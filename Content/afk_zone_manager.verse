using { /Fortnite.com/Devices }
using { /Verse.org/Simulation }
using { /Verse.org/Colors }
using { /Verse.org/Verse }
using { /Verse.org/Random }
using { /Fortnite.com/Teams }
using { /UnrealEngine.com/Assets }
using { /Fortnite.com/Game }
using { /Fortnite.com/Characters }
using { /Fortnite.com/Playspaces }
using { /Fortnite.com/UI }
using { /UnrealEngine.com/Temporary/Diagnostics }
using { /UnrealEngine.com/Temporary/SpatialMath }
using { /UnrealEngine.com/Temporary/UI }
using { /Verse.org/Assets}
using { /Verse.org/Colors/NamedColors }

afk_zone_manager := class(creative_device):

    @editable
    gm : game_manager = game_manager{}

    @editable
    AudioManager : audio_manager = audio_manager{}

    @editable
    SUP : signal_device_manager = signal_device_manager{}

    @editable
    Tutorial : tutorial_manager = tutorial_manager{}

    @editable
    AFKZone:mutator_zone_device = mutator_zone_device{}

    var DamageNumbers<private>: damage_numbers = damage_numbers:

    OnBegin<override>()<suspends>:void=

        AFKZone.AgentEntersEvent.SubscribeAgent(StartAFK, true)
        AFKZone.AgentExitsEvent.SubscribeAgent(StartAFK, false)


    PlayerIsAFK(CustomPlayer:Custom_Player)<suspends>:void=
        AFKZoneGlow:cancelable = SpawnParticleSystem(VFX.AFKZoneGlow, vector3{X:=3320.0,Y:=-1548.0,Z:=0.0})
        
        var PetStr : float = 0.0
        for(i:int=0..CustomPlayer.PetSlots.Length):
            if(CustomPlayer.PetSlots[i] <> -1):
                if:
                    PetLevel := CustomPlayer.PetLevels[CustomPlayer.PetSlots[i]]
                    Gather := CustomPlayer.PetGatherRates[CustomPlayer.PetSlots[i]]
                then:
                    set PetStr += Gather
        loop:
            if(CustomPlayer.CanGrantEXP = true, Accolade:=AudioManager.Accolades[9]): #AWARD PLAYER EXP FOR HITTING QUARRY ROCK
                set CustomPlayer.CanGrantEXP = false
                Accolade.Award(CustomPlayer.Player)
            if(CustomPlayer.IsAFK = true):
                if(CustomPlayer.AFKPetsUnlocked = true):
                    Gains:= 1.0 + ((CustomPlayer.Power/4.0) * PetStr/100.0) + (CustomPlayer.Power/4.0) + (CustomPlayer.Power/4.0 * CustomPlayer.AFKStoneBoost)
                    DamageNumbers.ShowDamageUIForPlayer(CustomPlayer, Gains, false, 0, false, "AFK", -1.0)
                    CustomPlayer.Collect(Gains, "Stone", 0)
                else:
                    Gains:= 1.0 + (CustomPlayer.Power/4.0) + (CustomPlayer.Power/4.0 * CustomPlayer.AFKStoneBoost)
                    DamageNumbers.ShowDamageUIForPlayer(CustomPlayer, Gains, false, 0, false, "AFK", -1.0)
                    CustomPlayer.Collect(Gains, "Stone", 0)
                if(MyUI:=CustomPlayer.UIMaps["SUP"]?):
                    SUP.UpdateSUPHUD(CustomPlayer)
                Sleep(0.5)
            else:
                AFKZoneGlow.Cancel()
                break:

    StartAFK(Agent:agent, InZone:logic):void=
        if(Player:=player[Agent], CustomPlayer := gm.PlayersMap[Player]):
            if(InZone = true):
                Tutorial.UpdateTutorialUI(CustomPlayer, "AFKZone", 1.0)
                set CustomPlayer.IsAFK = true
                spawn:
                    PlayerIsAFK(CustomPlayer)
            else:
                set CustomPlayer.IsAFK = false