using { /Fortnite.com/Devices }
using { /Verse.org/Simulation }
using { /Verse.org/Colors }
using { /Verse.org/Verse }
using { /Verse.org/Random }
using { /Fortnite.com/Teams }
using { /UnrealEngine.com/Assets }
using { /Fortnite.com/Game }
using { /Fortnite.com/Characters }
using { /Fortnite.com/Playspaces }
using { /Fortnite.com/UI }
using { /UnrealEngine.com/Temporary/Diagnostics }
using { /UnrealEngine.com/Temporary/SpatialMath }
using { /UnrealEngine.com/Temporary/UI }
using { /Verse.org/Assets}

#I KNOW ITS AUDIO MANAGER BUT ITS MORE OF A MISC HOLDER
# ANALYTICS, ACCOLADES, AUDIO AND OTHER EASY SLOT INS (IM A DEGEN I KNOW)

audio_manager := class(creative_device):

    @editable
    OneTimeAnalytics : []analytics_device = array{}

    @editable
    GeneralAnalytics : []analytics_device = array{}
    <#
    0 - STAGES UNLOCKED
    1 - ROCKS DESTROYED
    2 - QUARRY ROCK HIT
    3 - MINUTES AFK
    4 - EGGS HATCHED
    5 - PRESTIGE (NEED TO ADD)
    6 - EXPEDITIONS COMPLETED
    7 - METEOR ROCK DESTROYED
    8 - TIMES FULL RESET
    9 - HATS PURCHASED
    #>

    @editable
    MobAnalytics : []analytics_device = array{}
    
    @editable
    BossAnalytics : []analytics_device = array{}

    @editable
    WeaponAnalytics : []analytics_device = array{}


    @editable
    Accolades : []accolades_device = array{}

    #BACKGROUND MUSIC
    @editable
    BGMs : []audio_player_device = array{}
    # 0 = Default
    # 1 = Mob
    # 2 = Boss
    # 3 = Space
    # 4 = Space Mob
    # 5 = Space Boss
    # 6 = Post Space Boss

    BGMNames : []string = array{
        "Default",
        "Mob",
        "Boss",
        "Space",
        "Space Mob",
        "Space Boss",
        "Post Space Boss",
        "HalloweenBoss"
    }

    #GENERAL
    @editable
    Denied : audio_player_device = audio_player_device{}
    @editable
    EXPLaunch : audio_player_device = audio_player_device{}
    @editable
    EXPClaim : audio_player_device = audio_player_device{}
    @editable
    EXPClaimPet : audio_player_device = audio_player_device{}
    @editable
    Evolve : audio_player_device = audio_player_device{}
    @editable
    AchCompleted : audio_player_device = audio_player_device{}
    @editable
    LeveledUp : audio_player_device = audio_player_device{}
    @editable
    PlayerElimed : audio_player_device = audio_player_device{}
    @editable
    BossDefeatedVictory : audio_player_device = audio_player_device{}
    @editable
    ItemObtained : audio_player_device = audio_player_device{}
    @editable
    Prestige : audio_player_device = audio_player_device{}
    @editable
    Drill : audio_player_device = audio_player_device{}

    @editable
    MissileSFX : []audio_player_device = array{}

    #ROCK STUFF
    @editable
    RockDestroyed : audio_player_device = audio_player_device{}
    @editable
    ComboHit : audio_player_device = audio_player_device{}
    @editable
    CritHit : audio_player_device = audio_player_device{}
    @editable
    Earthquake : audio_player_device = audio_player_device{}

    #MOB AUDIO
    @editable
    Punch : audio_player_device = audio_player_device{}
    @editable
    Woosh : audio_player_device = audio_player_device{}
    @editable
    Chomp : audio_player_device = audio_player_device{}

    #BOSS AUDIO
    @editable
    Spores : audio_player_device = audio_player_device{}
    @editable
    Boss2Effects : []audio_player_device = array{}
    @editable
    Boss5Effects : []audio_player_device = array{}
    #0 - Roar
    #1 - Earthquake
    #2 - Vertical Beam

    @editable
    DemonEffects : []audio_player_device = array{}
    #0 - Roar1
    #1 - Roar2
    #2 - Tornado
    #3 - Vertical Beam
    #4 - Comet

    @editable
    HalloweenBossEffects : []audio_player_device = array{}
    #0 - Laugh
    #1 - Vertical Beam

    @editable
    Meteor : audio_player_device = audio_player_device{}

    #BOSS ROOMS
    @editable
    ChestOpen : []audio_player_device = array{}

    @editable
    ChestWaiting : []audio_player_device = array{}

    @editable
    ChestOpenGlobal : audio_player_device = audio_player_device{}

    @editable
    ChestBeacons : []beacon_device = array{}

    @editable
    BigTextHUD : hud_message_device = hud_message_device{}



    OnBegin<override>()<suspends>:void=
        block: