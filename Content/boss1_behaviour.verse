using {/Fortnite.com/Characters }
using { /Fortnite.com/AI }
using { /Fortnite.com/Animation/PlayAnimation }
using { /Fortnite.com/Devices }
using { /Fortnite.com/Game }
using { /UnrealEngine.com/Assets }
using { /UnrealEngine.com/Temporary/SpatialMath }
using { /Verse.org/Assets }
using { /Verse.org/Random }
using { /Verse.org/Simulation }

boss1_behaviour := class(npc_behavior):
    
    @editable
    AudioManager: audio_manager = audio_manager{}

    var TargetAgents : [int]?agent = map{}

    var HardMode : logic = false
    var HardModeDamageMult : float = 5.0

    var AttackWeapon<public>: animation_sequence = Bosses.MushroomKing_Anim_CharacterArmature_Weapon
    var AttackSpore<public>: animation_sequence = Bosses.MushroomKing_Anim_CharacterArmature_Wave
    var HitReact<public>: animation_sequence = Bosses.MushroomKing_Anim_CharacterArmature_HitReact
    var BossDefeated<public>: animation_sequence = Bosses.MushroomKing_Anim_CharacterArmature_Death

    var SporeAoE : tuple(float, float) = (800.0, 10.0)
    var BeamAttack : tuple(float, float) = (200.0, 10.0)

    var IsElim : logic = false

    OnBegin<override>()<suspends>:void=
        spawn{StartBossLogic()}

    StartBossLogic()<suspends>:void=
        set IsElim = false
        if:
            NPCAgent := GetAgent[]
            NPCChar := NPCAgent.GetFortCharacter[]
            NPCNav := NPCChar.GetNavigatable[]
            NPCFocus := NPCChar.GetFocusInterface[]
            NPCAnim := NPCChar.GetPlayAnimationController[]
        then:
            Sleep(0.5)
            FirstTarget:=GetRandomInt(0,TargetAgents.Length-1)
            if(NewAgent := TargetAgents[FirstTarget]?, NewTarget := NewAgent.GetFortCharacter[]):
                spawn{NPCFocus.MaintainFocus(NewAgent)}
            Sleep(3.5)
            if(HardMode = true):
                Print("HARD MODE")
                set SporeAoE = (SporeAoE(0), SporeAoE(1)*HardModeDamageMult)
                set BeamAttack = (BeamAttack(0), BeamAttack(1)*HardModeDamageMult)
            loop:
                Sleep(0.1)
                if(IsElim = true):
                    Print("BREAK BEHAVIOUR LOOP")
                    break:
                WhichTarget:=GetRandomInt(0,TargetAgents.Length-1)
                Print("Players in Boss Fight: {TargetAgents.Length-1}")
                if(NewAgent := TargetAgents[WhichTarget]?, NewTarget := NewAgent.GetFortCharacter[]):
                    spawn{NPCFocus.MaintainFocus(NewAgent)}

                    IDX:=GetRandomInt(1,5) #CHANGE TO 1, 5
                    if(IDX < 3): #RUN AT SOMEONE AND SPORE
                        Print("RUN AT SOMEONE")
                        NavTarget := MakeNavigationTarget(NewAgent)
                        NavResult:=NPCNav.NavigateTo(NavTarget, ?MovementType := movement_types.Running, ?ReachRadius := 5.0, ?AllowPartialPath:=true)
                        if(NavResult = navigation_result.Reached):
                            Print("REACHED")
                            NPCNav.StopNavigation()
                            NPCAnim.Play(AttackSpore)
                            spawn{SpawnSporeVFX(NewAgent, NPCChar.GetTransform().Translation)}
                            Sleep(1.75)
                            NPCAnim.Play(AttackSpore)
                            CheckForDamage(NPCChar.GetTransform().Translation, SporeAoE(0), SporeAoE(1))
                            Sleep(1.0)
                            CheckForDamage(NPCChar.GetTransform().Translation, SporeAoE(0), SporeAoE(1))
                            Sleep(1.0)
                            NPCAnim.Play(AttackSpore)
                            CheckForDamage(NPCChar.GetTransform().Translation, SporeAoE(0), SporeAoE(1))
                            Sleep(1.0)
                            CheckForDamage(NPCChar.GetTransform().Translation, SporeAoE(0), SporeAoE(1))
                        else:
                            Print("NOT REACHED")
                            NPCNav.StopNavigation()
                        if(Distance(NPCChar.GetTransform().Translation, NewTarget.GetTransform().Translation) < 100.0):
                            Print("MANUAL STOP")
                            NPCNav.StopNavigation()

                    else: #ATTACK ALL WITH BEAMS
                        Print("HIT EVERYONE")
                        for(TargetAgent:TargetAgents):
                            if(Target:=TargetAgent?, TargetFC:=Target.GetFortCharacter[]):
                                Print("FOUND A TARGET")
                                NPCAnim.Play(AttackSpore)
                                BeamLoc := TargetFC.GetTransform().Translation
                                spawn{SpawnAllBeamVFX(Target, BeamLoc)}
                                Sleep(0.2)
                        Sleep(3.0)
                if(NPCChar.GetTransform().Translation.Z < -6262.0):
                    Destination := vector3{
                        X := -3144.0
                        Y := 7656.0
                        Z :=-6160.0
                    }
                    if(NPCChar.TeleportTo[Destination, NPCChar.GetTransform().Rotation]) {}

    CheckForDamage(VFXCenter:vector3, DamageRange:float, Damage:float):void=
        if(IsElim = false):
            for(TargetAgent:TargetAgents):
                if(Target:=TargetAgent?, TargetFC:=Target.GetFortCharacter[]):
                    if(Distance(VFXCenter, TargetFC.GetTransform().Translation) < DamageRange):
                        Print("PLAYER HIT")
                        TargetFC.Damage(Damage)

    SpawnSporeVFX(Agent:agent, SporeLocation:vector3)<suspends>:void=
        SporeVFX:cancelable = SpawnParticleSystem(VFX.NS_VerticalBeam10_RedWhiteAqua, vector3{X:=SporeLocation.X,Y:=SporeLocation.Y,Z:=SporeLocation.Z-80.0})
        Sleep(5.0)
        SporeVFX.Cancel()

    SpawnAllBeamVFX(Agent:agent, SporeLocation:vector3)<suspends>:void=
        SporeVFX:cancelable = SpawnParticleSystem(VFX.NS_VerticalBeam09_GreenYellowPurple, vector3{X:=SporeLocation.X,Y:=SporeLocation.Y,Z:=SporeLocation.Z-80.0})
        Sleep(1.75)
        CheckForDamage(SporeLocation, BeamAttack(0), BeamAttack(1))
        Sleep(3.25)
        SporeVFX.Cancel()

    SetTargetAgents(Agents:[int]?agent):void=
        Print("UPDATED TARGETS FOR 1")
        set TargetAgents = Agents

    SetHardMode(IsHardMode:logic):void=
        set HardMode = IsHardMode

    SetIsElim(ElimResult:elimination_result):void=
        Print("SET ELIMINATED")
        set IsElim = true