using {/Fortnite.com/Characters }
using { /Fortnite.com/AI }
using { /Fortnite.com/Animation/PlayAnimation }
using { /Fortnite.com/Devices }
using { /Fortnite.com/Game }
using { /UnrealEngine.com/Assets }
using { /UnrealEngine.com/Temporary/SpatialMath }
using { /Verse.org/Assets }
using { /Verse.org/Random }
using { /Verse.org/Simulation }
 
boss3_behaviour := class(npc_behavior):
    
    @editable
    AudioManager: audio_manager = audio_manager{}

    var TargetAgents : [int]?agent = map{}

    var HardMode : logic = false
    var HardModeDamageMult : float = 5.0
    
    var AttackWeapon<public>: animation_sequence = Bosses.Cactoro_Anim_CharacterArmature_Weapon
    var AttackMeteor<public>: animation_sequence = Bosses.Cactoro_Anim_CharacterArmature_Wave
    var HitReact<public>: animation_sequence = Bosses.Cactoro_Anim_CharacterArmature_HitReact
    var BossDefeated<public>: animation_sequence = Bosses.Cactoro_Anim_CharacterArmature_Death

    var Attack1 : tuple(float, float) = (150.0, 35.0)
    var BeamAttack : tuple(float, float) = (250.0, 50.0)

    var IsElim : logic = false
    OnBegin<override>()<suspends>:void=
        spawn{StartBossLogic()}

    StartBossLogic()<suspends>:void=
        set IsElim = false
        if:
            NPCAgent := GetAgent[]
            NPCChar := NPCAgent.GetFortCharacter[]
            NPCNav := NPCChar.GetNavigatable[]
            NPCFocus := NPCChar.GetFocusInterface[]
            NPCAnim := NPCChar.GetPlayAnimationController[]
        then:
            Sleep(0.5)
            FirstTarget:=GetRandomInt(0,TargetAgents.Length-1)
            if(NewAgent := TargetAgents[FirstTarget]?, NewTarget := NewAgent.GetFortCharacter[]):
                spawn{NPCFocus.MaintainFocus(NewAgent)}
            Sleep(3.5)
            if(HardMode = true):
                Print("HARD MODE")
                set Attack1 = (Attack1(0), Attack1(1)*HardModeDamageMult)
                set BeamAttack = (BeamAttack(0), BeamAttack(1)*HardModeDamageMult)
            loop:
                Sleep(0.1)
                if(IsElim = true):
                    Print("BREAK BEHAVIOUR LOOP")
                    break:
                if(NPCChar.GetTransform().Translation.Z < -69.0):
                    Print("Fell Through Floor")
                    Destination := vector3{
                        X :=  11324.440766
                        Y := -774.03019
                        Z := 28.432465
                    }
                    if(NPCChar.TeleportTo[Destination, NPCChar.GetTransform().Rotation]) {}
                WhichTarget:=GetRandomInt(0,TargetAgents.Length-1)
                if(NewAgent := TargetAgents[WhichTarget]?, NewTarget := NewAgent.GetFortCharacter[]):
                    spawn{NPCFocus.MaintainFocus(NewAgent)}
                    IDX:=GetRandomInt(1,5) #CHANGE TO 1,5
                    if(IDX < 4): #SINGLE BEAM
                        NavTarget := MakeNavigationTarget(NewAgent)
                        NavResult:=NPCNav.NavigateTo(NavTarget, ?MovementType := movement_types.Running, ?ReachRadius := 5.0, ?AllowPartialPath:=true)
                        if(NavResult = navigation_result.Reached):
                            NPCNav.StopNavigation()
                            NPCAnim.Play(AttackMeteor)
                            BeamLoc := NewTarget.GetTransform().Translation
                            spawn{SpawnSingleBeamVFX(NewAgent, BeamLoc)}
                            Sleep(1.75)
                            CheckForDamage(BeamLoc, Attack1(0), Attack1(1))
                        else:
                            NPCNav.StopNavigation()
                    else: #ALL BEAMS
                        for(TargetAgent:TargetAgents):
                            if(Target:=TargetAgent?, TargetFC:=Target.GetFortCharacter[]):
                                NPCAnim.Play(AttackMeteor)
                                BeamLoc := TargetFC.GetTransform().Translation
                                spawn{SpawnAllBeamVFX(Target, BeamLoc)}
                                Sleep(0.2)
                        Sleep(3.0)

    CheckForDamage(VFXCenter:vector3, DamageRange:float, Damage:float):void=
        if(IsElim = false):
            for(TargetAgent:TargetAgents):
                if(Target:=TargetAgent?, TargetFC:=Target.GetFortCharacter[]):
                    if(Distance(VFXCenter, TargetFC.GetTransform().Translation) < DamageRange):
                        TargetFC.Damage(Damage)   
    
    SpawnSingleBeamVFX(Agent:agent, BeamLoc:vector3)<suspends>:void=
        SporeVFX:cancelable = SpawnParticleSystem(VFX.NS_VerticalBeam15_Fire, vector3{X:=BeamLoc.X,Y:=BeamLoc.Y,Z:=BeamLoc.Z-50.0})
        Sleep(7.0)
        SporeVFX.Cancel()     

    SpawnAllBeamVFX(Agent:agent, BeamLoc:vector3)<suspends>:void=
        SporeVFX:cancelable = SpawnParticleSystem(VFX.NS_VerticalBeam03_RedWhite, vector3{X:=BeamLoc.X,Y:=BeamLoc.Y,Z:=BeamLoc.Z-50.0})
        Sleep(1.75)
        CheckForDamage(BeamLoc, BeamAttack(0), BeamAttack(1))
        Sleep(5.25)
        SporeVFX.Cancel()     

    SetTargetAgents(Agents:[int]?agent):void=
        set TargetAgents = Agents

    SetHardMode(IsHardMode:logic):void=
        set HardMode = IsHardMode
        
    SetIsElim(ElimResult:elimination_result):void=
        Print("SET ELIMINATED")
        set IsElim = true
        
    OnEnd<override>():void=
        if:
            NPCAgent := GetAgent[]
            NPCChar := NPCAgent.GetFortCharacter[]
            NPCAnim := NPCChar.GetPlayAnimationController[]
        then:
            NPCAnim.Play(BossDefeated)