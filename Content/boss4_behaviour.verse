using {/Fortnite.com/Characters }
using { /Fortnite.com/AI }
using { /Fortnite.com/Animation/PlayAnimation }
using { /Fortnite.com/Devices }
using { /Fortnite.com/Game }
using { /UnrealEngine.com/Assets }
using { /UnrealEngine.com/Temporary/SpatialMath }
using { /Verse.org/Assets }
using { /Verse.org/Random }
using { /Verse.org/Simulation }
 

boss4_behaviour := class(npc_behavior):

    var TargetAgents : [int]?agent = map{}

    var BeamSingle : creative_prop_asset = DefaultCreativePropAsset
    var BeamAll : creative_prop_asset = DefaultCreativePropAsset
    var BeamRandom : creative_prop_asset = DefaultCreativePropAsset

    var TopLeftCorner : vector2 = vector2{X:=46810.0, Y:=34210.0}
    var BottomRightCorner : vector2 = vector2{X:=36270.0, Y:=43493.0}
    var ZAlways : float = 25607.0

    var StandLocations : []vector3 = array{
        vector3{X:= 44896.0, Y:= 39360.0, Z:=25727.232060},
        vector3{X:= 41503.0, Y:= 42005.0, Z:=25727.232060},
        vector3{X:= 39107.0, Y:= 39310.0, Z:=25727.232060},
        vector3{X:= 41583.0, Y:= 36659.0, Z:=25727.232060}
    }

    var OldCorner : int = 9

    #16_Spirals - Basic target 1 player 
    #07_BlueRedWhite - Basic target all players
    #22_Cosmic - Special random attack, 10 in a row in random place within zone

    <# USE THIS TO MOVE HIM TO DIFFERENT SIDES, DON'T CHASE PLAYERS
    NavTarget := MakeNavigationTarget(NewAgent)
    NPCNav.NavigateTo(NavTarget, ?MovementType := movement_types.Running, ?ReachRadius := 800.0)#>

    var SingleAttack<public>: animation_sequence = Bosses.Dragon_Evolved_Anim_CharacterArmature_Headbutt
    var MultiAttack<public>: animation_sequence = Bosses.Dragon_Evolved_Anim_CharacterArmature_Yes

    var IsElim : logic = false
    OnBegin<override>()<suspends>:void=
        spawn{StartBossLogic()}

    StartBossLogic()<suspends>:void=
        set IsElim = false
        if:
            NPCAgent := GetAgent[]
            NPCChar := NPCAgent.GetFortCharacter[]
            NPCNav := NPCChar.GetNavigatable[]
            NPCFocus := NPCChar.GetFocusInterface[]
            NPCAnim := NPCChar.GetPlayAnimationController[]
        then:
            Sleep(0.5)
            FirstTarget:=GetRandomInt(0,TargetAgents.Length-1)
            if(NewAgent := TargetAgents[FirstTarget]?, NewTarget := NewAgent.GetFortCharacter[]):
                spawn{NPCFocus.MaintainFocus(NewAgent)}
            Sleep(5.0)
            NPCAnim.Play(MultiAttack)
            spawn{StartSkyfall(NPCChar)}
            loop:
                Sleep(0.1)
                if(IsElim = true):
                    Print("BREAK BEHAVIOUR LOOP")
                    break:
                if(NPCChar.GetTransform().Translation.Z < ZAlways):
                    Print("Fell through floor")
                    Destination := vector3{
                        X := 39778.997233
                        Y := 39254.896265
                        Z := 26000.564975
                    }
                    if(NPCChar.TeleportTo[Destination, NPCChar.GetTransform().Rotation]) {}
                WhichTarget:=GetRandomInt(0,TargetAgents.Length-1)
                if(NewAgent := TargetAgents[WhichTarget]?, NewTarget := NewAgent.GetFortCharacter[]):
                    spawn{NPCFocus.MaintainFocus(NewAgent)}
                    IDX:=GetRandomInt(1,10) #CHANGE TO 1,10
                    if(IDX < 4): #SINGLE BEAM            
                        spawn{NPCFocus.MaintainFocus(NewAgent)}            
                        NPCAnim.Play(SingleAttack)
                        BeamLoc := NewTarget.GetTransform().Translation
                        spawn{SpawnSingleBeamVFX(BeamLoc)}
                        Sleep(2.0)
                    else if (IDX < 9): #ALL BEAMS
                        spawn{NPCFocus.MaintainFocus(NewAgent)}
                        for(TargetAgent:TargetAgents):
                            if(Target:=TargetAgent?, TargetFC:=Target.GetFortCharacter[]):
                                NPCAnim.Play(MultiAttack)
                                BeamLoc := TargetFC.GetTransform().Translation
                                spawn{SpawnAllBeamVFX(BeamLoc)}
                                Sleep(0.2)
                        Sleep(2.0)
                    else:
                        WhichCorner := GetRandomInt(0,3)
                        if:
                            WhichCorner <> OldCorner
                            TargetLoc :=StandLocations[WhichCorner]
                        then:
                            set OldCorner = WhichCorner
                            Print("WhichCorner {WhichCorner}")
                            #spawn{NPCFocus.MaintainFocus(TargetLoc)}
                            NavTarget := MakeNavigationTarget(TargetLoc)
                            NavResult:=NPCNav.NavigateTo(NavTarget, ?MovementType := movement_types.Running, ?ReachRadius := 5.0, ?AllowPartialPath:=true)
                            if(NavResult = navigation_result.Reached):
                                NPCNav.StopNavigation()
                            if(Distance(NPCChar.GetTransform().Translation, TargetLoc) < 100.0):
                                Print("MANUAL STOP")
                                NPCNav.StopNavigation()
                        Sleep(1.0)

    CheckForDamage(VFXCenter:vector3, DamageRange:float, Damage:float):void=
        if(IsElim = false):
            for(TargetAgent:TargetAgents):
                if(Target:=TargetAgent?, TargetFC:=Target.GetFortCharacter[]):
                    if(Distance(VFXCenter, TargetFC.GetTransform().Translation) < DamageRange):
                        TargetFC.Damage(Damage)   
    
    SpawnSingleBeamVFX(BeamLoc:vector3)<suspends>:void=
        NewRotation : rotation = IdentityRotation()
        NewScale : vector3 = vector3{X:=1.0, Y:=1.0, Z:=1.0}
        BeamVFX := SpawnProp(BeamSingle, transform{Translation:=vector3{X:=BeamLoc.X,Y:=BeamLoc.Y,Z:=BeamLoc.Z-40.0}, Rotation:=NewRotation, Scale:=NewScale})
        Sleep(1.5)
        CheckForDamage(BeamLoc, 250.0, 80.0)
        Sleep(7.0)
        PropInst := BeamVFX(0)
        if ( Prop := PropInst?):
            Prop.Dispose()  
        
    SpawnClusterBeamVFX(BeamLoc:vector3)<suspends>:void=
        NewRotation : rotation = IdentityRotation()
        NewScale : vector3 = vector3{X:=1.0, Y:=1.0, Z:=1.0}
        BeamVFX := SpawnProp(BeamRandom, transform{Translation:=vector3{X:=BeamLoc.X,Y:=BeamLoc.Y,Z:=BeamLoc.Z-40.0}, Rotation:=NewRotation, Scale:=NewScale})
        Sleep(1.5)
        CheckForDamage(BeamLoc, 600.0, 40.0)
        Sleep(7.0)
        PropInst := BeamVFX(0)
        if ( Prop := PropInst?):
            Prop.Dispose()   

    SpawnAllBeamVFX(BeamLoc:vector3)<suspends>:void=
        NewRotation : rotation = IdentityRotation()
        NewScale : vector3 = vector3{X:=1.0, Y:=1.0, Z:=1.0}
        BeamVFX := SpawnProp(BeamAll, transform{Translation:=vector3{X:=BeamLoc.X,Y:=BeamLoc.Y,Z:=BeamLoc.Z-40.0}, Rotation:=NewRotation, Scale:=NewScale})
        Sleep(1.5)
        for(i:int=0..9):
            CheckForDamage(BeamLoc, 500.0, 25.0)
            Sleep(0.525)
        PropInst := BeamVFX(0)
        if ( Prop := PropInst?):
            Prop.Dispose()  

    StartSkyfall(NPCChar:fort_character)<suspends>:void=
        loop:
            if(IsElim = true):
                Print("STOP SKYFALL")
                break:
            else:
                MinX := Min(TopLeftCorner.X, BottomRightCorner.X)
                MaxX := Max(TopLeftCorner.X, BottomRightCorner.X)
                MinY := Min(TopLeftCorner.Y, BottomRightCorner.Y)
                MaxY := Max(TopLeftCorner.Y, BottomRightCorner.Y)

                RandomX := GetRandomFloat(MinX, MaxX)
                RandomY := GetRandomFloat(MinY, MaxY)

                BeamLoc := vector3{X:=RandomX, Y:=RandomY, Z:=ZAlways}
                spawn{SpawnClusterBeamVFX(BeamLoc)}
            Sleep(0.5)

    SetTargetAgents(Agents:[int]?agent):void=
        set TargetAgents = Agents

    SetAnims(Boss4BeamSingle:creative_prop_asset, Boss4BeamAll:creative_prop_asset, Boss4BeamRandom:creative_prop_asset):void=
        set BeamSingle = Boss4BeamSingle
        set BeamAll = Boss4BeamAll
        set BeamRandom = Boss4BeamRandom

    SetIsElim(ElimResult:elimination_result):void=
        Print("SET ELIMINATED")
        set IsElim = true

    OnEnd<override>():void=
        block: