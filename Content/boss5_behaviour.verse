using {/Fortnite.com/Characters }
using { /Fortnite.com/AI }
using { /Fortnite.com/Animation/PlayAnimation }
using { /Fortnite.com/Devices }
using { /Fortnite.com/Game }
using { /UnrealEngine.com/Assets }
using { /UnrealEngine.com/Temporary/SpatialMath }
using { /Verse.org/Assets }
using { /Verse.org/Random }
using { /Verse.org/Simulation }
 
boss5_behaviour := class(npc_behavior):

    var TargetAgents : [int]?agent = map{}

    var AudioManager : ?audio_manager = false

    var BeamAll : creative_prop_asset = DefaultCreativePropAsset
    var GroundSlam : creative_prop_asset = DefaultCreativePropAsset

    var Center : vector3 = vector3{X:=57342.0, Y:=33698.0, Z:=1929.0}

    var HardMode : logic = false
    var HardModeDamageMult : float = 5.0
    
    var Roar<public>: animation_sequence = Bosses.A_Cavecrasher_Emote_Roar
    var Stun<public>: animation_sequence = Bosses.A_Cavecrasher_Combat_Stun

    var HandGroundSlam<public>: animation_sequence = Bosses.A_Cavecrasher_Combat_Unarmed_Attack
    var DoubleHit<public>: animation_sequence = Bosses.A_Cavecrasher_Combat_Unarmed_Attack01
    var BigGroundSlam<public>: animation_sequence = Bosses.A_Cavecrasher_Combat_Unarmed_Attack02

    var JumpStart<public>: animation_sequence = Bosses.A_Cavecrasher_Jump_Start
    var JumpAir<public>: animation_sequence = Bosses.A_Cavecrasher_Jump_Loop
    var JumpLand<public>: animation_sequence = Bosses.A_Cavecrasher_Jump_End

    var AttackSwipe : tuple(float, float) = (450.0, 80.0)
    var BeamAttack : tuple(float, float) = (250.0, 120.0)
    var AttackSlam : tuple(float, float) = (2200.0, 100.0)

    var IsElim : logic = false
    OnBegin<override>()<suspends>:void=
        spawn{StartBossLogic()}

    StartBossLogic()<suspends>:void=
        set IsElim = false
        if:
            NPCAgent := GetAgent[]
            NPCChar := NPCAgent.GetFortCharacter[]
            NPCNav := NPCChar.GetNavigatable[]
            NPCFocus := NPCChar.GetFocusInterface[]
            NPCAnim := NPCChar.GetPlayAnimationController[]
        then:
            Sleep(0.5)
            if:
                AM := AudioManager?
                BeamAudio:=AM.Boss5Effects[0]
                NewAgent := TargetAgents[0]?
                NewTarget := NewAgent.GetFortCharacter[]
            then:
                BeamAudio.Play(NewAgent)
                spawn{NPCFocus.MaintainFocus(NewAgent)}
            Sleep(0.2)
            NPCAnim.Play(Roar)
            Sleep(3.5)
            if(HardMode = true):
                Print("HARD MODE")
                set AttackSwipe = (AttackSwipe(0), AttackSwipe(1)*HardModeDamageMult)
                set BeamAttack = (BeamAttack(0), BeamAttack(1)*HardModeDamageMult)
                set AttackSlam = (AttackSlam(0), AttackSlam(1)*HardModeDamageMult)
            loop:
                Sleep(0.1)
                if(IsElim = true):
                    Print("BREAK BEHAVIOUR LOOP")
                    break:
                if(NPCChar.GetTransform().Translation.Z < 1920.0):
                    Destination := vector3{
                        X := 57946.0
                        Y := 33698.0
                        Z := 1996.0
                    }
                    if(NPCChar.TeleportTo[Destination, NPCChar.GetTransform().Rotation]) {}
                WhichTarget:=GetRandomInt(0,TargetAgents.Length-1)
                if:
                    NewAgent := TargetAgents[WhichTarget]?
                    NewTarget := NewAgent.GetFortCharacter[]
                then:
                    spawn{NPCFocus.MaintainFocus(NewAgent)}
                    IDX:=GetRandomInt(1,10) #CHANGE TO 1,10
                    if(IDX < 6): #RUN AND ATTACK      
                        #Print("RUN AT SOMEONE")
                        NavTarget := MakeNavigationTarget(NewAgent)
                        NavResult:=NPCNav.NavigateTo(NavTarget, ?MovementType := movement_types.Running, ?ReachRadius := 250.0, ?AllowPartialPath:=true)
                        if(NavResult = navigation_result.Reached):
                            #Print("REACHED")
                            NPCNav.StopNavigation()       

                            NPCAnim.Play(DoubleHit)
                            Sleep(0.34)
                            CheckForDamage(NPCChar.GetTransform().Translation, AttackSwipe(0), AttackSwipe(1))
                            Sleep(0.40)
                            CheckForDamage(NPCChar.GetTransform().Translation, AttackSwipe(0), AttackSwipe(1))
                    else if(IDX < 9): #GROUND SLAM INTO ALL x3      
                        #Print("GROUND SLAM BEAMS")
                        for(x := 0 .. 2):
                            #Print("{x} GROUND SLAM")
                            NPCAnim.Play(BigGroundSlam)
                            Sleep(0.50)
                            for(TargetAgent:TargetAgents):
                                if(Target:=TargetAgent?, TargetFC:=Target.GetFortCharacter[]):
                                    BeamLoc := TargetFC.GetTransform().Translation
                                    if:
                                        AM := AudioManager?
                                        BeamAudio:=AM.Boss5Effects[2]
                                    then:
                                        BeamAudio.Play(Target)
                                    spawn{SpawnAllBeamVFX(BeamLoc)}
                                    Sleep(0.05)
                            Sleep(1.0)
                    else if (IDX >= 9): #JUMP SLAMMING INTO EXHAUSTED
                        spawn{NPCFocus.MaintainFocus(Center)}
                        NavTarget := MakeNavigationTarget(Center)
                        NavResult:=NPCNav.NavigateTo(NavTarget, ?MovementType := movement_types.Running, ?ReachRadius := 5.0, ?AllowPartialPath:=true)
                        if(NavResult = navigation_result.Reached):
                            NPCNav.StopNavigation()
                            for(x := 0 .. 2):
                                #Print("JUMP")
                                NPCAnim.Play(JumpStart)
                                Sleep(0.64)
                                if:
                                    AM := AudioManager?
                                    BeamAudio:=AM.Boss5Effects[1]
                                then:
                                    BeamAudio.Play(NewAgent)
                                NPCAnim.Play(JumpLand)
                                Sleep(0.27)
                                spawn{SpawnGroundVFX(Center)}
                                Sleep(0.71)
                            #Print("STUNNED")
                            NPCAnim.Play(Stun)
                            Sleep(1.9)
                            NPCAnim.Play(Stun)
                            Sleep(1.9)


    CheckForDamage(VFXCenter:vector3, DamageRange:float, Damage:float):void=
        if(IsElim = false):
            for(TargetAgent:TargetAgents):
                if(Target:=TargetAgent?, TargetFC:=Target.GetFortCharacter[]):
                    if(Distance(VFXCenter, TargetFC.GetTransform().Translation) < DamageRange):
                        #Print("PLAYER DAMAGE")
                        TargetFC.Damage(Damage)   
                    else:
                        #Print("PLAYER OUT OF RANGE")


    CheckForGroundDamage(VFXCenter:vector3, DamageRange:float, Damage:float):void=
        if(IsElim = false):
            for(TargetAgent:TargetAgents):
                if(Target:=TargetAgent?, TargetFC:=Target.GetFortCharacter[]):
                    if(Distance(VFXCenter, TargetFC.GetTransform().Translation) < DamageRange):
                        if((VFXCenter.Z + 100.0) > TargetFC.GetTransform().Translation.Z):
                            #Print("GROUND DMG")
                            TargetFC.Damage(Damage)   
                        else:
                            #Print("PLAYER OFF GROUND")

    SpawnGroundVFX(BeamLoc:vector3)<suspends>:void=
        NewRotation : rotation = IdentityRotation()
        NewScale : vector3 = vector3{X:=1.0, Y:=1.0, Z:=1.0}
        BeamVFX := SpawnProp(GroundSlam, transform{Translation:=vector3{X:=BeamLoc.X,Y:=BeamLoc.Y,Z:=Center.Z}, Rotation:=NewRotation, Scale:=NewScale})
        CheckForGroundDamage(Center, AttackSlam(0), AttackSlam(1))
        Sleep(3.0)
        PropInst := BeamVFX(0)
        if (Prop := PropInst?):
            Prop.Dispose()

    SpawnAllBeamVFX(BeamLoc:vector3)<suspends>:void=
        NewRotation : rotation = IdentityRotation()
        NewScale : vector3 = vector3{X:=1.0, Y:=1.0, Z:=1.0}
        BeamVFX := SpawnProp(BeamAll, transform{Translation:=vector3{X:=BeamLoc.X,Y:=BeamLoc.Y,Z:=Center.Z}, Rotation:=NewRotation, Scale:=NewScale})
        Sleep(0.5)
        CheckForDamage(BeamLoc, BeamAttack(0), BeamAttack(1))
        Sleep(7.0)
        PropInst := BeamVFX(0)
        if ( Prop := PropInst?):
            Prop.Dispose()

    SetTargetAgents(Agents:[int]?agent, AM:?audio_manager):void=
        set TargetAgents = Agents
        set AudioManager = AM

    SetAnims(Boss5BeamAll:creative_prop_asset,Boss5GroundSlam:creative_prop_asset):void=
        set BeamAll = Boss5BeamAll
        set GroundSlam = Boss5GroundSlam

    SetHardMode(IsHardMode:logic):void=
        set HardMode = IsHardMode
        
    SetIsElim(ElimResult:elimination_result):void=
        Print("SET ELIMINATED")
        set IsElim = true