using {/Fortnite.com/Characters }
using { /Fortnite.com/AI }
using { /Fortnite.com/Animation/PlayAnimation }
using { /Fortnite.com/Devices }
using { /Fortnite.com/Game }
using { /UnrealEngine.com/Assets }
using { /UnrealEngine.com/Temporary/SpatialMath }
using { /Verse.org/Assets }
using { /Verse.org/Random }
using { /Verse.org/Simulation }

boss6_behaviour := class(npc_behavior):
    var TargetAgents : [int]?agent = map{}

    var AudioManager : ?audio_manager = false

    var Tornado : creative_prop_asset = DefaultCreativePropAsset
    var Comet : creative_prop_asset = DefaultCreativePropAsset

    var Center : vector3 = vector3{X:=-221052.0, Y:=-222096.0, Z:=16896.0}
    var TopLeftCorner : vector2 = vector2{X:=-216813.0, Y:=-226800.0}
    var BottomRightCorner : vector2 = vector2{X:=-225590.0, Y:=-214402.0}

    var HardMode : logic = false
    var HardModeDamageMult : float = 5.0
    
    var CometsOn : logic = false

    var RoarStart<public>: animation_sequence = Bosses.A_DemonBoss_Spell_Unarmed_Up_Start
    var RoarLoop<public>: animation_sequence = Bosses.A_DemonBoss_Spell_Unarmed_Up_Loop
    var RoarEnd<public>: animation_sequence = Bosses.A_DemonBoss_Spell_Unarmed_Up_Start

    var Stun<public>: animation_sequence = Bosses.A_DemonBoss_Combat_Stun

    var SwipeAttack<public>: animation_sequence = Bosses.A_DemonBoss_Combat_Unarmed_Attack
    var TornadoAttack<public>: animation_sequence = Bosses.A_DemonBoss_Combat_Unarmed_AttackSpecial

    var CometAttack1<public>: animation_sequence = Bosses.A_DemonBoss_Jump_Start
    var CometAttack2<public>: animation_sequence = Bosses.A_DemonBoss_Jump_End

    var AttackSwipe : tuple(float, float) = (450.0, 80.0)
    var AttackComet : tuple(float, float) = (350.0, 100.0)
    var AttackTornado : tuple(float, float) = (600.0, 5.0)

    var IsElim : logic = false
    OnBegin<override>()<suspends>:void=
        spawn{StartBossLogic()}

    StartBossLogic()<suspends>:void=
        set IsElim = false
        if:
            NPCAgent := GetAgent[]
            NPCChar := NPCAgent.GetFortCharacter[]
            NPCNav := NPCChar.GetNavigatable[]
            NPCFocus := NPCChar.GetFocusInterface[]
            NPCAnim := NPCChar.GetPlayAnimationController[]
        then:
            Sleep(0.5)
            if:
                AM := AudioManager?
                Roar:=AM.DemonEffects[0]
                NewAgent := TargetAgents[0]?
                NewTarget := NewAgent.GetFortCharacter[]
            then:
                Roar.Play(NewAgent)
                spawn{NPCFocus.MaintainFocus(NewAgent)}
            Sleep(0.2)
            NPCAnim.Play(RoarStart)
            Sleep(0.8)
            NPCAnim.Play(RoarLoop)
            Sleep(1.5)
            if(HardMode = true):
                set AttackSwipe = (AttackSwipe(0), AttackSwipe(1)*HardModeDamageMult)
                set AttackTornado = (AttackTornado(0), AttackTornado(1)*HardModeDamageMult)
                set AttackComet = (AttackComet(0), AttackComet(1)*HardModeDamageMult)
            loop:
                Sleep(0.1)
                if(IsElim = true):
                    Print("BREAK BEHAVIOUR LOOP")
                    break:
                if(NPCChar.GetTransform().Translation.Z < 16830.0):
                    Destination := vector3{
                        X := -220788.0
                        Y := -218140.0
                        Z := 16892.0
                    }
                    if(NPCChar.TeleportTo[Destination, NPCChar.GetTransform().Rotation]) {}
                WhichTarget:=GetRandomInt(0,TargetAgents.Length-1)
                if:
                    NewAgent := TargetAgents[WhichTarget]?
                    NewTarget := NewAgent.GetFortCharacter[]
                then:
                    spawn{NPCFocus.MaintainFocus(NewAgent)}
                    IDX:=GetRandomInt(1,10) #CHANGE TO 1,10
                    if(IDX < 6): #RUN AND ATTACK     
                        NavTarget := MakeNavigationTarget(NewAgent)
                        NavResult:=NPCNav.NavigateTo(NavTarget, ?MovementType := movement_types.Running, ?ReachRadius := 250.0, ?AllowPartialPath:=true)
                        if(NavResult = navigation_result.Reached):
                            NPCNav.StopNavigation()   
                            NPCAnim.Play(SwipeAttack)
                            Sleep(0.34)
                            CheckForDamage(NPCChar.GetTransform().Translation, AttackSwipe(0), AttackSwipe(1))
                            Sleep(0.40)
                            CheckForDamage(NPCChar.GetTransform().Translation, AttackSwipe(0), AttackSwipe(1))
                    else if(IDX < 9): #GROUND SLAM INTO ALL x3     
                        NPCAnim.Play(TornadoAttack)
                        Sleep(0.50)
                        for(TargetAgent:TargetAgents):
                            if(Target:=TargetAgent?, TargetFC:=Target.GetFortCharacter[]):
                                BeamLoc := TargetFC.GetTransform().Translation
                                if:
                                    AM := AudioManager?
                                    BeamAudio:=AM.DemonEffects[2]
                                then:
                                    BeamAudio.Play(Target)
                                spawn{SpawnAllBeamVFX(BeamLoc, TargetFC)}
                                Sleep(0.05)
                        Sleep(1.0)
                    else if (IDX >= 9 and CometsOn = false): #JUMP SLAMMING INTO EXHAUSTED
                        set CometsOn = true
                        spawn{NPCFocus.MaintainFocus(Center)}
                        NavTarget := MakeNavigationTarget(Center)
                        NavResult:=NPCNav.NavigateTo(NavTarget, ?MovementType := movement_types.Running, ?ReachRadius := 5.0, ?AllowPartialPath:=true)
                        if(NavResult = navigation_result.Reached):
                            NPCNav.StopNavigation()
                            NPCAnim.Play(CometAttack1)
                            Sleep(0.64)
                            NPCAnim.Play(CometAttack2)
                            Sleep(0.08)
                            if:
                                AM := AudioManager?
                                BeamAudio:=AM.DemonEffects[4]
                            then:
                                BeamAudio.Play(NewAgent)
                            Sleep(0.27)
                            spawn{StartSkyfall()}
                            Sleep(0.71)
                            #Print("STUNNED")
                            NPCAnim.Play(Stun)
                            Sleep(1.9)
                            NPCAnim.Play(Stun)
                            Sleep(1.9)
                            set CometsOn = false

    CheckForDamage(VFXCenter:vector3, DamageRange:float, Damage:float):void=
        if(IsElim = false):
            for(TargetAgent:TargetAgents):
                if:
                    #GET A TARGET
                    Target:=TargetAgent?
                    TargetFC:=Target.GetFortCharacter[]
                then:
                    if(Distance(VFXCenter, TargetFC.GetTransform().Translation) < DamageRange):
                        TargetFC.Damage(Damage)   
                    else:
                        #Print("PLAYER OUT OF RANGE")

    StartSkyfall()<suspends>:void=
        var HowManyComets : int = 10
        if(HardMode = true):
            set HowManyComets *= 5

        for(i:=0..HowManyComets):
            MinX := Min(TopLeftCorner.X, BottomRightCorner.X)
            MaxX := Max(TopLeftCorner.X, BottomRightCorner.X)
            MinY := Min(TopLeftCorner.Y, BottomRightCorner.Y)
            MaxY := Max(TopLeftCorner.Y, BottomRightCorner.Y)

            RandomX1 := GetRandomFloat(MinX, MaxX)
            RandomY1 := GetRandomFloat(MinY, MaxY)

            RandomX2 := GetRandomFloat(MinX, MaxX)
            RandomY2 := GetRandomFloat(MinY, MaxY)

            RandomX3:= GetRandomFloat(MinX, MaxX)
            RandomY3 := GetRandomFloat(MinY, MaxY)

            BeamLoc1 := vector3{X:=RandomX1, Y:=RandomY1, Z:=Center.Z}
            BeamLoc2 := vector3{X:=RandomX2, Y:=RandomY2, Z:=Center.Z}
            BeamLoc3 := vector3{X:=RandomX3, Y:=RandomY3, Z:=Center.Z}

            spawn{SpawnCometVFX(BeamLoc1)}
            spawn{SpawnCometVFX(BeamLoc2)}
            spawn{SpawnCometVFX(BeamLoc3)}

            Sleep(0.2)

    SpawnCometVFX(BeamLoc:vector3)<suspends>:void=
        NewRotation : rotation = IdentityRotation()
        NewScale : vector3 = vector3{X:=1.0, Y:=1.0, Z:=1.0}
        BeamVFX := SpawnProp(Comet, transform{Translation:=vector3{X:=BeamLoc.X,Y:=BeamLoc.Y,Z:=Center.Z}, Rotation:=NewRotation, Scale:=NewScale})
        Sleep(0.72)
        CheckForDamage(BeamLoc, AttackComet(0), AttackComet(1))
        Sleep(7.0)
        PropInst := BeamVFX(0)
        if ( Prop := PropInst?):
            Prop.Dispose()  

    SpawnAllBeamVFX(BeamLoc:vector3, FC:fort_character)<suspends>:void=
        var TornadoChaseTime : float = 2.0
        if(HardMode = true):
            set TornadoChaseTime = 1.0
        NewRotation : rotation = IdentityRotation()
        NewScale : vector3 = vector3{X:=1.0, Y:=1.0, Z:=1.0}
        BeamVFX := SpawnProp(Tornado, transform{Translation:=vector3{X:=BeamLoc.X,Y:=BeamLoc.Y,Z:=Center.Z}, Rotation:=NewRotation, Scale:=NewScale})
        Sleep(1.1)
        PropInst := BeamVFX(0)
        if(Prop := PropInst?):
            for(i:=0..4):
                if:
                    NPCAgent := GetAgent[]
                    NPCChar := NPCAgent.GetFortCharacter[]
                    NPCChar.IsActive[]
                then:
                    FortLoc := FC.GetTransform().Translation
                    spawn{TornadoLoop(Prop)}
                    Prop.MoveTo(vector3{X:=FortLoc.X,Y:=FortLoc.Y,Z:=Center.Z}, NewRotation, TornadoChaseTime)
                else:
                    Print("NPC BIG DEAD")
                #CheckForDamage(BeamLoc, AttackTornado(0), AttackTornado(1))
            Prop.Dispose()

    TornadoLoop(Prop:creative_prop)<suspends>:void=
        for(i:=0..10):
            PropLoc := Prop.GetTransform().Translation
            CheckForDamage(PropLoc, AttackTornado(0), AttackTornado(1))
            Sleep(0.1)

    SetTargetAgents(Agents:[int]?agent, AM:?audio_manager):void=
        set TargetAgents = Agents
        set AudioManager = AM

    SetAnims(BTornado:creative_prop_asset,BComet:creative_prop_asset):void=
        set Tornado = BTornado
        set Comet = BComet
        
    SetHardMode(IsHardMode:logic):void=
        set HardMode = IsHardMode

        
    SetIsElim(ElimResult:elimination_result):void=
        Print("SET ELIMINATED")
        set IsElim = true