using {/Fortnite.com/Characters }
using { /Fortnite.com/AI }
using { /Fortnite.com/Animation/PlayAnimation }
using { /Fortnite.com/Devices }
using { /Fortnite.com/Game }
using { /UnrealEngine.com/Assets }
using { /UnrealEngine.com/Temporary/SpatialMath }
using { /Verse.org/Assets }
using { /Verse.org/Random }
using { /Verse.org/Simulation }
 
boss6_behaviour2 := class(npc_behavior):
    var TargetAgents : [int]?agent = map{}

    var AudioManager : ?audio_manager = false

    var Tornado : creative_prop_asset = DefaultCreativePropAsset
    var Comet : creative_prop_asset = DefaultCreativePropAsset
    var MeteorShower : creative_prop_asset = DefaultCreativePropAsset
    var Beam : creative_prop_asset = DefaultCreativePropAsset

    var Center : vector3 = vector3{X:=-221052.0, Y:=-222096.0, Z:=16896.0}
    var TopLeftCorner : vector2 = vector2{X:=-216813.0, Y:=-226800.0}
    var BottomRightCorner : vector2 = vector2{X:=-225590.0, Y:=-214402.0}

    var HardMode : logic = false
    var HardModeDamageMult : float = 5.0

    var CometsOn : logic = false
    
    var RoarStart<public>: animation_sequence = Bosses.A_DemonBoss_Spell_Unarmed_Up_Start
    var RoarLoop<public>: animation_sequence = Bosses.A_DemonBoss_Spell_Unarmed_Up_Loop
    var RoarEnd<public>: animation_sequence = Bosses.A_DemonBoss_Spell_Unarmed_Up_Start

    var Stun<public>: animation_sequence = Bosses.A_DemonBoss_Combat_Fly_Stun

    var BeamAttack<public>: animation_sequence = Bosses.A_DemonBoss_Combat_Fly_Attack
    var TornadoAttack<public>: animation_sequence = Bosses.A_DemonBoss_Combat_Fly_Attack01

    var CometAttackStart<public>: animation_sequence = Bosses.A_DemonBoss_Spell_Fly_Forward_Start
    var CometAttackLoop<public>: animation_sequence = Bosses.A_DemonBoss_Spell_Fly_Forward_Loop

    var AttackBeam : tuple(float, float) = (300.0, 150.0)
    var AttackComet : tuple(float, float) = (350.0, 100.0)
    var AttackTornado : tuple(float, float) = (600.0, 5.0)

    var IsElim : logic = false
    OnBegin<override>()<suspends>:void=
        spawn{StartBossLogic()}

    StartBossLogic()<suspends>:void=
        set IsElim = false
        if:
            NPCAgent := GetAgent[]
            NPCChar := NPCAgent.GetFortCharacter[]
            NPCNav := NPCChar.GetNavigatable[]
            NPCFocus := NPCChar.GetFocusInterface[]
            NPCAnim := NPCChar.GetPlayAnimationController[]
        then:
            Sleep(0.5)
            if(HardMode = true):
                set AttackTornado = (AttackTornado(0), AttackTornado(1)*HardModeDamageMult)
                set AttackBeam = (AttackBeam(0), AttackBeam(1)*HardModeDamageMult)
                set AttackComet = (AttackComet(0), AttackComet(1)*HardModeDamageMult)
            loop:
                Sleep(0.1)
                if(IsElim = true):
                    Print("BREAK BEHAVIOUR LOOP")
                    break:
                <# if(NPCChar.GetTransform().Translation.Z < 16830.0):
                    Destination := vector3{
                        X := -221052.0
                        Y := -222096.0
                        Z := 16892.0
                    }
                    if(NPCChar.TeleportTo[Destination, NPCChar.GetTransform().Rotation]) {} #>
                WhichTarget:=GetRandomInt(0,TargetAgents.Length-1)
                if:
                    NewAgent := TargetAgents[WhichTarget]?
                    NewTarget := NewAgent.GetFortCharacter[]
                then:
                    spawn{NPCFocus.MaintainFocus(NewAgent)}
                    IDX:=GetRandomInt(1,10) #CHANGE TO 1,10
                    if(IDX < 4 and CometsOn = false): #RUN AND ATTACK   
                        set CometsOn = true  
                        NPCAnim.Play(TornadoAttack)
                        Sleep(0.64)
                        if:
                            AM := AudioManager?
                            BeamAudio:=AM.DemonEffects[4]
                        then:
                            BeamAudio.Play(NewAgent)
                        Sleep(0.27)
                        spawn{StartSkyfall()}
                        Sleep(0.71)
                        set CometsOn = false

                    else if(IDX < 7): #GROUND SLAM INTO ALL x3   
                        NPCAnim.Play(BeamAttack)
                        Sleep(0.50)
                        for(TargetAgent:TargetAgents):
                            if(Target:=TargetAgent?, TargetFC:=Target.GetFortCharacter[]):
                                BeamLoc := TargetFC.GetTransform().Translation
                                if:
                                    AM := AudioManager?
                                    BeamAudio:=AM.DemonEffects[2]
                                then:
                                    BeamAudio.Play(Target)
                                spawn{SpawnTornadoVFX(BeamLoc, TargetFC)}
                                Sleep(0.05)
                        Sleep(1.0)

                    else if (IDX >= 7): #ALL BEAMS 
                        for(x := 0 .. 2):
                            NPCAnim.Play(BeamAttack)
                            Sleep(0.50)
                            for(TargetAgent:TargetAgents):
                                if(Target:=TargetAgent?, TargetFC:=Target.GetFortCharacter[]):
                                    BeamLoc := TargetFC.GetTransform().Translation
                                    if:
                                        AM := AudioManager?
                                        BeamAudio:=AM.DemonEffects[3]
                                    then:
                                        BeamAudio.Play(Target)
                                    spawn{SpawnAllBeamVFX(BeamLoc)}
                                    Sleep(0.05)
                            Sleep(1.0)


    CheckForDamage(VFXCenter:vector3, DamageRange:float, Damage:float):void=
        if(IsElim = false):
            for(TargetAgent:TargetAgents):
                if:                
                    Target:=TargetAgent?
                    TargetFC:=Target.GetFortCharacter[]
                then:
                    if(Distance(VFXCenter, TargetFC.GetTransform().Translation) < DamageRange):
                        #Print("PLAYER DAMAGE")
                        TargetFC.Damage(Damage)   
                    else:
                        #Print("PLAYER OUT OF RANGE")

    StartSkyfall()<suspends>:void=
        var HowManyComets : int = 15
        if(HardMode = true):
            set HowManyComets *= 5

        for(i:=0..HowManyComets):
            MinX := Min(TopLeftCorner.X, BottomRightCorner.X)
            MaxX := Max(TopLeftCorner.X, BottomRightCorner.X)
            MinY := Min(TopLeftCorner.Y, BottomRightCorner.Y)
            MaxY := Max(TopLeftCorner.Y, BottomRightCorner.Y)

            RandomX1 := GetRandomFloat(MinX, MaxX)
            RandomY1 := GetRandomFloat(MinY, MaxY)

            RandomX2 := GetRandomFloat(MinX, MaxX)
            RandomY2 := GetRandomFloat(MinY, MaxY)

            RandomX3 := GetRandomFloat(MinX, MaxX)
            RandomY3 := GetRandomFloat(MinY, MaxY)

            RandomX4 := GetRandomFloat(MinX, MaxX)
            RandomY4 := GetRandomFloat(MinY, MaxY)

            RandomX5 := GetRandomFloat(MinX, MaxX)
            RandomY5 := GetRandomFloat(MinY, MaxY)

            BeamLoc1 := vector3{X:=RandomX1, Y:=RandomY1, Z:=Center.Z}
            BeamLoc2 := vector3{X:=RandomX2, Y:=RandomY2, Z:=Center.Z}
            BeamLoc3 := vector3{X:=RandomX3, Y:=RandomY3, Z:=Center.Z}
            BeamLoc4 := vector3{X:=RandomX4, Y:=RandomY4, Z:=Center.Z}
            BeamLoc5 := vector3{X:=RandomX5, Y:=RandomY5, Z:=Center.Z}

            spawn{SpawnCometVFX(BeamLoc1)}
            spawn{SpawnCometVFX(BeamLoc2)}
            spawn{SpawnCometVFX(BeamLoc3)}
            spawn{SpawnCometVFX(BeamLoc4)}
            spawn{SpawnCometVFX(BeamLoc5)}

            Sleep(0.1)

    SpawnCometVFX(BeamLoc:vector3)<suspends>:void=
        NewRotation : rotation = IdentityRotation()
        NewScale : vector3 = vector3{X:=1.0, Y:=1.0, Z:=1.0}
        BeamVFX := SpawnProp(Comet, transform{Translation:=vector3{X:=BeamLoc.X,Y:=BeamLoc.Y,Z:=Center.Z}, Rotation:=NewRotation, Scale:=NewScale})
        Sleep(0.7)
        CheckForDamage(BeamLoc, AttackComet(0), AttackComet(1))
        Sleep(7.0)
        PropInst := BeamVFX(0)
        if(Prop := PropInst?):
            Prop.Dispose()  

    SpawnTornadoVFX(BeamLoc:vector3, FC:fort_character)<suspends>:void=
        var TornadoChaseTime : float = 1.5
        if(HardMode = true):
            set TornadoChaseTime = 0.75
        NewRotation : rotation = IdentityRotation()
        NewScale : vector3 = vector3{X:=1.0, Y:=1.0, Z:=1.0}
        BeamVFX := SpawnProp(Tornado, transform{Translation:=vector3{X:=BeamLoc.X,Y:=BeamLoc.Y,Z:=Center.Z}, Rotation:=NewRotation, Scale:=NewScale})
        Sleep(1.1)
        PropInst := BeamVFX(0)
        if(Prop := PropInst?):
            for(i:=0..10):
                if:
                    NPCAgent := GetAgent[]
                    NPCChar := NPCAgent.GetFortCharacter[]
                    NPCChar.IsActive[]
                then:
                    FortLoc := FC.GetTransform().Translation
                    spawn{TornadoLoop(Prop)}
                    Prop.MoveTo(vector3{X:=FortLoc.X,Y:=FortLoc.Y,Z:=Center.Z}, NewRotation, TornadoChaseTime)
                else:
                    Print("NPC BIG DEAD")
                #CheckForDamage(BeamLoc, AttackTornado(0), AttackTornado(1))
            Prop.Dispose()

    TornadoLoop(Prop:creative_prop)<suspends>:void=
        for(i:=0..10):
                PropLoc := Prop.GetTransform().Translation
                CheckForDamage(PropLoc, AttackTornado(0), AttackTornado(1))
                Sleep(0.1)

    SpawnAllBeamVFX(BeamLoc:vector3)<suspends>:void=
        NewRotation : rotation = IdentityRotation()
        NewScale : vector3 = vector3{X:=1.0, Y:=1.0, Z:=1.0}
        BeamVFX := SpawnProp(Beam, transform{Translation:=vector3{X:=BeamLoc.X,Y:=BeamLoc.Y,Z:=Center.Z+10.0}, Rotation:=NewRotation, Scale:=NewScale})
        Sleep(0.72)
        CheckForDamage(BeamLoc, AttackBeam(0), AttackBeam(1))
        Sleep(7.0)
        PropInst := BeamVFX(0)
        if(Prop := PropInst?):
            Prop.Dispose()

    SetTargetAgents(Agents:[int]?agent, AM:?audio_manager):void=
        set TargetAgents = Agents
        set AudioManager = AM

    SetAnims(BTornado:creative_prop_asset,BComet:creative_prop_asset, BMeteorShower:creative_prop_asset, BBeam:creative_prop_asset):void=
        set Tornado = BTornado
        set Comet = BComet
        set MeteorShower = BMeteorShower
        set Beam = BBeam
        
    SetHardMode(IsHardMode:logic):void=
        set HardMode = IsHardMode
        
    SetIsElim(ElimResult:elimination_result):void=
        Print("SET ELIMINATED")
        set IsElim = true