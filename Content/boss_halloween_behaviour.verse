using {/Fortnite.com/Characters }
using { /Fortnite.com/AI }
using { /Fortnite.com/Animation/PlayAnimation }
using { /Fortnite.com/Devices }
using { /Fortnite.com/Game }
using { /UnrealEngine.com/Assets }
using { /UnrealEngine.com/Temporary/SpatialMath }
using { /Verse.org/Assets }
using { /Verse.org/Random }
using { /Verse.org/Simulation }
 
halloween_boss_behaviour := class(npc_behavior):

    var TargetAgents : [int]?agent = map{}

    var AudioManager : ?audio_manager = false

    var BeamAll : creative_prop_asset = DefaultCreativePropAsset
    var SummonBeam : creative_prop_asset = DefaultCreativePropAsset

    var Center : vector3 = vector3{X:=-249346.0, Y:=175866.0, Z:=31876.0}

    var Roar<public>: animation_sequence = Bosses.A_PumpkinCreature_Emote_Roar
    var Stun<public>: animation_sequence = Bosses.A_PumpkinCreature_Combat_Stun

    var CastStart<public>: animation_sequence = Bosses.A_PumpkinCreature_Combat_Cast_Up_Impact
    var CastLoop<public>: animation_sequence = Bosses.A_PumpkinCreature_Combat_Cast_Up_Channel

    var SummonStart<public>: animation_sequence = Bosses.A_PumpkinCreature_Combat_Cast_Forward_Impact
    var SummonLoop<public>: animation_sequence = Bosses.A_PumpkinCreature_Combat_Cast_Forward_Channel

    var Punch<public>: animation_sequence = Bosses.A_PumpkinCreature_Combat_Unarmed_Attack

    var StartLaugh<public>: animation_sequence = Bosses.A_PumpkinCreature_Emote_Laugh

    var IsNewPhase : logic = false
    var WhichPhase : int = 1

    var IsElim : logic = false
    OnBegin<override>()<suspends>:void=
        spawn{StartBossLogic()}

    StartBossLogic()<suspends>:void=
        set IsElim = false
        if:
            NPCAgent := GetAgent[]
            NPCChar := NPCAgent.GetFortCharacter[]
            NPCNav := NPCChar.GetNavigatable[]
            NPCFocus := NPCChar.GetFocusInterface[]
            NPCAnim := NPCChar.GetPlayAnimationController[]
        then:
            Sleep(0.6)
            set WhichPhase = 1
            if:
                AM := AudioManager?
                Laugh:=AM.HalloweenBossEffects[0]
                NewAgent := TargetAgents[0]?
                NewTarget := NewAgent.GetFortCharacter[]
            then:
                Laugh.Play(NewAgent)
                Sleep(0.5)
                NPCAnim.Play(StartLaugh)
                spawn{NPCFocus.MaintainFocus(NewAgent)}
                Sleep(2.2)
            loop:
                Sleep(0.1)
                if(IsElim = true):
                    Print("BREAK BEHAVIOUR LOOP")
                    break:
                <# if(NPCChar.GetTransform().Translation.Z < 31870.0):
                    Destination := vector3{
                        X := -250584.276547
                        Y := 175936.000001
                        Z := 31876.0
                    }
                    if(NPCChar.TeleportTo[Destination, NPCChar.GetTransform().Rotation]) {} #>
                WhichTarget:=GetRandomInt(0,TargetAgents.Length-1)
                if:
                    NewAgent := TargetAgents[WhichTarget]?
                    NewTarget := NewAgent.GetFortCharacter[]
                then:
                    if(IsNewPhase = false):
                        spawn{NPCFocus.MaintainFocus(NewAgent)}
                        IDX:=GetRandomInt(1,10) #CHANGE TO 1,10
                        if(IDX < 7): #RUN AND ATTACK      
                            #Print("RUN AT SOMEONE")
                            NavTarget := MakeNavigationTarget(NewAgent)
                            NavResult:=NPCNav.NavigateTo(NavTarget, ?MovementType := movement_types.Running, ?ReachRadius := 250.0, ?AllowPartialPath:=true)
                            if(NavResult = navigation_result.Reached):
                                #Print("REACHED")
                                NPCNav.StopNavigation()       

                                NPCAnim.Play(Punch)
                                Sleep(0.34)
                                CheckForDamage(NPCChar.GetTransform().Translation, 450.0, 40.0)
                                Sleep(0.66)
                        else: #GROUND SLAM INTO ALL x5
                            NPCAnim.Play(CastStart)
                            Sleep(0.62)
                            for(x := 0 .. 4):
                                NPCAnim.Play(CastLoop)
                                for(TargetAgent:TargetAgents):
                                    if(Target:=TargetAgent?, TargetFC:=Target.GetFortCharacter[]):
                                        BeamLoc := TargetFC.GetTransform().Translation
                                        if:
                                            AM := AudioManager?
                                            BeamAudio:=AM.HalloweenBossEffects[1]
                                        then:
                                            BeamAudio.Play(Target)
                                        spawn{SpawnAllBeamVFX(BeamLoc)}
                                        Sleep(0.05)
                                Sleep(0.85)
                    else:
                        NPCNav.StopNavigation()
                        NPCAnim.Play(SummonStart)
                        Sleep(0.50)
                        NPCAnim.Play(SummonLoop)
                        if(WhichPhase = 2):
                            for(x := 0 .. 47):
                                NPCAnim.Play(SummonLoop)
                                Sleep(0.90)
                        set IsNewPhase = false

    CheckForDamage(VFXCenter:vector3, DamageRange:float, Damage:float):void=
        for(TargetAgent:TargetAgents):
            if(Target:=TargetAgent?, TargetFC:=Target.GetFortCharacter[]):
                if(Distance(VFXCenter, TargetFC.GetTransform().Translation) < DamageRange):
                    #Print("PLAYER DAMAGE")
                    TargetFC.Damage(Damage)   
                else:
                    #Print("PLAYER OUT OF RANGE")

    SpawnAllBeamVFX(BeamLoc:vector3)<suspends>:void=
        NewRotation : rotation = IdentityRotation()
        NewScale : vector3 = vector3{X:=1.0, Y:=1.0, Z:=1.0}
        BeamVFX := SpawnProp(BeamAll, transform{Translation:=vector3{X:=BeamLoc.X,Y:=BeamLoc.Y,Z:=BeamLoc.Z-60.0}, Rotation:=NewRotation, Scale:=NewScale})
        Sleep(0.2)
        CheckForDamage(BeamLoc, 90.0, 60.0)
        Sleep(4.0)
        PropInst := BeamVFX(0)
        if ( Prop := PropInst?):
            Prop.Dispose()

    SetAnims(VFX1:creative_prop_asset,VFX2:creative_prop_asset):void=
        block:
        set BeamAll = VFX1
        set SummonBeam = VFX2

    SetBossPhase():void=
        #Print("NEW BOSS PHASE SPAWN MOOOOBS")
        set IsNewPhase = true
        set WhichPhase += 1
        #spawn{TimerToEndPhase()}

    TimerToEndPhase()<suspends>:void=
        Sleep(45.0)
        set IsNewPhase = false


    SetTargetAgents(Agents:[int]?agent, AM:?audio_manager):void=
        set TargetAgents = Agents
        set AudioManager = AM


    SetIsElim(ElimResult:elimination_result):void=
        Print("SET ELIMINATED")
        set IsElim = true