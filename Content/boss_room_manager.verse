using { /Fortnite.com/Devices }
using { /Verse.org/Simulation }
using { /Verse.org/Colors }
using { /Verse.org/Verse }
using { /Verse.org/Random }
using { /Fortnite.com/Teams }
using { /UnrealEngine.com/Assets }
using { /Fortnite.com/Game }
using { /Fortnite.com/Characters }
using { /Fortnite.com/Playspaces }
using { /Fortnite.com/UI }
using { /UnrealEngine.com/Temporary/Diagnostics }
using { /UnrealEngine.com/Temporary/SpatialMath }
using { /UnrealEngine.com/Temporary/UI }
using { /Verse.org/Assets}
using { /Verse.org/Colors/NamedColors }
using { /Fortnite.com/AI }

boss_room_manager := class(creative_device):

    @editable
    gm : game_manager = game_manager{}
    
    @editable
    Halloween : seasonal_halloween_manager = seasonal_halloween_manager{}
    
    @editable
    pm: pet_manager = pet_manager{}

    @editable
    prestige: ui_prestige_upgrades = ui_prestige_upgrades{}

    @editable
    AudioManager: audio_manager = audio_manager{}

    @editable
    Tutorial: tutorial_manager = tutorial_manager{}

    @editable
    DRM : new_daily_rewards_manager = new_daily_rewards_manager{}

    var DamageNumbers<private>: damage_numbers = damage_numbers{}

    HPBars: health_bar_device = health_bar_device{}

    @editable
    BossDefeatedVFX : creative_prop_asset = DefaultCreativePropAsset

    @editable
    Boss4BeamSingle : creative_prop_asset = DefaultCreativePropAsset
    @editable
    Boss4BeamAll : creative_prop_asset = DefaultCreativePropAsset
    @editable
    Boss4BeamRandom : creative_prop_asset = DefaultCreativePropAsset

    @editable
    Boss5BeamAll : creative_prop_asset = DefaultCreativePropAsset
    @editable
    Boss5GroundSlam : creative_prop_asset = DefaultCreativePropAsset

    @editable
    HalloweenBeam : creative_prop_asset = DefaultCreativePropAsset
    @editable
    HalloweenSummon : creative_prop_asset = DefaultCreativePropAsset

    @editable
    Boss6Tornado : creative_prop_asset = DefaultCreativePropAsset
    @editable
    Boss6Comet : creative_prop_asset = DefaultCreativePropAsset
    @editable
    Boss6Beam : creative_prop_asset = DefaultCreativePropAsset
    @editable
    Boss6MeteorShower : creative_prop_asset = DefaultCreativePropAsset

    @editable
    LootOrbCommon : creative_prop_asset = DefaultCreativePropAsset
    @editable
    LootOrbUncommon : creative_prop_asset = DefaultCreativePropAsset
    @editable
    LootOrbRare : creative_prop_asset = DefaultCreativePropAsset
    @editable
    LootOrbEpic : creative_prop_asset = DefaultCreativePropAsset
    @editable
    LootOrbLegendary : creative_prop_asset = DefaultCreativePropAsset
    @editable
    LootOrbTranscendant : creative_prop_asset = DefaultCreativePropAsset


    @editable
    BossSpawners : []npc_spawner_device = array{}
    @editable
    Boss6Phase2 : npc_spawner_device = npc_spawner_device{}

    @editable
    MobSpawners : []npc_spawner_device = array{}

    @editable
    BossEnterTPs : []teleporter_device = array{}
    @editable
    BossEnterButtons : []button_device = array{}

    @editable
    BossExitTPs : []teleporter_device = array{}
    @editable
    BossExitButtons : []button_device = array{}

    @editable
    HalloweenRewardButton : button_device = button_device{}
    @editable
    BossRewardButtons : []button_device = array{}

    @editable
    DemonBossStats : button_device = button_device{}

    @editable
    BossRoomFreezes : []mutator_zone_device = array{}
    @editable
    PreventBossScum : mutator_zone_device = mutator_zone_device{}

    @editable
    EnterBossRoomSeq1 : []cinematic_sequence_device = array{}
    @editable
    EnterBossRoomSeq2 : []cinematic_sequence_device = array{}
    @editable
    EnterBossRoomSeq3 : []cinematic_sequence_device = array{}
    @editable
    EnterBossRoomSeq4 : []cinematic_sequence_device = array{}
    @editable
    EnterBossRoomSeq5 : []cinematic_sequence_device = array{}
    @editable
    EnterBossRoomSeq6 : []cinematic_sequence_device = array{}
    @editable
    EnterBossRoomSeq7 : []cinematic_sequence_device = array{}
    @editable
    DemonBossPhaseSeq : []cinematic_sequence_device = array{}

    var AllEnterBossRoomSeqs :[][]cinematic_sequence_device = array{}

    var QueuedPlayers1 : [int]?agent = map{}
    var QueuedPlayers2 : [int]?agent = map{}
    var QueuedPlayers3 : [int]?agent = map{}
    var QueuedPlayers4 : [int]?agent = map{}
    var QueuedPlayers5 : [int]?agent = map{}
    var QueuedPlayers6 : [int]?agent = map{}
    var QueuedPlayers7 : [int]?agent = map{}
    var AllQueuedPlayers : [][int]?agent = array{}
    var TimerDefault : float = 20.0 #CHANGE TO CHANGE ALL 20.0
    var AllQueueTimers : []float = array{20.0, 20.0, 20.0, 20.0, 20.0, 20.0, 20.0, 20.0}
    var TimersStarted : []logic = array{false, false, false, false, false, false, false,false}
    var StageRequirements : []int = array{4, 9, 14, 19, 16, 0, 23, 23}

    #HALLOWEEN
    var CanBeDamaged : logic = true
    var BossPhase : int = 1
    var DemonPhase : int = 0
    var HalloweenPhaseTimer : float = 45.0
    var SpawningBosses : [int]logic = map{}

    var FBTimerDefault : float = 180.0 #CHANGE TO 180.0
    var FBTimer : float = 180.0  #CHANGE TO 180.0
    var DemonTimer : float = 120.0 #180.0  #CHANGE TO 180.0
    var TheoreticalMaxDmg : float = 1.5e22
    var BossCanBeElimed : logic = false
                                      #STONE | WEP MAT | SHARDS | EGGS | GATHER RATE
    var MaxFBRewards : []float = array{1.68e50, 12000.0, 18000.0, 120000.0, 6000.0}

    var EachBossHPs : [fort_character]float = map{}
    var EachBossLocation : [fort_character]vector3 = map{}
    var EachBoss : [int]fort_character = map{}
    var AllBossIcons : []texture = array{UI.Boss1,UI.Boss2,UI.Boss3,UI.Boss4,UI.Boss5,UI.BossHalloween,UI.Boss6, UI.Boss6}
    var BossNames : []string = array{"Mushroom King", "Razor Fang", "Cactoro", "Astralis", "Bramblerock", "GrimRoot", "The Blamed One", "The Eternal One"}
    var BossMaxHPs : []float = array{1500000.0, 5343750000.0, 1518750000000000.0, 42e12, 8e18, 100000.0, 666e22, 3.33e28} #CHANGE LAST TO 100000.0
    var DemonMaxHP : float = 3.33e28 # 666e22 / 3.33e28
    var BossMaxHP : float = 100000.0 #100000.0 #CHANGE TO 100000.0
    var BossKeyCosts : []float = array{1.0, 5.0, 10.0, 500.0, 80.0, 10.0, 2000.0, 2000.0}
    
    HalloweenBossRewards : []float = array{30.0, 150.0, 700.0, 800.0, 10.0}

    var DemonTotalDamage : float = 0.0

    DemonBossRewards : [int][]tuple(string, float) = map:
        0 => array:
            ("Stone", 250.0)
            ("WepMat", 15000.0)
            ("Egg", 860000.0) #<> 10%
            ("Shard", 480000.0)
        1 => array:
            ("Stone", 300.0)
            ("WepMat", 19000.0)
            ("Egg", 1184000.0) #<> 10%
            ("Shard", 592000.0)
        2 => array:
            ("Stone", 350.0)
            ("WepMat", 28600.0)
            ("Egg", 1497600.0) #<> 10%
            ("Shard", 748800.0)
        3 => array:
            ("Stone", 400.0)
            ("WepMat", 36040.0)
            ("Egg", 1736640.0) #<> 10%
            ("Shard", 968320.0)
        4 => array:
            ("Stone", 450.0)
            ("WepMat", 44456.0)
            ("Egg", 2651296.0) #<> 10%
            ("Shard", 1475648.0)

    # xGATHER RATE | WEP MAT | EGG MIN | EGG MAX | PRESTIGE SHARDS
    <#DemonBossRewards : [int]tuple(float, float, float, float, float) = map:
        0 => (200.0, 2500.0, 7000.0, 8000.0, 0.0)
        1 => (240.0, 3000.0, 7000.0, 8000.0, 0.0)
        2 => (275.0, 3500.0, 7000.0, 8000.0, 0.0)
        3 => (300.0, 4000.0, 7000.0, 8000.0, 500.0)
        4 => (350.0, 4500.0, 7000.0, 8000.0, 1000.0)
        5 => (400.0, 5000.0, 7000.0, 8000.0, 2000.0)#>

    # xGATHER RATE | WEP MAT | EGG MIN | EGG MAX
    var AllBossRewards : [][]float = array{
        array{20.0, 15.00, 50.00, 85.00},
        array{40.0, 55.00, 180.0, 220.0},
        array{60.0, 100.00, 440.0, 580.0},
        array{100.0, 1050.00, 3800.0, 5200.0},
        array{100.0, 1050.00, 3800.0, 5200.0},
        array{1.0, 1.0, 1.0, 1.0},
        
        array{450.0, 44456.0, 2651296.0, 1475648.0},
        array{450.0, 44456.0, 2651296.0, 1475648.0}
        }

    var HardModeMult : float = 10.0
    var LootOrbLoc : vector3 = vector3{X := -217382.0, Y := -221323.0, Z := 16935.0}

    var SomeoneQueued : []logic = array{false, false, false, false, false, false, false, false}
    var BossFightStarted : []logic = array{false, false, false, false, false, false, false, false}
    var BossHardMode : []logic = array{false, false, false, false, false, false, false, false}
    var BossDefeated : []logic = array{false, false, false, false, false, false, false, false}

    OnBegin<override>()<suspends>:void=
        set AllQueuedPlayers = {QueuedPlayers1, QueuedPlayers2, QueuedPlayers3, QueuedPlayers4, QueuedPlayers5, QueuedPlayers6, QueuedPlayers7}
        set AllEnterBossRoomSeqs = array{EnterBossRoomSeq1,EnterBossRoomSeq2,EnterBossRoomSeq3,EnterBossRoomSeq4,EnterBossRoomSeq5,EnterBossRoomSeq6, EnterBossRoomSeq7}

        PreventBossScum.AgentEntersEvent.SubscribeAgent(TryingToBossScum, 1)
        PreventBossScum.AgentExitsEvent.SubscribeAgent(TryingToBossScum, 0)

        for(i:int=0..BossSpawners.Length):
            if(BossSpawner:=BossSpawners[i]):
                BossSpawner.SpawnedEvent.SubscribeAgent(OnBossSpawned, i)
                
        Boss6Phase2.SpawnedEvent.SubscribeAgent(OnBossSpawned, 6)

        DemonBossStats.InteractedWithEvent.Subscribe(OnDemonStats)

        for(i:int=0..BossEnterButtons.Length):
            if(BossEnterButton:=BossEnterButtons[i]):
                BossEnterButton.InteractedWithEvent.SubscribeAgent(OpenBossHUD, i)
                
        for(i:int=0..BossExitButtons.Length):
            if(BossExitButton:=BossExitButtons[i]):
                BossExitButton.InteractedWithEvent.SubscribeAgent(LeaveBossRoom, i)

        for(i:int=0..BossRewardButtons.Length):
            if(BossRewardButton:=BossRewardButtons[i]):
                BossRewardButton.InteractedWithEvent.SubscribeAgent(OnRewardButton, i)
                
        for(i:int=0..MobSpawners.Length):
            if(MobSpawner:=MobSpawners[i]):
                MobSpawner.SpawnedEvent.SubscribeAgent(OnMobSpawned, i)

        HalloweenRewardButton.InteractedWithEvent.SubscribeAgent(OnRewardButton, 5)

        for(Spawner : gm.PlayerSpawners):     
            Spawner.SpawnedEvent.Subscribe(OnPlayerJoined)

    OnPlayerJoined(Agent:agent):void=
        spawn{OnPlayerJoinedBuffer(Agent)}

    OnPlayerJoinedBuffer(Agent:agent)<suspends>:void=
        Sleep(1.0)
        if(CustomPlayer := gm.PlayersMap[player[Agent]], Fort:=CustomPlayer.Player.GetFortCharacter[]):
            if(CustomPlayer.PlayTimeSessionSeconds = 0.0):
                CustomPlayer.JoinButton.OnClick().Subscribe(JoinQueue)
                CustomPlayer.LeaveButton.OnClick().Subscribe(LeaveQueue)
                CustomPlayer.CallHelpButton.OnClick().Subscribe(CallForHelp)
                CustomPlayer.HardModeButton.OnClick().Subscribe(ToggleHardMode)
                CustomPlayer.BossCloseButton.OnClick().Subscribe(RemoveBossHUD)
                CustomPlayer.PostFBCloseButton.OnClick().Subscribe(RemoveFBPostHUD)
                set CustomPlayer.AllTextBlocks = {CustomPlayer.BossWarning,CustomPlayer.BossKeyCostText,CustomPlayer.MessageText,CustomPlayer.QueueStatus,CustomPlayer.QueuedPlayer4,CustomPlayer.QueuedPlayer3,CustomPlayer.QueuedPlayer2,CustomPlayer.QueuedPlayer1,CustomPlayer.QueuedPlayersCount,CustomPlayer.QueuedPlayersTitle,CustomPlayer.Reward3Text,CustomPlayer.Reward2Text,CustomPlayer.BossKeyText, CustomPlayer.BossNameText, CustomPlayer.BossHealth, CustomPlayer.RewardTitle, CustomPlayer.Reward1Text}
                set CustomPlayer.QueuedPlayerBlocks = {CustomPlayer.QueuedPlayer1, CustomPlayer.QueuedPlayer2, CustomPlayer.QueuedPlayer3, CustomPlayer.QueuedPlayer4}
                set CustomPlayer.AllPostFBTextBlocks = array{CustomPlayer.PostFBBossDamageDealt,CustomPlayer.PostFBDamageTitle,CustomPlayer.PostFBName1,CustomPlayer.PostFBDamage1,CustomPlayer.PostFBName2,CustomPlayer.PostFBDamage2,CustomPlayer.PostFBName3,CustomPlayer.PostFBDamage3,CustomPlayer.PostFBName4,CustomPlayer.PostFBDamage4,CustomPlayer.PostFBBossRewardTitle,CustomPlayer.PostFBBossStoneReward,CustomPlayer.PostFBBossWepMatReward,CustomPlayer.PostFBBossShardReward,CustomPlayer.PostFBBossEggReward,CustomPlayer.PostFBBossGatherReward,CustomPlayer.PostFBPrestigeTitle,CustomPlayer.PostFBPrestigeShards,CustomPlayer.PostFBPrestigeLevel,CustomPlayer.PostFBPrestigeGather}
                set CustomPlayer.AllPostFBPlayerNames = array{CustomPlayer.PostFBName1,CustomPlayer.PostFBName2,CustomPlayer.PostFBName3,CustomPlayer.PostFBName4}
                set CustomPlayer.AllPostFBPlayerDamages = array{CustomPlayer.PostFBDamage1,CustomPlayer.PostFBDamage2,CustomPlayer.PostFBDamage3,CustomPlayer.PostFBDamage4}
                set CustomPlayer.AllBossRewardBlocks = {CustomPlayer.PostFBBossStoneReward,CustomPlayer.PostFBBossWepMatReward,CustomPlayer.PostFBBossShardReward,CustomPlayer.PostFBBossEggReward,CustomPlayer.PostFBBossGatherReward}        
    
                Fort.EliminatedEvent().SubscribeAgentElim(OnPlayerEliminated, CustomPlayer)
                
            if(CustomPlayer.CurrentBoss = 3):
                OpenFBPostHUD(CustomPlayer)
                EndFBFight(CustomPlayer, true)
                set CustomPlayer.CurrentBoss = -1

    OnDemonStats(Agent:agent):void=
        if:
            CustomPlayer := gm.PlayersMap[Agent]
        then:
            gm.ShowMessage(CustomPlayer.Player, "You did {NumberCondensor(CustomPlayer.FBDamage)} Damage, {NumberCondensor((CustomPlayer.FBDamage / DemonTotalDamage) * 100.0)}%!", 0)

    TryingToBossScum(Agent:agent, InOrOut:int): void=
        if:
            CustomPlayer := gm.PlayersMap[Agent]
            FC := Agent.GetFortCharacter[]
            CustomPlayer.CurrentBoss <> -1
        then:
            Print("TRYING TO SCUM")
            if(InOrOut = 1):
                Print("GOING IN")
                S:stasis_args = stasis_args{}
                if(S.AllowEmotes = false, S.AllowFalling = false, S.AllowTurning = false):
                    FC.PutInStasis(S)
            else:
                Print("COMING OUT")
                FC.ReleaseFromStasis()           

#CODE
    OpenBossHUD(Agent:agent, which:int): void=
        if(CustomPlayer := gm.PlayersMap[Agent]):
            Print("which: {which}")
            Print("CustomPlayer.StagesUnlocked: {CustomPlayer.StagesUnlocked}")
            if(CustomPlayer.CurrentBoss = -1 or CustomPlayer.CurrentBoss = which):
                if(CustomPlayer.StagesUnlocked >= StageRequirements[which]):
                    var NewPrestigeCost : float = 0.0
                    if(CustomPlayer.PrestigeCost * Pow(CustomPlayer.PrestigeCostMult, (CustomPlayer.PrestigeLevel)*1.0) > 1e300):
                        set NewPrestigeCost = 1e300
                    else:
                        set NewPrestigeCost = CustomPlayer.PrestigeCost * Pow(CustomPlayer.PrestigeCostMult, (CustomPlayer.PrestigeLevel)*1.0)
                    if:
                        which < 3 or
                        (which = 3 and CustomPlayer.PrestigeStoneAmount >= NewPrestigeCost) or
                        (which = 3 and CustomPlayer.PrestigeStoneAmount >= 1.0e154) or #1.0e154
                        which = 4 or
                        which = 6 or
                        (which = 5 and gm.SessionTime >= 900.0) #CHANGE TO 900.0
                    then:
                        if(MyUI:=CustomPlayer.UIMaps["Boss"]?):
                        else:
                            var MyBossCanvas : canvas = MakeBossCanvas(CustomPlayer)
                            CustomPlayer.AddFullHUD(MyBossCanvas)
                            set CustomPlayer.CurrentBoss = which
                            if(set CustomPlayer.UIMaps["Boss"] = option{MyBossCanvas}) {}
                            UpdateBossHUD(CustomPlayer)
                    else:
                        gm.ShowMessage(CustomPlayer.Player, "You are not ready!", 1)
                else:
                    gm.ShowMessage(CustomPlayer.Player, "You are not ready!", 1)
            else:
                gm.ShowMessage(CustomPlayer.Player, "You are already in queue for a boss!", 1)

                
    RemoveBossHUD(Wdata:widget_message):void=
        if(CustomPlayer := gm.PlayersMap[Wdata.Player]):
            if(Player:=Wdata.Player, PlayerUI:=GetPlayerUI[Player],SelectedButton:=Wdata.Source):
                if(MyUI:=CustomPlayer.UIMaps["Boss"]?):
                    PlayerUI.RemoveWidget(MyUI)
                    MyUI.SetVisibility(widget_visibility.Collapsed)
                    if(CustomPlayer.QueuedForBoss = true):
                        spawn{ShowBossTimer(CustomPlayer)}
                    else:
                        Print("REMOVE HUD SET TO -1")
                        set CustomPlayer.CurrentBoss = -1
                        set CustomPlayer.RequestedBossHelp = false
                    if(set CustomPlayer.UIMaps["Boss"] = false){}

    ForceRemoveBossHUD(CustomPlayer:Custom_Player):void=
        if(PlayerUI:=GetPlayerUI[CustomPlayer.Player]):
            if(MyUI:=CustomPlayer.UIMaps["Boss"]?):
                PlayerUI.RemoveWidget(MyUI)
                MyUI.SetVisibility(widget_visibility.Collapsed)
                if(set CustomPlayer.UIMaps["Boss"] = false){}

    UpdateBossHUD(CustomPlayer:Custom_Player):void=
        if:
            BossKeyCost:=BossKeyCosts[CustomPlayer.CurrentBoss]
            QueuedPlayers:=AllQueuedPlayers[CustomPlayer.CurrentBoss]
            CurrentBossIcon:=AllBossIcons[CustomPlayer.CurrentBoss]
            BossName:=BossNames[CustomPlayer.CurrentBoss]
            BossHP:=BossMaxHPs[CustomPlayer.CurrentBoss]
            BossRewards:=AllBossRewards[CustomPlayer.CurrentBoss]
            IsHardMode := BossHardMode[CustomPlayer.CurrentBoss]
        then:
            if(CustomPlayer.CurrentBoss <> 3 and CustomPlayer.CurrentBoss <> 5):
                CustomPlayer.HardModeButton.SetEnabled(true)
            else:
                CustomPlayer.HardModeButton.SetEnabled(false)

            CustomPlayer.JoinButton.SetText(StringToMessage("JOIN"))
            CustomPlayer.BossKeyText.SetText(StringToMessage(" {NumberCondensor(CustomPlayer.BossKeyAmount)}"))
            CustomPlayer.BossIcon.SetImage(CurrentBossIcon)
            if(IsHardMode = true):
                CustomPlayer.BossNameText.SetText(StringToMessage("HARD - {BossName}"))
                CustomPlayer.BossHealth.SetText(StringToMessage("Health: {NumberCondensor(BossHP * HardModeMult)}"))
                CustomPlayer.BossKeyCostText.SetText(StringToMessage(" {NumberCondensor(BossKeyCost * HardModeMult)}"))
            else:
                CustomPlayer.BossNameText.SetText(StringToMessage("{BossName}"))
                CustomPlayer.BossHealth.SetText(StringToMessage("Health: {NumberCondensor(BossHP)}"))
                CustomPlayer.BossKeyCostText.SetText(StringToMessage(" {NumberCondensor(BossKeyCost)}"))
            #CustomPlayer.BossNameText.SetText(StringToMessage("{BossName}"))
            #CustomPlayer.BossHealth.SetText(StringToMessage("Health: {NumberCondensor(BossHP)}"))
            if(CustomPlayer.CurrentBoss = 3):
                CustomPlayer.BossWarning.SetText(StringToMessage("PRESTIGE AFTER FIGHT"))
            else if(CustomPlayer.CurrentBoss = 4):
                CustomPlayer.BossWarning.SetText(StringToMessage("20 SEC IMMORTALITY"))
            else:
                CustomPlayer.BossWarning.SetText(StringToMessage(""))
            #CustomPlayer.BossKeyCostText.SetText(StringToMessage(" {NumberCondensor(BossKeyCost)}"))
            if:
                StoneReward:=BossRewards[0] * (1.0 + (CustomPlayer.PUPLevels[2]*0.20) + CustomPlayer.SinkBossKeyPower/100.0) 
                WepMatReward:=BossRewards[1] * (1.0 + (CustomPlayer.PUPLevels[2]*0.20) + CustomPlayer.SinkBossKeyPower/100.0) 
                EggMinReward:=BossRewards[2] * (1.0 + (CustomPlayer.PUPLevels[2]*0.20) + CustomPlayer.SinkBossKeyPower/100.0) 
                EggMaxReward:=BossRewards[3] * (1.0 + (CustomPlayer.PUPLevels[2]*0.20) + CustomPlayer.SinkBossKeyPower/100.0) 
                CustomPlayer.CurrentBoss <> 3 and CustomPlayer.CurrentBoss <> 5 and CustomPlayer.CurrentBoss <> 6
            then:
                TotalStone := CustomPlayer.GetPlayerDamage() * StoneReward
                if(IsHardMode = true):
                    CustomPlayer.Reward1Text.SetText(StringToMessage(" {NumberCondensor(TotalStone * HardModeMult)}"))
                    CustomPlayer.Reward2Text.SetText(StringToMessage(" {NumberCondensor(WepMatReward * HardModeMult)}"))
                    CustomPlayer.Reward3Text.SetText(StringToMessage(" {NumberCondensor(EggMinReward * HardModeMult)} - {NumberCondensor(EggMaxReward * 10.0)} Random Eggs"))
                else:
                    CustomPlayer.Reward1Text.SetText(StringToMessage(" {NumberCondensor(TotalStone)}"))
                    CustomPlayer.Reward2Text.SetText(StringToMessage(" {NumberCondensor(WepMatReward)}"))
                    CustomPlayer.Reward3Text.SetText(StringToMessage(" {NumberCondensor(EggMinReward)} - {NumberCondensor(EggMaxReward)} Random Eggs"))

                #CustomPlayer.Reward1Text.SetText(StringToMessage(" {NumberCondensor(TotalStone)}"))
                #CustomPlayer.Reward2Text.SetText(StringToMessage(" {NumberCondensor(WepMatReward)}"))
                #CustomPlayer.Reward3Text.SetText(StringToMessage(" {NumberCondensor(EggMinReward)} - {NumberCondensor(EggMaxReward)} Random Eggs"))
            else if(CustomPlayer.CurrentBoss = 3):
                if:
                    StoneReward := MaxFBRewards[0]
                    WepMatReward:=MaxFBRewards[1]
                    EggReward:=MaxFBRewards[3]
                then:
                    CustomPlayer.Reward1Text.SetText(StringToMessage(" Max: {NumberCondensor(StoneReward)}"))
                    CustomPlayer.Reward2Text.SetText(StringToMessage(" Max: {NumberCondensor(WepMatReward)}"))
                    CustomPlayer.Reward3Text.SetText(StringToMessage(" Max: {NumberCondensor(EggReward)} Random Eggs"))
            else if(CustomPlayer.CurrentBoss = 6):
                if:
                    StoneReward:=BossRewards[0] * (1.0 + (CustomPlayer.PUPLevels[2]*0.20) + CustomPlayer.SinkBossKeyPower/100.0) 
                    WepMatReward:=BossRewards[1] * (1.0 + (CustomPlayer.PUPLevels[2]*0.20) + CustomPlayer.SinkBossKeyPower/100.0)  
                    EggMinReward:=BossRewards[2] * (1.0 + (CustomPlayer.PUPLevels[2]*0.20) + CustomPlayer.SinkBossKeyPower/100.0) 
                    #EggMaxReward:=BossRewards[3] * ((1.0 + CustomPlayer.PUPLevels[2]*0.20) + (1.0 + CustomPlayer.SinkBossKeyPower/100.0)) 
                then:
                    TotalStone := CustomPlayer.GetPlayerDamage() * StoneReward
                    if(IsHardMode = true):
                        CustomPlayer.Reward1Text.SetText(StringToMessage(" Max: {NumberCondensor(TotalStone * HardModeMult)}"))
                        CustomPlayer.Reward2Text.SetText(StringToMessage(" Max: {NumberCondensor(WepMatReward * HardModeMult)}"))
                        CustomPlayer.Reward3Text.SetText(StringToMessage(" Max: {NumberCondensor(EggMinReward * HardModeMult)} Random Eggs"))
                    else:
                        CustomPlayer.Reward1Text.SetText(StringToMessage(" Max: {NumberCondensor(TotalStone)}"))
                        CustomPlayer.Reward2Text.SetText(StringToMessage(" Max: {NumberCondensor(WepMatReward)}"))
                        CustomPlayer.Reward3Text.SetText(StringToMessage(" Max: {NumberCondensor(EggMinReward)} Random Eggs"))
    
            else:
                CustomPlayer.BossHealth.SetText(StringToMessage("Health: ???"))
                CustomPlayer.Reward1Text.SetText(StringToMessage("???"))
                CustomPlayer.Reward2Text.SetText(StringToMessage("???"))
                CustomPlayer.Reward3Text.SetText(StringToMessage("???"))

            for(K -> CPlayer:gm.PlayersMap):
                set CPlayer.ActivePlayers = 0
                if(CPlayer.CurrentBoss = CustomPlayer.CurrentBoss):
                    for(i:int=0..QueuedPlayers.Length):
                        if(QueuedPlayer:=QueuedPlayers[i],QueuedPlayerBlock:=CPlayer.QueuedPlayerBlocks[i]):
                            if(PlayerAgent:=QueuedPlayer?):
                                QueuedPlayerBlock.SetText(PlayerToMessage(PlayerAgent))
                                set CPlayer.ActivePlayers += 1
                            else:
                                QueuedPlayerBlock.SetText(StringToMessage(""))
                    if(CPlayer.ActivePlayers = 0):
                        if(set SomeoneQueued[CPlayer.CurrentBoss] = false) {}
                    if:
                        QueuedPlayer := AllQueuedPlayers[CustomPlayer.CurrentBoss][0]?
                        QueuedPlayer = CustomPlayer.Player
                    then:
                        if(CustomPlayer.ActivePlayers >= 1):
                            CustomPlayer.JoinButton.SetText(StringToMessage("START"))
                        else:
                            CustomPlayer.JoinButton.SetText(StringToMessage("JOIN"))
                    else:
                        CustomPlayer.JoinButton.SetText(StringToMessage("JOIN"))
                    
                    CPlayer.QueuedPlayersCount.SetText(StringToMessage("{CPlayer.ActivePlayers}/4"))
                    if(FightInProgress:=BossFightStarted[CPlayer.CurrentBoss]):
                        if(FightInProgress = true):
                            CPlayer.QueueStatus.SetText(StringToMessage("Boss Fight In Progress!"))
                        else if(SomeoneQueued[CPlayer.CurrentBoss] = false):
                            CPlayer.QueueStatus.SetText(StringToMessage("JOIN to start Boss Queue..."))

    JoinQueue(Wdata:widget_message):void=
        if(CustomPlayer := gm.PlayersMap[Wdata.Player]):
            Print("JOIN QUEUE")
            if:
                QueuedPlayer := AllQueuedPlayers[CustomPlayer.CurrentBoss][0]?
                QueuedPlayer = CustomPlayer.Player
            then:
                if(set AllQueueTimers[CustomPlayer.CurrentBoss] = 1.0) {}
            else:
                for(i:int=0..CustomPlayer.QueuedPlayerBlocks.Length):
                    if(FightInProgress:=BossFightStarted[CustomPlayer.CurrentBoss], FightInProgress = false):
                        if(QueuedPlayerBlock:=CustomPlayer.QueuedPlayerBlocks[i],QueuedPlayers:=AllQueuedPlayers[CustomPlayer.CurrentBoss]):
                            PlayerBlockText:=QueuedPlayerBlock.GetText()
                            if(CustomPlayer.QueuedForBoss = false):
                                if(PlayerBlockText = "", BossKeyCost:=BossKeyCosts[CustomPlayer.CurrentBoss]):
                                    var TotalKeyCost : float = BossKeyCost
                                    if(BossHardMode[CustomPlayer.CurrentBoss] = true):
                                        set TotalKeyCost *= HardModeMult
                                    if(CustomPlayer.BossKeyAmount >= TotalKeyCost):
                                        if(set SomeoneQueued[CustomPlayer.CurrentBoss] = true) {}
                                        set CustomPlayer.QueuedForBoss = true
                                        if(set AllQueuedPlayers[CustomPlayer.CurrentBoss][i] = option{Wdata.Player}) {}
                                        QueuedPlayerBlock.SetText(PlayerToMessage(Wdata.Player))
                                        CustomPlayer.MessageText.SetTextColor(MakeColorFromHex("39FF39"))
                                        CustomPlayer.MessageText.SetText(StringToMessage("Joined Queue!"))
                                        spawn{BossQueueTimer(CustomPlayer.CurrentBoss)}
                                    else:
                                        CustomPlayer.MessageText.SetTextColor(MakeColorFromHex("fa4d56"))
                                        CustomPlayer.MessageText.SetText(StringToMessage("You do not have a boss key!"))
                                else:
                                    CustomPlayer.MessageText.SetTextColor(MakeColorFromHex("fa4d56"))
                                    CustomPlayer.MessageText.SetText(StringToMessage("Lobby full!"))
                    else:
                        CustomPlayer.MessageText.SetTextColor(MakeColorFromHex("fa4d56"))
                        CustomPlayer.MessageText.SetText(StringToMessage("Boss Fight in progress!"))

            UpdateBossHUD(CustomPlayer)

    LeaveQueue(Wdata:widget_message):void=
        var AnyoneLeft: logic = false
        if(CustomPlayer := gm.PlayersMap[Wdata.Player],QueuedPlayers:=AllQueuedPlayers[CustomPlayer.CurrentBoss], BossFightStarted[CustomPlayer.CurrentBoss] = false):
            Print("LEAVE QUEUE")
            for(i:int=0..QueuedPlayers.Length):
                if(QueuedPlayer:=QueuedPlayers[i],QueuedPlayerBlock:=CustomPlayer.QueuedPlayerBlocks[i],isPlayer:=QueuedPlayer?):
                    if(isPlayer = Wdata.Player):
                        QueuedPlayerBlock.SetText(StringToMessage(""))
                        if(set AllQueuedPlayers[CustomPlayer.CurrentBoss][i] = false) {}
                        set CustomPlayer.QueuedForBoss = false
                        CustomPlayer.MessageText.SetTextColor(MakeColorFromHex("39FF39"))
                        CustomPlayer.MessageText.SetText(StringToMessage("You have left the Queue"))

            for(i:int=0..CustomPlayer.QueuedPlayerBlocks.Length):
                if(QueuedPlayerBlock:=CustomPlayer.QueuedPlayerBlocks[i]):
                    PlayerBlockText:=QueuedPlayerBlock.GetText()
                    if(PlayerBlockText <> ""):
                        set AnyoneLeft = true

            if(AnyoneLeft = false):
                if(set SomeoneQueued[CustomPlayer.CurrentBoss] = false) {}
                
            UpdateBossHUD(CustomPlayer)

    CallForHelp(Wdata:widget_message):void=
        if(CustomPlayer := gm.PlayersMap[Wdata.Player], CustomPlayer.RequestedBossHelp = false):
            set CustomPlayer.RequestedBossHelp = true
            for(K -> CPlayer:gm.PlayersMap):
                if:
                    BossName := BossNames[CustomPlayer.CurrentBoss]
                then:
                    gm.ShowMessage(CPlayer.Player, "Someone is requesting help at {BossName}!", 0)
            CustomPlayer.MessageText.SetText(StringToMessage("Help Requested!"))
        #UpdateBossHUD(CustomPlayer)
  
    ToggleHardMode(Wdata:widget_message):void=
        if:
            CustomPlayer := gm.PlayersMap[Wdata.Player]
            IsHardMode := BossHardMode[CustomPlayer.CurrentBoss]
            CustomPlayer.ActivePlayers = 0
        then:
            if(IsHardMode = true):
                if(set BossHardMode[CustomPlayer.CurrentBoss] = false) {}
            else:
                if(set BossHardMode[CustomPlayer.CurrentBoss] = true) {}

            UpdateBossHUD(CustomPlayer)
        else:
            AudioManager.Denied.Play(Wdata.Player)

            
    #TIMER QUEUE TO JOIN A BOSS FIGHT
    BossQueueTimer(WhichBoss:int)<suspends>:void=
        if(TimersStarted[WhichBoss] = false):
            if(set AllQueueTimers[WhichBoss] = TimerDefault) {}
            if(set TimersStarted[WhichBoss] = true) {}
            loop:
                if:
                    SomeoneQueued[WhichBoss] = true
                    BossFightStarted[WhichBoss] = false
                then:
                    if(set AllQueueTimers[WhichBoss] -= 1.0) {}
                    for(K -> CPlayer:gm.PlayersMap):
                        if(CPlayer.CurrentBoss = WhichBoss,QueueTimer:=AllQueueTimers[WhichBoss]):
                            CPlayer.QueueStatus.SetText(StringToMessage("{ConvertToTime(QueueTimer, "Minutes")}"))
                    if(AllQueueTimers[WhichBoss] <= 0.0):
                        if(set AllQueueTimers[WhichBoss] = TimerDefault) {}
                        if(set TimersStarted[WhichBoss] = false) {}
                        StartBossFight(WhichBoss)
                        break:
                    Sleep(1.0)
                else:
                    if(set AllQueueTimers[WhichBoss] = TimerDefault) {}
                    if(set TimersStarted[WhichBoss] = false) {}
                    break:

    #FINAL BOSS FIGHT TIMER
    StartFBTimer(CustomPlayer:Custom_Player, WhichBoss:int)<suspends>:void=
        OpenBossTimerHUD(CustomPlayer.Player)
        if(TimersStarted[WhichBoss] = false):
            set FBTimer = FBTimerDefault
            if(set TimersStarted[WhichBoss] = true) {}
            loop:
                if:
                    BossFightStarted[WhichBoss] = true
                then:
                    set FBTimer -= 1.0
                    if(QueuedPlayers:=AllQueuedPlayers[WhichBoss]):
                        for(i:int=0..QueuedPlayers.Length):
                            if:
                                QueuedPlayer:=QueuedPlayers[i]
                                isPlayer:=QueuedPlayer?
                                CPlayer:=gm.PlayersMap[isPlayer]
                            then:
                                CPlayer.BossTimerText.SetText(StringToMessage(ConvertToTime(FBTimer*1.0,"Minutes")))
                                if(FBTimer <= 0.0):
                                    set FBTimer = 0.0
                                    Print("TIMER OVER")
                                    spawn{EndFBFight(CPlayer, false)}
                                    if(set TimersStarted[WhichBoss] = false) {}
                    if(FBTimer <= 0.0):
                        break:
                    Sleep(1.0)
                else:
                    RemoveBossTimerHUD(CustomPlayer)
                    if(set TimersStarted[WhichBoss] = false) {}
                    break:

    #DEMON FIGHT TIMER
    StartDemonTimer(CustomPlayer:Custom_Player, WhichBoss:int)<suspends>:void=
        OpenBossTimerHUD(CustomPlayer.Player)
        if(TimersStarted[WhichBoss] = false):
            set DemonTimer = 120.0 # 120.0 FBTimerDefault
            if(set TimersStarted[WhichBoss] = true) {}
            loop:
                if:
                    BossFightStarted[WhichBoss] = true
                then:
                    set DemonTimer -= 1.0
                    if(QueuedPlayers:=AllQueuedPlayers[WhichBoss]):
                        for(i:int=0..QueuedPlayers.Length):
                            if:
                                QueuedPlayer:=QueuedPlayers[i]
                                isPlayer:=QueuedPlayer?
                                CPlayer:=gm.PlayersMap[isPlayer]
                            then:
                                CPlayer.BossTimerText.SetText(StringToMessage(ConvertToTime(DemonTimer*1.0,"Minutes")))
                                if(DemonTimer <= 0.0):
                                    set DemonTimer = 0.0
                                    Print("TIMER OVER")
                                    spawn{EndDemonFight(CPlayer)}
                                    if(set TimersStarted[WhichBoss] = false) {}
                                    break:
                    if(DemonTimer <= 0.0):
                        Print("TIMER LESS THAN 0.0")
                        break:
                    Sleep(1.0)
                else:
                    Print("REMOVEBOSSTIMERHUD AND TIMERSSTARTED FALSE")
                    RemoveBossTimerHUD(CustomPlayer)
                    if(set TimersStarted[WhichBoss] = false) {}
                    break:

    #TIMER QUEUE FOR BOSS TO BE ELIMED
    BossFightTimer(CustomPlayer:Custom_Player, WhichBoss:int)<suspends>:void=
        CustomPlayer.RockBarColorTop.SetColor(MakeColorFromHex("474747"))
        CustomPlayer.RockBarColorTop2.SetColor(MakeColorFromHex("333333"))
        CustomPlayer.RockBarColorTopSmooth.SetColor(MakeColorFromHex("F5F5F5"))
        Print("GOT TO BOSS FIGHT")
        if(TimersStarted[WhichBoss] = false):
            set BossCanBeElimed = false
            if(set AllQueueTimers[WhichBoss] = 0.0) {}
            if(set TimersStarted[WhichBoss] = true) {}
            Print("SET FOR LOOP")
            loop:
                if:
                    BossFightStarted[WhichBoss] = true
                then:
                    Print("TICK TOCK ON THE CLOCK")
                    if(set AllQueueTimers[WhichBoss] += 1.0) {}
                    if(AllQueueTimers[WhichBoss] >= 20.0):
                        Print("BOSS CAN NOW BE ELIMED")
                        set BossCanBeElimed = true
                        for(K -> CPlayer:gm.PlayersMap):
                            if:
                                CPlayer.CurrentBoss = WhichBoss
                            then:
                                CPlayer.RockBarColorTop.SetColor(MakeColorFromHex("D61F28"))
                                CPlayer.RockBarColorTop2.SetColor(MakeColorFromHex("A1171E"))
                                CPlayer.RockBarColorTopSmooth.SetColor(MakeColorFromHex("EB7076"))
                        if(set AllQueueTimers[WhichBoss] = TimerDefault) {}
                        if(set TimersStarted[WhichBoss] = false) {}
                        break:
                    Sleep(1.0)
                else:
                    for(K -> CPlayer:gm.PlayersMap):
                        if:
                            CPlayer.CurrentBoss = WhichBoss
                        then:
                            CPlayer.RockBarColorTop.SetColor(MakeColorFromHex("D61F28"))
                            CPlayer.RockBarColorTop2.SetColor(MakeColorFromHex("A1171E"))
                            CPlayer.RockBarColorTopSmooth.SetColor(MakeColorFromHex("EB7076"))
                    if(set AllQueueTimers[WhichBoss] = TimerDefault) {}
                    if(set TimersStarted[WhichBoss] = false) {}
                    break:
                        
    StartBossFight(WhichBoss:int)<suspends>:void=
        if(set BossFightStarted[WhichBoss] = true) {}
        if(set SomeoneQueued[WhichBoss] = false) {}
        if(QueuedPlayers:=AllQueuedPlayers[WhichBoss]):
            for(i:int=0..QueuedPlayers.Length):
                if(QueuedPlayer:=QueuedPlayers[i],isPlayer:=QueuedPlayer?,CustomPlayer:=gm.PlayersMap[isPlayer]):
                    spawn{SendPlayerToBossFight(CustomPlayer,WhichBoss, i)}
                    set CustomPlayer.RequestedBossHelp = false
  
    EndFBFight(CustomPlayer:Custom_Player, OnElim:logic)<suspends>:void=
        if(CustomPlayer.CurrentBoss = 3):
            if(OnElim = false):
                Print("BOSS 3 END NO DEATH")
                for(i:int=0..AllQueuedPlayers[CustomPlayer.CurrentBoss].Length-1):
                    if:
                        QueuedPlayer := AllQueuedPlayers[CustomPlayer.CurrentBoss][i]?
                        CPlayer := gm.PlayersMap[QueuedPlayer]
                        BossDamage := PlayerDamages[i]
                        First:=Places[0]
                        Second:=Places[1]
                        Third:=Places[2]
                        Fourth:=Places[3]
                    then:
                        if(BossDamage > First):
                            if(OldPlace:=Places[2], set Places[3] = OldPlace) {}
                            if(OldPlace:=Places[1], set Places[2] = OldPlace) {}
                            if(OldPlace:=Places[0], set Places[1] = OldPlace) {}
                            if(set Places[0] = BossDamage) {}
                        else if(BossDamage > Second):
                            if(OldPlace:=Places[2], set Places[3] = OldPlace) {}
                            if(OldPlace:=Places[1], set Places[2] = OldPlace) {}
                            if(set Places[1] = BossDamage) {}
                        else if(BossDamage > Third):
                            if(OldPlace:=Places[2], set Places[3] = OldPlace) {}
                            if(set Places[2] = BossDamage) {}
                        else if(BossDamage > Fourth):
                            if(set Places[3] = BossDamage) {}

                OpenFBPostHUD(CustomPlayer)
                if(Boss:=EachBoss[3], Spawner:=BossSpawners[3]):
                    Boss.Damage(9e301)
                    Spawner.DespawnAll(false)
                LeaveBossRoom(CustomPlayer.Player, CustomPlayer.CurrentBoss)
                Print("END FB FIGHT -1")
                set CustomPlayer.CurrentBoss = -1
                Sleep(1.0)
            if(CustomPlayer.OneTimeAnalytics["FightPrestigeBoss"] = false):
                if(set CustomPlayer.OneTimeAnalytics["FightPrestigeBoss"] = true) {}
                if(Analytic:=AudioManager.OneTimeAnalytics[8]): #ANALYTIC
                    Analytic.Submit(CustomPlayer.Player)
            RemoveBossTimerHUD(CustomPlayer)
            GrantFBRewards(CustomPlayer)
            prestige.PrestigePlayer(CustomPlayer)  

    EndDemonFight(CustomPlayer:Custom_Player)<suspends>:void=
        Print("END DEMON FIGHT")
        if:
            Boss2 := EachBoss[6]
        then:
            Print("Got Demon BOSS")
        if:
            CustomPlayer.CurrentBoss = 6
            Boss2 := EachBoss[6]
        then:
            Print("GOT THROUGH DEMON IFS")
            Boss2.Damage(9e301)
            Boss6Phase2.DespawnAll(false)
            if(set EachBossHPs[Boss2] = 0.0){}
            HPBars.TakeDamage(CustomPlayer, 0.0)
            AudioManager.BigTextHUD.SetText(StringToMessage("BOSS FIGHT OVER!"))
            AudioManager.BigTextHUD.Show(CustomPlayer.Player)
            if(Location:=EachBossLocation[Boss2]):
                spawn{SpawnBossDefeatedVFX(Location)}
            AudioManager.BossDefeatedVictory.Play(CustomPlayer.Player)
            if(ChestReward:=AudioManager.ChestWaiting[CustomPlayer.CurrentBoss], Beacon:=AudioManager.ChestBeacons[CustomPlayer.CurrentBoss]):
                ChestReward.Play(CustomPlayer.Player)
                Beacon.AddToShowList(CustomPlayer.Player)
                CustomPlayer.RemoveRockHealthBarUI()
            if(set BossDefeated[7] = true) {}
            if(set SpawningBosses[CustomPlayer.CurrentBoss] = false) {}
            if(set SpawningBosses[7] = false) {}
            if(Spawner:=BossSpawners[6]):
                Spawner.DespawnAll(false)
            if(CustomPlayer.OneTimeAnalytics["DefeatFinalBoss"] = false):
                if(set CustomPlayer.OneTimeAnalytics["DefeatFinalBoss"] = true) {}
                if(Analytic:=AudioManager.OneTimeAnalytics[14]): #ANALYTIC
                    Analytic.Submit(CustomPlayer.Player)
            RemoveBossTimerHUD(CustomPlayer)

    SendPlayerToBossFight(CustomPlayer:Custom_Player,WhichBoss:int, WhichSlot:int)<suspends>:void=  
        if:
            BossHP:=BossMaxHPs[WhichBoss]
            BossName:=BossNames[WhichBoss]
            EnterTP:=BossEnterTPs[WhichBoss]
            BossKeyCost:=BossKeyCosts[WhichBoss]
            IsHardMode := BossHardMode[WhichBoss]
        then:
            CustomPlayer.RemoveRockHealthBarUI()
            ForceRemoveBossHUD(CustomPlayer)
            set CustomPlayer.CurrentRockHit = "" #For HP bar reset
            EnterTP.Teleport(CustomPlayer.Player)
            if(IsHardMode = true):
                CustomPlayer.Spend(BossKeyCost * HardModeMult, "BossKey", 0)
                set CustomPlayer.CanRegen = 9999999999
                set CustomPlayer.InHardBoss = true
            else:
                CustomPlayer.Spend(BossKeyCost, "BossKey", 0)

            if(WhichBoss <> 5):
                if(IsHardMode = true):
                    HPBars.Init(CustomPlayer, "{BossName}", BossHP * HardModeMult)
                else:
                    HPBars.Init(CustomPlayer, "{BossName}", BossHP)
            else:
                HPBars.Init(CustomPlayer, "{BossName}", 100.0, ?IsPerc:=true)
                if(CustomPlayer.OneTimeAnalytics["FightHalloweenBoss"] = false):
                    if(set CustomPlayer.OneTimeAnalytics["FightHalloweenBoss"] = true) {}
                    if(Analytic:=AudioManager.OneTimeAnalytics[11]): #ANALYTIC
                        Analytic.Submit(CustomPlayer.Player)
            HPBars.ShowUIForPlayer(CustomPlayer)
            set CustomPlayer.FBDamage = 0.0
            if(WhichBoss = 6):
                set DemonTotalDamage = 0.0
                if(set BossMaxHPs[7] = DemonMaxHP) {}
            if(WhichBoss <> 5):
                if(WhichBoss <> 3 and WhichBoss <> 4):
                    gm.SwitchMusic(CustomPlayer.Player, "Boss")
                else:
                    if(WhichBoss = 3):
                        set PlayerDamages = array{0.0, 0.0, 0.0, 0.0}
                        set Places = array{0.0, 0.0, 0.0, 0.0}
                        var NewPrestigeCost : float = 0.0
                        if(CustomPlayer.PrestigeCost * Pow(CustomPlayer.PrestigeCostMult, (CustomPlayer.PrestigeLevel)*1.0) > 1e300):
                            set NewPrestigeCost = 1e300
                        else:
                            set NewPrestigeCost = CustomPlayer.PrestigeCost * Pow(CustomPlayer.PrestigeCostMult, (CustomPlayer.PrestigeLevel)*1.0)
                        if(CustomPlayer.PrestigeStoneAmount < NewPrestigeCost):
                            set CustomPlayer.PrestigeStoneAmount = NewPrestigeCost + (NewPrestigeCost * 0.01)

                    gm.SwitchMusic(CustomPlayer.Player, "Space Boss")
                    set CustomPlayer.InSpace = false
                spawn{SpawnBoss(WhichBoss)}
            else:
                gm.SwitchMusic(CustomPlayer.Player, "HalloweenBoss")
                spawn{SpawnBoss(WhichBoss)}
                #StartBossFight(QueuedPlayers)

            if(BossRoomFreeze:=BossRoomFreezes[WhichBoss],EnterBossRoomSeq:=AllEnterBossRoomSeqs[WhichBoss][WhichSlot]):
                BossRoomFreeze.Enable()
                EnterBossRoomSeq.Play(CustomPlayer.Player)
                if(WhichBoss <> 3):
                    Sleep(3.5)
                else:
                    Sleep(5.0)
                BossRoomFreeze.Disable()
            if(WhichBoss = 3):
                spawn{StartFBTimer(CustomPlayer, WhichBoss)}
            if(WhichBoss = 4):
                spawn{BossFightTimer(CustomPlayer, WhichBoss)}

    LeaveBossRoom(Agent:agent, WhichBoss:int):void=
        var AnyoneLeft: logic = false
        if(QueuedPlayers:=AllQueuedPlayers[WhichBoss]):
            for(i:int=0..QueuedPlayers.Length):
                if(Player:=player[Agent], CustomPlayer:=gm.PlayersMap[Player],QueuedPlayer:=QueuedPlayers[i],isPlayer:=QueuedPlayer?,ExitTP:=BossExitTPs[WhichBoss]):
                    if(isPlayer = Player):
                        ExitTP.Teleport(Player)
                        if(set AllQueuedPlayers[WhichBoss][i] = false) {Print("REMOVED PLAYED FROM ROOM")}
                        UpdateBossHUD(CustomPlayer)
                        CustomPlayer.RockBarColorTop.SetColor(MakeColorFromHex("D61F28"))
                        CustomPlayer.RockBarColorTop2.SetColor(MakeColorFromHex("A1171E"))
                        CustomPlayer.RockBarColorTopSmooth.SetColor(MakeColorFromHex("EB7076"))
                         
                        gm.SetHP(CustomPlayer)

                        set CustomPlayer.CanRegen = 0
                        set CustomPlayer.InHardBoss = false
                        if:
                            CustomPlayer.CurrentBoss = 5
                            Laugh:=AudioManager.HalloweenBossEffects[0]
                        then:
                            Laugh.Play(CustomPlayer.Player)
                            
                        if(CurrentBoss:=EachBoss[WhichBoss], BossAgent:=CurrentBoss.GetAgent[]):
                            UpdateBossTargets(BossAgent, WhichBoss)

                        if(CustomPlayer.CurrentBoss <> 3):
                            Print("LEAVE BOSS ROOM SET TO -1")
                            set CustomPlayer.CurrentBoss = -1
                        else:
                            Print("IN SPACE FALSE")
                            set CustomPlayer.InSpace = false

                        set CustomPlayer.QueuedForBoss = false #TESTING IT HERE INSTEAD
                        gm.SwitchMusic(CustomPlayer.Player, "Default")
                        if(ChestReward:=AudioManager.ChestWaiting[WhichBoss], OpenChest:=AudioManager.ChestOpen[WhichBoss], Beacon:=AudioManager.ChestBeacons[WhichBoss]):
                            ChestReward.Stop(CustomPlayer.Player)
                            Beacon.RemoveFromShowList(CustomPlayer.Player)
                        CustomPlayer.RemoveRockHealthBarUI()

        if(QueuedPlayers:=AllQueuedPlayers[WhichBoss]):
            for(i:int=0..QueuedPlayers.Length):
                if(QueuedPlayer:=QueuedPlayers[i]):
                    if(isPlayer:=QueuedPlayer?):
                        set AnyoneLeft = true

            if(AnyoneLeft = false):
                if(set SpawningBosses[WhichBoss] = false) {}
                if(set BossFightStarted[WhichBoss] = false) {}
                if(set BossDefeated[WhichBoss] = false) {}
                Print("WhichBoss: {WhichBoss}")
                Print("DemonPhase: {DemonPhase}")
                if(WhichBoss = 6 and DemonPhase = 1):
                    Print("RESET DEMON")
                    if(set BossDefeated[7] = false) {}
                    if(set SpawningBosses[7] = false) {}
                    set DemonPhase = 0
                    if(Boss:=EachBoss[6]):
                        Boss.Damage(9e301)
                    Boss6Phase2.DespawnAll(false)
                if(Boss:=EachBoss[WhichBoss], Spawner:=BossSpawners[WhichBoss]):
                    Boss.Damage(9e301)
                    Spawner.DespawnAll(false)

    UpdateBossTargets(Agent:agent, WhichBoss:int):void=
        if(WhichBoss = 0, Fort:=Agent.GetFortCharacter[], QueuedPlayers:= AllQueuedPlayers[WhichBoss], NPCBehave:=boss1_behaviour[Agent.GetNPCBehavior[]]):
            NPCBehave.SetTargetAgents(QueuedPlayers)
            Fort.EliminatedEvent().Subscribe(NPCBehave.SetIsElim)
        if(WhichBoss = 1, Fort:=Agent.GetFortCharacter[], QueuedPlayers:= AllQueuedPlayers[WhichBoss], NPCBehave:=boss2_behaviour[Agent.GetNPCBehavior[]]):
            NPCBehave.SetTargetAgents(QueuedPlayers)
            Fort.EliminatedEvent().Subscribe(NPCBehave.SetIsElim)
        if(WhichBoss = 2, Fort:=Agent.GetFortCharacter[], QueuedPlayers:= AllQueuedPlayers[WhichBoss], NPCBehave:=boss3_behaviour[Agent.GetNPCBehavior[]]):
            NPCBehave.SetTargetAgents(QueuedPlayers)
            Fort.EliminatedEvent().Subscribe(NPCBehave.SetIsElim)
        if(WhichBoss = 3, Fort:=Agent.GetFortCharacter[], QueuedPlayers:= AllQueuedPlayers[WhichBoss], NPCBehave:=boss4_behaviour[Agent.GetNPCBehavior[]]):
            NPCBehave.SetTargetAgents(QueuedPlayers)
            Fort.EliminatedEvent().Subscribe(NPCBehave.SetIsElim)
            NPCBehave.SetAnims(Boss4BeamSingle, Boss4BeamAll, Boss4BeamRandom)
        if(WhichBoss = 4, Fort:=Agent.GetFortCharacter[], QueuedPlayers:= AllQueuedPlayers[WhichBoss], NPCBehave:=boss5_behaviour[Agent.GetNPCBehavior[]]):
            NPCBehave.SetTargetAgents(QueuedPlayers, option{AudioManager})
            Fort.EliminatedEvent().Subscribe(NPCBehave.SetIsElim)
            NPCBehave.SetAnims(Boss5BeamAll, Boss5GroundSlam)
        if(WhichBoss = 5, Fort:=Agent.GetFortCharacter[], QueuedPlayers:= AllQueuedPlayers[WhichBoss], NPCBehave:=halloween_boss_behaviour[Agent.GetNPCBehavior[]]):
            NPCBehave.SetTargetAgents(QueuedPlayers, option{AudioManager})
            Fort.EliminatedEvent().Subscribe(NPCBehave.SetIsElim)
            NPCBehave.SetAnims(HalloweenBeam, HalloweenSummon)
        if(WhichBoss = 6, DemonPhase = 0, Fort:=Agent.GetFortCharacter[], QueuedPlayers:= AllQueuedPlayers[WhichBoss], NPCBehave:=boss6_behaviour[Agent.GetNPCBehavior[]]):
            Print("BOSS 6 PHASE 1")
            NPCBehave.SetTargetAgents(QueuedPlayers, option{AudioManager})
            Fort.EliminatedEvent().Subscribe(NPCBehave.SetIsElim)
            NPCBehave.SetAnims(Boss6Tornado, Boss6Comet)
        if(WhichBoss = 6 and DemonPhase = 1, Fort:=Agent.GetFortCharacter[], QueuedPlayers:= AllQueuedPlayers[WhichBoss], NPCBehave:=boss6_behaviour2[Agent.GetNPCBehavior[]]):
            Print("BOSS 6 PHASE 2")
            NPCBehave.SetTargetAgents(QueuedPlayers, option{AudioManager})
            Fort.EliminatedEvent().Subscribe(NPCBehave.SetIsElim)
            NPCBehave.SetAnims(Boss6Tornado, Boss6Comet, Boss6MeteorShower, Boss6Beam)

    UpdateBossHardMode(Agent:agent, WhichBoss:int):void=
        var IsHard : logic = false
        if(BossHardMode[WhichBoss] = true):
            set IsHard = true

        if(WhichBoss = 0, NPCBehave:=boss1_behaviour[Agent.GetNPCBehavior[]]):
            NPCBehave.SetHardMode(IsHard)
        if(WhichBoss = 1, NPCBehave:=boss2_behaviour[Agent.GetNPCBehavior[]]):
            NPCBehave.SetHardMode(IsHard)
        if(WhichBoss = 2, NPCBehave:=boss3_behaviour[Agent.GetNPCBehavior[]]):
            NPCBehave.SetHardMode(IsHard)
        if(WhichBoss = 4, NPCBehave:=boss5_behaviour[Agent.GetNPCBehavior[]]):
            NPCBehave.SetHardMode(IsHard)
        if(WhichBoss = 6, NPCBehave:=boss6_behaviour[Agent.GetNPCBehavior[]]):
            NPCBehave.SetHardMode(IsHard)
        if(WhichBoss = 7, NPCBehave:=boss6_behaviour2[Agent.GetNPCBehavior[]]):
            NPCBehave.SetHardMode(IsHard)

    SpawnBoss(WhichBoss:int)<suspends>:void=
        if(WhichBoss = 5 ):
            set BossPhase = 1
        if(WhichBoss = 6):
            set DemonPhase = 0
        if(BossSpawner := BossSpawners[WhichBoss]):
            for(i:=0..10):
                Print("FORCE SPAWN")
                BossSpawner.Spawn()
                Sleep(0.06)

    SetupDemonPhaseTwo(Customplayer:Custom_Player, WhichPlayer:int)<suspends>:void=
        if(FC:=Customplayer.Player.GetFortCharacter[]):
            FC.SetVulnerability(false)

        if:
            BossRoomFreeze:=BossRoomFreezes[6]
            DemonPhaseSeq:=DemonBossPhaseSeq[WhichPlayer]
        then:
            BossRoomFreeze.Enable()
            DemonPhaseSeq.Play(Customplayer.Player)
            Sleep(5.0)
            BossRoomFreeze.Disable()

        spawn{StartDemonTimer(Customplayer, Customplayer.CurrentBoss)}

        if:
            BossHP:=BossMaxHPs[7]
            BossName:=BossNames[7]
            IsHardMode := BossHardMode[6]
        then:
            if(IsHardMode = true):
                HPBars.Init(Customplayer, "{BossName}", BossHP * HardModeMult)
            else:
                HPBars.Init(Customplayer, "{BossName}", BossHP)
            HPBars.ShowUIForPlayer(Customplayer)

        if(FC:=Customplayer.Player.GetFortCharacter[]):
            FC.SetVulnerability(true)

    SpawnDemonPhase(CustomPlayer:Custom_Player)<suspends>:void=
        Print("Got Boss Spawner")
        for(i:=0..AllQueuedPlayers[CustomPlayer.CurrentBoss].Length-1):
            if:
                PlayerAgent:=AllQueuedPlayers[CustomPlayer.CurrentBoss][i]?
                CPlayer := gm.PlayersMap[PlayerAgent]
            then:
                spawn{SetupDemonPhaseTwo(CPlayer, i)}

        Sleep(4.8)

        # Wait 3 seconds before spawning another Pet, if one is still "Spawning"
        if(SpawningBosses[7]?):
            race:
                block:
                    loop:
                        Sleep(0.25)
                        if (not SpawningBosses[7]?):
                            break
                block:
                    Sleep(3.0)
                    if(set SpawningBosses[7] = false) {}

        if(set SpawningBosses[7] = true) {}
        Boss6Phase2.Spawn()
        #spawn{StartDemonTimer(CustomPlayer, CustomPlayer.CurrentBoss)}

    OnBossSpawned(Agent:agent, WhichBoss:int):void=
        Print("BOSS SPAWNED: {WhichBoss}")
        if (Boss := Agent.GetFortCharacter[]):
            if:
                BossHP:=BossMaxHPs[WhichBoss]
                IsHardMode := BossHardMode[WhichBoss]
            then:
                Print("THROUGH IFS")
                if(IsHardMode = true):
                    if(set EachBossHPs[Boss] = BossHP * HardModeMult){}
                else:
                    if(set EachBossHPs[Boss] = BossHP){}

                if(set EachBoss[WhichBoss] = Boss) {}
                Boss.SetMaxHealth(2e300)
                UpdateBossTargets(Agent, WhichBoss)
                UpdateBossHardMode(Agent, WhichBoss)

                if(WhichBoss <> 5):
                    Boss.DamagedEvent().SubscribeAgent5(OnBossDamaged, Boss)
                else:
                    Print("HALLOWEEN BOSS")
                    Boss.DamagedEvent().SubscribeAgent5(OnHalloweenBossDamaged, Boss)

    OnMobSpawned(Agent:agent, whichRoom:int):void=
        if (Creep := Agent.GetFortCharacter[]):
            Creep.SetMaxHealth(1.0)
            if:
                Fort:=Agent.GetFortCharacter[]
                NPCBehave:=mob_halloween_behaviour[Agent.GetNPCBehavior[]]
                QueuedPlayers:=AllQueuedPlayers[5]
            then:
                if:
                    whichRoom = 0 or
                    whichRoom = 3 or
                    whichRoom = 5
                then:
                    NPCBehave.SetTargetAgents(QueuedPlayers, 0)
                else:
                    NPCBehave.SetTargetAgents(QueuedPlayers, 1)

                Creep.EliminatedEvent().Subscribe(NPCBehave.SetIsElim)
                Creep.EliminatedEvent().Subscribe(MobDefeated)

    MobDefeated(ElimResult:elimination_result):void=
        set HalloweenPhaseTimer -= 0.2
        if:
            Fort := ElimResult.EliminatingCharacter?
            Agent := Fort.GetAgent[]
            CustomPlayer := gm.PlayersMap[Agent]
        then:
            if (MobQuestStatus := CustomPlayer.QuestsStarted[1], MobQuestStatus?):
                set CustomPlayer.QuestMobsDefeated += 1
                DRM.UpdateCounterUI(CustomPlayer)

    SpawnAllMissiles(CustomPlayer:Custom_Player, Target:fort_character):void=
        if:
            CooldownLevel := CustomPlayer.Hut5Levels[2]
            CooldownAmount := 3.0 - (CooldownLevel * 0.3)
        then:
            set CustomPlayer.MissileCD = CooldownAmount
            spawn{gm.MissileCDTimer(CustomPlayer)}

        spawn{SpawnMissile1(CustomPlayer, Target)}
        if(CustomPlayer.ShowVFX = true):
            spawn{SpawnMissile2(CustomPlayer, Target)}
            spawn{SpawnMissile3(CustomPlayer, Target)}
        

    SpawnMissile1(CustomPlayer:Custom_Player, Target:fort_character)<suspends>:void=
        if(MissileSFX := AudioManager.MissileSFX[GetRandomInt(0,1)]):
            MissileSFX.Play(CustomPlayer.Player)
        if(Fort:=CustomPlayer.Player.GetFortCharacter[]):
            PlayerLoc:=Fort.GetTransform().Translation
            TargetLoc:=Target.GetTransform().Translation
            PlayerRot:=Fort.GetTransform().Rotation
            if(CustomPlayer.ShowVFX = true):
                Missile := SpawnProp(gm.FireMissile, transform{Translation:=vector3{X:=PlayerLoc.X,Y:=PlayerLoc.Y,Z:=PlayerLoc.Z}})
                if (Prop := Missile(0)?):
                    Prop.MoveTo(vector3{X:=PlayerLoc.X,Y:=PlayerLoc.Y,Z:=PlayerLoc.Z+300.0}, PlayerRot, 0.2)
                    Prop.MoveTo(vector3{X:=TargetLoc.X,Y:=TargetLoc.Y,Z:=TargetLoc.Z+50.0}, PlayerRot, 0.5)
                    Prop.Dispose()
            else:
                Sleep(0.7)

            if(CustomPlayer.CurrentBoss <> 5):
                VerseOnBossDamaged(CustomPlayer, Target)
            else:
                Print("HALLOWEEN BOSS")
                VerseOnHalloweenBossDamaged(CustomPlayer, Target)

            if:
                HealLevel := CustomPlayer.Hut5Levels[3]
                HealAmount := HealLevel * 2.0
                HealLevel > 0
            then:
                Fort.Heal(HealAmount)
                

    SpawnMissile2(CustomPlayer:Custom_Player, Target:fort_character)<suspends>:void=
        if(Fort:=CustomPlayer.Player.GetFortCharacter[]):
            PlayerLoc:=Fort.GetTransform().Translation
            TargetLoc:=Target.GetTransform().Translation
            PlayerRot:=Fort.GetTransform().Rotation
            Missile := SpawnProp(gm.FireMissile, transform{Translation:=vector3{X:=PlayerLoc.X,Y:=PlayerLoc.Y,Z:=PlayerLoc.Z}})
            if (Prop := Missile(0)?):
                Prop.MoveTo(vector3{X:=PlayerLoc.X,Y:=PlayerLoc.Y+300.0,Z:=PlayerLoc.Z+300.0}, PlayerRot, 0.2)
                Prop.MoveTo(vector3{X:=TargetLoc.X,Y:=TargetLoc.Y,Z:=TargetLoc.Z+50.0}, PlayerRot, 0.5)
                Prop.Dispose()

    SpawnMissile3(CustomPlayer:Custom_Player, Target:fort_character)<suspends>:void=
        if(Fort:=CustomPlayer.Player.GetFortCharacter[]):
            PlayerLoc:=Fort.GetTransform().Translation
            TargetLoc:=Target.GetTransform().Translation
            PlayerRot:=Fort.GetTransform().Rotation
            Missile := SpawnProp(gm.FireMissile, transform{Translation:=vector3{X:=PlayerLoc.X,Y:=PlayerLoc.Y,Z:=PlayerLoc.Z}})
            if (Prop := Missile(0)?):
                Prop.MoveTo(vector3{X:=PlayerLoc.X,Y:=PlayerLoc.Y-300.0,Z:=PlayerLoc.Z+300.0}, PlayerRot, 0.5)
                Prop.MoveTo(vector3{X:=TargetLoc.X,Y:=TargetLoc.Y,Z:=TargetLoc.Z+50.0}, PlayerRot, 0.5)
                Prop.Dispose()

    VerseOnBossDamaged(CustomPlayer : Custom_Player, Boss:fort_character) : void=
        var TotalDamage : float = 0.0
        var DamageResult : tuple(float, logic, float, int) = CustomPlayer.CalcTotalDamage("Mob")
        if(MissileDamage := DamageResult(0) * (0.2 + CustomPlayer.Hut5Levels[1] * 0.1)):#CHANGE TO THIS IS 
            set TotalDamage = MissileDamage
        if(set EachBossHPs[Boss] -= TotalDamage) {}
        DamageNumbers.ShowDamageUIForPlayer(CustomPlayer, TotalDamage, DamageResult(1), DamageResult(3), false, "Missile", -1.0)
        set CustomPlayer.FBDamage += TotalDamage
        
        if(CustomPlayer.CurrentBoss = 6):
            set DemonTotalDamage += TotalDamage

        if(Boss.IsActive[]):
            if(set EachBossLocation[Boss] = Boss.GetTransform().Translation) {}
        else:
            Print("BOSS NOT ACTIVE")
            if(set BossDefeated[CustomPlayer.CurrentBoss] = true) {}
            if(set SpawningBosses[CustomPlayer.CurrentBoss] = false) {}
            CustomPlayer.RemoveRockHealthBarUI()
        if(QueuedPlayers:=AllQueuedPlayers[CustomPlayer.CurrentBoss]):
            for(i:int=0..QueuedPlayers.Length):
                if(CustomPlayer.Player = QueuedPlayers[i]?):
                    if(set PlayerDamages[i] += TotalDamage) {}
                if(PlayerAgent:=QueuedPlayers[i]?,CPlayer := gm.PlayersMap[PlayerAgent]):
                    if(EachBossHPs[Boss] <= 0.0 or not Boss.IsActive[]):
                        if(CPlayer.CurrentBoss = 6 and DemonPhase = 0):
                            Print("DEMON PHASE 1 HP GONE")
                            set DemonPhase = 1
                            if(Spawner:=BossSpawners[CPlayer.CurrentBoss]):
                                Boss.Damage(9e301)
                                Spawner.DespawnAll(false)
                            spawn{SpawnDemonPhase(CPlayer)}
                        else if(CPlayer.CurrentBoss = 6 and DemonPhase = 1):
                            if:
                                set BossMaxHPs[7] *= 2.0
                                BossHP:= BossMaxHPs[7]
                                set EachBossHPs[Boss] = BossHP
                                NewBossHP := EachBossHPs[Boss]
                                BossName:=BossNames[7]
                            then:
                                Print("DEMON PHASE 2 HP GONE")
                                HPBars.Init(CustomPlayer, "{BossName}", NewBossHP)

                        else if(CPlayer.CurrentBoss <> 3):
                            if:
                                CPlayer.CurrentBoss <> 4 or
                                (CPlayer.CurrentBoss = 4 and BossCanBeElimed = true)
                            then:
                                Print("CURRENT BOSS IS NOT 4 OR IT IS AND HE CAN BE ELIMED")
                                if(Spawner:=BossSpawners[CPlayer.CurrentBoss]):
                                    Boss.Damage(9e301)
                                    Spawner.DespawnAll(false)
                                if(set EachBossHPs[Boss] = 0.0){}
                                HPBars.TakeDamage(CPlayer, 0.0)
                                AudioManager.BigTextHUD.SetText(StringToMessage("BOSS DEFEATED!"))
                                AudioManager.BigTextHUD.Show(CPlayer.Player)
                                if(Location:=EachBossLocation[Boss]):
                                    spawn{SpawnBossDefeatedVFX(Location)}
                                    AudioManager.BossDefeatedVictory.Play(CPlayer.Player)
                                if(ChestReward:=AudioManager.ChestWaiting[CPlayer.CurrentBoss], Beacon:=AudioManager.ChestBeacons[CPlayer.CurrentBoss]):
                                    ChestReward.Play(CPlayer.Player)
                                    Beacon.AddToShowList(CPlayer.Player)
                                CPlayer.RemoveRockHealthBarUI()
                                if(set BossDefeated[CPlayer.CurrentBoss] = true) {}
                                if(set SpawningBosses[CustomPlayer.CurrentBoss] = false) {}
                            else:
                                Print("CURRENT BOSS IS 4")
                                if(set EachBossHPs[Boss] = 1.0){}
                                HPBars.TakeDamage(CPlayer, 1.0)
                        else:
                            if:
                                set BossMaxHPs[3] *= 2.0
                                BossHP:= BossMaxHPs[3]
                                set EachBossHPs[Boss] = BossHP
                                NewBossHP := EachBossHPs[Boss]
                                BossName:=BossNames[CustomPlayer.CurrentBoss]
                            then:
                                HPBars.Init(CustomPlayer, "{BossName}", NewBossHP)
                                #HPBars.TakeDamage(CPlayer, NewBossHP)

                    else:
                        if(NewBossHP:=EachBossHPs[Boss]): #Need this line to make sure you're updating the new HP after damage
                            HPBars.TakeDamage(CPlayer, NewBossHP)

    OnBossDamaged(DmgResult : damage_result, Boss:fort_character) : void=
        MaybeInstigator := DmgResult.Instigator
        if (Instigator := MaybeInstigator?):
            if (InstagatingAgent := Instigator.GetInstigatorAgent[],Player:=player[InstagatingAgent]):
                if(CustomPlayer := gm.PlayersMap[Player]):
                    var DamageResult : tuple(float, logic, float, int) = CustomPlayer.CalcTotalDamage("Mob")
                    #DmgResult.Target.Damage(DamageResult(0))
                    if(set EachBossHPs[Boss] -= DamageResult(0) + DmgResult.Amount) {}
                    DamageNumbers.ShowDamageUIForPlayer(CustomPlayer, DamageResult(0), DamageResult(1), DamageResult(3), false, "Mob", -1.0)
                    set CustomPlayer.FBDamage += DamageResult(0)

                    if(CustomPlayer.CurrentBoss = 6):
                        set DemonTotalDamage += DamageResult(0)

                    if:
                        CustomPlayer.Hut5Levels[0] = 1
                        CustomPlayer.MissileCD <= 0.0
                        CustomPlayer.HutEnabled[4] = true
                    then:
                        SpawnAllMissiles(CustomPlayer, Boss)
                        
                    if(Boss.IsActive[]):
                        if(set EachBossLocation[Boss] = Boss.GetTransform().Translation) {}
                    else:
                        if(CustomPlayer.CurrentBoss <> 6 or (CustomPlayer.CurrentBoss = 6 and DemonPhase = 1)):
                            if(set BossDefeated[CustomPlayer.CurrentBoss] = true) {}
                            if(set SpawningBosses[CustomPlayer.CurrentBoss] = false) {}
                            CustomPlayer.RemoveRockHealthBarUI()
                    if(QueuedPlayers:=AllQueuedPlayers[CustomPlayer.CurrentBoss]):
                        for(i:int=0..QueuedPlayers.Length):
                            if(CustomPlayer.Player = QueuedPlayers[i]?):
                                if(set PlayerDamages[i] += DamageResult(0)) {}
                            if(PlayerAgent:=QueuedPlayers[i]?,CPlayer := gm.PlayersMap[PlayerAgent]):
                                if(EachBossHPs[Boss] <= 0.0 or not Boss.IsActive[]):
                                    if(CPlayer.CurrentBoss = 6 and DemonPhase = 0):
                                        Print("DEMON PHASE 1 HP GONE")
                                        if(Spawner:=BossSpawners[CPlayer.CurrentBoss]):
                                            Boss.Damage(9e301)
                                            Spawner.DespawnAll(false)
                                        set DemonPhase = 1
                                        spawn{SpawnDemonPhase(CPlayer)}

                                    else if(CPlayer.CurrentBoss = 6 and DemonPhase = 1):
                                        if:
                                            set BossMaxHPs[7] *= 2.0
                                            BossHP:= BossMaxHPs[7]
                                            set EachBossHPs[Boss] = BossHP
                                            NewBossHP := EachBossHPs[Boss]
                                            BossName:=BossNames[7]
                                        then:
                                            Print("DEMON PHASE 2 HP GONE")
                                            HPBars.Init(CustomPlayer, "{BossName}", NewBossHP)
            
                                    else if(CPlayer.CurrentBoss <> 3):
                                        if:
                                            CPlayer.CurrentBoss <> 4 or
                                            (CPlayer.CurrentBoss = 4 and BossCanBeElimed = true)
                                        then:
                                            if(Spawner:=BossSpawners[CPlayer.CurrentBoss]):
                                                Boss.Damage(9e301)
                                                Spawner.DespawnAll(false)
                                            if(set EachBossHPs[Boss] = 0.0){}
                                            HPBars.TakeDamage(CPlayer, 0.0)
                                            Print("Show Boss Defeated")
                                            AudioManager.BigTextHUD.SetText(StringToMessage("BOSS DEFEATED!"))
                                            AudioManager.BigTextHUD.Show(CPlayer.Player)
                                            if(Location:=EachBossLocation[Boss]):
                                                spawn{SpawnBossDefeatedVFX(Location)}
                                            AudioManager.BossDefeatedVictory.Play(CPlayer.Player)
                                            if(ChestReward:=AudioManager.ChestWaiting[CPlayer.CurrentBoss], Beacon:=AudioManager.ChestBeacons[CPlayer.CurrentBoss]):
                                                ChestReward.Play(CPlayer.Player)
                                                Beacon.AddToShowList(CPlayer.Player)
                                            CPlayer.RemoveRockHealthBarUI()
                                            if(set BossDefeated[CPlayer.CurrentBoss] = true) {}
                                            if(set SpawningBosses[CustomPlayer.CurrentBoss] = false) {}
                                        else:
                                            Print("CURRENT BOSS IS 4")
                                            if(set EachBossHPs[Boss] = 1.0){}
                                            HPBars.TakeDamage(CPlayer, 1.0)
                                    else:
                                        if:
                                            set BossMaxHPs[3] *= 2.0
                                            BossHP:= BossMaxHPs[3]
                                            set EachBossHPs[Boss] = BossHP
                                            BossName:=BossNames[CustomPlayer.CurrentBoss]
                                        then:
                                            HPBars.Init(CustomPlayer, "{BossName}", BossHP)
                                            #HPBars.TakeDamage(CPlayer, BossHP)

                                else:
                                    if(NewBossHP:=EachBossHPs[Boss]): #Need this line to make sure you're updating the new HP after damage
                                        HPBars.TakeDamage(CPlayer, NewBossHP)

    VerseOnHalloweenBossDamaged(CustomPlayer:Custom_Player, Boss:fort_character) : void=
        var TotalDamage : float = 0.0
        if:
            CanBeDamaged = true #CHANGE TO UNCOMMENTED
            QueuedPlayers:=AllQueuedPlayers[CustomPlayer.CurrentBoss]
            MissileLevel := CustomPlayer.Hut5Levels[1]
        then:
            var DamageResult : tuple(float, logic, float, int) = CustomPlayer.CalcTotalDamage("Mob")
            var HalloweenPlayerDamage : float = 100.0 * (0.1 + MissileLevel * 0.1)
            if(set EachBossHPs[Boss] -= HalloweenPlayerDamage) {}
            DamageNumbers.ShowDamageUIForPlayer(CustomPlayer, HalloweenPlayerDamage, DamageResult(1), DamageResult(3), false, "Missile", -1.0)
            for(i:int=0..QueuedPlayers.Length):
                if:
                    PlayerAgent:=QueuedPlayers[i]?
                    CPlayer := gm.PlayersMap[PlayerAgent]
                then:
                    if(EachBossHPs[Boss] <= 0.0):
                        Boss.Damage(9e301)
                        if(Spawner:=BossSpawners[CPlayer.CurrentBoss]):
                            Spawner.DespawnAll(false)
                        if(set EachBossHPs[Boss] = 0.0){}
                        HPBars.TakeDamage(CPlayer, 0.0, ?IsPerc:=true)
                        AudioManager.BigTextHUD.SetText(StringToMessage("BOSS DEFEATED!"))
                        AudioManager.BigTextHUD.Show(CPlayer.Player)
                        AudioManager.BossDefeatedVictory.Play(CPlayer.Player)
                        if(ChestReward:=AudioManager.ChestWaiting[CPlayer.CurrentBoss], Beacon:=AudioManager.ChestBeacons[CPlayer.CurrentBoss]):
                            ChestReward.Play(CPlayer.Player)
                            Beacon.AddToShowList(CPlayer.Player)
                        CPlayer.RemoveRockHealthBarUI()
                        if(set BossDefeated[CPlayer.CurrentBoss] = true) {}
                        if(set SpawningBosses[CustomPlayer.CurrentBoss] = false) {}
                        if(set CPlayer.AllBossesDefeated[CPlayer.CurrentBoss] = true) {Print("{CPlayer.CurrentBoss} BOSS DEFEATED")}
                    else:
                        if(EachBossHPs[Boss] <= (BossMaxHP / 3.0)):
                            if:
                                BossPhase <> 3
                                BossAgent := Boss.GetAgent[]
                                NPCBehave := halloween_boss_behaviour[BossAgent.GetNPCBehavior[]]
                            then:
                                set BossPhase = 3
                                NPCBehave.SetBossPhase()
                                spawn{StartMobSpawn(QueuedPlayers)}

                        else if(EachBossHPs[Boss] <= (BossMaxHP / 3.0) + (BossMaxHP / 3.0)):
                            if:
                                BossPhase <> 2
                                BossAgent := Boss.GetAgent[]
                                NPCBehave := halloween_boss_behaviour[BossAgent.GetNPCBehavior[]]
                            then:
                                set BossPhase = 2
                                NPCBehave.SetBossPhase()
                                spawn{StartMobSpawn(QueuedPlayers)}
                        else:
                            set BossPhase = 1

                        if:
                            NewBossHP:=EachBossHPs[Boss]
                            BossHPPerc := (NewBossHP / BossMaxHP) * 100.0
                        then:
                            HPBars.TakeDamage(CPlayer, BossHPPerc, ?IsPerc:=true)

    OnHalloweenBossDamaged(DmgResult : damage_result, Boss:fort_character) : void=
        if:
            MaybeInstigator := DmgResult.Instigator
            Instigator := MaybeInstigator?
            InstagatingAgent := Instigator.GetInstigatorAgent[]
            Player:=player[InstagatingAgent]
            CustomPlayer := gm.PlayersMap[Player]
            CanBeDamaged = true #CHANGE TO UNCOMMENTED
            QueuedPlayers:=AllQueuedPlayers[CustomPlayer.CurrentBoss]
        then:
            var DamageResult : tuple(float, logic, float, int) = CustomPlayer.CalcTotalDamage("Mob")
            var HalloweenPlayerDamage : float = 100.0

            if:
                CustomPlayer.Hut5Levels[0] = 1
                CustomPlayer.MissileCD <= 0.0
                CustomPlayer.HutEnabled[4] = true
            then:
                SpawnAllMissiles(CustomPlayer, Boss)

            if(set EachBossHPs[Boss] -= HalloweenPlayerDamage) {}
            DamageNumbers.ShowDamageUIForPlayer(CustomPlayer, HalloweenPlayerDamage, DamageResult(1), DamageResult(3), false, "Mob", -1.0)
            for(i:int=0..QueuedPlayers.Length):
                if:
                    PlayerAgent:=QueuedPlayers[i]?
                    CPlayer := gm.PlayersMap[PlayerAgent]
                then:
                    if(EachBossHPs[Boss] <= 0.0):
                        Print("DEFEATED")
                        Boss.Damage(9e301)
                        if(Spawner:=BossSpawners[CPlayer.CurrentBoss]):
                            Spawner.DespawnAll(false)
                        if(set EachBossHPs[Boss] = 0.0){}
                        HPBars.TakeDamage(CPlayer, 0.0, ?IsPerc:=true)
                        AudioManager.BigTextHUD.SetText(StringToMessage("BOSS DEFEATED!"))
                        AudioManager.BigTextHUD.Show(CPlayer.Player)
                        AudioManager.BossDefeatedVictory.Play(CPlayer.Player)
                        if(ChestReward:=AudioManager.ChestWaiting[CPlayer.CurrentBoss], Beacon:=AudioManager.ChestBeacons[CPlayer.CurrentBoss]):
                            ChestReward.Play(CPlayer.Player)
                            Beacon.AddToShowList(CPlayer.Player)
                        CPlayer.RemoveRockHealthBarUI()
                        if(set BossDefeated[CPlayer.CurrentBoss] = true) {}
                        if(set SpawningBosses[CustomPlayer.CurrentBoss] = false) {}
                        if(set CPlayer.AllBossesDefeated[CPlayer.CurrentBoss] = true) {Print("{CPlayer.CurrentBoss} BOSS DEFEATED")}
                    else:
                        if(EachBossHPs[Boss] <= (BossMaxHP / 3.0)):
                            Print("PHASE 3")
                            if:
                                BossPhase <> 3
                                BossAgent := Boss.GetAgent[]
                                NPCBehave := halloween_boss_behaviour[BossAgent.GetNPCBehavior[]]
                            then:
                                set BossPhase = 3
                                NPCBehave.SetBossPhase()
                                spawn{StartMobSpawn(QueuedPlayers)}

                        else if(EachBossHPs[Boss] <= (BossMaxHP / 3.0) + (BossMaxHP / 3.0)):
                            Print("PHASE 2")
                            if:
                                BossPhase <> 2
                                BossAgent := Boss.GetAgent[]
                                NPCBehave := halloween_boss_behaviour[BossAgent.GetNPCBehavior[]]
                            then:
                                set BossPhase = 2
                                NPCBehave.SetBossPhase()
                                spawn{StartMobSpawn(QueuedPlayers)}
                        else:
                            Print("PHASE 1")
                            set BossPhase = 1

                        if:
                            NewBossHP:=EachBossHPs[Boss]
                            BossHPPerc := (NewBossHP / BossMaxHP) * 100.0
                        then:
                            HPBars.TakeDamage(CPlayer, BossHPPerc, ?IsPerc:=true)
             
    StartMobSpawn(Agents:[int]?agent)<suspends>:void=
        Print("RELEASE THE MOBS")
        set HalloweenPhaseTimer = 45.0
        set CanBeDamaged = false
        for(MobSpawner:MobSpawners):
            MobSpawner.Enable()
        for(i:int=0..Agents.Length):
            if:
                PlayerAgent:=Agents[i]?
                CPlayer := gm.PlayersMap[PlayerAgent]
            then:
                CPlayer.RockBarColorTop.SetColor(MakeColorFromHex("474747"))
                CPlayer.RockBarColorTop2.SetColor(MakeColorFromHex("333333"))
                CPlayer.RockBarColorTopSmooth.SetColor(MakeColorFromHex("F5F5F5"))

        spawn{TrackMobSpawn(Agents)}
        #Sleep(45.0) #Make sure this sleep is same as behaviour


    TrackMobSpawn(Agents:[int]?agent)<suspends>:void=
        loop:
            Sleep(1.0)
            if(HalloweenPhaseTimer > 0.0):
                Print("Timer down by 1")
                set HalloweenPhaseTimer -= 1.0
            else:
                set HalloweenPhaseTimer = 0.0
                Print("TIMER DONE")
                set CanBeDamaged = true
                for(MobSpawner:MobSpawners):
                    MobSpawner.Disable()
                for(i:int=0..Agents.Length):
                    if:
                        PlayerAgent:=Agents[i]?
                        CPlayer := gm.PlayersMap[PlayerAgent]
                    then:
                        CPlayer.RockBarColorTop.SetColor(MakeColorFromHex("D61F28"))
                        CPlayer.RockBarColorTop2.SetColor(MakeColorFromHex("A1171E"))
                        CPlayer.RockBarColorTopSmooth.SetColor(MakeColorFromHex("EB7076"))
                break:
        Print("REMOVE THE MOBS")

    SpawnBossDefeatedVFX(Location:vector3)<suspends>:void=
        Rubble := SpawnProp(BossDefeatedVFX, transform{Translation:=vector3{X:=Location.X,Y:=Location.Y,Z:=Location.Z - 80.0}})
        Sleep(2.2)
        PropInst := Rubble(0)
        if ( Prop := PropInst?):
            Prop.Dispose()
        #BossDefeatedVFX:cancelable = SpawnParticleSystem(VFX.BossDefeated, vector3{X:=Location.X,Y:=Location.Y,Z:=Location.Z - 80.0})
        #BossDefeatedVFX.Cancel()
        
    OnRewardButton(Agent:agent, WhichBoss:int):void=
        if(CustomPlayer := gm.PlayersMap[Agent]):
            if:
                (BossDefeated[7] = true and CustomPlayer.CurrentBoss = 6) or
                (CustomPlayer.CurrentBoss <> 6 and BossDefeated[WhichBoss] = true and CustomPlayer.CurrentBoss = WhichBoss)
            then:
                Tutorial.UpdateTutorialUI(CustomPlayer, "Boss", 1.0)
                if(Analytic:=AudioManager.BossAnalytics[WhichBoss]): #ANALYTIC
                    Print("ANALYTIC FOR {WhichBoss} BOSS")
                    Analytic.Submit(CustomPlayer.Player)
                if(Accolade:=AudioManager.Accolades[2]): #AWARD PLAYER EXP
                    Accolade.Award(CustomPlayer.Player)
                UpdateBossHUD(CustomPlayer)
                if(ChestReward:=AudioManager.ChestWaiting[WhichBoss], OpenChest:=AudioManager.ChestOpen[WhichBoss], Beacon:=AudioManager.ChestBeacons[WhichBoss]):
                    ChestReward.Stop(CustomPlayer.Player)
                    OpenChest.Play(CustomPlayer.Player)
                    Beacon.RemoveFromShowList(CustomPlayer.Player)
                if(set CustomPlayer.AllBossesDefeated[CustomPlayer.CurrentBoss] = true) {}
                if(WhichBoss = 6 and BossHardMode[WhichBoss] = true):
                    if(set CustomPlayer.AllBossesDefeated[7] = true) {} #COMPLETED HARD MODE
                set CustomPlayer.CurrentBoss = -1
                CustomPlayer.UpAchValue(6, 1.0)
                if (BossQuestStatus := CustomPlayer.QuestsStarted[2], BossQuestStatus?):
                    set CustomPlayer.QuestBossesDefeated += 1
                    DRM.UpdateCounterUI(CustomPlayer)
                if(BossDefeated[7] = true):
                    spawn{GiveRewards(CustomPlayer,WhichBoss)}
                else if(WhichBoss <> 5 and WhichBoss <> 6):
                    spawn{GiveRewards(CustomPlayer,WhichBoss)}
                else if (WhichBoss = 5):
                    spawn{GiveHalloweenRewards(CustomPlayer)}

    SpawnOrb(LootOrb:creative_prop_asset)<suspends>:void=
        NewRotation : rotation = IdentityRotation()
        NewScale : vector3 = vector3{X:=1.0, Y:=1.0, Z:=1.0}
        LootOrbVFX := SpawnProp(LootOrb, transform{Translation := LootOrbLoc, Rotation:=NewRotation, Scale:=NewScale})
        Sleep(10.0)
        PropInst := LootOrbVFX(0)
        if(Prop := PropInst?):
            Prop.Dispose()  


    GiveRewards(CustomPlayer:Custom_Player,WhichBoss:int)<suspends>:void=
        var RewardMult : float = 1.0
        var Reward100Oc : float = 0.0
        if(BossHardMode[WhichBoss] = true):
            set RewardMult = HardModeMult
        if(WhichBoss = 6):
            if(BossHardMode[WhichBoss] = true):
                if(Analytic:=AudioManager.GeneralAnalytics[14]): #ANALYTIC
                    Analytic.Submit(CustomPlayer.Player)
                
            if(Accolade:=AudioManager.Accolades[2]): #AWARD PLAYER EXP
                Accolade.Award(CustomPlayer.Player)

            Print("DEMON BOSS REWARDS")
            var Tier : float = DemonTotalDamage / (DemonMaxHP * 0.75)
            Print("Tier: {Tier}")
            if(Tier > 4.0):
                set Tier = 4.0
            if:
                FlooredTier := Floor[Tier]
                TierRewards:=DemonBossRewards[FlooredTier]
            then:
                Print("FlooredTier: {FlooredTier}")
                if(FlooredTier = 0):
                    Print("Common ORB")
                    spawn{SpawnOrb(LootOrbCommon)}
                else if(FlooredTier = 1):
                    Print("Uncommon ORB")
                    spawn{SpawnOrb(LootOrbUncommon)}
                else if(FlooredTier = 2):
                    Print("Rare ORB")
                    spawn{SpawnOrb(LootOrbRare)}
                else if(FlooredTier = 3):
                    Print("Epic ORB")
                    spawn{SpawnOrb(LootOrbEpic)}
                else if(FlooredTier = 4):
                    Print("Legendary ORB")
                    spawn{SpawnOrb(LootOrbLegendary)}
                    RewardRatio := CustomPlayer.FBDamage / 100.0e30
                    set Reward100Oc = Log(10.0, RewardRatio + 1.0)

                    Print("FBDamage: {NumberCondensor(CustomPlayer.FBDamage)}")
                    Print("RewardRatio: {RewardRatio}")
                    Print("Reward100Oc Mult: {Reward100Oc}")

                for(i:=0..3):
                    for(Reward:TierRewards):
                        if:
                            #PlayerCut := CustomPlayer.FBDamage / DemonTotalDamage
                            PlayerMult := 1.0 + ((CustomPlayer.PUPLevels[2]*0.20) + CustomPlayer.SinkBossKeyPower/100.0 + Reward100Oc)
                            TotalReward := Floor[((Reward(1) * GetRandomFloat(0.95,1.05)) * PlayerMult * RewardMult) / 4.0] * 1.0
                        then:
                            var whichEgg : int = 0
                            if(Reward(0) = "Egg"):
                                if:
                                    whichDropTable:=gm.EggDropOdds[StageRequirements[WhichBoss]]
                                then:
                                    set whichEgg = SelectItem(whichDropTable)
                                    CustomPlayer.Collect(TotalReward, Reward(0), whichEgg)
                            else if(Reward(0) = "Stone"):
                                StoneReward := CustomPlayer.GetPlayerDamage() * TotalReward
                                Print("StoneReward: {StoneReward}")
                                CustomPlayer.Collect(StoneReward, Reward(0), 0)
                                DamageNumbers.ShowDamageUIForPlayer(CustomPlayer, StoneReward, false, 0, false, "Rock", -1.0)
                            else:
                                CustomPlayer.Collect(TotalReward, Reward(0), 0)

                            if(Reward(0) = "Stone"):
                            else if(Reward(0) = "Egg"):
                                DamageNumbers.ShowDamageUIForPlayer(CustomPlayer, TotalReward, false, 0, false, "Egg{whichEgg}", -1.0)
                            else:
                                DamageNumbers.ShowDamageUIForPlayer(CustomPlayer, TotalReward, false, 0, false, Reward(0), -1.0)
                        Sleep(0.3)

            if(BossHardMode[WhichBoss] = true):
                Print("ROLL FOR PET")
                if(GetRandomInt(0,100) < 1): #CHANGE TO 1
                    Print("PET OBTAINED")
                    
                    if(CustomPlayer.OneTimeAnalytics["ObtainBossPet"] = false):
                        if(set CustomPlayer.OneTimeAnalytics["ObtainBossPet"] = true) {}
                        if(Analytic:=AudioManager.OneTimeAnalytics[15]): #ANALYTIC
                            Analytic.Submit(CustomPlayer.Player)
                        if(Accolade:=AudioManager.Accolades[18]): #AWARD PLAYER EXP
                            Accolade.Award(CustomPlayer.Player)

                    set CustomPlayer.DemonPetObtained = true
                    gm.ShowMessage(CustomPlayer.Player, "You got a funny feeling!", 0)
                    Sleep(1.5)
        else:
            for(i:int=0..3): #LOOP 4 TIMES TO BREAK UP REWARDS TO FEEL BETTER
                if(BossRewards:=AllBossRewards[WhichBoss]):
                    if(StoneReward:=BossRewards[0] * (1.0 + ((CustomPlayer.PUPLevels[2]*0.20) + CustomPlayer.SinkBossKeyPower/100.0)) * RewardMult): #GIVE STONE
                        TotalReward := ((CustomPlayer.GetPlayerDamage()*StoneReward) / 4.0)
                        CustomPlayer.Collect(TotalReward, "Stone", 0)
                        DamageNumbers.ShowDamageUIForPlayer(CustomPlayer, TotalReward, false, 0, false, "Rock", -1.0)
                        #gm.StoneGranter.GrantItem(CustomPlayer.Player)
                        Sleep(0.3)

                    if(WepMatReward:=BossRewards[1] * (1.0 + ((CustomPlayer.PUPLevels[2]*0.20) + CustomPlayer.SinkBossKeyPower/100.0)) * RewardMult): #GIVE WEP MATS
                        TotalReward := (WepMatReward / 4.0)
                        CustomPlayer.Collect(TotalReward, "WepMat", 0)
                        DamageNumbers.ShowDamageUIForPlayer(CustomPlayer, TotalReward, false, 0, false, "WepMat", -1.0)
                        #gm.WepMatGranter.GrantItem(CustomPlayer.Player)
                        Sleep(0.3)

                    if: #GIVE EGGS
                        EggMinReward:=BossRewards[2] * (1.0 + ((CustomPlayer.PUPLevels[2]*0.20) + CustomPlayer.SinkBossKeyPower/100.0)) * RewardMult
                        EggMaxReward:=BossRewards[3] * (1.0 + ((CustomPlayer.PUPLevels[2]*0.20) + CustomPlayer.SinkBossKeyPower/100.0)) * RewardMult
                    then:
                        var whichEgg : int = 0
                        if(whichDropTable:=gm.EggDropOdds[StageRequirements[WhichBoss]],DropAmount := Floor[GetRandomFloat(EggMinReward, EggMaxReward)]*1.0, EggGranter:=gm.EggGranters[whichEgg]):
                            set whichEgg = SelectItem(whichDropTable)
                            TotalReward := (DropAmount / 4.0)
                            CustomPlayer.Collect(TotalReward, "Egg", whichEgg)
                            DamageNumbers.ShowDamageUIForPlayer(CustomPlayer, TotalReward, false, 0, false, "Egg{whichEgg}", -1.0)
                            #EggGranter.GrantItem(CustomPlayer.Player)
                            Sleep(0.3)
        SaveAllData(CustomPlayer)
        LeaveBossRoom(CustomPlayer.Player, WhichBoss)

    GiveHalloweenRewards(CustomPlayer:Custom_Player)<suspends>:void=
        if (BossQuestStatus := CustomPlayer.QuestsStarted[2], BossQuestStatus?):
            set CustomPlayer.QuestBossesDefeated += 1
            DRM.UpdateCounterUI(CustomPlayer)
            
        var TotalMult : float = 1.0
        if(QueuedPlayers:=AllQueuedPlayers[5]):
            if(PrestigeUpgradeLvl:=CustomPlayer.PUPLevels[2]):
                set TotalMult += PrestigeUpgradeLvl * 0.20
            set TotalMult += QueuedPlayers.Length * 0.25
            set TotalMult += (CustomPlayer.StagesUnlocked*1.0) / 2.0
            set TotalMult += CustomPlayer.SinkBossKeyPower/100.0

        for(i:int=0..3): #LOOP 4 TIMES TO BREAK UP REWARDS TO FEEL BETTER
            if(StoneReward:=HalloweenBossRewards[0] * TotalMult): #GIVE STONE
                TotalReward := ((CustomPlayer.GetPlayerDamage()*StoneReward) / 4.0)
                CustomPlayer.Collect(TotalReward, "Stone", 0)
                DamageNumbers.ShowDamageUIForPlayer(CustomPlayer, TotalReward, false, 0, false, "Rock", -1.0)
                Sleep(0.3)

            if(WepMatReward:=HalloweenBossRewards[1] * TotalMult): #GIVE WEP MATS
                TotalReward := (WepMatReward / 4.0)
                CustomPlayer.Collect(TotalReward, "WepMat", 0)
                DamageNumbers.ShowDamageUIForPlayer(CustomPlayer, TotalReward, false, 0, false, "WepMat", -1.0)
                Sleep(0.3)

            if(EggMinReward:=HalloweenBossRewards[2] * TotalMult, EggMaxReward:=HalloweenBossRewards[3] * TotalMult): #GIVE EGGS
                var whichEgg : int = 0
                if(whichDropTable:=gm.EggDropOdds[CustomPlayer.StagesUnlocked],DropAmount := Floor[GetRandomFloat(EggMinReward, EggMaxReward)]*1.0, EggGranter:=gm.EggGranters[whichEgg]):
                    set whichEgg = SelectItem(whichDropTable)
                    TotalReward := (DropAmount / 4.0)
                    CustomPlayer.Collect(TotalReward, "Egg", whichEgg)
                    DamageNumbers.ShowDamageUIForPlayer(CustomPlayer, TotalReward, false, 0, false, "Egg{whichEgg}", -1.0)

                    Sleep(0.3)

            if(CustomPlayer.StagesUnlocked >= 15):
                if(ShardReward:=HalloweenBossRewards[1] * TotalMult): #GIVE SHARDS
                    TotalReward := (ShardReward / 4.0)
                    CustomPlayer.Collect(TotalReward, "Shard", 0)
                    DamageNumbers.ShowDamageUIForPlayer(CustomPlayer, TotalReward, false, 0, false, "Shard", -1.0)
                    Sleep(0.3)

        SaveAllData(CustomPlayer)
        LeaveBossRoom(CustomPlayer.Player, 5)

    GrantFBRewards(CustomPlayer:Custom_Player):void=
        
        if (BossQuestStatus := CustomPlayer.QuestsStarted[2], BossQuestStatus?):
            set CustomPlayer.QuestBossesDefeated += 1
            DRM.UpdateCounterUI(CustomPlayer)

        var RewardRatio : float = CustomPlayer.FBDamage / TheoreticalMaxDmg
        var TotalReward : float = (Log(100.0, RewardRatio) + 1.0) * (1.0 + CustomPlayer.SinkBossKeyPower/100.0)

        if(TotalReward < 0.02):
            set TotalReward = 0.02

        if(Analytic:=AudioManager.BossAnalytics[3]): #ANALYTIC
            Analytic.Submit(CustomPlayer.Player)
        #var MaxFBRewards : []float = array{1.68e90, 20000.0, 20000.0, 200000.0, 5000.0}

        if(StoneReward:=MaxFBRewards[0] * TotalReward): #GIVE STONE
            CustomPlayer.Collect(StoneReward, "Stone", 0)
            Print("StoneReward: {StoneReward}")

        if(WepMatReward:=MaxFBRewards[1] * TotalReward): #GIVE WEP MATS
            CustomPlayer.Collect(WepMatReward, "WepMat", 0)
            Print("WepMatReward: {WepMatReward}")

        if(ShardReward:=MaxFBRewards[2] * TotalReward): #GIVE SHARDS
            CustomPlayer.Collect(ShardReward, "Shard", 0)
            Print("ShardReward: {ShardReward}")

        for(i:int=0..3):
            if(EggReward:=MaxFBRewards[3] * TotalReward): #GIVE EGGS
                var whichEgg : int = 0
                if(whichDropTable:=gm.EggDropOdds[19]):
                    set whichEgg = SelectItem(whichDropTable)
                    CustomPlayer.Collect(EggReward / 4.0, "Egg", whichEgg)

        if(GatherRateReward:=MaxFBRewards[4] * TotalReward): #GIVE GATHER RATE
            set CustomPlayer.HarvestBuff += GatherRateReward/100.0
            set CustomPlayer.PrestigeHarvestBuff += GatherRateReward

        SaveAllData(CustomPlayer)
        
    OnPlayerEliminated(ElimResult:elimination_result,CustomPlayer:Custom_Player):void=
        AudioManager.PlayerElimed.Play(CustomPlayer.Player)
        LeaveBossRoom(CustomPlayer.Player, CustomPlayer.CurrentBoss)
        
#UI BLOCKS
    
#CANVAS
    MakeBossTimerCanvas(CustomPlayer:Custom_Player):canvas=
        CustomPlayer.BossTimerText.SetText(StringToMessage(""))
        CustomPlayer.BossTimerText.SetShadowColor(CustomPlayer.ShadowDefaults.DefaultShadowColor)
        CustomPlayer.BossTimerText.SetShadowOffset(CustomPlayer.ShadowDefaults.DefaultShadowOffset)
        CustomPlayer.BossTimerText.SetShadowOpacity(1.0)

        var MyBossTimerCanvas: canvas = canvas:
            Slots := array:
                canvas_slot:
                    Widget:=CustomPlayer.BossTimerHUDBG
                    ZOrder:=0
                    SizeToContent:=false
                    Anchors := anchors{Minimum := vector2{X:= 1.0, Y:= 0.0}, Maximum := vector2{X:= 1.0, Y:= 0.0}}
                    Alignment := vector2{X:= 0.5, Y:= 0.0}
                    Offsets:=margin{Top:=304.0,Left:=-100.0,Bottom:=92.0,Right:=192.0}
                canvas_slot:
                    Widget:=CustomPlayer.BossTimerText
                    ZOrder:=1
                    SizeToContent:=true
                    Anchors := anchors{Minimum := vector2{X:= 1.0, Y:= 0.0}, Maximum := vector2{X:= 1.0, Y:= 0.0}}
                    Offsets:=margin{Top:=332.0,Left:=-80.0}
                    Alignment := vector2{X:= 0.5, Y:= 0.0}

        return MyBossTimerCanvas

    OpenBossTimerHUD(Agent:agent): void=
        if(CustomPlayer := gm.PlayersMap[Agent]):
            if(MyUI:=CustomPlayer.UIMaps["BossTimer"]?):
                CustomPlayer.AddUI(MyUI)
            else:
                var MyBossTimerCanvas : canvas = MakeBossTimerCanvas(CustomPlayer)
                CustomPlayer.AddUI(MyBossTimerCanvas)
                if(set CustomPlayer.UIMaps["BossTimer"] = option{MyBossTimerCanvas}) {}

    RemoveBossTimerHUD(CustomPlayer:Custom_Player):void=
            if(PlayerUI:=GetPlayerUI[CustomPlayer.Player]):
                if(MyUI:=CustomPlayer.UIMaps["BossTimer"]?):
                    PlayerUI.RemoveWidget(MyUI)
                    MyUI.SetVisibility(widget_visibility.Collapsed)
                    #if(set CustomPlayer.UIMaps["BossTimer"] = false){}

    ShowBossTimer(CustomPlayer:Custom_Player)<suspends>:void=
        OpenBossTimerHUD(CustomPlayer.Player)
        loop:
            if:
                CustomPlayer.QueuedForBoss = true and BossFightStarted[CustomPlayer.CurrentBoss] = false
                QueueTime:=AllQueueTimers[CustomPlayer.CurrentBoss]
            then:
                CustomPlayer.BossTimerText.SetText(StringToMessage(ConvertToTime(QueueTime*1.0,"Minutes")))
            else:
                RemoveBossTimerHUD(CustomPlayer)
                break:
            Sleep(1.0)

    OpenFBPostHUD(CustomPlayer:Custom_Player): void=
        Print("OPEN FB HUD")
        if(MyUI:=CustomPlayer.UIMaps["PostFB"]?):
        else:
            #MAKE YOU INVINCIBLE WHILE HUD IS OPEN
            if(FC:=CustomPlayer.Player.GetFortCharacter[]):
                FC.SetVulnerability(false)
            var MyPostFBCanvas : canvas = MakeFBPostFight(CustomPlayer)
            CustomPlayer.AddFullHUD(MyPostFBCanvas)
            if(set CustomPlayer.UIMaps["PostFB"] = option{MyPostFBCanvas}) {}

    RemoveFBPostHUD(Wdata:widget_message):void=
        if(CustomPlayer := gm.PlayersMap[Wdata.Player]):
            if(PlayerUI:=GetPlayerUI[CustomPlayer.Player]):
                if(MyUI:=CustomPlayer.UIMaps["PostFB"]?):
                    if(FC:=CustomPlayer.Player.GetFortCharacter[]):
                        FC.SetVulnerability(true)
                    PlayerUI.RemoveWidget(MyUI)
                    MyUI.SetVisibility(widget_visibility.Collapsed)
                    if(set CustomPlayer.UIMaps["PostFB"] = false){}

    var PlayerDamages : []float = array{0.0, 0.0, 0.0, 0.0}
    var Places : []float = array{0.0, 0.0, 0.0, 0.0}

    MakeFBPostFight(CustomPlayer:Custom_Player):canvas=

        CustomPlayer.PostFBDamageTitle.SetText(StringToMessage("Player Damage"))
        CustomPlayer.PostFBBossRewardTitle.SetText(StringToMessage("Boss Rewards"))
        CustomPlayer.PostFBPrestigeTitle.SetText(StringToMessage("Prestige Rewards"))
        CustomPlayer.PostFBCloseButton.SetText(StringToMessage("X"))

        #SET PLAYER NEW HIGHEST IF APPLICABLE
        if(CustomPlayer.FBDamage > CustomPlayer.HighestFBDamage):
            set CustomPlayer.HighestFBDamage = CustomPlayer.FBDamage

        #ALL PLAYER DAMAGES
        for(x:int=0..AllQueuedPlayers[CustomPlayer.CurrentBoss].Length - 1):
            if:
                QueuedPlayer := AllQueuedPlayers[CustomPlayer.CurrentBoss][x]?
                CPlayer := gm.PlayersMap[QueuedPlayer]
                FBDamage := CPlayer.FBDamage
            then:
                for(i:int=0..Places.Length - 1):
                    if(FBDamage = Places[i]):
                        if:
                            PlayerName:=CPlayer.AllPostFBPlayerNames[i]
                            PlayerDamage:=CPlayer.AllPostFBPlayerDamages[i]
                        then:
                            PlayerName.SetText(PlayerToMessage(CPlayer.Player))
                            PlayerDamage.SetText(StringToMessage("{NumberCondensor(FBDamage)}"))
                
            CustomPlayer.PostFBBossDamageDealt.SetText(StringToMessage("Your Damage: {NumberCondensor(CustomPlayer.FBDamage)}"))

        #PLAYERS REWARDS
        RewardRatio := CustomPlayer.FBDamage / TheoreticalMaxDmg 
        for(i:int=0..CustomPlayer.AllBossRewardBlocks.Length-1):
            if:
                TextBlock:=CustomPlayer.AllBossRewardBlocks[i]
                MaxReward:= MaxFBRewards[i]
            then:
                var TotalReward : float = Log(100.0, RewardRatio) + 1.0 * (1.0 + CustomPlayer.SinkBossKeyPower/100.0)
                if(TotalReward < 0.02):
                    set TotalReward = 0.02
                if(i = 4):
                    TextBlock.SetText(StringToMessage(" +{NumberCondensor(MaxReward * TotalReward)}%"))
                else:
                    TextBlock.SetText(StringToMessage(" +{NumberCondensor(MaxReward * TotalReward)}"))

        #PRESTIGE REWARDS
        CustomPlayer.PostFBPrestigeLevel.SetText(StringToMessage("Prestige Level: {CustomPlayer.PrestigeLevel}"))

        var NewPrestigeCost : float = 0.0
        if(CustomPlayer.PrestigeCost * Pow(CustomPlayer.PrestigeCostMult, (CustomPlayer.PrestigeLevel)*1.0) > 1e300):
            set NewPrestigeCost = 1e300
        else:
            set NewPrestigeCost = CustomPlayer.PrestigeCost * Pow(CustomPlayer.PrestigeCostMult, (CustomPlayer.PrestigeLevel)*1.0)
        PrestigeRatio := CustomPlayer.PrestigeStoneAmount / NewPrestigeCost
        var PrestigeShardAmount : float = 0.0
        var PrestigeGatherAmount : float = 0.0
        if(CustomPlayer.PrestigeLevel <= 220):
            if(CustomPlayer.PrestigeLevel >= 1):
                set PrestigeShardAmount = ((Log(5.0, PrestigeRatio)) * 5) * (1.5 * CustomPlayer.PrestigeLevel)
            else:
                set PrestigeShardAmount = ((Log(5.0, PrestigeRatio)) * 5)

            if(CustomPlayer.StagesUnlocked >= 15):
                set PrestigeGatherAmount = (PrestigeShardAmount/5.0) * 3.0
            else:
                set PrestigeGatherAmount = (PrestigeShardAmount/5.0)

            if(PrestigeShardAmount < 1.0):
                set PrestigeShardAmount = 1.0
            if(PrestigeGatherAmount < 1.0):
                set PrestigeGatherAmount = 1.0

            set PrestigeShardAmount = PrestigeShardAmount * (1.0 + CustomPlayer.SinkPrestigePower / 10000.0)
            set PrestigeGatherAmount = PrestigeGatherAmount * (1.0 + CustomPlayer.SinkPrestigePower / 10000.0)
        else:
            set PrestigeShardAmount = CustomPlayer.EndGamePrestigeShardAmount * (1.0 + CustomPlayer.SinkPrestigePower / 10000.0)
            set PrestigeGatherAmount = CustomPlayer.EndGamePrestigeGatherAmount * (1.0 + CustomPlayer.SinkPrestigePower / 10000.0)

        CustomPlayer.PostFBPrestigeShards.SetText(StringToMessage(" +{NumberCondensor(PrestigeShardAmount)}"))
        CustomPlayer.PostFBPrestigeGather.SetText(StringToMessage(" +{NumberCondensor(PrestigeGatherAmount)}% Strength"))

        #CAVAS STUFF
        for(TextBlock : CustomPlayer.AllPostFBTextBlocks):
            TextBlock.SetShadowColor(CustomPlayer.ShadowDefaults.DefaultShadowColor)
            TextBlock.SetShadowOffset(CustomPlayer.ShadowDefaults.DefaultShadowOffset)
            TextBlock.SetShadowOpacity(1.0)

        var MyPostFBCanvas: canvas = canvas:
            Slots := array:
                canvas_slot:
                    Widget:=CustomPlayer.PostFBHUDBG
                    ZOrder:=0
                    SizeToContent:=false
                    Alignment := vector2{X := 0.5, Y := 0.0}
                    Anchors := anchors{Minimum := vector2{X:= 0.5, Y:= 0.0}, Maximum := vector2{X:= 0.5, Y:= 0.0}}
                    Offsets:=margin{Top:=20.0,Left:=0.0,Bottom:=1024.0,Right:=2048.0}
                canvas_slot:
                    Widget:=CustomPlayer.PostFBHUDTitle
                    ZOrder:=1
                    SizeToContent:=false
                    Anchors := anchors{Minimum := vector2{X:= 0.5, Y:= 0.0}, Maximum := vector2{X:= 0.5, Y:= 0.0}}
                    Offsets:=margin{Top:=150.0,Left:=0.0, Bottom:=56.0, Right:=677.0}
                    Alignment := vector2{X:= 0.5, Y:= 0.0}
                canvas_slot:
                    Widget:=CustomPlayer.PostFBCloseButton
                    ZOrder:=9
                    SizeToContent:=false
                    Anchors := anchors{Minimum := vector2{X:= 0.5, Y:= 0.0}, Maximum := vector2{X:= 0.5, Y:= 0.0}}
                    Offsets:=margin{Top:=152.0,Left:=640.0, Bottom:=56.0, Right:=76.0}
                    Alignment := vector2{X:= 0.5, Y:= 0.0}
                canvas_slot:
                    Widget:=CustomPlayer.PostFBDamageTitle
                    ZOrder:=1
                    SizeToContent:=true
                    Anchors := anchors{Minimum := vector2{X:= 0.5, Y:= 0.0}, Maximum := vector2{X:= 0.5, Y:= 0.0}}
                    Offsets:=margin{Top:=316.0,Left:=-576.0}
                    Alignment := vector2{X:= 0.5, Y:= 0.0}
                canvas_slot:
                    Widget:=CustomPlayer.PostFBName1
                    ZOrder:=1
                    SizeToContent:=true
                    Anchors := anchors{Minimum := vector2{X:= 0.5, Y:= 0.0}, Maximum := vector2{X:= 0.5, Y:= 0.0}}
                    Offsets:=margin{Top:=396.0,Left:=-576.0}
                    Alignment := vector2{X:= 0.5, Y:= 0.0}
                canvas_slot:
                    Widget:=CustomPlayer.PostFBDamage1
                    ZOrder:=1
                    SizeToContent:=true
                    Anchors := anchors{Minimum := vector2{X:= 0.5, Y:= 0.0}, Maximum := vector2{X:= 0.5, Y:= 0.0}}
                    Offsets:=margin{Top:=456.0,Left:=-576.0}
                    Alignment := vector2{X:= 0.5, Y:= 0.0}
                canvas_slot:
                    Widget:=CustomPlayer.PostFBName2
                    ZOrder:=1
                    SizeToContent:=true
                    Anchors := anchors{Minimum := vector2{X:= 0.5, Y:= 0.0}, Maximum := vector2{X:= 0.5, Y:= 0.0}}
                    Offsets:=margin{Top:=544.0,Left:=-576.0}
                    Alignment := vector2{X:= 0.5, Y:= 0.0}
                canvas_slot:
                    Widget:=CustomPlayer.PostFBDamage2
                    ZOrder:=1
                    SizeToContent:=true
                    Anchors := anchors{Minimum := vector2{X:= 0.5, Y:= 0.0}, Maximum := vector2{X:= 0.5, Y:= 0.0}}
                    Offsets:=margin{Top:=604.0,Left:=-576.0}
                    Alignment := vector2{X:= 0.5, Y:= 0.0}
                canvas_slot:
                    Widget:=CustomPlayer.PostFBName3
                    ZOrder:=1
                    SizeToContent:=true
                    Anchors := anchors{Minimum := vector2{X:= 0.5, Y:= 0.0}, Maximum := vector2{X:= 0.5, Y:= 0.0}}
                    Offsets:=margin{Top:=684.0,Left:=-576.0}
                    Alignment := vector2{X:= 0.5, Y:= 0.0}
                canvas_slot:
                    Widget:=CustomPlayer.PostFBDamage3
                    ZOrder:=1
                    SizeToContent:=true
                    Anchors := anchors{Minimum := vector2{X:= 0.5, Y:= 0.0}, Maximum := vector2{X:= 0.5, Y:= 0.0}}
                    Offsets:=margin{Top:=744.0, Left:=-576.0}
                    Alignment := vector2{X:= 0.5, Y:= 0.0}
                canvas_slot:
                    Widget:=CustomPlayer.PostFBName4
                    ZOrder:=1
                    SizeToContent:=true
                    Anchors := anchors{Minimum := vector2{X:= 0.5, Y:= 0.0}, Maximum := vector2{X:= 0.5, Y:= 0.0}}
                    Offsets:=margin{Top:=824.0,Left:=-576.0}
                    Alignment := vector2{X:= 0.5, Y:= 0.0}
                canvas_slot:
                    Widget:=CustomPlayer.PostFBDamage4
                    ZOrder:=1
                    SizeToContent:=true
                    Anchors := anchors{Minimum := vector2{X:= 0.5, Y:= 0.0}, Maximum := vector2{X:= 0.5, Y:= 0.0}}
                    Offsets:=margin{Top:=884.0,Left:=-576.0}
                    Alignment := vector2{X:= 0.5, Y:= 0.0}

                    
                canvas_slot:
                    Widget:=CustomPlayer.PostFBBossRewardTitle
                    ZOrder:=1
                    SizeToContent:=true
                    Anchors := anchors{Minimum := vector2{X:= 0.5, Y:= 0.0}, Maximum := vector2{X:= 0.5, Y:= 0.0}}
                    Offsets:=margin{Top:=316.0,Left:=0.0}
                    Alignment := vector2{X:= 0.5, Y:= 0.0}
                canvas_slot:
                    Widget:=CustomPlayer.PostFBBossDamageDealt
                    ZOrder:=1
                    SizeToContent:=true
                    Anchors := anchors{Minimum := vector2{X:= 0.5, Y:= 0.0}, Maximum := vector2{X:= 0.5, Y:= 0.0}}
                    Offsets:=margin{Top:=396.0,Left:=0.0}
                    Alignment := vector2{X:= 0.5, Y:= 0.0}
                canvas_slot:
                    ZOrder:=1
                    SizeToContent:=true
                    Anchors := anchors{Minimum := vector2{X:= 0.5, Y:= 0.0}, Maximum := vector2{X:= 0.5, Y:= 0.0}}
                    Offsets:=margin{Top:=544.0,Left:=0.0}
                    Alignment := vector2{X:= 0.5, Y:= 0.0}
                    Widget := stack_box:
                        Orientation := orientation.Horizontal
                        Slots := array:
                            stack_box_slot:
                                Widget := texture_block:
                                    DefaultImage := UI.StoneIcon
                            stack_box_slot:
                                Widget := CustomPlayer.PostFBBossStoneReward
                canvas_slot:
                    ZOrder:=1
                    SizeToContent:=true
                    Anchors := anchors{Minimum := vector2{X:= 0.5, Y:= 0.0}, Maximum := vector2{X:= 0.5, Y:= 0.0}}
                    Offsets:=margin{Top:=604.0,Left:=0.0}
                    Alignment := vector2{X:= 0.5, Y:= 0.0}
                    Widget := stack_box:
                        Orientation := orientation.Horizontal
                        Slots := array:
                            stack_box_slot:
                                Widget := texture_block:
                                    DefaultImage := UI.WepMat
                            stack_box_slot:
                                Widget := CustomPlayer.PostFBBossWepMatReward
                canvas_slot:
                    ZOrder:=1
                    SizeToContent:=true
                    Anchors := anchors{Minimum := vector2{X:= 0.5, Y:= 0.0}, Maximum := vector2{X:= 0.5, Y:= 0.0}}
                    Offsets:=margin{Top:=664.0,Left:=0.0}
                    Alignment := vector2{X:= 0.5, Y:= 0.0}
                    Widget := stack_box:
                        Orientation := orientation.Horizontal
                        Slots := array:
                            stack_box_slot:
                                Widget := texture_block:
                                    DefaultImage := UI.ShardIcon
                            stack_box_slot:
                                Widget := CustomPlayer.PostFBBossShardReward
                canvas_slot:
                    ZOrder:=1
                    SizeToContent:=true
                    Anchors := anchors{Minimum := vector2{X:= 0.5, Y:= 0.0}, Maximum := vector2{X:= 0.5, Y:= 0.0}}
                    Offsets:=margin{Top:=724.0,Left:=0.0}
                    Alignment := vector2{X:= 0.5, Y:= 0.0}
                    Widget := stack_box:
                        Orientation := orientation.Horizontal
                        Slots := array:
                            stack_box_slot:
                                Widget := texture_block:
                                    DefaultImage := UI.Egg1
                            stack_box_slot:
                                Widget := CustomPlayer.PostFBBossEggReward
                canvas_slot:
                    ZOrder:=1
                    SizeToContent:=true
                    Anchors := anchors{Minimum := vector2{X:= 0.5, Y:= 0.0}, Maximum := vector2{X:= 0.5, Y:= 0.0}}
                    Offsets:=margin{Top:=784.0,Left:=0.0}
                    Alignment := vector2{X:= 0.5, Y:= 0.0}
                    Widget := stack_box:
                        Orientation := orientation.Horizontal
                        Slots := array:
                            stack_box_slot:
                                Widget := texture_block:
                                    DefaultImage := UI.AchRocksDestroyed
                            stack_box_slot:
                                Widget := CustomPlayer.PostFBBossGatherReward

                canvas_slot:
                    Widget:=CustomPlayer.PostFBPrestigeTitle
                    ZOrder:=1
                    SizeToContent:=true
                    Anchors := anchors{Minimum := vector2{X:= 0.5, Y:= 0.0}, Maximum := vector2{X:= 0.5, Y:= 0.0}}
                    Offsets:=margin{Top:=412.0,Left:=580.0}
                    Alignment := vector2{X:= 0.5, Y:= 0.0}
                canvas_slot:
                    Widget:=CustomPlayer.PostFBPrestigeLevel
                    ZOrder:=1
                    SizeToContent:=true
                    Anchors := anchors{Minimum := vector2{X:= 0.5, Y:= 0.0}, Maximum := vector2{X:= 0.5, Y:= 0.0}}
                    Offsets:=margin{Top:=484.0,Left:=580.0}
                    Alignment := vector2{X:= 0.5, Y:= 0.0}
                canvas_slot:
                    ZOrder:=1
                    SizeToContent:=true
                    Anchors := anchors{Minimum := vector2{X:= 0.5, Y:= 0.0}, Maximum := vector2{X:= 0.5, Y:= 0.0}}
                    Offsets:=margin{Top:=556.0,Left:=580.0}
                    Alignment := vector2{X:= 0.5, Y:= 0.0}
                    Widget := stack_box:
                        Orientation := orientation.Horizontal
                        Slots := array:
                            stack_box_slot:
                                Widget := texture_block:
                                    DefaultImage := UI.ShardIcon
                            stack_box_slot:
                                Widget := CustomPlayer.PostFBPrestigeShards
                canvas_slot:
                    ZOrder:=1
                    SizeToContent:=true
                    Anchors := anchors{Minimum := vector2{X:= 0.5, Y:= 0.0}, Maximum := vector2{X:= 0.5, Y:= 0.0}}
                    Offsets:=margin{Top:=616.0,Left:=580.0}
                    Alignment := vector2{X:= 0.5, Y:= 0.0}
                    Widget := stack_box:
                        Orientation := orientation.Horizontal
                        Slots := array:
                            stack_box_slot:
                                Widget := texture_block:
                                    DefaultImage := UI.AchRocksDestroyed
                            stack_box_slot:
                                Widget := CustomPlayer.PostFBPrestigeGather
        return MyPostFBCanvas

        #CANVAS 
    MakeBossCanvas(CustomPlayer:Custom_Player):canvas=
        CustomPlayer.BossCloseButton.SetText(StringToMessage("X"))
        CustomPlayer.RewardTitle.SetText(StringToMessage("Rewards"))
        CustomPlayer.QueuedPlayersTitle.SetText(StringToMessage("Queued Players"))
        
        CustomPlayer.JoinButton.SetText(StringToMessage("JOIN"))
        CustomPlayer.LeaveButton.SetText(StringToMessage("LEAVE"))
        CustomPlayer.CallHelpButton.SetText(StringToMessage("CALL FOR HELP"))
        CustomPlayer.HardModeButton.SetText(StringToMessage("HARD MODE"))

        CustomPlayer.MessageText.SetTextColor(MakeColorFromHex("FFFFFF"))
        CustomPlayer.MessageText.SetText(StringToMessage("Closing HUD while in Queue\nwill keep you in Queue"))
        CustomPlayer.MessageText.SetJustification(text_justification.Center)
        
        for(i:int=0..CustomPlayer.QueuedPlayerBlocks.Length):
            if(QueuedPlayerBlock:=CustomPlayer.QueuedPlayerBlocks[i]):
                QueuedPlayerBlock.SetText(StringToMessage(""))

        for(TextBlock : CustomPlayer.AllTextBlocks):
            TextBlock.SetShadowColor(CustomPlayer.ShadowDefaults.DefaultShadowColor)
            TextBlock.SetShadowOffset(CustomPlayer.ShadowDefaults.DefaultShadowOffset)
            TextBlock.SetShadowOpacity(1.0)

        var MyBossCanvas: canvas = canvas:
            Slots := array:
                canvas_slot:
                    Widget:=CustomPlayer.BossHUDBG
                    ZOrder:=0
                    SizeToContent:=false
                    Alignment := vector2{X := 0.5, Y := 0.0}
                    Anchors := anchors{Minimum := vector2{X:= 0.5, Y:= 0.0}, Maximum := vector2{X:= 0.5, Y:= 0.0}}
                    Offsets:=margin{Top:=0.0,Left:=0.0,Bottom:=1076.0,Right:=2140.5}
                canvas_slot:
                    Widget:=CustomPlayer.BossCloseButton
                    ZOrder:=9
                    SizeToContent:=false
                    Anchors := anchors{Minimum := vector2{X:= 0.5, Y:= 0.0}, Maximum := vector2{X:= 0.5, Y:= 0.0}}
                    Offsets:=margin{Top:= 140.0,Left:= 664.0,Bottom:=CustomPlayer.BossClose_Size.Y,Right:=CustomPlayer.BossClose_Size.X}
                canvas_slot:
                    Widget:=CustomPlayer.BossHUDTitle
                    ZOrder:=1
                    SizeToContent:=false
                    Anchors := anchors{Minimum := vector2{X:= 0.5, Y:= 0.0}, Maximum := vector2{X:= 0.5, Y:= 0.0}}
                    Offsets:=margin{Top:=160.0,Left:=0.0, Bottom:=56.0, Right:=359.0}
                    Alignment := vector2{X:= 0.5, Y:= 0.0}
                canvas_slot:
                    ZOrder:=1
                    SizeToContent:=true
                    Anchors := anchors{Minimum := vector2{X:= 0.5, Y:= 0.0}, Maximum := vector2{X:= 0.5, Y:= 0.0}}
                    Offsets:=margin{Top:=104.0,Left:=-660.0}
                    Alignment := vector2{X:= 0.5, Y:= 0.0}
                    Widget := stack_box:
                        Orientation := orientation.Horizontal
                        Slots := array:
                            stack_box_slot:
                                Widget := CustomPlayer.BossKeyIcon   
                            stack_box_slot:
                                Widget := CustomPlayer.BossKeyText


                canvas_slot:
                    Widget:=CustomPlayer.BossIcon
                    ZOrder:=1
                    SizeToContent:=false
                    Anchors := anchors{Minimum := vector2{X:= 0.5, Y:= 0.0}, Maximum := vector2{X:= 0.5, Y:= 0.0}}
                    Offsets:=margin{Top:=336.0,Left:=-776.0, Bottom:=200.0,Right:=200.0}
                canvas_slot:
                    Widget:=CustomPlayer.BossNameText
                    ZOrder:=1
                    SizeToContent:=true
                    Anchors := anchors{Minimum := vector2{X:= 0.5, Y:= 0.0}, Maximum := vector2{X:= 0.5, Y:= 0.0}}
                    Offsets:=margin{Top:=324.0,Left:=-348.0}
                    Alignment := vector2{X:= 0.5, Y:= 0.0}
                canvas_slot:
                    Widget:=CustomPlayer.BossHealth
                    ZOrder:=1
                    SizeToContent:=true
                    Anchors := anchors{Minimum := vector2{X:= 0.5, Y:= 0.0}, Maximum := vector2{X:= 0.5, Y:= 0.0}}
                    Offsets:=margin{Top:=392.0,Left:=-544.0}
                    Alignment := vector2{X:= 0.0, Y:= 0.0}
                canvas_slot:
                    ZOrder:=1
                    SizeToContent:=true
                    Anchors := anchors{Minimum := vector2{X:= 0.5, Y:= 0.0}, Maximum := vector2{X:= 0.5, Y:= 0.0}}
                    Offsets:=margin{Top:=444.0,Left:=-544.0}
                    Alignment := vector2{X:= 0.0, Y:= 0.0}
                    Widget := stack_box:
                        Orientation := orientation.Horizontal
                        Slots := array:
                            stack_box_slot:
                                Widget := CustomPlayer.BossKeyCostIcon   
                            stack_box_slot:
                                Widget := CustomPlayer.BossKeyCostText
                canvas_slot:
                    Widget:=CustomPlayer.BossWarning
                    ZOrder:=1
                    SizeToContent:=true
                    Anchors := anchors{Minimum := vector2{X:= 0.5, Y:= 0.0}, Maximum := vector2{X:= 0.5, Y:= 0.0}}
                    Offsets:=margin{Top:=496.0,Left:=-544.0}
                    Alignment := vector2{X:= 0.0, Y:= 0.0}
                canvas_slot:
                    Widget:=CustomPlayer.RewardTitle
                    ZOrder:=1
                    SizeToContent:=true
                    Anchors := anchors{Minimum := vector2{X:= 0.5, Y:= 0.0}, Maximum := vector2{X:= 0.5, Y:= 0.0}}
                    Offsets:=margin{Top:=584.0,Left:=-460.0}
                    Alignment := vector2{X:= 0.5, Y:= 0.0}
                canvas_slot:
                    ZOrder:=1
                    SizeToContent:=true
                    Anchors := anchors{Minimum := vector2{X:= 0.5, Y:= 0.0}, Maximum := vector2{X:= 0.5, Y:= 0.0}}
                    Offsets:=margin{Top:=636.0,Left:=-728.0}
                    Alignment := vector2{X:= 0.0, Y:= 0.0}
                    Widget := stack_box:
                        Orientation := orientation.Horizontal
                        Slots := array:
                            stack_box_slot:
                                Widget := CustomPlayer.Reward1Icon   
                            stack_box_slot:
                                Widget := CustomPlayer.Reward1Text
                canvas_slot:
                    ZOrder:=1
                    SizeToContent:=true
                    Anchors := anchors{Minimum := vector2{X:= 0.5, Y:= 0.0}, Maximum := vector2{X:= 0.5, Y:= 0.0}}
                    Offsets:=margin{Top:=700.0,Left:=-728.0}
                    Alignment := vector2{X:= 0.0, Y:= 0.0}
                    Widget := stack_box:
                        Orientation := orientation.Horizontal
                        Slots := array:
                            stack_box_slot:
                                Widget := CustomPlayer.Reward2Icon   
                            stack_box_slot:
                                Widget := CustomPlayer.Reward2Text
                canvas_slot:
                    ZOrder:=1
                    SizeToContent:=true
                    Anchors := anchors{Minimum := vector2{X:= 0.5, Y:= 0.0}, Maximum := vector2{X:= 0.5, Y:= 0.0}}
                    Offsets:=margin{Top:=764.0,Left:=-728.0}
                    Alignment := vector2{X:= 0.0, Y:= 0.0}
                    Widget := stack_box:
                        Orientation := orientation.Horizontal
                        Slots := array:
                            stack_box_slot:
                                Widget := CustomPlayer.Reward3Icon   
                            stack_box_slot:
                                Widget := CustomPlayer.Reward3Text

                canvas_slot:
                    Widget:=CustomPlayer.QueuedPlayersTitle
                    ZOrder:=1
                    SizeToContent:=true
                    Anchors := anchors{Minimum := vector2{X:= 0.5, Y:= 0.0}, Maximum := vector2{X:= 0.5, Y:= 0.0}}
                    Offsets:=margin{Top:=324.0,Left:=120.0}
                    Alignment := vector2{X:= 0.0, Y:= 0.0}
                canvas_slot:
                    Widget:=CustomPlayer.QueuedPlayersCount
                    ZOrder:=1
                    SizeToContent:=true
                    Anchors := anchors{Minimum := vector2{X:= 0.5, Y:= 0.0}, Maximum := vector2{X:= 0.5, Y:= 0.0}}
                    Offsets:=margin{Top:=324.0,Left:=788.0}
                    Alignment := vector2{X:= 1.0, Y:= 0.0}
                canvas_slot:
                    Widget:=CustomPlayer.QueuedPlayer1
                    ZOrder:=1
                    SizeToContent:=true
                    Anchors := anchors{Minimum := vector2{X:= 0.5, Y:= 0.0}, Maximum := vector2{X:= 0.5, Y:= 0.0}}
                    Offsets:=margin{Top:=396.0,Left:=156.0}
                    Alignment := vector2{X:= 0.0, Y:= 0.0}
                canvas_slot:
                    Widget:=CustomPlayer.QueuedPlayer2
                    ZOrder:=1
                    SizeToContent:=true
                    Anchors := anchors{Minimum := vector2{X:= 0.5, Y:= 0.0}, Maximum := vector2{X:= 0.5, Y:= 0.0}}
                    Offsets:=margin{Top:=448.0,Left:=156.0}
                    Alignment := vector2{X:= 0.0, Y:= 0.0}
                canvas_slot:
                    Widget:=CustomPlayer.QueuedPlayer3
                    ZOrder:=1
                    SizeToContent:=true
                    Anchors := anchors{Minimum := vector2{X:= 0.5, Y:= 0.0}, Maximum := vector2{X:= 0.5, Y:= 0.0}}
                    Offsets:=margin{Top:=500.0,Left:=156.0}
                    Alignment := vector2{X:= 0.0, Y:= 0.0}
                canvas_slot:
                    Widget:=CustomPlayer.QueuedPlayer4
                    ZOrder:=1
                    SizeToContent:=true
                    Anchors := anchors{Minimum := vector2{X:= 0.5, Y:= 0.0}, Maximum := vector2{X:= 0.5, Y:= 0.0}}
                    Offsets:=margin{Top:=552.0,Left:=156.0}
                    Alignment := vector2{X:= 0.0, Y:= 0.0}
                canvas_slot:
                    Widget:=CustomPlayer.QueueStatus
                    ZOrder:=1
                    SizeToContent:=true
                    Anchors := anchors{Minimum := vector2{X:= 0.5, Y:= 0.0}, Maximum := vector2{X:= 0.5, Y:= 0.0}}
                    Offsets:=margin{Top:=672.0,Left:=456.0}
                    Alignment := vector2{X:= 0.5, Y:= 0.0}
                canvas_slot:
                    Widget:=CustomPlayer.MessageText
                    ZOrder:=1
                    SizeToContent:=true
                    Anchors := anchors{Minimum := vector2{X:= 0.5, Y:= 0.0}, Maximum := vector2{X:= 0.5, Y:= 0.0}}
                    Offsets:=margin{Top:=892.0,Left:=-460.0}
                    Alignment := vector2{X:= 0.5, Y:= 0.0}
                canvas_slot:
                    Widget:=CustomPlayer.JoinButton
                    ZOrder:=1
                    SizeToContent:=false
                    Anchors := anchors{Minimum := vector2{X:= 0.5, Y:= 0.0}, Maximum := vector2{X:= 0.5, Y:= 0.0}}
                    Offsets:=margin{Top:=772.0,Left:=480.0, Bottom:=80.0,Right:=320.0}
                    Alignment := vector2{X:= 0.0, Y:= 0.0}
                canvas_slot:
                    Widget:=CustomPlayer.LeaveButton
                    ZOrder:=1
                    SizeToContent:=false
                    Anchors := anchors{Minimum := vector2{X:= 0.5, Y:= 0.0}, Maximum := vector2{X:= 0.5, Y:= 0.0}}
                    Offsets:=margin{Top:=772.0,Left:=112.0, Bottom:=80.0,Right:=320.0}
                    Alignment := vector2{X:= 0.0, Y:= 0.0}
                canvas_slot:
                    Widget:=CustomPlayer.CallHelpButton
                    ZOrder:=1
                    SizeToContent:=false
                    Anchors := anchors{Minimum := vector2{X:= 0.5, Y:= 0.0}, Maximum := vector2{X:= 0.5, Y:= 0.0}}
                    Offsets:=margin{Top:=880.0,Left:=112.0, Bottom:=80.0,Right:=320.0}
                    Alignment := vector2{X:= 0.0, Y:= 0.0}
                canvas_slot:
                    Widget:=CustomPlayer.HardModeButton
                    ZOrder:=1
                    SizeToContent:=false
                    Anchors := anchors{Minimum := vector2{X:= 0.5, Y:= 0.0}, Maximum := vector2{X:= 0.5, Y:= 0.0}}
                    Offsets:=margin{Top:=880.0,Left:=480.0, Bottom:=80.0,Right:=320.0}
                    Alignment := vector2{X:= 0.0, Y:= 0.0}
        return MyBossCanvas
