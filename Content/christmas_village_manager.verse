
using { /Fortnite.com/Devices }
using { /Verse.org/Simulation }
using { /UnrealEngine.com/Temporary/Diagnostics }
using { /UnrealEngine.com/Temporary/SpatialMath }
using { /Verse.org/Random }

# See https://dev.epicgames.com/documentation/en-us/uefn/create-your-own-device-in-verse for how to create a verse device.

# A Verse-authored creative device that can be placed in a level
christmas_village_manager := class(creative_device):

    DateData : date_data = date_data{}
    
    @editable
    gm : game_manager = game_manager{}

    @editable
    AimTrainerProp : creative_prop = creative_prop{}
    @editable
    AimTrainerManip : prop_manipulator_device = prop_manipulator_device{}
    @editable
    TrainerBarrier : barrier_device = barrier_device{}
    @editable
    GameExitTP : teleporter_device = teleporter_device{}
    @editable
    ChristmasVillageEnterTp : teleporter_device = teleporter_device{}
    @editable
    ChristmasVillageExitTp : teleporter_device = teleporter_device{}

    @editable
    MinPos : vector3 = vector3{}
    @editable
    MaxPos : vector3 = vector3{}
    @editable
    StartGameBtn : button_device = button_device{}
    @editable
    ChristmasVillageEnterBtn : button_device = button_device{}
    @editable
    ChristmasVillageExitBtn : button_device = button_device{}

    var GameTime : float = 30.0

    var Score : float = 0.0
    var TimeStamp : float = 0.0

    var InUse : logic = false

    var AllRewards : []float = array{10.0, 10000.0, 15000.0} #Stone Multi, Egg Min, Egg Max

    var DamageNumbers<private>: damage_numbers = damage_numbers{}

    # Runs when the device is started in a running game
    OnBegin<override>()<suspends>:void=
        AimTrainerManip.HideProps()
        AimTrainerManip.DamagedEvent.Subscribe(OnHit)
        StartGameBtn.InteractedWithEvent.Subscribe(StartGameHelper)
        ChristmasVillageEnterBtn.InteractedWithEvent.Subscribe(VillageEnterHandler)
        ChristmasVillageExitBtn.InteractedWithEvent.Subscribe(VillageExitHandler)

    VillageEnterHandler(Agent:agent):void=
        ChristmasVillageEnterTp.Teleport(Agent)

    VillageExitHandler(Agent:agent):void=
        ChristmasVillageExitTp.Teleport(Agent)

    StartGameHelper(Agent:agent):void=
        var Today : date_data = DateData.ParseEpoch(GetSecondsSinceEpoch())
        if(CustomPlayer := gm.PlayersMap[Agent]):
            if(CustomPlayer.CurrentBoss = -1):
                if(InUse?):
                    gm.ShowMessage(CustomPlayer.Player, "The shooting range is currently in use!", 1)
                else:
                    if (not DateData.IsSameDay(Today, CustomPlayer.LastCompletedHolidayDate)?):
                        set CustomPlayer.LastCompletedHolidayDate = Today
                        set InUse = true
                        spawn{StartGame(Agent)}
                    else:
                        gm.ShowMessage(CustomPlayer.Player, "Be sure to visit again tomorrow!", 0)
            else:
                gm.ShowMessage(CustomPlayer.Player, "Cannot use while waiting for boss queue!", 1)



    StartGame(Agent:agent)<suspends>:void=
        var CurrentGameTime: float = GameTime
        if(CustomPlayer := gm.PlayersMap[Agent]):
            set Score = 0.0
            TrainerBarrier.AddToIgnoreList(Agent)
            gm.ShowMessage(CustomPlayer.Player, "The game will begin in 3 seconds, enter the arena and prepare yourself!", 0)
            Sleep(3.0)
            ShowNextTarget()
            loop:
                if (CurrentGameTime <= 0.0):
                    gm.ShowMessage(CustomPlayer.Player, "The game has ended, enjoy your gifts!", 0)
                    TrainerBarrier.RemoveFromIgnoreList(Agent)
                    GameExitTP.Teleport(Agent)
                    spawn{GiveRewards(CustomPlayer)}
                    set InUse = false
                    AimTrainerManip.HideProps()
                    break:
                set CurrentGameTime -= 1.0
                Sleep(1.0)

                    
    GiveRewards(CustomPlayer:Custom_Player)<suspends>:void=
        #Print("Score: {Score}")
        for(i:=0..3):
            if(StoneReward:=AllRewards[0]): #GIVE STONE
                TotalReward := ((CustomPlayer.GetPlayerDamage()*StoneReward * Score) / 4.0) #* CustomPlayer.BossRewardMult
                CustomPlayer.Collect(TotalReward, "Stone", 0)
                DamageNumbers.ShowDamageUIForPlayer(CustomPlayer, TotalReward, false, 0, false, "Rock", -1.0)
                #gm.StoneGranter.GrantItem(CustomPlayer.Player)
                Sleep(0.3)

            if(EggMinReward:=AllRewards[1],EggMaxReward:=AllRewards[2]): #GIVE EGGS
                var whichEgg : int = 0
                if(CustomPlayer.StagesUnlocked = -1):
                    if(whichDropTable:=gm.EggDropOdds[0],DropAmount := Floor[GetRandomFloat(EggMinReward * Score, EggMaxReward * Score)]*1.0, EggGranter:=gm.EggGranters[whichEgg]):
                        set whichEgg = SelectItem(whichDropTable)
                        TotalReward := (DropAmount / 4.0) #* CustomPlayer.BossRewardMult
                        CustomPlayer.Collect(TotalReward, "Egg", whichEgg)
                        DamageNumbers.ShowDamageUIForPlayer(CustomPlayer, TotalReward, false, 0, false, "Egg{whichEgg}", -1.0)
                        #EggGranter.GrantItem(CustomPlayer.Player)
                        Sleep(0.3)
                else:
                    if(whichDropTable:=gm.EggDropOdds[CustomPlayer.StagesUnlocked],DropAmount := Floor[GetRandomFloat(EggMinReward * Score, EggMaxReward * Score)]*1.0, EggGranter:=gm.EggGranters[whichEgg]):
                        set whichEgg = SelectItem(whichDropTable)
                        TotalReward := (DropAmount / 4.0)
                        CustomPlayer.Collect(TotalReward, "Egg", whichEgg)
                        DamageNumbers.ShowDamageUIForPlayer(CustomPlayer, TotalReward, false, 0, false, "Egg{whichEgg}", -1.0)
                        #EggGranter.GrantItem(CustomPlayer.Player)
                        Sleep(0.3)


    OnHit(Agent:agent):void=
        AimTrainerManip.HideProps()
        spawn{UpdateScore()}

    UpdateScore()<suspends>:void=
        CurrentTime := GetSimulationElapsedTime()
        AttackSpeed := CurrentTime - TimeStamp

        var EarnedPoints : float = 1.0

        if (AttackSpeed <= 1.0):
            set EarnedPoints = 10.0
        else if (AttackSpeed <= 2.0):
            set EarnedPoints = 5.0
        else if (AttackSpeed <= 3.0):
            set EarnedPoints = 3.0

        set TimeStamp = CurrentTime
        set Score += EarnedPoints

        spawn{
            ShowNextTarget()
        }

    ShowNextTarget()<suspends>:void=
        Sleep(0.1)
        RandX := GetRandomFloat(MinPos.X, MaxPos.X)
        RandY := GetRandomFloat(MinPos.Y, MaxPos.Y)
        RandZ := GetRandomFloat(MinPos.Z, MaxPos.Z)
        NewPosition := vector3{X := RandX, Y := RandY, Z := RandZ }
        AimTrainerProp.MoveTo(NewPosition, AimTrainerProp.GetTransform().Rotation, 0.1)
        AimTrainerManip.ShowProps()
        set TimeStamp = GetSimulationElapsedTime()

