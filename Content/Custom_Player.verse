using { /Fortnite.com/Devices }
using { /Verse.org/Simulation }
using { /Verse.org/Colors }
using { /Verse.org/Verse }
using { /Verse.org/Random }
using { /Fortnite.com/Teams }
using { /UnrealEngine.com/Assets }
using { /Fortnite.com/Game }
using { /Fortnite.com/Characters }
using { /Fortnite.com/Playspaces }
using { /Fortnite.com/UI }
using { /UnrealEngine.com/Temporary/Diagnostics }
using { /UnrealEngine.com/Temporary/SpatialMath }
using { /UnrealEngine.com/Temporary/UI }
using { /Verse.org/Assets}
using { /Verse.org/Colors/NamedColors }

Custom_Player := class<unique>():       #custom player class that holds the players values
    Player : player

    var UIMaps : [string]?canvas = map{}
    
    var DamageMap : [int]?canvas = map{}

    var MyUI : canvas = canvas{}

    var Version : int = 9 #CHANGING THIS DOESN'T DO ANYTHING

    var CanRegen : int = 0

    #RESOURCES
    var StoneAmount : float = 0.0 #CHANGE TO 0.0 # 1.0e40 FOR TESTING
    var ShardAmount : float = 0.0 #CHANGE TO 0.0 #PRESTIGE POINTS
    var EggAmounts : []float = array{0.0,0.0,0.0,0.0,0.0,0.0}
    var WepMatAmount : float = 0.0
    var BossKeyAmount : float = 0.0 #CHANGE TO 0.0
    var StardustAmount : float = 0.0 #CHANGE TO 0.0 # 1.0e40 FOR TESTING

    #CAVE EXPLORATION
    var CurrentRockHit : string = ""
    var CurrentMobHit : ?fort_character = false
    var StagesUnlocked : int = -1 #CHANGE TO -1
    var CurrentMobRoom : int = -1
    var CurrentMobHP : float = 0.0
    var MaxHP : float = 100.0 #CHANGE TO 100.0
    var WepMatDropRate : float = 0.02 #CHANGE TO 0.02

    #AFK ZONE
    var AFKStoneBoost : float = 0.0
    var AFKPetsUnlocked : logic = false
    var IsAFK : logic = false
    var CanGrantEXP : logic = true

    #STONE HARVEST UPGRADES
    var BaseDamage : float = 1.0 #DO NOT CHANGE THIS, THIS IS HOW MUCH DMG PLAYER DOES AT START OF GAME
    var Power : float = 1.0
    var HarvestBuff : float = 0.0 #CHANGE TO 0.0 # percentage bonus added to player power
    var CritRate : float = 0.0  #CHANGE TO 0.0 || 1.0 IS 100%
    var CritDMG : float = 1.5 #CHANGE TO 1.5
    var FinalHitBonus : float = 0.20 #CHANGE TO 0.20
    var MultiCrit : float = 1.0 #CHANGE TO 1.0

    var ComboStrikeUnlocked : logic = false
    var ComboStrikeMultiplier : float = 10.0
    var ComboAccelerator : float = 1.0
    var ComboCount : float = -1.0
    var ComboStrikeStreak : float = 1.0
    var CurrentComboStrikeStreak : float = 0.0

    var RoomUnlockGather : float = 0.0

    #MOVEMENT SPEED STATS
    var MovementSpeedLevel : int = 0 #CHANGE TO 0
    var DoubleJumpUnlocked : logic = false
    var DoubleJumpAmount : int = 0
    var DoubleJumpCount : int = 0

    #WEAPON UPGRADES
    var WeaponTier : int = 0 #CHANGE TO 0
    var WeaponPerkTier : int = 0

    var WeaponStatLevel : float = 0.0 #HOW MANY TIMES UPGRADED STATS
    var WeaponBonusLevel : float = 0.0 #HOW MANY TIMES UPGRADED BONUSES
    var WeaponPerkLevel : float = 0.0 #HOW MANY TIMES UPGRADED PERKS

    var WeaponStatRolling : logic = false
    var WeaponBonusRolling : logic = false
    
    var WeaponBaseDamage : float = 24.0 #DO NOT CHANGE THIS, THIS IS BASE DMG OF WEAPON IN GAME
    var WeaponStats : []float = array{0.0,0.0,0.0}
    var WeaponBonuses : []float = array{0.0,0.0,0.0}

    var StatCost : []float = array{1000000.0, 1.0} #STONE, EGG
    var BonusCost : float = 5000000.0 #STONE
    var PerkCost : float = 20000000.0 #STONE
    var EvolveCost : []float = array{40.0, 2375000.0, 50.0} #LV. REQ, STONE, WEP MAT #CHANGE TO {40.0, 2375000.0, 50.0}

    var PerksUnlocked : logic = false
    var ActivePerks : []tuple(string, float) = array{
        ("Perk 1", 0.0),
        ("Perk 2", 0.0),
        ("Perk 3", 0.0)
    }
    var LockedPerks : []logic = array{false, false, false}

    #HUT UPGRADE INFO
    var Hut1Levels : []int = array{0,0,0,0}
    var Hut1Costs : []float = array{15000.0,250000.0,400000.0,8000000.0} #NOT PERSISTED
    var Hut1Upgrades : []float = array{0.0,0.0,0.0,1.0} #THIS IS PURELY FOR THE HUD

    var Hut2Levels : []int = array{0,0,0,0}
    var Hut2Costs : []float = array{1.24e14,3.12e16,4.68e18,4.68e20} #NOT PERSISTED
    var Hut2Upgrades : []float = array{0.0,0.0,1.0,1.0} #THIS IS PURELY FOR THE HUD

    var Hut3Levels : []int = array{0,0,0,0}
    var Hut3Costs : []float = array{1.95e25,2.95e25,2.95e25,1.95e25} #NOT PERSISTED
    var Hut3Upgrades : []float = array{0.0,600.0,0.30,0.50} #THIS IS PURELY FOR THE HUD

    var Hut4Levels : []int = array{0,0,0,0}

    var Hut5Levels : []int = array{0,0,0,0}

    var HutEnabled : []logic = array{true, true, true, true, true}

    #THIS IS ALL JUST FOR UI STUFF, NOT SAVED
    AllHutMaxLevels : [][]int = array:
        array{10, 10, 10, 5}
        array{1, 5, 9, 4}
        array{1, 10, 10, 10}

    AllHutCostMults : [][]float = array:
        array{350.0, 600.0, 1000.0, 80000.0}
        array{1.0, 10000.0, 3000.0, 120000.0}
        array{1.0, 1000.0, 500.0, 250.0}

        
    #POP UP UI
    var PopUpPrestige : texture_block = texture_block{DefaultImage:=UI.PopUpPrestige}
    var PopUpPrestigeText : text_block = text_block{DefaultTextColor := MakeColorFromHex("39FF39")}
    
    var PopUpHut1 : texture_block = texture_block{DefaultImage:=UI.PopUpCrit}
    var PopUpHut2 : texture_block = texture_block{DefaultImage:=UI.PopUpCombo}
    var PopUpHut3 : texture_block = texture_block{DefaultImage:=UI.PopUpEarthquake}

    #GENERAL UPGRADE INFO
    var GUP1Levels : []int = array{0,0}
    var GUP1Costs : []float = array{1500.0,10.0}
    var GUP1Upgrades : []float = array{0.0,0.0} #THIS IS PURELY FOR THE HUD

    var GUP2Levels : []int = array{0,0}
    var GUP2Costs : []float = array{10.0,10.0}
    var GUP2Upgrades : []float = array{0.0,0.0} #THIS IS PURELY FOR THE HUD

    var GUP3Levels : []int = array{0,0}
    var GUP3Costs : []float = array{10.0,10.0}
    var GUP3Upgrades : []float = array{0.0,0.20} #THIS IS PURELY FOR THE HUD

    #PRESTIGE UPGRADE INFO
    var PrestigeLevel : int = 0
    var PrestigeCost : float = 1.12e16
    var PrestigeCostMult : float = 5.0
    var PrestigeHarvestBuff : float = 0.0
    var PrestigeStoneAmount : float = 0.0

    var PrestigeFix : logic = false
    var EndGamePrestigeShardAmount : float = 250000.0
    var EndGamePrestigeGatherAmount : float = 80000.0
    
    var PUPLevels : []int = array{0, 0, 0} #0 INDEX IS STRENGTH SURGE LEVELS
    var PUPCosts : []float = array{10.0, 10.0, 10.0}
    var PUPUpgrades : []float = array{0.0, 0.0, 0.0} #THIS IS PURELY FOR THE HUD

    var EggPrestigeLevels : []int = array{0, 0, 0} 
    var EggPrestigeCosts : []float = array{10.0, 10.0, 10.0}
    var EggPrestigeUpgrades : []float = array{0.0, 0.0, 0.0} 

    var PUPQOLLevels : []int = array{0, 0, 0}

    var SurgeMultiplier : float = 3.0
    var AutoGatherRate : float = 0.0
    var BossRewardMult : float = 1.0

    var AutoUpgradeOnHit : logic = false

    var EggAmountMultiplier : float = 0.0
    var EggDropRateMultiplier : float = 0.0

    #STRENGTH UPGRADES
    #CHANGE IN GAME MANAGER
    var BaseStrengthCosts : []float = array{
        15.0,   150.0,  800.0,  4.80e03,3.00e05,
        1.80e06,3.00e07,1.80e08,4.50e09,2.70e10,
        4.50e11,2.70e12,3.38e13,2.03e14,1.87e16,
        1.12e17,3.75e18,2.25e19,1.12e21,6.74e21,
        2.25e23,1.35e24,3.37e25,2.02e26,3.75e28,
        2.25e29,1.12e31,6.74e31,5.06e33,3.03e34,
        1.52e36,9.10e36,3.41e38,2.05e39,5.31e40,
        3.18e41,2.12e42,1.27e43,1.27e44,2.64e45, #7.64e45

        3.41e49, 2.05e51, 1.23e53, 7.37e54, 4.42e56, 
        2.65e58, 3.47e59, 2.08e61, 1.25e62, 7.50e63, 
        4.50e64, 2.70e65, 9.71e65, 4.63e66, 2.78e68, 
        1.67e70, 9.99e71, 6.00e72, 1.45e74, 8.70e76, 

        5.22e78, 3.13e79, 1.88e81, 1.13e83, 1.67e84, 
        3.00e85, 8.01e86, 7.61e88, 8.16e90, 9.30e91, 
        7.09e92, 8.51e93, 6.91e94, 5.34e95, 8.41e96, 
        8.44e97, 5.06e98, 3.04e99, 1.82e100, 1.09e101,
        
        9.81e105, 8.83e107, 7.95e109, 7.15e111, 6.44e113, 
        5.79e115, 5.21e117, 4.69e119, 4.22e121, 3.80e123, 
        3.42e125, 3.08e127, 2.77e129, 2.49e131, 2.24e133, 
        2.02e135, 1.82e137, 1.64e139, 1.47e141, 1.33e143, 
        1.19e145, 1.07e147, 9.66e148, 8.69e150, 7.83e152, 
        7.04e154, 6.34e156, 5.70e158, 5.13e160, 4.62e162     
        } 

    var StrengthBaseUpgrades : []float = array{
        1.0,    10.0,   25.0,   1.00e02,2.22e03,
        8.33e03,3.00e05,1.07e06,2.73e07,9.38e07,
        1.88e09,6.25e09,1.28e11,1.25e12,3.57e13,
        1.14e14,5.00e15,1.56e16,4.68e17,1.44e18,
        1.10e19,1.00e20,6.93e21,2.08e22,6.57e23,
        1.95e24,9.36e25,2.75e26,8.91e27,2.60e28,
        6.38e29,1.85e30,1.35e32,3.89e33,1.30e34,
        3.71e35,1.87e36,5.31e37,1.80e38,5.08e39,
        
        1.40e41,4.21e42,1.26e44,3.79e45,1.14e47,
        3.41e48,2.81e49,8.42e50,2.53e52,7.58e53,
        2.27e55,6.82e56,1.40e57,4.21e58,1.26e60,
        3.79e61,1.14e63,3.41e64,1.12e65,3.37e66,

        1.01e68,3.03e69,9.09e70,2.73e72,1.38e73,
        4.14e74,1.24e76,3.73e77,1.12e79,3.36e80,
        1.70e81,5.10e82,1.53e84,4.59e85,1.38e87,
        4.13e88,1.24e90,3.71e91,1.11e93,3.34e94,

        1.67e96, 8.35e97, 4.18e99, 2.09e101, 1.04e103, 
        5.22e104, 2.61e106, 1.30e108, 6.52e109, 3.26e111, 
        1.63e113, 8.15e114, 4.08e116, 2.04e118, 1.02e120, 
        5.10e121, 2.55e123, 1.27e125, 6.37e126, 3.19e128, 
        1.59e130, 7.96e131, 3.98e133, 1.99e135, 9.95e136, 
        2.49e139, 6.22e141, 1.56e144, 3.89e146, 9.72e148
        }
    
    <#var StrengthUpgrades : []float = array{
        0.0, 0.0, 0.0, 0.0, 0.0,
        0.0, 0.0, 0.0, 0.0, 0.0,
        0.0, 0.0, 0.0, 0.0, 0.0,
        0.0, 0.0, 0.0, 0.0, 0.0,
        0.0, 0.0, 0.0, 0.0, 0.0,
        0.0, 0.0, 0.0, 0.0, 0.0,
        0.0, 0.0, 0.0, 0.0, 0.0,
        0.0, 0.0, 0.0, 0.0, 0.0,

        0.0, 0.0, 0.0, 0.0, 0.0,
        0.0, 0.0, 0.0, 0.0, 0.0,
        0.0, 0.0, 0.0, 0.0, 0.0,
        0.0, 0.0, 0.0, 0.0, 0.0,
        0.0, 0.0, 0.0, 0.0, 0.0,
        0.0, 0.0, 0.0, 0.0, 0.0,
        0.0, 0.0, 0.0, 0.0, 0.0,
        0.0, 0.0, 0.0, 0.0, 0.0
        }#>
    
    var StrengthEfficiencies : []int = array{
        1, 1, 1, 1, 1,
        1, 1, 1, 1, 1,
        1, 1, 1, 1, 1,
        1, 1, 1, 1, 1,
        1, 1, 1, 1, 1,
        1, 1, 1, 1, 1,
        1, 1, 1, 1, 1,
        1, 1, 1, 1, 1,

        1, 1, 1, 1, 1,
        1, 1, 1, 1, 1,
        1, 1, 1, 1, 1,
        1, 1, 1, 1, 1,
        1, 1, 1, 1, 1,
        1, 1, 1, 1, 1,
        1, 1, 1, 1, 1,
        1, 1, 1, 1, 1,
        
        1, 1, 1, 1, 1,
        1, 1, 1, 1, 1,
        1, 1, 1, 1, 1,
        1, 1, 1, 1, 1,
        1, 1, 1, 1, 1,
        1, 1, 1, 1, 1
        }
    
    var StrengthUpgradeLevels : []int = array{
        0,0,0,0,0,
        0,0,0,0,0,
        0,0,0,0,0,
        0,0,0,0,0,
        0,0,0,0,0,
        0,0,0,0,0,
        0,0,0,0,0,
        0,0,0,0,0,

        0,0,0,0,0,
        0,0,0,0,0,
        0,0,0,0,0,
        0,0,0,0,0,
        0,0,0,0,0,
        0,0,0,0,0,
        0,0,0,0,0,
        0,0,0,0,0,
        
        0,0,0,0,0,
        0,0,0,0,0,
        0,0,0,0,0,
        0,0,0,0,0,
        0,0,0,0,0,
        0,0,0,0,0
        }
        
    var EfficiencyLevelRqs : []int = array{1, 5, 25, 50, 75, 100}

    #EARTHQUAKE STATS
    var EarthquakeUnlocked : logic = false
    var EarthquakeDistance : float = 600.0 
    var EarthquakeGatherAmount : float = 0.20
    var EarthquakeDamage : float = 0.50 #% BASED ON PLAYER DAMAGE
    var EarthquakeShowing : logic = false

    #VOID FRACTURE STATS
    var VoidUnlocked : logic = false
    var VoidDamage : float = 0.50 #% BASED ON PLAYER DAMAGE
    var VoidDuration : float = 5.0 #SECONDS
    var VoidStacks : float = 1.0

    var VoidActiveTimer : float = 0.0
    var VoidActiveStacks : float = 0.0

    #MISSILE STATS
    var MissileCD : float = 0.0

    var ShowVFX : logic = true

    var InSpace : logic = false

    var A4Unlocked : logic = false #If you've ever been to A4

    var A5Unlocked : logic = false #If you've ever been to A4

    #INFINITE SINKS

    var SinkPrestigePower : float = 0.0
    var SinkWepMatPower : float = 0.0
    var SinkBossKeyPower : float = 0.0

    #REAL TIME DATA
    var CurrentDay : int = -1
    var LastClaimed : time_and_date = time_and_date{Year:=-1,Month:=-1,Day:=-1,Hours:=-1,Minutes:=-1,Seconds:=-1}
    var HatPrestiges : int = 0
    var HatCost : float = 1000.0
    var HatsUnlocked : [int]logic = map{
        0 => false,
        1 => false,
        2 => false,
        3 => false,
        4 => false,
        5 => false,
        6 => false,
        7 => false,
        8 => false,
        9 => false,
        10 => false,
        11 => false,
        12 => false,
        13 => false,
        14 => false,
        15 => false,
        16 => false,
        17 => false,
        18 => false,
        19 => false,
        20 => false,
        21 => false,
        22 => false,
        23 => false,
        24 => false,
        25 => false,
        26 => false,
        27 => false,
        28 => false,
        29 => false,
        30 => false,
        31 => false,
        32 => false,
        33 => false,
        34 => false,
        35 => false,
        36 => false,
        37 => false,
        38 => false,
        39 => false,
        40 => false,
        41 => false,
        42 => false,
        43 => false,
        44 => false,
        45 => false,
        46 => false,
        47 => false,
        48 => false,
        49 => false,
        50 => false,
        51 => false,
        52 => false,
        53 => false,
        54 => false,
        55 => false,
        56 => false,
        57 => false,
        58 => false,
        59 => false,
        60 => false,
        61 => false,
        62 => false,
        63 => false,
        64 => false,
        65 => false,
        66 => false,
        67 => false,
        68 => false,
        69 => false,
        70 => false,
        71 => false,
        72 => false,
        73 => false,
        74 => false,
        75 => false,
        76 => false,
        77 => false,
        78 => false,
        79 => false
    }
    var CodeHatsUnlocked : [int]logic = map{
        0=>false,
        1=>false
    }

    #NEW DAILY REWARDS

    var QuestRocksDestroyed : int = 0
    var QuestMobsDefeated : int = 0
    var QuestBossesDefeated : int = 0
    var QuestFinalStatus : logic = false
    var QuestsStarted : []logic = array{false,false,false}
    var QuestsCompleted : []logic = array{false,false,false}
    var LastCompletedDate : date_data = date_data{}

    #CHRISTMAS VILLAGE

    var LastCompletedHolidayDate : date_data = date_data{}

    #New Meteor

    var DamageToNewMeteor : float = 0.0
    
    #EXPEDITIONS
    #FOR HUD
    var EXPSelectedPet : int = 0
    var EXPSelectedExpedition : int = 0
    var ClaimingRewards : logic = false

    #PERSIST
    var NewEXPPetsLevel : [int]float = map: #CHANGE TO 0.0, 0.0 #0 IS DAVE, ALWAYS 0.0
        0 => 0.0,
        1 => 0.0,
        2 => 0.0
    var EXPPetsLevel : []float = array{0.0, 0.0, 0.0} #CHANGE TO 0.0, 0.0 #0 IS DAVE, ALWAYS 0.0
    var CurrentExpedition : int = -1
    var EXPCurrentPet : int = 0
    var ExpeditionTimeLeft : float = 0.0
    var EXPTotalRewards : []float = array{0.0,0.0,0.0,0.0}

    
    #GOLDEN DRILL
    var GoldenDrillUnlocked : logic = false
    var GoldenDrillTimer : logic = false
    var DrillActive : logic = false
    var DrillActivations : int = 0
    var BuffMultiplier : float = 2.0 #100% BONUS TO START (1.0 wouldn't do anything)
    var BuffTimeSeconds : float = 30.0
    var BuffCurrentTimeSeconds : float = 0.0 #ORIGINAL VALUE 30.0
    var BuffChargeTimer : float = 900.0 #15 MINUTE CHARGE TIME
    var BuffCurrentCharges : float = 1.0
    var BuffMaxCharges : float = 1.0
    var GoldenDrillLevels : []int = array{0,0,0,0}
    #var GoldenDrillCosts : []float = array{7.50e07,5.63e11,7.50e07,4.68e18}
    #var GoldenDrillUpgrades : []float = array{0.0,1.0,30.0,1.0} #THIS IS PURELY FOR THE HUD

    #SECRETS
    var SecretsClaimed : []logic = array{false,false,false,false,false,false,false,false,false}

    #USE THIS ONE
    var NewSecretsClaimed : [int]logic = map{
        0 => false,
        1 => false,
        2 => false,
        3 => false,
        4 => false,
        5 => false,
        6 => false,
        7 => false,
        8 => false,
        9 => false,
        10 => false,
        11 => false,
        12 => false,
        13 => false,
        14 => false,
        15 => false,
        16 => false,
        17 => false,
        18 => false,
        19 => false,
        20 => false,
        21 => false,
        22 => false,
        23 => false,
        24 => false,
        25 => false,
        26 => false,
        27 => false,
        28 => false,
        29 => false,
        30 => false
    }
    #PET INFO
    var EggsUnlocked : logic = false #CHANGE TO FALSE
    var RockEggDropChance : float = 1.0 #CHANGE TO 1.0
    var EggDropAmount : []float = array{1.0, 2.0}
    var MobEggDropChance : float = 10.0 #CHANGE TO 10.0
    var PetDataReset : logic = false

    var NewestPetOwner : ?agent = false

    #THIS IS HERE TO MAKE SURE IT'S UNIQUE FOR EACH PLAYER, NOT PERSISTABLE DATA
    var PetFortChars : [int]?fort_character = map{}

    # -1 is NO PET, 0-5 is whichever pet is equipped in that slot
    var PetSlots : []int = array{-1,-1,-1,-1,-1,-1}
    var CurrentPetSlot : int = 0
    var PetEquipSlots : []logic = array{true, true, true, false, false, false}
    var PetUnlocked : []logic = array{false, false, false, false, false, false}
    var PetLevels : []int = array{0,0,0,0,0,0}
    var PetEvolveLevels : []int = array{0,0,0,0,0,0}
    var PetGatherRates : []float = array{0.0,0.0,0.0,0.0,0.0,0.0}
    var PetAttackRates : []float = array{0.0,0.0,0.0,0.0,0.0,0.0}
    var PetCritRates : []float = array{0.0,0.0,0.0,0.0,0.0,0.0}
    var PetCritDMG : []float = array{0.0,0.0,0.0,0.0,0.0,0.0}
    var PetFinalStrike : []float = array{0.0,0.0,0.0,0.0,0.0,0.0}

    #0,1,2 = Pet 1  || 3,4,5 = Pet 2
    #var PetMutatorLevels : []int = array{0,0,0,0,0,0}
    <#    "Pet1CritChance" => 0
        "Pet1CritDmg" => 0
        "Pet1FS" => 0
        "Pet2CritChance" => 0
        "Pet2CritDmg" => 0
        "Pet2FS" => 0#>

    var EggsHatched : []float = array{0.0,0.0,0.0,0.0,0.0,0.0}
    var PetSelected : int = 0
    var MPetSelected : int = 0
    var PetCurrentPage : int = 0
    var MPetCurrentPage : int = 0
    var SelectedEgg : int = 0

    var HidePets : logic = false

    var SpawningPets : logic = false

    #BOSS INFO
    var QueuedForBoss : logic = false
    var CurrentBoss : int = -1
    var RequestedBossHelp : logic = false
    var InHardBoss : logic = false

    var BossMenuOnCooldown : logic = false

    var AllBossesDefeated : [int]logic = map:
        0 => false #Mushroom King
        1 => false #Razor Fang
        2 => false #Cactoro
        3 => false #Prestige Boss
        4 => false #Brambleback
        5 => false #Grimroot (Halloween Seasonal)
        6 => false #The Blamed One
        7 => false #The Blamed One Hard Mode
    
    var FBDamage : float = 0.0
    var HighestFBDamage : float = 0.0

    var DemonPetObtained : logic = false

    #UI GARBAGE
    var StoneUIText : text_block = text_block{DefaultTextColor := MakeColorFromHex("FFFFFF"),DefaultShadowOffset := option{vector2{X:=3.0, Y:=3.0}},DefaultShadowColor := MakeColorFromHex("000000")}
    var AutoStoneText : text_block = text_block{DefaultTextColor := MakeColorFromHex("39FF39"),DefaultShadowOffset := option{vector2{X:=3.0, Y:=3.0}},DefaultShadowColor := MakeColorFromHex("000000")}
    var AutoStoneIcon : texture_block = texture_block{DefaultImage := UI.AutoGather,DefaultDesiredSize:=vector2{X:=60.0,Y:=60.0}}
    var ShardUIText : text_block = text_block{DefaultTextColor := MakeColorFromHex("FFFFFF"),DefaultShadowOffset := option{vector2{X:=3.0, Y:=3.0}},DefaultShadowColor := MakeColorFromHex("000000")}
    
    var CTAHUD : texture_block = texture_block{DefaultImage := UI.cta}
    
    var DrillUI : texture_block = texture_block{DefaultImage:=UI.DrillUI}
    var DrillUIText : text_block = text_block{DefaultTextColor := MakeColorFromHex("FFFFFF")}
    var DrillActiveTimer : text_block = text_block{DefaultTextColor := MakeColorFromHex("73ABFF")}
    var DrillChargeTimer : text_block = text_block{DefaultTextColor := MakeColorFromHex("DDDDDD")}
    var DrillChargeTimerIcon : texture_block = texture_block{DefaultImage:=UI.drillcharges}
    var CurrentPage : int = 0 #FOR WEAPON UPGRADE HUD
    var CurrentHut : int = 0 #FOR HUT UPGRADES HUD
    var CurrentGUP : int = 0 #FOR GUP UPGRADES HUD
    var CurrentPUP : int = 0 #FOR PUP UPGRADES HUD
    var CurrentStat : int = 0 #For Player Stat Pages
    var TotalAmountGained : float = 0.0
    var EQTotalAmountGained : float = 0.0
    var VoidTotalAmountGained : float = 0.0
    var MissileTotalAmountGained : float = 0.0
    var WepTotalAmountGained : float = 0.0

    var CurrentMusic : string = "Default"
    var AchTitle : text_block = text_block{DefaultTextColor := MakeColorFromHex("FFFFFF"),DefaultShadowOffset := option{vector2{X:=3.0, Y:=3.0}},DefaultShadowColor := MakeColorFromHex("000000")}
    var GatherRateBonus : text_block = text_block{DefaultTextColor := MakeColorFromHex("39FF39"),DefaultShadowOffset := option{vector2{X:=3.0, Y:=3.0}},DefaultShadowColor := MakeColorFromHex("000000")}
    
    var TutorialLevel : int = 0
    var TutorialProgress : float = 0.0

    var AchTitles : []string = array{
        "Total Stone Gathered",
        "Max Strength",
        "Total Mobs Defeated",
        "Prestige Points Gathered",
        "Total Rocks Destroyed",
        "Total Pet Level",
        "Total Bosses Defeated",
        "Total Time Played"
        }

    #ACHIEVEMENTS
    var AchIncreases : []float = array{100.0, 100.0, 10.0, 10.0, 10.0, 2.0, 2.0, 1800.0}
    var AchLevels : []int = array{1, 1, 1, 1, 1, 1, 1, 1}
    var AchBuffTotal : float = 0.0 #VISUAL FOR HUD ONLY, USE "HarvestBuff"
    var AllAchs : [][]float = array{
        array{0.0,1000000.0} #Total Stone Gathered #CHANGE TO 1000000.0
        array{0.0,1000.0} #Max Strength
        array{0.0,100.0} #Total Mobs Defeated
        array{0.0,10.0} #Prestige Points Obtained
        array{0.0,1000.0} #Total Rocks Destroyed #CHANGE TO 1000.0
        array{0.0,100.0} #Total Pet Level
        array{0.0,1.0} #Total Bosses Defeated
        array{0.0,1800.0} #Total Time Played IN SECONDS, CHANGE TO 1800 = 30 MINUTES
        }

    var TutorialAmounts : []float = array{15.0,10.0,1000.0,1.0,1.0,1.0,5.0,1.0,2.50e05,2.50e06,20.0,1.0,1.0,7.50e07,1.0,1.0,50.0,1.12e16}

    var PlayTimeSeconds : float = 0.0
    var PlayTimeSessionSeconds : float = 0.0

    #ONE TIME ANALYTICS
    var OneTimeAnalytics : [string]logic = map{
        "Play10Min"=>false,
        "Play30Min"=>false,
        "Play60Min"=>false,
        "Play120Min"=>false,
        "Play240Min"=>false,
        "Play6000Min"=>false,
        "Play30000Min"=>false,
        "DestroyMeteor"=>false,
        "FightPrestigeBoss"=>false,
        "Reach80StrTiers"=>false,
        "Reach80SurgeTiers"=>false,
        "FightHalloweenBoss"=>false,
        
        "UnlockArea5"=>false,
        "DestroyNewMeteor"=>false,
        "DefeatFinalBoss"=>false,
        "ObtainBossPet"=>false
    }

    GetPlayerDamage(): float=
        var TotalDamage: float = 0.0
        var PerkDamageBonus : float = 0.0
        var PetGather : float = 0.0
        var SecretGather : float = 0.0
        for(i:int=0..PetSlots.Length):
            if(PetSlots[i] <> -1):
                if:
                    Gather := PetGatherRates[PetSlots[i]]
                then:
                    set PetGather += Gather

        for(i:int=0..NewSecretsClaimed.Length):
            if(NewSecretsClaimed[i] = true):
                set SecretGather += 0.05

        for(ActivePerk:ActivePerks):
            if(ActivePerk(0) = "Strength"):
                set PerkDamageBonus = ActivePerk(1)
        if(StrStat:=WeaponStats[2],StrBonus:=WeaponBonuses[2], VoidWither:=Hut4Levels[2]): #STR UPGRADE
            #Print("Void Wither Bonus: {0.02 * VoidWither}")
            set TotalDamage = Power + (Power * HarvestBuff) + (Power * RoomUnlockGather) + (Power * (StrStat + (StrStat*StrBonus))) + (Power * (PetGather/100.0)) + (Power * PerkDamageBonus) + (Power * SecretGather) + (Power * SinkPrestigePower/100.0)
            set TotalDamage *= 1.0 + (0.02 * VoidWither)
            #Print("{PetGather/100.0}")
        if(DrillActive = true and QueuedForBoss = false):
            set TotalDamage *= BuffMultiplier
        return TotalDamage

    GetPlayerWeaponDamage(): float=
        var PerkDamageBonus : float = 0.0
        var PetStr : float = 0.0
        for(i:int=0..PetSlots.Length):
            if(PetSlots[i] <> -1):
                if:
                    PetLevel := PetLevels[PetSlots[i]]
                    Strength := PetAttackRates[PetSlots[i]]
                then:
                    set PetStr += Strength
        for(ActivePerk:ActivePerks):
            if(ActivePerk(0) = "Attack"):
                set PerkDamageBonus = ActivePerk(1)
        WeaponTierDamage := 226.0 * Pow(25.0, (WeaponTier-1)*1.0) #BASE WE CHOOSE, MULTIPLIER, HOW MANY TIMES
        var TotalDamage: float = WeaponBaseDamage + WeaponTierDamage #ADD REAL DAMAGE TO FORNITE BASE DAMAGE NON NEGOTIABLE
        if:
            AttackIncrease := WeaponStats[0] + (WeaponStats[0]*WeaponBonuses[0])
            AttackPercIncrease :=WeaponStats[1] + (WeaponStats[1]*WeaponBonuses[1])
        then:
            set TotalDamage += AttackIncrease + (AttackIncrease * AttackPercIncrease)
            set TotalDamage *= 1.0 + PerkDamageBonus 
            set TotalDamage += TotalDamage * (PetStr/100.0)
            set TotalDamage += TotalDamage * (SinkWepMatPower/100.0)
            set TotalDamage *= 2.0 #MAYBE KEEP?
        return TotalDamage

    CalcTotalDamage(MobOrRock:string): tuple(float, logic, float, int)=
        var TotalDamage : float = 0.0
        var isCrit : logic = false 

        #BASE DAMAGE + COMBO
        if(MobOrRock = "Mob"):
            set TotalDamage = GetPlayerWeaponDamage()
        if(MobOrRock = "Earthquake"): #EARTHQUAKE DOESN'T STACK COMBO
            set TotalDamage = GetPlayerDamage()
        if(MobOrRock = "Void"):
            set TotalDamage = GetPlayerDamage()
        if(MobOrRock = "Stone"):
            set TotalDamage = GetPlayerDamage()
            if(ComboStrikeUnlocked = true and HutEnabled[1] = true):
                if(ComboCount = -1.0 or (ComboCount = 100.0 and CurrentComboStrikeStreak = 0.0)):
                    set ComboCount = 0.0
                if(ComboCount < 100.0): #CHANGE TO 100.0
                    set ComboCount += ComboAccelerator
                if(ComboCount >= 100.0): #CHANGE TO 100.0
                    set ComboCount = 100.0
                    if(CurrentComboStrikeStreak = 0.0): #If 0 set up to how many hits
                        set CurrentComboStrikeStreak = ComboStrikeStreak
                    set CurrentComboStrikeStreak -= 1.0
                    set TotalDamage *= ComboStrikeMultiplier
                    
        #CRIT
        var CritCount : int = 0
        if(HutEnabled[0] = true):
            if(MobOrRock <> "Void"):
                var CurrentCritChance : float = CritRate
                var CurrentCritDmg : float = CritDMG
                for(i:int=0..PetSlots.Length):
                    if(PetSlots[i] <> -1):
                        if:
                            PetLevel := PetLevels[PetSlots[i]]
                            PetCritRate := PetCritRates[PetSlots[i]]
                            PetCritDmg := PetCritDMG[PetSlots[i]]
                        then:
                            #Print("Pet {i} - PetLevel: {PetLevel}")
                            set CurrentCritChance += PetCritRate/100.0
                            set CurrentCritDmg += PetCritDmg/100.0
                IDX:=GetRandomFloat(0.0,1.0)
                if(CurrentCritChance >= IDX):
                    set isCrit = true
                    set CurrentCritChance /= 2.0
                    set CritCount += 1
                    if(HowManyCrits := Floor[MultiCrit+000.1]):
                        if(HowManyCrits > 1 and MobOrRock <> "Mob"):
                            for(i:int=1..HowManyCrits - 1):
                                NewIDX:=GetRandomFloat(0.0,1.0)
                                if(CurrentCritChance >= NewIDX):
                                    set CurrentCritDmg += CritDMG
                                    set CurrentCritChance /= 2.0
                                    set CritCount += 1
                    set TotalDamage *= CurrentCritDmg
        
        return (TotalDamage, isCrit, ComboCount, CritCount)

    RemoveActivePerks():void=
        #Strength & Attack are handled directly within with respected GetDamages()
        #Instant Reload is handled by SetClass() in Game Manager
        for(ActivePerk:ActivePerks):
            if(ActivePerk(0) = "Crit Chance"):
                set CritRate -= ActivePerk(1)
            else if(ActivePerk(0) = "Crit Damage"):
                set CritDMG -= ActivePerk(1)
            else if(ActivePerk(0) = "Final Strike"):
                set FinalHitBonus -= ActivePerk(1)

    AddActivePerks():void=
        for(ActivePerk:ActivePerks):
            if(ActivePerk(0) = "Crit Chance"):
                set CritRate += ActivePerk(1)
            else if(ActivePerk(0) = "Crit Damage"):
                set CritDMG += ActivePerk(1)
            else if(ActivePerk(0) = "Final Strike"):
                set FinalHitBonus += ActivePerk(1)

    GetPlayer()<transacts>:player=
        return Player

    CreateMyUI() : canvas =   #Creates the Main Hud UI
        AutoStoneIcon.SetDesiredSize(vector2{X:=AutoStoneIcon.DefaultDesiredSize.X,Y:=AutoStoneIcon.DefaultDesiredSize.Y})
        PlayerGatherRate:=GetPlayerDamage()
        if(PlayerGatherRate * AutoGatherRate > 0.0):
            AutoStoneIcon.SetVisibility(widget_visibility.Visible)
            if(PlayerGatherRate * AutoGatherRate > 1.0):
                AutoStoneText.SetText(StringToMessage(" +{NumberCondensor(PlayerGatherRate * AutoGatherRate)}/s"))
            else:
                AutoStoneText.SetText(StringToMessage(" <+1/s"))
        else:
            AutoStoneIcon.SetVisibility(widget_visibility.Collapsed)
            AutoStoneText.SetText(StringToMessage(""))

        AutoStoneText.SetShadowColor(AutoStoneText.DefaultShadowColor)
        AutoStoneText.SetShadowOffset(AutoStoneText.DefaultShadowOffset)
        AutoStoneText.SetShadowOpacity(1.0)

        StoneUIText.SetText(StringToMessage(NumberCondensor(StoneAmount)))
        StoneUIText.SetShadowColor(StoneUIText.DefaultShadowColor)
        StoneUIText.SetShadowOffset(StoneUIText.DefaultShadowOffset)
        StoneUIText.SetShadowOpacity(1.0)

        ShardUIText.SetText(StringToMessage(NumberCondensor(ShardAmount)))
        ShardUIText.SetShadowColor(ShardUIText.DefaultShadowColor)
        ShardUIText.SetShadowOffset(ShardUIText.DefaultShadowOffset)
        ShardUIText.SetShadowOpacity(1.0)

        DrillUIText.SetShadowColor(ShardUIText.DefaultShadowColor)
        DrillUIText.SetShadowOffset(ShardUIText.DefaultShadowOffset)
        DrillUIText.SetShadowOpacity(1.0)
        DrillActiveTimer.SetShadowColor(ShardUIText.DefaultShadowColor)
        DrillActiveTimer.SetShadowOffset(ShardUIText.DefaultShadowOffset)
        DrillActiveTimer.SetShadowOpacity(1.0)
        DrillChargeTimer.SetShadowColor(ShardUIText.DefaultShadowColor)
        DrillChargeTimer.SetShadowOffset(ShardUIText.DefaultShadowOffset)
        DrillChargeTimer.SetShadowOpacity(1.0)
        

        if(GoldenDrillUnlocked = true):
            DrillUI.SetVisibility(widget_visibility.Visible)
            DrillUIText.SetText(StringToMessage(NumberCondensor(BuffCurrentCharges)))
        else: #HIDE IF YOU DON'T HAVE DRILL UNLOCKED
            DrillUI.SetVisibility(widget_visibility.Collapsed)
            DrillUIText.SetText(StringToMessage(""))

        PopUpPrestige.SetVisibility(widget_visibility.Collapsed)
        PopUpPrestigeText.SetText(StringToMessage(""))
        PopUpPrestigeText.SetShadowColor(ShardUIText.DefaultShadowColor)
        PopUpPrestigeText.SetShadowOffset(ShardUIText.DefaultShadowOffset)
        PopUpPrestigeText.SetShadowOpacity(1.0)
        PopUpHut1.SetVisibility(widget_visibility.Collapsed)
        PopUpHut2.SetVisibility(widget_visibility.Collapsed)
        PopUpHut3.SetVisibility(widget_visibility.Collapsed)
        CTAHUD.SetVisibility(widget_visibility.Collapsed)

        MyCanvas: canvas = canvas:
            Slots := array:
                canvas_slot:
                    ZOrder:=1
                    SizeToContent:=true
                    Alignment := vector2{X := 1.0, Y := 1.0}
                    Anchors := anchors{Minimum := vector2{X := 1.0, Y := 1.0}, Maximum := vector2{X := 1.0, Y := 1.0}}
                    Offsets:=margin{Top := -412.5,Left:=-6.0}
                    Widget := stack_box:
                        Orientation := orientation.Horizontal
                        Slots := array:
                            stack_box_slot:
                                Widget := AutoStoneIcon
                            stack_box_slot:
                                Widget := AutoStoneText
                canvas_slot:
                    ZOrder:=0
                    SizeToContent:=false
                    Alignment := vector2{X := 0.5, Y := 0.0}
                    Anchors := anchors{Minimum := vector2{X := 0.5, Y := 0.0}, Maximum := vector2{X := 0.5, Y := 0.0}}
                    Offsets:=margin{Top := 0.0,Left:=0.0,Bottom:=75.0,Right:=157.5}
                    Widget := texture_block:
                        DefaultImage := UI.MinerOdysseyCode

                canvas_slot:
                    ZOrder:=0
                    SizeToContent:=false
                    Alignment := vector2{X := 0.0, Y := 0.5}
                    Anchors := anchors{Minimum := vector2{X := 0.0, Y := 0.5}, Maximum := vector2{X := 0.0, Y := 0.5}}
                    Offsets:=margin{Top := 0.0,Left:=32.0,Bottom:=170.0,Right:=355.0}
                    Widget := CTAHUD

                canvas_slot:
                    ZOrder:=0
                    SizeToContent:=false
                    Alignment := vector2{X := 1.0, Y := 1.0}
                    Anchors := anchors{Minimum := vector2{X := 1.0, Y := 1.0}, Maximum := vector2{X := 1.0, Y := 1.0}}
                    Offsets:=margin{Top := -348.5,Left:=0.0,Bottom:=64.0,Right:=271.0}
                    Widget := texture_block:
                        DefaultImage := UI.stoneUI

                canvas_slot:
                    ZOrder:=1
                    SizeToContent:=false
                    Alignment := vector2{X := 1.0, Y := 1.0}
                    Anchors := anchors{Minimum := vector2{X := 1.0, Y := 1.0}, Maximum := vector2{X := 1.0, Y := 1.0}}
                    Offsets:=margin{Top := -399.5,Left:=0.0,Bottom:=0.0,Right:=195.0}
                    Widget := StoneUIText

                canvas_slot:
                    ZOrder:=0
                    SizeToContent:=false
                    Alignment := vector2{X := 1.0, Y := 1.0}
                    Anchors := anchors{Minimum := vector2{X := 1.0, Y := 1.0}, Maximum := vector2{X := 1.0, Y := 1.0}}
                    Offsets:=margin{Top := -268.5,Left:=0.0,Bottom:=64.0,Right:=271.0}
                    Widget := texture_block:
                        DefaultImage := UI.ShardUI

                canvas_slot:
                    ZOrder:=1
                    SizeToContent:=false
                    Alignment := vector2{X := 1.0, Y := 1.0}
                    Anchors := anchors{Minimum := vector2{X := 1.0, Y := 1.0}, Maximum := vector2{X := 1.0, Y := 1.0}}
                    Offsets:=margin{Top := -319.5,Left:=0.0,Bottom:=0.0,Right:=195.0}
                    Widget := ShardUIText
                
                canvas_slot:
                    ZOrder:=0
                    SizeToContent:=false
                    Anchors := anchors{Minimum := vector2{X := 0.0, Y := 1.0}, Maximum := vector2{X := 0.0, Y := 1.0}}
                    Offsets:=margin{Left:=0.0,Top:=-324.0,Right:=179.0,Bottom:=64.0}
                    Alignment := vector2{X := 0.0, Y := 0.0}
                    Widget := DrillUI

                canvas_slot:
                    Widget:=DrillUIText
                    ZOrder:=1
                    SizeToContent:=true
                    Anchors := anchors{Minimum := vector2{X := 0.0, Y := 1.0}, Maximum := vector2{X := 0.0, Y := 1.0}}
                    Offsets:=margin{Left:=110.0,Top:=-312.0}
                    Alignment := vector2{X := 0.5, Y := 0.0}

                canvas_slot:
                    Widget:=DrillActiveTimer
                    ZOrder:=0
                    SizeToContent:=true
                    Anchors := anchors{Minimum := vector2{X := 0.0, Y := 1.0}, Maximum := vector2{X := 0.0, Y := 1.0}}
                    Offsets:=margin{Left:=100.0,Top:=-364.0}
                    Alignment := vector2{X := 0.5, Y := 0.0}

                canvas_slot:
                    ZOrder:=1
                    SizeToContent:=false
                    Anchors := anchors{Minimum := vector2{X := 0.0, Y := 1.0}, Maximum := vector2{X := 0.0, Y := 1.0}}
                    Offsets:=margin{Left:=110.0,Top:=-260.0,Right:=179.0,Bottom:=64.0}
                    Alignment := vector2{X := 0.5, Y := 0.0}
                    Widget := stack_box:
                        Orientation := orientation.Horizontal
                        Slots := array:
                            stack_box_slot:
                                Widget := DrillChargeTimerIcon
                            stack_box_slot:
                                Widget := DrillChargeTimer

                #CAN UPGRADE POP UPS
                canvas_slot:
                    Widget := PopUpHut1
                    ZOrder:=0
                    SizeToContent:=false
                    Anchors := anchors{Minimum := vector2{X := 1.0, Y := 0.0}, Maximum := vector2{X := 1.0, Y := 0.0}}
                    Offsets:=margin{Left:=-664.0,Top:=0.0, Right:=166.0,Bottom:=120.0}
                    Alignment := vector2{X := 0.0, Y := 0.0}
                canvas_slot:
                    Widget := PopUpHut2
                    ZOrder:=0
                    SizeToContent:=false
                    Anchors := anchors{Minimum := vector2{X := 1.0, Y := 0.0}, Maximum := vector2{X := 1.0, Y := 0.0}}
                    Offsets:=margin{Left:=-496.0,Top:=0.0, Right:=166.0,Bottom:=120.0}
                    Alignment := vector2{X := 0.0, Y := 0.0}
                canvas_slot:
                    Widget := PopUpHut3
                    ZOrder:=0
                    SizeToContent:=false
                    Anchors := anchors{Minimum := vector2{X := 1.0, Y := 0.0}, Maximum := vector2{X := 1.0, Y := 0.0}}
                    Offsets:=margin{Left:=-328.0,Top:=0.0, Right:=166.0,Bottom:=120.0}
                    Alignment := vector2{X := 0.0, Y := 0.0}

                canvas_slot:
                    Widget := PopUpPrestige
                    ZOrder:=0
                    SizeToContent:=false
                    Anchors := anchors{Minimum := vector2{X := 1.0, Y := 0.0}, Maximum := vector2{X := 1.0, Y := 0.0}}
                    Offsets:=margin{Left:=-152.0,Top:=0.0, Right:=166.0,Bottom:=120.0}
                    Alignment := vector2{X := 0.0, Y := 0.0}
                canvas_slot:
                    Widget:=PopUpPrestigeText
                    ZOrder:=1
                    SizeToContent:=true
                    Anchors := anchors{Minimum := vector2{X := 1.0, Y := 0.0}, Maximum := vector2{X := 1.0, Y := 0.0}}
                    Offsets:=margin{Left:=-71.0,Top:=58.0}
                    Alignment := vector2{X := 0.5, Y := 0.0}
                        

    InitiateUI(): void=       #Initalizes and adds the players UI
        if(PlayerObj := Player, PlayerUI := GetPlayerUI[PlayerObj]):
            if(MainUI:=UIMaps["MainUI"]?): #If already exists don't overlap
            else:
                set MyUI = CreateMyUI()
                PlayerUI.AddWidget(MyUI)
                if(set UIMaps["MainUI"] = option{MyUI}) {}

    RemoveUI(): void=       #Removes UI on leave
        if(PlayerObj := Player, PlayerUI := GetPlayerUI[PlayerObj]):
            if(MainUI:=UIMaps["MainUI"]?):
                PlayerUI.RemoveWidget(MyUI)
                MyUI.SetVisibility(widget_visibility.Collapsed)
                if(set UIMaps["MainUI"] = false) {}
                
            for(Map:DamageMap): #Keeping total until it needs to be reset
                if(NumberUI:=Map?):
                    NumberUI.SetVisibility(widget_visibility.Collapsed)

    UpdateUI(): void=        #Updates the players UI to new values
        PlayerGatherRate:=GetPlayerDamage()
        if(PlayerGatherRate * AutoGatherRate > 0.0):
            AutoStoneIcon.SetVisibility(widget_visibility.Visible)
            if(PlayerGatherRate * AutoGatherRate > 1.0):
                AutoStoneText.SetText(StringToMessage(" +{NumberCondensor(PlayerGatherRate * AutoGatherRate)}/s"))
            else:
                AutoStoneText.SetText(StringToMessage(" <+1/s"))
        else:
            AutoStoneIcon.SetVisibility(widget_visibility.Collapsed)
            AutoStoneText.SetText(StringToMessage(""))
        StoneUIText.SetText(StringToMessage(NumberCondensor(StoneAmount)))
        ShardUIText.SetText(StringToMessage(NumberCondensor(ShardAmount)))

        #DRILL STUFF
        if(GoldenDrillUnlocked = true):
            DrillUI.SetVisibility(widget_visibility.Visible)
            DrillUIText.SetText(StringToMessage(NumberCondensor(BuffCurrentCharges)))
        else: #HIDE IF YOU DON'T HAVE DRILL UNLOCKED
            DrillUI.SetVisibility(widget_visibility.Collapsed)
            DrillUIText.SetText(StringToMessage(""))

        #POP UP HUD STUFF
        
        var NewPrestigeCost : float = 0.0
        if(PrestigeCost * Pow(PrestigeCostMult, (PrestigeLevel)*1.0) > 1e300):
            set NewPrestigeCost = 1e300
        else:
            set NewPrestigeCost = PrestigeCost * Pow(PrestigeCostMult, (PrestigeLevel)*1.0)
        if(PrestigeStoneAmount >= NewPrestigeCost):
            var PrestigeShardAmount : float = 0.0
            ratio := PrestigeStoneAmount / NewPrestigeCost
            
            if(PrestigeLevel <= 220):
                if(PrestigeLevel >= 1):
                    set PrestigeShardAmount = ((Log(5.0, ratio)) * 5) * (1.5 * PrestigeLevel)
                else:
                    set PrestigeShardAmount = ((Log(5.0, ratio)) * 5)
                if(PrestigeShardAmount < 1.0):
                    set PrestigeShardAmount = 1.0

                set PrestigeShardAmount = PrestigeShardAmount * (1.0 + SinkPrestigePower / 10000.0)
                PopUpPrestige.SetVisibility(widget_visibility.Visible)
                PopUpPrestigeText.SetText(StringToMessage("+{NumberCondensor(PrestigeShardAmount)}"))
            else:
                set PrestigeShardAmount = EndGamePrestigeShardAmount * (1.0 + SinkPrestigePower / 10000.0)
        else:
            PopUpPrestige.SetVisibility(widget_visibility.Collapsed)
            PopUpPrestigeText.SetText(StringToMessage(""))
        
        PopUpHut1.SetVisibility(widget_visibility.Collapsed)
        PopUpHut2.SetVisibility(widget_visibility.Collapsed)
        PopUpHut3.SetVisibility(widget_visibility.Collapsed)

        if(StagesUnlocked >= 0):
            for(i:int=0..AllHutMaxLevels[0].Length-1):
                if:
                    NewHutCost:=AllHutCosts[0][i]
                    HutLevel:=AllHutLevels[0][i]
                    NewCostMult:= Pow(AllHutCostMults[0][i], (HutLevel*1.0))
                    StoneAmount >= NewHutCost * NewCostMult
                    AllHutMaxLevels[0][i] > Hut1Levels[i]
                then:
                    PopUpHut1.SetVisibility(widget_visibility.Visible)

        if(StagesUnlocked >= 5):
            for(i:int=0..AllHutMaxLevels[1].Length-1):
                if:
                    NewHutCost:=AllHutCosts[1][i]
                    HutLevel:=AllHutLevels[1][i]
                    NewCostMult:= Pow(AllHutCostMults[1][i], (HutLevel*1.0))
                    StoneAmount >= NewHutCost * NewCostMult
                    AllHutMaxLevels[1][i] > HutLevel
                then:
                    PopUpHut2.SetVisibility(widget_visibility.Visible)

        if(StagesUnlocked >= 10):
            for(i:int=0..AllHutMaxLevels[2].Length-1):
                if:
                    NewHutCost:=AllHutCosts[2][i]
                    HutLevel:=AllHutLevels[2][i]
                    NewCostMult:= Pow(AllHutCostMults[2][i], (HutLevel*1.0))
                    StoneAmount >= NewHutCost * NewCostMult
                    AllHutMaxLevels[2][i] > HutLevel
                then:
                    PopUpHut3.SetVisibility(widget_visibility.Visible)
        
    AddRockHealthBarUI(Canvas : canvas):void=
        if(PlayerObj := Player, PlayerUI := GetPlayerUI[PlayerObj]):
            if(RockHPUI:=UIMaps["RockHP"]?): #If already exists don't overlap
                RockHPUI.SetVisibility(widget_visibility.Visible)
            else:
                PlayerUI.AddWidget(Canvas)
                if(set UIMaps["RockHP"] = option{Canvas}) {}

    RemoveRockHealthBarUI():void=
        if(PlayerObj := Player, PlayerUI := GetPlayerUI[PlayerObj]):
            if(RockHPUI:=UIMaps["RockHP"]?):
                RockHPUI.SetVisibility(widget_visibility.Collapsed)

    AddFullHUD(Canvas : canvas):void=
        if(PlayerObj := Player, PlayerUI := GetPlayerUI[PlayerObj]):
            PlayerUI.AddWidget(Canvas, player_ui_slot{InputMode:=ui_input_mode.All})
            Canvas.ForceRefresh()

    AddUI(Canvas : canvas):void=
        if(PlayerObj := Player, PlayerUI := GetPlayerUI[PlayerObj]):
            PlayerUI.AddWidget(Canvas)

    Collect(Amount : float, ResourceType: string, whichEgg:int): void=   #Increases a players resource amount by a specific value
        if(ResourceType = "Stone"):
            set StoneAmount += Amount
            set PrestigeStoneAmount += Amount
            UpAchValue(0, Amount)

            #HAVE TO HAVE THIS HERE TO MAKE SURE THE TUTORIAL ALWAYS UPDATES CORRECTLY
            if(TutorialAmount:=TutorialAmounts[TutorialLevel], TutorialLevel = 2 or TutorialLevel = 8 or TutorialLevel = 9 or TutorialLevel = 13 or TutorialLevel = 17):
                if(TutorialLevel = 17):
                    ProgressIcon.SetImage(UI.StoneIcon)
                    TutorialProgressText.SetText(StringToMessage(" {NumberCondensor(PrestigeStoneAmount)}/{NumberCondensor(TutorialAmount)}"))
                    if(PrestigeStoneAmount >= TutorialAmount):
                        TutorialProgressText.SetTextColor(MakeColorFromHex("39FF39"))
                    else:
                        TutorialProgressText.SetTextColor(MakeColorFromHex("fa4d56"))
                else if(TutorialLevel = 2 or TutorialLevel = 8 or TutorialLevel = 9 or TutorialLevel = 13):
                    ProgressIcon.SetImage(UI.StoneIcon)
                    TutorialProgressText.SetText(StringToMessage(" {NumberCondensor(StoneAmount)}/{NumberCondensor(TutorialAmount)}"))
                    if(StoneAmount >= TutorialAmount):
                        TutorialProgressText.SetTextColor(MakeColorFromHex("39FF39"))
                    else:
                        TutorialProgressText.SetTextColor(MakeColorFromHex("fa4d56"))

        else if(ResourceType = "Egg"):
            if(set EggAmounts[whichEgg] += Amount) {}

            if(TutorialLevel = 11, TutorialAmount:=TutorialAmounts[11]):
                var totalEggAmount : float = 0.0
                for(i:int=0..EggAmounts.Length):
                    if(EggAmounts[i] >= 1.0):
                        if(set totalEggAmount += EggAmounts[i]) {}
                ProgressIcon.SetImage(UI.Egg1)
                TutorialProgressText.SetText(StringToMessage(" {NumberCondensor(totalEggAmount)}/{NumberCondensor(TutorialAmount)}"))
                if(totalEggAmount >= TutorialAmount):
                    TutorialProgressText.SetTextColor(MakeColorFromHex("39FF39"))
                else:
                    TutorialProgressText.SetTextColor(MakeColorFromHex("fa4d56"))

        else if(ResourceType = "WepMat"):
            set WepMatAmount += Amount

            if(TutorialLevel = 16, TutorialAmount:=TutorialAmounts[16]):
                ProgressIcon.SetImage(UI.WepMat)
                TutorialProgressText.SetText(StringToMessage(" {NumberCondensor(WepMatAmount)}/{NumberCondensor(TutorialAmount)}"))
                if(WepMatAmount >= TutorialAmount):
                    TutorialProgressText.SetTextColor(MakeColorFromHex("39FF39"))
                else:
                    TutorialProgressText.SetTextColor(MakeColorFromHex("fa4d56"))

        else if(ResourceType = "Shard"):
            set ShardAmount += Amount
            UpAchValue(3, Amount)

        else if(ResourceType = "BossKey"):
            set BossKeyAmount += Amount

        else:
            Print("COLLECT Invalid Resource Name")

        #SaveAllData(CustomPlayer)
        #SaveResources(Player, StoneAmount, EggAmounts, WepMatAmount, ShardAmount, BossKeyAmount)
        UpdateUI()

    Spend(Amount : float, ResourceType: string, whichEgg:int): void=   #Increases a players resource amount by a specific value
        if(ResourceType = "Stone"):
            set StoneAmount -= Amount

            #HAVE TO HAVE THIS HERE TO MAKE SURE THE TUTORIAL ALWAYS UPDATES CORRECTLY
            if(TutorialAmount:=TutorialAmounts[TutorialLevel], TutorialLevel = 2 or TutorialLevel = 8 or TutorialLevel = 9 or TutorialLevel = 13 or TutorialLevel = 17):
                ProgressIcon.SetImage(UI.StoneIcon)
                if(TutorialLevel = 17):
                    TutorialProgressText.SetText(StringToMessage("{NumberCondensor(PrestigeStoneAmount)}/{NumberCondensor(TutorialAmount)}"))
                    if(PrestigeStoneAmount >= TutorialAmount):
                        TutorialProgressText.SetTextColor(MakeColorFromHex("39FF39"))
                    else:
                        TutorialProgressText.SetTextColor(MakeColorFromHex("fa4d56"))
                else:
                    TutorialProgressText.SetText(StringToMessage("{NumberCondensor(StoneAmount)}/{NumberCondensor(TutorialAmount)}"))
                    if(StoneAmount >= TutorialAmount):
                        TutorialProgressText.SetTextColor(MakeColorFromHex("39FF39"))
                    else:
                        TutorialProgressText.SetTextColor(MakeColorFromHex("fa4d56"))

        else if(ResourceType = "Egg"):
            if(set EggAmounts[whichEgg] -= Amount) {}

        else if(ResourceType = "WepMat"):
            set WepMatAmount -= Amount

        else if(ResourceType = "Shard"):
            set ShardAmount -= Amount

        else if(ResourceType = "BossKey"):
            set BossKeyAmount -= Amount

        else:
            Print("SPEND Invalid Resource Name")
        #SaveResources(Player, StoneAmount, EggAmounts, WepMatAmount, ShardAmount, BossKeyAmount)
        UpdateUI()

    SetAchValue(whichQuest:int, NewValue:float) :void=
        if(Value:=AllAchs[whichQuest][0],Target:=AllAchs[whichQuest][1]):
            if(set AllAchs[whichQuest][0] = NewValue){}
            if(AllAchs[whichQuest][0] >= Target):
                CompleteAch(whichQuest)
            #SaveAchievements(Player, AchLevels, AchBuffTotal, AllAchs)

    UpAchValue(whichQuest:int, ValueIncrease:float) :void=
        if(Value:=AllAchs[whichQuest][0],Target:=AllAchs[whichQuest][1]):
            if(set AllAchs[whichQuest][0] += ValueIncrease){}
            if(ValueIncrease + Value >= Target):
                CompleteAch(whichQuest)
            #SaveAchievements(Player, AchLevels, AchBuffTotal, AllAchs)

    UpAchTarget(whichQuest:int) :void=
        if(Increase:=AchIncreases[whichQuest], Target:=AllAchs[whichQuest][1]):
            if(whichQuest = 7): #ADD 30 MINUTES
                if(set AllAchs[whichQuest][1] += Increase){}
            else:
                if(set AllAchs[whichQuest][1] *= Increase){}
        
    CompleteAch(whichQuest:int) : void=
        if(Value:=AllAchs[whichQuest][0],Target:=AllAchs[whichQuest][1]):
            if(Value >= Target):
                set HarvestBuff += 0.01
                set AchBuffTotal += 0.01
                if(set AchLevels[whichQuest] += 1) {}
                UpAchTarget(whichQuest)
                #SaveAchievements(Player, AchLevels, AchBuffTotal, AllAchs)
                if(AchUI:=UIMaps["AchCompleted"]?):
                    if(CurrentTitle:=AchTitles[whichQuest]):
                        AchTitle.SetText(StringToMessage(CurrentTitle))
                    AchUI.SetVisibility(Visible := widget_visibility.Visible)
                    spawn{RemoveAchHUD(AchUI)}
                else:
                    var AchCompletedUI : canvas = CreateAchUI(whichQuest)
                    AddUI(AchCompletedUI)
                    spawn{RemoveAchHUD(AchCompletedUI)}
                    if(set UIMaps["AchCompleted"] = option{AchCompletedUI}) {}

    RemoveAchHUD(AchCompletedUI:canvas)<suspends>:void=
        Sleep(5.0)
        AchCompletedUI.SetVisibility(Collapsed := widget_visibility.Collapsed)

    CreateAchUI(whichQuest:int) : canvas =   #Creates the Ach Hud UI
        if(CurrentTitle:=AchTitles[whichQuest]):
            AchTitle.SetText(StringToMessage(CurrentTitle))
        AchTitle.SetShadowColor(StoneUIText.DefaultShadowColor)
        AchTitle.SetShadowOffset(StoneUIText.DefaultShadowOffset)
        AchTitle.SetShadowOpacity(1.0)

        GatherRateBonus.SetText(StringToMessage("+1% Strength"))
        GatherRateBonus.SetShadowColor(StoneUIText.DefaultShadowColor)
        GatherRateBonus.SetShadowOffset(StoneUIText.DefaultShadowOffset)
        GatherRateBonus.SetShadowOpacity(1.0)

        MyCanvas: canvas = canvas:
            Slots := array:
                canvas_slot:
                    ZOrder:=1
                    SizeToContent:=false
                    Alignment := vector2{X := 0.5, Y := 1.0}
                    Anchors := anchors{Minimum := vector2{X := 0.5, Y := 1.0}, Maximum := vector2{X := 0.5, Y := 1.0}}
                    Offsets:=margin{Top := -20.0,Left:=0.0,Bottom:=250.0,Right:=500.0}
                    Widget := texture_block:
                        DefaultImage := UI.achCompleted
                canvas_slot:
                    ZOrder:=2
                    SizeToContent:=true
                    Anchors := anchors{Minimum := vector2{X := 0.5, Y := 1.0}, Maximum := vector2{X := 0.5, Y := 1.0}}
                    Alignment := vector2{X := 0.5, Y := 0.0}
                    Offsets:=margin{Top := -220.0,Left:=0.0}
                    Widget := AchTitle
                canvas_slot:
                    ZOrder:=2
                    SizeToContent:=true
                    Anchors := anchors{Minimum := vector2{X := 0.5, Y := 1.0}, Maximum := vector2{X := 0.5, Y := 1.0}}
                    Alignment := vector2{X := 0.5, Y := 0.0}
                    Offsets:=margin{Top := -135.0,Left:=75.0}
                    Widget := GatherRateBonus


#EXTERNAL UI GARBAGE
    ShadowDefaults : text_block = text_block:
        DefaultTextColor := MakeColorFromHex("FFFFFF")
        DefaultShadowOffset := option{vector2{X:=3.0, Y:=3.0}}
        DefaultShadowColor := MakeColorFromHex("000000")
    WhiteShadowDefaults : text_block = text_block:
        DefaultTextColor := MakeColorFromHex("FFFFFF")
        DefaultShadowOffset := option{vector2{X:=2.0, Y:=2.0}}
        DefaultShadowColor := MakeColorFromHex("FFFFFF")
#HEALTH_BAR
    var RockBarColorBase : color_block = color_block{
        DefaultColor := MakeColorFromHex("000000")
        DefaultDesiredSize := vector2{X:= 600.0, Y := 40.0}} 
    var HPBarFinalStrikeAmount : text_block = text_block{DefaultTextColor := MakeColorFromHex("FFFFFF")}
    var HPBarFinalStrikeIcon : texture_block = texture_block{DefaultImage := UI.FinalStrikeRock}

    var iTotalHP : float = 1000.0
    var iCurrentHP : float = 1000.0
    var iRockName : string = ""

    var RockBarColorTop : color_block = color_block{}  
    var RockBarColorTop2 : color_block = color_block{}  
    var RockBarColorTopSmooth : color_block = color_block{}  

    var HPBarTitleTextBlock : text_block = text_block{DefaultTextColor := MakeColorFromHex("FFFFFF")}
    var SmoothNewHealthBarWidth : float = 0.0
    
    NewProgressUpdatePolledEvent : event() = event(){}

#BOSS_ROOM_MANAGER DONE AND TESTED
    var BossTimerHUDBG : texture_block = texture_block{DefaultImage := UI.BossTimerBG}
    var BossTimerText : text_block = text_block{DefaultTextColor := MakeColorFromHex("FFFFFF")}

    var BossCloseButton : button_loud = button_loud{}
    var BossClose_Position : vector2 = vector2{X:=1710.0, Y:=100.0}
    var BossClose_Size : vector2 = vector2{X:=76.0, Y:= 56.0}

    var BossHUDBG : texture_block = texture_block{DefaultImage := UI.BOSSHUDBG}
    var BossHUD_Position : vector2 = vector2{X:= 0.0, Y:= 0.0}
    var BossHUDTitle : texture_block = texture_block{DefaultImage := UI.BOSS_FIGHT}

    #ARRAYS
    var QueuedPlayerBlocks : []text_block = array{}
    var AllTextBlocks : []text_block = array{}

    #EACH VALUE      
    var ActivePlayers : int = 0
    var BossKeyIcon : texture_block = texture_block{DefaultImage := UI.BossKeyIcon}
    var BossKeyText : text_block = text_block{DefaultTextColor := MakeColorFromHex("FFFFFF")}

    var BossIcon : texture_block = texture_block{DefaultImage := UI.Boss1}
    var BossNameText : text_block = text_block{DefaultTextColor := MakeColorFromHex("FFFFFF")}
    var BossHealth : text_block = text_block{DefaultTextColor := MakeColorFromHex("FFFFFF")}
    var BossWarning : text_block = text_block{DefaultTextColor := MakeColorFromHex("FA4D56")}
    var BossKeyCostIcon : texture_block = texture_block{DefaultImage := UI.BossKeyIcon}
    var BossKeyCostText : text_block = text_block{DefaultTextColor := MakeColorFromHex("FFFFFF")}

    var RewardTitle : text_block = text_block{DefaultTextColor := MakeColorFromHex("FFFFFF")}
    var Reward1Text : text_block = text_block{DefaultTextColor := MakeColorFromHex("FFFFFF")}
    var Reward1Icon : texture_block = texture_block{DefaultImage := UI.StoneIcon}
    var Reward2Text : text_block = text_block{DefaultTextColor := MakeColorFromHex("FFFFFF")}
    var Reward2Icon : texture_block = texture_block{DefaultImage := UI.WepMat}
    var Reward3Text : text_block = text_block{DefaultTextColor := MakeColorFromHex("FFFFFF")}
    var Reward3Icon : texture_block = texture_block{DefaultImage := UI.Egg1}

    var QueueStatus : text_block = text_block{DefaultTextColor := MakeColorFromHex("FFFFFF")}
    var QueuedPlayersTitle : text_block = text_block{DefaultTextColor := MakeColorFromHex("FFFFFF")}
    var QueuedPlayersCount : text_block = text_block{DefaultTextColor := MakeColorFromHex("FFFFFF")}
    var QueuedPlayer1 : text_block = text_block{DefaultTextColor := MakeColorFromHex("DDDDDD")}
    var QueuedPlayer2 : text_block = text_block{DefaultTextColor := MakeColorFromHex("DDDDDD")}
    var QueuedPlayer3 : text_block = text_block{DefaultTextColor := MakeColorFromHex("DDDDDD")}
    var QueuedPlayer4 : text_block = text_block{DefaultTextColor := MakeColorFromHex("DDDDDD")}
    var MessageText : text_block = text_block{DefaultTextColor := MakeColorFromHex("FFFFFF")}
        
    var JoinButton : button_regular = button_regular{}
    var LeaveButton : button_quiet = button_quiet{}
    var CallHelpButton : button_quiet = button_quiet{}
    var HardModeButton : button_quiet = button_quiet{}
#UI_DAMAGE_NUMBERS DONE AND TESTED
    var TextBlockTotal : text_block = text_block{DefaultTextColor := MakeColorFromHex("FFFFFF")}
    var TextureTotal : texture_block = texture_block{DefaultImage := UI.StoneIcon, DefaultDesiredSize := vector2{X := 50.0, Y:= 50.0}}
    var WepTextBlockTotal : text_block = text_block{DefaultTextColor := MakeColorFromHex("FFFFFF")}
    var WepTextureTotal : texture_block = texture_block{DefaultImage := UI.statAttack, DefaultDesiredSize := vector2{X := 50.0, Y:= 50.0}}
    var TextBlockCombo : text_block = text_block{DefaultTextColor := MakeColorFromHex("FFFFFF")}
    var TextureCombo : texture_block = texture_block{DefaultImage := UI.ComboStrike, DefaultDesiredSize := vector2{X := 50.0, Y:= 50.0}}
    var TextBlockEQ : text_block = text_block{DefaultTextColor := MakeColorFromHex("FFFFFF")}
    var TextureEQ : texture_block = texture_block{DefaultImage := UI.Earthquake, DefaultDesiredSize := vector2{X := 50.0, Y:= 50.0}}
    var TextBlockVoid : text_block = text_block{DefaultTextColor := MakeColorFromHex("FFFFFF")}
    var TextureVoid : texture_block = texture_block{DefaultImage := UI.VoidFracture, DefaultDesiredSize := vector2{X := 50.0, Y:= 50.0}}
    var TextBlockMissile : text_block = text_block{DefaultTextColor := MakeColorFromHex("FFFFFF")}
    var TextureMissile : texture_block = texture_block{DefaultImage := UI.FireUnlock, DefaultDesiredSize := vector2{X := 50.0, Y:= 50.0}}

#GOLDEN_DRILL_MANAGER DONE AND TESTED
    #HUD STUFF
    var GDCloseButton : button_loud = button_loud{}
    var GDClose_Position : vector2 = vector2{X:=200.0, Y:=976.0}
    var GDClose_Size : vector2 = vector2{X:=200.0, Y:= 50.0}

    var GDHUDBG : texture_block = texture_block{DefaultImage := UI.HUPHUDBG}
    var GDHUD_Position : vector2 = vector2{X:= 0.0, Y:= 0.0}
    var GDTotalDMG : text_block = text_block{DefaultTextColor := MakeColorFromHex("FFFFFF")}
    var GDHUDBGCover : texture_block = texture_block{DefaultImage := UI.HUPHUDBGCover}
    var GDHUDTitle : texture_block = texture_block{DefaultImage := UI.DRILL_UPGRADES}

    #ARRAYS
    var StoneIcons : []texture_block = array{}
    var GDUpgradeTitles : []text_block = array{}
    var GDUpgradeLevels : []text_block = array{}
    var GDUpgradeIcons : []texture_block = array{}
    var GDUpgradeDescTexts : []text_block = array{}
    var GDUpgradeBonuses : []text_block = array{}
    var GDUpgradeNextBonuses : []text_block = array{}
    var GDUpgradeCosts : []text_block = array{}
    var GDUpgradeCostTexts : []text_block = array{}
    var GDUpgradeCostIcons : []text_block = array{}
    var GDUpgradeButtons : []button_quiet = array{}
    var GDAllTexts : []text_block = array{}

    #EACH VALUE
    var StoneIcon1 : texture_block = texture_block{DefaultImage := UI.StoneIcon}
    var StoneIcon2 : texture_block = texture_block{DefaultImage := UI.StoneIcon}
    var StoneIcon3 : texture_block = texture_block{DefaultImage := UI.StoneIcon}
    var StoneIcon4 : texture_block = texture_block{DefaultImage := UI.StoneIcon}
    var StoneAmountIcon : texture_block = texture_block{DefaultImage := UI.StoneIcon}
    var StoneAmountText : text_block = text_block{DefaultTextColor := MakeColorFromHex("FFFFFF")}
      
    var Upgrade1Title : text_block = text_block{DefaultTextColor := MakeColorFromHex("FFFFFF")}
    var Upgrade1Level : text_block = text_block{DefaultTextColor := MakeColorFromHex("FFFFFF")}
    var Upgrade1Icon : texture_block = texture_block{DefaultImage := UI.miningdrill}
    var Upgrade1Desc : text_block = text_block{DefaultTextColor := MakeColorFromHex("DDDDDD")}
    var Upgrade1Bonus : text_block = text_block{DefaultTextColor := MakeColorFromHex("FFFFFF")}
    var Upgrade1NextBonus : text_block = text_block{DefaultTextColor := MakeColorFromHex("39FF39")}
    var Upgrade1CostText : text_block = text_block{DefaultTextColor := MakeColorFromHex("FFFFFF")}
    var Upgrade1CostIcon : texture_block = texture_block{DefaultImage := UI.ShardIcon}
    var Upgrade1CostAmount : text_block = text_block{DefaultTextColor := MakeColorFromHex("FFFFFF")}
    var GDUpgradeButton1 : button_quiet = button_quiet{}

    var Upgrade2Title : text_block = text_block{DefaultTextColor := MakeColorFromHex("FFFFFF")}
    var Upgrade2Level : text_block = text_block{DefaultTextColor := MakeColorFromHex("FFFFFF")}
    var Upgrade2Icon : texture_block = texture_block{DefaultImage := UI.miningpowerup}
    var Upgrade2Desc : text_block = text_block{DefaultTextColor := MakeColorFromHex("DDDDDD")}
    var Upgrade2Bonus : text_block = text_block{DefaultTextColor := MakeColorFromHex("FFFFFF")}
    var Upgrade2NextBonus : text_block = text_block{DefaultTextColor := MakeColorFromHex("39FF39")}
    var Upgrade2CostText : text_block = text_block{DefaultTextColor := MakeColorFromHex("FFFFFF")}
    var Upgrade2CostIcon : texture_block = texture_block{DefaultImage := UI.ShardIcon}
    var Upgrade2CostAmount : text_block = text_block{DefaultTextColor := MakeColorFromHex("FFFFFF")}
    var GDUpgradeButton2 : button_quiet = button_quiet{}

    var Upgrade3Title : text_block = text_block{DefaultTextColor := MakeColorFromHex("FFFFFF")}
    var Upgrade3Level : text_block = text_block{DefaultTextColor := MakeColorFromHex("FFFFFF")}
    var Upgrade3Icon : texture_block = texture_block{DefaultImage := UI.drilltime}
    var Upgrade3Desc : text_block = text_block{DefaultTextColor := MakeColorFromHex("DDDDDD")}
    var Upgrade3Bonus : text_block = text_block{DefaultTextColor := MakeColorFromHex("FFFFFF")}
    var Upgrade3NextBonus : text_block = text_block{DefaultTextColor := MakeColorFromHex("39FF39")}
    var Upgrade3CostText : text_block = text_block{DefaultTextColor := MakeColorFromHex("FFFFFF")}
    var Upgrade3CostIcon : texture_block = texture_block{DefaultImage := UI.ShardIcon}
    var Upgrade3CostAmount : text_block = text_block{DefaultTextColor := MakeColorFromHex("FFFFFF")}
    var GDUpgradeButton3 : button_quiet = button_quiet{}

    var Upgrade4Title : text_block = text_block{DefaultTextColor := MakeColorFromHex("FFFFFF")}
    var Upgrade4Level : text_block = text_block{DefaultTextColor := MakeColorFromHex("FFFFFF")}
    var Upgrade4Icon : texture_block = texture_block{DefaultImage := UI.drillcharges}
    var Upgrade4Desc : text_block = text_block{DefaultTextColor := MakeColorFromHex("DDDDDD")}
    var Upgrade4Bonus : text_block = text_block{DefaultTextColor := MakeColorFromHex("FFFFFF")}
    var Upgrade4NextBonus : text_block = text_block{DefaultTextColor := MakeColorFromHex("39FF39")}
    var Upgrade4CostText : text_block = text_block{DefaultTextColor := MakeColorFromHex("FFFFFF")}
    var Upgrade4CostIcon : texture_block = texture_block{DefaultImage := UI.ShardIcon}
    var Upgrade4CostAmount : text_block = text_block{DefaultTextColor := MakeColorFromHex("FFFFFF")}
    var GDUpgradeButton4 : button_quiet = button_quiet{}

#PET MUTATOR UI
    var PetMBG : texture_block = texture_block{DefaultImage := UI.PetMutationBG}
    var PetMCloseButton : button_loud = button_loud{}
    var PetMTitle : texture_block = texture_block{DefaultImage := UI.PET_MUTATIONS}

    # LEFT SECTION

    var PetMStoneAmount : text_block = text_block:
        DefaultTextColor := MakeColorFromHex("FFFFFF")
        DefaultShadowOffset := option{vector2{X:=3.0, Y:=3.0}}
        DefaultShadowColor := MakeColorFromHex("000000")
        DefaultShadowOpacity := 1.0
    var PetMPrestigeAmount : text_block = text_block:
        DefaultTextColor := MakeColorFromHex("FFFFFF")
        DefaultShadowOffset := option{vector2{X:=3.0, Y:=3.0}}
        DefaultShadowColor := MakeColorFromHex("000000")
        DefaultShadowOpacity := 1.0
    var PetMEgg1Amount : text_block = text_block:
        DefaultTextColor := MakeColorFromHex("FFFFFF")
        DefaultShadowOffset := option{vector2{X:=3.0, Y:=3.0}}
        DefaultShadowColor := MakeColorFromHex("000000")
        DefaultShadowOpacity := 1.0
    var PetMEgg2Amount : text_block = text_block:
        DefaultTextColor := MakeColorFromHex("FFFFFF")
        DefaultShadowOffset := option{vector2{X:=3.0, Y:=3.0}}
        DefaultShadowColor := MakeColorFromHex("000000")
        DefaultShadowOpacity := 1.0

    var MutateEggAmount : []text_block = array{} 
    

    var PetMStoneIcon : texture_block = texture_block{DefaultImage := UI.StoneIcon}
    var PetMPrestigeIcon : texture_block = texture_block{DefaultImage := UI.ShardIcon}
    var PetEgg1con : texture_block = texture_block{DefaultImage := UI.Egg1}
    var PetMEgg2Icon : texture_block = texture_block{DefaultImage := UI.Egg2}

    # MIDDLE SECTION

    var PetMPet1Icon : texture_block = texture_block{DefaultImage := UI.pet1}
    var PetMPet2Icon : texture_block = texture_block{DefaultImage := UI.pet2}

    var PetM1LevelTitle : text_block = text_block{DefaultTextColor := MakeColorFromHex("FFFFFF"),DefaultShadowOffset := option{vector2{X:=3.0, Y:=3.0}},DefaultShadowColor := MakeColorFromHex("000000"),DefaultShadowOpacity := 1.0}
    var PetM2LevelTitle : text_block = text_block{DefaultTextColor := MakeColorFromHex("FFFFFF"),DefaultShadowOffset := option{vector2{X:=3.0, Y:=3.0}},DefaultShadowColor := MakeColorFromHex("000000"),DefaultShadowOpacity := 1.0}
    var PetM1TierTitle : text_block = text_block{DefaultTextColor := MakeColorFromHex("FFFFFF"),DefaultShadowOffset := option{vector2{X:=3.0, Y:=3.0}},DefaultShadowColor := MakeColorFromHex("000000"),DefaultShadowOpacity := 1.0}
    var PetM2TierTitle : text_block = text_block{DefaultTextColor := MakeColorFromHex("FFFFFF"),DefaultShadowOffset := option{vector2{X:=3.0, Y:=3.0}},DefaultShadowColor := MakeColorFromHex("000000"),DefaultShadowOpacity := 1.0}

    var PetM1ViewButton : button_quiet = button_quiet{}
    var PetM2ViewButton : button_quiet = button_quiet{}

    # TOP RIGHT SECTION

    var PetMSPetStatsIcon : texture_block = texture_block{DefaultImage := UI.pet1}
    var PetMGatherIcon : texture_block = texture_block{DefaultImage := UI.AchMobsDefeated}
    var PetMStrengthIcon : texture_block = texture_block{DefaultImage := UI.AchMaxSTR}
    var PetMCritRateIcon : texture_block = texture_block{DefaultImage := UI.CritChance}
    var PetMCritDMGIcon : texture_block = texture_block{DefaultImage := UI.CritDamage}
    var PetMFSIcon : texture_block = texture_block{DefaultImage := UI.FinalStrike}

    var PetMGatherTitle : text_block = text_block{DefaultTextColor := MakeColorFromHex("FFFFFF"),DefaultShadowOffset := option{vector2{X:=3.0, Y:=3.0}},DefaultShadowColor := MakeColorFromHex("000000"),DefaultShadowOpacity := 1.0}
    var PetMGatherAmount : text_block = text_block{DefaultTextColor := MakeColorFromHex("39FF39"),DefaultShadowOffset := option{vector2{X:=3.0, Y:=3.0}},DefaultShadowColor := MakeColorFromHex("000000"),DefaultShadowOpacity := 1.0}

    var PetMStrengthTitle : text_block = text_block{DefaultTextColor := MakeColorFromHex("FFFFFF"),DefaultShadowOffset := option{vector2{X:=3.0, Y:=3.0}},DefaultShadowColor := MakeColorFromHex("000000"),DefaultShadowOpacity := 1.0}
    var PetMStrengthAmount : text_block = text_block{DefaultTextColor := MakeColorFromHex("39FF39"),DefaultShadowOffset := option{vector2{X:=3.0, Y:=3.0}},DefaultShadowColor := MakeColorFromHex("000000"),DefaultShadowOpacity := 1.0}

    var PetMCritRateTitle : text_block = text_block{DefaultTextColor := MakeColorFromHex("FFFFFF"),DefaultShadowOffset := option{vector2{X:=3.0, Y:=3.0}},DefaultShadowColor := MakeColorFromHex("000000"),DefaultShadowOpacity := 1.0}
    var PetMCritRateAmount : text_block = text_block{DefaultTextColor := MakeColorFromHex("39FF39"),DefaultShadowOffset := option{vector2{X:=3.0, Y:=3.0}},DefaultShadowColor := MakeColorFromHex("000000"),DefaultShadowOpacity := 1.0}

    var PetMCritDMGTitle : text_block = text_block{DefaultTextColor := MakeColorFromHex("FFFFFF"),DefaultShadowOffset := option{vector2{X:=3.0, Y:=3.0}},DefaultShadowColor := MakeColorFromHex("000000"),DefaultShadowOpacity := 1.0}
    var PetMCritDMGAmount : text_block = text_block{DefaultTextColor := MakeColorFromHex("39FF39"),DefaultShadowOffset := option{vector2{X:=3.0, Y:=3.0}},DefaultShadowColor := MakeColorFromHex("000000"),DefaultShadowOpacity := 1.0}

    var PetMFSTitle : text_block = text_block{DefaultTextColor := MakeColorFromHex("FFFFFF"),DefaultShadowOffset := option{vector2{X:=3.0, Y:=3.0}},DefaultShadowColor := MakeColorFromHex("000000"),DefaultShadowOpacity := 1.0}
    var PetMFSAmounnt : text_block = text_block{DefaultTextColor := MakeColorFromHex("39FF39"),DefaultShadowOffset := option{vector2{X:=3.0, Y:=3.0}},DefaultShadowColor := MakeColorFromHex("000000"),DefaultShadowOpacity := 1.0}

    # BOTTOM RIGHT SECTION

    var PetMStoneCostIcon : texture_block = texture_block{DefaultImage := UI.StoneIcon}
    var PetMPrestigeCostIcon : texture_block = texture_block{DefaultImage := UI.ShardIcon}
    var PetMEggCostIcon : texture_block = texture_block{DefaultImage := UI.Egg1}
    var PetMAllStoneCostIcon : texture_block = texture_block{DefaultImage := UI.StoneIcon}
    var PetMAllPrestigeCostIcon : texture_block = texture_block{DefaultImage := UI.ShardIcon}
    var PetMAllEggCostIcon : texture_block = texture_block{DefaultImage := UI.Egg1}
    var PetMWhichEggIcon : texture_block = texture_block{DefaultImage := UI.Egg1}

    var PetMCostTitle : text_block = text_block{DefaultTextColor := MakeColorFromHex("FFFFFF"),DefaultShadowOffset := option{vector2{X:=3.0, Y:=3.0}},DefaultShadowColor := MakeColorFromHex("000000"),DefaultShadowOpacity := 1.0}

    var PetMStoneCostTitle : text_block = text_block{DefaultTextColor := MakeColorFromHex("FFFFFF"),DefaultShadowOffset := option{vector2{X:=3.0, Y:=3.0}},DefaultShadowColor := MakeColorFromHex("000000"),DefaultShadowOpacity := 1.0}
    var PetMPrestigeCostTitle : text_block = text_block{DefaultTextColor := MakeColorFromHex("FFFFFF"),DefaultShadowOffset := option{vector2{X:=3.0, Y:=3.0}},DefaultShadowColor := MakeColorFromHex("000000"),DefaultShadowOpacity := 1.0}
    var PetMEggCostTitle : text_block = text_block{DefaultTextColor := MakeColorFromHex("FFFFFF"),DefaultShadowOffset := option{vector2{X:=3.0, Y:=3.0}},DefaultShadowColor := MakeColorFromHex("000000"),DefaultShadowOpacity := 1.0}

    var PetMAllStoneCostTitle : text_block = text_block{DefaultTextColor := MakeColorFromHex("FFFFFF"),DefaultShadowOffset := option{vector2{X:=3.0, Y:=3.0}},DefaultShadowColor := MakeColorFromHex("000000"),DefaultShadowOpacity := 1.0}
    var PetMAllPrestigeCostTitle : text_block = text_block{DefaultTextColor := MakeColorFromHex("FFFFFF"),DefaultShadowOffset := option{vector2{X:=3.0, Y:=3.0}},DefaultShadowColor := MakeColorFromHex("000000"),DefaultShadowOpacity := 1.0}
    var PetMAllEggCostTitle : text_block = text_block{DefaultTextColor := MakeColorFromHex("FFFFFF"),DefaultShadowOffset := option{vector2{X:=3.0, Y:=3.0}},DefaultShadowColor := MakeColorFromHex("000000"),DefaultShadowOpacity := 1.0}

    var PetMBackButton : button_quiet = button_quiet{}
    var PetMForwardButton : button_quiet = button_quiet{}
    var PetMHatchButton : button_quiet = button_quiet{}
    var PetMHatchAllButton : button_quiet = button_quiet{}
    var PetHatchBG : texture_block = texture_block:
        DefaultImage := UI.TransparentButton
    var PetHatchAllBG : texture_block = texture_block:
        DefaultImage := UI.TransparentButton

    var PetHatchMessage : text_block = text_block{DefaultTextColor := MakeColorFromHex("FFFFFF"),DefaultShadowOffset := option{vector2{X:=3.0, Y:=3.0}},DefaultShadowColor := MakeColorFromHex("000000"),DefaultShadowOpacity := 1.0}



#PET SHOP DONE
    var PSAllTexts : []text_block = array{}
    var EggAmountsText : []text_block = array{}
    var PetLevelsText : []text_block = array{}

    var PetShopCanvas : canvas = canvas{}
    var petShopBackGround : texture_block = texture_block{DefaultImage := UI.PetShopHUDBG, DefaultDesiredSize := vector2{X := 1920.0, Y:= 1080.0}}
    var PetShopTitle : texture_block = texture_block{DefaultImage := UI.PET_SHOP, DefaultDesiredSize := vector2{X := 301.0, Y:= 56.0}}
    var PetShopCloseButton : button_loud = button_loud{}
    var PetShopClose_Position : vector2 = vector2{X:=1468.0, Y:=112.0}
    var PetShopClose_Size : vector2 = vector2{X:=76.0, Y:= 56.0}

    var InvTitle : texture_block = texture_block{DefaultImage := UI.INVENTORY, DefaultDesiredSize := vector2{X := 198.0, Y:= 38.0}}
    var InvStone : text_block = text_block{DefaultTextColor := MakeColorFromHex("FFFFFF")}
    var InvStoneIcon : texture_block = texture_block{DefaultImage := UI.StoneIcon, DefaultDesiredSize := vector2{X := 40.0, Y:= 40.0}}
    var InvEgg1 : text_block = text_block{DefaultTextColor := MakeColorFromHex("FFFFFF")}
    var InvEgg1Icon : texture_block = texture_block{DefaultImage := UI.Egg1, DefaultDesiredSize := vector2{X := 40.0, Y:= 40.0}}
    var InvEgg2 : text_block = text_block{DefaultTextColor := MakeColorFromHex("FFFFFF")}
    var InvEgg2Icon : texture_block = texture_block{DefaultImage := UI.Egg2, DefaultDesiredSize := vector2{X := 40.0, Y:= 40.0}}
    var InvEgg3 : text_block = text_block{DefaultTextColor := MakeColorFromHex("FFFFFF")}
    var InvEgg3Icon : texture_block = texture_block{DefaultImage := UI.Egg3, DefaultDesiredSize := vector2{X := 40.0, Y:= 40.0}}
    var InvEgg4 : text_block = text_block{DefaultTextColor := MakeColorFromHex("FFFFFF")}
    var InvEgg4Icon : texture_block = texture_block{DefaultImage := UI.Egg4, DefaultDesiredSize := vector2{X := 40.0, Y:= 40.0}}
    var InvEgg5 : text_block = text_block{DefaultTextColor := MakeColorFromHex("FFFFFF")}
    var InvEgg5Icon : texture_block = texture_block{DefaultImage := UI.Egg5, DefaultDesiredSize := vector2{X := 40.0, Y:= 40.0}}
    var InvEgg6 : text_block = text_block{DefaultTextColor := MakeColorFromHex("FFFFFF")}
    var InvEgg6Icon : texture_block = texture_block{DefaultImage := UI.Egg6, DefaultDesiredSize := vector2{X := 40.0, Y:= 40.0}}
    var InvWepMat : text_block = text_block{DefaultTextColor := MakeColorFromHex("FFFFFF")}
    var InvWepMatIcon : texture_block = texture_block{DefaultImage := UI.WepMat, DefaultDesiredSize := vector2{X := 40.0, Y:= 40.0}}

    var PetsShopTitle : text_block = text_block{DefaultTextColor := MakeColorFromHex("FFFFFF")}
    var Pet1Icon : texture_block = texture_block{DefaultImage := UI.pet1, DefaultDesiredSize := vector2{X := 100.0, Y:= 100.0}}
    var Pet1Level : text_block = text_block{DefaultTextColor := MakeColorFromHex("FFFFFF")}
    var Pet1Tier : text_block = text_block{DefaultTextColor := MakeColorFromHex("DDDDDD")}
    var Pet1MenuButton : button_quiet = button_quiet{}
    var Pet2Icon : texture_block = texture_block{DefaultImage := UI.pet2, DefaultDesiredSize := vector2{X := 100.0, Y:= 100.0}}
    var Pet2Level : text_block = text_block{DefaultTextColor := MakeColorFromHex("FFFFFF")}
    var Pet2Tier : text_block = text_block{DefaultTextColor := MakeColorFromHex("b3f069")}
    var Pet2MenuButton : button_quiet = button_quiet{}
    var Pet3Icon : texture_block = texture_block{DefaultImage := UI.pet3, DefaultDesiredSize := vector2{X := 100.0, Y:= 100.0}}
    var Pet3Level : text_block = text_block{DefaultTextColor := MakeColorFromHex("FFFFFF")}
    var Pet3Tier : text_block = text_block{DefaultTextColor := MakeColorFromHex("56ccd3")}
    var Pet3MenuButton : button_quiet = button_quiet{}
    var Pet4Icon : texture_block = texture_block{DefaultImage := UI.pet4, DefaultDesiredSize := vector2{X := 100.0, Y:= 100.0}}
    var Pet4Level : text_block = text_block{DefaultTextColor := MakeColorFromHex("FFFFFF")}
    var Pet4Tier : text_block = text_block{DefaultTextColor := MakeColorFromHex("c167da")}
    var Pet4MenuButton : button_quiet = button_quiet{}
    var Pet5Icon : texture_block = texture_block{DefaultImage := UI.pet5, DefaultDesiredSize := vector2{X := 100.0, Y:= 100.0}}
    var Pet5Level : text_block = text_block{DefaultTextColor := MakeColorFromHex("FFFFFF")}
    var Pet5Tier : text_block = text_block{DefaultTextColor := MakeColorFromHex("dda24f")}
    var Pet5MenuButton : button_quiet = button_quiet{}
    var Pet6Icon : texture_block = texture_block{DefaultImage := UI.pet6, DefaultDesiredSize := vector2{X := 100.0, Y:= 100.0}}
    var Pet6Level : text_block = text_block{DefaultTextColor := MakeColorFromHex("FFFFFF")}
    var Pet6Tier : text_block = text_block{DefaultTextColor := MakeColorFromHex("eecf56")}
    var Pet6MenuButton : button_quiet = button_quiet{}

    var PetDisplayIMG : texture_block = texture_block{DefaultImage := UI.pet1, DefaultDesiredSize := vector2{X := 200.0, Y:= 200.0}}
    var PetGatherIcon : texture_block = texture_block{DefaultImage := UI.AchMobsDefeated, DefaultDesiredSize := vector2{X := 55.0, Y:= 55.0}}
    var PetGatherText : text_block = text_block{DefaultTextColor := MakeColorFromHex("FFFFFF")}
    var PetGatherAmount : text_block = text_block{DefaultTextColor := MakeColorFromHex("39FF39")}
    var StrengthIcon : texture_block = texture_block{DefaultImage := UI.statAttack, DefaultDesiredSize := vector2{X := 55.0, Y:= 55.0}}
    var StrengthText : text_block = text_block{DefaultTextColor := MakeColorFromHex("FFFFFF")}
    var AttackAmount : text_block = text_block{DefaultTextColor := MakeColorFromHex("39FF39")}
    var CritRateIcon : texture_block = texture_block{DefaultImage := UI.CritChance, DefaultDesiredSize := vector2{X := 50.0, Y:= 50.0}}
    var CritRateText : text_block = text_block{DefaultTextColor := MakeColorFromHex("FFFFFF")}
    var CritRateAmount : text_block = text_block{DefaultTextColor := MakeColorFromHex("39FF39")}
    var CritDMGIcon : texture_block = texture_block{DefaultImage := UI.CritDamage, DefaultDesiredSize := vector2{X := 55.0, Y:= 55.0}}
    var CritDMGText : text_block = text_block{DefaultTextColor := MakeColorFromHex("FFFFFF")}
    var CritDMGAmount : text_block = text_block{DefaultTextColor := MakeColorFromHex("39FF39")}
    var FinalStrikeIcon : texture_block = texture_block{DefaultImage := UI.FinalStrike, DefaultDesiredSize := vector2{X := 55.0, Y:= 55.0}}
    var FinalStrikeText : text_block = text_block{DefaultTextColor := MakeColorFromHex("FFFFFF")}
    var FinalStrikeAmount : text_block = text_block{DefaultTextColor := MakeColorFromHex("39FF39")}

    var TiersTitle : text_block = text_block{DefaultTextColor := MakeColorFromHex("FFFFFF")}
    var TiersText : text_block = text_block{DefaultTextColor := MakeColorFromHex("FFFFFF")}
    
    var CostTitle : text_block = text_block{DefaultTextColor := MakeColorFromHex("FFFFFF")}
    var StoneCostIMG : texture_block = texture_block{DefaultImage := UI.StoneIcon, DefaultDesiredSize := vector2{X := 50.0, Y:= 50.0}}
    var StoneCostText : text_block = text_block{DefaultTextColor := MakeColorFromHex("FFFFFF")}
    var EggCostIMG : texture_block = texture_block{DefaultImage := UI.Egg1, DefaultDesiredSize := vector2{X := 50.0, Y:= 50.0}}
    var EggCostText : text_block = text_block{DefaultTextColor := MakeColorFromHex("FFFFFF")}
    var AllStoneCostIMG : texture_block = texture_block{DefaultImage := UI.StoneIcon, DefaultDesiredSize := vector2{X := 50.0, Y:= 50.0}}
    var AllStoneCostText : text_block = text_block{DefaultTextColor := MakeColorFromHex("FFFFFF")}
    var AllEggCostIMG : texture_block = texture_block{DefaultImage := UI.Egg1, DefaultDesiredSize := vector2{X := 50.0, Y:= 50.0}}
    var AllEggCostText : text_block = text_block{DefaultTextColor := MakeColorFromHex("FFFFFF")}
    var EggSelectedIMG : texture_block = texture_block{DefaultImage := UI.Egg1, DefaultDesiredSize := vector2{X := 50.0, Y:= 50.0}}
    
    var BackArrow : texture_block = texture_block{DefaultImage := UI.LeftArrow}
    var ForwardArrow : texture_block = texture_block{DefaultImage := UI.RightArrow}
   
    var BackButton : button_quiet = button_quiet{}
    var ForwardButton : button_quiet = button_quiet{}
    var HatchButton : button_quiet = button_quiet{}
    var HatchAllButton : button_quiet = button_quiet{}

#SIGNAL_DEVICE_MANAGER DONE
    var TempBuffCalc : float = 0.0
    var EfficiencyDMGBonus : float = 0.0
    var AutoUpgrading : logic = false

    var SUPCloseButton : button_loud = button_loud{}
    var SUPClose_Position : vector2 = vector2{X:=1412.0, Y:=104.0}
    var SUPClose_Size : vector2 = vector2{X:=76.0, Y:= 56.0}

    var SUPHUDBG : texture_block = texture_block{DefaultImage := UI.SUPHUDBG, DefaultDesiredSize := vector2{X := 2065.0, Y:= 1025.0}}
    var SUPHUD_Position : vector2 = vector2{X:= -208.0, Y:= 24.0}
    var SUPTotalDMG : text_block = text_block{DefaultTextColor := MakeColorFromHex("FFFFFF")}
    var SUPTotalDMGIcon : texture_block = texture_block{DefaultImage := UI.SUP1, DefaultDesiredSize := vector2{X := 45.0, Y:= 45.0}}
    var SUPStoneAmountText : text_block = text_block{DefaultTextColor := MakeColorFromHex("FFFFFF")}
    var SUPStoneAmountIcon : texture_block = texture_block{DefaultImage := UI.StoneIcon}
    var SUPAutoUpgrade: button_quiet = button_quiet{}
    var SUPEnableUpgrade: button_quiet = button_quiet{}
    SUPEnableBG : texture_block = texture_block:
        DefaultImage := UI.ButtonOff
    
    var SUPTitle : texture_block = texture_block{DefaultImage := UI.STRENGTH_UPGRADES}
    var SUPLevelTitle : texture_block = texture_block{DefaultImage := UI.Level}
    var SUPBonusTitle : texture_block = texture_block{DefaultImage := UI.Bonus}
    var SUPUpgradeTitle : texture_block = texture_block{DefaultImage := UI.Upgrade}
    var SUPUpgradeEffTitle : texture_block = texture_block{DefaultImage := UI.Upgrade_Efficiency}
 
    var SUPIcons : []texture_block = array{}
    var SUPSurges : []texture_block = array{}
    var SUPRomanNums : []text_block = array{}
    var SUPLevels : []text_block = array{}
    var SUPLevelEffs : []text_block = array{}
    var SUPBonuses : []text_block = array{}
    var SUPNextBonuses : []text_block = array{}
    var SUPUpgradesTexts : []text_block = array{}
    var SUPEfficienciesTexts : []text_block = array{}
    var SUPUpgradesStones : []texture_block = array{}
    var SUPEfficienciesStones : []texture_block = array{}
    var SUPUpgrades : []button_quiet = array{}
    var SUPEfficiencies : []button_quiet = array{}

    var SUPEquip_Sizes : []vector2 = array{}
    
    var SUP1Icon : texture_block = texture_block{DefaultImage := UI.SUP1, DefaultDesiredSize := vector2{X := 67.0, Y:= 67.0}}
    var SUP1Surge : texture_block = texture_block{DefaultImage := UI.Surge, DefaultDesiredSize := vector2{X := 45.0, Y:= 45.0}}
    var SUP1RomanNum : text_block = text_block{DefaultTextColor := MakeColorFromHex("DDDDDD")}
    var SUP1Upgrade: button_quiet = button_quiet{}
    var SUP1Efficiency: button_quiet = button_quiet{}
    var SUP1UpgradeText : text_block = text_block{DefaultTextColor := MakeColorFromHex("FFFFFF")}
    var SUP1EfficiencyText : text_block = text_block{DefaultTextColor := MakeColorFromHex("FFFFFF")}
    var SUP1StoneIcon1 : texture_block = texture_block{DefaultImage := UI.StoneIcon}
    var SUP1StoneIcon2 : texture_block = texture_block{DefaultImage := UI.StoneIcon}
    var SUP1Level : text_block = text_block{DefaultTextColor := MakeColorFromHex("FFFFFF")}
    var SUP1LevelEff : text_block = text_block{DefaultTextColor := MakeColorFromHex("FFFFFF")}
    var SUP1Bonus : text_block = text_block{DefaultTextColor := MakeColorFromHex("FFFFFF")}
    var SUP1NextBonus : text_block = text_block{DefaultTextColor := MakeColorFromHex("39FF39")}
    
    var SUP2Icon : texture_block = texture_block{DefaultImage := UI.SUP2, DefaultDesiredSize := vector2{X := 67.0, Y:= 67.0}}
    var SUP2Surge : texture_block = texture_block{DefaultImage := UI.Surge, DefaultDesiredSize := vector2{X := 45.0, Y:= 45.0}}
    var SUP2RomanNum : text_block = text_block{DefaultTextColor := MakeColorFromHex("DDDDDD")}
    var SUP2Upgrade: button_quiet = button_quiet{}
    var SUP2Efficiency: button_quiet = button_quiet{}
    var SUP2UpgradeText : text_block = text_block{DefaultTextColor := MakeColorFromHex("FFFFFF")}
    var SUP2EfficiencyText : text_block = text_block{DefaultTextColor := MakeColorFromHex("FFFFFF")}
    var SUP2StoneIcon1 : texture_block = texture_block{DefaultImage := UI.StoneIcon}
    var SUP2StoneIcon2 : texture_block = texture_block{DefaultImage := UI.StoneIcon}
    var SUP2Level : text_block = text_block{DefaultTextColor := MakeColorFromHex("FFFFFF")}
    var SUP2LevelEff : text_block = text_block{DefaultTextColor := MakeColorFromHex("FFFFFF")}
    var SUP2Bonus : text_block = text_block{DefaultTextColor := MakeColorFromHex("FFFFFF")}
    var SUP2NextBonus : text_block = text_block{DefaultTextColor := MakeColorFromHex("39FF39")}
    
    var SUP3Icon : texture_block = texture_block{DefaultImage := UI.SUP3, DefaultDesiredSize := vector2{X := 67.0, Y:= 67.0}}
    var SUP3Surge : texture_block = texture_block{DefaultImage := UI.Surge, DefaultDesiredSize := vector2{X := 45.0, Y:= 45.0}}
    var SUP3RomanNum : text_block = text_block{DefaultTextColor := MakeColorFromHex("DDDDDD")}
    var SUP3Upgrade: button_quiet = button_quiet{}
    var SUP3Efficiency: button_quiet = button_quiet{}
    var SUP3UpgradeText : text_block = text_block{DefaultTextColor := MakeColorFromHex("FFFFFF")}
    var SUP3EfficiencyText : text_block = text_block{DefaultTextColor := MakeColorFromHex("FFFFFF")}
    var SUP3StoneIcon1 : texture_block = texture_block{DefaultImage := UI.StoneIcon}
    var SUP3StoneIcon2 : texture_block = texture_block{DefaultImage := UI.StoneIcon}
    var SUP3Level : text_block = text_block{DefaultTextColor := MakeColorFromHex("FFFFFF")}
    var SUP3LevelEff : text_block = text_block{DefaultTextColor := MakeColorFromHex("FFFFFF")}
    var SUP3Bonus : text_block = text_block{DefaultTextColor := MakeColorFromHex("FFFFFF")}
    var SUP3NextBonus : text_block = text_block{DefaultTextColor := MakeColorFromHex("39FF39")}
    
    var SUP4Icon : texture_block = texture_block{DefaultImage := UI.SUP4, DefaultDesiredSize := vector2{X := 67.0, Y:= 67.0}}
    var SUP4Surge : texture_block = texture_block{DefaultImage := UI.Surge, DefaultDesiredSize := vector2{X := 45.0, Y:= 45.0}}
    var SUP4RomanNum : text_block = text_block{DefaultTextColor := MakeColorFromHex("DDDDDD")}
    var SUP4Upgrade: button_quiet = button_quiet{}
    var SUP4Efficiency: button_quiet = button_quiet{}
    var SUP4UpgradeText : text_block = text_block{DefaultTextColor := MakeColorFromHex("FFFFFF")}
    var SUP4EfficiencyText : text_block = text_block{DefaultTextColor := MakeColorFromHex("FFFFFF")}
    var SUP4StoneIcon1 : texture_block = texture_block{DefaultImage := UI.StoneIcon}
    var SUP4StoneIcon2 : texture_block = texture_block{DefaultImage := UI.StoneIcon}
    var SUP4Level : text_block = text_block{DefaultTextColor := MakeColorFromHex("FFFFFF")}
    var SUP4LevelEff : text_block = text_block{DefaultTextColor := MakeColorFromHex("FFFFFF")}
    var SUP4Bonus : text_block = text_block{DefaultTextColor := MakeColorFromHex("FFFFFF")}
    var SUP4NextBonus : text_block = text_block{DefaultTextColor := MakeColorFromHex("39FF39")}
    
    var SUP5Icon : texture_block = texture_block{DefaultImage := UI.SUP5, DefaultDesiredSize := vector2{X := 67.0, Y:= 67.0}}
    var SUP5Surge : texture_block = texture_block{DefaultImage := UI.Surge, DefaultDesiredSize := vector2{X := 45.0, Y:= 45.0}}
    var SUP5RomanNum : text_block = text_block{DefaultTextColor := MakeColorFromHex("DDDDDD")}
    var SUP5Upgrade: button_quiet = button_quiet{}
    var SUP5Efficiency: button_quiet = button_quiet{}
    var SUP5UpgradeText : text_block = text_block{DefaultTextColor := MakeColorFromHex("FFFFFF")}
    var SUP5EfficiencyText : text_block = text_block{DefaultTextColor := MakeColorFromHex("FFFFFF")}
    var SUP5StoneIcon1 : texture_block = texture_block{DefaultImage := UI.StoneIcon}
    var SUP5StoneIcon2 : texture_block = texture_block{DefaultImage := UI.StoneIcon}
    var SUP5Level : text_block = text_block{DefaultTextColor := MakeColorFromHex("FFFFFF")}
    var SUP5LevelEff : text_block = text_block{DefaultTextColor := MakeColorFromHex("FFFFFF")}
    var SUP5Bonus : text_block = text_block{DefaultTextColor := MakeColorFromHex("FFFFFF")}
    var SUP5NextBonus : text_block = text_block{DefaultTextColor := MakeColorFromHex("39FF39")}

#UI_HUT_UPGRADES DONE
    var HUPCloseButton : button_loud = button_loud{}
    var HUPClose_Position : vector2 = vector2{X:=200.0, Y:=976.0}
    var HUPClose_Size : vector2 = vector2{X:=200.0, Y:= 50.0}

    var HUPHUDBG : texture_block = texture_block{DefaultImage := UI.HUPHUDBG}
    var HUPHUD_Position : vector2 = vector2{X:= 0.0, Y:= 0.0}
    var HUPHUDBGCover : texture_block = texture_block{DefaultImage := UI.HUPHUDBGCover, DefaultDesiredSize:=vector2{X:=2140.5,Y:=1076.0}}
    var HUPHUDTitle : texture_block = texture_block{DefaultImage := UI.MINER_UPGRADES}

    var HUPUpgradeButton1 : button_quiet = button_quiet{}
    var HUPUpgradeButton2 : button_quiet = button_quiet{}
    var HUPUpgradeButton3 : button_quiet = button_quiet{}
    var HUPUpgradeButton4 : button_quiet = button_quiet{}
    var HUPUpgradeAllButton : button_quiet = button_quiet{}
    var HUPEnableButton : button_quiet = button_quiet{}

    HUPEnableBG : texture_block = texture_block:
        DefaultImage := UI.ButtonOff

    #ARRAYS
    var HUPStoneIcons : []texture_block = array{}
    var HUPUpgradeTitles : []text_block = array{}
    var HUPUpgradeLevels : []text_block = array{}
    var HUPUpgradeIcons : []texture_block = array{}
    var HUPUpgradeDescTexts : []text_block = array{}
    var HUPUpgradeBonuses : []text_block = array{}
    var HUPUpgradeNextBonuses : []text_block = array{}
    var HUPUpgradeCosts : []text_block = array{}
    var HUPUpgradeCostTexts : []text_block = array{}
    var HUPUpgradeCostIcons : []texture_block = array{}
    var HUPUpgradeButtons : []button_quiet = array{}

    var HUPAllTexts : []text_block = array{}

    var AllHutLevels : [][]int = array{}
    var AllHutCosts : [][]float = array{}
    var AllHutUpgrades : [][]float = array{}

#UI_MAIN_MENU DONE
    #MAIN MENU
    var MainMenuCanvas : canvas = canvas{}
    var MainMenuBackGround : texture_block = texture_block{DefaultImage := UI.HUDBackground}
    var CloseBtn : button_loud = button_loud{}
    var MainMenuTitle : texture_block = texture_block{DefaultImage := UI.MAIN_MENU}

    var AchievementsIcon : texture_block = texture_block{DefaultImage := UI.achievementIcon}
    var AchievementsBtn : button_quiet = button_quiet{}
    var PlayerStatsIcon : texture_block = texture_block{DefaultImage := UI.playerstats}
    var PlayerStatsBtn : button_quiet = button_quiet{}
    var PetsIcon : texture_block = texture_block{DefaultImage := UI.AchPetLevel}
    var PetsBtn : button_quiet = button_quiet{}
    var InventoryIcon : texture_block = texture_block{DefaultImage := UI.inventoryIcon}
    var InventoryBtn : button_quiet = button_quiet{}

    var NumbersIcon : texture_block = texture_block{DefaultImage := UI.NumberNotations}
    var NumbersBtn : button_quiet = button_quiet{}
    var MapIcon : texture_block = texture_block{DefaultImage := UI.Map}
    var MapBtn : button_quiet = button_quiet{}
    var MapBtnBG : texture_block = texture_block{DefaultImage := UI.ButtonOn}
    var TeleportIcon : texture_block = texture_block{DefaultImage := UI.teleporters}
    var TeleportBtn : button_quiet = button_quiet{}
    var PrestigeIcon : texture_block = texture_block{DefaultImage := UI.AchPrestigePoints}
    var PrestigeBtn : button_quiet = button_quiet{}

    var ReloadHUDIcon : texture_block = texture_block{DefaultImage := UI.ReloadHUD}
    var ReloadHUDBtn : button_quiet = button_quiet{}
    var HideNumbersIcon : texture_block = texture_block{DefaultImage := UI.NumberNotations}
    var HideNumbersBtn : button_quiet = button_quiet{}
    var HideNumbersBtnBG : texture_block = texture_block{DefaultImage := UI.ButtonOn}
    var MuteMusicIcon : texture_block = texture_block{DefaultImage := UI.MuteMusic}
    var MuteMusicBtn : button_quiet = button_quiet{}
    var MuteMusicBtnBG : texture_block = texture_block{DefaultImage := UI.ButtonOn}
    var DrillIcon : texture_block = texture_block{DefaultImage := UI.miningdrill}
    var DrillBtn : button_quiet = button_quiet{}

    #PETS
    var AllPetTexts : []text_block = array{}
    var PetImageTextureBlocks : []texture_block = array{}
    var EquipButtons : []button_quiet = array{}
    var PetButtons : []button_quiet = array{}
    var InvEggTexts : []text_block = array{}
    var SelectButtons: []button_quiet = array{}

    var AllPrestigeLevels : [][]int = array{}
    var AllPrestigeCosts : [][]float = array{}
    var AllPrestigeUpgrades : [][]float = array{}

    var PetMenuCanvas : canvas = canvas{}
    var PetMenuBackGround : texture_block = texture_block{DefaultImage := UI.PetEquipHUDBG}
    var ClosePetMenuBtn : button_loud = button_loud{}
    var PetsTitle : texture_block = texture_block{DefaultImage := UI.PETS}

    var EquipSlot1Btn : button_quiet = button_quiet{}
    var EquipSlot2Btn : button_quiet = button_quiet{}
    var EquipSlot3Btn : button_quiet = button_quiet{}
    var EquipSlot4Btn : button_quiet = button_quiet{}
    var EquipSlot5Btn : button_quiet = button_quiet{}
    var EquipSlot6Btn : button_quiet = button_quiet{}
    var Equip1Image : texture_block = texture_block{DefaultImage := UI.Blank}
    var Equip2Image : texture_block = texture_block{DefaultImage := UI.Blank}
    var Equip3Image : texture_block = texture_block{DefaultImage := UI.Blank}
    var Equip4Image : texture_block = texture_block{DefaultImage := UI.Blank}
    var Equip5Image : texture_block = texture_block{DefaultImage := UI.Blank}
    var Equip6Image : texture_block = texture_block{DefaultImage := UI.Blank}

    var Pet1Btn : button_quiet = button_quiet{}
    var Pet2Btn : button_quiet = button_quiet{}
    var Pet3Btn : button_quiet = button_quiet{}
    var Pet4Btn : button_quiet = button_quiet{}
    var Pet5Btn : button_quiet = button_quiet{}
    var Pet6Btn : button_quiet = button_quiet{}
    var Pet1Image : texture_block = texture_block{DefaultImage := UI.pet1}
    var Pet2Image : texture_block = texture_block{DefaultImage := UI.pet2}
    var Pet3Image : texture_block = texture_block{DefaultImage := UI.pet3}
    var Pet4Image : texture_block = texture_block{DefaultImage := UI.pet4}
    var Pet5Image : texture_block = texture_block{DefaultImage := UI.pet5}
    var Pet6Image : texture_block = texture_block{DefaultImage := UI.pet6}

    var UnequipPetsBtn : button_quiet = button_quiet{}
    var UnequipPetsText : text_block = text_block{DefaultTextColor := MakeColorFromHex("FFFFFF")}
    var HidePetsBtn : button_quiet = button_quiet{}

    var PetTotalTitle : text_block = text_block{DefaultTextColor := MakeColorFromHex("FFFFFF")}
    var GatherTotalText : text_block = text_block{DefaultTextColor := MakeColorFromHex("FFFFFF")}
    var GatherTotalAmount : text_block = text_block{DefaultTextColor := MakeColorFromHex("39FF39")}
    var StrengthAmount : text_block = text_block{DefaultTextColor := MakeColorFromHex("39FF39")}
    var CritDmgText : text_block = text_block{DefaultTextColor := MakeColorFromHex("FFFFFF")}
    var CritDmgAmount : text_block = text_block{DefaultTextColor := MakeColorFromHex("39FF39")}

    var EquipImages : []texture_block = array{}

    #INVENTORY UI

    var InvMenuCanvas : canvas = canvas{}
    var InvMenuBackGround : texture_block = texture_block{DefaultImage := UI.inventhud}

    var AllInvTexts : []text_block = array{}

    var InvShardIcon : texture_block = texture_block{DefaultImage:=UI.ShardIcon}
    var InvBossKeyIcon : texture_block = texture_block{DefaultImage:=UI.BossKeyIcon}

    var InventoryTitle : texture_block = texture_block{DefaultImage:=UI.INVENTORY}
    var CurrencyTitle : texture_block = texture_block{DefaultImage:=UI.currencies}
    var EggTitle : texture_block = texture_block{DefaultImage:=UI.EggsTitle}

    var InvStoneText : text_block = text_block{DefaultTextColor := MakeColorFromHex("FFFFFF")}
    var InvShardText : text_block = text_block{DefaultTextColor := MakeColorFromHex("FFFFFF")}
    var InvWepMatText : text_block = text_block{DefaultTextColor := MakeColorFromHex("FFFFFF")}
    var InvBossKeyText : text_block = text_block{DefaultTextColor := MakeColorFromHex("FFFFFF")}

    var InvEgg1Text : text_block = text_block{DefaultTextColor := MakeColorFromHex("FFFFFF")}
    var InvEgg2Text : text_block = text_block{DefaultTextColor := MakeColorFromHex("FFFFFF")}
    var InvEgg3Text : text_block = text_block{DefaultTextColor := MakeColorFromHex("FFFFFF")}
    var InvEgg4Text : text_block = text_block{DefaultTextColor := MakeColorFromHex("FFFFFF")}
    var InvEgg5Text : text_block = text_block{DefaultTextColor := MakeColorFromHex("FFFFFF")}
    var InvEgg6Text : text_block = text_block{DefaultTextColor := MakeColorFromHex("FFFFFF")}

    var CloseInvBtn : button_loud = button_loud{}

    #GB 
    var GBGBBtn : button_regular = button_regular{}
    var GBGenStatsBtn : button_regular = button_regular{}
    var GBGameStatsBtn : button_regular = button_regular{}
    var GBTitle : texture_block = texture_block{DefaultImage:=UI.PLAYER_STATS}
    var GBMenuTitle : text_block = text_block{DefaultTextColor := MakeColorFromHex("FFFFFF")}
    var GSTitle : texture_block = texture_block{DefaultImage:=UI.PLAYER_STATS}
    var GSMenuTitle : text_block = text_block{DefaultTextColor := MakeColorFromHex("FFFFFF")}

    var GBCloseBtn : button_loud = button_loud{}

    var GBBtns : []button_regular = array{}
    var GSBtns : []button_regular = array{}
    
    var GatherBreakdownCanvas : canvas = canvas{}
    var GBBackground : texture_block = texture_block{DefaultImage:=UI.gatherbreakdown}

    var GBIcon1 : texture_block = texture_block{DefaultImage:=UI.SUP1}
    var GBIcon2 : texture_block = texture_block{DefaultImage:=UI.AchMobsDefeated}
    var GBIcon3 : texture_block = texture_block{DefaultImage:=UI.statAttack}
    var GBIcon4 : texture_block = texture_block{DefaultImage:=UI.SUP16}
    var GBIcon5 : texture_block = texture_block{DefaultImage:=UI.areamultiplier}
    var GBIcon6 : texture_block = texture_block{DefaultImage:=UI.achievementIcon}
    var GBIcon7 : texture_block = texture_block{DefaultImage:=UI.ShardIcon}
    var GBIcon8 : texture_block = texture_block{DefaultImage:=UI.FinalStrike}
    var GBIcon9 : texture_block = texture_block{DefaultImage:=UI.AutoGather}
    var GBIcon10 : texture_block = texture_block{DefaultImage:=UI.AutoGather}
    
    var AllGBTexts : []text_block = array{}

    var GBTitleText : text_block = text_block{DefaultTextColor := MakeColorFromHex("FFFFFF")}

    var GBTitle1 : text_block = text_block{DefaultTextColor := MakeColorFromHex("FFFFFF")}
    var GBTitle2 : text_block = text_block{DefaultTextColor := MakeColorFromHex("FFFFFF")}
    var GBTitle3 : text_block = text_block{DefaultTextColor := MakeColorFromHex("FFFFFF")}
    var GBTitle4 : text_block = text_block{DefaultTextColor := MakeColorFromHex("FFFFFF")}
    var GBTitle5 : text_block = text_block{DefaultTextColor := MakeColorFromHex("FFFFFF")}
    var GBTitle6 : text_block = text_block{DefaultTextColor := MakeColorFromHex("FFFFFF")}
    var GBTitle7 : text_block = text_block{DefaultTextColor := MakeColorFromHex("FFFFFF")}
    var GBTitle8 : text_block = text_block{DefaultTextColor := MakeColorFromHex("FFFFFF")}
    var GBTitle9 : text_block = text_block{DefaultTextColor := MakeColorFromHex("FFFFFF")}
    var GBTitle10 : text_block = text_block{DefaultTextColor := MakeColorFromHex("FFFFFF")}

    var GBTitles : []text_block = array{}

    var GBAmount1 : text_block = text_block{DefaultTextColor := MakeColorFromHex("FFFFFF")}
    var GBAmount2 : text_block = text_block{DefaultTextColor := MakeColorFromHex("FFFFFF")}
    var GBAmount3 : text_block = text_block{DefaultTextColor := MakeColorFromHex("FFFFFF")}
    var GBAmount4 : text_block = text_block{DefaultTextColor := MakeColorFromHex("FFFFFF")}
    var GBAmount5 : text_block = text_block{DefaultTextColor := MakeColorFromHex("FFFFFF")}
    var GBAmount6 : text_block = text_block{DefaultTextColor := MakeColorFromHex("FFFFFF")}
    var GBAmount7 : text_block = text_block{DefaultTextColor := MakeColorFromHex("FFFFFF")}
    var GBAmount8 : text_block = text_block{DefaultTextColor := MakeColorFromHex("FFFFFF")}
    var GBAmount9 : text_block = text_block{DefaultTextColor := MakeColorFromHex("FFFFFF")}
    var GBAmount10 : text_block = text_block{DefaultTextColor := MakeColorFromHex("FFFFFF")}

    #GS
    
    var GSCanvas : canvas = canvas{}

    var GSCloseBtn : button_loud = button_loud{}
    var GSGBBtn : button_regular = button_regular{}
    var GSGenStatsBtn : button_regular = button_regular{}
    var GSGameStatsBtn : button_regular = button_regular{}

    var GSBackground : texture_block = texture_block{DefaultImage:=UI.generalstats}

    var AllGSTexts : []text_block = array{}

    var GSIcon1 : texture_block = texture_block{DefaultImage:=UI.statAttack}
    var GSIcon2 : texture_block = texture_block{DefaultImage:=UI.statAttack}
    var GSIcon3 : texture_block = texture_block{DefaultImage:=UI.CritChance}
    var GSIcon4 : texture_block = texture_block{DefaultImage:=UI.CritDamage}
    var GSIcon5 : texture_block = texture_block{DefaultImage:=UI.FinalStrike}
    var GSIcon6 : texture_block = texture_block{DefaultImage:=UI.movementspeed}
    var GSIcon7 : texture_block = texture_block{DefaultImage:=UI.EggDropRate}
    var GSIcon8 : texture_block = texture_block{DefaultImage:=UI.EggDrop}

    var GSIconBlock : []texture_block = array{}

    var GSIcons1: []texture = array{UI.statAttack,UI.statAttack,UI.CritChance,UI.CritDamage,UI.FinalStrike,UI.movementspeed,UI.EggDropRate,UI.EggDrop}
    var GSIcons2: []texture = array{UI.AchTimePlayed,UI.AchTimePlayed,UI.AchMobsDefeated,UI.AchRocksDestroyed,UI.miningdrill,UI.AchBossesDefeated,UI.StoneIcon,UI.ShardIcon}

    var GSTitleText : text_block = text_block{DefaultTextColor := MakeColorFromHex("FFFFFF")}
    var GSHeadingTitle1 : text_block = text_block{DefaultTextColor := MakeColorFromHex("FFFFFF")}
    var GSHeadingTitle2 : text_block = text_block{DefaultTextColor := MakeColorFromHex("FFFFFF")}

    var GSTitle1 : text_block = text_block{DefaultTextColor := MakeColorFromHex("FFFFFF")}
    var GSTitle2 : text_block = text_block{DefaultTextColor := MakeColorFromHex("FFFFFF")}
    var GSTitle3 : text_block = text_block{DefaultTextColor := MakeColorFromHex("FFFFFF")}
    var GSTitle4 : text_block = text_block{DefaultTextColor := MakeColorFromHex("FFFFFF")}
    var GSTitle5 : text_block = text_block{DefaultTextColor := MakeColorFromHex("FFFFFF")}
    var GSTitle6 : text_block = text_block{DefaultTextColor := MakeColorFromHex("FFFFFF")}
    var GSTitle7 : text_block = text_block{DefaultTextColor := MakeColorFromHex("FFFFFF")}
    var GSTitle8 : text_block = text_block{DefaultTextColor := MakeColorFromHex("FFFFFF")}
    var GSTitle9 : text_block = text_block{DefaultTextColor := MakeColorFromHex("FFFFFF")}
    var GSTitle10 : text_block = text_block{DefaultTextColor := MakeColorFromHex("FFFFFF")}

    var GSTitles : []text_block = array{}

    var GSAmount1 : text_block = text_block{DefaultTextColor := MakeColorFromHex("FFFFFF")}
    var GSAmount2 : text_block = text_block{DefaultTextColor := MakeColorFromHex("FFFFFF")}
    var GSAmount3 : text_block = text_block{DefaultTextColor := MakeColorFromHex("FFFFFF")}
    var GSAmount4 : text_block = text_block{DefaultTextColor := MakeColorFromHex("FFFFFF")}
    var GSAmount5 : text_block = text_block{DefaultTextColor := MakeColorFromHex("FFFFFF")}
    var GSAmount6 : text_block = text_block{DefaultTextColor := MakeColorFromHex("FFFFFF")}
    var GSAmount7 : text_block = text_block{DefaultTextColor := MakeColorFromHex("FFFFFF")}
    var GSAmount8 : text_block = text_block{DefaultTextColor := MakeColorFromHex("FFFFFF")}
    var GSAmount9 : text_block = text_block{DefaultTextColor := MakeColorFromHex("FFFFFF")}
    var GSAmount10 : text_block = text_block{DefaultTextColor := MakeColorFromHex("FFFFFF")}

    #ACHIEVEMENTS
    var AchMenuCanvas : canvas = canvas{}
    AchBarColorBase1 : color_block = color_block{DefaultColor := MakeColorFromHex("222222"),DefaultDesiredSize := vector2{X:= 876.0, Y := 41.0}} 
    AchBarColorBase2 : color_block = color_block{DefaultColor := MakeColorFromHex("222222"),DefaultDesiredSize := vector2{X:= 876.0, Y := 41.0}} 
    AchBarColorBase3 : color_block = color_block{DefaultColor := MakeColorFromHex("222222"),DefaultDesiredSize := vector2{X:= 876.0, Y := 41.0}} 
    AchBarColorBase4 : color_block = color_block{DefaultColor := MakeColorFromHex("222222"),DefaultDesiredSize := vector2{X:= 876.0, Y := 41.0}} 
    AchBarColorBase5 : color_block = color_block{DefaultColor := MakeColorFromHex("222222"),DefaultDesiredSize := vector2{X:= 876.0, Y := 41.0}} 
    AchBarColorBase6 : color_block = color_block{DefaultColor := MakeColorFromHex("222222"),DefaultDesiredSize := vector2{X:= 876.0, Y := 41.0}} 
    AchBarColorBase7 : color_block = color_block{DefaultColor := MakeColorFromHex("222222"),DefaultDesiredSize := vector2{X:= 876.0, Y := 41.0}} 
    AchBarColorBase8 : color_block = color_block{DefaultColor := MakeColorFromHex("222222"),DefaultDesiredSize := vector2{X:= 876.0, Y := 41.0}} 

    var AchOverlay : overlay = overlay{}

    var CloseAch_Position : vector2 = vector2{X:=1710.0, Y:=100.0}
    var CloseAch_Size : vector2 = vector2{X:=76.0, Y:= 56.0}

    var AchHUDBG : texture_block = texture_block{DefaultImage := UI.ACHHUDBG, DefaultDesiredSize := vector2{X := 2140.5, Y:= 1076.0}}
    var AchHUD_Position : vector2 = vector2{X:= 0.0, Y:= 0.0}
    var AchTitleText : texture_block = texture_block{DefaultImage := UI.ACHIEVEMENTS}

    var AchHUDTitles : []text_block = array{}
    var AchHUDDescs : []text_block = array{}
    var AchHUDProgress : []text_block = array{}
    var AchBarColorTops : []color_block = array{}

    var CloseAchButton : button_loud = button_loud{}

    var AchBuffTotalText : text_block = text_block{DefaultTextColor := MakeColorFromHex("39FF39")}

    ## Stone Gathered
    var Ach1Title : text_block = text_block{DefaultTextColor := MakeColorFromHex("FFFFFF")}
    var Ach1Desc : text_block = text_block{DefaultTextColor := MakeColorFromHex("DDDDDD")}
    var Ach1Progress :  text_block = text_block{DefaultTextColor := MakeColorFromHex("FFFFFF")}
    var Ach1Icon : texture_block = texture_block{DefaultImage := UI.AchStoneGatheredWhite}
    var Ach1Width : float = 0.0
    var Ach1BarColorTop : color_block = color_block{
        DefaultColor := MakeColorFromHex("FFA800")
        DefaultDesiredSize := vector2{X:=0.0, Y:=20.0}}  

    ## Max Strength
    var Ach2Title : text_block = text_block{DefaultTextColor := MakeColorFromHex("FFFFFF")}
    var Ach2Desc : text_block = text_block{DefaultTextColor := MakeColorFromHex("DDDDDD")}
    var Ach2Progress :  text_block = text_block{DefaultTextColor := MakeColorFromHex("FFFFFF")}
    var Ach2Icon : texture_block = texture_block{DefaultImage := UI.AchMaxSTR}
    var Ach2Width : float = 0.0
    var Ach2BarColorTop : color_block = color_block{
        DefaultColor := MakeColorFromHex("FFA800")
        DefaultDesiredSize := vector2{X:=0.0, Y:=20.0}}  

    ## Total Mobs Defeated
    var Ach3Title : text_block = text_block{DefaultTextColor := MakeColorFromHex("FFFFFF")}
    var Ach3Desc : text_block = text_block{DefaultTextColor := MakeColorFromHex("DDDDDD")}
    var Ach3Progress :  text_block = text_block{DefaultTextColor := MakeColorFromHex("FFFFFF")}
    var Ach3Icon : texture_block = texture_block{DefaultImage := UI.AchMobsDefeated}
    var Ach3Width : float = 0.0
    var Ach3BarColorTop : color_block = color_block{
        DefaultColor := MakeColorFromHex("FFA800")
        DefaultDesiredSize := vector2{X:=0.0, Y:=20.0}}  

    ## Prestige Points Collected
    var Ach4Title : text_block = text_block{DefaultTextColor := MakeColorFromHex("FFFFFF")}
    var Ach4Desc : text_block = text_block{DefaultTextColor := MakeColorFromHex("DDDDDD")}
    var Ach4Progress :  text_block = text_block{DefaultTextColor := MakeColorFromHex("FFFFFF")}
    var Ach4Icon : texture_block = texture_block{DefaultImage := UI.AchPrestigePoints}
    var Ach4Width : float = 0.0
    var Ach4BarColorTop : color_block = color_block{
        DefaultColor := MakeColorFromHex("FFA800")
        DefaultDesiredSize := vector2{X:=0.0, Y:=20.0}}  

    ## Rocks Destroyed
    var Ach5Title : text_block = text_block{DefaultTextColor := MakeColorFromHex("FFFFFF")}
    var Ach5Desc : text_block = text_block{DefaultTextColor := MakeColorFromHex("DDDDDD")}
    var Ach5Progress :  text_block = text_block{DefaultTextColor := MakeColorFromHex("FFFFFF")}
    var Ach5Icon : texture_block = texture_block{DefaultImage := UI.AchRocksDestroyed}
    var Ach5Width : float = 0.0
    var Ach5BarColorTop : color_block = color_block{
        DefaultColor := MakeColorFromHex("FFA800")
        DefaultDesiredSize := vector2{X:=0.0, Y:=20.0}}  

    ## Total Pet Level
    var Ach6Title : text_block = text_block{DefaultTextColor := MakeColorFromHex("FFFFFF")}
    var Ach6Desc : text_block = text_block{DefaultTextColor := MakeColorFromHex("DDDDDD")}
    var Ach6Progress :  text_block = text_block{DefaultTextColor := MakeColorFromHex("FFFFFF")}
    var Ach6Icon : texture_block = texture_block{DefaultImage := UI.AchPetLevel}
    var Ach6Width : float = 0.0
    var Ach6BarColorTop : color_block = color_block{
        DefaultColor := MakeColorFromHex("FFA800")
        DefaultDesiredSize := vector2{X:=0.0, Y:=20.0}}  

    ## Total Bosses Defeated
    var Ach7Title : text_block = text_block{DefaultTextColor := MakeColorFromHex("FFFFFF")}
    var Ach7Desc : text_block = text_block{DefaultTextColor := MakeColorFromHex("DDDDDD")}
    var Ach7Progress :  text_block = text_block{DefaultTextColor := MakeColorFromHex("FFFFFF")}
    var Ach7Icon : texture_block = texture_block{DefaultImage := UI.AchBossesDefeated}
    var Ach7Width : float = 0.0
    var Ach7BarColorTop : color_block = color_block{
        DefaultColor := MakeColorFromHex("FFA800")
        DefaultDesiredSize := vector2{X:=0.0, Y:=20.0}}  

    ## Total Time Played
    var Ach8Title : text_block = text_block{DefaultTextColor := MakeColorFromHex("FFFFFF")}
    var Ach8Desc : text_block = text_block{DefaultTextColor := MakeColorFromHex("DDDDDD")}
    var Ach8Progress :  text_block = text_block{DefaultTextColor := MakeColorFromHex("FFFFFF")}
    var Ach8Icon : texture_block = texture_block{DefaultImage := UI.AchTimePlayed}
    var Ach8Width : float = 0.0
    var Ach8BarColorTop : color_block = color_block{
        DefaultColor := MakeColorFromHex("FFA800")
        DefaultDesiredSize := vector2{X:=0.0, Y:=20.0}}  


    #NUMBER NOTATIONS
    var NNCloseButton : button_loud = button_loud{}
    var NNClose_Position : vector2 = vector2{X:=1710.0, Y:=100.0}
    var NNClose_Size : vector2 = vector2{X:=76.0, Y:= 56.0}

    var NNHUDBG : texture_block = texture_block{DefaultImage := UI.NumberNotationHUDBG, DefaultDesiredSize := vector2{X := 2048.0, Y:= 1076.0}}
    var NNHUDTitle : texture_block = texture_block{DefaultImage := UI.NUMBER_NOTATIONS}

    var AllNNTexts : []text_block = array{}

    #EACH VALUE
    var Row1Letter : text_block = text_block{DefaultTextColor := MakeColorFromHex("FFFFFF")}
    var Row1Number : text_block = text_block{DefaultTextColor := MakeColorFromHex("FFFFFF")}
    var Row2Letter : text_block = text_block{DefaultTextColor := MakeColorFromHex("FFFFFF")}
    var Row2Number : text_block = text_block{DefaultTextColor := MakeColorFromHex("FFFFFF")}
    var Row3Letter : text_block = text_block{DefaultTextColor := MakeColorFromHex("FFFFFF")}
    var Row3Number : text_block = text_block{DefaultTextColor := MakeColorFromHex("FFFFFF")}
    var Row4Letter : text_block = text_block{DefaultTextColor := MakeColorFromHex("FFFFFF")}
    var Row4Number : text_block = text_block{DefaultTextColor := MakeColorFromHex("FFFFFF")}

    #INFORMATION BOARDS
    var CaveMapHUDBG : texture_block = texture_block{DefaultImage := UI.CaveMap2, DefaultDesiredSize := vector2{X := 2048.0, Y:= 1076.0}}
    var AccoladesNoBG : texture_block = texture_block{DefaultImage := UI.AccoladesNoBG, DefaultDesiredSize := vector2{X := 561.0, Y:= 600.0}}
    var SignalNoBG : texture_block = texture_block{DefaultImage := UI.SignalNoBG, DefaultDesiredSize := vector2{X := 561.0, Y:= 600.0}}
    var InfoNoBG : texture_block = texture_block{DefaultImage := UI.InfoNoBG, DefaultDesiredSize := vector2{X := 561.0, Y:= 600.0}}
    var InfoBackground: texture_block = texture_block{DefaultImage := UI.HUDBackground}

    var InfoCloseButton : button_loud = button_loud{}
    var MapCloseButton : button_loud = button_loud{}
    var InfoClose_Position : vector2 = vector2{X:=1710.0, Y:=100.0}
    var InfoClose_Size : vector2 = vector2{X:=76.0, Y:= 56.0}

#UI_PRESTIGE_UPGRADES DONE
    var PrestigeCloseButton : button_loud = button_loud{}
    var PrestigeClose_Position : vector2 = vector2{X:=1520.0, Y:=100.0}
    var PrestigeClose_Size : vector2 = vector2{X:=76.0, Y:= 56.0}

    var PrestigeHUDBG : texture_block = texture_block{DefaultImage := UI.PrestigeHUDBG, DefaultDesiredSize := vector2{X := 2048.0, Y:= 1076.0}}
    var PrestigeHUDTitle : texture_block = texture_block{DefaultImage := UI.PRESTIGE}

    var AllPrestigeTexts : []text_block = array{}

    #EACH VALUE
    var InfoText : text_block = text_block{DefaultTextColor := MakeColorFromHex("DDDDDD")}
    var KeepTitle : text_block = text_block{DefaultTextColor := MakeColorFromHex("FFFFFF")}
    var KeepText : text_block = text_block{DefaultTextColor := MakeColorFromHex("39FF39")}
    var ResetTitle : text_block = text_block{DefaultTextColor := MakeColorFromHex("FFFFFF")}
    var ResetText : text_block = text_block{DefaultTextColor := MakeColorFromHex("fa4d56")}
    
    var PUPUpgradeButton1 : button_quiet = button_quiet{}
    var PUPUpgradeButton2 : button_quiet = button_quiet{}
    var PUPUpgradeButton3 : button_quiet = button_quiet{}

    var CurrentPrestigeLevelText : text_block = text_block{DefaultTextColor := MakeColorFromHex("FFFFFF")}
    var NextPrestigeTitle : text_block = text_block{DefaultTextColor := MakeColorFromHex("FFFFFF")}
    var NextPrestigeAmount : text_block = text_block{DefaultTextColor := MakeColorFromHex("FFFFFF")}
    var NextPrestigeStoneIcon : texture_block = texture_block{DefaultImage := UI.StoneIcon}

    var PrestigeNowTitle : text_block = text_block{DefaultTextColor := MakeColorFromHex("FFFFFF")}
    var PrestigeNowShards : text_block = text_block{DefaultTextColor := MakeColorFromHex("39FF39")}
    var PrestigeNowShardIcon : texture_block = texture_block{DefaultImage := UI.ShardIcon}
    var PrestigeNowGather : text_block = text_block{DefaultTextColor := MakeColorFromHex("39FF39")}
    var PrestigeNowGatherIcon : texture_block = texture_block{DefaultImage := UI.FinalStrike}

    var PrestigeNowText : text_block = text_block{DefaultTextColor := MakeColorFromHex("FFFFFF")}
    
    var PUPCloseButton : button_loud = button_loud{}
    var PUPClose_Position : vector2 = vector2{X:=1520.0, Y:=148.0}
    var PUPClose_Size : vector2 = vector2{X:=76.0, Y:= 56.0}

    var PUPHUDBG : texture_block = texture_block{DefaultImage := UI.PUPHUDBG, DefaultDesiredSize := vector2{X := 2048.0, Y:= 1076.0}}
    var PUPHUD_Position : vector2 = vector2{X:= 0.0, Y:= 0.0}
    var PUPHUDTitle : texture_block = texture_block{DefaultImage := UI.PRESTIGE_UPGRADES}

    #ARRAYS
    var PUPShardIcons : []texture_block = array{}
    var PUPUpgradeTitlesBlock : []text_block = array{}
    var PUPUpgradeLevelsBlock : []text_block = array{}
    var PUPUpgradeIconsBlock : []texture_block = array{}
    var PUPUpgradeDescTextsBlock : []text_block = array{}
    var PUPUpgradeBonusesBlock : []text_block = array{}
    var PUPUpgradeNextBonusesBlock : []text_block = array{}
    var PUPUpgradeCostsBlock : []text_block = array{}
    var PUPUpgradeCostTextsBlock : []text_block = array{}
    var PUPUpgradeButtonsBlock : []button_quiet = array{}

    #EACH VALUE
    var ShardIcon1 : texture_block = texture_block{DefaultImage := UI.ShardIcon}
    var ShardIcon2 : texture_block = texture_block{DefaultImage := UI.ShardIcon}
    var ShardIcon3 : texture_block = texture_block{DefaultImage := UI.ShardIcon}
    var ShardAmountIcon : texture_block = texture_block{DefaultImage := UI.ShardIcon}
    var ShardAmountText : text_block = text_block{DefaultTextColor := MakeColorFromHex("FFFFFF")}

    var AllPUPTexts : []text_block = array{}

    var PrestigeNowButton : button_regular = button_regular {}

#UI_WEAPON_UPGRADES DONE
    var WUPCloseButton : button_loud = button_loud{}
    var WUPClose_Position : vector2 = vector2{X:=421.0, Y:=100.0}
    var WUPClose_Size : vector2 = vector2{X:=76.0, Y:= 56.0}

    var WUPHUDBG : texture_block = texture_block{DefaultImage := UI.WUPHUDBG, DefaultDesiredSize := vector2{X := 2150.0, Y:= 1080.0}}
    var WUPHUD_Position : vector2 = vector2{X:= 0.0, Y:= 0.0}
    var WUPTotalDMG : text_block = text_block{DefaultTextColor := MakeColorFromHex("FFFFFF")}

    StoneIcon : texture_block = texture_block{DefaultImage := UI.StoneIcon}
    Egg1Icon : texture_block = texture_block{DefaultImage := UI.Egg1}
    Egg2Icon : texture_block = texture_block{DefaultImage := UI.Egg2}
    Egg3Icon : texture_block = texture_block{DefaultImage := UI.Egg3}
    Egg4Icon : texture_block = texture_block{DefaultImage := UI.Egg4}
    Egg5Icon : texture_block = texture_block{DefaultImage := UI.Egg5}
    Egg6Icon : texture_block = texture_block{DefaultImage := UI.Egg6}
    WepMatIcon : texture_block = texture_block{DefaultImage := UI.WepMat}

    var EggIcons : []texture_block = array{}
    var AllWUPTexts : []text_block = array{}
    var InvTexts : []text_block = array{}
    var StatTexts : []text_block = array{}
    var BonusTexts : []text_block = array{}
    var UpgradeTitles : []string = array{"Stats","Bonus Stats","Perks","Evolve", "Reset Stats"}
    var UpgradeBtns : []string = array{"Reforge","Upgrade","Reveal Perks", "Evolve", "Reset Stats"}
    var UpgradeDescs : []string = array{
        "Increase your weapon's base stats.",
        "Increases the bonus stat multiplier\nthat is applied on top of your base stats.",
        "Unlock & randomize your weapons perks.\n Locking a perk will increase the price.",
        "Upgrades to the next weapon tier with\nsignificantly higher stats.\nResets all weapon upgrades.",
        "Resets your weapons stats & bonuses\n to attempt to reroll for better stats."
        }
    var WUPTitle : texture_block = texture_block{DefaultImage := UI.WEAPON_UPGRADES, DefaultDesiredSize := vector2{X := 675.0, Y:= 61.0}}

    var UnlockWUPCover : texture_block = texture_block{DefaultImage := UI.WUPHUDBGCover, DefaultDesiredSize := vector2{X := 2150.0, Y:= 1080.0}}
    var UnlockWUPStone : texture_block = texture_block{DefaultImage := UI.StoneIcon}
    var UnlockWUPCost : text_block = text_block{DefaultTextColor := MakeColorFromHex("FFFFFF")}
    var UnlockWUPButton : button_regular = button_regular{}
    var UnlockWUPTitle : texture_block = texture_block{DefaultImage := UI.UNLOCK_WEAPON, DefaultDesiredSize := vector2{X := 485.0, Y:= 52.0}}
    var UnlockWUPIcon : texture_block = texture_block{DefaultImage := UI.weapon, DefaultDesiredSize := vector2{X := 200.0, Y:= 200.0}}
      
    var WeaponIcon : texture_block = texture_block{DefaultImage := UI.weapon, DefaultDesiredSize := vector2{X := 200.0, Y:= 200.0}}
    var AttackIcon : texture_block = texture_block{DefaultImage := UI.statAttack, DefaultDesiredSize := vector2{X := 55.0, Y:= 55.0}}
    var AttackPercIcon : texture_block = texture_block{DefaultImage := UI.statAttack, DefaultDesiredSize := vector2{X := 55.0, Y:= 55.0}}
    var StrengthPercIcon : texture_block = texture_block{DefaultImage := UI.SUP1, DefaultDesiredSize := vector2{X := 55.0, Y:= 55.0}}
  
    var WepInvTitle : texture_block = texture_block{DefaultImage := UI.INVENTORY, DefaultDesiredSize := vector2{X := 198.0, Y:= 38.0}}
   
    var UpgradeTitleText : texture_block = texture_block{DefaultImage := UI.UPGRADES, DefaultDesiredSize := vector2{X := 188.0, Y:= 38.0}}
    var ButtonStats : button_quiet = button_quiet{}
    var ButtonBonus : button_quiet = button_quiet{}
    var ButtonPerk : button_quiet = button_quiet{}
    var ButtonEvolve : button_quiet = button_quiet{}
    var ButtonReset : button_quiet = button_quiet{}

    var UpgradeTitle : text_block = text_block{DefaultTextColor := MakeColorFromHex("FFFFFF")}
    var UpgradeDesc : text_block = text_block{DefaultTextColor := MakeColorFromHex("DDDDDD")}
    
    var UpgradeReqTitle1 : text_block = text_block{DefaultTextColor := MakeColorFromHex("FFFFFF")}
    var UpgradeReqAmount1 : text_block = text_block{DefaultTextColor := MakeColorFromHex("FFFFFF")}
    var UpgradeReqTexture1 : texture_block = texture_block{DefaultImage := UI.StoneIcon, DefaultDesiredSize := vector2{X := 40.0, Y:= 40.0}}

    var UpgradeReqTitle2 : text_block = text_block{DefaultTextColor := MakeColorFromHex("FFFFFF")}
    var UpgradeReqAmount2 : text_block = text_block{DefaultTextColor := MakeColorFromHex("FFFFFF")}
    var UpgradeReqTexture2 : texture_block = texture_block{DefaultImage := UI.StoneIcon, DefaultDesiredSize := vector2{X := 40.0, Y:= 40.0}}

    var UpgradeReqTitle3 : text_block = text_block{DefaultTextColor := MakeColorFromHex("FFFFFF")}
    var UpgradeReqAmount3 : text_block = text_block{DefaultTextColor := MakeColorFromHex("FFFFFF")}
    var UpgradeReqTexture3 : texture_block = texture_block{DefaultImage := UI.StoneIcon, DefaultDesiredSize := vector2{X := 40.0, Y:= 40.0}}
    var UpgradeReqAmount4 : text_block = text_block{DefaultTextColor := MakeColorFromHex("FFFFFF")}
    var UpgradeReqTexture4 : texture_block = texture_block{DefaultImage := UI.ShardIcon, DefaultDesiredSize := vector2{X := 40.0, Y:= 40.0}}

    var ButtonUpgrade : button_regular = button_regular{}
    var ButtonUpgradeAll : button_regular = button_regular{}

    var WeaponTitle : text_block = text_block{DefaultTextColor := MakeColorFromHex("FFFFFF")}
    var WeaponDesc : text_block = text_block{DefaultTextColor := MakeColorFromHex("DDDDDD")}
    var WeaponCalc : text_block = text_block:
        DefaultTextColor := MakeColorFromHex("DDDDDD")
        DefaultShadowOffset := option{vector2{X:=3.0, Y:=3.0}}
        DefaultShadowColor := MakeColorFromHex("000000")
        DefaultShadowOpacity := 1.0
    var WeaponDamage : text_block = text_block{DefaultTextColor := MakeColorFromHex("FFFFFF")}

    var StatsTitle : text_block = text_block{DefaultTextColor := MakeColorFromHex("FFFFFF")}
    var StatsAttackTitle : text_block = text_block{DefaultTextColor := MakeColorFromHex("DDDDDD")}
    var StatsAttackAmount : text_block = text_block{DefaultTextColor := MakeColorFromHex("FFFFFF")}
    var StatsAttackPercTitle : text_block = text_block{DefaultTextColor := MakeColorFromHex("DDDDDD")}
    var StatsAttackPercAmount : text_block = text_block{DefaultTextColor := MakeColorFromHex("FFFFFF")}
    var StatsStrPercTitle : text_block = text_block{DefaultTextColor := MakeColorFromHex("DDDDDD")}
    var StatsStrPercAmount : text_block = text_block{DefaultTextColor := MakeColorFromHex("FFFFFF")}

    var BonusTitle : text_block = text_block{DefaultTextColor := MakeColorFromHex("FFFFFF")}
    var BonusAttack : text_block = text_block{DefaultTextColor := MakeColorFromHex("8EFF5C")}
    var BonusAttackPerc : text_block = text_block{DefaultTextColor := MakeColorFromHex("8EFF5C")}
    var BonusStrPerc : text_block = text_block{DefaultTextColor := MakeColorFromHex("8EFF5C")}

    var UnlockPerksTitle : text_block = text_block{DefaultTextColor := MakeColorFromHex("FFFFFF")}
    var UnlockPerksLockIcon : texture_block = texture_block{DefaultImage := UI.LockedIconWhite, DefaultDesiredSize := vector2{X := 80.0, Y:= 80.0}}
    var UnlockPerksBG : color_block = color_block{DefaultColor := MakeColorFromHex("222222"), DefaultDesiredSize := vector2{X:=669.819763, Y:=229.189194}}  

    var PerksTitle : text_block = text_block{DefaultTextColor := MakeColorFromHex("FFFFFF")}
    var PerksTier : text_block = text_block{DefaultTextColor := MakeColorFromHex("FFFFFF")}
    var Perk1 : text_block = text_block{DefaultTextColor := MakeColorFromHex("FFFFFF")}
    var Perk2 : text_block = text_block{DefaultTextColor := MakeColorFromHex("FFFFFF")}
    var Perk3 : text_block = text_block{DefaultTextColor := MakeColorFromHex("FFFFFF")}
    var Perk1Icon : texture_block = texture_block{DefaultImage := UI.LockedIcon, DefaultDesiredSize := vector2{X := 40.0, Y:= 40.0}}
    var Perk2Icon : texture_block = texture_block{DefaultImage := UI.LockedIcon, DefaultDesiredSize := vector2{X := 40.0, Y:= 40.0}}
    var Perk3Icon : texture_block = texture_block{DefaultImage := UI.LockedIcon, DefaultDesiredSize := vector2{X := 40.0, Y:= 40.0}}

    var PerkIcons : []texture_block = array{}

    var ButtonPerk1 : button_quiet = button_quiet{}
    var ButtonPerk2 : button_quiet = button_quiet{}
    var ButtonPerk3 : button_quiet = button_quiet{}

#UPGRADES_MANAGER DONE

    var AllLevels : [][]int = array{}
    var AllCosts : [][]float = array{}
    var AllUpgrades : [][]float = array{}
    #HUD STUFF
    var GUPCloseButton : button_loud = button_loud{}
    var GUPClose_Position : vector2 = vector2{X:=1302.0, Y:=100.0}
    var GUPClose_Size : vector2 = vector2{X:=76.0, Y:= 56.0}

    var GUPHUDBG : texture_block = texture_block{DefaultImage := UI.GUPHUDBG, DefaultDesiredSize := vector2{X := 2048.0, Y:= 1076.0}}
    var GUPHUD_Position : vector2 = vector2{X:= 0.0, Y:= 0.0}
    var GUPHUDTitle : texture_block = texture_block{DefaultImage := UI.GUPUPGRADES}

    #ARRAYS
    var GUPStoneIcons : []texture_block = array{}
    var GUPUpgradeTitlesBlock : []text_block = array{}
    var GUPUpgradeLevelsBlock : []text_block = array{}
    var GUPUpgradeIconsBlock : []texture_block = array{}
    var GUPUpgradeDescTextsBlock : []text_block = array{}
    var GUPUpgradeBonusesBlock : []text_block = array{}
    var GUPUpgradeNextBonusesBlock : []text_block = array{}
    var GUPUpgradeCostsBlock : []text_block = array{}
    var GUPUpgradeCostTextsBlock : []text_block = array{}
    var GUPUpgradeButtonsBlock : []button_quiet = array{}

    var AllGUPTexts : []text_block = array{}

    #EACH VALUE
    var GUPUpgradeButton1 : button_quiet = button_quiet{}
    var GUPUpgradeButton2 : button_quiet = button_quiet{}

#TUTORIAL_MANAGER
    #HUD STUFF
    var TutorialHUDBG : texture_block = texture_block{
        DefaultImage := UI.TutorialPopupBG,
        DefaultDesiredSize := vector2{X := 512.0, Y:= 256.0}
        }

    #EACH VALUE
    var TutorialTitleText : text_block = text_block{DefaultTextColor := MakeColorFromHex("FFFFFF")}
    var TutorialDescText : text_block = text_block{DefaultTextColor := MakeColorFromHex("DDDDDD")}
    var TutorialProgressText : text_block = text_block{DefaultTextColor := MakeColorFromHex("FFFFFF")}
    var ProgressIcon : texture_block = texture_block{DefaultImage := UI.StoneIcon}


#RESET ALL DATA 
    var ResetAllCloseButton : button_loud = button_loud{}

    var ResetAllHUDBG : texture_block = texture_block{DefaultImage := UI.HUDBackground, DefaultDesiredSize := vector2{X := 2048.0, Y:= 1076.0}}

    var AllResetAllTexts : []text_block = array{}

    #EACH VALUE
    var ResetInfoTitle : text_block = text_block{DefaultTextColor := MakeColorFromHex("FFFFFF")}
    var ResetWarning : text_block = text_block{DefaultTextColor := MakeColorFromHex("FBFF00")}
    var ResetLoseAll : text_block = text_block{DefaultTextColor := MakeColorFromHex("FA4D56")}

    var ResetAllNowButton : button_quiet = button_quiet {}
    var ResetAllConfirmButton : button_quiet = button_quiet {}

#A4 UNLOCK HUD 
    var A4CloseButton : button_loud = button_loud{}

    var A4UnlockHUDBG : texture_block = texture_block{DefaultImage := UI.A4UnlockHUDBG, DefaultDesiredSize := vector2{X := 2048.0, Y:= 1076.0}}
    var A4UnlockHUDTitle : texture_block = texture_block{DefaultImage := UI.AREA_4_UNLOCK, DefaultDesiredSize := vector2{X := 2048.0, Y:= 1076.0}}

    var AllA4Texts : []text_block = array{}

    #EACH VALUE
    var WhatsNewTitle : text_block = text_block{DefaultTextColor := MakeColorFromHex("FFFFFF")}
    var WhatsNewText : text_block = text_block{DefaultTextColor := MakeColorFromHex("FFFFFF")}
    var A4ReqTitle : text_block = text_block{DefaultTextColor := MakeColorFromHex("FFFFFF")}

    var A4StoneIcon : texture_block = texture_block{DefaultImage := UI.StoneIcon}
    var A4Stone : text_block = text_block{DefaultTextColor := MakeColorFromHex("FFFFFF")}
    var A4ReqStone : text_block = text_block{DefaultTextColor := MakeColorFromHex("FFFFFF")}

    var A4WepTier : text_block = text_block{DefaultTextColor := MakeColorFromHex("FFFFFF")}
    var A4ReqWepTier : text_block = text_block{DefaultTextColor := MakeColorFromHex("FFFFFF")}

    var A4MaxStr : text_block = text_block{DefaultTextColor := MakeColorFromHex("FFFFFF")}
    var A4ReqMaxStr : text_block = text_block{DefaultTextColor := MakeColorFromHex("FFFFFF")}

    var A4UnlockButton : button_regular = button_regular {}
    
#EXPEDITION HUD 
    var EXPTimerHUDBG : texture_block = texture_block{DefaultImage := UI.ExpeditionTimerBG}
    var EXPTimerText : text_block = text_block{DefaultTextColor := MakeColorFromHex("FFFFFF")}

    var EXPCloseButton : button_loud = button_loud{}

    var EXPHUDBG : texture_block = texture_block{DefaultImage := UI.ExpeditionHUDBG2, DefaultDesiredSize := vector2{X := 2048.0, Y:= 1024.0}}
    var EXPHUDTitle : texture_block = texture_block{DefaultImage := UI.EXPEDITIONS, DefaultDesiredSize := vector2{X := 405.0, Y:= 56.0}}

    var AllEXPTexts : []text_block = array{}

    var AllEXPStoneRewards : []text_block = array{}
    var AllEXPEggRewards : []text_block = array{}
    var AllEXPWepMatRewards : []text_block = array{}
    var AllEXPPetRewards : []text_block = array{}

    #EACH VALUE
    var EXPStoneAmountText : text_block = text_block{DefaultTextColor := MakeColorFromHex("FFFFFF")}
    var EXPStoneAmountIcon : texture_block = texture_block{DefaultImage := UI.StoneIcon}
    var EXPShardAmountText : text_block = text_block{DefaultTextColor := MakeColorFromHex("FFFFFF")}
    var EXPShardAmountIcon : texture_block = texture_block{DefaultImage := UI.BossKeyIcon}

    var EXPPetTitle : text_block = text_block{DefaultTextColor := MakeColorFromHex("FFFFFF")}

    var EXPPet1Equip : button_quiet = button_quiet{}
    var EXPPet1Icon : texture_block = texture_block{DefaultImage := UI.dave}
    var EXPPet1Bonuses : text_block = text_block{DefaultTextColor := MakeColorFromHex("FFFFFF")}

    var EXPPet2Equip : button_quiet = button_quiet{}
    var EXPPet2Icon : texture_block = texture_block{DefaultImage := UI.pet7}
    var EXPPet2level : text_block = text_block{DefaultTextColor := MakeColorFromHex("DDDDDD")}
    var EXPPet2Bonuses : text_block = text_block{DefaultTextColor := MakeColorFromHex("FFFFFF")}

    var EXPPet3Equip : button_quiet = button_quiet{}
    var EXPPet3Icon : texture_block = texture_block{DefaultImage := UI.pet8}
    var EXPPet3level : text_block = text_block{DefaultTextColor := MakeColorFromHex("DDDDDD")}
    var EXPPet3Bonuses : text_block = text_block{DefaultTextColor := MakeColorFromHex("FFFFFF")}

    var EXPEasy : button_quiet = button_quiet{}
    var EXP1Title : text_block = text_block{DefaultTextColor := MakeColorFromHex("57df87")}
    var EXP1Time : text_block = text_block{DefaultTextColor := MakeColorFromHex("FFFFFF")}
    var EXP1RewardStone : text_block = text_block{DefaultTextColor := MakeColorFromHex("FFFFFF")}
    var EXP1RewardEgg : text_block = text_block{DefaultTextColor := MakeColorFromHex("FFFFFF")}
    var EXP1RewardWepMat : text_block = text_block{DefaultTextColor := MakeColorFromHex("FFFFFF")}
    var EXP1RewardPet : text_block = text_block{DefaultTextColor := MakeColorFromHex("FFFFFF")}

    var EXPMedium : button_quiet = button_quiet{}
    var EXP2Title : text_block = text_block{DefaultTextColor := MakeColorFromHex("f1da57")}
    var EXP2Time : text_block = text_block{DefaultTextColor := MakeColorFromHex("FFFFFF")}
    var EXP2RewardStone : text_block = text_block{DefaultTextColor := MakeColorFromHex("FFFFFF")}
    var EXP2RewardEgg : text_block = text_block{DefaultTextColor := MakeColorFromHex("FFFFFF")}
    var EXP2RewardWepMat : text_block = text_block{DefaultTextColor := MakeColorFromHex("FFFFFF")}
    var EXP2RewardPet : text_block = text_block{DefaultTextColor := MakeColorFromHex("FFFFFF")}

    var EXPHard : button_quiet = button_quiet{}
    var EXP3Title : text_block = text_block{DefaultTextColor := MakeColorFromHex("f93f3f")}
    var EXP3Time : text_block = text_block{DefaultTextColor := MakeColorFromHex("FFFFFF")}
    var EXP3RewardStone : text_block = text_block{DefaultTextColor := MakeColorFromHex("FFFFFF")}
    var EXP3RewardEgg : text_block = text_block{DefaultTextColor := MakeColorFromHex("FFFFFF")}
    var EXP3RewardWepMat : text_block = text_block{DefaultTextColor := MakeColorFromHex("FFFFFF")}
    var EXP3RewardPet : text_block = text_block{DefaultTextColor := MakeColorFromHex("FFFFFF")}

    var EXPExtreme : button_quiet = button_quiet{}
    var EXP4Title : text_block = text_block{DefaultTextColor := MakeColorFromHex("3336FF")}
    var EXP4Time : text_block = text_block{DefaultTextColor := MakeColorFromHex("FFFFFF")}
    var EXP4RewardStone : text_block = text_block{DefaultTextColor := MakeColorFromHex("FFFFFF")}
    var EXP4RewardEgg : text_block = text_block{DefaultTextColor := MakeColorFromHex("FFFFFF")}
    var EXP4RewardWepMat : text_block = text_block{DefaultTextColor := MakeColorFromHex("FFFFFF")}
    var EXP4RewardPet : text_block = text_block{DefaultTextColor := MakeColorFromHex("FFFFFF")}

    var EXPDetailsTitle : text_block = text_block{DefaultTextColor := MakeColorFromHex("FFFFFF")}
    var EXPDetailsPet : text_block = text_block{DefaultTextColor := MakeColorFromHex("FFFFFF")}
    var EXPDetailsEXP : text_block = text_block{DefaultTextColor := MakeColorFromHex("FFFFFF")}
    var EXPDetailsStone : text_block = text_block{DefaultTextColor := MakeColorFromHex("FFFFFF")}
    var EXPDetailsEgg : text_block = text_block{DefaultTextColor := MakeColorFromHex("FFFFFF")}
    var EXPDetailsWepMat : text_block = text_block{DefaultTextColor := MakeColorFromHex("FFFFFF")}
    var EXPDetailsPetChance : text_block = text_block{DefaultTextColor := MakeColorFromHex("FFFFFF")}
    
    var EXPStoneCostText : text_block = text_block{DefaultTextColor := MakeColorFromHex("FFFFFF")}
    var EXPStoneCostIcon : texture_block = texture_block{DefaultImage := UI.StoneIcon}
    var EXPShardCostText : text_block = text_block{DefaultTextColor := MakeColorFromHex("FFFFFF")}
    var EXPShardCostIcon : texture_block = texture_block{DefaultImage := UI.BossKeyIcon}

    var EXPLaunch : button_regular = button_regular{}

#DAMAGE NUMBER TEST
    var ShowNumbers : logic = true
    var CurrentImages : [int]string = map:
        0 => "",
        1 => "",
        2 => "",
        3 => "",
        4 => "",
        5 => "",
        6 => "",
        7 => "",
        8 => ""
        
    var DmgBlockDMGs : []text_block = array{}
    var DmgColorBlockDMGs : []color_block = array{}
    var DmgTextures : []texture_block = array{}

    var DmgBlockDMG1 :text_block= text_block{DefaultTextColor := MakeColorFromHex("FFFFFF"),DefaultShadowOffset := option{vector2{X:=3.0, Y:=3.0}},DefaultShadowColor := MakeColorFromHex("000000")}
    var DmgColorBlockDMG1 :color_block= color_block{DefaultColor := MakeColorFromHex("000000"),DefaultDesiredSize := vector2{X := 0.0,Y := 250.0},DefaultOpacity := 1.0}
    var DmgTexture1 :texture_block= texture_block{DefaultImage := UI.StoneIcon, DefaultDesiredSize := vector2{X := 50.0, Y:= 50.0}}
    var DmgBlockDMG2 :text_block= text_block{DefaultTextColor := MakeColorFromHex("FFFFFF"),DefaultShadowOffset := option{vector2{X:=3.0, Y:=3.0}},DefaultShadowColor := MakeColorFromHex("000000")}
    var DmgColorBlockDMG2 :color_block= color_block{DefaultColor := MakeColorFromHex("000000"),DefaultDesiredSize := vector2{X := 0.0,Y := 250.0},DefaultOpacity := 1.0}
    var DmgTexture2 :texture_block= texture_block{DefaultImage := UI.StoneIcon, DefaultDesiredSize := vector2{X := 50.0, Y:= 50.0}}
    var DmgBlockDMG3 :text_block= text_block{DefaultTextColor := MakeColorFromHex("FFFFFF"),DefaultShadowOffset := option{vector2{X:=3.0, Y:=3.0}},DefaultShadowColor := MakeColorFromHex("000000")}
    var DmgColorBlockDMG3 :color_block= color_block{DefaultColor := MakeColorFromHex("000000"),DefaultDesiredSize := vector2{X := 0.0,Y := 250.0},DefaultOpacity := 1.0}
    var DmgTexture3 :texture_block= texture_block{DefaultImage := UI.StoneIcon, DefaultDesiredSize := vector2{X := 50.0, Y:= 50.0}}
    var DmgBlockDMG4 :text_block= text_block{DefaultTextColor := MakeColorFromHex("FFFFFF"),DefaultShadowOffset := option{vector2{X:=3.0, Y:=3.0}},DefaultShadowColor := MakeColorFromHex("000000")}
    var DmgColorBlockDMG4 :color_block= color_block{DefaultColor := MakeColorFromHex("000000"),DefaultDesiredSize := vector2{X := 0.0,Y := 250.0},DefaultOpacity := 1.0}
    var DmgTexture4 :texture_block= texture_block{DefaultImage := UI.StoneIcon, DefaultDesiredSize := vector2{X := 50.0, Y:= 50.0}}
    var DmgBlockDMG5 :text_block= text_block{DefaultTextColor := MakeColorFromHex("FFFFFF"),DefaultShadowOffset := option{vector2{X:=3.0, Y:=3.0}},DefaultShadowColor := MakeColorFromHex("000000")}
    var DmgColorBlockDMG5 :color_block= color_block{DefaultColor := MakeColorFromHex("000000"),DefaultDesiredSize := vector2{X := 0.0,Y := 250.0},DefaultOpacity := 1.0}
    var DmgTexture5 :texture_block= texture_block{DefaultImage := UI.StoneIcon, DefaultDesiredSize := vector2{X := 50.0, Y:= 50.0}}
    var DmgBlockDMG6 :text_block= text_block{DefaultTextColor := MakeColorFromHex("FFFFFF"),DefaultShadowOffset := option{vector2{X:=3.0, Y:=3.0}},DefaultShadowColor := MakeColorFromHex("000000")}
    var DmgColorBlockDMG6 :color_block= color_block{DefaultColor := MakeColorFromHex("000000"),DefaultDesiredSize := vector2{X := 0.0,Y := 250.0},DefaultOpacity := 1.0}
    var DmgTexture6 :texture_block= texture_block{DefaultImage := UI.StoneIcon, DefaultDesiredSize := vector2{X := 50.0, Y:= 50.0}}
    var DmgBlockDMG7 :text_block= text_block{DefaultTextColor := MakeColorFromHex("FFFFFF"),DefaultShadowOffset := option{vector2{X:=3.0, Y:=3.0}},DefaultShadowColor := MakeColorFromHex("000000")}
    var DmgColorBlockDMG7 :color_block= color_block{DefaultColor := MakeColorFromHex("000000"),DefaultDesiredSize := vector2{X := 0.0,Y := 250.0},DefaultOpacity := 1.0}
    var DmgTexture7 :texture_block= texture_block{DefaultImage := UI.StoneIcon, DefaultDesiredSize := vector2{X := 50.0, Y:= 50.0}}
    var DmgBlockDMG8 :text_block= text_block{DefaultTextColor := MakeColorFromHex("FFFFFF"),DefaultShadowOffset := option{vector2{X:=3.0, Y:=3.0}},DefaultShadowColor := MakeColorFromHex("000000")}
    var DmgColorBlockDMG8 :color_block= color_block{DefaultColor := MakeColorFromHex("000000"),DefaultDesiredSize := vector2{X := 0.0,Y := 250.0},DefaultOpacity := 1.0}
    var DmgTexture8 :texture_block= texture_block{DefaultImage := UI.StoneIcon, DefaultDesiredSize := vector2{X := 50.0, Y:= 50.0}}

#FINAL BOSS POST GAME HUD

    var PostFBHUDBG : texture_block = texture_block{DefaultImage := UI.PostFBHUDBG}
    var PostFBHUDTitle : texture_block = texture_block{DefaultImage := UI.POST_BATTLE_RESULTS}
    var PostFBCloseButton : button_loud = button_loud{}

    var AllPostFBTextBlocks : []text_block = array{}
    var AllPostFBPlayerNames : []text_block = array{}
    var AllPostFBPlayerDamages : []text_block = array{}
    var AllBossRewardBlocks : []text_block = array{}
    
    #EACH VALUE
    var PostFBDamageTitle : text_block = text_block{DefaultTextColor := MakeColorFromHex("FFFFFF")}
    var PostFBName1 : text_block = text_block{DefaultTextColor := MakeColorFromHex("fada5a")}
    var PostFBDamage1 : text_block = text_block{DefaultTextColor := MakeColorFromHex("FFFFFF")}
    var PostFBName2 : text_block = text_block{DefaultTextColor := MakeColorFromHex("cacaca")}
    var PostFBDamage2 : text_block = text_block{DefaultTextColor := MakeColorFromHex("FFFFFF")}
    var PostFBName3 : text_block = text_block{DefaultTextColor := MakeColorFromHex("a87122")}
    var PostFBDamage3 : text_block = text_block{DefaultTextColor := MakeColorFromHex("FFFFFF")}
    var PostFBName4 : text_block = text_block{DefaultTextColor := MakeColorFromHex("FFFFFF")}
    var PostFBDamage4 : text_block = text_block{DefaultTextColor := MakeColorFromHex("FFFFFF")}

    var PostFBBossRewardTitle : text_block = text_block{DefaultTextColor := MakeColorFromHex("FFFFFF")}
    var PostFBBossDamageDealt : text_block = text_block{DefaultTextColor := MakeColorFromHex("FFFFFF")}
    var PostFBBossStoneReward : text_block = text_block{DefaultTextColor := MakeColorFromHex("FFFFFF")}
    var PostFBBossWepMatReward : text_block = text_block{DefaultTextColor := MakeColorFromHex("FFFFFF")}
    var PostFBBossShardReward : text_block = text_block{DefaultTextColor := MakeColorFromHex("FFFFFF")}
    var PostFBBossEggReward : text_block = text_block{DefaultTextColor := MakeColorFromHex("FFFFFF")}
    var PostFBBossGatherReward : text_block = text_block{DefaultTextColor := MakeColorFromHex("FFFFFF")}

    var PostFBPrestigeTitle : text_block = text_block{DefaultTextColor := MakeColorFromHex("FFFFFF")}
    var PostFBPrestigeLevel : text_block = text_block{DefaultTextColor := MakeColorFromHex("FFFFFF")}
    var PostFBPrestigeShards : text_block = text_block{DefaultTextColor := MakeColorFromHex("FFFFFF")}
    var PostFBPrestigeGather: text_block = text_block{DefaultTextColor := MakeColorFromHex("FFFFFF")}

#PINPAD HUD
    
    var PPButton0 : button_loud = button_loud{}
    var PPButton1 : button_loud = button_loud{}
    var PPButton2 : button_loud = button_loud{}
    var PPButton3 : button_loud = button_loud{}
    var PPButton4 : button_loud = button_loud{}
    var PPButton5 : button_loud = button_loud{}
    var PPButton6 : button_loud = button_loud{}
    var PPButton7 : button_loud = button_loud{}
    var PPButton8 : button_loud = button_loud{}
    var PPButton9 : button_loud = button_loud{}
    var PPButtonDelete : button_loud = button_loud{}
    var PPButtonEnter : button_loud = button_loud{}
    var NumTB : text_block = text_block{DefaultTextColor := MakeColorFromHex("FFFFFF")}

# QUEST TRACKER HUD

    var RockCounterBG : texture_block = texture_block{DefaultImage := UI.RockCounterUI}
    var MobCounterBG : texture_block = texture_block{DefaultImage := UI.MobCounterUI}
    var BossCounterBG : texture_block = texture_block{DefaultImage := UI.BossCounterUI}
    var RockCounterText : text_block = text_block:
        DefaultTextColor := MakeColorFromHex("DDDDDD")
        DefaultShadowOffset := option{vector2{X:=3.0, Y:=3.0}}
        DefaultShadowColor := MakeColorFromHex("000000")
        DefaultShadowOpacity := 1.0
    var MobCounterText : text_block = text_block:
        DefaultTextColor := MakeColorFromHex("DDDDDD")
        DefaultShadowOffset := option{vector2{X:=3.0, Y:=3.0}}
        DefaultShadowColor := MakeColorFromHex("000000")
        DefaultShadowOpacity := 1.0
    var BossCounterText : text_block = text_block:
        DefaultTextColor := MakeColorFromHex("DDDDDD")
        DefaultShadowOffset := option{vector2{X:=3.0, Y:=3.0}}
        DefaultShadowColor := MakeColorFromHex("000000")
        DefaultShadowOpacity := 1.0

    var AllCounterTexts : []text_block = array{}

# INFINITE SINKS HUD
    
    var ISBG : texture_block = texture_block{DefaultImage := UI.InfiniteSinkBG}
    
    var ISShardIcon : texture_block = texture_block{DefaultImage := UI.ShardIcon}
    var ISWepMatIcon : texture_block = texture_block{DefaultImage := UI.WepMat}
    var ISBossKeyIcon : texture_block = texture_block{DefaultImage := UI.BossKeyIcon}
    var ISShardPercentIcon : texture_block = texture_block{DefaultImage := UI.ShardIcon}
    var ISWepMatPercentIcon : texture_block = texture_block{DefaultImage := UI.WepMat}
    var ISBossKeyPercentIcon : texture_block = texture_block{DefaultImage := UI.BossKeyIcon}
    var ISTitle : texture_block = texture_block{DefaultImage := UI.InfiniteSinkTitle}

    var ISShardSlider : slider_regular = slider_regular{}
    var ISWepMatSlider : slider_regular = slider_regular{}
    var ISBossKeySlider : slider_regular = slider_regular{}

    var ISShardButton : button_quiet = button_quiet{}
    var ISWepMatButton : button_quiet = button_quiet{}
    var ISCloseButton : button_loud = button_loud{}
    var ISAllShardButton : button_quiet = button_quiet{}
    var ISAllWepMatButton : button_quiet = button_quiet{}
    var ISBossKeyButton : button_quiet = button_quiet{}
    var ISAllBossKeyButton : button_quiet = button_quiet{}

    var ISConverstionRateTitle : text_block = text_block{
        DefaultTextColor := MakeColorFromHex("FFFFFF")
        DefaultShadowOffset := option{vector2{X:=3.0, Y:=3.0}}
        DefaultShadowColor := MakeColorFromHex("000000")
        DefaultShadowOpacity := 1.0
        }
    var ISShardAmountTitle : text_block = text_block{
        DefaultTextColor := MakeColorFromHex("FFFFFF")
        DefaultShadowOffset := option{vector2{X:=3.0, Y:=3.0}}
        DefaultShadowColor := MakeColorFromHex("000000")
        DefaultShadowOpacity := 1.0
        }
    var ISBossKeyAmountTitle : text_block = text_block{
        DefaultTextColor := MakeColorFromHex("FFFFFF")
        DefaultShadowOffset := option{vector2{X:=3.0, Y:=3.0}}
        DefaultShadowColor := MakeColorFromHex("000000")
        DefaultShadowOpacity := 1.0
        }
    var ISWeppMatAmountTitle : text_block = text_block{
        DefaultTextColor := MakeColorFromHex("FFFFFF")
        DefaultShadowOffset := option{vector2{X:=3.0, Y:=3.0}}
        DefaultShadowColor := MakeColorFromHex("000000")
        DefaultShadowOpacity := 1.0
        }
    var ISPrestigePercentTitle : text_block = text_block{
        DefaultTextColor := MakeColorFromHex("FFFFFF")
        DefaultShadowOffset := option{vector2{X:=3.0, Y:=3.0}}
        DefaultShadowColor := MakeColorFromHex("000000")
        DefaultShadowOpacity := 1.0
        }
    var ISWeppMatPercentTitle : text_block = text_block{
        DefaultTextColor := MakeColorFromHex("FFFFFF")
        DefaultShadowOffset := option{vector2{X:=3.0, Y:=3.0}}
        DefaultShadowColor := MakeColorFromHex("000000")
        DefaultShadowOpacity := 1.0
        }
    var ISBossKeyPercentTitle : text_block = text_block{
        DefaultTextColor := MakeColorFromHex("FFFFFF")
        DefaultShadowOffset := option{vector2{X:=3.0, Y:=3.0}}
        DefaultShadowColor := MakeColorFromHex("000000")
        DefaultShadowOpacity := 1.0
        }
