
using { /Fortnite.com/Devices }
using { /Verse.org/Simulation }
using { /UnrealEngine.com/Temporary/Diagnostics }
using { /UnrealEngine.com/Temporary/SpatialMath }
using { /Verse.org/Random }

time_and_date := class<persistable><final><concrete>:
    @editable Year : int = -1
    @editable Month : int = -1
    @editable Day : int = -1
    @editable Hours : int = -1
    @editable Minutes : int = -1
    @editable Seconds : int = -1

MakeTimeAndDate(Year:int,Month:int,Day:int,Hours:int,Minutes:int,Seconds:int):time_and_date=
    return time_and_date:
        Year:=Year
        Month:=Month
        Day:=Day
        Hours:=Hours
        Minutes:=Minutes
        Seconds:=Seconds

MakeTimeAndDate(Prop:creative_prop):time_and_date=
    Transform:=Prop.GetTransform()
    if:
        Year:=Int[Transform.Translation.X]
        Month:=Int[Transform.Translation.Y]
        Day:=Int[Transform.Translation.Z]
        Hours:=Int[Transform.Scale.X]-1
        Minutes:=Int[Transform.Scale.Y]-1
        Seconds:=Int[Transform.Scale.Z]-1
    then:
        return MakeTimeAndDate(Year,Month,Day,Hours,Minutes,Seconds)
        
    return time_and_date{}

# A Verse-authored creative device that can be placed in a level
daily_rewards_manager := class(creative_device):

    @editable
    gm : game_manager = game_manager{}

    @editable
    GiveRewardsBtn : button_device = button_device{}

    @editable
    TimeProp : creative_prop = creative_prop{}

    var Month_Days : []int = array{31, 28, 31, 30, 31, 30, 31, 31, 30, 31, 30, 31}
    var DamageNumbers<private>: damage_numbers = damage_numbers{}

    # xGATHER RATE | EGG MIN | EGG MAX
    var AllDailyRewards : [][]float = array{
        array{150.0, 5000.0, 10000.0},
        array{200.0, 7500.0, 12500.0},
        array{400.0, 10000.0, 15000.0},
        array{600.0, 12500.0, 17500.0},
        array{850.0, 15000.0, 20000.0},
        array{1000.0, 17500.0, 22500.0},
        array{1200.0, 20000.0, 25000.0}
        }

    IsLeapYear(Year:int)<transacts>:logic=
        if ((Mod[Year,4] = 0 and Mod[Year,100] <> 0) or (Mod[Year,400]=0)):
            return true
        else:
            return false
    
    DaysInMonth(Month:int,Year:int):int=
        if (Month = 2 and IsLeapYear(Year)):
            return 29
        else if(Days:=Month_Days[Month-1]):
            return Days
        else:
            return -1
    
    DateToDays(Year:int,Month:int,Day:int):int=
        var total_days : int = Day
        for(m:int=1..Month):
            set total_days += DaysInMonth(m,Year)
        set total_days += 365*Year
        if:
            Int1 := 4-Year
            Int2 := 100+Year
            Floor1 := Floor(Year/Int1)
            Floor2 := Floor(Floor1/Int2)
            Floor3 := Floor(Floor2/400)
        then:
            set total_days += Floor3
        return total_days

    IsSameDay(Year1:int,Month1:int,Day1:int,Year2:int,Month2:int,Day2:int)<transacts>:logic=
        if (Year1 = Year2 and Month1 = Month2 and Day1 = Day2):
            return true
        else:
            return false

    GiveDailyRewards(CustomPlayer:Custom_Player)<suspends>:void=
        gm.ShowMessage(CustomPlayer.Player, "Rewards Claimed!", 0)
        for(i:int=0..3): #LOOP 4 TIMES TO BREAK UP REWARDS TO FEEL BETTER
            if(BossRewards:=AllDailyRewards[CustomPlayer.CurrentDay - 1]):
                if(StoneReward:=BossRewards[0]): #GIVE STONE
                    TotalReward := ((CustomPlayer.GetPlayerDamage()*StoneReward) / 4.0) #* CustomPlayer.BossRewardMult
                    CustomPlayer.Collect(TotalReward, "Stone", 0)
                    DamageNumbers.ShowDamageUIForPlayer(CustomPlayer, TotalReward, false, 0, false, "Rock", -1.0)
                    #gm.StoneGranter.GrantItem(CustomPlayer.Player)
                    Sleep(0.3)

                if(EggMinReward:=BossRewards[1],EggMaxReward:=BossRewards[2]): #GIVE EGGS
                    var whichEgg : int = 0
                    if(CustomPlayer.StagesUnlocked = -1):
                        if(whichDropTable:=gm.EggDropOdds[0],DropAmount := Floor[GetRandomFloat(EggMinReward, EggMaxReward)]*1.0, EggGranter:=gm.EggGranters[whichEgg]):
                            set whichEgg = SelectItem(whichDropTable)
                            TotalReward := (DropAmount / 4.0) #* CustomPlayer.BossRewardMult
                            CustomPlayer.Collect(TotalReward, "Egg", whichEgg)
                            DamageNumbers.ShowDamageUIForPlayer(CustomPlayer, TotalReward, false, 0, false, "Egg{whichEgg}", -1.0)
                            #EggGranter.GrantItem(CustomPlayer.Player)
                            Sleep(0.3)
                    else:
                        if(whichDropTable:=gm.EggDropOdds[CustomPlayer.StagesUnlocked],DropAmount := Floor[GetRandomFloat(EggMinReward, EggMaxReward)]*1.0, EggGranter:=gm.EggGranters[whichEgg]):
                            set whichEgg = SelectItem(whichDropTable)
                            TotalReward := (DropAmount / 4.0)
                            CustomPlayer.Collect(TotalReward, "Egg", whichEgg)
                            DamageNumbers.ShowDamageUIForPlayer(CustomPlayer, TotalReward, false, 0, false, "Egg{whichEgg}", -1.0)
                            #EggGranter.GrantItem(CustomPlayer.Player)
                            Sleep(0.3)
        if (CustomPlayer.CurrentDay < 7):            
            set CustomPlayer.CurrentDay += 1
        else:
            set CustomPlayer.CurrentDay = 1
        
        SaveAllData(CustomPlayer)

    ResetPlayerDay(CustomPlayer:Custom_Player,Date:time_and_date):void=
        set CustomPlayer.CurrentDay = 1
        set CustomPlayer.LastClaimed = Date

    CheckAndGiveDailyReward(CustomPlayer:Custom_Player, Prop:creative_prop):void=
        #Print("Get to Check")
        var Today : time_and_date = MakeTimeAndDate(Prop)
        if(CustomPlayer.CurrentDay = -1):
            #Print("Current day is -1")
            set CustomPlayer.CurrentDay = 1
            set CustomPlayer.LastClaimed = Today
            SaveAllData(CustomPlayer)
            spawn{GiveDailyRewards(CustomPlayer)}

        if (IsSameDay(CustomPlayer.LastClaimed.Year,CustomPlayer.LastClaimed.Month,CustomPlayer.LastClaimed.Day,Today.Year,Today.Month,Today.Day) = false):
            #Print("Checking if same day")
            days_diff := DateToDays(Today.Year,Today.Month,Today.Day) - DateToDays(CustomPlayer.LastClaimed.Year,CustomPlayer.LastClaimed.Month,CustomPlayer.LastClaimed.Day)
            #Print("days_diff: {days_diff}")
            if (days_diff >= 1):
                #Print("Days diff is more or equal to 1")
                ResetPlayerDay(CustomPlayer, Today)
                SaveAllData(CustomPlayer)
            spawn{GiveDailyRewards(CustomPlayer)}
        else:
            gm.ShowMessage(CustomPlayer.Player, "Come back tomorrow!", 1)

    OnDailyButnPress(Agent:agent):void=
        if(CustomPlayer := gm.PlayersMap[Agent]):
            CheckAndGiveDailyReward(CustomPlayer,TimeProp)

    OnBegin<override>()<suspends>:void=
        Sleep(2.0)
        GiveRewardsBtn.InteractedWithEvent.Subscribe(OnDailyButnPress)