date_data := class<persistable><final><concrete><public>:
    # Default values for the date (January 1, 1970, 00:00:00)
    Year : int = 1970
    Month : int = 1
    Day : int = 1
    Hour : int = 0
    Minute : int = 0
    Second : int = 0
    WeekDay : int = 0

    SECONDS_IN_A_DAY : int = 86400
    SECONDS_IN_AN_HOUR : int = 3600
    SECONDS_IN_A_MINUTE : int = 60

    # Function that initializes the date using epoch seconds
    ParseEpoch(EpochSeconds : float):date_data= 

        var DateData : date_data = date_data{}

        if:
            IntEpochSeconds : int = Round[EpochSeconds]
            EpochDays := IntEpochSeconds / SECONDS_IN_A_DAY
            DaysSinceEpoch : int = Floor(EpochDays)
            WK := GetWeekDay[IntEpochSeconds]
        then:
            RemainingSeconds := IntEpochSeconds - (DaysSinceEpoch * SECONDS_IN_A_DAY)
            DateVal : tuple(int,int,int) = Self.ConvertDaysToDate(DaysSinceEpoch)
            TimeVal : tuple(int,int,int) = Self.ConvertSecondsToTime(RemainingSeconds)
            set DateData = date_data{
                Year := DateVal(0), Month := DateVal(1), Day := DateVal(2),
                Hour := TimeVal(0), Minute := TimeVal(1), Second := TimeVal(2),
                WeekDay := WK
            }

        return DateData

    #Method to determine if two date data classes are the same day
    IsSameDay(DateData1:date_data,DateData2:date_data)<transacts>:logic=
        if (DateData1.Year = DateData2.Year and DateData1.Month = DateData2.Month and DateData1.Day = DateData2.Day):
            return true
        else:
            return false

    # Method to convert seconds left to time
    ConvertSecondsToTime(EpochSeconds : int):tuple(int,int,int)=
        var Time : tuple(int,int,int) = (0,0,0)
        if:
            Hours : int = Floor(EpochSeconds / 3600)
            Minutes : int = Floor((EpochSeconds - (Hours * 3600)) / 60)
            Seconds : int = EpochSeconds - (Hours * 3600) - (Minutes * 60)
        then:
            set Time = (Hours, Minutes, Seconds)
        
        return Time

    # Method to convert days since epoch to year, month, and day
    ConvertDaysToDate(DaysSinceEpoch : int) : tuple(int, int, int) =
        var CurrentYear : int = 1970
        var DaysRemaining : int = DaysSinceEpoch

        loop:
            DayCount := Self.DaysInYear(CurrentYear)
            if(DaysRemaining >= DayCount):
                set DaysRemaining -= Self.DaysInYear(CurrentYear)
                set CurrentYear += 1
            else:
                break

        var DaysInMonths : []int = Self.GetDaysInMonths(CurrentYear)
        var CurrentMonth : int = 1

        loop:
            if(DaysRemaining >= DaysInMonths[CurrentMonth - 1]):
                if:
                    set DaysRemaining -= DaysInMonths[CurrentMonth - 1]
                set CurrentMonth += 1
            else:
                break

        return (CurrentYear, CurrentMonth, DaysRemaining + 1)

    # Method to determine the number of days in a year
    DaysInYear(YearVal : int): int =
        return if (IsLeapYear(YearVal)?) then 366 else 365

    # Method to determine if a year is a leap year
    IsLeapYear(YearVal : int)<transacts>: logic =
        return if((Mod[YearVal,4] = 0) and (Mod[YearVal,100] <> 0) or (Mod[YearVal,400] = 0)) then true else false

    # Method to get the number of days in each month for a given year
    GetDaysInMonths(YearVal : int) : []int =
        return if (Self.IsLeapYear(YearVal)?) then array{31, 29, 31, 30, 31, 30, 31, 31, 30, 31, 30, 31} else array{31, 28, 31, 30, 31, 30, 31, 31, 30, 31, 30, 31}
    
    GetWeekDay(UTime:int)<transacts><decides>:int=
        DaysSinceEpoch := Floor(UTime / 86400)
        NewWeekDay := Mod[DaysSinceEpoch + 3,7]