
using { /Fortnite.com/Devices }
using { /Verse.org/Simulation }
using { /UnrealEngine.com/Temporary/Diagnostics }
using { /UnrealEngine.com/Temporary/UI }
using { /UnrealEngine.com/Temporary/SpatialMath }
using { /Verse.org/Colors }
using { /Verse.org/Random }

expeditions_manager := class(creative_device):

    @editable
    gm : game_manager = game_manager{}

    @editable
    AudioManager : audio_manager = audio_manager{}

    @editable
    ExpeditionHUDButton : button_device = button_device{}

    var DamageNumbers<private>: damage_numbers = damage_numbers{}

    var EXPPetNames : []string = array{"Dave", "Alien", "Ghost"}
    var EXPPetDrops : []string = array{"Alien", "Alien", "Ghost", "Ghost"}

    var EXPNames : []string = array{"Easy", "Medium", "Hard", "Extreme"}
    var EXPColors : []string = array{"57df87", "f1da57", "f93f3f", "7073FF"}

    var EXPTimes : []float = array{300.0,600.0,1200.0,1800.0} #IN SECONDS #300.0,600.0,1200.0,1800.0
    var EXPMinTimes : []float = array{150.0,300.0,600.0,900.0} #IN SECONDS #150.0,300.0,600.0,900.0

    var EXPStoneCosts : []float = array{1.71e49,3.41e60,1.71e73,1.68e87}
    var EXPBossKeyCosts : []float = array{5.0,25.0,100.0,300.0}

    #SHARD (CALLED Stone IN CODE THO BECAUSE LAZY) | EGG | WEP MAT | CHANCE FOR PET
    var ExpeditionRewards : [][]float = array{
        #Easy
        array{15.0, 100.0, 10.0, 0.10},
        #Medium
        array{80.0, 400.0, 40.0, 0.25},
        #Hard
        array{300.0, 1600.0, 80.0, 0.10},
        #Extreme
        array{1200.0, 6400.0, 460.0, 0.25} 
    }
    var PetInitialBonuses : [][]float = array{
        array{0.0, 0.0},
        array{0.05, 30.0},
        array{0.10, 60.0}
    }
    var PetBonuses : [][]float = array{
        array{0.0, 0.0},
        array{0.01, 15.0},
        array{0.02, 30.0}
    }
    
    OnBegin<override>()<suspends>:void=
        ExpeditionHUDButton.InteractedWithEvent.Subscribe(OpenExpeditionHUD)

        for(Spawner : gm.PlayerSpawners):     
            Spawner.SpawnedEvent.Subscribe(OnPlayerJoined)

    OnPlayerJoined(Agent:agent):void=
        spawn{OnPlayerJoinedBuffer(Agent)}

    OnPlayerJoinedBuffer(Agent:agent)<suspends>:void=
        Sleep(1.0)
        if(CustomPlayer := gm.PlayersMap[player[Agent]]):
            if(CustomPlayer.PlayTimeSessionSeconds = 0.0):
                CustomPlayer.EXPCloseButton.OnClick().Subscribe(RemoveExpeditionHUD)
                set CustomPlayer.AllEXPTexts = array{CustomPlayer.EXP1RewardPet,CustomPlayer.EXP2RewardPet,CustomPlayer.EXP3RewardPet,CustomPlayer.EXP4RewardPet, CustomPlayer.EXP1RewardWepMat,CustomPlayer.EXP2RewardWepMat,CustomPlayer.EXP3RewardWepMat,CustomPlayer.EXP4RewardWepMat,CustomPlayer.EXP1RewardEgg,CustomPlayer.EXP2RewardEgg,CustomPlayer.EXP3RewardEgg,CustomPlayer.EXP4RewardEgg,CustomPlayer.EXP1RewardStone,CustomPlayer.EXP2RewardStone,CustomPlayer.EXP3RewardStone,CustomPlayer.EXP4RewardStone,CustomPlayer.EXPStoneAmountText,CustomPlayer.EXPShardAmountText,CustomPlayer.EXPPetTitle,CustomPlayer.EXPPet1Bonuses,CustomPlayer.EXPPet2level,CustomPlayer.EXPPet2Bonuses,CustomPlayer.EXPPet3level,CustomPlayer.EXPPet3Bonuses,CustomPlayer.EXP1Title,CustomPlayer.EXP1Time,CustomPlayer.EXP2Title,CustomPlayer.EXP2Time,CustomPlayer.EXP3Title,CustomPlayer.EXP3Time,CustomPlayer.EXP4Title,CustomPlayer.EXP4Time,CustomPlayer.EXPDetailsTitle,CustomPlayer.EXPDetailsPet,CustomPlayer.EXPDetailsEXP,CustomPlayer.EXPDetailsStone,CustomPlayer.EXPDetailsEgg,CustomPlayer.EXPDetailsWepMat,CustomPlayer.EXPDetailsPetChance,CustomPlayer.EXPStoneCostText,CustomPlayer.EXPShardCostText}

                set CustomPlayer.AllEXPStoneRewards = array{CustomPlayer.EXP1RewardStone,CustomPlayer.EXP2RewardStone,CustomPlayer.EXP3RewardStone,CustomPlayer.EXP4RewardStone}
                set CustomPlayer.AllEXPEggRewards = array{CustomPlayer.EXP1RewardEgg,CustomPlayer.EXP2RewardEgg,CustomPlayer.EXP3RewardEgg,CustomPlayer.EXP4RewardEgg}
                set CustomPlayer.AllEXPWepMatRewards = array{CustomPlayer.EXP1RewardWepMat,CustomPlayer.EXP2RewardWepMat,CustomPlayer.EXP3RewardWepMat,CustomPlayer.EXP4RewardWepMat}
                set CustomPlayer.AllEXPPetRewards = array{CustomPlayer.EXP1RewardPet,CustomPlayer.EXP2RewardPet,CustomPlayer.EXP3RewardPet,CustomPlayer.EXP4RewardPet}
               
                set CustomPlayer.EXPSelectedPet = CustomPlayer.EXPCurrentPet
                if(CustomPlayer.CurrentExpedition <> -1):
                    set CustomPlayer.EXPSelectedExpedition = CustomPlayer.CurrentExpedition
                else:
                    set CustomPlayer.EXPSelectedExpedition = 0

                CustomPlayer.EXPPet1Equip.OnClick().SubscribeAgent3(EXPEquipPet, 0)
                CustomPlayer.EXPPet2Equip.OnClick().SubscribeAgent3(EXPEquipPet, 1)
                CustomPlayer.EXPPet3Equip.OnClick().SubscribeAgent3(EXPEquipPet, 2)

                CustomPlayer.EXPEasy.OnClick().SubscribeAgent3(SelectExpedition, 0)
                CustomPlayer.EXPMedium.OnClick().SubscribeAgent3(SelectExpedition, 1)
                CustomPlayer.EXPHard.OnClick().SubscribeAgent3(SelectExpedition, 2)
                CustomPlayer.EXPExtreme.OnClick().SubscribeAgent3(SelectExpedition, 3)
                CustomPlayer.EXPLaunch.OnClick().Subscribe(LaunchExpedition)

                if(CustomPlayer.EXPTotalRewards[0] > 0.0):
                    #Print("START EXPEDITION LOGIN")
                    spawn{StartExpeditionTimer(CustomPlayer)}
                else:
                    set CustomPlayer.CurrentExpedition = -1

                if(CustomPlayer.ExpeditionTimeLeft > 0.0 and CustomPlayer.EXPTotalRewards[0] = 0.0):
                    #Print("NO ACTUAL EXPEDITION")
                    set CustomPlayer.ExpeditionTimeLeft = 0.0

    EXPEquipPet(Wdata:widget_message, WhichPet:int): void=
        if(CustomPlayer := gm.PlayersMap[Wdata.Player]):
            if:
                WhichPet = 1
                CustomPlayer.NewEXPPetsLevel[WhichPet] >= 1.0
            then:
                #Print("Alien Equipped")
                set CustomPlayer.EXPSelectedPet = WhichPet
                UpdateExpeditionHUD(CustomPlayer)
            if:
                WhichPet = 2
                CustomPlayer.NewEXPPetsLevel[WhichPet] >= 1.0
            then:
                #Print("Ghost Equipped")
                set CustomPlayer.EXPSelectedPet = WhichPet
                UpdateExpeditionHUD(CustomPlayer)
            if:
                WhichPet = 0
            then:
                #Print("Dave Equipped")
                set CustomPlayer.EXPSelectedPet = WhichPet
                UpdateExpeditionHUD(CustomPlayer)

    SelectExpedition(Wdata:widget_message, WhichEXP:int): void=
        if(CustomPlayer := gm.PlayersMap[Wdata.Player]):
            set CustomPlayer.EXPSelectedExpedition = WhichEXP
            UpdateExpeditionHUD(CustomPlayer)

    LaunchExpedition(Wdata:widget_message): void=
        if(CustomPlayer := gm.PlayersMap[Wdata.Player]):
            if(CustomPlayer.EXPTotalRewards[0] = 0.0):
                if(StoneCost:=EXPStoneCosts[CustomPlayer.EXPSelectedExpedition], BossKeyCost:=EXPBossKeyCosts[CustomPlayer.EXPSelectedExpedition]):
                    if(CustomPlayer.StoneAmount >= StoneCost and CustomPlayer.BossKeyAmount >= BossKeyCost):
                        #CustomPlayer.EXPLaunch.SetEnabled(false)
                        RemoveExpeditionHUD(Wdata)

                        CustomPlayer.Spend(StoneCost, "Stone", 0)
                        CustomPlayer.Spend(BossKeyCost, "BossKey", 0)

                        set CustomPlayer.EXPCurrentPet = CustomPlayer.EXPSelectedPet

                        if:
                            EXPTime := EXPTimes[CustomPlayer.EXPSelectedExpedition]
                            EXPMinTime := EXPMinTimes[CustomPlayer.EXPSelectedExpedition]
                            PetLevel := CustomPlayer.NewEXPPetsLevel[CustomPlayer.EXPSelectedPet]
                            TimeInitialBonus := PetInitialBonuses[CustomPlayer.EXPSelectedPet][1]
                            TimeBonus := PetBonuses[CustomPlayer.EXPSelectedPet][1]
                        then:
                            TimeAfterBonuses := EXPTime - (TimeInitialBonus + (PetLevel * TimeBonus))
                            if(TimeAfterBonuses > EXPMinTime):
                                set CustomPlayer.ExpeditionTimeLeft = TimeAfterBonuses
                            else:
                                set CustomPlayer.ExpeditionTimeLeft = EXPMinTime
                        if:
                            PetLevel := CustomPlayer.NewEXPPetsLevel[CustomPlayer.EXPSelectedPet]
                            PetRewardBonus := (PetInitialBonuses[CustomPlayer.EXPSelectedPet][0]) + (PetLevel * (PetBonuses[CustomPlayer.EXPSelectedPet][0])) #% MORE GAINED
                            Stone := ExpeditionRewards[CustomPlayer.EXPSelectedExpedition][0] + (ExpeditionRewards[CustomPlayer.EXPSelectedExpedition][0] * PetRewardBonus)
                            Eggs := ExpeditionRewards[CustomPlayer.EXPSelectedExpedition][1] + (ExpeditionRewards[CustomPlayer.EXPSelectedExpedition][1] * PetRewardBonus)
                            WepMat := ExpeditionRewards[CustomPlayer.EXPSelectedExpedition][2] + (ExpeditionRewards[CustomPlayer.EXPSelectedExpedition][2] * PetRewardBonus)

                            StoneReward := GetRandomFloat(Stone * 0.90, Stone)
                            EggReward := GetRandomFloat(Eggs * 0.90, Eggs)
                            WepMatReward := GetRandomFloat(WepMat * 0.90, WepMat)
                            PetDropChance := ExpeditionRewards[CustomPlayer.EXPSelectedExpedition][3]
                        then:
                            set CustomPlayer.EXPTotalRewards = array{StoneReward, EggReward, WepMatReward, PetDropChance}
                        spawn{StartExpeditionTimer(CustomPlayer)}
                        AudioManager.EXPLaunch.Play(CustomPlayer.Player)
                        SaveAllData(CustomPlayer)
                    else:
                        AudioManager.Denied.Play(CustomPlayer.Player)
            else:
                #CustomPlayer.EXPLaunch.SetEnabled(false)
                spawn{ClaimExpeditionRewards(CustomPlayer, CustomPlayer.EXPTotalRewards, CustomPlayer.CurrentExpedition)}
                RemoveExpeditionHUD(Wdata)
                SaveAllData(CustomPlayer)
            

    OpenExpeditionHUD(Agent:agent): void=
        if(CustomPlayer := gm.PlayersMap[Agent]):
                if(MyUI:=CustomPlayer.UIMaps["Expedition"]?):
                else:
                    if(CustomPlayer.ExpeditionTimeLeft = 0.0):
                        if(CustomPlayer.EXPTotalRewards[0] = 0.0):
                            var MyExpeditionCanvas : canvas = MakeExpeditionCanvas(CustomPlayer)
                            CustomPlayer.AddFullHUD(MyExpeditionCanvas)
                            if(set CustomPlayer.UIMaps["Expedition"] = option{MyExpeditionCanvas}) {}
                            UpdateExpeditionHUD(CustomPlayer)
                        else:
                            spawn{ClaimExpeditionRewards(CustomPlayer, CustomPlayer.EXPTotalRewards, CustomPlayer.CurrentExpedition)}
                            SaveAllData(CustomPlayer)
                            gm.ShowMessage(CustomPlayer.Player, "Rewards Claimed!", 0)
                    else:
                        gm.ShowMessage(CustomPlayer.Player, "Expedition still in progress!", 1)

    RemoveExpeditionHUD(Wdata:widget_message):void=
        if(CustomPlayer := gm.PlayersMap[Wdata.Player]):
            if(Player:=Wdata.Player, PlayerUI:=GetPlayerUI[Player]):
                if(MyUI:=CustomPlayer.UIMaps["Expedition"]?):
                    PlayerUI.RemoveWidget(MyUI)
                    MyUI.SetVisibility(widget_visibility.Collapsed)
                    if(set CustomPlayer.UIMaps["Expedition"] = false){}
    
    StartExpeditionTimer(CustomPlayer:Custom_Player)<suspends>:void=
        CustomPlayer.EXPTimerText.SetText(StringToMessage(""))
        if(CustomPlayer.CurrentExpedition = -1 or CustomPlayer.PlayTimeSessionSeconds = 0.0):
            OpenEXPTimerHUD(CustomPlayer.Player)
            if(CustomPlayer.CurrentExpedition = -1):
                set CustomPlayer.CurrentExpedition = CustomPlayer.EXPSelectedExpedition
            loop:
                if(CustomPlayer.ExpeditionTimeLeft > 0.0):
                    set CustomPlayer.ExpeditionTimeLeft -= 1.0
                    if(PlayerUI:=GetPlayerUI[CustomPlayer.Player],MyUI:=CustomPlayer.UIMaps["Expedition"]?):
                        UpdateExpeditionHUD(CustomPlayer)
                    if(PlayerUI:=GetPlayerUI[CustomPlayer.Player],MyUI:=CustomPlayer.UIMaps["EXPTimer"]?):
                        if(CustomPlayer.ExpeditionTimeLeft > 0.0):
                            CustomPlayer.EXPTimerText.SetText(StringToMessage(ConvertToTime(CustomPlayer.ExpeditionTimeLeft,"Minutes")))
                        else:
                            CustomPlayer.EXPTimerText.SetText(StringToMessage("Done!"))
                            break:
                    SaveAllData(CustomPlayer)
                else:
                    Print("TIMER SHOULD BE OVER")
                    set CustomPlayer.ExpeditionTimeLeft = 0.0
                    if(CustomPlayer.EXPTotalRewards[0] > 0.0):
                        CustomPlayer.EXPTimerText.SetText(StringToMessage("Done!"))

                    #RemoveEXPTimerHUD(CustomPlayer)
                    break:
                Sleep(1.0)

    ClaimExpeditionRewards(CustomPlayer:Custom_Player, Rewards:[]float, CurrentExpedition:int)<suspends>:void=
        if(CustomPlayer.CurrentExpedition <> -1 and CustomPlayer.ExpeditionTimeLeft <= 0.0):
            RemoveEXPTimerHUD(CustomPlayer)
            set CustomPlayer.CurrentExpedition = -1
            set CustomPlayer.EXPTotalRewards = array{0.0, 0.0, 0.0, 0.0}
            set CustomPlayer.EXPCurrentPet = 0 

            if(Accolade:=AudioManager.Accolades[13]): #AWARD PLAYER EXP
                Accolade.Award(CustomPlayer.Player)
            if(Analytic:=AudioManager.GeneralAnalytics[6]): #ANALYTIC
                Analytic.Submit(CustomPlayer.Player)

            if(StoneReward:=Rewards[0]): #GIVE SHARDS
                CustomPlayer.Collect(StoneReward, "Shard", 0)
                DamageNumbers.ShowDamageUIForPlayer(CustomPlayer, StoneReward, false, 0, false, "Shard", -1.0)
                Sleep(0.3)

            if(EggReward:=Rewards[1]): #GIVE EGGS
                var whichEgg : int = 0
                if(whichDropTable:=gm.EggDropOdds[CustomPlayer.StagesUnlocked], EggGranter:=gm.EggGranters[whichEgg]):
                    set whichEgg = SelectItem(whichDropTable)
                    CustomPlayer.Collect(EggReward, "Egg", whichEgg)
                    DamageNumbers.ShowDamageUIForPlayer(CustomPlayer, EggReward, false, 0, false, "Egg{whichEgg}", -1.0)
                    Sleep(0.3)   

            if(WepMatReward:=Rewards[2]): #GIVE WEP MATS
                CustomPlayer.Collect(WepMatReward, "WepMat", 0)
                DamageNumbers.ShowDamageUIForPlayer(CustomPlayer, WepMatReward, false, 0, false, "WepMat", -1.0)
                Sleep(0.3)

            IDX:=GetRandomInt(0,100)
            if(Floor[Rewards[3]*100.0] > IDX):
                if(CurrentExpedition = 0 or CurrentExpedition = 1):
                    AudioManager.EXPClaimPet.Play(CustomPlayer.Player)
                    gm.ShowMessage(CustomPlayer.Player, "You found an Alien!", 0)
                    if(set CustomPlayer.NewEXPPetsLevel[1] += 1.0) {}
                if(CurrentExpedition = 2 or CurrentExpedition = 3):
                    AudioManager.EXPClaimPet.Play(CustomPlayer.Player)
                    gm.ShowMessage(CustomPlayer.Player, "You found a Ghost!", 0)
                    if(set CustomPlayer.NewEXPPetsLevel[2] += 1.0) {}
            else:
                AudioManager.EXPClaim.Play(CustomPlayer.Player)

        SaveAllData(CustomPlayer)

    UpdateExpeditionHUD(CustomPlayer:Custom_Player):void=
        CustomPlayer.EXPStoneAmountText.SetText(StringToMessage(" {NumberCondensor(CustomPlayer.StoneAmount)}"))
        CustomPlayer.EXPShardAmountText.SetText(StringToMessage(" {NumberCondensor(CustomPlayer.BossKeyAmount)}"))
        
        CustomPlayer.EXPStoneCostText.SetTextColor(MakeColorFromHex("FFFFFF"))
        CustomPlayer.EXPShardCostText.SetTextColor(MakeColorFromHex("FFFFFF"))
        CustomPlayer.EXPStoneCostIcon.SetVisibility(widget_visibility.Collapsed)
        CustomPlayer.EXPShardCostIcon.SetVisibility(widget_visibility.Collapsed)
        <#if(CustomPlayer.ExpeditionTimeLeft > 0.0):
            CustomPlayer.EXPLaunch.SetText(StringToMessage("PROGRESS"))
            CustomPlayer.EXPStoneCostText.SetText(StringToMessage("Time Left:"))
            CustomPlayer.EXPShardCostText.SetText(StringToMessage("{ConvertToTime(CustomPlayer.ExpeditionTimeLeft,"Minutes")}"))
        else if(CustomPlayer.EXPTotalRewards[0] > 0.0):
            CustomPlayer.EXPLaunch.SetEnabled(true)
            CustomPlayer.EXPLaunch.SetText(StringToMessage("CLAIM"))
            CustomPlayer.EXPStoneCostText.SetText(StringToMessage("Claim Expedition"))
            CustomPlayer.EXPShardCostText.SetText(StringToMessage(""))
        else:#>
        CustomPlayer.EXPLaunch.SetText(StringToMessage("LAUNCH"))
        CustomPlayer.EXPStoneCostIcon.SetVisibility(widget_visibility.Visible)
        CustomPlayer.EXPShardCostIcon.SetVisibility(widget_visibility.Visible)
        <#CustomPlayer.EXPPet1Equip.SetEnabled(true)
        CustomPlayer.EXPEasy.SetEnabled(true)
        CustomPlayer.EXPMedium.SetEnabled(true)
        CustomPlayer.EXPHard.SetEnabled(true)
        CustomPlayer.EXPExtreme.SetEnabled(true)
        CustomPlayer.EXPLaunch.SetEnabled(true)
        if(CustomPlayer.NewEXPPetsLevel[1] >= 1.0):
            CustomPlayer.EXPPet2Equip.SetEnabled(true)
        if(CustomPlayer.NewEXPPetsLevel[2] >= 1.0):
            CustomPlayer.EXPPet3Equip.SetEnabled(true)#>

        if(StoneCost:=EXPStoneCosts[CustomPlayer.EXPSelectedExpedition], BossKeyCost:=EXPBossKeyCosts[CustomPlayer.EXPSelectedExpedition]):
            CustomPlayer.EXPStoneCostText.SetText(StringToMessage("{NumberCondensor(StoneCost)}"))
            CustomPlayer.EXPShardCostText.SetText(StringToMessage("{NumberCondensor(BossKeyCost)}"))
            if(CustomPlayer.StoneAmount >= StoneCost):
                CustomPlayer.EXPStoneCostText.SetTextColor(MakeColorFromHex("39FF39"))
            else:    
                CustomPlayer.EXPStoneCostText.SetTextColor(MakeColorFromHex("FA4D56"))

            if(CustomPlayer.BossKeyAmount >= BossKeyCost):
                CustomPlayer.EXPShardCostText.SetTextColor(MakeColorFromHex("39FF39"))
            else:    
                CustomPlayer.EXPShardCostText.SetTextColor(MakeColorFromHex("FA4D56"))
               
        #ITS SET UP LIKE THIS TO CATCH WHAT THE PROBLEM IS
        if:
            PetEquipped:=EXPPetNames[CustomPlayer.EXPSelectedPet]
        then:
            #Print("PET EQUIPPED GOOD")
            CustomPlayer.EXPDetailsPet.SetText(StringToMessage("Pet: {PetEquipped}"))
        if:
            AlienLevel := CustomPlayer.NewEXPPetsLevel[1]
        then:       
            #Print("ALIEN LEVEL GOOD")    
            CustomPlayer.EXPPet2level.SetText(StringToMessage("Level: {NumberCondensor(AlienLevel)}"))
            if:
                AlienRewardBonus := (PetInitialBonuses[1][0]*100.0) + (AlienLevel * (PetBonuses[1][0]*100.0)) #% MORE GAINED
                AlienTimeBonus :=  PetInitialBonuses[1][1] + (AlienLevel * PetBonuses[1][1]) #SECONDS OFF EXPEDITION
            then:
                #Print("ALIEN BONUSES GOOD")
                CustomPlayer.EXPPet2Bonuses.SetText(StringToMessage("+{NumberCondensor(AlienRewardBonus)}% Rewards\n-{NumberCondensor(AlienTimeBonus)} Seconds"))
        if:
            GhostLevel := CustomPlayer.NewEXPPetsLevel[2]
        then:
            #Print("GHOST LEVEL GOOD")
            CustomPlayer.EXPPet3level.SetText(StringToMessage("Level: {NumberCondensor(GhostLevel)}"))
            if:
                GhostRewardBonus := (PetInitialBonuses[2][0]*100.0) + (GhostLevel * (PetBonuses[2][0]*100.0)) #% MORE GAINED
                GhostTimeBonus := PetInitialBonuses[2][1] + (GhostLevel * PetBonuses[2][1]) #SECONDS OFF EXPEDITION
            then:
                #Print("GHOST BONUSES GOOD")
                CustomPlayer.EXPPet3Bonuses.SetText(StringToMessage("+{NumberCondensor(GhostRewardBonus)}% Rewards\n-{NumberCondensor(GhostTimeBonus)} Seconds"))


        for(i:int=0..CustomPlayer.AllEXPEggRewards.Length-1):
            if:
                Stone := ExpeditionRewards[i][0]
                Eggs := ExpeditionRewards[i][1]
                WepMat := ExpeditionRewards[i][2]
                PetDropChance := ExpeditionRewards[i][3]
                PetDropName := EXPPetDrops[i]
                CurrentStoneText := CustomPlayer.AllEXPStoneRewards[i]
                CurrentEggText := CustomPlayer.AllEXPEggRewards[i]
                CurrentWepmatText := CustomPlayer.AllEXPWepMatRewards[i]
                CurrentPetText := CustomPlayer.AllEXPPetRewards[i]
            then:
                CurrentStoneText.SetText(StringToMessage(" {NumberCondensor(Stone * 0.90)} - {NumberCondensor(Stone)}"))
                CurrentEggText.SetText(StringToMessage(" {NumberCondensor(Eggs * 0.90)} - {NumberCondensor(Eggs)}"))
                CurrentWepmatText.SetText(StringToMessage(" {NumberCondensor(WepMat * 0.90)} - {NumberCondensor(WepMat)}"))
                CurrentPetText.SetText(StringToMessage("{NumberCondensor(PetDropChance*100.0)}% Chance for {PetDropName}"))
        if:
            CurrentEXP:=EXPNames[CustomPlayer.EXPSelectedExpedition]
            EXPColor:=EXPColors[CustomPlayer.EXPSelectedExpedition]
        then:
            CustomPlayer.EXPDetailsEXP.SetText(StringToMessage("{CurrentEXP} Expedition"))
            CustomPlayer.EXPDetailsEXP.SetTextColor(MakeColorFromHex(EXPColor))

        #SET UP REWARDS AND SHOW HERE
        if:
            PetLevel := CustomPlayer.NewEXPPetsLevel[CustomPlayer.EXPSelectedPet]
            PetRewardBonus := (PetInitialBonuses[CustomPlayer.EXPSelectedPet][0]) + (PetLevel * (PetBonuses[CustomPlayer.EXPSelectedPet][0])) #% MORE GAINED
            Stone := ExpeditionRewards[CustomPlayer.EXPSelectedExpedition][0] + (ExpeditionRewards[CustomPlayer.EXPSelectedExpedition][0] * PetRewardBonus)
            Eggs := ExpeditionRewards[CustomPlayer.EXPSelectedExpedition][1] + (ExpeditionRewards[CustomPlayer.EXPSelectedExpedition][1] * PetRewardBonus)
            WepMat := ExpeditionRewards[CustomPlayer.EXPSelectedExpedition][2] + (ExpeditionRewards[CustomPlayer.EXPSelectedExpedition][2] * PetRewardBonus)
            PetDropChance := ExpeditionRewards[CustomPlayer.EXPSelectedExpedition][3]
            PetDropName := EXPPetDrops[CustomPlayer.EXPSelectedExpedition]
        then:
            CustomPlayer.EXPDetailsStone.SetText(StringToMessage(" {NumberCondensor(Stone*0.90)} - {NumberCondensor(Stone)}"))
            CustomPlayer.EXPDetailsEgg.SetText(StringToMessage(" {NumberCondensor(Eggs*0.90)} - {NumberCondensor(Eggs)}"))
            CustomPlayer.EXPDetailsWepMat.SetText(StringToMessage(" {NumberCondensor(WepMat*0.90)} - {NumberCondensor(WepMat)}"))
            CustomPlayer.EXPDetailsPetChance.SetText(StringToMessage(" {NumberCondensor(PetDropChance*100.0)}% Chance for {PetDropName}"))
            



    MakeExpeditionCanvas(CustomPlayer:Custom_Player):canvas=
        CustomPlayer.EXPCloseButton.SetText(StringToMessage("Close"))
        CustomPlayer.EXPPetTitle.SetText(StringToMessage("Pets"))
        CustomPlayer.EXPPet1Bonuses.SetText(StringToMessage("No Bonuses\nNo Time Speed Up"))
        CustomPlayer.EXP1Title.SetText(StringToMessage("Easy Expedition"))
        CustomPlayer.EXP1Time.SetText(StringToMessage("5 Minutes"))
        CustomPlayer.EXP2Title.SetText(StringToMessage("Medium Expedition"))
        CustomPlayer.EXP2Time.SetText(StringToMessage("10 Minutes"))
        CustomPlayer.EXP3Title.SetText(StringToMessage("Hard Expedition"))
        CustomPlayer.EXP3Time.SetText(StringToMessage("20 Minutes"))
        CustomPlayer.EXP4Title.SetText(StringToMessage("Extreme Expedition"))
        CustomPlayer.EXP4Time.SetText(StringToMessage("30 Minutes"))
        CustomPlayer.EXPDetailsTitle.SetText(StringToMessage("Details"))
        CustomPlayer.EXPPet1Equip.SetText(StringToMessage("Equip: Dave"))
        CustomPlayer.EXPPet2Equip.SetText(StringToMessage("Equip: Alien"))
        CustomPlayer.EXPPet3Equip.SetText(StringToMessage("Equip: Ghost"))
        
        <#CustomPlayer.EXPPet1Equip.SetEnabled(false)
        CustomPlayer.EXPPet2Equip.SetEnabled(false)
        CustomPlayer.EXPPet3Equip.SetEnabled(false)
        CustomPlayer.EXPEasy.SetEnabled(false)
        CustomPlayer.EXPMedium.SetEnabled(false)
        CustomPlayer.EXPHard.SetEnabled(false)
        CustomPlayer.EXPExtreme.SetEnabled(false)
        CustomPlayer.EXPLaunch.SetEnabled(false)#>
        
        for(TextBlock : CustomPlayer.AllEXPTexts):
            TextBlock.SetShadowColor(CustomPlayer.ShadowDefaults.DefaultShadowColor)
            TextBlock.SetShadowOffset(CustomPlayer.ShadowDefaults.DefaultShadowOffset)
            TextBlock.SetShadowOpacity(1.0)

        var MyExpeditionCanvas: canvas = canvas:
            Slots := array:
                canvas_slot:
                    Widget:=CustomPlayer.EXPHUDBG
                    ZOrder:=0
                    SizeToContent:=false
                    Alignment := vector2{X := 0.5, Y := 0.0}
                    Anchors := anchors{Minimum := vector2{X:= 0.5, Y:= 0.0}, Maximum := vector2{X:= 0.5, Y:= 0.0}}
                    Offsets:=margin{Top:=20.0,Left:=0.0,Bottom:=1024.0,Right:=2048.0}
                canvas_slot:
                    Widget:=CustomPlayer.EXPHUDTitle
                    ZOrder:=1
                    SizeToContent:=false
                    Alignment := vector2{X := 0.5, Y := 0.0}
                    Anchors := anchors{Minimum := vector2{X:= 0.5, Y:= 0.0}, Maximum := vector2{X:= 0.5, Y:= 0.0}}
                    Offsets:=margin{Top:=130.0,Left:=100.0,Bottom:=56.0,Right:=405.0}
                
                canvas_slot:
                    Widget:=CustomPlayer.EXPStoneAmountIcon
                    ZOrder:=1
                    SizeToContent:=true
                    Alignment := vector2{X:= 0.0, Y:= 0.0}
                    Anchors := anchors{Minimum := vector2{X:= 0.5, Y:= 0.0}, Maximum := vector2{X:= 0.5, Y:= 0.0}}
                    Offsets:=margin{Top:=128.0, Left:=-764.0}
                canvas_slot:
                    Widget:=CustomPlayer.EXPStoneAmountText
                    ZOrder:=1
                    SizeToContent:=true
                    Alignment := vector2{X:= 0.0, Y:= 0.0}
                    Anchors := anchors{Minimum := vector2{X:= 0.5, Y:= 0.0}, Maximum := vector2{X:= 0.5, Y:= 0.0}}
                    Offsets:=margin{Top:=128.0,Left:=-708.0}
                canvas_slot:
                    Widget:=CustomPlayer.EXPShardAmountIcon
                    ZOrder:=1
                    SizeToContent:=true
                    Alignment := vector2{X:= 0.0, Y:= 0.0}
                    Anchors := anchors{Minimum := vector2{X:= 0.5, Y:= 0.0}, Maximum := vector2{X:= 0.5, Y:= 0.0}}
                    Offsets:=margin{Top:=128.0, Left:=-444.0}
                canvas_slot:
                    Widget:=CustomPlayer.EXPShardAmountText
                    ZOrder:=1
                    SizeToContent:=true
                    Alignment := vector2{X:= 0.0, Y:= 0.0}
                    Anchors := anchors{Minimum := vector2{X:= 0.5, Y:= 0.0}, Maximum := vector2{X:= 0.5, Y:= 0.0}}
                    Offsets:=margin{Top:=128.0,Left:=-388.0}

                canvas_slot:
                    Widget:=CustomPlayer.EXPPetTitle
                    ZOrder:=1
                    SizeToContent:=true
                    Alignment := vector2{X:= 0.5, Y:= 0.0}
                    Anchors := anchors{Minimum := vector2{X:= 0.5, Y:= 0.0}, Maximum := vector2{X:= 0.5, Y:= 0.0}}
                    Offsets:=margin{Top:=244.0,Left:=-548.0}

                canvas_slot:
                    Widget:=CustomPlayer.EXPPet1Equip
                    ZOrder:=1
                    SizeToContent:=false
                    Alignment := vector2{X:= 0.5, Y:= 0.0}
                    Anchors := anchors{Minimum := vector2{X:= 0.5, Y:= 0.0}, Maximum := vector2{X:= 0.5, Y:= 0.0}}
                    Offsets:=margin{Top:=336.0,Left:=-612.0, Bottom:= 50.0, Right:=250.0}
                canvas_slot:
                    Widget:=CustomPlayer.EXPPet1Icon
                    ZOrder:=1
                    SizeToContent:=false
                    Alignment := vector2{X:= 0.0, Y:= 0.0}
                    Anchors := anchors{Minimum := vector2{X:= 0.5, Y:= 0.0}, Maximum := vector2{X:= 0.5, Y:= 0.0}}
                    Offsets:=margin{Top:=332.0,Left:=-444.0, Bottom:=90.0, Right:= 90.0}
                canvas_slot:
                    Widget:=CustomPlayer.EXPPet1Bonuses
                    ZOrder:=1
                    SizeToContent:=true
                    Alignment := vector2{X:= 0.0, Y:= 0.0}
                    Anchors := anchors{Minimum := vector2{X:= 0.5, Y:= 0.0}, Maximum := vector2{X:= 0.5, Y:= 0.0}}
                    Offsets:=margin{Top:=416.0,Left:=-740.0}

                canvas_slot:
                    Widget:=CustomPlayer.EXPPet2Equip
                    ZOrder:=1
                    SizeToContent:=false
                    Alignment := vector2{X:= 0.5, Y:= 0.0}
                    Anchors := anchors{Minimum := vector2{X:= 0.5, Y:= 0.0}, Maximum := vector2{X:= 0.5, Y:= 0.0}}
                    Offsets:=margin{Top:=564.0,Left:=-612.0, Bottom:= 50.0, Right:=250.0}
                canvas_slot:
                    Widget:=CustomPlayer.EXPPet2Icon
                    ZOrder:=1
                    SizeToContent:=false
                    Alignment := vector2{X:= 0.0, Y:= 0.0}
                    Anchors := anchors{Minimum := vector2{X:= 0.5, Y:= 0.0}, Maximum := vector2{X:= 0.5, Y:= 0.0}}
                    Offsets:=margin{Top:=556.0,Left:=-444.0, Bottom:=90.0, Right:= 90.0}
                canvas_slot:
                    Widget:=CustomPlayer.EXPPet2level
                    ZOrder:=1
                    SizeToContent:=true
                    Alignment := vector2{X:= 0.5, Y:= 0.0}
                    Anchors := anchors{Minimum := vector2{X:= 0.5, Y:= 0.0}, Maximum := vector2{X:= 0.5, Y:= 0.0}}
                    Offsets:=margin{Top:=628.0,Left:=-612.0}
                canvas_slot:
                    Widget:=CustomPlayer.EXPPet2Bonuses
                    ZOrder:=1
                    SizeToContent:=true
                    Alignment := vector2{X:= 0.0, Y:= 0.0}
                    Anchors := anchors{Minimum := vector2{X:= 0.5, Y:= 0.0}, Maximum := vector2{X:= 0.5, Y:= 0.0}}
                    Offsets:=margin{Top:=676.0,Left:=-740.0}

                canvas_slot:
                    Widget:=CustomPlayer.EXPPet3Equip
                    ZOrder:=1
                    SizeToContent:=false
                    Alignment := vector2{X:= 0.5, Y:= 0.0}
                    Anchors := anchors{Minimum := vector2{X:= 0.5, Y:= 0.0}, Maximum := vector2{X:= 0.5, Y:= 0.0}}
                    Offsets:=margin{Top:=792.0,Left:=-612.0, Bottom:= 50.0, Right:=250.0}
                canvas_slot:
                    Widget:=CustomPlayer.EXPPet3Icon
                    ZOrder:=1
                    SizeToContent:=false
                    Alignment := vector2{X:= 0.0, Y:= 0.0}
                    Anchors := anchors{Minimum := vector2{X:= 0.5, Y:= 0.0}, Maximum := vector2{X:= 0.5, Y:= 0.0}}
                    Offsets:=margin{Top:=784.0,Left:=-444.0, Bottom:=90.0, Right:= 90.0}
                canvas_slot:
                    Widget:=CustomPlayer.EXPPet3level
                    ZOrder:=1
                    SizeToContent:=true
                    Alignment := vector2{X:= 0.5, Y:= 0.0}
                    Anchors := anchors{Minimum := vector2{X:= 0.5, Y:= 0.0}, Maximum := vector2{X:= 0.5, Y:= 0.0}}
                    Offsets:=margin{Top:=856.0,Left:=-612.0}
                canvas_slot:
                    Widget:=CustomPlayer.EXPPet3Bonuses
                    ZOrder:=1
                    SizeToContent:=true
                    Alignment := vector2{X:= 0.0, Y:= 0.0}
                    Anchors := anchors{Minimum := vector2{X:= 0.5, Y:= 0.0}, Maximum := vector2{X:= 0.5, Y:= 0.0}}
                    Offsets:=margin{Top:=904.0,Left:=-740.0}

                canvas_slot:
                    Widget:=CustomPlayer.EXPEasy
                    ZOrder:=2
                    SizeToContent:=false
                    Alignment := vector2{X:= 0.5, Y:= 0.0}
                    Anchors := anchors{Minimum := vector2{X:= 0.5, Y:= 0.0}, Maximum := vector2{X:= 0.5, Y:= 0.0}}
                    Offsets:=margin{Top:=224.0,Left:=-4.0, Bottom:= 275.0, Right:=450.0}
                canvas_slot:
                    Widget:=CustomPlayer.EXP1Title
                    ZOrder:=1
                    SizeToContent:=true
                    Alignment := vector2{X:= 0.5, Y:= 0.0}
                    Anchors := anchors{Minimum := vector2{X:= 0.5, Y:= 0.0}, Maximum := vector2{X:= 0.5, Y:= 0.0}}
                    Offsets:=margin{Top:=240.0,Left:=-4.0}
                canvas_slot:
                    Widget:=CustomPlayer.EXP1Time
                    ZOrder:=1
                    SizeToContent:=true
                    Alignment := vector2{X:= 0.0, Y:= 0.0}
                    Anchors := anchors{Minimum := vector2{X:= 0.5, Y:= 0.0}, Maximum := vector2{X:= 0.5, Y:= 0.0}}
                    Offsets:=margin{Top:=280.0,Left:=-220.0}
                canvas_slot:
                    ZOrder:=1
                    SizeToContent:=true
                    Alignment := vector2{X:= 0.5, Y:= 0.0}
                    Anchors := anchors{Minimum := vector2{X:= 0.5, Y:= 0.0}, Maximum := vector2{X:= 0.5, Y:= 0.0}}
                    Offsets:=margin{Top:=324.0,Left:=-4.0}
                    Widget := stack_box:
                        Orientation := orientation.Horizontal
                        Slots := array:
                            stack_box_slot:
                                Widget := texture_block:
                                    DefaultImage := UI.ShardIcon
                            stack_box_slot:
                                Widget := CustomPlayer.EXP1RewardStone
                canvas_slot:
                    ZOrder:=1
                    SizeToContent:=true
                    Alignment := vector2{X:= 0.5, Y:= 0.0}
                    Anchors := anchors{Minimum := vector2{X:= 0.5, Y:= 0.0}, Maximum := vector2{X:= 0.5, Y:= 0.0}}
                    Offsets:=margin{Top:=368.0,Left:=-4.0}
                    Widget := stack_box:
                        Orientation := orientation.Horizontal
                        Slots := array:
                            stack_box_slot:
                                Widget := texture_block:
                                    DefaultImage := UI.Egg1
                            stack_box_slot:
                                Widget := CustomPlayer.EXP1RewardEgg
                canvas_slot:
                    ZOrder:=1
                    SizeToContent:=true
                    Alignment := vector2{X:= 0.5, Y:= 0.0}
                    Anchors := anchors{Minimum := vector2{X:= 0.5, Y:= 0.0}, Maximum := vector2{X:= 0.5, Y:= 0.0}}
                    Offsets:=margin{Top:=412.0,Left:=-4.0}
                    Widget := stack_box:
                        Orientation := orientation.Horizontal
                        Slots := array:
                            stack_box_slot:
                                Widget := texture_block:
                                    DefaultImage := UI.WepMat
                            stack_box_slot:
                                Widget := CustomPlayer.EXP1RewardWepMat
                canvas_slot:
                    Widget:=CustomPlayer.EXP1RewardPet
                    ZOrder:=1
                    SizeToContent:=true
                    Alignment := vector2{X:= 0.5, Y:= 0.0}
                    Anchors := anchors{Minimum := vector2{X:= 0.5, Y:= 0.0}, Maximum := vector2{X:= 0.5, Y:= 0.0}}
                    Offsets:=margin{Top:=456.0,Left:=-4.0}


                canvas_slot:
                    Widget:=CustomPlayer.EXPMedium
                    ZOrder:=2
                    SizeToContent:=false
                    Alignment := vector2{X:= 0.5, Y:= 0.0}
                    Anchors := anchors{Minimum := vector2{X:= 0.5, Y:= 0.0}, Maximum := vector2{X:= 0.5, Y:= 0.0}}
                    Offsets:=margin{Top:=224.0,Left:=496.0, Bottom:= 275.0, Right:=450.0}
                canvas_slot:
                    Widget:=CustomPlayer.EXP2Title
                    ZOrder:=1
                    SizeToContent:=true
                    Alignment := vector2{X:= 0.5, Y:= 0.0}
                    Anchors := anchors{Minimum := vector2{X:= 0.5, Y:= 0.0}, Maximum := vector2{X:= 0.5, Y:= 0.0}}
                    Offsets:=margin{Top:=240.0,Left:=496.0}
                canvas_slot:
                    Widget:=CustomPlayer.EXP2Time
                    ZOrder:=1
                    SizeToContent:=true
                    Alignment := vector2{X:= 0.0, Y:= 0.0}
                    Anchors := anchors{Minimum := vector2{X:= 0.5, Y:= 0.0}, Maximum := vector2{X:= 0.5, Y:= 0.0}}
                    Offsets:=margin{Top:=280.0,Left:=284.0}
                canvas_slot:
                    ZOrder:=1
                    SizeToContent:=true
                    Alignment := vector2{X:= 0.5, Y:= 0.0}
                    Anchors := anchors{Minimum := vector2{X:= 0.5, Y:= 0.0}, Maximum := vector2{X:= 0.5, Y:= 0.0}}
                    Offsets:=margin{Top:=324.0,Left:=496.0}
                    Widget := stack_box:
                        Orientation := orientation.Horizontal
                        Slots := array:
                            stack_box_slot:
                                Widget := texture_block:
                                    DefaultImage := UI.ShardIcon
                            stack_box_slot:
                                Widget := CustomPlayer.EXP2RewardStone
                canvas_slot:
                    ZOrder:=1
                    SizeToContent:=true
                    Alignment := vector2{X:= 0.5, Y:= 0.0}
                    Anchors := anchors{Minimum := vector2{X:= 0.5, Y:= 0.0}, Maximum := vector2{X:= 0.5, Y:= 0.0}}
                    Offsets:=margin{Top:=368.0,Left:=496.0}
                    Widget := stack_box:
                        Orientation := orientation.Horizontal
                        Slots := array:
                            stack_box_slot:
                                Widget := texture_block:
                                    DefaultImage := UI.Egg1
                            stack_box_slot:
                                Widget := CustomPlayer.EXP2RewardEgg
                canvas_slot:
                    ZOrder:=1
                    SizeToContent:=true
                    Alignment := vector2{X:= 0.5, Y:= 0.0}
                    Anchors := anchors{Minimum := vector2{X:= 0.5, Y:= 0.0}, Maximum := vector2{X:= 0.5, Y:= 0.0}}
                    Offsets:=margin{Top:=412.0,Left:=496.0}
                    Widget := stack_box:
                        Orientation := orientation.Horizontal
                        Slots := array:
                            stack_box_slot:
                                Widget := texture_block:
                                    DefaultImage := UI.WepMat
                            stack_box_slot:
                                Widget := CustomPlayer.EXP2RewardWepMat
                canvas_slot:
                    Widget:=CustomPlayer.EXP2RewardPet
                    ZOrder:=1
                    SizeToContent:=true
                    Alignment := vector2{X:= 0.5, Y:= 0.0}
                    Anchors := anchors{Minimum := vector2{X:= 0.5, Y:= 0.0}, Maximum := vector2{X:= 0.5, Y:= 0.0}}
                    Offsets:=margin{Top:=456.0,Left:=496.0}


                canvas_slot:
                    Widget:=CustomPlayer.EXPHard
                    ZOrder:=2
                    SizeToContent:=false
                    Alignment := vector2{X:= 0.5, Y:= 0.0}
                    Anchors := anchors{Minimum := vector2{X:= 0.5, Y:= 0.0}, Maximum := vector2{X:= 0.5, Y:= 0.0}}
                    Offsets:=margin{Top:=516.0,Left:=-4.0, Bottom:= 275.0, Right:=450.0}
                canvas_slot:
                    Widget:=CustomPlayer.EXP3Title
                    ZOrder:=1
                    SizeToContent:=true
                    Alignment := vector2{X:= 0.5, Y:= 0.0}
                    Anchors := anchors{Minimum := vector2{X:= 0.5, Y:= 0.0}, Maximum := vector2{X:= 0.5, Y:= 0.0}}
                    Offsets:=margin{Top:=520.0,Left:=-4.0}
                canvas_slot:
                    Widget:=CustomPlayer.EXP3Time
                    ZOrder:=1
                    SizeToContent:=true
                    Alignment := vector2{X:= 0.0, Y:= 0.0}
                    Anchors := anchors{Minimum := vector2{X:= 0.5, Y:= 0.0}, Maximum := vector2{X:= 0.5, Y:= 0.0}}
                    Offsets:=margin{Top:=564.0,Left:=-220.0}
                canvas_slot:
                    ZOrder:=1
                    SizeToContent:=true
                    Alignment := vector2{X:= 0.5, Y:= 0.0}
                    Anchors := anchors{Minimum := vector2{X:= 0.5, Y:= 0.0}, Maximum := vector2{X:= 0.5, Y:= 0.0}}
                    Offsets:=margin{Top:=604.0,Left:=-4.0}
                    Widget := stack_box:
                        Orientation := orientation.Horizontal
                        Slots := array:
                            stack_box_slot:
                                Widget := texture_block:
                                    DefaultImage := UI.ShardIcon
                            stack_box_slot:
                                Widget := CustomPlayer.EXP3RewardStone
                canvas_slot:
                    ZOrder:=1
                    SizeToContent:=true
                    Alignment := vector2{X:= 0.5, Y:= 0.0}
                    Anchors := anchors{Minimum := vector2{X:= 0.5, Y:= 0.0}, Maximum := vector2{X:= 0.5, Y:= 0.0}}
                    Offsets:=margin{Top:=648.0,Left:=-4.0}
                    Widget := stack_box:
                        Orientation := orientation.Horizontal
                        Slots := array:
                            stack_box_slot:
                                Widget := texture_block:
                                    DefaultImage := UI.Egg1
                            stack_box_slot:
                                Widget := CustomPlayer.EXP3RewardEgg
                canvas_slot:
                    ZOrder:=1
                    SizeToContent:=true
                    Alignment := vector2{X:= 0.5, Y:= 0.0}
                    Anchors := anchors{Minimum := vector2{X:= 0.5, Y:= 0.0}, Maximum := vector2{X:= 0.5, Y:= 0.0}}
                    Offsets:=margin{Top:=692.0,Left:=-4.0}
                    Widget := stack_box:
                        Orientation := orientation.Horizontal
                        Slots := array:
                            stack_box_slot:
                                Widget := texture_block:
                                    DefaultImage := UI.WepMat
                            stack_box_slot:
                                Widget := CustomPlayer.EXP3RewardWepMat
                canvas_slot:
                    Widget:=CustomPlayer.EXP3RewardPet
                    ZOrder:=1
                    SizeToContent:=true
                    Alignment := vector2{X:= 0.5, Y:= 0.0}
                    Anchors := anchors{Minimum := vector2{X:= 0.5, Y:= 0.0}, Maximum := vector2{X:= 0.5, Y:= 0.0}}
                    Offsets:=margin{Top:=736.0,Left:=-4.0}


                canvas_slot:
                    Widget:=CustomPlayer.EXPExtreme
                    ZOrder:=2
                    SizeToContent:=false
                    Alignment := vector2{X:= 0.5, Y:= 0.0}
                    Anchors := anchors{Minimum := vector2{X:= 0.5, Y:= 0.0}, Maximum := vector2{X:= 0.5, Y:= 0.0}}
                    Offsets:=margin{Top:=516.0,Left:=496.0, Bottom:= 275.0, Right:=450.0}
                canvas_slot:
                    Widget:=CustomPlayer.EXP4Title
                    ZOrder:=1
                    SizeToContent:=true
                    Alignment := vector2{X:= 0.5, Y:= 0.0}
                    Anchors := anchors{Minimum := vector2{X:= 0.5, Y:= 0.0}, Maximum := vector2{X:= 0.5, Y:= 0.0}}
                    Offsets:=margin{Top:=520.0,Left:=496.0}
                canvas_slot:
                    Widget:=CustomPlayer.EXP4Time
                    ZOrder:=1
                    SizeToContent:=true
                    Alignment := vector2{X:= 0.0, Y:= 0.0}
                    Anchors := anchors{Minimum := vector2{X:= 0.5, Y:= 0.0}, Maximum := vector2{X:= 0.5, Y:= 0.0}}
                    Offsets:=margin{Top:=564.0,Left:=284.0}
                canvas_slot:
                    ZOrder:=1
                    SizeToContent:=true
                    Alignment := vector2{X:= 0.5, Y:= 0.0}
                    Anchors := anchors{Minimum := vector2{X:= 0.5, Y:= 0.0}, Maximum := vector2{X:= 0.5, Y:= 0.0}}
                    Offsets:=margin{Top:=604.0,Left:=496.0}
                    Widget := stack_box:
                        Orientation := orientation.Horizontal
                        Slots := array:
                            stack_box_slot:
                                Widget := texture_block:
                                    DefaultImage := UI.ShardIcon
                            stack_box_slot:
                                Widget := CustomPlayer.EXP4RewardStone
                canvas_slot:
                    ZOrder:=1
                    SizeToContent:=true
                    Alignment := vector2{X:= 0.5, Y:= 0.0}
                    Anchors := anchors{Minimum := vector2{X:= 0.5, Y:= 0.0}, Maximum := vector2{X:= 0.5, Y:= 0.0}}
                    Offsets:=margin{Top:=648.0,Left:=496.0}
                    Widget := stack_box:
                        Orientation := orientation.Horizontal
                        Slots := array:
                            stack_box_slot:
                                Widget := texture_block:
                                    DefaultImage := UI.Egg1
                            stack_box_slot:
                                Widget := CustomPlayer.EXP4RewardEgg
                canvas_slot:
                    ZOrder:=1
                    SizeToContent:=true
                    Alignment := vector2{X:= 0.5, Y:= 0.0}
                    Anchors := anchors{Minimum := vector2{X:= 0.5, Y:= 0.0}, Maximum := vector2{X:= 0.5, Y:= 0.0}}
                    Offsets:=margin{Top:=692.0,Left:=496.0}
                    Widget := stack_box:
                        Orientation := orientation.Horizontal
                        Slots := array:
                            stack_box_slot:
                                Widget := texture_block:
                                    DefaultImage := UI.WepMat
                            stack_box_slot:
                                Widget := CustomPlayer.EXP4RewardWepMat
                canvas_slot:
                    Widget:=CustomPlayer.EXP4RewardPet
                    ZOrder:=1
                    SizeToContent:=true
                    Alignment := vector2{X:= 0.5, Y:= 0.0}
                    Anchors := anchors{Minimum := vector2{X:= 0.5, Y:= 0.0}, Maximum := vector2{X:= 0.5, Y:= 0.0}}
                    Offsets:=margin{Top:=736.0,Left:=496.0}

                canvas_slot:
                    Widget:=CustomPlayer.EXPDetailsTitle
                    ZOrder:=1
                    SizeToContent:=false
                    Alignment := vector2{X:= 0.5, Y:= 0.0}
                    Anchors := anchors{Minimum := vector2{X:= 0.5, Y:= 0.0}, Maximum := vector2{X:= 0.5, Y:= 0.0}}
                    Offsets:=margin{Top:=816.0,Left:=-188.0}
                canvas_slot:
                    Widget:=CustomPlayer.EXPDetailsPet
                    ZOrder:=1
                    SizeToContent:=false
                    Alignment := vector2{X:= 0.5, Y:= 0.0}
                    Anchors := anchors{Minimum := vector2{X:= 0.5, Y:= 0.0}, Maximum := vector2{X:= 0.5, Y:= 0.0}}
                    Offsets:=margin{Top:=868.0,Left:=-224.0}
                canvas_slot:
                    Widget:=CustomPlayer.EXPDetailsEXP
                    ZOrder:=1
                    SizeToContent:=false
                    Alignment := vector2{X:= 0.5, Y:= 0.0}
                    Anchors := anchors{Minimum := vector2{X:= 0.5, Y:= 0.0}, Maximum := vector2{X:= 0.5, Y:= 0.0}}
                    Offsets:=margin{Top:=920.0,Left:=-224.0}
                canvas_slot:
                    ZOrder:=1
                    SizeToContent:=true
                    Alignment := vector2{X:= 1.0, Y:= 0.0}
                    Anchors := anchors{Minimum := vector2{X:= 0.5, Y:= 0.0}, Maximum := vector2{X:= 0.5, Y:= 0.0}}
                    Offsets:=margin{Top:=808.0,Left:=472.0}
                    Widget := stack_box:
                        Orientation := orientation.Horizontal
                        Slots := array:
                            stack_box_slot:
                                Widget := texture_block:
                                    DefaultImage := UI.ShardIcon
                            stack_box_slot:
                                Widget := CustomPlayer.EXPDetailsStone
                canvas_slot:
                    ZOrder:=1
                    SizeToContent:=true
                    Alignment := vector2{X:= 1.0, Y:= 0.0}
                    Anchors := anchors{Minimum := vector2{X:= 0.5, Y:= 0.0}, Maximum := vector2{X:= 0.5, Y:= 0.0}}
                    Offsets:=margin{Top:=852.0,Left:=472.0}
                    Widget := stack_box:
                        Orientation := orientation.Horizontal
                        Slots := array:
                            stack_box_slot:
                                Widget := texture_block:
                                    DefaultImage := UI.Egg1
                            stack_box_slot:
                                Widget := CustomPlayer.EXPDetailsEgg
                canvas_slot:
                    ZOrder:=1
                    SizeToContent:=true
                    Alignment := vector2{X:= 1.0, Y:= 0.0}
                    Anchors := anchors{Minimum := vector2{X:= 0.5, Y:= 0.0}, Maximum := vector2{X:= 0.5, Y:= 0.0}}
                    Offsets:=margin{Top:=896.0,Left:=472.0}
                    Widget := stack_box:
                        Orientation := orientation.Horizontal
                        Slots := array:
                            stack_box_slot:
                                Widget := texture_block:
                                    DefaultImage := UI.WepMat
                            stack_box_slot:
                                Widget := CustomPlayer.EXPDetailsWepMat
                canvas_slot:
                    Widget:=CustomPlayer.EXPDetailsPetChance
                    ZOrder:=1
                    SizeToContent:=true
                    Alignment := vector2{X:= 1.0, Y:= 0.0}
                    Anchors := anchors{Minimum := vector2{X:= 0.5, Y:= 0.0}, Maximum := vector2{X:= 0.5, Y:= 0.0}}
                    Offsets:=margin{Top:=940.0,Left:=472.0}
                canvas_slot:
                    ZOrder:=1
                    SizeToContent:=true
                    Alignment := vector2{X:= 1.0, Y:= 0.0}
                    Anchors := anchors{Minimum := vector2{X:= 0.5, Y:= 0.0}, Maximum := vector2{X:= 0.5, Y:= 0.0}}
                    Offsets:=margin{Top:=812.0,Left:=748.0}
                    Widget := stack_box:
                        Orientation := orientation.Horizontal
                        Slots := array:
                            stack_box_slot:
                                Widget := CustomPlayer.EXPStoneCostIcon  
                            stack_box_slot:
                                Widget := CustomPlayer.EXPStoneCostText
                canvas_slot:
                    ZOrder:=1
                    SizeToContent:=true
                    Alignment := vector2{X:= 1.0, Y:= 0.0}
                    Anchors := anchors{Minimum := vector2{X:= 0.5, Y:= 0.0}, Maximum := vector2{X:= 0.5, Y:= 0.0}}
                    Offsets:=margin{Top:=860.0,Left:=748.0}
                    Widget := stack_box:
                        Orientation := orientation.Horizontal
                        Slots := array:
                            stack_box_slot:
                                Widget := CustomPlayer.EXPShardCostIcon  
                            stack_box_slot:
                                Widget := CustomPlayer.EXPShardCostText
                canvas_slot:
                    Widget:=CustomPlayer.EXPLaunch
                    ZOrder:=1
                    SizeToContent:=false
                    Alignment := vector2{X:= 0.5, Y:= 0.0}
                    Anchors := anchors{Minimum := vector2{X:= 0.5, Y:= 0.0}, Maximum := vector2{X:= 0.5, Y:= 0.0}}
                    Offsets:=margin{Top:=920.0,Left:=632.0, Bottom:= 50.0, Right:=250.0}

                canvas_slot:
                    Widget:=CustomPlayer.EXPCloseButton
                    ZOrder:=9
                    SizeToContent:=false
                    Alignment := vector2{X:= 0.5, Y:= 0.0}
                    Anchors := anchors{Minimum := vector2{X:= 0.5, Y:= 0.0}, Maximum := vector2{X:= 0.5, Y:= 0.0}}
                    Offsets:=margin{Top:= 128.0,Left:= 632.0,Bottom:= 50.0,Right:= 200.0}

        return MyExpeditionCanvas

                
#CANVAS
    MakeEXPTimerCanvas(CustomPlayer:Custom_Player):canvas=
        CustomPlayer.EXPTimerText.SetText(StringToMessage(""))
        CustomPlayer.EXPTimerText.SetShadowColor(CustomPlayer.ShadowDefaults.DefaultShadowColor)
        CustomPlayer.EXPTimerText.SetShadowOffset(CustomPlayer.ShadowDefaults.DefaultShadowOffset)
        CustomPlayer.EXPTimerText.SetShadowOpacity(1.0)
        if(EXPColor:=EXPColors[CustomPlayer.CurrentExpedition]):
            CustomPlayer.EXPTimerText.SetTextColor(MakeColorFromHex("{EXPColor}"))

        var MyEXPTimerCanvas: canvas = canvas:
            Slots := array:
                canvas_slot:
                    Widget:=CustomPlayer.EXPTimerHUDBG
                    ZOrder:=0
                    SizeToContent:=false
                    Anchors := anchors{Minimum := vector2{X:= 1.0, Y:= 0.0}, Maximum := vector2{X:= 1.0, Y:= 0.0}}
                    Alignment := vector2{X:= 0.5, Y:= 0.0}
                    Offsets:=margin{Top:=120.0,Left:=-68.0,Bottom:=120.0,Right:=166.0}
                canvas_slot:
                    Widget:=CustomPlayer.EXPTimerText
                    ZOrder:=1
                    SizeToContent:=true
                    Anchors := anchors{Minimum := vector2{X:= 1.0, Y:= 0.0}, Maximum := vector2{X:= 1.0, Y:= 0.0}}
                    Offsets:=margin{Top:=180.0,Left:=-68.0}
                    Alignment := vector2{X:= 0.5, Y:= 0.0}

        return MyEXPTimerCanvas

    OpenEXPTimerHUD(Agent:agent): void=
        Print("SHOW EXP TIMER")
        if(CustomPlayer := gm.PlayersMap[Agent]):
            if(MyUI:=CustomPlayer.UIMaps["EXPTimer"]?):
                if(EXPColor:=EXPColors[CustomPlayer.EXPSelectedExpedition]):
                    CustomPlayer.EXPTimerText.SetTextColor(MakeColorFromHex("{EXPColor}"))
                MyUI.SetVisibility(widget_visibility.Visible)
            else:
                var MyEXPTimerCanvas : canvas = MakeEXPTimerCanvas(CustomPlayer)
                CustomPlayer.AddUI(MyEXPTimerCanvas)
                if(set CustomPlayer.UIMaps["EXPTimer"] = option{MyEXPTimerCanvas}) {}
                if(EXPColor:=EXPColors[CustomPlayer.EXPSelectedExpedition]):
                    CustomPlayer.EXPTimerText.SetTextColor(MakeColorFromHex("{EXPColor}"))

    RemoveEXPTimerHUD(CustomPlayer:Custom_Player):void=
        Print("HIDE EXP TIMER")
            if(PlayerUI:=GetPlayerUI[CustomPlayer.Player]):
                if(MyUI:=CustomPlayer.UIMaps["EXPTimer"]?):
                    MyUI.SetVisibility(widget_visibility.Collapsed)
