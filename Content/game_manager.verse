using { /Fortnite.com/Devices }
using { /Verse.org/Simulation }
using { /Verse.org/Colors }
using { /Verse.org/Verse }
using { /Verse.org/Random }
using { /Fortnite.com/Teams }
using { /UnrealEngine.com/Assets }
using { /Fortnite.com/Game }
using { /Fortnite.com/Characters }
using { /Fortnite.com/Playspaces }
using { /Fortnite.com/UI }
using { /UnrealEngine.com/Temporary/Diagnostics }
using { /UnrealEngine.com/Temporary/SpatialMath }
using { /UnrealEngine.com/Temporary/UI }
using { /Verse.org/Assets}
using { /Verse.org/Colors/NamedColors }


game_manager := class(creative_device):
    var Version : int = 9 #CHANGE THIS HERE & IN PERSISTENCE TO WIPE PEOPLES PROGRESS
    #CHANGE IT BACK TO 9 PLEASE
    
    @editable
    PlayerSpawners : []player_spawner_device = array{}

    @editable
    FireMissile : creative_prop_asset = DefaultCreativePropAsset

    @editable
    BaseClassSelector : class_and_team_selector_device = class_and_team_selector_device{}
    @editable
    ClassInstantReload : class_and_team_selector_device = class_and_team_selector_device{}

    var PlayersMap : [player]Custom_Player = map{}

    var SessionTime : float = 0.0

    @editable
    MessageDevices : []hud_message_device = array{}
    @editable
    MovementUpgradeGranters : []movement_modulator_device = array{}

    @editable
    NPCSpawners : []npc_spawner_device = array{}
    @editable
    NPCSpawnTriggers : []mutator_zone_device = array{}


    @editable
    MatchmakingPortal : mutator_zone_device = mutator_zone_device{}
    @editable
    PortalAnalytic : analytics_device = analytics_device{}


    @editable
    AudioManager: audio_manager = audio_manager{}

    @editable
    TESTStoneButtons: []button_device = array{}

    @editable
    TESTBossKeyButtons: []button_device = array{}

    @editable
    TESTShardsButtons: []button_device = array{}

    @editable
    TESTSilverButtons: []button_device = array{}

    @editable
    TESTPMButton: button_device = button_device{}

    @editable
    TESTEggsButton: button_device = button_device{}

    #@editable
    #PatchNotesButton: button_device = button_device{}

    #PatchNotesHUDDevice : hud_message_device = hud_message_device{}

    @editable
    LeaderboardPlayers : []player_reference_device = array{}

    @editable
    LeaderboardBillboards : []billboard_device = array{}
    
    @editable
    StoneGranter : item_granter_device = item_granter_device{}
    @editable
    WepMatGranter : item_granter_device = item_granter_device{}
    @editable
    EggGranters : []item_granter_device = array{}

    @editable
    SpaceTeleporter : teleporter_device = teleporter_device{}

    @editable
    SpaceGravity : mutator_zone_device = mutator_zone_device{}

    @editable
    SeasonalBillboard: billboard_device = billboard_device{}

    EggDropOdds : [][]float = array{
        array{100.0, 00.00, 00.00, 00.00, 00.00, 00.00},
        array{79.50, 20.00, 00.50, 00.00, 00.00, 00.00},
        array{59.00, 40.00, 01.00, 00.00, 00.00, 00.00},
        array{38.00, 60.00, 02.00, 00.00, 00.00, 00.00},
        array{35.00, 59.50, 05.00, 00.50, 00.00, 00.00},
        array{32.00, 46.00, 20.00, 02.00, 00.00, 00.00},
        array{26.00, 29.00, 40.00, 05.00, 00.00, 00.00},
        array{20.00, 24.50, 45.00, 10.00, 00.50, 00.00},
        array{15.00, 20.00, 34.00, 30.00, 01.00, 00.00},
        array{12.00, 19.00, 32.00, 35.00, 02.00, 00.00},
        array{09.00, 15.50, 30.00, 40.00, 05.00, 00.50},
        array{06.00, 13.00, 25.00, 30.00, 25.00, 01.00},
        array{03.00, 12.00, 22.00, 25.00, 30.00, 08.00},
        array{02.00, 07.50, 17.50, 25.50, 32.50, 15.00},
        array{00.00, 05.00, 15.50, 22.50, 35.00, 22.00},
        array{00.00, 03.00, 13.50, 23.50, 36.00, 24.00},
        array{00.00, 02.00, 13.50, 23.00, 36.50, 25.00},
        array{00.00, 01.00, 12.00, 22.00, 38.00, 27.00},
        array{00.00, 00.00, 12.00, 20.00, 40.00, 28.00},
        array{00.00, 00.00, 05.50, 20.00, 39.50, 35.00},

        array{35.00, 30.00, 02.00, 03.00, 08.00, 22.00},
        array{35.00, 30.00, 02.00, 03.00, 08.00, 22.00},
        array{35.00, 30.00, 02.00, 03.00, 08.00, 22.00},
        array{35.00, 30.00, 02.00, 03.00, 08.00, 22.00},
        array{35.00, 30.00, 02.00, 03.00, 08.00, 22.00}
    }

    var PetInitialBonusValues : []float = array{2.0,4.0,8.0,16.0,28.0,44.0}

    OnBegin<override>()<suspends>:void=

        #PatchNotesButton.InteractedWithEvent.Subscribe(ShowPatchNotes)

        for(i:int=0..TESTStoneButtons.Length):
            if(HutUpgradeButton:=TESTStoneButtons[i]):
                HutUpgradeButton.InteractedWithEvent.SubscribeAgent(OnTestStoneButton, i)
                
        for(i:int=0..TESTBossKeyButtons.Length):
            if(HutUpgradeButton:=TESTBossKeyButtons[i]):
                HutUpgradeButton.InteractedWithEvent.SubscribeAgent(OnTestBossKeyButton, i)
                
        for(i:int=0..TESTShardsButtons.Length):
            if(HutUpgradeButton:=TESTShardsButtons[i]):
                HutUpgradeButton.InteractedWithEvent.SubscribeAgent(OnTestShardButton, i)

        for(i:int=0..TESTSilverButtons.Length):
            if(HutUpgradeButton:=TESTSilverButtons[i]):
                HutUpgradeButton.InteractedWithEvent.SubscribeAgent(OnTestSilverButton, i)

        TESTPMButton.InteractedWithEvent.Subscribe(OnTestPMButton)

        TESTEggsButton.InteractedWithEvent.Subscribe(OnTestEggsButton)

        SpaceGravity.AgentEntersEvent.Subscribe(EnterSpace)

        MatchmakingPortal.AgentEntersEvent.Subscribe(StartPortalCheck)
        MatchmakingPortal.AgentExitsEvent.Subscribe(StopPortalCheck)

        for(NPCSpawnTrigger:NPCSpawnTriggers):
            NPCSpawnTrigger.AgentEntersEvent.Subscribe(SpawnShopNPCs)

        GetPlayspace().PlayerRemovedEvent().Subscribe(OnPlayerLeftServer)

        spawn{TrackSessionTime()}

        for(Spawner : PlayerSpawners):     
            Spawner.SpawnedEvent.Subscribe(OnPlayerLoad)  #Activates when a player is spawned on a spawn pad

    #ShowPatchNotes(Agent:agent):void=
    #    PatchNotesHUDDevice.Show(Agent)

    var LookingAtPortal : logic = false

    StartPortalCheck(Agent:agent):void=
        set LookingAtPortal = true
        spawn{TimerToAnalytic(Agent)}
        
    StopPortalCheck(Agent:agent):void=
        set LookingAtPortal = false

    TimerToAnalytic(Agent:agent)<suspends>:void=
        Sleep(3.0)
        if(LookingAtPortal = true):
            PortalAnalytic.Submit(Agent)
        else:
            Print("NO ONE LOOKING AT PORTAL")
        set LookingAtPortal = false

    OnTestStoneButton(Agent:agent, WhichButton:int):void=
        if(CustomPlayer := PlayersMap[player[Agent]]):
            if(WhichButton = 0):
                CustomPlayer.Collect(CustomPlayer.StoneAmount*1.1, "Stone", 0)

            else if(WhichButton = 1):
                CustomPlayer.Collect(CustomPlayer.StoneAmount*1.5, "Stone", 0)
                
            else if(WhichButton = 2):
                CustomPlayer.Collect(CustomPlayer.StoneAmount*2.0, "Stone", 0)
                
            else if(WhichButton = 3):
                CustomPlayer.Spend(CustomPlayer.StoneAmount, "Stone", 0)

    OnTestPMButton(Agent:agent):void=
        if(CustomPlayer := PlayersMap[player[Agent]]):
            ShowMessage(CustomPlayer.Player, "PM gained!", 0)
            set CustomPlayer.HarvestBuff += 100000.0/100.0
            set CustomPlayer.PrestigeHarvestBuff += 100000.0
            SaveAllData(CustomPlayer)

    OnTestEggsButton(Agent:agent):void=
        if(CustomPlayer := PlayersMap[player[Agent]]):
            ShowMessage(CustomPlayer.Player, "Eggs gained!", 0)
            CustomPlayer.Collect(1000000.0, "Egg", 0)
            CustomPlayer.Collect(1000000.0, "Egg", 1)
            CustomPlayer.Collect(1000000.0, "Egg", 2)
            CustomPlayer.Collect(1000000.0, "Egg", 3)
            CustomPlayer.Collect(1000000.0, "Egg", 4)
            CustomPlayer.Collect(1000000.0, "Egg", 5)
            SaveAllData(CustomPlayer)

    OnTestBossKeyButton(Agent:agent, WhichButton:int):void=
        if(CustomPlayer := PlayersMap[player[Agent]]):
            if(WhichButton = 0):
                CustomPlayer.Collect(100.0, "BossKey", 0)
                ShowMessage(CustomPlayer.Player, "Boss Keys Gained!", 0)

            else if(WhichButton = 1):
                CustomPlayer.Collect(1000.0, "BossKey", 0)
                ShowMessage(CustomPlayer.Player, "Boss Keys Gained!", 0)
                
            else if(WhichButton = 2):
                CustomPlayer.Collect(10000.0, "BossKey", 0)
                ShowMessage(CustomPlayer.Player, "Boss Keys Gained!", 0)
                
            else if(WhichButton = 3):
                CustomPlayer.Spend(CustomPlayer.BossKeyAmount, "BossKey", 0)
                ShowMessage(CustomPlayer.Player, "Boss Keys Reset!", 0)
        
    OnTestShardButton(Agent:agent, WhichButton:int):void=
        if(CustomPlayer := PlayersMap[player[Agent]]):
            if(WhichButton = 0):
                CustomPlayer.Collect(1000.0, "Shard", 0)
                ShowMessage(CustomPlayer.Player, "Shards Gained!", 0)

            else if(WhichButton = 1):
                CustomPlayer.Collect(100000.0, "Shard", 0)
                ShowMessage(CustomPlayer.Player, "Shards Gained!", 0)
                
            else if(WhichButton = 2):
                CustomPlayer.Collect(10000000.0, "Shard", 0)
                ShowMessage(CustomPlayer.Player, "Shards Gained!", 0)
                
            else if(WhichButton = 3):
                CustomPlayer.Spend(CustomPlayer.ShardAmount, "Shard", 0)
                ShowMessage(CustomPlayer.Player, "Shards Reset!", 0)

    OnTestSilverButton(Agent:agent, WhichButton:int):void=
        if(CustomPlayer := PlayersMap[player[Agent]]):
            if(WhichButton = 0):
                CustomPlayer.Collect(1000.0, "WepMat", 0)
                ShowMessage(CustomPlayer.Player, "Silver Gained!", 0)

            else if(WhichButton = 1):
                CustomPlayer.Collect(100000.0, "WepMat", 0)
                ShowMessage(CustomPlayer.Player, "Silver Gained!", 0)
                
            else if(WhichButton = 2):
                CustomPlayer.Collect(10000000.0, "WepMat", 0)
                ShowMessage(CustomPlayer.Player, "Silver Gained!", 0)
                
            else if(WhichButton = 3):
                CustomPlayer.Spend(CustomPlayer.WepMatAmount, "WepMat", 0)
                ShowMessage(CustomPlayer.Player, "Silver Reset!", 0)



    OnPlayerLoad(Agent:agent):void=
        spawn{NewInitiatePlayer(Agent)}
        
        <#var IsSlotted : logic = false
        for(i:=0..3):
            if(ActivePlayer:=ActivePlayers[i]?):
                Print("PLAYER ALREADY EXISTS HERE")
                if(ActivePlayer = Agent):
                    Print("THIS PLAYER ALREADY LOADED")
                    set IsSlotted = true
            else:
                if(IsSlotted = false):
                    Print("NEW PLAYER ADDED")
                    set IsSlotted = true
                    if(set ActivePlayers[i] = option{player[Agent]}) {}
                else:
                    Print("NEW PLAYER ALREADY ADDED")#>

    EnterSpace(Agent:agent):void=
        if(CustomPlayer := PlayersMap[player[Agent]]):
            set CustomPlayer.InSpace = true

    UpdateLeaderboards():void=
        var StonePlaces : []float = array{0.0, 0.0, 0.0}
        var BossPlaces : []float = array{0.0, 0.0, 0.0}

        for(K -> CustomPlayer:PlayersMap):
            if:
                PlayerTotalStone:=CustomPlayer.AllAchs[0][0]
                FBDamage:=CustomPlayer.HighestFBDamage
            then:
                if(PlayerTotalStone > StonePlaces[0]):
                    if(OldPlace:=StonePlaces[1], set StonePlaces[2] = OldPlace) {}
                    if(OldPlace:=StonePlaces[0], set StonePlaces[1] = OldPlace) {}
                    if(set StonePlaces[0] = PlayerTotalStone) {}
                else if(PlayerTotalStone > StonePlaces[1]):
                    if(OldPlace:=StonePlaces[1], set StonePlaces[2] = OldPlace) {}
                    if(set StonePlaces[1] = PlayerTotalStone) {}
                else if(PlayerTotalStone > StonePlaces[2]):
                    if(set StonePlaces[2] = PlayerTotalStone) {}

                if(FBDamage > BossPlaces[0]):
                    if(OldPlace:=BossPlaces[1], set BossPlaces[2] = OldPlace) {}
                    if(OldPlace:=BossPlaces[0], set BossPlaces[1] = OldPlace) {}
                    if(set BossPlaces[0] = FBDamage) {}
                else if(FBDamage > BossPlaces[1]):
                    if(OldPlace:=BossPlaces[1], set BossPlaces[2] = OldPlace) {}
                    if(set BossPlaces[1] = FBDamage) {}
                else if(PlayerTotalStone > BossPlaces[2]):
                    if(set BossPlaces[2] = FBDamage) {}

        for(i:int=0..StonePlaces.Length - 1):
            for(K -> CustomPlayer:PlayersMap):
                if:
                    PlayerTotalStone:=CustomPlayer.AllAchs[0][0]
                    FBDamage:=CustomPlayer.HighestFBDamage
                then:
                    if(PlayerTotalStone = StonePlaces[i]):
                        if(LeaderboardText:=LeaderboardBillboards[i], LeaderboardSpot:=LeaderboardPlayers[i]):
                            LeaderboardSpot.Register(CustomPlayer.Player)
                            LeaderboardText.SetText(StringToMessage("{NumberCondensor(PlayerTotalStone)}"))

                    if(FBDamage = BossPlaces[i]):
                        if(LeaderboardText:=LeaderboardBillboards[i+3], LeaderboardSpot:=LeaderboardPlayers[i+3]):
                            LeaderboardSpot.Register(CustomPlayer.Player)
                            LeaderboardText.SetText(StringToMessage("{NumberCondensor(FBDamage)}"))

    ResetPlayerProgress(Agent:agent): void=
        if(Player := player[Agent]):            
            if(set MySavedPlayerData[Player] = player_profile_data{}){}
            LoadPlayerData(Agent)
            Print("RESET ALL DATA SAVED")

    NewInitiatePlayer(Agent:agent)<suspends>: void=
        if(Player := player[Agent]):                    
            #CHECKING/APPLYING SAVE DATA
            if(PlayerExists := PlayersMap[Player]):
            else:
                if(set PlayersMap[Player] = Custom_Player{Player := Player}){}
                if:
                    not MySavedPlayerData[Player]
                    set MySavedPlayerData[Player] = player_profile_data{}

                if(Version <> MySavedPlayerData[Player].Version):
                    if(set MySavedPlayerData[Player] = player_profile_data{}){}
                    if(CustomPlayer := PlayersMap[Player]):
                        Print("SAVE CUSTOM PLAYER NEW VERSION")
                        SaveAllData(CustomPlayer)

            #PLAYER STUFF
            if(CustomPlayer := PlayersMap[Player],Fort :=CustomPlayer.Player.GetFortCharacter[]):
                if(CustomPlayer.PlayTimeSessionSeconds = 0.0):
                    LoadPlayerData(CustomPlayer.Player)
                    SwitchMusic(CustomPlayer.Player, "Default")
                    spawn:
                        TrackPlayTime(CustomPlayer)
                    CustomPlayer.InitiateUI()
                    Fort.DamagedEvent().SubscribeAgent5(OnPlayerDamaged, Agent)
                    set CustomPlayer.DmgBlockDMGs = array{CustomPlayer.DmgBlockDMG1,CustomPlayer.DmgBlockDMG2,CustomPlayer.DmgBlockDMG3,CustomPlayer.DmgBlockDMG4,CustomPlayer.DmgBlockDMG5,CustomPlayer.DmgBlockDMG6,CustomPlayer.DmgBlockDMG7,CustomPlayer.DmgBlockDMG8}
                    set CustomPlayer.DmgColorBlockDMGs = array{CustomPlayer.DmgColorBlockDMG1,CustomPlayer.DmgColorBlockDMG2,CustomPlayer.DmgColorBlockDMG3,CustomPlayer.DmgColorBlockDMG4,CustomPlayer.DmgColorBlockDMG5,CustomPlayer.DmgColorBlockDMG6,CustomPlayer.DmgColorBlockDMG7,CustomPlayer.DmgColorBlockDMG8}
                    set CustomPlayer.DmgTextures = array{CustomPlayer.DmgTexture1,CustomPlayer.DmgTexture2,CustomPlayer.DmgTexture3,CustomPlayer.DmgTexture4,CustomPlayer.DmgTexture5,CustomPlayer.DmgTexture6,CustomPlayer.DmgTexture7,CustomPlayer.DmgTexture8}                
    
                #if(CustomPlayer.InSpace = true and CustomPlayer.StagesUnlocked >= 15):
                #    SpaceTeleporter.Teleport(CustomPlayer.Player)
                #else:
                SwitchMusic(CustomPlayer.Player, "Default")

                SetClass(Agent, CustomPlayer)
                SetHP(CustomPlayer)
                ApplyMovementBuff(CustomPlayer)
                SaveAllData(CustomPlayer)

    OnPlayerDamaged(DmgResult : damage_result, Agent:agent) : void=
        if(CustomPlayer := PlayersMap[player[Agent]]):
            if(CustomPlayer.InHardBoss = false):
                set CustomPlayer.CanRegen = 25
            else:
                set CustomPlayer.CanRegen = 9999999999

    ReloadHUD(Agent:agent):void=
        if(CustomPlayer := PlayersMap[player[Agent]]):
            CustomPlayer.RemoveUI()
            CustomPlayer.InitiateUI()

    HideActionNumbers(Agent:agent):void=
        if(CustomPlayer := PlayersMap[player[Agent]]):
            if(CustomPlayer.ShowNumbers = true):
                set CustomPlayer.ShowNumbers = false
            else:
                set CustomPlayer.ShowNumbers = true

    OnPlayerLeftServer(PlayerLeaving : player):void=
        spawn{RemovePlayerFromMap(PlayerLeaving)}

    RemovePlayerFromMap(PlayerLeaving : player)<suspends>:void=
        Sleep(2.0) #DOING THIS SO WE CAN DESPAWN PETS
        if (CustomPlayer := PlayersMap[PlayerLeaving], A := agent[PlayerLeaving]):
            CustomPlayer.RemoveUI()
            var NewCustomPlayerMap : [player]Custom_Player = map{}
            for (Key -> Value : PlayersMap, Key <> PlayerLeaving):
                set NewCustomPlayerMap = ConcatenateMaps(NewCustomPlayerMap, map{Key => Value})
            set PlayersMap = NewCustomPlayerMap

    SetClass(Agent:agent,CustomPlayer:Custom_Player):void=
        BaseClassSelector.ChangeClass(Agent)
        for(ActivePerk : CustomPlayer.ActivePerks):
            if(ActivePerk(0) = "Instant Reload"):
                ClassInstantReload.ChangeClass(Agent)
        SetHP(CustomPlayer)

    SetHP(CustomPlayer:Custom_Player):void=
        if(Fort :=CustomPlayer.Player.GetFortCharacter[]):
            Fort.SetMaxHealth(CustomPlayer.MaxHP)
            Fort.Heal(CustomPlayer.MaxHP)

    ApplyMovementBuff(CustomPlayer:Custom_Player):void=
        if(MovementUpgradeGranter:=MovementUpgradeGranters[CustomPlayer.MovementSpeedLevel-1]):
            MovementUpgradeGranter.Activate(CustomPlayer.Player)

    SpawnShopNPCs(Agent:agent):void=
        for(ShopNPC:NPCSpawners):
            ShopNPC.Spawn()

    #PERSISTABLE DATA LOAD ON BEGIN
    LoadPlayerData(Agent:agent):void=
        if(Player := player[Agent],CustomPlayer := PlayersMap[Player]):
            if:
                MySavedPlayerData[Player]
                NewVersion := MySavedPlayerData[Player].Version
                set CustomPlayer.Version = NewVersion
                StoneAmount := MySavedPlayerData[Player].StoneAmount
                if(StoneAmount < 0.0):
                    set CustomPlayer.StoneAmount = 0.0
                else:
                    set CustomPlayer.StoneAmount = StoneAmount
                NewStardustAmount := MySavedPlayerData[Player].StardustAmount
                    set CustomPlayer.StardustAmount = NewStardustAmount
                NewShardAmount := MySavedPlayerData[Player].ShardAmount
                set CustomPlayer.ShardAmount = NewShardAmount
                NewEggAmounts := MySavedPlayerData[Player].EggAmounts
                set CustomPlayer.EggAmounts = NewEggAmounts
                NewWepMatAmount := MySavedPlayerData[Player].WepMatAmount
                set CustomPlayer.WepMatAmount = NewWepMatAmount
                NewStagesUnlocked := MySavedPlayerData[Player].StagesUnlocked
                set CustomPlayer.StagesUnlocked = NewStagesUnlocked
                if(CustomPlayer.StagesUnlocked > 24):
                    set CustomPlayer.StagesUnlocked = 24
                MaxHP := MySavedPlayerData[Player].MaxHP
                set CustomPlayer.MaxHP = MaxHP
                WepMatDropRate := MySavedPlayerData[Player].WepMatDropRate
                set CustomPlayer.WepMatDropRate = WepMatDropRate
                AFKStoneBoost := MySavedPlayerData[Player].AFKStoneBoost
                set CustomPlayer.AFKStoneBoost = AFKStoneBoost
                AFKPetsUnlocked := MySavedPlayerData[Player].AFKPetsUnlocked
                set CustomPlayer.AFKPetsUnlocked = AFKPetsUnlocked
                Power := MySavedPlayerData[Player].Power
                set CustomPlayer.Power = Power
                HarvestBuff := MySavedPlayerData[Player].HarvestBuff
                set CustomPlayer.HarvestBuff = HarvestBuff
                CritRate := MySavedPlayerData[Player].CritRate
                set CustomPlayer.CritRate = CritRate
                CritDMG := MySavedPlayerData[Player].CritDMG
                set CustomPlayer.CritDMG = CritDMG
                FinalHitBonus := MySavedPlayerData[Player].FinalHitBonus
                set CustomPlayer.FinalHitBonus = FinalHitBonus
                MultiCrit := MySavedPlayerData[Player].MultiCrit
                set CustomPlayer.MultiCrit = MultiCrit
                ComboStrikeUnlocked := MySavedPlayerData[Player].ComboStrikeUnlocked
                set CustomPlayer.ComboStrikeUnlocked = ComboStrikeUnlocked
                ComboStrikeMultiplier := MySavedPlayerData[Player].ComboStrikeMultiplier
                set CustomPlayer.ComboStrikeMultiplier = ComboStrikeMultiplier
                ComboAccelerator := MySavedPlayerData[Player].ComboAccelerator
                set CustomPlayer.ComboAccelerator = ComboAccelerator
                ComboStrikeStreak := MySavedPlayerData[Player].ComboStrikeStreak
                set CustomPlayer.ComboStrikeStreak = ComboStrikeStreak
                RoomUnlockGather := MySavedPlayerData[Player].RoomUnlockGather
                set CustomPlayer.RoomUnlockGather = RoomUnlockGather
                MovementSpeedLevel := MySavedPlayerData[Player].MovementSpeedLevel
                set CustomPlayer.MovementSpeedLevel = MovementSpeedLevel
                DoubleJumpUnlocked := MySavedPlayerData[Player].DoubleJumpUnlocked
                set CustomPlayer.DoubleJumpUnlocked = DoubleJumpUnlocked
                DoubleJumpAmount := MySavedPlayerData[Player].DoubleJumpAmount
                set CustomPlayer.DoubleJumpAmount = DoubleJumpAmount
                WeaponTier := MySavedPlayerData[Player].WeaponTier
                set CustomPlayer.WeaponTier = WeaponTier
                WeaponPerkTier := MySavedPlayerData[Player].WeaponPerkTier
                set CustomPlayer.WeaponPerkTier = WeaponPerkTier
                WeaponStatLevel := MySavedPlayerData[Player].WeaponStatLevel
                set CustomPlayer.WeaponStatLevel = WeaponStatLevel
                WeaponBonusLevel := MySavedPlayerData[Player].WeaponBonusLevel
                set CustomPlayer.WeaponBonusLevel = WeaponBonusLevel
                WeaponPerkLevel := MySavedPlayerData[Player].WeaponPerkLevel
                set CustomPlayer.WeaponPerkLevel = WeaponPerkLevel
                WeaponStats := MySavedPlayerData[Player].WeaponStats
                set CustomPlayer.WeaponStats = WeaponStats
                WeaponBonuses := MySavedPlayerData[Player].WeaponBonuses
                set CustomPlayer.WeaponBonuses = WeaponBonuses
                <#StatCost := MySavedPlayerData[Player].StatCost
                set CustomPlayer.StatCost = StatCost
                BonusCost := MySavedPlayerData[Player].BonusCost
                set CustomPlayer.BonusCost = BonusCost
                PerkCost := MySavedPlayerData[Player].PerkCost
                set CustomPlayer.PerkCost = PerkCost
                EvolveCost := MySavedPlayerData[Player].EvolveCost
                set CustomPlayer.EvolveCost = EvolveCost#>
                PerksUnlocked := MySavedPlayerData[Player].PerksUnlocked
                set CustomPlayer.PerksUnlocked = PerksUnlocked
                ActivePerks := MySavedPlayerData[Player].ActivePerks
                set CustomPlayer.ActivePerks = ActivePerks
                Hut1Levels := MySavedPlayerData[Player].Hut1Levels
                set CustomPlayer.Hut1Levels = Hut1Levels
                Hut1Costs := MySavedPlayerData[Player].Hut1Costs
                set CustomPlayer.Hut1Costs = Hut1Costs
                #Hut1Upgrades := MySavedPlayerData[Player].Hut1Upgrades
                #set CustomPlayer.Hut1Upgrades = Hut1Upgrades
                Hut2Levels := MySavedPlayerData[Player].Hut2Levels
                set CustomPlayer.Hut2Levels = Hut2Levels
                Hut2Costs := MySavedPlayerData[Player].Hut2Costs
                set CustomPlayer.Hut2Costs = Hut2Costs
                #Hut2Upgrades := MySavedPlayerData[Player].Hut2Upgrades
                #set CustomPlayer.Hut2Upgrades = Hut2Upgrades
                Hut3Levels := MySavedPlayerData[Player].Hut3Levels
                set CustomPlayer.Hut3Levels = Hut3Levels
                Hut3Costs := MySavedPlayerData[Player].Hut3Costs
                set CustomPlayer.Hut3Costs = Hut3Costs
                Hut4Levels := MySavedPlayerData[Player].Hut4Levels
                set CustomPlayer.Hut4Levels = Hut4Levels
                Hut5Levels := MySavedPlayerData[Player].Hut5Levels
                set CustomPlayer.Hut5Levels = Hut5Levels
                #Hut3Upgrades := MySavedPlayerData[Player].Hut3Upgrades
                #set CustomPlayer.Hut3Upgrades = Hut3Upgrades
                GUP1Levels := MySavedPlayerData[Player].GUP1Levels
                set CustomPlayer.GUP1Levels = GUP1Levels
                #GUP1Costs := MySavedPlayerData[Player].GUP1Costs
                #set CustomPlayer.GUP1Costs = GUP1Costs
                #GUP1Upgrades := MySavedPlayerData[Player].GUP1Upgrades
                #set CustomPlayer.GUP1Upgrades = GUP1Upgrades
                GUP2Levels := MySavedPlayerData[Player].GUP2Levels
                set CustomPlayer.GUP2Levels = GUP2Levels
                #GUP2Costs := MySavedPlayerData[Player].GUP2Costs
                #set CustomPlayer.GUP2Costs = GUP2Costs
                #GUP2Upgrades := MySavedPlayerData[Player].GUP2Upgrades
                #set CustomPlayer.GUP2Upgrades = GUP2Upgrades
                GUP3Levels := MySavedPlayerData[Player].GUP3Levels
                set CustomPlayer.GUP3Levels = GUP3Levels
                #GUP3Costs := MySavedPlayerData[Player].GUP3Costs
                #set CustomPlayer.GUP3Costs = GUP3Costs
                #GUP3Upgrades := MySavedPlayerData[Player].GUP3Upgrades
                #set CustomPlayer.GUP3Upgrades = GUP3Upgrades
                PrestigeLevel := MySavedPlayerData[Player].PrestigeLevel
                set CustomPlayer.PrestigeLevel = PrestigeLevel
                PrestigeFix := MySavedPlayerData[Player].PrestigeFix
                set CustomPlayer.PrestigeFix = PrestigeFix
                #PrestigeCost := MySavedPlayerData[Player].PrestigeCost
                set CustomPlayer.PrestigeCost = 1.12e16 #CHANGE TO DEFAULT VALUE
                PrestigeHarvestBuff := MySavedPlayerData[Player].PrestigeHarvestBuff
                set CustomPlayer.PrestigeHarvestBuff = PrestigeHarvestBuff
                PrestigeStoneAmount := MySavedPlayerData[Player].PrestigeStoneAmount
                set CustomPlayer.PrestigeStoneAmount = PrestigeStoneAmount
                PUPLevels := MySavedPlayerData[Player].PUPLevels
                set CustomPlayer.PUPLevels = PUPLevels
                #PUPCosts := MySavedPlayerData[Player].PUPCosts
                #set CustomPlayer.PUPCosts = PUPCosts
                #PUPUpgrades := MySavedPlayerData[Player].PUPUpgrades
                #set CustomPlayer.PUPUpgrades = PUPUpgrades
                #EggPrestigeUpgrades := MySavedPlayerData[Player].EggPrestigeUpgrades
                #set CustomPlayer.EggPrestigeUpgrades = EggPrestigeUpgrades
                #EggPrestigeCosts := MySavedPlayerData[Player].EggPrestigeCosts
                #set CustomPlayer.EggPrestigeCosts = EggPrestigeCosts
                EggPrestigeLevels := MySavedPlayerData[Player].EggPrestigeLevels
                set CustomPlayer.EggPrestigeLevels = EggPrestigeLevels
                PetEquipSlots := MySavedPlayerData[Player].PetEquipSlots
                set CustomPlayer.PetEquipSlots = PetEquipSlots
                PUPQOLLevels := MySavedPlayerData[Player].PUPQOLLevels
                set CustomPlayer.PUPQOLLevels = PUPQOLLevels
                AutoGatherRate := MySavedPlayerData[Player].AutoGatherRate
                set CustomPlayer.AutoGatherRate = AutoGatherRate
                BossRewardMult := MySavedPlayerData[Player].BossRewardMult
                set CustomPlayer.BossRewardMult = BossRewardMult
                EggAmountMultiplier := MySavedPlayerData[Player].EggAmountMultiplier
                set CustomPlayer.EggAmountMultiplier = EggAmountMultiplier
                EggDropRateMultiplier := MySavedPlayerData[Player].EggDropRateMultiplier
                set CustomPlayer.EggDropRateMultiplier = EggDropRateMultiplier
                BaseStrengthCosts := MySavedPlayerData[Player].BaseStrengthCosts
                set CustomPlayer.BaseStrengthCosts = array{
                    15.0,   150.0,  800.0,  4.80e03,3.00e05,
                    1.80e06,3.00e07,1.80e08,4.50e09,2.70e10,
                    4.50e11,2.70e12,3.38e13,2.03e14,1.87e16,
                    1.12e17,3.75e18,2.25e19,1.12e21,6.74e21,
                    2.25e23,1.35e24,3.37e25,2.02e26,3.75e28,
                    2.25e29,1.12e31,6.74e31,5.06e33,3.03e34,
                    1.52e36,9.10e36,3.41e38,2.05e39,5.31e40,
                    3.18e41,2.12e42,1.27e43,1.27e44,2.64e45, #7.64e45
            
                    3.41e49, 2.05e51, 1.23e53, 7.37e54, 4.42e56, 
                    2.65e58, 3.47e59, 2.08e61, 1.25e62, 7.50e63, 
                    4.50e64, 2.70e65, 9.71e65, 4.63e66, 2.78e68, 
                    1.67e70, 9.99e71, 6.00e72, 1.45e74, 8.70e76, 
            
                    5.22e78, 3.13e79, 1.88e81, 1.13e83, 1.67e84, 
                    3.00e85, 8.01e86, 7.61e88, 8.16e90, 9.30e91, 
                    7.09e92, 8.51e93, 6.91e94, 5.34e95, 8.41e96, 
                    8.44e97, 5.06e98, 3.04e99, 1.82e100, 1.09e101,
                    
                    9.81e105, 8.83e107, 7.95e109, 7.15e111, 6.44e113, 
                    5.79e115, 5.21e117, 4.69e119, 4.22e121, 3.80e123, 
                    3.42e125, 3.08e127, 2.77e129, 2.49e131, 2.24e133, 
                    2.02e135, 1.82e137, 1.64e139, 1.47e141, 1.33e143, 
                    1.19e145, 1.07e147, 9.66e148, 8.69e150, 7.83e152, 
                    7.04e154, 6.34e156, 5.70e158, 5.13e160, 4.62e162            
                    } 
                #StrengthUpgrades := MySavedPlayerData[Player].StrengthUpgrades
                #set CustomPlayer.StrengthUpgrades = StrengthUpgrades
                StrengthEfficiencies := MySavedPlayerData[Player].StrengthEfficiencies
                set CustomPlayer.StrengthEfficiencies = StrengthEfficiencies
                StrengthUpgradeLevels := MySavedPlayerData[Player].StrengthUpgradeLevels
                set CustomPlayer.StrengthUpgradeLevels = StrengthUpgradeLevels

                if(StrEff := StrengthEfficiencies[81]):
                    Print("StrEff: {StrEff}")
                else:
                    Print("NO STR EFF AT 81 WTFFFF")

                if(StrengthEfficiencies[81] >= 0):
                    Print("You have efficiencies, you chillin")
                else:
                    Print("DO NOT HAVE A5 STRENGTH EFFICIENCIES")
                    if:
                        NewStrEffArray := array:
                            StrengthEfficiencies[0],StrengthEfficiencies[1],StrengthEfficiencies[2],StrengthEfficiencies[3],StrengthEfficiencies[4],
                            StrengthEfficiencies[5],StrengthEfficiencies[6],StrengthEfficiencies[7],StrengthEfficiencies[8],StrengthEfficiencies[9],
                            StrengthEfficiencies[10],StrengthEfficiencies[11],StrengthEfficiencies[12],StrengthEfficiencies[13],StrengthEfficiencies[14],
                            StrengthEfficiencies[15],StrengthEfficiencies[16],StrengthEfficiencies[17],StrengthEfficiencies[18],StrengthEfficiencies[19],
                            StrengthEfficiencies[20],StrengthEfficiencies[21],StrengthEfficiencies[22],StrengthEfficiencies[23],StrengthEfficiencies[24],
                            StrengthEfficiencies[25],StrengthEfficiencies[26],StrengthEfficiencies[27],StrengthEfficiencies[28],StrengthEfficiencies[29],
                            StrengthEfficiencies[30],StrengthEfficiencies[31],StrengthEfficiencies[32],StrengthEfficiencies[33],StrengthEfficiencies[34],
                            StrengthEfficiencies[35],StrengthEfficiencies[36],StrengthEfficiencies[37],StrengthEfficiencies[38],StrengthEfficiencies[39],
                            StrengthEfficiencies[40],StrengthEfficiencies[41],StrengthEfficiencies[42],StrengthEfficiencies[43],StrengthEfficiencies[44],
                            StrengthEfficiencies[45],StrengthEfficiencies[46],StrengthEfficiencies[47],StrengthEfficiencies[48],StrengthEfficiencies[49],
                            StrengthEfficiencies[50],StrengthEfficiencies[51],StrengthEfficiencies[52],StrengthEfficiencies[53],StrengthEfficiencies[54],
                            StrengthEfficiencies[55],StrengthEfficiencies[56],StrengthEfficiencies[57],StrengthEfficiencies[58],StrengthEfficiencies[59],
                            StrengthEfficiencies[60],StrengthEfficiencies[61],StrengthEfficiencies[62],StrengthEfficiencies[63],StrengthEfficiencies[64],
                            StrengthEfficiencies[65],StrengthEfficiencies[66],StrengthEfficiencies[67],StrengthEfficiencies[68],StrengthEfficiencies[69],
                            StrengthEfficiencies[70],StrengthEfficiencies[71],StrengthEfficiencies[72],StrengthEfficiencies[73],StrengthEfficiencies[74],
                            StrengthEfficiencies[75],StrengthEfficiencies[76],StrengthEfficiencies[77],StrengthEfficiencies[78],StrengthEfficiencies[79],
                            1, 1, 1, 1, 1,
                            1, 1, 1, 1, 1,
                            1, 1, 1, 1, 1,
                            1, 1, 1, 1, 1,
                            1, 1, 1, 1, 1,
                            1, 1, 1, 1, 1

                        NewStrArray := array:
                            StrengthUpgradeLevels[0],StrengthUpgradeLevels[1],StrengthUpgradeLevels[2],StrengthUpgradeLevels[3],StrengthUpgradeLevels[4],
                            StrengthUpgradeLevels[5],StrengthUpgradeLevels[6],StrengthUpgradeLevels[7],StrengthUpgradeLevels[8],StrengthUpgradeLevels[9],
                            StrengthUpgradeLevels[10],StrengthUpgradeLevels[11],StrengthUpgradeLevels[12],StrengthUpgradeLevels[13],StrengthUpgradeLevels[14],
                            StrengthUpgradeLevels[15],StrengthUpgradeLevels[16],StrengthUpgradeLevels[17],StrengthUpgradeLevels[18],StrengthUpgradeLevels[19],
                            StrengthUpgradeLevels[20],StrengthUpgradeLevels[21],StrengthUpgradeLevels[22],StrengthUpgradeLevels[23],StrengthUpgradeLevels[24],
                            StrengthUpgradeLevels[25],StrengthUpgradeLevels[26],StrengthUpgradeLevels[27],StrengthUpgradeLevels[28],StrengthUpgradeLevels[29],
                            StrengthUpgradeLevels[30],StrengthUpgradeLevels[31],StrengthUpgradeLevels[32],StrengthUpgradeLevels[33],StrengthUpgradeLevels[34],
                            StrengthUpgradeLevels[35],StrengthUpgradeLevels[36],StrengthUpgradeLevels[37],StrengthUpgradeLevels[38],StrengthUpgradeLevels[39],
                            StrengthUpgradeLevels[40],StrengthUpgradeLevels[41],StrengthUpgradeLevels[42],StrengthUpgradeLevels[43],StrengthUpgradeLevels[44],
                            StrengthUpgradeLevels[45],StrengthUpgradeLevels[46],StrengthUpgradeLevels[47],StrengthUpgradeLevels[48],StrengthUpgradeLevels[49],
                            StrengthUpgradeLevels[50],StrengthUpgradeLevels[51],StrengthUpgradeLevels[52],StrengthUpgradeLevels[53],StrengthUpgradeLevels[54],
                            StrengthUpgradeLevels[55],StrengthUpgradeLevels[56],StrengthUpgradeLevels[57],StrengthUpgradeLevels[58],StrengthUpgradeLevels[59],
                            StrengthUpgradeLevels[60],StrengthUpgradeLevels[61],StrengthUpgradeLevels[62],StrengthUpgradeLevels[63],StrengthUpgradeLevels[64],
                            StrengthUpgradeLevels[65],StrengthUpgradeLevels[66],StrengthUpgradeLevels[67],StrengthUpgradeLevels[68],StrengthUpgradeLevels[69],
                            StrengthUpgradeLevels[70],StrengthUpgradeLevels[71],StrengthUpgradeLevels[72],StrengthUpgradeLevels[73],StrengthUpgradeLevels[74],
                            StrengthUpgradeLevels[75],StrengthUpgradeLevels[76],StrengthUpgradeLevels[77],StrengthUpgradeLevels[78],StrengthUpgradeLevels[79],
                            0,0,0,0,0,
                            0,0,0,0,0,
                            0,0,0,0,0,
                            0,0,0,0,0,
                            0,0,0,0,0,
                            0,0,0,0,0
                    then:
                        set CustomPlayer.StrengthEfficiencies = NewStrEffArray
                        set CustomPlayer.StrengthUpgradeLevels = NewStrArray
                        Print("Done")

                EarthquakeUnlocked := MySavedPlayerData[Player].EarthquakeUnlocked
                set CustomPlayer.EarthquakeUnlocked = EarthquakeUnlocked
                EarthquakeDistance := MySavedPlayerData[Player].EarthquakeDistance
                set CustomPlayer.EarthquakeDistance = EarthquakeDistance
                EarthquakeGatherAmount := MySavedPlayerData[Player].EarthquakeGatherAmount
                set CustomPlayer.EarthquakeGatherAmount = EarthquakeGatherAmount
                EarthquakeDamage := MySavedPlayerData[Player].EarthquakeDamage
                set CustomPlayer.EarthquakeDamage = EarthquakeDamage
                VoidUnlocked := MySavedPlayerData[Player].VoidUnlocked
                set CustomPlayer.VoidUnlocked = VoidUnlocked
                #VoidDamage := MySavedPlayerData[Player].VoidDamage
                #set CustomPlayer.VoidDamage = VoidDamage
                VoidDuration := MySavedPlayerData[Player].VoidDuration
                set CustomPlayer.VoidDuration = VoidDuration
                VoidStacks := MySavedPlayerData[Player].VoidStacks
                set CustomPlayer.VoidStacks = VoidStacks
                A4Unlocked := MySavedPlayerData[Player].A4Unlocked
                set CustomPlayer.A4Unlocked = A4Unlocked
                A5Unlocked := MySavedPlayerData[Player].A5Unlocked
                set CustomPlayer.A5Unlocked = A5Unlocked
                SinkPrestigePower := MySavedPlayerData[Player].SinkPrestigePower
                set CustomPlayer.SinkPrestigePower = SinkPrestigePower
                SinkWepMatPower := MySavedPlayerData[Player].SinkWepMatPower
                set CustomPlayer.SinkWepMatPower = SinkWepMatPower
                SinkBossKeyPower := MySavedPlayerData[Player].SinkBossKeyPower
                set CustomPlayer.SinkBossKeyPower = SinkBossKeyPower
                CurrentDay := MySavedPlayerData[Player].CurrentDay
                set CustomPlayer.CurrentDay = CurrentDay
                LastClaimed := MySavedPlayerData[Player].LastClaimed
                set CustomPlayer.LastClaimed = LastClaimed
                HatsUnlocked :=MySavedPlayerData[Player].HatsUnlocked
                set CustomPlayer.HatsUnlocked = HatsUnlocked
                HatPrestiges :=MySavedPlayerData[Player].HatPrestiges
                set CustomPlayer.HatPrestiges = HatPrestiges
                CodeHatsUnlocked :=MySavedPlayerData[Player].CodeHatsUnlocked
                set CustomPlayer.CodeHatsUnlocked = CodeHatsUnlocked

                EXPPetsLevel := MySavedPlayerData[Player].EXPPetsLevel
                NewEXPPetsLevel := MySavedPlayerData[Player].NewEXPPetsLevel
                for(i:int=0..NewEXPPetsLevel.Length):
                    if(PetLevel:=EXPPetsLevel[i],NewPetLevel:=NewEXPPetsLevel[i]):
                        if(NewPetLevel > PetLevel):
                            if(set CustomPlayer.NewEXPPetsLevel[i] = NewPetLevel) {}
                        else:
                            if(set CustomPlayer.NewEXPPetsLevel[i] = PetLevel) {}
                QuestRocksDestroyed := MySavedPlayerData[Player].QuestRocksDestroyed
                set CustomPlayer.QuestRocksDestroyed = QuestRocksDestroyed
                QuestMobsDefeated := MySavedPlayerData[Player].QuestMobsDefeated
                set CustomPlayer.QuestMobsDefeated = QuestMobsDefeated
                QuestBossesDefeated := MySavedPlayerData[Player].QuestBossesDefeated
                set CustomPlayer.QuestBossesDefeated = QuestBossesDefeated
                QuestFinalStatus := MySavedPlayerData[Player].QuestFinalStatus
                set CustomPlayer.QuestFinalStatus = QuestFinalStatus
                QuestsStarted := MySavedPlayerData[Player].QuestsStarted
                set CustomPlayer.QuestsStarted = QuestsStarted
                QuestsCompleted := MySavedPlayerData[Player].QuestsCompleted
                set CustomPlayer.QuestsCompleted = QuestsCompleted
                LastCompletedDate := MySavedPlayerData[Player].LastCompletedDate
                set CustomPlayer.LastCompletedDate = LastCompletedDate
                LastCompletedHolidayDate := MySavedPlayerData[Player].LastCompletedHolidayDate
                set CustomPlayer.LastCompletedHolidayDate = LastCompletedHolidayDate
                CurrentExpedition := MySavedPlayerData[Player].CurrentExpedition
                set CustomPlayer.CurrentExpedition = CurrentExpedition
                EXPCurrentPet := MySavedPlayerData[Player].EXPCurrentPet
                set CustomPlayer.EXPCurrentPet = EXPCurrentPet
                ExpeditionTimeLeft := MySavedPlayerData[Player].ExpeditionTimeLeft
                set CustomPlayer.ExpeditionTimeLeft = ExpeditionTimeLeft
                EXPTotalRewards := MySavedPlayerData[Player].EXPTotalRewards
                set CustomPlayer.EXPTotalRewards = EXPTotalRewards
                GoldenDrillUnlocked := MySavedPlayerData[Player].GoldenDrillUnlocked
                set CustomPlayer.GoldenDrillUnlocked = GoldenDrillUnlocked
                DrillActivations := MySavedPlayerData[Player].DrillActivations
                set CustomPlayer.DrillActivations = DrillActivations
                BuffMultiplier := MySavedPlayerData[Player].BuffMultiplier
                set CustomPlayer.BuffMultiplier = BuffMultiplier
                BuffTimeSeconds := MySavedPlayerData[Player].BuffTimeSeconds
                set CustomPlayer.BuffTimeSeconds = BuffTimeSeconds
                BuffChargeTimer := MySavedPlayerData[Player].BuffChargeTimer
                set CustomPlayer.BuffChargeTimer = BuffChargeTimer
                BuffCurrentCharges := MySavedPlayerData[Player].BuffCurrentCharges
                set CustomPlayer.BuffCurrentCharges = BuffCurrentCharges
                BuffMaxCharges := MySavedPlayerData[Player].BuffMaxCharges
                set CustomPlayer.BuffMaxCharges = BuffMaxCharges
                GoldenDrillLevels := MySavedPlayerData[Player].GoldenDrillLevels
                set CustomPlayer.GoldenDrillLevels = GoldenDrillLevels
                #GoldenDrillCosts := MySavedPlayerData[Player].GoldenDrillCosts
                #set CustomPlayer.GoldenDrillCosts = GoldenDrillCosts
                #GoldenDrillUpgrades := MySavedPlayerData[Player].GoldenDrillUpgrades
                #set CustomPlayer.GoldenDrillUpgrades = GoldenDrillUpgrades
                SecretsClaimed := MySavedPlayerData[Player].SecretsClaimed
                for(i:int=0..SecretsClaimed.Length):
                    if(SecretClaimed:=SecretsClaimed[i], SecretClaimed = true):
                        set CustomPlayer.NewSecretsClaimed[i] = SecretClaimed
                NewSecretsClaimed := MySavedPlayerData[Player].NewSecretsClaimed
                for(i:int=0..NewSecretsClaimed.Length):
                    if(SecretClaimed:=NewSecretsClaimed[i], SecretClaimed = true):
                        set CustomPlayer.NewSecretsClaimed[i] = SecretClaimed
                
                EggsUnlocked := MySavedPlayerData[Player].EggsUnlocked
                set CustomPlayer.EggsUnlocked = EggsUnlocked
                RockEggDropChance := MySavedPlayerData[Player].RockEggDropChance
                set CustomPlayer.RockEggDropChance = RockEggDropChance
                EggDropAmount := MySavedPlayerData[Player].EggDropAmount
                set CustomPlayer.EggDropAmount = EggDropAmount
                MobEggDropChance := MySavedPlayerData[Player].MobEggDropChance
                set CustomPlayer.MobEggDropChance = MobEggDropChance
                PetDataReset := MySavedPlayerData[Player].PetDataReset
                set CustomPlayer.PetDataReset = PetDataReset
                PetUnlocked := MySavedPlayerData[Player].PetUnlocked
                set CustomPlayer.PetUnlocked = PetUnlocked
                PetSlots := MySavedPlayerData[Player].PetSlots
                set CustomPlayer.PetSlots = PetSlots
                HidePets := MySavedPlayerData[Player].HidePets
                set CustomPlayer.HidePets = HidePets
                PetLevels := MySavedPlayerData[Player].PetLevels
                set CustomPlayer.PetLevels = PetLevels
                PetEvolveLevels := MySavedPlayerData[Player].PetEvolveLevels
                set CustomPlayer.PetEvolveLevels = PetEvolveLevels
                PetGatherRates := MySavedPlayerData[Player].PetGatherRates
                set CustomPlayer.PetGatherRates = PetGatherRates
                PetAttackRates := MySavedPlayerData[Player].PetAttackRates
                set CustomPlayer.PetAttackRates = PetAttackRates
                PetCritRates := MySavedPlayerData[Player].PetCritRates
                set CustomPlayer.PetCritRates = PetCritRates
                PetCritDMG := MySavedPlayerData[Player].PetCritDMG
                set CustomPlayer.PetCritDMG = PetCritDMG
                PetFinalStrike := MySavedPlayerData[Player].PetFinalStrike
                set CustomPlayer.PetFinalStrike = PetFinalStrike
                EggsHatched := MySavedPlayerData[Player].EggsHatched
                set CustomPlayer.EggsHatched = EggsHatched
                HighestFBDamage := MySavedPlayerData[Player].HighestFBDamage
                set CustomPlayer.HighestFBDamage = HighestFBDamage
                AllBossesDefeated := MySavedPlayerData[Player].AllBossesDefeated
                set CustomPlayer.AllBossesDefeated = AllBossesDefeated
                DemonPetObtained := MySavedPlayerData[Player].DemonPetObtained
                set CustomPlayer.DemonPetObtained = DemonPetObtained
                BossKeyAmount := MySavedPlayerData[Player].BossKeyAmount
                set CustomPlayer.BossKeyAmount = BossKeyAmount
                TutorialLevel := MySavedPlayerData[Player].TutorialLevel
                set CustomPlayer.TutorialLevel = TutorialLevel
                TutorialProgress := MySavedPlayerData[Player].TutorialProgress
                set CustomPlayer.TutorialProgress = TutorialProgress
                CurrentMusic := MySavedPlayerData[Player].CurrentMusic
                set CustomPlayer.CurrentMusic = CurrentMusic
                AchLevels := MySavedPlayerData[Player].AchLevels
                set CustomPlayer.AchLevels = AchLevels
                AchBuffTotal := MySavedPlayerData[Player].AchBuffTotal
                set CustomPlayer.AchBuffTotal = AchBuffTotal
                Ach1 := MySavedPlayerData[Player].Ach1
                if(set CustomPlayer.AllAchs[0] = Ach1) {}
                Ach2 := MySavedPlayerData[Player].Ach2
                if(set CustomPlayer.AllAchs[1] = Ach2) {}
                Ach3 := MySavedPlayerData[Player].Ach3
                if(set CustomPlayer.AllAchs[2] = Ach3) {}
                Ach4 := MySavedPlayerData[Player].Ach4
                if(set CustomPlayer.AllAchs[3] = Ach4) {}
                Ach5 := MySavedPlayerData[Player].Ach5
                if(set CustomPlayer.AllAchs[4] = Ach5) {}
                Ach6 := MySavedPlayerData[Player].Ach6
                if(set CustomPlayer.AllAchs[5] = Ach6) {}
                Ach7 := MySavedPlayerData[Player].Ach7
                if(set CustomPlayer.AllAchs[6] = Ach7) {}
                Ach8 := MySavedPlayerData[Player].Ach8
                if(set CustomPlayer.AllAchs[7] = Ach8) {}
                PlayTimeSeconds := MySavedPlayerData[Player].PlayTimeSeconds
                set CustomPlayer.PlayTimeSeconds = PlayTimeSeconds
                OneTimeAnalytics := MySavedPlayerData[Player].OneTimeAnalytics
                set CustomPlayer.OneTimeAnalytics = OneTimeAnalytics

                #UNCOMMENT THESE FOR TEST VALUES

    #ALWAYS UPDATE THE LEADERBOARDS

                <#set CustomPlayer.StoneAmount = 2.43e200
                set CustomPlayer.ShardAmount = 2.43e200
                set CustomPlayer.EggAmounts = array{1.0e13,1.0e13,10000000.0,10000000.0,10000000.0,10000000.0}
                set CustomPlayer.WepMatAmount = 8000000000.0
                set CustomPlayer.BossKeyAmount = 8000000000.0
                set CustomPlayer.StagesUnlocked = 24
                set CustomPlayer.MovementSpeedLevel = 10
                set CustomPlayer.DoubleJumpAmount = 4
                set CustomPlayer.DoubleJumpUnlocked = true
                set CustomPlayer.WeaponTier = 9
                set CustomPlayer.A4Unlocked = true
                set CustomPlayer.A5Unlocked = true
                if(set CustomPlayer.OneTimeAnalytics["DestroyMeteor"] = true) {}
                set CustomPlayer.TutorialLevel = 18
                set CustomPlayer.PetSlots = array{0, 1, 2, 3, 4, 5}
                set CustomPlayer.PetUnlocked = array{true,true,true,true,true,true}
                set CustomPlayer.PetLevels = array{7000000,7777007,6000000,10000000,8000000,3200000} #0,0,0,0,0,0 #70000,77777,60000,100000,80000,32000  #30100,48340,29640,31750,24850,11210
                set CustomPlayer.MaxHP = 600.0
                set CustomPlayer.HidePets = true
                set CustomPlayer.EggsHatched = array{0.0,0.0,0.0,0.0,0.0,0.0}
                set CustomPlayer.BuffCurrentCharges = 4.0
                set CustomPlayer.PrestigeStoneAmount = 2.0e204
                set CustomPlayer.PrestigeLevel = 2#>

                #set CustomPlayer.HarvestBuff = 5100.0
                #set CustomPlayer.PrestigeHarvestBuff = 5100000.0

    ShowMessage(Agent:agent, Message:string, i:int):void=
        if(Player := player[Agent]):
           if(MessageDevice:=MessageDevices[i]):
               MessageDevice.SetText(StringToMessage(Message))
               MessageDevice.Show(Agent)

    TrackPlayTime(CustomPlayer:Custom_Player)<suspends>:void=
        if(CustomPlayer.PlayTimeSessionSeconds = 0.0):
            Sleep(5.0)
            set CustomPlayer.PlayTimeSessionSeconds += 5.0

            var MinReqs : []int = array{10,30,60,120,240,6000,30000}
            if(MinPlaytime := Floor[CustomPlayer.PlayTimeSeconds / 60.0]):
                for(i:=0..MinReqs.Length-1):
                    if:
                        MinReq := MinReqs[i]
                        MinPlaytime >= MinReq
                        CustomPlayer.OneTimeAnalytics["Play{MinReq}Min"] = false
                    then:
                        if(set CustomPlayer.OneTimeAnalytics["Play{MinReq}Min"] = true) {}
                        if(Analytic:=AudioManager.OneTimeAnalytics[i]): #ANALYTIC
                            Analytic.Submit(CustomPlayer.Player)
            loop:
                Sleep(60.0)
                set CustomPlayer.CanGrantEXP = true
                CustomPlayer.UpAchValue(7, 60.0)
                UpdateLeaderboards()
                set CustomPlayer.PlayTimeSeconds += 60.0
                set CustomPlayer.PlayTimeSessionSeconds += 60.0
                SaveAllData(CustomPlayer)
                if(Analytic:=AudioManager.GeneralAnalytics[3]): #ANALYTIC
                    Analytic.Submit(CustomPlayer.Player)

                if:
                    Playtime:=Floor[CustomPlayer.PlayTimeSessionSeconds - 5.0]
                    Playtime <= 2700
                    Mod[Playtime, 900] = 0
                then:
                    spawn{ShowCTA(CustomPlayer)}

    MissileCDTimer(CustomPlayer:Custom_Player)<suspends>:void=
        if:
            CooldownLevel := CustomPlayer.Hut5Levels[2]
            CooldownAmount := 3.0 - (CooldownLevel * 0.3)
        then:
            Sleep(CooldownAmount)
            Print("ABILITY REFRESHED")
            set CustomPlayer.MissileCD = 0.0

    ShowCTA(CustomPlayer:Custom_Player)<suspends>:void=
        CustomPlayer.CTAHUD.SetVisibility(widget_visibility.Visible)
        AudioManager.Evolve.Play(CustomPlayer.Player)
        Sleep(5.0)
        CustomPlayer.CTAHUD.SetVisibility(widget_visibility.Collapsed)


    TrackSessionTime()<suspends>:void=
        loop:
            var SeasonalBossTimer : float = 900.0
            #Print("SESSION SECOND")
            Sleep(1.0)
            set SessionTime += 1.0
            set SeasonalBossTimer -= SessionTime
            if(SeasonalBossTimer >= 0.0):
                SeasonalBillboard.SetText(StringToMessage("{ConvertToTime(SeasonalBossTimer, "Minutes")}"))
            else:
                SeasonalBillboard.SetText(StringToMessage("Ready!"))

            for(K -> V:PlayersMap):
                if:
                    V.CanRegen = 0
                    Fort := V.Player.GetFortCharacter[]
                then:
                    Fort.Heal(5.0)
                else:
                    set V.CanRegen -= 1

                #Print("V.CanRegen: {V.CanRegen}")


    MuteMusic(Agent:agent):void=
        if(CustomPlayer := PlayersMap[player[Agent]]):
            if(CustomPlayer.CurrentMusic <> "Muted"):
                set CustomPlayer.CurrentMusic = "Muted"
                for(i:int=0..AudioManager.BGMs.Length):
                    if(BGM:=AudioManager.BGMs[i]):
                        BGM.Stop(CustomPlayer.Player)
            else:
                set CustomPlayer.CurrentMusic = "Default"
                SwitchMusic(CustomPlayer.Player, "Default")

    SwitchMusic(Agent:agent, WhichMusic:string):void=
        if(CustomPlayer := PlayersMap[player[Agent]]):
            if(CustomPlayer.CurrentMusic <> "Muted"):
                for(i:int=0..AudioManager.BGMs.Length):
                    if(BGM:=AudioManager.BGMs[i], BGMName:=AudioManager.BGMNames[i]):
                        if(CustomPlayer.CurrentMusic = BGMName):
                            BGM.Stop(CustomPlayer.Player)
                        if(WhichMusic = BGMName):
                            BGM.Play(CustomPlayer.Player)

                set CustomPlayer.CurrentMusic = WhichMusic
                SaveAllData(CustomPlayer)
