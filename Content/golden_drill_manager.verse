using { /Fortnite.com/Devices }
using { /Verse.org/Simulation }
using { /Verse.org/Colors }
using { /Verse.org/Verse }
using { /Verse.org/Random }
using { /Fortnite.com/Teams }
using { /UnrealEngine.com/Assets }
using { /Fortnite.com/Game }
using { /Fortnite.com/Characters }
using { /Fortnite.com/Playspaces }
using { /Fortnite.com/UI }
using { /UnrealEngine.com/Temporary/Diagnostics }
using { /UnrealEngine.com/Temporary/SpatialMath }
using { /UnrealEngine.com/Temporary/UI }
using { /Verse.org/Assets}


golden_drill_manager := class(creative_device):

    #@editable
    #StartBuff : button_device = button_device{}
    
    @editable
    gm : game_manager = game_manager{}
    @editable
    OpenUpgrade : button_device = button_device{}
    @editable
    AudioManager : audio_manager = audio_manager{}

    @editable
    Tutorial : tutorial_manager = tutorial_manager{}

    @editable
    DrillProcEffect : visual_effect_powerup_device = visual_effect_powerup_device{}

    var DrillCostMult : []float = array{1.0,6000000.0,100000.0,12000000.0}
    var GDUpgradeAmounts : []float = array{1.0,0.20,6.0,1.0}
    var GDMaxLevels : []int = array{1, 10, 10, 3}
    var GoldenDrillCosts : []float = array{7.50e07,5.63e11,7.50e07,4.68e18}
    var GoldenDrillUpgrades : []float = array{0.0,1.0,30.0,1.0} #THIS IS PURELY FOR THE HUD

    var BuffChargeTimer : float = 900.0 #CHANGE TO 900.0 #15 MINUTE TIMER

    # Runs when the device is started in a running game
    OnBegin<override>()<suspends>:void=
        #StartBuff.InteractedWithEvent.Subscribe(ActivateButton)
        OpenUpgrade.InteractedWithEvent.Subscribe(OpenGDUHUD)

        for(Spawner : gm.PlayerSpawners):     
            Spawner.SpawnedEvent.Subscribe(OnPlayerJoined)

    OnPlayerJoined(Agent:agent):void=
        spawn{OnPlayerJoinedBuffer(Agent)}

    OnPlayerJoinedBuffer(Agent:agent)<suspends>:void=
        Sleep(1.0)
        if(CustomPlayer := gm.PlayersMap[player[Agent]]):
            
            if(CustomPlayer.PlayTimeSessionSeconds = 0.0):
                CustomPlayer.GDCloseButton.OnClick().Subscribe(RemoveGDUHUD)
                set CustomPlayer.StoneIcons = array{CustomPlayer.StoneIcon1,CustomPlayer.StoneIcon2,CustomPlayer.StoneIcon3,CustomPlayer.StoneIcon4}
                set CustomPlayer.GDUpgradeTitles = array{CustomPlayer.Upgrade1Title,CustomPlayer.Upgrade2Title,CustomPlayer.Upgrade3Title,CustomPlayer.Upgrade4Title}
                set CustomPlayer.GDUpgradeLevels = array{CustomPlayer.Upgrade1Level,CustomPlayer.Upgrade2Level,CustomPlayer.Upgrade3Level,CustomPlayer.Upgrade4Level}
                set CustomPlayer.GDUpgradeIcons = array{CustomPlayer.Upgrade1Icon,CustomPlayer.Upgrade2Icon,CustomPlayer.Upgrade3Icon,CustomPlayer.Upgrade4Icon}
                set CustomPlayer.GDUpgradeDescTexts = array{CustomPlayer.Upgrade1Desc,CustomPlayer.Upgrade2Desc,CustomPlayer.Upgrade3Desc,CustomPlayer.Upgrade4Desc}
                set CustomPlayer.GDUpgradeBonuses = array{CustomPlayer.Upgrade1Bonus,CustomPlayer.Upgrade2Bonus,CustomPlayer.Upgrade3Bonus,CustomPlayer.Upgrade4Bonus}
                set CustomPlayer.GDUpgradeNextBonuses = array{CustomPlayer.Upgrade1NextBonus,CustomPlayer.Upgrade2NextBonus,CustomPlayer.Upgrade3NextBonus,CustomPlayer.Upgrade4NextBonus}
                set CustomPlayer.GDUpgradeCosts = array{CustomPlayer.Upgrade1CostAmount,CustomPlayer.Upgrade2CostAmount,CustomPlayer.Upgrade3CostAmount,CustomPlayer.Upgrade4CostAmount}
                set CustomPlayer.GDUpgradeCostTexts = array{CustomPlayer.Upgrade1CostText,CustomPlayer.Upgrade2CostText,CustomPlayer.Upgrade3CostText,CustomPlayer.Upgrade4CostText}
                set CustomPlayer.GDUpgradeButtons = array{CustomPlayer.GDUpgradeButton1,CustomPlayer.GDUpgradeButton2,CustomPlayer.GDUpgradeButton3,CustomPlayer.GDUpgradeButton4}
                
                set CustomPlayer.GDAllTexts = array{CustomPlayer.StoneAmountText, CustomPlayer.Upgrade1Title, CustomPlayer.Upgrade1Level,CustomPlayer.Upgrade1Desc,CustomPlayer.Upgrade1Bonus,CustomPlayer.Upgrade1NextBonus,CustomPlayer.Upgrade1CostText,CustomPlayer.Upgrade1CostAmount,CustomPlayer.Upgrade2Title,CustomPlayer.Upgrade2Level,CustomPlayer.Upgrade2Desc,CustomPlayer.Upgrade2Bonus,CustomPlayer.Upgrade2NextBonus,CustomPlayer.Upgrade2CostText,CustomPlayer.Upgrade2CostAmount,CustomPlayer.Upgrade3Title,CustomPlayer.Upgrade3Level,CustomPlayer.Upgrade3Desc,CustomPlayer.Upgrade3Bonus,CustomPlayer.Upgrade3NextBonus,CustomPlayer.Upgrade3CostText,CustomPlayer.Upgrade3CostAmount,CustomPlayer.Upgrade4Title,CustomPlayer.Upgrade4Level,CustomPlayer.Upgrade4Desc,CustomPlayer.Upgrade4Bonus,CustomPlayer.Upgrade4NextBonus,CustomPlayer.Upgrade4CostText,CustomPlayer.Upgrade4CostAmount}
       
                CustomPlayer.DrillChargeTimerIcon.SetVisibility(widget_visibility.Collapsed)
                CustomPlayer.DrillChargeTimer.SetText(StringToMessage(""))
                CustomPlayer.DrillActiveTimer.SetText(StringToMessage(""))

                for(i:int=0..CustomPlayer.GDUpgradeButtons.Length):
                    if(UpgradeButton:=CustomPlayer.GDUpgradeButtons[i]):
                        UpgradeButton.OnClick().SubscribeAgent3(OnUpgradeButton, i)

                if(CustomPlayer.BuffMaxCharges > CustomPlayer.BuffCurrentCharges and CustomPlayer.GoldenDrillTimer = false):
                    set CustomPlayer.GoldenDrillTimer = true
                    ChargeTimer(CustomPlayer.Player)

    ActivateButton(Agent : agent):void=
        if(CustomPlayer := gm.PlayersMap[Agent]):
            if (CustomPlayer.GoldenDrillUnlocked = true and CustomPlayer.BuffCurrentCharges >= 1.0):
                if(Accolade:=AudioManager.Accolades[12]): #AWARD PLAYER EXP
                    Accolade.Award(CustomPlayer.Player)
                spawn:
                    BuffTimer(Agent)
                if(CustomPlayer.BuffChargeTimer <= 0.0 and CustomPlayer.GoldenDrillTimer = false):
                    spawn:
                        ChargeTimer(Agent)
                    set CustomPlayer.GoldenDrillTimer = true
                gm.ShowMessage(CustomPlayer.Player, "Drill Activated!", 0)
            else:
                gm.ShowMessage(CustomPlayer.Player, "No Drill Charges!", 1)

    BuffTimer(Agent : agent)<suspends>:void=
        if(CustomPlayer := gm.PlayersMap[Agent]):
            DrillProcEffect.Spawn()
            DrillProcEffect.SetDuration(CustomPlayer.BuffCurrentTimeSeconds + CustomPlayer.BuffTimeSeconds)
            DrillProcEffect.Pickup(Agent)
            AudioManager.Drill.Play(Agent)
            if (CurrentCharges := Round[CustomPlayer.BuffCurrentCharges]):
                set CustomPlayer.DrillActivations += 1
            set CustomPlayer.BuffCurrentTimeSeconds += CustomPlayer.BuffTimeSeconds
            set CustomPlayer.BuffCurrentCharges -= 1.0
            SaveAllData(CustomPlayer)
            Print("{CustomPlayer.BuffCurrentTimeSeconds}")
            CustomPlayer.UpdateUI()
            if(CustomPlayer.DrillActive = false):
                Print("NEW DRILL")
                set CustomPlayer.DrillActive = true
                loop:
                    TimeLeft : string = "{ConvertToTime(CustomPlayer.BuffCurrentTimeSeconds, "Minutes")}"
                    CustomPlayer.DrillActiveTimer.SetText(StringToMessage(" {TimeLeft}"))
                    if(CustomPlayer.BuffCurrentTimeSeconds <= 0.0):
                        set CustomPlayer.DrillActive = false
                        CustomPlayer.DrillActiveTimer.SetText(StringToMessage(""))
                        set CustomPlayer.BuffCurrentTimeSeconds = 0.0
                        break:
                    Sleep(1.0)
                    set CustomPlayer.BuffCurrentTimeSeconds -= 1.0
            else:
                Print("ALREADY DRILL GOING")
                        
    ChargeTimer(Agent:agent)<suspends>:void=       
        if(CustomPlayer := gm.PlayersMap[Agent]):  
            if(CustomPlayer.BuffChargeTimer <= 0.0): 
                set CustomPlayer.BuffChargeTimer = BuffChargeTimer
            if (CustomPlayer.GoldenDrillUnlocked = true):       
                CustomPlayer.DrillChargeTimerIcon.SetVisibility(widget_visibility.Visible)
                loop:
                    TimeLeft : string = "{ConvertToTime(CustomPlayer.BuffChargeTimer, "Minutes")}"
                    CustomPlayer.DrillChargeTimer.SetText(StringToMessage(" {TimeLeft}"))
                    if(CustomPlayer.BuffChargeTimer <= 0.0):
                        if(CustomPlayer.BuffMaxCharges > CustomPlayer.BuffCurrentCharges):
                            set CustomPlayer.BuffCurrentCharges += 1.0
                        CustomPlayer.DrillChargeTimer.SetText(StringToMessage(""))
                        CustomPlayer.DrillChargeTimerIcon.SetVisibility(widget_visibility.Collapsed)
                        CustomPlayer.UpdateUI()
                        if(CustomPlayer.BuffMaxCharges > CustomPlayer.BuffCurrentCharges):
                            set CustomPlayer.BuffChargeTimer = BuffChargeTimer
                        else:
                            set CustomPlayer.GoldenDrillTimer = false
                            break:
                    Sleep(1.0)
                    set CustomPlayer.BuffChargeTimer -= 1.0
                    SaveAllData(CustomPlayer)

    OpenGDUHUD(Agent:agent): void=
        if(CustomPlayer := gm.PlayersMap[Agent]):
            if(MyUI:=CustomPlayer.UIMaps["GDU"]?):
            else:
                var MyGDCanvas : canvas = MakeGDUCanvas(CustomPlayer)
                CustomPlayer.AddFullHUD(MyGDCanvas)
                if(set CustomPlayer.UIMaps["GDU"] = option{MyGDCanvas}) {}
                UpdateGDUHUD(CustomPlayer)

    RemoveGDUHUD(Wdata:widget_message):void=
        if(CustomPlayer := gm.PlayersMap[Wdata.Player]):
            if(Player:=Wdata.Player, PlayerUI:=GetPlayerUI[Player]):
                if(MyUI:=CustomPlayer.UIMaps["GDU"]?):
                    PlayerUI.RemoveWidget(MyUI)
                    MyUI.SetVisibility(widget_visibility.Collapsed)
                    if(set CustomPlayer.UIMaps["GDU"] = false){}

    
    UpdateGDUHUD(CustomPlayer : Custom_Player): void=
        if (CustomPlayer.GoldenDrillUnlocked = false):
            CustomPlayer.GDHUDBGCover.SetDesiredSize(vector2{X:=2140.5,Y:=1076.0})
            CustomPlayer.GDUpgradeButton2.SetEnabled(false)
            CustomPlayer.GDUpgradeButton3.SetEnabled(false)
            CustomPlayer.GDUpgradeButton4.SetEnabled(false)
        else:
            CustomPlayer.GDHUDBGCover.SetDesiredSize(vector2{X:=0.0,Y:=0.0})
            CustomPlayer.GDUpgradeButton2.SetEnabled(true)
            CustomPlayer.GDUpgradeButton3.SetEnabled(true)
            CustomPlayer.GDUpgradeButton4.SetEnabled(true)

        CustomPlayer.StoneAmountText.SetText(StringToMessage(NumberCondensor(CustomPlayer.StoneAmount)))

        for(i:int=0..CustomPlayer.GDUpgradeTitles.Length):
            if:
                StoneIcon:=CustomPlayer.StoneIcons[i]
                UpgradeNextBonus:=CustomPlayer.GDUpgradeNextBonuses[i]
                UpgradeCostText:=CustomPlayer.GDUpgradeCostTexts[i]
                MaxLevel:=GDMaxLevels[i]
                HutUpgrade:=GoldenDrillUpgrades[i]
                HutUpgradeAmount:=GDUpgradeAmounts[i]
                HutLevel:=CustomPlayer.GoldenDrillLevels[i]
                GDCostMult:= DrillCostMult[i]
                HutCost:=GoldenDrillCosts[i] * Pow(GDCostMult, HutLevel*1.0)
                HutUpgradeTitle:=GDUpgradeTitles[i]
                HutUpgradeDesc:=GDUpgradeDescs[i]
                HutUpgradeIcon:=Hut1Textures[i]
                UpgradeTitle:=CustomPlayer.GDUpgradeTitles[i]
                UpgradeLevel:=CustomPlayer.GDUpgradeLevels[i]
                UpgradeIcon:=CustomPlayer.GDUpgradeIcons[i]
                UpgradeDesc:=CustomPlayer.GDUpgradeDescTexts[i]
                UpgradeBonus:=CustomPlayer.GDUpgradeBonuses[i]
                UpgradeCost:=CustomPlayer.GDUpgradeCosts[i]
                UpgradeButton:=CustomPlayer.GDUpgradeButtons[i]
                UpgradeTotal := HutUpgrade + (HutUpgradeAmount*HutLevel)
            then:
                UpgradeTitle.SetText(StringToMessage(HutUpgradeTitle))
                UpgradeLevel.SetText(StringToMessage("Lv. {HutLevel}"))
                UpgradeIcon.SetImage(HutUpgradeIcon)
                UpgradeDesc.SetText(StringToMessage(HutUpgradeDesc))
                UpgradeCost.SetText(StringToMessage(" {NumberCondensor(HutCost)}"))
                var isPerc : logic = true

                if(CustomPlayer.StoneAmount >= HutCost):
                    UpgradeCost.SetTextColor(MakeColorFromHex("39FF39"))
                else:
                    UpgradeCost.SetTextColor(MakeColorFromHex("fa4d56"))

                if(i = 0 or i = 2 or i = 3):
                    set isPerc = false
                       
                if(isPerc = true):
                    UpgradeBonus.SetText(StringToMessage("{NumberCondensor(100.0 + (UpgradeTotal*100.0))}% "))
                else:
                    UpgradeBonus.SetText(StringToMessage("{NumberCondensor(UpgradeTotal)} "))

                if(HutLevel < MaxLevel):
                    UpgradeButton.SetText(StringToMessage("UPGRADE"))
                    StoneIcon.SetVisibility(widget_visibility.Visible)
                    UpgradeCostText.SetText(StringToMessage("COST:"))
                    UpgradeCostText.SetTextColor(MakeColorFromHex("DDDDDD"))
                    if(isPerc = true):
                        UpgradeNextBonus.SetText(StringToMessage("(+{NumberCondensor(HutUpgradeAmount*100.0)}%)"))
                    else:
                        UpgradeNextBonus.SetText(StringToMessage("(+{NumberCondensor(HutUpgradeAmount)})"))
                else:
                    UpgradeButton.SetText(StringToMessage("MAX"))
                    StoneIcon.SetVisibility(widget_visibility.Collapsed)
                    UpgradeCostText.SetText(StringToMessage(""))
                    UpgradeCost.SetText(StringToMessage(""))
                    UpgradeNextBonus.SetText(StringToMessage(""))

    OnUpgradeButton(Wdata:widget_message, WhicGDgrade:int):void=
        if(CustomPlayer := gm.PlayersMap[Wdata.Player],HutMaxLevels:=GDMaxLevels,HutUpgradeAmounts:=GDUpgradeAmounts, HutCosts:=GoldenDrillCosts,GDCostMult:= DrillCostMult[WhicGDgrade],HutLevels:=CustomPlayer.GoldenDrillLevels, Cost:=HutCosts[WhicGDgrade] * Pow(GDCostMult, HutLevels[WhicGDgrade]*1.0) ):
            if(HutLevels[WhicGDgrade] < HutMaxLevels[WhicGDgrade]):
                if(CustomPlayer.StoneAmount >= Cost): #If you don't have enough stone nothing else matters
                    AudioManager.LeveledUp.Play(CustomPlayer.Player)
                    CustomPlayer.Spend(Cost, "Stone", 0)
                    if(WhicGDgrade = 0):                   
                        Tutorial.UpdateTutorialUI(CustomPlayer, "Drill", 75000000.0)   
                        set CustomPlayer.GoldenDrillUnlocked = true
                    else if(WhicGDgrade = 1):
                        if(set CustomPlayer.BuffMultiplier += HutUpgradeAmounts[WhicGDgrade]) {}
                    else if(WhicGDgrade = 2):
                        if(set CustomPlayer.BuffTimeSeconds += HutUpgradeAmounts[WhicGDgrade]){}
                    else if(WhicGDgrade = 3):
                        if(set CustomPlayer.BuffMaxCharges += HutUpgradeAmounts[WhicGDgrade]) {}
                        set CustomPlayer.BuffCurrentCharges += 1.0
                    #if(set CustomPlayer.GoldenDrillUpgrades[WhicGDgrade] += HutUpgradeAmounts[WhicGDgrade]) {} #Up the amount the hud has given
                    if(set CustomPlayer.GoldenDrillLevels[WhicGDgrade] += 1){} #Up level
                    #if(set CustomPlayer.GoldenDrillCosts[WhicGDgrade] *= DrillCostMult[WhicGDgrade]){} #Up Cost
                    SaveAllData(CustomPlayer)
                else:
                    AudioManager.Denied.Play(CustomPlayer.Player)
                    Print("Not Enough Stone!")
            else:
                AudioManager.Denied.Play(CustomPlayer.Player)
                Print("Max Level!")
            CustomPlayer.UpdateUI()
            UpdateGDUHUD(CustomPlayer)

#UI BLOCKS
    var GDUpgradeTitles : []string = array{
        "Mining Drill",
        "Intensify Power Up",
        "Increase Time",
        "Drill Charges"
        }
 
    var GDUpgradeDescs : []string = array{
        "A large drill used to\ncrack even the strongest\nof stones",
        "Increase the multiplier of\nthe drill power",
        "Increase the duration of\nthe drill buff",
        "Gather drill charges\nto use later"
        }

    var Hut1Textures :[]texture = array{UI.miningdrill,UI.miningpowerup,UI.drilltime,UI.drillcharges}

#CANVAS
    MakeGDUCanvas(CustomPlayer:Custom_Player):canvas=
        CustomPlayer.GDCloseButton.SetText(StringToMessage("CLOSE"))
        CustomPlayer.Upgrade1Desc.SetJustification(text_justification.Center)
        CustomPlayer.Upgrade2Desc.SetJustification(text_justification.Center)
        CustomPlayer.Upgrade3Desc.SetJustification(text_justification.Center)
        CustomPlayer.Upgrade4Desc.SetJustification(text_justification.Center)

        for(TextBlock : CustomPlayer.GDAllTexts):
            TextBlock.SetShadowColor(CustomPlayer.ShadowDefaults.DefaultShadowColor)
            TextBlock.SetShadowOffset(CustomPlayer.ShadowDefaults.DefaultShadowOffset)
            TextBlock.SetShadowOpacity(1.0)

        var MyGDCanvas: canvas = canvas:
            Slots := array:
                canvas_slot:
                    Widget:=CustomPlayer.GDHUDBG
                    ZOrder:=0
                    SizeToContent:=false
                    Alignment := vector2{X := 0.5, Y := 0.0}
                    Anchors := anchors{Minimum := vector2{X:= 0.5, Y:= 0.0}, Maximum := vector2{X:= 0.5, Y:= 0.0}}
                    Offsets:=margin{Top:=0.0,Left:=0.0,Bottom:=1076.0,Right:=2140.5}
                canvas_slot:
                    Widget:=CustomPlayer.GDHUDBGCover
                    ZOrder:=2
                    SizeToContent:=true
                    Alignment := vector2{X := 0.5, Y := 0.0}
                    Anchors := anchors{Minimum := vector2{X:= 0.5, Y:= 0.0}, Maximum := vector2{X:= 0.5, Y:= 0.0}}
                    Offsets:=margin{Top:=0.0,Left:=0.0,Bottom:=1076.0,Right:=2140.5}
                canvas_slot:
                    Widget:=CustomPlayer.GDCloseButton
                    ZOrder:=9
                    SizeToContent:=false
                    Alignment := vector2{X:= 0.5, Y:= 0.0}
                    Anchors := anchors{Minimum := vector2{X:= 0.5, Y:= 0.0}, Maximum := vector2{X:= 0.5, Y:= 0.0}}
                    Offsets:=margin{Top:=976.0,Left:=-660.0,Bottom:=CustomPlayer.GDClose_Size.Y,Right:=CustomPlayer.GDClose_Size.X}
                
                canvas_slot:
                    Widget:=CustomPlayer.GDHUDTitle
                    ZOrder:=1
                    SizeToContent:=false
                    Anchors := anchors{Minimum := vector2{X:= 0.5, Y:= 0.0}, Maximum := vector2{X:= 0.5, Y:= 0.0}}
                    Offsets:=margin{Top:=120.0,Left:=-100.0, Bottom:=56.0, Right:=521.0}

                canvas_slot:
                    Widget:=CustomPlayer.StoneAmountIcon
                    ZOrder:=2
                    SizeToContent:=false
                    Anchors := anchors{Minimum := vector2{X:= 0.5, Y:= 0.0}, Maximum := vector2{X:= 0.5, Y:= 0.0}}
                    Offsets:=margin{Top:=162.0,Left:=-828.0, Bottom:=40.0, Right:=40.0}
                canvas_slot:
                    Widget:=CustomPlayer.StoneAmountText
                    ZOrder:=2
                    SizeToContent:=true
                    Anchors := anchors{Minimum := vector2{X:= 0.5, Y:= 0.0}, Maximum := vector2{X:= 0.5, Y:= 0.0}}
                    Offsets:=margin{Top:=162.0,Left:=-776.0}
                    Alignment := vector2{X:= 0.0, Y:= 0.0}

                canvas_slot:
                    Widget:=CustomPlayer.Upgrade1Title
                    ZOrder:=2
                    SizeToContent:=true
                    Anchors := anchors{Minimum := vector2{X:= 0.5, Y:= 0.0}, Maximum := vector2{X:= 0.5, Y:= 0.0}}
                    Offsets:=margin{Top:=282.0,Left:=-660.0}
                    Alignment := vector2{X:= 0.5, Y:= 0.0}
                canvas_slot:
                    Widget:=CustomPlayer.Upgrade1Level
                    ZOrder:=2
                    SizeToContent:=true
                    Anchors := anchors{Minimum := vector2{X:= 0.5, Y:= 0.0}, Maximum := vector2{X:= 0.5, Y:= 0.0}}
                    Offsets:=margin{Top:=341.0,Left:=-660.0}
                    Alignment := vector2{X:= 0.5, Y:= 0.0}
                canvas_slot:
                    Widget:=CustomPlayer.Upgrade1Icon
                    ZOrder:=2
                    SizeToContent:=false
                    Anchors := anchors{Minimum := vector2{X:= 0.5, Y:= 0.0}, Maximum := vector2{X:= 0.5, Y:= 0.0}}
                    Offsets:=margin{Top:=421.0,Left:=-660.0, Bottom:=140.0,Right:=140.0}
                    Alignment := vector2{X:= 0.5, Y:= 0.0}
                canvas_slot:
                    Widget:=CustomPlayer.Upgrade1Desc
                    ZOrder:=2
                    SizeToContent:=true
                    Anchors := anchors{Minimum := vector2{X:= 0.5, Y:= 0.0}, Maximum := vector2{X:= 0.5, Y:= 0.0}}
                    Offsets:=margin{Top:=592.0,Left:=-660.0}
                    Alignment := vector2{X:= 0.5, Y:= 0.0}
                canvas_slot:
                    #Widget:=Upgrade1Bonus
                    ZOrder:=1
                    SizeToContent:=true
                    Anchors := anchors{Minimum := vector2{X:= 0.5, Y:= 0.0}, Maximum := vector2{X:= 0.5, Y:= 0.0}}
                    Offsets:=margin{Top:=740.0,Left:=-660.0}
                    Alignment := vector2{X:= 0.5, Y:= 0.0}
                    Widget := stack_box:
                        Orientation := orientation.Horizontal
                        Slots := array:
                            stack_box_slot:
                                Widget := CustomPlayer.Upgrade1Bonus   
                            stack_box_slot:
                                Widget := CustomPlayer.Upgrade1NextBonus
                canvas_slot:
                    Widget:=CustomPlayer.Upgrade1CostText
                    ZOrder:=2
                    SizeToContent:=true
                    Anchors := anchors{Minimum := vector2{X:= 0.5, Y:= 0.0}, Maximum := vector2{X:= 0.5, Y:= 0.0}}
                    Offsets:=margin{Top:=800.0,Left:=-776.0}
                    Alignment := vector2{X:= 0.5, Y:= 0.0}
                canvas_slot:
                    ZOrder:=2
                    SizeToContent:=true
                    Anchors := anchors{Minimum := vector2{X:= 0.5, Y:= 0.0}, Maximum := vector2{X:= 0.5, Y:= 0.0}}
                    Offsets:=margin{Top:=800.0,Left:=-484.0}
                    Alignment := vector2{X:= 1.0, Y:= 0.0}
                    Widget := stack_box:
                        Orientation := orientation.Horizontal
                        Slots := array:
                            stack_box_slot:
                                Widget := CustomPlayer.StoneIcon1    
                            stack_box_slot:
                                Widget := CustomPlayer.Upgrade1CostAmount
                canvas_slot:
                    Widget:=CustomPlayer.GDUpgradeButton1
                    ZOrder:=3
                    SizeToContent:=true
                    Anchors := anchors{Minimum := vector2{X:= 0.5, Y:= 0.0}, Maximum := vector2{X:= 0.5, Y:= 0.0}}
                    Offsets:=margin{Top:=870.0,Left:=-660.0, Bottom:=60.0, Right:=220.0}
                    Alignment := vector2{X:= 0.5, Y:= 0.0}
                    
                canvas_slot:
                    Widget:=CustomPlayer.Upgrade2Title
                    ZOrder:=1
                    SizeToContent:=true
                    Anchors := anchors{Minimum := vector2{X:= 0.5, Y:= 0.0}, Maximum := vector2{X:= 0.5, Y:= 0.0}}
                    Offsets:=margin{Top:=282.0,Left:=-220.0}
                    Alignment := vector2{X:= 0.5, Y:= 0.0}
                canvas_slot:
                    Widget:=CustomPlayer.Upgrade2Level
                    ZOrder:=1
                    SizeToContent:=true
                    Anchors := anchors{Minimum := vector2{X:= 0.5, Y:= 0.0}, Maximum := vector2{X:= 0.5, Y:= 0.0}}
                    Offsets:=margin{Top:=341.0,Left:=-220.0}
                    Alignment := vector2{X:= 0.5, Y:= 0.0}
                canvas_slot:
                    Widget:=CustomPlayer.Upgrade2Icon
                    ZOrder:=1
                    SizeToContent:=false
                    Anchors := anchors{Minimum := vector2{X:= 0.5, Y:= 0.0}, Maximum := vector2{X:= 0.5, Y:= 0.0}}
                    Offsets:=margin{Top:=421.0,Left:=-220.0, Bottom:=140.0,Right:=140.0}
                    Alignment := vector2{X:= 0.5, Y:= 0.0}
                canvas_slot:
                    Widget:=CustomPlayer.Upgrade2Desc
                    ZOrder:=1
                    SizeToContent:=true
                    Anchors := anchors{Minimum := vector2{X:= 0.5, Y:= 0.0}, Maximum := vector2{X:= 0.5, Y:= 0.0}}
                    Offsets:=margin{Top:=592.0,Left:=-220.0}
                    Alignment := vector2{X:= 0.5, Y:= 0.0}
                canvas_slot:
                    #Widget:=Upgrade2Bonus
                    ZOrder:=1
                    SizeToContent:=true
                    Anchors := anchors{Minimum := vector2{X:= 0.5, Y:= 0.0}, Maximum := vector2{X:= 0.5, Y:= 0.0}}
                    Offsets:=margin{Top:=740.0,Left:=-220.0}
                    Alignment := vector2{X:= 0.5, Y:= 0.0}
                    Widget := stack_box:
                        Orientation := orientation.Horizontal
                        Slots := array:
                            stack_box_slot:
                                Widget := CustomPlayer.Upgrade2Bonus   
                            stack_box_slot:
                                Widget := CustomPlayer.Upgrade2NextBonus
                canvas_slot:
                    Widget:=CustomPlayer.Upgrade2CostText
                    ZOrder:=1
                    SizeToContent:=true
                    Anchors := anchors{Minimum := vector2{X:= 0.5, Y:= 0.0}, Maximum := vector2{X:= 0.5, Y:= 0.0}}
                    Offsets:=margin{Top:=800.0,Left:=-336.0}
                    Alignment := vector2{X:= 0.5, Y:= 0.0}
                canvas_slot:
                    ZOrder:=1
                    SizeToContent:=true
                    Anchors := anchors{Minimum := vector2{X:= 0.5, Y:= 0.0}, Maximum := vector2{X:= 0.5, Y:= 0.0}}
                    Offsets:=margin{Top:=800.0,Left:=-44.0}
                    Alignment := vector2{X:= 1.0, Y:= 0.0}
                    Widget := stack_box:
                        Orientation := orientation.Horizontal
                        Slots := array:
                            stack_box_slot:
                                Widget := CustomPlayer.StoneIcon2   
                            stack_box_slot:
                                Widget := CustomPlayer.Upgrade2CostAmount
                canvas_slot:
                    Widget:=CustomPlayer.GDUpgradeButton2
                    ZOrder:=1
                    SizeToContent:=true
                    Anchors := anchors{Minimum := vector2{X:= 0.5, Y:= 0.0}, Maximum := vector2{X:= 0.5, Y:= 0.0}}
                    Offsets:=margin{Top:=870.0,Left:=-220.0, Bottom:=60.0, Right:=220.0}
                    Alignment := vector2{X:= 0.5, Y:= 0.0}
                    
                canvas_slot:
                    Widget:=CustomPlayer.Upgrade3Title
                    ZOrder:=1
                    SizeToContent:=true
                    Anchors := anchors{Minimum := vector2{X:= 0.5, Y:= 0.0}, Maximum := vector2{X:= 0.5, Y:= 0.0}}
                    Offsets:=margin{Top:=282.0,Left:=216.0}
                    Alignment := vector2{X:= 0.5, Y:= 0.0}
                canvas_slot:
                    Widget:=CustomPlayer.Upgrade3Level
                    ZOrder:=1
                    SizeToContent:=true
                    Anchors := anchors{Minimum := vector2{X:= 0.5, Y:= 0.0}, Maximum := vector2{X:= 0.5, Y:= 0.0}}
                    Offsets:=margin{Top:=341.0,Left:=216.0}
                    Alignment := vector2{X:= 0.5, Y:= 0.0}
                canvas_slot:
                    Widget:=CustomPlayer.Upgrade3Icon
                    ZOrder:=1
                    SizeToContent:=false
                    Anchors := anchors{Minimum := vector2{X:= 0.5, Y:= 0.0}, Maximum := vector2{X:= 0.5, Y:= 0.0}}
                    Offsets:=margin{Top:=421.0,Left:=216.0, Bottom:=140.0,Right:=140.0}
                    Alignment := vector2{X:= 0.5, Y:= 0.0}
                canvas_slot:
                    Widget:=CustomPlayer.Upgrade3Desc
                    ZOrder:=1
                    SizeToContent:=true
                    Anchors := anchors{Minimum := vector2{X:= 0.5, Y:= 0.0}, Maximum := vector2{X:= 0.5, Y:= 0.0}}
                    Offsets:=margin{Top:=592.0,Left:=216.0}
                    Alignment := vector2{X:= 0.5, Y:= 0.0}
                canvas_slot:
                    #Widget:=Upgrade3Bonus
                    ZOrder:=1
                    SizeToContent:=true
                    Anchors := anchors{Minimum := vector2{X:= 0.5, Y:= 0.0}, Maximum := vector2{X:= 0.5, Y:= 0.0}}
                    Offsets:=margin{Top:=740.0,Left:=216.0}
                    Alignment := vector2{X:= 0.5, Y:= 0.0}
                    Widget := stack_box:
                        Orientation := orientation.Horizontal
                        Slots := array:
                            stack_box_slot:
                                Widget := CustomPlayer.Upgrade3Bonus   
                            stack_box_slot:
                                Widget := CustomPlayer.Upgrade3NextBonus
                canvas_slot:
                    Widget:=CustomPlayer.Upgrade3CostText
                    ZOrder:=1
                    SizeToContent:=true
                    Anchors := anchors{Minimum := vector2{X:= 0.5, Y:= 0.0}, Maximum := vector2{X:= 0.5, Y:= 0.0}}
                    Offsets:=margin{Top:=800.0,Left:=100.0}
                    Alignment := vector2{X:= 0.5, Y:= 0.0}
                canvas_slot:
                    ZOrder:=1
                    SizeToContent:=true
                    Anchors := anchors{Minimum := vector2{X:= 0.5, Y:= 0.0}, Maximum := vector2{X:= 0.5, Y:= 0.0}}
                    Offsets:=margin{Top:=800.0,Left:=392.0}
                    Alignment := vector2{X:= 1.0, Y:= 0.0}
                    Widget := stack_box:
                        Orientation := orientation.Horizontal
                        Slots := array:
                            stack_box_slot:
                                Widget := CustomPlayer.StoneIcon3    
                            stack_box_slot:
                                Widget := CustomPlayer.Upgrade3CostAmount
                canvas_slot:
                    Widget:=CustomPlayer.GDUpgradeButton3
                    ZOrder:=1
                    SizeToContent:=true
                    Anchors := anchors{Minimum := vector2{X:= 0.5, Y:= 0.0}, Maximum := vector2{X:= 0.5, Y:= 0.0}}
                    Offsets:=margin{Top:=870.0,Left:=216.0, Bottom:=60.0, Right:=220.0}
                    Alignment := vector2{X:= 0.5, Y:= 0.0}
                    
                canvas_slot:
                    Widget:=CustomPlayer.Upgrade4Title
                    ZOrder:=1
                    SizeToContent:=true
                    Anchors := anchors{Minimum := vector2{X:= 0.5, Y:= 0.0}, Maximum := vector2{X:= 0.5, Y:= 0.0}}
                    Offsets:=margin{Top:=282.0,Left:=664.0}
                    Alignment := vector2{X:= 0.5, Y:= 0.0}
                canvas_slot:
                    Widget:=CustomPlayer.Upgrade4Level
                    ZOrder:=1
                    SizeToContent:=true
                    Anchors := anchors{Minimum := vector2{X:= 0.5, Y:= 0.0}, Maximum := vector2{X:= 0.5, Y:= 0.0}}
                    Offsets:=margin{Top:=341.0,Left:=664.0}
                    Alignment := vector2{X:= 0.5, Y:= 0.0}
                canvas_slot:
                    Widget:=CustomPlayer.Upgrade4Icon
                    ZOrder:=1
                    SizeToContent:=false
                    Anchors := anchors{Minimum := vector2{X:= 0.5, Y:= 0.0}, Maximum := vector2{X:= 0.5, Y:= 0.0}}
                    Offsets:=margin{Top:=421.0,Left:=664.0, Bottom:=140.0,Right:=140.0}
                    Alignment := vector2{X:= 0.5, Y:= 0.0}
                canvas_slot:
                    Widget:=CustomPlayer.Upgrade4Desc
                    ZOrder:=1
                    SizeToContent:=true
                    Anchors := anchors{Minimum := vector2{X:= 0.5, Y:= 0.0}, Maximum := vector2{X:= 0.5, Y:= 0.0}}
                    Offsets:=margin{Top:=592.0,Left:=664.0}
                    Alignment := vector2{X:= 0.5, Y:= 0.0}
                canvas_slot:
                    #Widget:=Upgrade4Bonus
                    ZOrder:=1
                    SizeToContent:=true
                    Anchors := anchors{Minimum := vector2{X:= 0.5, Y:= 0.0}, Maximum := vector2{X:= 0.5, Y:= 0.0}}
                    Offsets:=margin{Top:=740.0,Left:=664.0}
                    Alignment := vector2{X:= 0.5, Y:= 0.0}
                    Widget := stack_box:
                        Orientation := orientation.Horizontal
                        Slots := array:
                            stack_box_slot:
                                Widget := CustomPlayer.Upgrade4Bonus   
                            stack_box_slot:
                                Widget := CustomPlayer.Upgrade4NextBonus
                canvas_slot:
                    Widget:=CustomPlayer.Upgrade4CostText
                    ZOrder:=1
                    SizeToContent:=true
                    Anchors := anchors{Minimum := vector2{X:= 0.5, Y:= 0.0}, Maximum := vector2{X:= 0.5, Y:= 0.0}}
                    Offsets:=margin{Top:=800.0,Left:=552.0}
                    Alignment := vector2{X:= 0.5, Y:= 0.0}
                canvas_slot:
                    ZOrder:=1
                    SizeToContent:=true
                    Anchors := anchors{Minimum := vector2{X:= 0.5, Y:= 0.0}, Maximum := vector2{X:= 0.5, Y:= 0.0}}
                    Offsets:=margin{Top:=800.0,Left:=840.0}
                    Alignment := vector2{X:= 1.0, Y:= 0.0}
                    Widget := stack_box:
                        Orientation := orientation.Horizontal
                        Slots := array:
                            stack_box_slot:
                                Widget := CustomPlayer.StoneIcon4    
                            stack_box_slot:
                                Widget := CustomPlayer.Upgrade4CostAmount
                canvas_slot:
                    Widget:=CustomPlayer.GDUpgradeButton4
                    ZOrder:=1
                    SizeToContent:=true
                    Anchors := anchors{Minimum := vector2{X:= 0.5, Y:= 0.0}, Maximum := vector2{X:= 0.5, Y:= 0.0}}
                    Offsets:=margin{Top:=870.0,Left:=664.0, Bottom:=60.0, Right:=220.0}
                    Alignment := vector2{X:= 0.5, Y:= 0.0}
