using { /Fortnite.com/Devices }
using { /Verse.org/Simulation }
using { /Verse.org/Colors }
using { /Verse.org/Verse }
using { /Verse.org/Random }
using { /Fortnite.com/Teams }
using { /UnrealEngine.com/Assets }
using { /Fortnite.com/Game }
using { /Fortnite.com/Characters }
using { /Fortnite.com/Playspaces }
using { /Fortnite.com/UI }
using { /UnrealEngine.com/Temporary/Diagnostics }
using { /UnrealEngine.com/Temporary/SpatialMath }
using { /UnrealEngine.com/Temporary/UI }
using { /Verse.org/Assets}

# See https://dev.epicgames.com/documentation/en-us/uefn/create-your-own-device-in-verse for how to create a verse device.

CustomRNG := class:
    var State : int
    A:int=1664525
    C:int=1013904223
    M:int=4294967296


    Next():int=
        if (NewState := Mod[(A * State + C), M], set State=NewState){}
        return State

    Random():float=
        return (Next()*1.0) / (M*1.0)

# A Verse-authored creative device that can be placed in a level
hat_shop_manager := class(creative_device):

    DateData : date_data = date_data{}

    @editable
    gm : game_manager = game_manager{}
    @editable
    AudioManager : audio_manager = audio_manager{}
    @editable
    TimeProp : creative_prop = creative_prop{}
    @editable
    HatVFXList : []visual_effect_powerup_device = array{}
    @editable
    drm : daily_rewards_manager = daily_rewards_manager{}
    @editable
    HatProps : []creative_prop = array{}
    @editable
    ShopButtons : []button_device = array{}
    @editable
    CodeShopButtons : []button_device = array{}
    @editable
    EquipButtons : []button_device = array{}
    @editable
    CodeEquipButtons : []button_device = array{}
    @editable
    HatPrestigeUnlockButtons : []button_device = array{}
    @editable
    HatPrestigeButton : button_device = button_device{}
    @editable
    EquipHatProps : []creative_prop = array{}
    @editable
    Sequencers : []cinematic_sequence_device = array{}
    @editable
    CodeSequencers : []cinematic_sequence_device = array{}
    @editable
    HatDimension : mutator_zone_device = mutator_zone_device{}
    @editable
    UnequipHatBtn : button_device = button_device{}
    @editable
    UnequipVFX : visual_effect_powerup_device = visual_effect_powerup_device{}
    @editable
    CodeHatVFXList : []visual_effect_powerup_device = array{}
    @editable
    UnequipTrailTrigger : trigger_device = trigger_device{} 
    @editable
    UnlockAllHatsBtn : button_device = button_device{}
    @editable
    HatPrestigePopup : popup_dialog_device = popup_dialog_device{}

    @editable
    SpawnTrail : trigger_device = trigger_device{}
    @editable
    VFXCreator : vfx_creator_device = vfx_creator_device{}
    @editable
    ExclusiveVFXSpawners : []visual_effect_powerup_device = array{} 
    @editable
    ExclusiveEquipButtons : []button_device = array{}
    @editable
    PigSlippers : []visual_effect_powerup_device = array{}
    @editable
    FancyCrown : visual_effect_powerup_device = visual_effect_powerup_device{}
    @editable
    UnequipFeet : []visual_effect_powerup_device = array{}

    var Yesterday : date_data = date_data{Hour:=-1}
    var Today : date_data = date_data{}
    var CurrentHats : []int = array{}
    var ShopChoices : [][]int = array{}
    var Tomorrow : date_data = date_data{Year:=2024,Month:=11,Day:=6,Hour:=0,Minute:=0,Second:=0,WeekDay:=2}

    OnUnlockAllHatsPress(Agent:agent):void=
        if(CustomPlayer := gm.PlayersMap[Agent]):
            for(i:=0..CustomPlayer.HatsUnlocked.Length-1):
                if (set CustomPlayer.HatsUnlocked[i] = true){}

    OnExclusiveButtonPress(Agent:agent, Which:int):void=
        if(CustomPlayer := gm.PlayersMap[Agent]):
            if (Which = 0):
                if(Hat := ExclusiveVFXSpawners[Which]):
                    Hat.Pickup(Agent)
            else if (Which = 1):
                if(LeftShoe := ExclusiveVFXSpawners[Which], RightShoe := ExclusiveVFXSpawners[Which + 1]):
                    LeftShoe.Pickup(Agent)
                    RightShoe.Pickup(Agent)
            else if (Which = 2):
                Print("Spawn Trail")
                SpawnTrail.Trigger(CustomPlayer.Player)


    OnUnEquipPress(Agent:agent):void=
        UnequipVFX.Pickup(Agent)
        UnequipTrailTrigger.Trigger(Agent)
        for(i:int=0..UnequipFeet.Length-1):
            if (VFXPowerup := UnequipFeet[i]):
                VFXPowerup.Pickup(Agent)

    GetCurrentShopChoices():[]int=
        var Choices : []int = array{}
        if(Epoch := Floor[GetSecondsSinceEpoch()]):
            WeekHours := (Today.WeekDay * 24) + Today.Hour
            if(Value := ShopChoices[WeekHours]):
                set Choices = Value
        return Choices

    GetWeekDay(UTime:int)<transacts><decides>:int=
        DaysSinceEpoch := Floor(UTime / 86400)
        NewWeekDay := Mod[DaysSinceEpoch + 3,7]
        #Print("Weekday: {NewWeekDay}")
        return NewWeekDay

    OnHatEquipButtonPress(Agent:agent, Which:int):void=
        if(CustomPlayer := gm.PlayersMap[Agent]):
            for(i:int=0..EquipButtons.Length):
                if (Which = i):
                    if (Hat := HatVFXList[i], CustomPlayer.HatsUnlocked[i] = true):
                        Hat.Pickup(Agent)
                    else:
                        gm.ShowMessage(CustomPlayer.Player, "Hat Not Unlocked!", 1)

    CheckIfInDimension(Agent:agent)<suspends>:void=
        if(CustomPlayer := gm.PlayersMap[Agent]):
            loop:
                if (HatDimension.IsInVolume[Agent]):
                    for (i:int=0..CustomPlayer.HatsUnlocked.Length):
                        if (CustomPlayer.HatsUnlocked[i] = true):
                            if (Sequencer := Sequencers[i]):
                                Sequencer.Play(Agent)
                    for (i:int=0..CustomPlayer.CodeHatsUnlocked.Length):
                        if (CustomPlayer.CodeHatsUnlocked[i] = true):
                            if (Sequencer := CodeSequencers[i]):
                                Sequencer.Play(Agent)
                else:
                    break
                Sleep(1.0)

    CheckDimensionHelper(Agent:agent):void=
        spawn:
            CheckIfInDimension(Agent)

    CheckForShopUpdate()<suspends>:void=
        #Print("Checking for update")
        #Print("Year: {Yesterday.Year}/Month: {Yesterday.Month}/Day: {Yesterday.Day}/Hour: {Yesterday.Hour}/Min: {Yesterday.Minute}/Second: {Yesterday.Second}")
        loop:  
            set Today = DateData.ParseEpoch(GetSecondsSinceEpoch())          
            if (Yesterday.Hour <> Today.Hour or Today.Hour=0):
                set Yesterday = Today
                Print("going to update")
                #Print("Year: {Today.Year}/Month: {Today.Month}/Day: {Today.Day}/Hour: {Today.Hour}/Min: {Today.Minute}/Second: {Today.Second}")
                set CurrentHats = GetCurrentShopChoices()
                for(i:CurrentHats):
                    #Print("Current Hats: {i}")
                for (i:int=0..CurrentHats.Length):
                    if (Hat := CurrentHats[i]):
                        if (HatMesh := HatMeshesList[Hat]):
                            if (HatProp := HatProps[i]):
                                HatProp.SetMesh(HatMesh)
                            else:
                                #Print("No hat prop found sucker!")
                        else:
                            #Print("No hat mesh found sucker!")        
                    else:
                        #Print("No hat found sucker!")
            Sleep(60.0)

    InitShop()<suspends>:void=
        Print("Init Shop")
        CRNG := CustomRNG{State:=69420}
        CosmeticList : []int = array{0,1,2,3,4,5,6,7,8,9,10,11,12,13,14,15,16,17,18,19,20,21,22,23,24,25,26,27,28,29,30,31,32,33,34,35,36,37,38,39,40,41,42,43,44,45,46,47,48,49,50,51,52,53,54,55,56,57,58,59,60,61,62,63,64,65,66,67,68,69,70,71,72,73,74,75,76,77,78,79}
        for(X:=0..167):
            var Options : []int = CosmeticList
            var Opt : []int = array{}
            for (i:int=0..3):
                IDX := CRNG.Random()
                if(SV := Floor[IDX*Options.Length], AV := Options[SV]):
                    set Opt += array{AV}
                    if (NewArray := Options.RemoveElement[SV]):
                        set Options = NewArray
            set ShopChoices += array{Opt}
            if (Mod[X,25] = 0):
                Sleep(0.0)
        set CurrentHats = GetCurrentShopChoices()
        CheckForShopUpdate()

    OnShopButtonPress(Agent:agent, Which:int):void=
        if(CustomPlayer := gm.PlayersMap[Agent]):
            HatCost := 1000.0 + (1000.0 * (CustomPlayer.HatPrestiges))
            if (CustomPlayer.ShardAmount > HatCost):
                if (WhichHat := CurrentHats[Which]):
                    if (Hat := HatVFXList[WhichHat]):
                        if(CustomPlayer.HatsUnlocked[WhichHat] = false):        
                            CustomPlayer.Spend(HatCost, "Shard", 0)                  
                            if (set CustomPlayer.HatsUnlocked[WhichHat] = true){}
                            Hat.Pickup(Agent)
                            if(Analytic:=AudioManager.GeneralAnalytics[9]): #ANALYTIC
                                Analytic.Submit(CustomPlayer.Player)
                            SaveAllData(CustomPlayer)
                            if(Accolade:=AudioManager.Accolades[15]): #AWARD PLAYER EXP
                                Accolade.Award(CustomPlayer.Player)
                            gm.ShowMessage(CustomPlayer.Player, "Hat Purchased!", 0)
                        else:
                            gm.ShowMessage(CustomPlayer.Player, "Hat already unlocked!", 1)
                    else:
                        Print("cant Vind Hat VFX")
                else:
                    Print("cant Find whichHat")
            else:
                gm.ShowMessage(CustomPlayer.Player, "Not Enough Prestige Shards!", 1)

    # Runs when the device is started in a running game
    OnBegin<override>()<suspends>:void=
        Sleep(2.0)
        spawn:
            InitShop()
        HatDimension.AgentEntersEvent.Subscribe(CheckDimensionHelper)
        UnequipHatBtn.InteractedWithEvent.Subscribe(OnUnEquipPress)
        HatPrestigeButton.InteractedWithEvent.Subscribe(OnHatPrestigeButton)
        UnlockAllHatsBtn.InteractedWithEvent.Subscribe(OnUnlockAllHatsPress)
        HatPrestigePopup.RespondingButtonEvent.Subscribe(PopupPress)

        for (i:int=0..ShopButtons.Length-1):
            if (Btn := ShopButtons[i]):
                Btn.InteractedWithEvent.SubscribeAgent(OnShopButtonPress, i)
        for (i:int=0..CodeShopButtons.Length-1):
            if (Btn := CodeShopButtons[i]):
                Btn.InteractedWithEvent.SubscribeAgent(OnCodeShopButtonPress, i)
        for (i:int=0..EquipButtons.Length-1):
            if (Btn := EquipButtons[i]):
                Btn.InteractedWithEvent.SubscribeAgent(OnHatEquipButtonPress, i)
        for (i:int=0..CodeEquipButtons.Length-1):
            if (Btn := CodeEquipButtons[i]):
                Btn.InteractedWithEvent.SubscribeAgent(OnCodeHatEquipButtonPress, i)
        for (i:int=0..ExclusiveEquipButtons.Length-1):
            if (Btn := ExclusiveEquipButtons[i]):
                Btn.InteractedWithEvent.SubscribeAgent(OnExclusiveButtonPress, i)
        for (i:int=0..HatPrestigeUnlockButtons.Length-1):
            if (Btn := HatPrestigeUnlockButtons[i]):
                Btn.InteractedWithEvent.SubscribeAgent(OnPrestigeHatEquipButton, i)
        if (VFXCreator.TeleportTo[VFXCreator.GetTransform().Translation, VFXCreator.GetTransform().Rotation]):
            block:

    
    OnCodeShopButtonPress(Agent:agent, Which:int):void=
        if(CustomPlayer := gm.PlayersMap[Agent]):
            #if (CustomPlayer.ShardAmount > 1000.0):
            if (Hat := CodeHatVFXList[Which]):
                if(CustomPlayer.CodeHatsUnlocked[Which] = false):        
                    #CustomPlayer.Spend(1000.0, "Shard", 0)                  
                    if (set CustomPlayer.CodeHatsUnlocked[Which] = true){}
                    Hat.Pickup(Agent)
                    SaveAllData(CustomPlayer)
                    gm.ShowMessage(CustomPlayer.Player, "Hat Purchased!", 0)
                else:
                    gm.ShowMessage(CustomPlayer.Player, "Hat already unlocked!", 1)
            else:
                Print("cant Vind Hat VFX")
            #else:
            #    gm.ShowMessage(CustomPlayer.Player, "Not Enough Prestige Shards!", 1)

    OnCodeHatEquipButtonPress(Agent:agent, Which:int):void=
        if(CustomPlayer := gm.PlayersMap[Agent]):
            for(i:int=0..CodeEquipButtons.Length):
                if (Which = i):
                    if (Hat := CodeHatVFXList[i], CustomPlayer.CodeHatsUnlocked[i] = true):
                        Hat.Pickup(Agent)
                    else:
                        gm.ShowMessage(CustomPlayer.Player, "Hat Not Unlocked!", 1)

    OnHatPrestigeButton(Agent:agent):void=
        #var CanPrestige : logic = true
        if(CustomPlayer := gm.PlayersMap[Agent]):
            HatPrestigePopup.Show(Agent)
            <# for(HatUnlocked:CustomPlayer.HatsUnlocked):
                if(HatUnlocked = false):
                    Print("DON'T HAVE A HAT")
                    set CanPrestige = false
            if(CanPrestige = true):
                gm.ShowMessage(CustomPlayer.Player, "You've Prestiged!", 0)
                AudioManager.Prestige.Play(CustomPlayer.Player)
                set CustomPlayer.HatPrestiges += 1
                set CustomPlayer.NewHatPrestiges += 1
                set CustomPlayer.HatCost *= 2.0
                for(i:=0..CustomPlayer.HatsUnlocked.Length-1):
                    if (set CustomPlayer.HatsUnlocked[i] = false){}
                spawn{CheckIfInDimension(Agent)}
                SaveAllData(CustomPlayer)
            else:
                gm.ShowMessage(CustomPlayer.Player, "You cannot prestige!", 1)
                AudioManager.Denied.Play(CustomPlayer.Player) #>
    
    PopupPress(Agent:agent, Number : int):void=
        var CanPrestige : logic = true
        if(Number = 0):
            if(CustomPlayer := gm.PlayersMap[Agent]):
                for(HatUnlocked:CustomPlayer.HatsUnlocked):
                    if(HatUnlocked = false):
                        set CanPrestige = false
                if(CanPrestige = true):
                    gm.ShowMessage(CustomPlayer.Player, "You've Hat Prestiged!", 0)
                    AudioManager.Prestige.Play(CustomPlayer.Player)
                    set CustomPlayer.HatPrestiges += 1
                    for(i:=0..CustomPlayer.HatsUnlocked.Length-1):
                        if (set CustomPlayer.HatsUnlocked[i] = false){}
                    spawn{CheckIfInDimension(Agent)}
                    SaveAllData(CustomPlayer)
                else:
                    gm.ShowMessage(CustomPlayer.Player, "You cannot prestige!", 1)
                    AudioManager.Denied.Play(CustomPlayer.Player)

    OnPrestigeHatEquipButton(Agent:agent, Which:int):void=
        if(CustomPlayer := gm.PlayersMap[Agent]):
            if(Which = 0):
                if (CustomPlayer.HatPrestiges >= 1):
                    if(LeftShoe := PigSlippers[Which], RightShoe := PigSlippers[Which + 1]):
                        LeftShoe.Pickup(Agent)
                        RightShoe.Pickup(Agent)
                else:
                    gm.ShowMessage(CustomPlayer.Player, "You must Prestige your Hats at least 1 time to equip!", 1)
                    AudioManager.Denied.Play(CustomPlayer.Player)
            if (Which = 1):
                if (CustomPlayer.HatPrestiges >= 5):
                    Print("Spawn Trail")
                    SpawnTrail.Trigger(CustomPlayer.Player)
                else:
                    gm.ShowMessage(CustomPlayer.Player, "You must Prestige your Hats at least 5 times to equip!", 1)
                    AudioManager.Denied.Play(CustomPlayer.Player)
            if (Which = 2):
                if (CustomPlayer.HatPrestiges >= 10):
                    FancyCrown.Pickup(Agent)
                else:
                    gm.ShowMessage(CustomPlayer.Player, "You must Prestige your Hats at least 10 times to equip!", 1)
                    AudioManager.Denied.Play(CustomPlayer.Player)

                


    var CodeHatMeshesList : []mesh = array{
        HatMeshes.DuckHat,
        RealTime.Propeller_hat_01
    }

    var HatMeshesList : []mesh = array{
        HatMeshes.m_AlienAntennaHat01,
        HatMeshes.m_Antenna01,
        HatMeshes.m_AntlerHat01,
        HatMeshes.m_ArmyHelmet01,
        HatMeshes.m_AstronautHelmet01,
        HatMeshes.m_BachelorHat01,
        HatMeshes.m_BaseballHat01,
        HatMeshes.m_BeachHat01,
        HatMeshes.m_BigHornedHelmet01,
        HatMeshes.m_BikeHelmet01,
        HatMeshes.m_Bunnyears01,
        HatMeshes.m_CactusHat01,
        HatMeshes.m_CandleHat01,
        HatMeshes.m_CaptainsHat01,
        HatMeshes.m_ChefHat01,
        HatMeshes.m_ChristmasElfHat01,
        HatMeshes.m_CowboyHat01,
        HatMeshes.m_CrownHat01,
        HatMeshes.m_CrownHat02,
        HatMeshes.m_CylinderHat01,
        HatMeshes.m_DetectiveHat01,
        HatMeshes.m_DevilHorns01,
        HatMeshes.m_DinosaurHat01,
        HatMeshes.m_DivingHelmet01,
        HatMeshes.m_EiffeltowerHat01,
        HatMeshes.m_ElephantHat01,
        HatMeshes.m_ExecutionerHat01,
        HatMeshes.m_FBIHat01,
        HatMeshes.m_FirefighterHat01,
        HatMeshes.m_FlamingoHat01,
        HatMeshes.m_FlowerCrown01,
        HatMeshes.m_Flowerpot01,
        HatMeshes.m_FrenchHat01,
        HatMeshes.m_FruitHat01,
        HatMeshes.m_GladiatorHelmet01,
        HatMeshes.m_HardHat01,
        HatMeshes.m_JesterHat01,
        HatMeshes.m_KettleHat01,
        HatMeshes.m_LaurelWreath01,
        HatMeshes.m_LeatherCap01,
        HatMeshes.m_LeprechaunHat01,
        HatMeshes.m_MafiaHat01,
        HatMeshes.m_MelonHat01,
        HatMeshes.m_MinerHat01,
        HatMeshes.m_MongolianHat01,
        HatMeshes.m_MorionHelmet01,
        HatMeshes.m_MushroomHat01,
        HatMeshes.m_NurseHat01,
        HatMeshes.m_Paperbag01,
        HatMeshes.m_PartyHat01,
        HatMeshes.m_PharaohHat01,
        HatMeshes.m_PirateHat01,
        HatMeshes.m_PoliceHat01,
        HatMeshes.m_PopeHat01,
        HatMeshes.m_RiceHat01,
        HatMeshes.m_RobinHoodHat01,
        HatMeshes.m_RussianHat01,
        HatMeshes.m_SafariHat01,
        HatMeshes.m_SailorHat01,
        HatMeshes.m_SamuraiHelmet01,
        HatMeshes.m_SantaHat01,
        HatMeshes.m_ShamanHat01,
        HatMeshes.m_SharkHat01,
        HatMeshes.m_SleepyHat01,
        HatMeshes.m_SmallHornedHelmet01,
        HatMeshes.m_SombreroHat01,
        HatMeshes.m_SpiderHat01,
        HatMeshes.m_SpikyHat01,
        HatMeshes.m_TiaraCrown01,
        HatMeshes.m_TinfoilHat01,
        HatMeshes.m_ToyWinderHat01,
        HatMeshes.m_TrafficCone01,
        HatMeshes.m_TurbanHelmet01,
        HatMeshes.m_UmbrellaHat01,
        HatMeshes.m_UncleSamHat01,
        HatMeshes.m_UnicornHat01,
        HatMeshes.m_VikingHelmet01,
        HatMeshes.m_WingedHelmet01,
        HatMeshes.m_WinterHat01,
        HatMeshes.m_WitchHat01
    }
