
using { /Fortnite.com/Devices }
using { /Verse.org/Simulation }
using { /Verse.org/Colors }
using { /Verse.org/Verse }
using { /Verse.org/Random }
using { /Fortnite.com/Teams }
using { /Fortnite.com/Game }
using { /Fortnite.com/Characters }
using { /Fortnite.com/Playspaces }
using { /Fortnite.com/UI }
using { /UnrealEngine.com/Temporary/Diagnostics }
using { /UnrealEngine.com/Temporary/SpatialMath }
using { /UnrealEngine.com/Temporary/UI }

health_bar_device := class(creative_device):

    @editable GameManager : game_manager = game_manager{}

    ## HP BAR FOR ROCK
    SmoothHPBar(CustomPlayer:Custom_Player, Intervals : type{ X : int where X > 10})<suspends>:void=
        NewBarLength := CustomPlayer.SmoothNewHealthBarWidth # New size of Foreground Color based on Percentage
        CurrentX := CustomPlayer.RockBarColorTopSmooth.GetDesiredSize().X      # Current Foreground Color Size
        race:
            CustomPlayer.NewProgressUpdatePolledEvent.Await() # If a change is made while the bar is being updated, cancel the previous update
            for (I := 1..Intervals):
                LerpParameter := (I * 1.0) / (Intervals * 1.0) # Used to Linearly Interpolate Current Size to New Size, based on current interval
                NewSizeVec2 := Lerp(vector2{ X := CurrentX, Y := 42.0}, vector2{ X := NewBarLength, Y := 42.0}, LerpParameter)
                CustomPlayer.RockBarColorTopSmooth.SetDesiredSize(NewSizeVec2)
                if (Mod[I, 2] = 0) then Sleep(0.0)

    Init(CustomPlayer:Custom_Player, Name : string, TotalHP :float,?IsPerc:logic=false):void=
        set CustomPlayer.iRockName = Name
        set CustomPlayer.iTotalHP = TotalHP
        set CustomPlayer.iCurrentHP = TotalHP
        if(CustomPlayer.CurrentRockHit <> ""):
            var PetFinalHitBonus : float = 0.0
            for(i:int=0..CustomPlayer.PetSlots.Length):
                if(CustomPlayer.PetSlots[i] <> -1):
                    if:
                        PetLevel := CustomPlayer.PetLevels[CustomPlayer.PetSlots[i]]
                        FinalHit := CustomPlayer.PetFinalStrike[CustomPlayer.PetSlots[i]]
                    then:
                        set PetFinalHitBonus += FinalHit/100.0
            CustomPlayer.HPBarFinalStrikeAmount.SetText(StringToMessage(" {NumberCondensor((CustomPlayer.FinalHitBonus + PetFinalHitBonus)*100.0)}%"))
            CustomPlayer.HPBarFinalStrikeIcon.SetVisibility(widget_visibility.Visible)
        else:
            CustomPlayer.HPBarFinalStrikeAmount.SetText(StringToMessage(""))
            CustomPlayer.HPBarFinalStrikeIcon.SetVisibility(widget_visibility.Collapsed)

        if(IsPerc = false):
            CustomPlayer.HPBarTitleTextBlock.SetText(StringToMessage("{CustomPlayer.iRockName}: {NumberCondensor(CustomPlayer.iCurrentHP)}/{NumberCondensor(CustomPlayer.iTotalHP)}"))
        else:
            CustomPlayer.HPBarTitleTextBlock.SetText(StringToMessage("{CustomPlayer.iRockName}: {NumberCondensor(CustomPlayer.iCurrentHP)}%/{NumberCondensor(CustomPlayer.iTotalHP)}%"))
        BarSize := CustomPlayer.RockBarColorBase.GetDesiredSize()
        set CustomPlayer.SmoothNewHealthBarWidth = BarSize.X * 1.0
        CustomPlayer.RockBarColorTopSmooth.SetDesiredSize(vector2{X:=CustomPlayer.SmoothNewHealthBarWidth,Y:=40.0})
        UpdateHealthBar(CustomPlayer, CustomPlayer.SmoothNewHealthBarWidth)

    TakeDamage(CustomPlayer:Custom_Player, RockHP : float,?IsPerc:logic=false):void=
        HealthPercentage := RockHP / CustomPlayer.iTotalHP
        BarSize := CustomPlayer.RockBarColorBase.GetDesiredSize()
        set CustomPlayer.SmoothNewHealthBarWidth = BarSize.X * HealthPercentage
        if(IsPerc = false):
            CustomPlayer.HPBarTitleTextBlock.SetText(StringToMessage("{CustomPlayer.iRockName}: {NumberCondensor(RockHP)}/{NumberCondensor(CustomPlayer.iTotalHP)}"))
        else:
            CustomPlayer.HPBarTitleTextBlock.SetText(StringToMessage("{CustomPlayer.iRockName}: {NumberCondensor(RockHP)}%/{NumberCondensor(CustomPlayer.iTotalHP)}%"))
        
        UpdateHealthBar(CustomPlayer, CustomPlayer.SmoothNewHealthBarWidth)

    UpdateHealthBar(CustomPlayer:Custom_Player,NewWidth : float):void=
        CustomPlayer.NewProgressUpdatePolledEvent.Signal()
        CustomPlayer.RockBarColorTop.SetDesiredSize(vector2{X:=NewWidth,Y:=28.0})
        CustomPlayer.RockBarColorTop2.SetDesiredSize(vector2{X:=NewWidth,Y:=14.0})
        spawn:
            SmoothHPBar(CustomPlayer, 60)

    CreateUI<private>(CustomPlayer:Custom_Player) : canvas=
        CustomPlayer.HPBarTitleTextBlock.SetShadowColor(MakeColorFromHex("000000"))
        CustomPlayer.HPBarTitleTextBlock.SetShadowOffset(option{vector2{X:=3.0, Y:=3.0}})
        CustomPlayer.HPBarTitleTextBlock.SetShadowOpacity(1.0)
        CustomPlayer.HPBarFinalStrikeAmount.SetShadowColor(MakeColorFromHex("000000"))
        CustomPlayer.HPBarFinalStrikeAmount.SetShadowOffset(option{vector2{X:=3.0, Y:=3.0}})
        CustomPlayer.HPBarFinalStrikeAmount.SetShadowOpacity(1.0)
        CustomPlayer.RockBarColorTop.SetColor(MakeColorFromHex("D61F28"))
        CustomPlayer.RockBarColorTop2.SetColor(MakeColorFromHex("A1171E"))
        CustomPlayer.RockBarColorTopSmooth.SetColor(MakeColorFromHex("EB7076"))
        CustomPlayer.RockBarColorTopSmooth.SetDesiredSize(vector2{X:=600.0,Y:=42.0})
        TheCanvas: canvas = canvas:
            Slots := array:
                canvas_slot:
                    Widget:=CustomPlayer.RockBarColorBase
                    ZOrder:=0
                    SizeToContent:=false
                    Alignment := vector2{X := 0.0, Y := 1.0}
                    Anchors := anchors{Minimum := vector2{X :=  0.5, Y := 0.0}, Maximum := vector2{X := 0.5, Y := 0.0}}
                    Offsets:=margin{Top := 124.0,Left:=-305.0,Bottom:=52.0,Right:=610.0} 
                canvas_slot:
                    ZOrder:=0
                    SizeToContent:=true
                    Alignment := vector2{X := 0.5, Y := 0.0}
                    Anchors := anchors{Minimum := vector2{X :=  0.5, Y := 0.0}, Maximum := vector2{X := 0.5, Y := 0.0}}
                    Offsets:=margin{Top := 132.0,Left:=0.0}
                    Widget := stack_box:
                        Orientation := orientation.Horizontal
                        Slots := array:
                            stack_box_slot:
                                Widget := CustomPlayer.HPBarFinalStrikeIcon
                                HorizontalAlignment := horizontal_alignment.Left   
                            stack_box_slot:
                                Widget:=CustomPlayer.HPBarFinalStrikeAmount
                                HorizontalAlignment := horizontal_alignment.Left   
                canvas_slot:
                    ZOrder:=2
                    SizeToContent:=true
                    Alignment := vector2{X := 0.0, Y := 1.0}
                    Anchors := anchors{Minimum := vector2{X :=  0.5, Y := 0.0}, Maximum := vector2{X := 0.5, Y := 0.0}}
                    Offsets:=margin{Top := 105.0,Left:=-300.0,Bottom:=28.0,Right:=600.0}
                    Widget := stack_box:
                        Orientation := orientation.Vertical
                        Slots := array:
                            stack_box_slot:
                                Widget := CustomPlayer.RockBarColorTop
                                HorizontalAlignment := horizontal_alignment.Left   
                canvas_slot:
                    ZOrder:=2
                    SizeToContent:=true
                    Alignment := vector2{X := 0.0, Y := 1.0}
                    Anchors := anchors{Minimum := vector2{X :=  0.5, Y := 0.0}, Maximum := vector2{X := 0.5, Y := 0.0}}
                    Offsets:=margin{Top := 119.0,Left:=-300.0,Bottom:=14.0,Right:=600.0}
                    Widget := stack_box:
                        Orientation := orientation.Vertical
                        Slots := array:
                            stack_box_slot:
                                Widget := CustomPlayer.RockBarColorTop2
                                HorizontalAlignment := horizontal_alignment.Left   
                canvas_slot:
                    ZOrder:=1
                    SizeToContent:=true
                    Alignment := vector2{X := 0.0, Y := 1.0}
                    Anchors := anchors{Minimum := vector2{X :=  0.5, Y := 0.0}, Maximum := vector2{X := 0.5, Y := 0.0}}
                    Offsets:=margin{Top := 119.0,Left:=-300.0,Bottom:=42.0,Right:=600.0}
                    Widget := stack_box:
                        Orientation := orientation.Vertical
                        Slots := array:
                            stack_box_slot:
                                Widget := CustomPlayer.RockBarColorTopSmooth
                                HorizontalAlignment := horizontal_alignment.Left   
                canvas_slot:
                    ZOrder:=3
                    SizeToContent:=true
                    Alignment := vector2{X := 0.5, Y := 1.0}
                    Anchors := anchors{Minimum := vector2{X :=  0.5, Y := 0.0}, Maximum := vector2{X := 0.5, Y := 0.0}}
                    Offsets:=margin{Top := 119.0,Left:=0.0,Bottom:=0.0,Right:=0.0} #TOP 75.0
                    Widget := stack_box:
                        Orientation := orientation.Vertical
                        Slots := array:
                            stack_box_slot:
                                Widget := CustomPlayer.HPBarTitleTextBlock     
                                HorizontalAlignment := horizontal_alignment.Left                       
        return TheCanvas  
                        

    ShowUIForPlayer(CustomPlayer : Custom_Player):void=
        MyUI := CreateUI(CustomPlayer)
        CustomPlayer.AddRockHealthBarUI(MyUI)

    HideUIForPlayer(CustomPlayer : Custom_Player):void=
        CustomPlayer.RemoveRockHealthBarUI()
            

    OnBegin<override>()<suspends>:void=
        block: