using { /Fortnite.com/Devices }
using { /Verse.org/Simulation }
using { /Verse.org/Colors }
using { /Verse.org/Verse }
using { /Verse.org/Random }
using { /Fortnite.com/Teams }
using { /UnrealEngine.com/Assets }
using { /Fortnite.com/Game }
using { /Fortnite.com/Characters }
using { /Fortnite.com/Playspaces }
using { /Fortnite.com/UI }
using { /UnrealEngine.com/Temporary/Diagnostics }
using { /UnrealEngine.com/Temporary/SpatialMath }
using { /UnrealEngine.com/Temporary/UI }
using { /Verse.org/Assets}


hit_rock_manager := class(creative_device):

    @editable
    gm : game_manager = game_manager{}

    @editable
    Tutorial : tutorial_manager = tutorial_manager{}

    @editable
    SUP : signal_device_manager = signal_device_manager{}

    @editable
    pm: pet_manager = pet_manager{}

    @editable
    AudioManager: audio_manager = audio_manager{}

    @editable
    DRM : new_daily_rewards_manager = new_daily_rewards_manager{}

    HPBars: health_bar_device = health_bar_device{}

    @editable
    Earthquake1 : creative_prop_asset = DefaultCreativePropAsset
    @editable
    Earthquake2 : creative_prop_asset = DefaultCreativePropAsset
    @editable
    Earthquake3 : creative_prop_asset = DefaultCreativePropAsset

    @editable
    RockRubble : creative_prop_asset = DefaultCreativePropAsset
    
    @editable
    VoidBP : creative_prop_asset = DefaultCreativePropAsset

    @editable
    TwoStarBP : creative_prop_asset = DefaultCreativePropAsset
    @editable
    ThreeStarBP : creative_prop_asset = DefaultCreativePropAsset

    @editable
    TheRock:creative_prop = creative_prop{}
    @editable
    MeteorRock:creative_prop = creative_prop{}
    @editable
    NewMeteorRock:creative_prop = creative_prop{}

    @editable
    RockGranter:item_granter_device = item_granter_device{}
    @editable
    WepMatGranter:item_granter_device = item_granter_device{}
    @editable
    PrestigeGranter:item_granter_device = item_granter_device{}

    @editable
    HoldingWepButtons: []conditional_button_device = array{}
    @editable
    MeteorInfoButtons: []button_device = array{}

    @editable
    MainRockManipulator : prop_manipulator_device = prop_manipulator_device{} 
    @editable
    MeteorRockManip : prop_manipulator_device = prop_manipulator_device{} 
    @editable
    NewMeteorRockManip : prop_manipulator_device = prop_manipulator_device{}

    @editable
    A1S1Rocks : []prop_manipulator_device = array{} 
    @editable
    A1S1RockProps : []creative_prop = array{} 
    
    @editable
    A1S2Rocks : []prop_manipulator_device = array{} 
    @editable
    A1S2RockProps : []creative_prop = array{} 
    @editable
    A1S3Rocks : []prop_manipulator_device = array{} 
    @editable
    A1S3RockProps : []creative_prop = array{} 
    @editable
    A1S4Rocks : []prop_manipulator_device = array{} 
    @editable
    A1S4RockProps : []creative_prop = array{} 
    @editable
    A1S5Rocks : []prop_manipulator_device = array{} 
    @editable
    A1S5RockProps : []creative_prop = array{} 
    
    @editable
    A2S1Rocks : []prop_manipulator_device = array{} 
    @editable
    A2S1RockProps : []creative_prop = array{} 
    @editable
    A2S2Rocks : []prop_manipulator_device = array{} 
    @editable
    A2S2RockProps : []creative_prop = array{} 
    @editable
    A2S3Rocks : []prop_manipulator_device = array{} 
    @editable
    A2S3RockProps : []creative_prop = array{} 
    @editable
    A2S4Rocks : []prop_manipulator_device = array{} 
    @editable
    A2S4RockProps : []creative_prop = array{} 
    @editable
    A2S5Rocks : []prop_manipulator_device = array{} 
    @editable
    A2S5RockProps : []creative_prop = array{} 
    
    @editable
    A3S1Rocks : []prop_manipulator_device = array{} 
    @editable
    A3S1RockProps : []creative_prop = array{} 
    @editable
    A3S2Rocks : []prop_manipulator_device = array{} 
    @editable
    A3S2RockProps : []creative_prop = array{} 
    @editable
    A3S3Rocks : []prop_manipulator_device = array{} 
    @editable
    A3S3RockProps : []creative_prop = array{} 
    @editable
    A3S4Rocks : []prop_manipulator_device = array{} 
    @editable
    A3S4RockProps : []creative_prop = array{} 
    @editable
    A3S5Rocks : []prop_manipulator_device = array{} 
    @editable
    A3S5RockProps : []creative_prop = array{} 

    @editable
    A4S1Rocks : []prop_manipulator_device = array{} 
    @editable
    A4S1RockProps : []creative_prop = array{} 
    @editable
    A4S2Rocks : []prop_manipulator_device = array{} 
    @editable
    A4S2RockProps : []creative_prop = array{} 
    @editable
    A4S3Rocks : []prop_manipulator_device = array{} 
    @editable
    A4S3RockProps : []creative_prop = array{} 
    @editable
    A4S4Rocks : []prop_manipulator_device = array{} 
    @editable
    A4S4RockProps : []creative_prop = array{} 
    @editable
    A4S5Rocks : []prop_manipulator_device = array{} 
    @editable
    A4S5RockProps : []creative_prop = array{} 

    @editable
    A5S1Rocks : []prop_manipulator_device = array{} 
    @editable
    A5S1RockProps : []creative_prop = array{} 
    @editable
    A5S2Rocks : []prop_manipulator_device = array{} 
    @editable
    A5S2RockProps : []creative_prop = array{} 
    @editable
    A5S3Rocks : []prop_manipulator_device = array{} 
    @editable
    A5S3RockProps : []creative_prop = array{} 
    @editable
    A5S4Rocks : []prop_manipulator_device = array{} 
    @editable
    A5S4RockProps : []creative_prop = array{} 
    @editable
    A5S5Rocks : []prop_manipulator_device = array{} 
    @editable
    A5S5RockProps : []creative_prop = array{}


    var AllRocks : [][]prop_manipulator_device = array{}

    var AllRockProps : [][]creative_prop = array{}

    var EachRockHPs : [string]float = map{}

    var RocksDown : [string]logic = map{}

    var StageRockHPs : []float = array{
        5.00e03,5.00e05,7.50e07,7.50e09,5.63e11, #AREA 1
        3.12e14,6.24e16,1.87e19,3.75e21,5.62e23, #AREA 2
        6.24e27,1.87e29,8.43e32,2.53e34,5.69e40, #AREA 3
        1.12e53,1.81e61,2.71e69,4.37e78,6.68e88, #AREA 4
        1.12e114,1.81e123,2.71e132,4.37e141,6.68e150 #AREA 5
    }

    var DamageNumbers<private>: damage_numbers = damage_numbers:

    var MeteorRockHP : float = 3.09e110
    var NewMeteorRockHP : float = 3.03e167

    var WepMatAmount : float = 5000.0
    var PrestigeAmount : float = 40000.0
    var BossKeyAmount : float = 6000.0
    var EggMinAmount : float = 80000.0
    var EggMaxAmount : float = 120000.0

    var NewWepMatAmount : float = 500000.0
    var NewPrestigeAmount : float = 11500000.0
    var NewBossKeyAmount : float = 600000.0
    var NewEggMinAmount : float = 13000000.0
    var NewEggMaxAmount : float = 20110000.0

    # Runs when the device is started in a running game
    OnBegin<override>()<suspends>:void=
        set AllRocks = array:
            A1S1Rocks,A1S2Rocks,A1S3Rocks,A1S4Rocks,A1S5Rocks,
            A2S1Rocks,A2S2Rocks,A2S3Rocks,A2S4Rocks,A2S5Rocks,
            A3S1Rocks,A3S2Rocks,A3S3Rocks,A3S4Rocks,A3S5Rocks,
            A4S1Rocks,A4S2Rocks,A4S3Rocks,A4S4Rocks,A4S5Rocks,
            A5S1Rocks,A5S2Rocks,A5S3Rocks,A5S4Rocks,A5S5Rocks

         #Put each stage of rocks into array
        set AllRockProps = array:
            A1S1RockProps,A1S2RockProps,A1S3RockProps,A1S4RockProps,A1S5RockProps,
            A2S1RockProps,A2S2RockProps,A2S3RockProps,A2S4RockProps,A2S5RockProps,
            A3S1RockProps,A3S2RockProps,A3S3RockProps,A3S4RockProps,A3S5RockProps,
            A4S1RockProps,A4S2RockProps,A4S3RockProps,A4S4RockProps,A4S5RockProps,
            A5S1RockProps,A5S2RockProps,A5S3RockProps,A5S4RockProps,A5S5RockProps
        
        for(i:=0..MeteorInfoButtons.Length):
            if(MeteorInfoButton:=MeteorInfoButtons[i]):
                MeteorInfoButton.InteractedWithEvent.SubscribeAgent(ShowMeteorHelp, i)

        MainRockManipulator.DamagedEvent.Subscribe(OnMainRockHit)

        if(set EachRockHPs["Meteor"] = MeteorRockHP){}
        if(set RocksDown["Meteor"] = false){}
        MeteorRockManip.DamagedEvent.Subscribe(OnMeteorRockHit)

        if(set EachRockHPs["NewMeteor"] = NewMeteorRockHP){}
        if(set RocksDown["NewMeteor"] = false){}
        NewMeteorRockManip.DamagedEvent.Subscribe(OnNewMeteorRockHit)

        for(i:int=0..AllRocks.Length-1): #For each stage of rocks
            if(CurrentStage:=AllRocks[i]): #Get current stage we're on
                for(x:int=0..CurrentStage.Length-1): #For each rock in this stage
                    if(CurrentRock:=CurrentStage[x]): #Get current rock we're on
                        TempString := "S{i+1}R{x+1}"
                        RealRockHP := GetRockHP(CurrentStage, i, x)
                        if(set EachRockHPs[TempString] = RealRockHP){}
                        if(set RocksDown[TempString] = false){}
                        CurrentRock.DamagedEvent.SubscribeAgent(OnCaveRockHit, TempString) #Add Damage Event to that rock

        for(i:int=0..AllRocks.Length-1): #For each stage of rocks
            if(CurrentStage:=AllRocks[i]): #Get current stage we're on
                for(x:int=0..CurrentStage.Length-1): #For each rock in this stage
                    if(CurrentRock:=CurrentStage[x]): #Get current rock we're on   
                        TempString := "S{i+1}R{x+1}"         
                        if(i >= 14):
                            if(Third := Floor[(CurrentStage.Length*1.0) / 3.0]):
                                if(x < Third):
                                else if(x < Third * 2):
                                    if(i >= 20):
                                        spawn{ShowStars(TempString, 1, true)}
                                    else:
                                        spawn{ShowStars(TempString, 1, false)}
                                else:
                                    if(i >= 20):
                                        spawn{ShowStars(TempString, 2, true)}
                                    else:
                                        spawn{ShowStars(TempString, 2, false)}
                        Sleep(0.2)
        
    var MeteorHelpCD : float = 0.0
    ShowMeteorHelp(Agent:agent, WhichMeteor:int):void=
        if(CustomPlayer := gm.PlayersMap[Agent]):
            if(MeteorHelpCD <= 0.0):
                set MeteorHelpCD = 60.0
                spawn{MeteorHelpTimer()}
                for(K -> CPlayer:gm.PlayersMap):
                        if(WhichMeteor = 0):
                            gm.ShowMessage(CPlayer.Player, "Someone is requesting help at the Area 4 Meteor!", 0)
                        else:
                            gm.ShowMessage(CPlayer.Player, "Someone is requesting help at the Area 5 Meteor!", 0)

                gm.ShowMessage(CustomPlayer.Player, "Help Requested!", 0)
            else:
                gm.ShowMessage(CustomPlayer.Player, "Wait a bit before requesting help again", 1)

    MeteorHelpTimer()<suspends>:void=
        loop:
            if(MeteorHelpCD > 0.0):
                set MeteorHelpCD -= 1.0
                Sleep(1.0)
            else:
                set MeteorHelpCD = 0.0
                break:
                
    GetRockHP(CurrentStage:[]prop_manipulator_device, Stage:int, WhichRockNum:int):float=
        var StageRockHP : float = 0.0
        if(WhichRockNum <> 99 and WhichRockNum <> 999):
            if(NewHP:=StageRockHPs[Stage]):
                set StageRockHP = NewHP
            if(Stage >= 14):
                if(Third := Floor[(CurrentStage.Length*1.0) / 3.0]):
                    if(WhichRockNum < Third):
                        set StageRockHP /= 1000.0
                    else if(WhichRockNum < Third * 2):
                        #set StageRockHP
                    else:
                        set StageRockHP *= 1000.0
        else:
            if(WhichRockNum = 99):
                if(MeteorHP:=EachRockHPs["Meteor"]):
                    set StageRockHP = MeteorHP
            else if(WhichRockNum = 999):
                if(NewMeteorHP:=EachRockHPs["NewMeteor"]):
                    set StageRockHP = NewMeteorHP
        return StageRockHP

    GetCurrentRock(whichRock:string):prop_manipulator_device=
        if(whichRock = "Meteor"):
            return MeteorRockManip
        else if(whichRock = "NewMeteor"):
            return NewMeteorRockManip
        else:
            for(i:int=0..AllRocks.Length-1): #For each stage of rocks
                if(CurrentStage:=AllRocks[i]): #Get current stage we're on
                    for(x:int=0..CurrentStage.Length-1): #For each rock in this stage
                        if(CurrentRock:=CurrentStage[x]): #Get current rock we're on
                            TempString := "S{i+1}R{x+1}"
                            if(TempString = whichRock):
                                return CurrentRock
        return MainRockManipulator


    GetCurrentRockProp(whichRock:string):creative_prop=
        if(whichRock = "Meteor"):
            return MeteorRock
        else if (whichRock = "NewMeteor"):
            return NewMeteorRock
        else:
            for(i:int=0..AllRockProps.Length-1): #For each stage of rocks
                if(CurrentStage:=AllRockProps[i]): #Get current stage we're on
                    for(x:int=0..CurrentStage.Length-1): #For each rock in this stage
                        if(CurrentRock:=CurrentStage[x]): #Get current rock we're on
                            TempString := "S{i+1}R{x+1}"
                            if(TempString = whichRock):
                                return CurrentRock
        return TheRock

    ShowStars(whichRock:string, StarLevel:int, IsArea5:logic)<suspends>:void=
        NewRotation : rotation = IdentityRotation()
        NewScale : vector3 = vector3{X:=1.0, Y:=1.0, Z:=1.0}
        CurrentRockProp:=GetCurrentRockProp(whichRock)
        RockLocation:=CurrentRockProp.GetTransform().Translation
        var ZLoc : float = RockLocation.Z
        if(IsArea5 = false):
            set ZLoc += 150.0
        
        Sleep(0.1)
        if(StarLevel = 1):
            TwoStar := SpawnProp(TwoStarBP, transform{Translation:=vector3{X:=RockLocation.X,Y:=RockLocation.Y,Z:=ZLoc}, Rotation:=NewRotation, Scale:=NewScale})
        else:
            ThreeStar := SpawnProp(ThreeStarBP, transform{Translation:=vector3{X:=RockLocation.X,Y:=RockLocation.Y,Z:=ZLoc}, Rotation:=NewRotation, Scale:=NewScale})


    #Attacking the Main Rock outside the cave (Separate because it acts different, No HP & No Elim bonus, just your damage value)
    OnMainRockHit(Agent : agent):void=
        if(CustomPlayer := gm.PlayersMap[Agent]):
            IsHolding:=IsHoldingWeapon(Agent)
            if(IsHolding = false):
                Tutorial.UpdateTutorialUI(CustomPlayer, "HitMainRock", 1.0)
                var DamageResult : tuple(float, logic, float, int) = CustomPlayer.CalcTotalDamage("Stone")
                if(DamageResult(2) = 100.0):
                    AudioManager.ComboHit.Play(CustomPlayer.Player)
                else if(DamageResult(1) = true):
                    AudioManager.CritHit.Play(CustomPlayer.Player)
                #RockGranter.GrantItem(CustomPlayer.Player)
                if(CustomPlayer.CanGrantEXP = true, Accolade:=AudioManager.Accolades[10]): #AWARD PLAYER EXP FOR HITTING QUARRY ROCK
                    set CustomPlayer.CanGrantEXP = false
                    Accolade.Award(CustomPlayer.Player)
                if(Analytic:=AudioManager.GeneralAnalytics[2]): #ANALYTIC
                    Analytic.Submit(CustomPlayer.Player)
                DamageNumbers.ShowDamageUIForPlayer(CustomPlayer, DamageResult(0), DamageResult(1), DamageResult(3), false, "Rock", DamageResult(2))
                CustomPlayer.Collect(DamageResult(0), "Stone", 0)
                
                var slimit : int = 39
                if(CustomPlayer.A4Unlocked = true):
                    set slimit = 79
                if(CustomPlayer.A5Unlocked = true):
                    set slimit = 109
                if:
                    CustomPlayer.PUPQOLLevels[0] = 2
                    CustomPlayer.AutoUpgradeOnHit = true
                    CustomPlayer.StrengthEfficiencies[slimit] <> 7
                then:
                    spawn{SUP.AutoUpgrade(CustomPlayer)}
                #SaveAllData(CustomPlayer)

    #Attacking the Meteor Rock in A4S3
    OnMeteorRockHit(Agent : agent):void=
        var HittingMeteorMult : float = 1.0
        if(CustomPlayer := gm.PlayersMap[Agent]):
            IsHolding:=IsHoldingWeapon(Agent)
            if(IsHolding = false):
                var DamageResult : tuple(float, logic, float, int) = CustomPlayer.CalcTotalDamage("Stone")
                for(K -> CPlayer:gm.PlayersMap):
                    if:
                        CPlayer.CurrentRockHit = "Meteor"
                    then:
                        set HittingMeteorMult += 0.15
                TotalDamage := DamageResult(0) * HittingMeteorMult
                if(DamageResult(2) = 100.0):
                    AudioManager.ComboHit.Play(CustomPlayer.Player)
                else if(DamageResult(1) = true):
                    AudioManager.CritHit.Play(CustomPlayer.Player)
                    
                if(CustomPlayer.CurrentRockHit <> "Meteor"):
                    set CustomPlayer.CurrentRockHit = "Meteor"
                    HPBars.Init(CustomPlayer, "Meteor Rock", MeteorRockHP)
                    HPBars.ShowUIForPlayer(CustomPlayer)

                CustomPlayer.Collect(TotalDamage, "Stone", 0)
                DamageNumbers.ShowDamageUIForPlayer(CustomPlayer, TotalDamage, DamageResult(1), DamageResult(3), false, "Rock", DamageResult(2))
                if(set EachRockHPs["Meteor"] -= TotalDamage){} #Reduce that Rocks HP
                           
                var slimit : int = 39
                if(CustomPlayer.A4Unlocked = true):
                    set slimit = 79
                if(CustomPlayer.A5Unlocked = true):
                    set slimit = 109
                if:
                    CustomPlayer.PUPQOLLevels[0] = 2
                    CustomPlayer.AutoUpgradeOnHit = true
                    CustomPlayer.StrengthEfficiencies[slimit] <> 7
                then:
                    spawn{SUP.AutoUpgrade(CustomPlayer)}

                if (EachRockHPs["Meteor"] <= 0.0): #ROCK DESTROYED
                    if(set EachRockHPs["Meteor"] = 0.0) {}
                    for(K -> CPlayer:gm.PlayersMap):
                        if:
                            CPlayer.CurrentRockHit = "Meteor"
                        then:
                            HPBars.TakeDamage(CPlayer, 0.0)
                            spawn{MaybeRemoveHPBar(CPlayer, CPlayer.CurrentRockHit)}
                            if(Accolade:=AudioManager.Accolades[8]): #AWARD PLAYER EXP FOR DESTROYING BIG BOI ROCK
                                Accolade.Award(CPlayer.Player) 
                            if(Analytic:=AudioManager.GeneralAnalytics[7]): #ANALYTIC
                                Analytic.Submit(CPlayer.Player)
                            if(CPlayer.OneTimeAnalytics["DestroyMeteor"] = false):
                                if(set CPlayer.OneTimeAnalytics["DestroyMeteor"] = true) {}
                                if(Analytic:=AudioManager.OneTimeAnalytics[7]): #ANALYTIC
                                    Analytic.Submit(CPlayer.Player)
                            spawn{GrantMeteorRewards(CPlayer)}
                            if (RockQuestStatus := CPlayer.QuestsStarted[0], RockQuestStatus?):
                                set CPlayer.QuestRocksDestroyed += 1
                                DRM.UpdateCounterUI(CustomPlayer)
                            SaveAllData(CPlayer)
                    spawn{MeteorDestroyed()}
                else:
                    if(NewRockHP:=EachRockHPs["Meteor"]):
                        for(K -> CPlayer:gm.PlayersMap):
                            if:
                                CPlayer.CurrentRockHit = "Meteor"
                            then:
                                HPBars.TakeDamage(CPlayer, NewRockHP)
                
                if(CustomPlayer.EarthquakeUnlocked = true and CustomPlayer.HutEnabled[2] = true):
                    spawn{RockShockwave(CustomPlayer, 16, "Meteor")}
                if(CustomPlayer.VoidUnlocked = true and CustomPlayer.HutEnabled[3] = true):
                    spawn{StartVoidFracture(CustomPlayer, "Meteor", 16)}
                SaveAllData(CustomPlayer)

    OnNewMeteorRockHit(Agent : agent):void=
        var HittingMeteorMult : float = 1.0
        var ActualDamage : float = 0.0
        if(CustomPlayer := gm.PlayersMap[Agent]):
            if (CustomPlayer.StagesUnlocked >= 24):
                IsHolding:=IsHoldingWeapon(Agent)
                if(IsHolding = false):
                    var DamageResult : tuple(float, logic, float, int) = CustomPlayer.CalcTotalDamage("Stone")
                    for(K -> CPlayer:gm.PlayersMap):
                        if:
                            CPlayer.CurrentRockHit = "NewMeteor"
                        then:
                            set HittingMeteorMult += 0.15
                    TotalDamage := DamageResult(0) * HittingMeteorMult
                    if(DamageResult(2) = 100.0):
                        AudioManager.ComboHit.Play(CustomPlayer.Player)
                    else if(DamageResult(1) = true):
                        AudioManager.CritHit.Play(CustomPlayer.Player)
                        
                    if(CustomPlayer.CurrentRockHit <> "NewMeteor"):
                        set CustomPlayer.CurrentRockHit = "NewMeteor"
                        HPBars.Init(CustomPlayer, "Blazing Meteor", NewMeteorRockHP)
                        HPBars.ShowUIForPlayer(CustomPlayer)

                    CustomPlayer.Collect(TotalDamage, "Stone", 0)
                    if(set ActualDamage += Min(TotalDamage, EachRockHPs["NewMeteor"])){}
                    DamageNumbers.ShowDamageUIForPlayer(CustomPlayer, TotalDamage, DamageResult(1), DamageResult(3), false, "Rock", DamageResult(2))
                    if(set EachRockHPs["NewMeteor"] -= TotalDamage){} #Reduce that Rocks HP
                    set CustomPlayer.DamageToNewMeteor += ActualDamage
                            
                    var slimit : int = 39
                    if(CustomPlayer.A4Unlocked = true):
                        set slimit = 79
                    if(CustomPlayer.A5Unlocked = true):
                        set slimit = 109
                    if:
                        CustomPlayer.PUPQOLLevels[0] = 2
                        CustomPlayer.AutoUpgradeOnHit = true
                        CustomPlayer.StrengthEfficiencies[slimit] <> 7
                    then:
                        spawn{SUP.AutoUpgrade(CustomPlayer)}

                    if (EachRockHPs["NewMeteor"] <= 0.0): #ROCK DESTROYED
                        if(set EachRockHPs["NewMeteor"] = 0.0) {}
                        for(K -> CPlayer:gm.PlayersMap):
                            if:
                                CPlayer.CurrentRockHit = "NewMeteor"
                            then:
                                HPBars.TakeDamage(CPlayer, 0.0)
                                spawn{MaybeRemoveHPBar(CPlayer, CPlayer.CurrentRockHit)}
                                if(Accolade:=AudioManager.Accolades[8]): #AWARD PLAYER EXP FOR DESTROYING BIG BOI ROCK
                                    Accolade.Award(CPlayer.Player) 
                                if(Analytic:=AudioManager.GeneralAnalytics[12]): #ANALYTIC
                                    Analytic.Submit(CPlayer.Player)
                                if(CustomPlayer.OneTimeAnalytics["DestroyNewMeteor"] = false):
                                    if(set CustomPlayer.OneTimeAnalytics["DestroyNewMeteor"] = true) {}
                                    if(Analytic:=AudioManager.OneTimeAnalytics[13]): #ANALYTIC
                                        Analytic.Submit(CustomPlayer.Player)
                                spawn{GrantNewMeteorRewards(CPlayer)}
                                if (RockQuestStatus := CPlayer.QuestsStarted[0], RockQuestStatus?):
                                    set CPlayer.QuestRocksDestroyed += 1
                                    DRM.UpdateCounterUI(CustomPlayer)
                                SaveAllData(CPlayer)
                        spawn{NewMeteorDestroyed()}
                    else:
                        if(NewRockHP:=EachRockHPs["NewMeteor"]):
                            for(K -> CPlayer:gm.PlayersMap):
                                if:
                                    CPlayer.CurrentRockHit = "NewMeteor"
                                then:
                                    HPBars.TakeDamage(CPlayer, NewRockHP)
                    
                    if(CustomPlayer.EarthquakeUnlocked = true and CustomPlayer.HutEnabled[2] = true):
                        spawn{RockShockwave(CustomPlayer, 22, "NewMeteor")}
                    if(CustomPlayer.VoidUnlocked = true and CustomPlayer.HutEnabled[3] = true):
                        spawn{StartVoidFracture(CustomPlayer, "NewMeteor", 22)}
                    SaveAllData(CustomPlayer)
            else:
                gm.ShowMessage(CustomPlayer.Player, "You must have A5S5 unlocked!", 1)

    GrantMeteorRewards(CustomPlayer:Custom_Player)<suspends>:void=
        if(CustomPlayer.ClaimingRewards = false):
            set CustomPlayer.ClaimingRewards = true
            for(i:int=0..3): #LOOP 4 TIMES TO BREAK UP REWARDS TO FEEL BETTER
                StoneReward := ((CustomPlayer.GetPlayerDamage() * 10.0) / 4.0)# * MeteorRewardMult
                CustomPlayer.Collect(StoneReward, "Stone", 0)
                DamageNumbers.ShowDamageUIForPlayer(CustomPlayer, StoneReward, false, 0, false, "Rock", -1.0)
                Sleep(0.3)

                WepMatReward := (WepMatAmount / 4.0)# * MeteorRewardMult
                CustomPlayer.Collect(WepMatReward, "WepMat", 0)
                DamageNumbers.ShowDamageUIForPlayer(CustomPlayer, WepMatReward, false, 0, false, "WepMat", -1.0)
                #WepMatGranter.GrantItem(CustomPlayer.Player)
                Sleep(0.3)

                var whichEgg : int = 0
                if(whichDropTable:=gm.EggDropOdds[19],DropAmount := Floor[GetRandomFloat(EggMinAmount, EggMaxAmount)]*1.0):
                    set whichEgg = SelectItem(whichDropTable)
                    EggReward := (DropAmount / 4.0)# * MeteorRewardMult
                    CustomPlayer.Collect(EggReward, "Egg", whichEgg)
                    DamageNumbers.ShowDamageUIForPlayer(CustomPlayer, EggReward, false, 0, false, "Egg{whichEgg}", -1.0)
                    Sleep(0.3)

                PrestigeReward := (PrestigeAmount / 4.0)
                CustomPlayer.Collect(PrestigeReward, "Shard", 0)
                DamageNumbers.ShowDamageUIForPlayer(CustomPlayer, PrestigeReward, false, 0, false, "Shard", -1.0)
                #PrestigeGranter.GrantItem(CustomPlayer.Player)
                Sleep(0.3)
                
                BossKeyReward := (BossKeyAmount / 4.0)
                CustomPlayer.Collect(BossKeyReward, "BossKey", 0)
                DamageNumbers.ShowDamageUIForPlayer(CustomPlayer, BossKeyReward, false, 0, false, "BossKey", -1.0)
                #PrestigeGranter.GrantItem(CustomPlayer.Player)
                Sleep(0.3)
            set CustomPlayer.ClaimingRewards = false
            SaveAllData(CustomPlayer)

    GrantNewMeteorRewards(CustomPlayer:Custom_Player)<suspends>:void=
        <# var PercentDone : float = 0.0
        set PercentDone = CustomPlayer.DamageToNewMeteor/NewMeteorRockHP
        if(PercentDone >= 0.05): #>
        if(CustomPlayer.ClaimingRewards = false):
            set CustomPlayer.ClaimingRewards = true
            for(i:int=0..3): #LOOP 4 TIMES TO BREAK UP REWARDS TO FEEL BETTER
                StoneReward := (CustomPlayer.GetPlayerDamage() * 1000.0) / 4.0
                CustomPlayer.Collect(StoneReward, "Stone", 0)
                DamageNumbers.ShowDamageUIForPlayer(CustomPlayer, StoneReward, false, 0, false, "Rock", -1.0)
                Sleep(0.3)

                WepMatReward := NewWepMatAmount / 4.0
                CustomPlayer.Collect(WepMatReward, "WepMat", 0)
                DamageNumbers.ShowDamageUIForPlayer(CustomPlayer, WepMatReward, false, 0, false, "WepMat", -1.0)
                #WepMatGranter.GrantItem(CustomPlayer.Player)
                Sleep(0.3)

                var whichEgg : int = 0
                if(whichDropTable:=gm.EggDropOdds[20],DropAmount := Floor[GetRandomFloat(NewEggMinAmount, NewEggMaxAmount)]*1.0):
                    set whichEgg = SelectItem(whichDropTable)
                    EggReward := ((DropAmount) / 4.0) 
                    CustomPlayer.Collect(EggReward, "Egg", whichEgg)
                    DamageNumbers.ShowDamageUIForPlayer(CustomPlayer, EggReward, false, 0, false, "Egg{whichEgg}", -1.0)
                    Sleep(0.3)

                PrestigeReward := NewPrestigeAmount / 4.0
                CustomPlayer.Collect(PrestigeReward, "Shard", 0)
                DamageNumbers.ShowDamageUIForPlayer(CustomPlayer, PrestigeReward, false, 0, false, "Shard", -1.0)
                #PrestigeGranter.GrantItem(CustomPlayer.Player)
                Sleep(0.3)
                
                BossKeyReward := NewBossKeyAmount / 4.0
                CustomPlayer.Collect(BossKeyReward, "BossKey", 0)
                DamageNumbers.ShowDamageUIForPlayer(CustomPlayer, BossKeyReward, false, 0, false, "BossKey", -1.0)
                #PrestigeGranter.GrantItem(CustomPlayer.Player)
                Sleep(0.3)
            set CustomPlayer.ClaimingRewards = false
            set CustomPlayer.DamageToNewMeteor = 0.0
            SaveAllData(CustomPlayer)
        <# else:
            gm.ShowMessage(CustomPlayer.Player, "You must do at least 5% total HP for rewards!", 1) #>

    OnCaveRockHit(Agent : agent, whichRock:string):void=
        var whichStage : int = 0
        var currentStage : string = ""
        if(CustomPlayer := gm.PlayersMap[Agent], rockStage1:=whichRock[1], rockStage2:=whichRock[2]):
            IsHolding:=IsHoldingWeapon(Agent)
            if(IsHolding = false):
                if(ToString(rockStage2) = "R"):
                    set currentStage = "{rockStage1}"
                else:
                    set currentStage = "{rockStage1}{rockStage2}"
                rockStage:=ToString(currentStage)
                set whichStage = StringToInt(rockStage)
                WhichRockNum := GetWhichRock(whichRock)
                if(CustomPlayer.CurrentRockHit <> whichRock, CurrentStage := AllRocks[whichStage-1]):
                    set CustomPlayer.CurrentRockHit = whichRock
                    RealRockHP := GetRockHP(CurrentStage, whichStage-1, WhichRockNum)
                    HPBars.Init(CustomPlayer, "Lv. {whichStage} Stone", RealRockHP)
                    HPBars.ShowUIForPlayer(CustomPlayer)
                spawn{OnRockDamaged(CustomPlayer, whichRock, whichStage-1)}
                if(CustomPlayer.EarthquakeUnlocked = true and CustomPlayer.HutEnabled[2] = true):
                    spawn{RockShockwave(CustomPlayer, whichStage-1, whichRock)}
                if(CustomPlayer.VoidUnlocked = true and CustomPlayer.HutEnabled[3] = true):
                    spawn{StartVoidFracture(CustomPlayer, whichRock, whichStage-1)}

    OnRockDamaged(CustomPlayer : Custom_Player, whichRock:string, whichStage:int)<suspends>: void=
        var DamageResult : tuple(float, logic, float, int) = CustomPlayer.CalcTotalDamage("Stone")
        if(CustomPlayer.ComboCount >= 100.0):
            AudioManager.ComboHit.Play(CustomPlayer.Player)
        else if(DamageResult(1) = true):
            AudioManager.CritHit.Play(CustomPlayer.Player)

        if(CurrentRockHP:=EachRockHPs[whichRock],CurrentStage := AllRocks[whichStage]):
            WhichRockNum := GetWhichRock(whichRock)
            StageRockHP := GetRockHP(CurrentStage, whichStage, WhichRockNum)
            if(CurrentRockHP >= DamageResult(0)):
                CustomPlayer.Collect(DamageResult(0), "Stone", 0)
                DamageNumbers.ShowDamageUIForPlayer(CustomPlayer, DamageResult(0), DamageResult(1), DamageResult(3), false, "Rock", DamageResult(2))
            else: #If final stike
                var PetFinalHitBonus : float = 0.0
                for(i:int=0..CustomPlayer.PetSlots.Length):
                    if(CustomPlayer.PetSlots[i] <> -1):
                        if:
                            PetLevel := CustomPlayer.PetLevels[CustomPlayer.PetSlots[i]]
                            FinalHit := CustomPlayer.PetFinalStrike[CustomPlayer.PetSlots[i]]
                        then:
                            set PetFinalHitBonus += FinalHit/100.0

                TotalReward := DamageResult(0) + (StageRockHP*CustomPlayer.FinalHitBonus) + (StageRockHP*PetFinalHitBonus)
                CustomPlayer.Collect(TotalReward, "Stone", 0)
                DamageNumbers.ShowDamageUIForPlayer(CustomPlayer, TotalReward, DamageResult(1), DamageResult(3), true, "Rock", DamageResult(2))

            if(set EachRockHPs[whichRock] -= DamageResult(0)){} #Reduce that Rocks HP
            
            var slimit : int = 39
            if(CustomPlayer.A4Unlocked = true):
                set slimit = 79
            if(CustomPlayer.A5Unlocked = true):
                set slimit = 109
            if:
                CustomPlayer.PUPQOLLevels[0] = 2
                CustomPlayer.AutoUpgradeOnHit = true
                CustomPlayer.StrengthEfficiencies[slimit] <> 7
            then:
                spawn{SUP.AutoUpgrade(CustomPlayer)}

            #RockGranter.GrantItem(CustomPlayer.Player)
            if (EachRockHPs[whichRock] <= 0.0): #ROCK DESTROYED
                for(K -> CPlayer:gm.PlayersMap):
                    if:
                        CPlayer.CurrentRockHit = whichRock
                    then:
                        HPBars.TakeDamage(CPlayer, 0.0)
                        spawn{MaybeRemoveHPBar(CPlayer, CPlayer.CurrentRockHit)}
                        AudioManager.RockDestroyed.Play(CPlayer.Player)
                        Tutorial.UpdateTutorialUI(CPlayer, "RockDestroyed", 1.0)
                        if(Accolade:=AudioManager.Accolades[0]): #AWARD PLAYER EXP
                            Accolade.Award(CPlayer.Player)
                        CPlayer.UpAchValue(4, 1.0)
                        spawn{DoesEggDrop(CPlayer, whichStage, "Stone")}

                if(Analytic:=AudioManager.GeneralAnalytics[1]): #ANALYTIC
                    Analytic.Submit(CustomPlayer.Player)
                spawn{RockDestroyed(CustomPlayer, whichRock, whichStage)}
                Sleep(0.2) #TO OFFSET FLOATING NUMBERS
            else:
                if(NewRockHP:=EachRockHPs[whichRock]): #Need this line to make sure you're updating the new HP after damage
                    for(K -> CPlayer:gm.PlayersMap):
                        if:
                            CPlayer.CurrentRockHit = whichRock
                        then:
                            HPBars.TakeDamage(CPlayer, NewRockHP)
                    spawn{MaybeResetRockHP(CustomPlayer, whichRock, NewRockHP, whichStage, false)}

    MaybeResetRockHP(CustomPlayer:Custom_Player, ThisRock:string, RockHP:float, whichStage:int, IsOther:logic)<suspends>:void=
        Sleep(5.0)
        if(CurrentRockHP:=EachRockHPs[ThisRock]): #FOR PERSONAL HP BAR RESET
            if(CurrentRockHP = RockHP and CustomPlayer.CurrentRockHit = ThisRock and CustomPlayer.QueuedForBoss = false):
                CustomPlayer.RemoveRockHealthBarUI()
                set CustomPlayer.CurrentRockHit = ""    
        Sleep(295.0) #CHANGE TO 295.0
        if:
            ThisRock <> "Meteor"
            ThisRock <> "NewMeteor"
            NewCurrentRockHP:=EachRockHPs[ThisRock]
            NewCurrentRockHP = RockHP
        then:
            SetRockHP(ThisRock, whichStage)


    GetWhichRock(whichRock:string):int=
        var CurrentRock : string = ""
        if(whichRock <> "Meteor" and whichRock <> "NewMeteor"):
            if:
                rockNum1:=whichRock[whichRock.Length-2]
                rockNum2:=whichRock[whichRock.Length-1]
            then:
                if(ToString(rockNum1) = "R"):
                    set CurrentRock = "{rockNum2}"
                else:
                    set CurrentRock = "{rockNum1}{rockNum2}"
                rockNum:=ToString(CurrentRock)
                return StringToInt(rockNum) - 1
        else:
            if (whichRock = "Meteor"):
                return 99
            if (whichRock = "NewMeteor"):
                return 999
        return 100

    SetRockHP(whichRock:string, whichStage:int):void=
        if:
            CurrentStage := AllRocks[whichStage]
        then:
            WhichRockNum := GetWhichRock(whichRock)
            RealRockHP := GetRockHP(CurrentStage, whichStage, WhichRockNum)
            if(set EachRockHPs[whichRock] = RealRockHP){}

    MaybeRemoveHPBar(CustomPlayer:Custom_Player, ThisRock:string)<suspends>:void=
        Sleep(2.0)
        if(CustomPlayer.CurrentRockHit = ThisRock and CustomPlayer.QueuedForBoss = false):
            CustomPlayer.RemoveRockHealthBarUI()
            set CustomPlayer.CurrentRockHit = ""

    RockDestroyed(CustomPlayer:Custom_Player, whichRock:string, WhichStage:int)<suspends>:void=
        if (RockQuestStatus := CustomPlayer.QuestsStarted[0], RockQuestStatus?):
            set CustomPlayer.QuestRocksDestroyed += 1
            DRM.UpdateCounterUI(CustomPlayer)
        SaveAllData(CustomPlayer)
        var IsRockDown : logic = false
        var PlayerLoc : vector3 = vector3{X:=0.0, Y:=0.0, Z:=0.0}
        NewRotation : rotation = IdentityRotation()
        NewScale : vector3 = vector3{X:=1.0, Y:=1.0, Z:=1.0}
        if(RocksDown[whichRock] = true):
            set IsRockDown = true
        if(IsRockDown = false):    
            if(set RocksDown[whichRock] = true) {}
            CurrentRock:=GetCurrentRock(whichRock)
            CurrentRockProp:=GetCurrentRockProp(whichRock)
            RockLocation:=CurrentRockProp.GetTransform().Translation
            RockRotation:=CurrentRockProp.GetTransform().Rotation
            CurrentRock.Disable()
            if(FC := CustomPlayer.Player.GetFortCharacter[]):
                set PlayerLoc = FC.GetTransform().Translation
            Rubble := SpawnProp(RockRubble, transform{Translation:=vector3{X:=RockLocation.X,Y:=RockLocation.Y,Z:=PlayerLoc.Z-95.0}, Rotation:=NewRotation, Scale:=NewScale})
            CurrentRockProp.MoveTo(vector3{X:=RockLocation.X,Y:=RockLocation.Y,Z:=RockLocation.Z-800.0}, RockRotation, 1.75)
            PropInst := Rubble(0)
            if ( Prop := PropInst?):
                Prop.Dispose()
            Sleep(5.0)
            CurrentRockProp.MoveTo(vector3{X:=RockLocation.X,Y:=RockLocation.Y,Z:=RockLocation.Z}, RockRotation, 1.5)
            #Sleep(0.5)
            CurrentRock.Enable()
            SetRockHP(whichRock, WhichStage)
            if(set RocksDown[whichRock] = false) {}

    MeteorDestroyed()<suspends>:void=
        var IsRockDown : logic = false
        NewRotation : rotation = IdentityRotation()
        NewScale : vector3 = vector3{X:=1.0, Y:=1.0, Z:=1.0}
        if(RocksDown["Meteor"] = true):
            set IsRockDown = true
        if(IsRockDown = false):    
            if(set RocksDown["Meteor"] = true) {}
            CurrentRock:=GetCurrentRock("Meteor")
            CurrentRockProp:=GetCurrentRockProp("Meteor")
            RockLocation:=CurrentRockProp.GetTransform().Translation
            RockRotation:=CurrentRockProp.GetTransform().Rotation
            CurrentRock.Disable()
            CurrentRockProp.MoveTo(vector3{X:=RockLocation.X,Y:=RockLocation.Y,Z:=RockLocation.Z-2770.0}, RockRotation, 15.0)
            
            Sleep(25.0)
            CurrentRockProp.MoveTo(vector3{X:=RockLocation.X,Y:=RockLocation.Y,Z:=RockLocation.Z}, RockRotation, 4.0)
            CurrentRock.Enable()
            if(set EachRockHPs["Meteor"] = MeteorRockHP){} #Set it back to Full HP
            if(set RocksDown["Meteor"] = false) {}

    NewMeteorDestroyed()<suspends>:void=
        var IsRockDown : logic = false
        NewRotation : rotation = IdentityRotation()
        NewScale : vector3 = vector3{X:=1.0, Y:=1.0, Z:=1.0}
        if(RocksDown["NewMeteor"] = true):
            set IsRockDown = true
        if(IsRockDown = false):    
            if(set RocksDown["NewMeteor"] = true) {}
            CurrentRock:=GetCurrentRock("NewMeteor")
            CurrentRockProp:=GetCurrentRockProp("NewMeteor")
            RockLocation:=CurrentRockProp.GetTransform().Translation
            RockRotation:=CurrentRockProp.GetTransform().Rotation
            CurrentRock.Disable()
            CurrentRockProp.MoveTo(vector3{X:=RockLocation.X,Y:=RockLocation.Y,Z:=RockLocation.Z-2770.0}, RockRotation, 15.0)
            
            Sleep(25.0)
            CurrentRockProp.MoveTo(vector3{X:=RockLocation.X,Y:=RockLocation.Y,Z:=RockLocation.Z}, RockRotation, 4.0)
            CurrentRock.Enable()
            if(set EachRockHPs["NewMeteor"] = NewMeteorRockHP){} #Set it back to Full HP
            if(set RocksDown["NewMeteor"] = false) {}

    RockShockwave(CustomPlayer:Custom_Player, whichStage:int, WhichRock:string)<suspends>:void=
        var  CheckRange : float = CustomPlayer.EarthquakeDistance
        if(CurrentStage:=AllRockProps[whichStage],FC := CustomPlayer.Player.GetFortCharacter[]): #Get current stage we're on
            PlayerLoc:=FC.GetTransform().Translation
            spawn{ShowEarthquake(CustomPlayer,CustomPlayer.EarthquakeDistance,PlayerLoc)}
            if(WhichRock = "Meteor" and Distance(PlayerLoc, MeteorRock.GetTransform().Translation) < CheckRange * 3.0):
                Sleep(0.2)
                CauseEarthquake(CustomPlayer, "Meteor", whichStage)
            else if(WhichRock = "NewMeteor" and Distance(PlayerLoc, NewMeteorRock.GetTransform().Translation) < CheckRange * 3.0):
                Sleep(0.2)
                CauseEarthquake(CustomPlayer, "NewMeteor", whichStage)
            else:
                for(x:int=0..CurrentStage.Length-1): #For each rock in this stage
                    if(CurrentRock:=CurrentStage[x]): #Get current rock we're on
                        if(DistanceDiff := Distance(PlayerLoc, CurrentRock.GetTransform().Translation) < CheckRange):
                            TempString := "S{whichStage+1}R{x+1}"
                            Sleep(0.2)
                            CauseEarthquake(CustomPlayer, TempString, whichStage)

    CauseEarthquake(CustomPlayer : Custom_Player, whichRock:string, whichStage:int)<suspends>:void=
        var DamageResult : tuple(float, logic, float, int) = CustomPlayer.CalcTotalDamage("Earthquake")
        var TotalDamage : float = DamageResult(0) * CustomPlayer.EarthquakeDamage
        if(CurrentRockHP:=EachRockHPs[whichRock],CurrentStage := AllRocks[whichStage], CurrentRockHP > 0.0):
            WhichRockNum := GetWhichRock(whichRock)
            StageRockHP := GetRockHP(CurrentStage, whichStage, WhichRockNum)

            if(CurrentRockHP > TotalDamage):
                CustomPlayer.Collect(TotalDamage * CustomPlayer.EarthquakeGatherAmount, "Stone", 0)
                DamageNumbers.ShowDamageUIForPlayer(CustomPlayer, TotalDamage * CustomPlayer.EarthquakeGatherAmount, DamageResult(1), DamageResult(3), false, "Earthquake", DamageResult(2))
                if(whichRock = "NewMeteor"):
                    set CustomPlayer.DamageToNewMeteor += TotalDamage * CustomPlayer.EarthquakeGatherAmount
            else:
                if(whichRock <> "Meteor" and whichRock <> "NewMeteor"):
                    var PetFinalHitBonus : float = 0.0
                    for(i:int=0..CustomPlayer.PetSlots.Length):
                        if(CustomPlayer.PetSlots[i] <> -1):
                            if:
                                PetLevel := CustomPlayer.PetLevels[CustomPlayer.PetSlots[i]]
                                FinalHit := CustomPlayer.PetFinalStrike[CustomPlayer.PetSlots[i]]
                            then:
                                set PetFinalHitBonus += FinalHit/100.0
                    if(whichRock = CustomPlayer.CurrentRockHit):
                        TotalReward := DamageResult(0) + (StageRockHP*CustomPlayer.FinalHitBonus) + (StageRockHP*PetFinalHitBonus)
                        CustomPlayer.Collect(TotalReward, "Stone", 0)
                        DamageNumbers.ShowDamageUIForPlayer(CustomPlayer, TotalReward, DamageResult(1), DamageResult(3), true, "Earthquake", DamageResult(2))
                    else:
                        TotalReward := TotalDamage + (StageRockHP*CustomPlayer.FinalHitBonus) + (StageRockHP*PetFinalHitBonus)
                        CustomPlayer.Collect(TotalReward * CustomPlayer.EarthquakeGatherAmount, "Stone", 0)
                        DamageNumbers.ShowDamageUIForPlayer(CustomPlayer, TotalReward * CustomPlayer.EarthquakeGatherAmount, DamageResult(1), DamageResult(3), true, "Earthquake", DamageResult(2))

                    spawn{RockDestroyed(CustomPlayer, whichRock, whichStage)}
                else:
                    if (whichRock = "Meteor"):
                        spawn{MeteorDestroyed()}
                    if (whichRock = "NewMeteor"):
                        spawn{NewMeteorDestroyed()}
                
            if(set EachRockHPs[whichRock] -= TotalDamage){} #Reduce that Rocks HP
            if (RducedHP:=EachRockHPs[whichRock]):

            if (EachRockHPs[whichRock] <= 0.0):
                if(Accolade:=AudioManager.Accolades[0]): #AWARD PLAYER EXP
                    Accolade.Award(CustomPlayer.Player)

            #RockGranter.GrantItem(CustomPlayer.Player)
            if(CustomPlayer.CurrentRockHit = whichRock):
                if (EachRockHPs[whichRock] <= 0.0):
                    for(K -> CPlayer:gm.PlayersMap):
                        if:
                            CPlayer.CurrentRockHit = whichRock
                        then:
                            HPBars.TakeDamage(CPlayer, 0.0)
                            CPlayer.UpAchValue(4, 1.0)
                            spawn{MaybeRemoveHPBar(CPlayer, CPlayer.CurrentRockHit)}
                            if(whichRock = "Meteor"):
                                spawn{GrantMeteorRewards(CPlayer)}
                            else if(whichRock = "NewMeteor"):
                                spawn{GrantNewMeteorRewards(CPlayer)}
                            else:
                                DoesEggDrop(CPlayer, whichStage, "Stone")
                else:
                    if(NewRockHP:=EachRockHPs[whichRock]): #Need this line to make sure you're updating the new HP after damage
                        for(K -> CPlayer:gm.PlayersMap):
                            if:
                                CPlayer.CurrentRockHit = whichRock
                            then:
                                HPBars.TakeDamage(CPlayer, NewRockHP)
                        
            if(NewRockHP:=EachRockHPs[whichRock]):
                spawn{MaybeResetRockHP(CustomPlayer, whichRock, NewRockHP, whichStage, true)}


    ShowEarthquake(CustomPlayer:Custom_Player, EQDistance:float, FortLoc:vector3)<suspends>:void=
        AudioManager.Earthquake.Play(CustomPlayer.Player)
        if(CustomPlayer.ShowVFX = true):
            NewRotation : rotation = IdentityRotation()
            NewScale : vector3 = vector3{X:=1.0, Y:=1.0, Z:=1.0}
            if(CustomPlayer.EarthquakeShowing = false):
                set CustomPlayer.EarthquakeShowing = true
                if(EQDistance >= 1400.0):
                    Earthquake := SpawnProp(Earthquake3, transform{Translation:=vector3{X:=FortLoc.X,Y:=FortLoc.Y,Z:=FortLoc.Z-75.0}, Rotation:=NewRotation, Scale:=NewScale})
                    Sleep(1.5)
                    PropInst := Earthquake(0)
                    if ( Prop := PropInst?):
                        Prop.Dispose()

                else if(EQDistance >= 1000.0):
                    Earthquake := SpawnProp(Earthquake2, transform{Translation:=vector3{X:=FortLoc.X,Y:=FortLoc.Y,Z:=FortLoc.Z-75.0}, Rotation:=NewRotation, Scale:=NewScale})
                    Sleep(1.5)
                    PropInst := Earthquake(0)
                    if ( Prop := PropInst?):
                        Prop.Dispose()

                else:
                    Earthquake := SpawnProp(Earthquake1, transform{Translation:=vector3{X:=FortLoc.X,Y:=FortLoc.Y,Z:=FortLoc.Z-75.0}, Rotation:=NewRotation, Scale:=NewScale})
                    Sleep(1.5)
                    PropInst := Earthquake(0)
                    if ( Prop := PropInst?):
                        Prop.Dispose()
                set CustomPlayer.EarthquakeShowing = false
                

    StartVoidFracture(CustomPlayer:Custom_Player, WhichRock:string, WhichStage:int)<suspends>:void=
        CurrentRockProp:=GetCurrentRockProp(WhichRock)
        RockLocation:=CurrentRockProp.GetTransform().Translation
        var TotalDamage : float = 0.0
        if(CustomPlayer.VoidActiveStacks < CustomPlayer.VoidStacks):
            set CustomPlayer.VoidActiveStacks += 1.0
            if(WhichRock <> "Meteor" and WhichRock <> "NewMeteor"):
                spawn{ShowVoidFracture(CustomPlayer, WhichRock, 5.0, RockLocation)}
            loop:
                if(CustomPlayer.VoidActiveTimer <= 5.0 and EachRockHPs[WhichRock] > 0.0):
                    var DamageResult : tuple(float, logic, float, int) = CustomPlayer.CalcTotalDamage("Void")
                    if(VoidDamage:= 2.0 + (2.0 * CustomPlayer.Hut4Levels[1])):
                        set TotalDamage = (DamageResult(0) * VoidDamage) / 2.0
                    if(CurrentRockHP:=EachRockHPs[WhichRock],CurrentStage := AllRocks[WhichStage], CurrentRockHP > 0.0):
                        WhichRockNum := GetWhichRock(WhichRock)
                        StageRockHP := GetRockHP(CurrentStage, WhichStage, WhichRockNum)
                        if(CurrentRockHP > TotalDamage):
                            CustomPlayer.Collect(TotalDamage, "Stone", 0)
                            DamageNumbers.ShowDamageUIForPlayer(CustomPlayer, TotalDamage, DamageResult(1), DamageResult(3), false, "Void", DamageResult(2))
                            if(WhichRock = "NewMeteor"):
                                set CustomPlayer.DamageToNewMeteor += TotalDamage
                        else:
                            if(WhichRock <> "Meteor" and WhichRock <> "NewMeteor"):
                                var PetFinalHitBonus : float = 0.0
                                for(i:int=0..CustomPlayer.PetSlots.Length):
                                    if(CustomPlayer.PetSlots[i] <> -1):
                                        if:
                                            PetLevel := CustomPlayer.PetLevels[CustomPlayer.PetSlots[i]]
                                            FinalHit := CustomPlayer.PetFinalStrike[CustomPlayer.PetSlots[i]]
                                        then:
                                            set PetFinalHitBonus += FinalHit/100.0
                                TotalReward := TotalDamage + (StageRockHP*CustomPlayer.FinalHitBonus) + (StageRockHP*PetFinalHitBonus)
                                CustomPlayer.Collect(TotalReward, "Stone", 0)
                                DamageNumbers.ShowDamageUIForPlayer(CustomPlayer, TotalReward, DamageResult(1), DamageResult(3), true, "Void", DamageResult(2))

                                spawn{RockDestroyed(CustomPlayer, WhichRock, WhichStage)}
                            else:
                                if (WhichRock = "Meteor"):
                                    spawn{MeteorDestroyed()}
                                if (WhichRock = "NewMeteor"):
                                    spawn{NewMeteorDestroyed()}
                                    
                        if(set EachRockHPs[WhichRock] -= TotalDamage){} #Reduce that Rocks HP

                        if (EachRockHPs[WhichRock] <= 0.0):
                            if(Accolade:=AudioManager.Accolades[0]): #AWARD PLAYER EXP
                                Accolade.Award(CustomPlayer.Player)
                        if(CustomPlayer.CurrentRockHit = WhichRock):
                            if (EachRockHPs[WhichRock] <= 0.0):
                                for(K -> CPlayer:gm.PlayersMap):
                                    if:
                                        CPlayer.CurrentRockHit = WhichRock
                                    then:
                                        HPBars.TakeDamage(CPlayer, 0.0)
                                        CPlayer.UpAchValue(4, 1.0)
                                        spawn{MaybeRemoveHPBar(CPlayer, CPlayer.CurrentRockHit)}
                                        if(WhichRock = "Meteor"):
                                            spawn{GrantMeteorRewards(CPlayer)}
                                        else if(WhichRock = "NewMeteor"):
                                            spawn{GrantNewMeteorRewards(CPlayer)}
                                        else:
                                            DoesEggDrop(CPlayer, WhichStage, "Stone")
                            else:
                                if(NewRockHP:=EachRockHPs[WhichRock]):
                                    for(K -> CPlayer:gm.PlayersMap):
                                        if:
                                            CPlayer.CurrentRockHit = WhichRock
                                        then:
                                            HPBars.TakeDamage(CPlayer, NewRockHP)

                        if(NewRockHP:=EachRockHPs[WhichRock]):
                            spawn{MaybeResetRockHP(CustomPlayer, WhichRock, NewRockHP, WhichStage, true)}
                                            
                    set CustomPlayer.VoidActiveTimer += 1.0
                    Sleep(0.5)
                else:
                    set CustomPlayer.VoidActiveTimer = 0.0
                    set CustomPlayer.VoidActiveStacks = 0.0
                    break:          

    ShowVoidFracture(CustomPlayer:Custom_Player, WhichRock:string, VoidDuration:float, FortLoc:vector3)<suspends>:void=
        #AudioManager.Earthquake.Play(CustomPlayer.Player) #FIND AudioManager FOR THIS?
        if(CustomPlayer.ShowVFX = true):
            NewRotation : rotation = IdentityRotation()
            NewScale : vector3 = vector3{X:=1.0, Y:=1.0, Z:=1.0}
            var TimeSoFar : float = 0.0

            VoidFracture := SpawnProp(VoidBP, transform{Translation:=vector3{X:=FortLoc.X,Y:=FortLoc.Y,Z:=FortLoc.Z + 100.0}, Rotation:=NewRotation, Scale:=NewScale})
            loop:
                set TimeSoFar += 1.0
                if(EachRockHPs[WhichRock] <= 0.0 or VoidDuration <= TimeSoFar):
                    PropInst := VoidFracture(0)
                    if ( Prop := PropInst?):
                        Prop.Dispose()
                    break:
                Sleep(1.0)

    IsHoldingWeapon(Agent:agent):logic=
        var IsHolding : logic = false
        for(HoldingWepButton:HoldingWepButtons):
            if(HoldingWepButton.IsHoldingItem[Agent]):
                set IsHolding = true
        return IsHolding

        
    #WE DO THIS FOR VISUAL CLARITY
    DoesEggDrop(CPlayer:Custom_Player, whichStage:int, WhereDropped:string)<suspends>:void=
        Sleep(0.3)
        pm.DoesEggDrop(CPlayer, whichStage, WhereDropped)
