
using { /Fortnite.com/Devices }
using { /Verse.org/Simulation }
using { /UnrealEngine.com/Temporary/Diagnostics }

# See https://dev.epicgames.com/documentation/en-us/uefn/create-your-own-device-in-verse for how to create a verse device.

# A Verse-authored creative device that can be placed in a level
infinite_rock_manager := class(creative_device):

    @editable
    gm : game_manager = game_manager{}
    @editable
    HRM : hit_rock_manager = hit_rock_manager{}

    @editable
    InfRockPropManip : prop_manipulator_device = prop_manipulator_device{}
    @editable
    InfRockMenuBtn : button_device = button_device{}

    @editable
    Earthquake1 : creative_prop_asset = DefaultCreativePropAsset
    @editable
    Earthquake2 : creative_prop_asset = DefaultCreativePropAsset
    @editable
    Earthquake3 : creative_prop_asset = DefaultCreativePropAsset

    @editable
    VoidBP : creative_prop_asset = DefaultCreativePropAsset

    @editable
    HoldingWepButtons: []conditional_button_device = array{}

    var InfRockHPs : []float = array{
        3.03e167,3.03e168,3.03e169,3.03e170,3.03e171
    }

    HPBars: health_bar_device = health_bar_device{}

    OnInfRockHit(Agent:agent):void=
        if(CustomPlayer := gm.PlayersMap[Agent]):
            IsHolding:=HRM.IsHoldingWeapon(Agent)
            if(IsHolding = false):

    # Runs when the device is started in a running game
    OnBegin<override>()<suspends>:void=
        # TODO: Replace this with your code
        Print("Hello, world!")
        Print("2 + 2 = {2 + 2}")