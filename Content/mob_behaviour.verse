using { /Fortnite.com/AI }
using { /Verse.org/Simulation }
using { /Fortnite.com/AI }
using { /Verse.org/Simulation }
using {/Fortnite.com/Characters }
using { /Fortnite.com/Animation/PlayAnimation }
using { /Verse.org/Assets }
using { /UnrealEngine.com/Temporary/SpatialMath }
using { /Fortnite.com/Game }

mob_behaviour := class(npc_behavior):

    var MobAttacks : []animation_sequence = array:
        Monsters.GreenSpikyBlob_Anim_CharacterArmature_Bite_Front
        Monsters.Mushnub_Evolved_Anim_CharacterArmature_Bite_Front
        Monsters.Fish_Anim_CharacterArmature_Bite_Front
        Monsters.Yeti_Anim_CharacterArmature_Bite_Front
        Monsters.Orc_Anim_CharacterArmature_Bite_Front
        Monsters.CactoroReal_Anim_CharacterArmature_Bite_Front
        Monsters.Alien_Anim_CharacterArmature_Bite_Front
        Monsters.Ghost_Anim_CharacterArmature_Punch

    var TargetAgent : ?agent = false
    var WhichAnim : int = 0
    var IsElim : logic = false

    OnBegin<override>()<suspends>:void=
        spawn{StartMobLogic()}

    StartMobLogic()<suspends>:void=
        set IsElim = false
        if:
            NPCAgent := GetAgent[]
            NPCChar := NPCAgent.GetFortCharacter[]
            NPCNav := NPCChar.GetNavigatable[]
            NPCFocus := NPCChar.GetFocusInterface[]
            NPCAnim := NPCChar.GetPlayAnimationController[]
        then:
            loop:
                Sleep(0.01)
                if(IsElim = false):
                    if(NewAgent := TargetAgent?, NewTarget := NewAgent.GetFortCharacter[]):
                        spawn{NPCFocus.MaintainFocus(NewAgent)}
                        NavTarget := MakeNavigationTarget(NewAgent)
                        NavResult:=NPCNav.NavigateTo(NavTarget, ?MovementType := movement_types.Running, ?ReachRadius := 125.0, ?AllowPartialPath:=true)
                        if(NavResult = navigation_result.Reached):
                            NPCNav.StopNavigation()
                            if(Attack:=MobAttacks[WhichAnim]):
                                NPCAnim.Play(Attack)
                            Sleep(0.69)
                            CheckForDamage(NewAgent, NPCChar.GetTransform().Translation, 275.0, 10.0)
                            Sleep(1.50)
                        else:
                            NPCNav.StopNavigation()     
                            Sleep(0.02)
                else:
                    Print("BREAK LOOP")
                    break:

    CheckForDamage(Target:agent, VFXCenter:vector3, DamageRange:float, Damage:float):void=
        if(TargetFC:=Target.GetFortCharacter[]):
            if(Distance(VFXCenter, TargetFC.GetTransform().Translation) < DamageRange):
                TargetFC.Damage(Damage)

    SetTargetAgent(Agent:?agent, WhichRoom:int):void=
        set TargetAgent = Agent
        set WhichAnim = WhichRoom

    SetIsElim(ElimResult:elimination_result):void=
        #Print("SET ELIMINATED")
        set IsElim = true


