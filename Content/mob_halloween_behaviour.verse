using { /Fortnite.com/AI }
using { /Verse.org/Simulation }
using { /Fortnite.com/AI }
using { /Verse.org/Simulation }
using {/Fortnite.com/Characters }
using { /Fortnite.com/Animation/PlayAnimation }
using { /Verse.org/Assets }
using { /UnrealEngine.com/Temporary/SpatialMath }
using { /Fortnite.com/Game }
using { /Verse.org/Random }

mob_halloween_behaviour := class(npc_behavior):

    var MobAttacks : []animation_sequence = array:
        Monsters.A_HalloweenPuppy_Combat_Unarmed_Attack
        Monsters.A_HalloweenGhost_Combat_Unarmed_AttackSpecial

    var TargetAgents : [int]?agent = map{}
    var WhichAnim : int = 0
    var IsElim : logic = false

    OnBegin<override>()<suspends>:void=
        spawn{StartMobLogic()}

    StartMobLogic()<suspends>:void=
        set IsElim = false
        if:
            NPCAgent := GetAgent[]
            NPCChar := NPCAgent.GetFortCharacter[]
            NPCNav := NPCChar.GetNavigatable[]
            NPCFocus := NPCChar.GetFocusInterface[]
            NPCAnim := NPCChar.GetPlayAnimationController[]
        then:
            loop:
                Sleep(0.01)
                if(IsElim = false):
                    WhichTarget:=GetRandomInt(0,TargetAgents.Length-1)
                    if(NewAgent := TargetAgents[WhichTarget]?, NewTarget := NewAgent.GetFortCharacter[]):
                        spawn{NPCFocus.MaintainFocus(NewAgent)}
                        NavTarget := MakeNavigationTarget(NewAgent)
                        NavResult:=NPCNav.NavigateTo(NavTarget, ?MovementType := movement_types.Running, ?ReachRadius := 145.0, ?AllowPartialPath:=true)
                        if(NavResult = navigation_result.Reached):
                            NPCNav.StopNavigation()
                            if(Attack:=MobAttacks[WhichAnim]):
                                NPCAnim.Play(Attack)
                            Sleep(0.22)
                            CheckForDamage(NewAgent, NPCChar.GetTransform().Translation, 220.0, 20.0)
                            Sleep(0.30)
                        else:
                            NPCNav.StopNavigation()     
                            Sleep(0.02)
                else:
                    #Print("BREAK LOOP")
                    break:

    CheckForDamage(Target:agent, VFXCenter:vector3, DamageRange:float, Damage:float):void=
        if(TargetFC:=Target.GetFortCharacter[]):
            if(Distance(VFXCenter, TargetFC.GetTransform().Translation) < DamageRange):
                TargetFC.Damage(Damage)

    SetTargetAgents(Agents:[int]?agent, WhichRoom:int):void=
        #Print("SET TARGETS")
        set TargetAgents = Agents
        set WhichAnim = WhichRoom

    SetIsElim(ElimResult:elimination_result):void=
        #Print("SET ELIMINATED")
        set IsElim = true
