using { /Fortnite.com/Devices }
using { /Verse.org/Simulation }
using { /Verse.org/Colors }
using { /Verse.org/Verse }
using { /Verse.org/Random }
using { /Fortnite.com/Teams }
using { /UnrealEngine.com/Assets }
using { /Fortnite.com/Game }
using { /Fortnite.com/Characters }
using { /Fortnite.com/Playspaces }
using { /Fortnite.com/UI }
using { /UnrealEngine.com/Temporary/Diagnostics }
using { /UnrealEngine.com/Temporary/SpatialMath }
using { /UnrealEngine.com/Temporary/UI }
using { /Verse.org/Assets}
using { /Verse.org/Colors/NamedColors }
using { /Fortnite.com/AI }

mob_room_manager := class(creative_device):

    @editable
    gm : game_manager = game_manager{}
    
    @editable
    pm: pet_manager = pet_manager{}
    
    @editable
    Tutorial: tutorial_manager = tutorial_manager{}

    @editable
    TestDummyManip: prop_manipulator_device = prop_manipulator_device{}

    @editable
    AudioManager : audio_manager = audio_manager{}

    @editable
    DRM : new_daily_rewards_manager = new_daily_rewards_manager{}

    var DamageNumbers<private>: damage_numbers = damage_numbers:

    HPBars: health_bar_device = health_bar_device{}

    var NewestTargets : [int]?agent = map{}

    var MobRoomTarget1 : []?agent = array{false,false,false,false,false,false,false,false}
    var MobRoomTarget2 : []?agent = array{false,false,false,false,false,false,false,false}
    var MobRoomTarget3 : []?agent = array{false,false,false,false,false,false,false,false}
    var MobRoomTarget4 : []?agent = array{false,false,false,false,false,false,false,false}
    var MobRoomTarget5 : []?agent = array{false,false,false,false,false,false,false,false}
    var MobRoomTarget6 : []?agent = array{false,false,false,false,false,false,false,false}
    var MobRoomTarget7 : []?agent = array{false,false,false,false,false,false,false,false}
    var MobRoomTarget8 : []?agent = array{false,false,false,false,false,false,false,false}
    var MobRoomTarget9 : []?agent = array{false,false,false,false,false,false,false,false}
    var MobRoomTarget10 : []?agent = array{false,false,false,false,false,false,false,false}
    var MobRoomTargets : [][]?agent = array{}

    @editable
    Area1Room1Spawners : []npc_spawner_device = array{}
    @editable
    Area1Room2Spawners : []npc_spawner_device = array{}
    @editable
    Area2Room1Spawners : []npc_spawner_device = array{}
    @editable
    Area2Room2Spawners : []npc_spawner_device = array{}
    @editable
    Area3Room1Spawners : []npc_spawner_device = array{}
    @editable
    Area3Room2Spawners : []npc_spawner_device = array{}
    @editable
    Area4Room1Spawners : []npc_spawner_device = array{}
    @editable
    Area4Room2Spawners : []npc_spawner_device = array{}
    @editable
    Area5Room1Spawners : []npc_spawner_device = array{}
    @editable
    Area5Room2Spawners : []npc_spawner_device = array{}

    @editable
    StoneGranter : item_granter_device = item_granter_device{}
    @editable
    WepMatGranter : item_granter_device = item_granter_device{}
    @editable
    BossKeyGranter : item_granter_device = item_granter_device{}

    @editable
    SetCurrentRooms : []mutator_zone_device = array{}

    var AllRoomSpawners : [][]npc_spawner_device = array{}

    var RoomNumbers : []int = array{1, 3, 6, 8, 11, 13, 16, 18, 21, 23}

    #[which zone] amount of players in zone
    var PlayersInZones : [int]int = map{}

    var MobsElimed : int = 0

    var EachMobHPs : [fort_character]float = map{}
    var MobRoomHPs : []float = array:
        2500.0, 37500.0,
        1687500.0, 75937500.0,
        7593750000.0, 1063125000000.0,
        637875000000000.0,7.6545e18,
        2.3e20,1.2e22

    var BossKeyDropRates : []float =  array{0.04, 0.05, 0.07, 0.08, 0.10, 0.12, 0.15, 0.18, 0.22, 0.28}

    OnBegin<override>()<suspends>:void=
        Sleep(3.0)
        TestDummyManip.DamagedEvent.Subscribe(OnTestDummyHit)

        set MobRoomTargets = array:
            MobRoomTarget1
            MobRoomTarget2
            MobRoomTarget3
            MobRoomTarget4
            MobRoomTarget5
            MobRoomTarget6
            MobRoomTarget7
            MobRoomTarget8
            MobRoomTarget9
            MobRoomTarget10

        set AllRoomSpawners = array:
            Area1Room1Spawners
            Area1Room2Spawners
            Area2Room1Spawners
            Area2Room2Spawners
            Area3Room1Spawners
            Area3Room2Spawners
            Area4Room1Spawners
            Area4Room2Spawners
            Area5Room1Spawners
            Area5Room2Spawners
            
        for(i:int=0..SetCurrentRooms.Length):
            if(set PlayersInZones[i] = 0) {}
            if(MobRoom:=SetCurrentRooms[i]):
                MobRoom.AgentEntersEvent.SubscribeAgent(SetCurrentRoom, i)
                if(i < 6):
                    MobRoom.AgentExitsEvent.SubscribeAgent(SetCurrentRoom, 99)
                else:
                    MobRoom.AgentExitsEvent.SubscribeAgent(SetCurrentRoom, 999)

        for(i:int=0..AllRoomSpawners.Length):
            if(MobRoom:=AllRoomSpawners[i]):
                for(MobSpawner:MobRoom):
                    MobSpawner.SpawnedEvent.SubscribeAgent(OnMobSpawned, i)

    SetCurrentRoom(Agent:agent, whichRoom:int):void=
        if(CustomPlayer := gm.PlayersMap[Agent]):
            if(whichRoom = 99):
                gm.SwitchMusic(Agent, "Default")
                for(i:=0..MobRoomTargets[CustomPlayer.CurrentMobRoom].Length):
                    if:
                        RoomTargets := MobRoomTargets[CustomPlayer.CurrentMobRoom]
                        NewestTarget := RoomTargets[i]
                        Player := NewestTarget?
                        Player = CustomPlayer.Player
                    then:
                        Print("MobRoom {CustomPlayer.CurrentMobRoom} Target at {i} removed")
                        if(set MobRoomTargets[CustomPlayer.CurrentMobRoom][i] = false) {}

                if(set PlayersInZones[CustomPlayer.CurrentMobRoom] -= 1) {}
                spawn{DespawnAllMobs(CustomPlayer.CurrentMobRoom)}
                set CustomPlayer.CurrentMobRoom = -1
            else if(whichRoom = 999):
                gm.SwitchMusic(Agent, "Space")
                for(i:=0..MobRoomTargets[CustomPlayer.CurrentMobRoom].Length):
                    if:
                        RoomTargets := MobRoomTargets[CustomPlayer.CurrentMobRoom]
                        NewestTarget := RoomTargets[i]
                        Player := NewestTarget?
                        Player = CustomPlayer.Player
                    then:
                        Print("MobRoom {CustomPlayer.CurrentMobRoom} Target at {i} removed")
                        if(set MobRoomTargets[CustomPlayer.CurrentMobRoom][i] = false) {}

                if(set PlayersInZones[CustomPlayer.CurrentMobRoom] -= 1) {}
                spawn{DespawnAllMobs(CustomPlayer.CurrentMobRoom)}
                set CustomPlayer.CurrentMobRoom = -1
            else:
                set CustomPlayer.CurrentMobRoom = whichRoom
                if(set PlayersInZones[whichRoom] += 1) {}
                if(PlayerCount := PlayersInZones[whichRoom]):
                    if(set MobRoomTargets[whichRoom][PlayerCount-1] = option{CustomPlayer.Player}) {}
                    Print("MobRoom {whichRoom} Target at {PlayerCount-1} added")
                spawn{SpawnMobs(CustomPlayer, whichRoom)}
                if(whichRoom < 6):
                    gm.SwitchMusic(Agent, "Mob")
                else:
                    gm.SwitchMusic(Agent, "Space Mob")
                    
    SpawnMobs(CustomPlayer:Custom_Player, whichRoom:int)<suspends>:void=
        if(MobRoom:=AllRoomSpawners[whichRoom]):
            for(MobSpawner:MobRoom):
                MobSpawner.Enable()

    DespawnAllMobs(whichRoom:int)<suspends>:void=
        if(PlayersInZones[whichRoom] <= 0):
            if(MobRoom:=AllRoomSpawners[whichRoom]):
                for(MobSpawner:MobRoom):
                    Print("DESPAWN")
                    MobSpawner.DespawnAll(false)
                    MobSpawner.Disable()

    OnMobSpawned(Agent:agent, whichRoom:int):void=
        if (Creep := Agent.GetFortCharacter[]):
            if(MobHP:=MobRoomHPs[whichRoom]):
                if(set EachMobHPs[Creep] = MobHP){}
                Creep.SetMaxHealth(MobHP)
            if:
                RoomTargets := MobRoomTargets[whichRoom]
                #RandomTarget := GetRandomInt(0, PlayersInZones[whichRoom]-1)
                NewestTarget := RoomTargets[PlayersInZones[whichRoom]-1]
                Owner := NewestTarget?
                CustomPlayer := gm.PlayersMap[Owner]
                Fort:=Agent.GetFortCharacter[]
                NPCBehave:=mob_behaviour[Agent.GetNPCBehavior[]]
            then:
                #Print("RandomTarget: {RandomTarget}")
                NPCBehave.SetTargetAgent(NewestTarget, whichRoom)
                Creep.EliminatedEvent().Subscribe(NPCBehave.SetIsElim)
            Creep.DamagedEvent().SubscribeAgent5(OnMobDamaged, Creep)

    OnTestDummyHit(Agent : agent) : void=
        if:
            Player:=player[Agent]
            CustomPlayer := gm.PlayersMap[Player]
        then:
            var DamageResult : tuple(float, logic, float, int) = CustomPlayer.CalcTotalDamage("Mob")
            DamageNumbers.ShowDamageUIForPlayer(CustomPlayer, DamageResult(0), DamageResult(1), DamageResult(3), false, "Mob", -1.0)
    

    SpawnAllMissiles(CustomPlayer:Custom_Player, Target:fort_character):void=
        if:
            CooldownLevel := CustomPlayer.Hut5Levels[2]
            CooldownAmount := 3.0 - (CooldownLevel * 0.3)
        then:
            set CustomPlayer.MissileCD = CooldownAmount
            spawn{gm.MissileCDTimer(CustomPlayer)}

        spawn{SpawnMissile1(CustomPlayer, Target)}
        if(CustomPlayer.ShowVFX = true):
            spawn{SpawnMissile2(CustomPlayer, Target)}
            spawn{SpawnMissile3(CustomPlayer, Target)}

        
    SpawnMissile1(CustomPlayer:Custom_Player, Target:fort_character)<suspends>:void=
        if(MissileSFX := AudioManager.MissileSFX[GetRandomInt(0,1)]):
            MissileSFX.Play(CustomPlayer.Player)
        if(Fort:=CustomPlayer.Player.GetFortCharacter[]):
            PlayerLoc:=Fort.GetTransform().Translation
            TargetLoc:=Target.GetTransform().Translation
            PlayerRot:=Fort.GetTransform().Rotation
            if(CustomPlayer.ShowVFX = true):
                Missile := SpawnProp(gm.FireMissile, transform{Translation:=vector3{X:=PlayerLoc.X,Y:=PlayerLoc.Y,Z:=PlayerLoc.Z}})
                if (Prop := Missile(0)?):
                    Prop.MoveTo(vector3{X:=PlayerLoc.X,Y:=PlayerLoc.Y,Z:=PlayerLoc.Z+300.0}, PlayerRot, 0.2)
                    Prop.MoveTo(vector3{X:=TargetLoc.X,Y:=TargetLoc.Y,Z:=TargetLoc.Z}, PlayerRot, 0.5)
                    Prop.Dispose()
            else:
                Sleep(0.7)

            VerseOnMobDamaged(CustomPlayer, Target)

            if:
                HealLevel := CustomPlayer.Hut5Levels[3]
                HealAmount := HealLevel * 2.0
                HealLevel > 0
            then:
                Fort.Heal(HealAmount)

    SpawnMissile2(CustomPlayer:Custom_Player, Target:fort_character)<suspends>:void=
        if(Fort:=CustomPlayer.Player.GetFortCharacter[]):
            PlayerLoc:=Fort.GetTransform().Translation
            TargetLoc:=Target.GetTransform().Translation
            PlayerRot:=Fort.GetTransform().Rotation
            Missile := SpawnProp(gm.FireMissile, transform{Translation:=vector3{X:=PlayerLoc.X,Y:=PlayerLoc.Y,Z:=PlayerLoc.Z}})
            if (Prop := Missile(0)?):
                Prop.MoveTo(vector3{X:=PlayerLoc.X,Y:=PlayerLoc.Y+300.0,Z:=PlayerLoc.Z+300.0}, PlayerRot, 0.2)
                Prop.MoveTo(vector3{X:=TargetLoc.X,Y:=TargetLoc.Y,Z:=TargetLoc.Z}, PlayerRot, 0.5)
                Prop.Dispose()

    SpawnMissile3(CustomPlayer:Custom_Player, Target:fort_character)<suspends>:void=
        if(Fort:=CustomPlayer.Player.GetFortCharacter[]):
            PlayerLoc:=Fort.GetTransform().Translation
            TargetLoc:=Target.GetTransform().Translation
            PlayerRot:=Fort.GetTransform().Rotation
            Missile := SpawnProp(gm.FireMissile, transform{Translation:=vector3{X:=PlayerLoc.X,Y:=PlayerLoc.Y,Z:=PlayerLoc.Z}})
            if (Prop := Missile(0)?):
                Prop.MoveTo(vector3{X:=PlayerLoc.X,Y:=PlayerLoc.Y-300.0,Z:=PlayerLoc.Z+300.0}, PlayerRot, 0.5)
                Prop.MoveTo(vector3{X:=TargetLoc.X,Y:=TargetLoc.Y,Z:=TargetLoc.Z}, PlayerRot, 0.5)
                Prop.Dispose()

    VerseOnMobDamaged(CustomPlayer:Custom_Player, Creep:fort_character) : void=
        var TotalDamage : float = 0.0
        if:
            FC := CustomPlayer.Player.GetFortCharacter[]
            MobHP:=MobRoomHPs[CustomPlayer.CurrentMobRoom]
        then:
            if(MobHit:=CustomPlayer.CurrentMobHit?):
            else:
                set CustomPlayer.CurrentMobHit = option{FC}
                set CustomPlayer.CurrentRockHit = ""
            if(CustomPlayer.CurrentMobHit? <> Creep):
                set CustomPlayer.CurrentMobHit = option{Creep}
                HPBars.Init(CustomPlayer, "Lv. {CustomPlayer.CurrentMobRoom+1} Monster", MobHP)
                HPBars.ShowUIForPlayer(CustomPlayer)
            var DamageResult : tuple(float, logic, float, int) = CustomPlayer.CalcTotalDamage("Mob")
            if(MissileDamage := DamageResult(0) * (0.2 + CustomPlayer.Hut5Levels[1] * 0.1)):#CHANGE TO THIS IS 
                set TotalDamage = MissileDamage
            if(set EachMobHPs[Creep] -= TotalDamage) {}
            if(NewMobHP:=EachMobHPs[Creep]):
                if(EachMobHPs[Creep] <= 0.0):
                    Creep.Damage(2e200)
                    spawn{OnMobElim(CustomPlayer, Creep)}
                    HPBars.TakeDamage(CustomPlayer, 0.0)
                    set CustomPlayer.CurrentMobHP = 0.0
                    set EachMobHPs = RemoveKeyFromMap(EachMobHPs, Creep)
                else:
                    HPBars.TakeDamage(CustomPlayer, NewMobHP)
                    set CustomPlayer.CurrentMobHP = NewMobHP
                spawn{MaybeRemoveHPBar(CustomPlayer, Creep)}
            DamageNumbers.ShowDamageUIForPlayer(CustomPlayer, TotalDamage, DamageResult(1), DamageResult(3), false, "Missile", -1.0)
    

    OnMobDamaged(DmgResult : damage_result, Creep:fort_character) : void=
        MaybeInstigator := DmgResult.Instigator
        if (Instigator := MaybeInstigator?):
            if (InstagatingAgent := Instigator.GetInstigatorAgent[],Player:=player[InstagatingAgent], FC := Player.GetFortCharacter[]):
                if(CustomPlayer := gm.PlayersMap[Player],MobHP:=MobRoomHPs[CustomPlayer.CurrentMobRoom]):
                    if(MobHit:=CustomPlayer.CurrentMobHit?):
                    else:
                        set CustomPlayer.CurrentMobHit = option{FC}
                        set CustomPlayer.CurrentRockHit = ""
                    if(CustomPlayer.CurrentMobHit? <> Creep):
                        set CustomPlayer.CurrentMobHit = option{Creep}
                        HPBars.Init(CustomPlayer, "Lv. {CustomPlayer.CurrentMobRoom+1} Monster", MobHP)
                        HPBars.ShowUIForPlayer(CustomPlayer)
                    var DamageResult : tuple(float, logic, float, int) = CustomPlayer.CalcTotalDamage("Mob")
                    if(DamageResult(1) = true):
                        AudioManager.CritHit.Play(CustomPlayer.Player)
                    if:
                        CustomPlayer.Hut5Levels[0] = 1
                        CustomPlayer.MissileCD <= 0.0
                        CustomPlayer.HutEnabled[4] = true
                    then:
                        SpawnAllMissiles(CustomPlayer, Creep)
                    DmgResult.Target.Damage(DamageResult(0))
                    if(set EachMobHPs[Creep] -= DamageResult(0) + DmgResult.Amount) {}
                    if(NewMobHP:=EachMobHPs[Creep]):
                        if(EachMobHPs[Creep] <= 0.0):
                            spawn{OnMobElim(CustomPlayer, Creep)}
                            HPBars.TakeDamage(CustomPlayer, 0.0)
                            set CustomPlayer.CurrentMobHP = 0.0
                            set EachMobHPs = RemoveKeyFromMap(EachMobHPs, Creep)
                        else:
                            HPBars.TakeDamage(CustomPlayer, NewMobHP)
                            set CustomPlayer.CurrentMobHP = NewMobHP
                        spawn{MaybeRemoveHPBar(CustomPlayer, Creep)}
                    DamageNumbers.ShowDamageUIForPlayer(CustomPlayer, DamageResult(0), DamageResult(1), DamageResult(3), false, "Mob", -1.0)
    
    MaybeRemoveHPBar(CustomPlayer:Custom_Player, Creep:fort_character)<suspends>:void=
        Sleep(5.0)
        if(FC:=CustomPlayer.Player.GetFortCharacter[]):
            if(MobHP:=EachMobHPs[Creep]):
                if:
                    OldCreep := CustomPlayer.CurrentMobHit?
                    (MobHP = CustomPlayer.CurrentMobHP or MobHP <= 0.0)
                    Creep = OldCreep
                    CustomPlayer.QueuedForBoss = false
                    CustomPlayer.CurrentRockHit = ""
                then:
                    CustomPlayer.RemoveRockHealthBarUI()
                   set CustomPlayer.CurrentMobHit = option{FC}
            else:
                if:
                    OldCreep := CustomPlayer.CurrentMobHit?
                    Creep = OldCreep
                    CustomPlayer.QueuedForBoss = false
                    CustomPlayer.CurrentRockHit = ""
                then:
                    CustomPlayer.RemoveRockHealthBarUI()
                    set CustomPlayer.CurrentMobHit = option{FC}

    OnMobElim(CustomPlayer:Custom_Player, Creep:fort_character)<suspends>:void=
        if (MobQuestStatus := CustomPlayer.QuestsStarted[1], MobQuestStatus?):
            set CustomPlayer.QuestMobsDefeated += 1
            DRM.UpdateCounterUI(CustomPlayer)
        set MobsElimed += 1
        #Print("Mobs Elimed: {MobsElimed}")
        Tutorial.UpdateTutorialUI(CustomPlayer, "MobElim", 1.0)

        if(Accolade:=AudioManager.Accolades[1]): #AWARD PLAYER EXP
            Accolade.Award(CustomPlayer.Player)
        if(CurrentRoom:=RoomNumbers[CustomPlayer.CurrentMobRoom]):
            if(ElimReward := MobRoomHPs[CustomPlayer.CurrentMobRoom]):
                CustomPlayer.UpAchValue(2, 1.0)
                CustomPlayer.Collect(ElimReward, "Stone", 0)
                #StoneGranter.GrantItem(CustomPlayer.Player)
                pm.DoesEggDrop(CustomPlayer, CurrentRoom, "Mob")

            if(RoomNumbers[CustomPlayer.CurrentMobRoom] = 21 or RoomNumbers[CustomPlayer.CurrentMobRoom] = 23):
                Sleep(0.2) #CHANGE TO 0.2
                #SHARD DROP
                ShardDrop:=GetRandomFloat(1.0,100.0)
                if(ShardDropRate:= 0.20 + (CustomPlayer.GUP3Levels[1] * 0.02)):
                    if(ShardDropRate*100.0 > ShardDrop):
                        ShardAmount := GetRandomFloat(150.0, 300.0)
                        AudioManager.ItemObtained.Play(CustomPlayer.Player)
                        CustomPlayer.Collect(ShardAmount, "Shard", 0)
                        DamageNumbers.ShowDamageUIForPlayer(CustomPlayer, ShardAmount, false, 0, false, "Shard", -1.0)

            Sleep(0.2) #CHANGE TO 0.2
            #WEP MAT DROP 
            WepMatDrop:=GetRandomFloat(1.0,100.0)
            if(WepMatDropRate:= 0.02 + (CustomPlayer.GUP3Levels[1] * 0.02)):
                #Print("WepMatDropRate {WepMatDropRate}")
                if(WepMatDropRate*100.0 > WepMatDrop):
                    AudioManager.ItemObtained.Play(CustomPlayer.Player)
                    CustomPlayer.Collect(1.0, "WepMat", 0)
                    DamageNumbers.ShowDamageUIForPlayer(CustomPlayer, 1.0, false, 0, false, "WepMat", -1.0)

            Sleep(0.2) #CHANGE TO 0.2
            #BOSS ROOM KEY DROP
            BossKeyDrop:=GetRandomFloat(1.0,100.0)
            if:
                BossKeyLevel := CustomPlayer.GUP3Levels[1]
                BossKeyDropRate:=BossKeyDropRates[CustomPlayer.CurrentMobRoom] + (BossKeyLevel * 0.02)
            then:
                #Print("BossKeyDropRate {BossKeyDropRate}")
                if(BossKeyDropRate*100.0 > BossKeyDrop):
                    AudioManager.ItemObtained.Play(CustomPlayer.Player)
                    #BossKeyGranter.GrantItem(CustomPlayer.Player)
                    CustomPlayer.Collect((1.0 + BossKeyLevel*1.0), "BossKey", 0)
                    DamageNumbers.ShowDamageUIForPlayer(CustomPlayer, (1.0 + BossKeyLevel*1.0), false, 0, false, "BossKey", -1.0)

            if(Analytic:=AudioManager.MobAnalytics[CustomPlayer.CurrentMobRoom]): #ANALYTIC
                Analytic.Submit(CustomPlayer.Player)

        SaveAllData(CustomPlayer)