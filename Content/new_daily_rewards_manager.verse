using { /Fortnite.com/Devices }
using { /Verse.org/Simulation }
using { /UnrealEngine.com/Temporary/Diagnostics }
using { /UnrealEngine.com/Temporary/SpatialMath }
using { /Verse.org/Random }
using { /UnrealEngine.com/Temporary/UI }
using { /Verse.org/Colors }

new_daily_rewards_manager := class(creative_device):

    DateData : date_data = date_data{}

    @editable
    gm : game_manager = game_manager{}
    @editable
    AudioManager : audio_manager = audio_manager{}
    @editable
    QuestBtns : []button_device = array{}
    @editable
    QuestPopupsAccept : []popup_dialog_device = array{}
    @editable
    QuestPopupsComplete : []popup_dialog_device = array{}
    @editable
    QuestPopUpFinal : popup_dialog_device = popup_dialog_device{}
    @editable
    ChangeData : button_device = button_device{}

    var DamageNumbers<private>: damage_numbers = damage_numbers{}

    <# var RocksTarget : int = 10
    var MobsTarget : int = 10
    var BossesTarget : int = 10 #>

    var QuestTargets : []int = array{250, 500, 25} # Rocks | Mobs | Bosses #CHANGE TO 250, 500, 25

    var Tomorrow : date_data = date_data{Year:=2024,Month:=11,Day:=07,Hour:=4,Minute:=0,Second:=0}
    #var Today : date_data = date_data{Year:=2024,Month:=11,Day:=06,Hour:=3,Minute:=0,Second:=0}

    var AllDailyRewards : [][]float = array{
        array{600.0, 12500.0, 17500.0},
        array{850.0, 15000.0, 20000.0},
        array{1000.0, 17500.0, 22500.0},
        array{1200.0, 20000.0, 25000.0}
        }

    <# ChangeDatePress(Agent:agent):void=
        set Today = Tomorrow #>

    CheckFinalQuestStatus(Agent:agent, Index:int):void=
        var DoneAllQuests : logic = true
        if(CustomPlayer := gm.PlayersMap[Agent]):
            if (Index = 0):
                for (i:int=0..CustomPlayer.QuestsCompleted.Length-1):
                    if (Quest := CustomPlayer.QuestsCompleted[i], Quest = false):
                        set DoneAllQuests = false
                if (DoneAllQuests?):
                    gm.ShowMessage(CustomPlayer.Player, "Congrats on completing all quests! Enjoy your reward!", 0)
                    spawn{GiveRewards(CustomPlayer,3)}
                    set CustomPlayer.QuestFinalStatus = true
                else:
                    gm.ShowMessage(CustomPlayer.Player, "You have not completed all quests, come back when you are finished!", 1)

    
    GiveRewards(CustomPlayer:Custom_Player, Which:int)<suspends>:void=
        if(Which < 3):
            Print("A QUEST")
            if(Accolade:=AudioManager.Accolades[16]): #AWARD PLAYER EXP
                Accolade.Award(CustomPlayer.Player)
            if(Analytic:=AudioManager.GeneralAnalytics[10]): #AWARD PLAYER EXP
                Analytic.Submit(CustomPlayer.Player)
        else:
            Print("ALL QUESTS")
            if(Accolade:=AudioManager.Accolades[17]): #AWARD PLAYER EXP
                Accolade.Award(CustomPlayer.Player)
            if(Analytic:=AudioManager.GeneralAnalytics[11]): #AWARD PLAYER EXP
                Analytic.Submit(CustomPlayer.Player)
        for(i:=0..3):
            if(BossRewards:=AllDailyRewards[Which]):
                if(StoneReward:=BossRewards[0]): #GIVE STONE
                    TotalReward := ((CustomPlayer.GetPlayerDamage()*StoneReward) / 4.0) #* CustomPlayer.BossRewardMult
                    CustomPlayer.Collect(TotalReward, "Stone", 0)
                    DamageNumbers.ShowDamageUIForPlayer(CustomPlayer, TotalReward, false, 0, false, "Rock", -1.0)
                    #gm.StoneGranter.GrantItem(CustomPlayer.Player)
                    Sleep(0.3)

                if(EggMinReward:=BossRewards[1],EggMaxReward:=BossRewards[2]): #GIVE EGGS
                    var whichEgg : int = 0
                    if(CustomPlayer.StagesUnlocked = -1):
                        if(whichDropTable:=gm.EggDropOdds[0],DropAmount := Floor[GetRandomFloat(EggMinReward, EggMaxReward)]*1.0, EggGranter:=gm.EggGranters[whichEgg]):
                            set whichEgg = SelectItem(whichDropTable)
                            TotalReward := (DropAmount / 4.0) #* CustomPlayer.BossRewardMult
                            CustomPlayer.Collect(TotalReward, "Egg", whichEgg)
                            DamageNumbers.ShowDamageUIForPlayer(CustomPlayer, TotalReward, false, 0, false, "Egg{whichEgg}", -1.0)
                            #EggGranter.GrantItem(CustomPlayer.Player)
                            Sleep(0.3)
                    else:
                        if(whichDropTable:=gm.EggDropOdds[CustomPlayer.StagesUnlocked],DropAmount := Floor[GetRandomFloat(EggMinReward, EggMaxReward)]*1.0, EggGranter:=gm.EggGranters[whichEgg]):
                            set whichEgg = SelectItem(whichDropTable)
                            TotalReward := (DropAmount / 4.0)
                            CustomPlayer.Collect(TotalReward, "Egg", whichEgg)
                            DamageNumbers.ShowDamageUIForPlayer(CustomPlayer, TotalReward, false, 0, false, "Egg{whichEgg}", -1.0)
                            #EggGranter.GrantItem(CustomPlayer.Player)
                            Sleep(0.3)

    OnQuestPress(Agent:agent, Which:int):void=
        var Today : date_data = DateData.ParseEpoch(GetSecondsSinceEpoch())
        if(CustomPlayer := gm.PlayersMap[Agent]):
            if (not DateData.IsSameDay(Today, CustomPlayer.LastCompletedDate)?):
                Print("NEW DAY")
                set CustomPlayer.LastCompletedDate = Today
                set CustomPlayer.QuestRocksDestroyed = 0
                set CustomPlayer.QuestMobsDefeated = 0
                set CustomPlayer.QuestBossesDefeated = 0
                set CustomPlayer.QuestFinalStatus = false
                for (i:int=0..CustomPlayer.QuestsStarted.Length-1):
                    if (set CustomPlayer.QuestsStarted[i] = false){}
                    if (set CustomPlayer.QuestsCompleted[i] = false){}
                UpdateCounterUI(CustomPlayer)
            else:
                Print("SAME DAY")

            if (Which = 3):
                if (CustomPlayer.QuestFinalStatus =  true):
                    gm.ShowMessage(CustomPlayer.Player, "You have completed all daily quests. Come back tomorrow!", 0)
                else:                   
                    QuestPopUpFinal.Show(Agent)
            else:
                if (QuestComplete := CustomPlayer.QuestsCompleted[Which], QuestComplete = true):
                    gm.ShowMessage(CustomPlayer.Player, "You have already completed this Quest today!", 0)
                else:
                    if (Bool := CustomPlayer.QuestsStarted[Which], Bool = false):
                        if (QuestPopop := QuestPopupsAccept[Which]):
                            QuestPopop.Show(Agent)
                    else:
                        if(QuestPopup := QuestPopupsComplete[Which]):
                            QuestPopup.Show(Agent)
            SaveAllData(CustomPlayer)

    OnQuestAccept(Agent:agent, Index:int,Which:int):void=
        if(CustomPlayer := gm.PlayersMap[Agent]):
            if (Index = 0):
                if (set CustomPlayer.QuestsStarted[Which] = true){}
                if (Which = 0):
                    OpenRockCounterUI(Agent)
                else if (Which = 1):
                    OpenMobCounterUI(Agent)
                else:
                    OpenBossCounterUI(Agent)
                SaveAllData(CustomPlayer)

    OnQuestComplete(Agent:agent, Index:int, Which:int):void=
        if(CustomPlayer := gm.PlayersMap[Agent]):
            if (Index = 0):
                if (Which = 0, RocksTarget := QuestTargets[Which]):
                    if:
                        CustomPlayer.QuestRocksDestroyed >= RocksTarget
                        CustomPlayer.QuestsCompleted[Which] = false
                    then:
                        CloseRockCounterUI(Agent)
                        if (set CustomPlayer.QuestsCompleted[Which] = true){}
                        spawn{GiveRewards(CustomPlayer, Which)}
                        gm.ShowMessage(CustomPlayer.Player, "Quest Complete!", 0)
                    else:
                        gm.ShowMessage(CustomPlayer.Player, "You have only destroyed {CustomPlayer.QuestRocksDestroyed} rocks, come back when you have destroyed {RocksTarget}", 1)
                
                if (Which = 1, MobsTarget := QuestTargets[Which]):
                    if:
                        CustomPlayer.QuestMobsDefeated >= MobsTarget
                        CustomPlayer.QuestsCompleted[Which] = false
                    then:
                        CloseMobCounterUI(Agent)
                        if (set CustomPlayer.QuestsCompleted[Which] = true){}
                        spawn{GiveRewards(CustomPlayer, Which)}
                        gm.ShowMessage(CustomPlayer.Player, "Quest Complete!", 0)
                    else:
                        gm.ShowMessage(CustomPlayer.Player, "You have only defeated {CustomPlayer.QuestMobsDefeated} monsters. Come back when you have defeated {MobsTarget}", 1)
                
                if (Which = 2, BossesTarget := QuestTargets[Which]):
                    if:
                        CustomPlayer.QuestBossesDefeated >= BossesTarget
                        CustomPlayer.QuestsCompleted[Which] = false
                    then:
                        CloseBossCounterUI(Agent)
                        if (set CustomPlayer.QuestsCompleted[Which] = true){}
                        spawn{GiveRewards(CustomPlayer, Which)}
                        gm.ShowMessage(CustomPlayer.Player, "Quest Complete!", 0)
                    else:
                        gm.ShowMessage(CustomPlayer.Player, "You have only defeated {CustomPlayer.QuestBossesDefeated} bosses. Come back when you have defeated {BossesTarget}", 1)
                SaveAllData(CustomPlayer)

    OnPlayerJoined(Agent:agent):void=
        spawn{OnPlayerJoinedBuffer(Agent)}

    OnPlayerJoinedBuffer(Agent:agent)<suspends>:void=
        var Today : date_data = DateData.ParseEpoch(GetSecondsSinceEpoch())
        Sleep(1.0)
        if(CustomPlayer := gm.PlayersMap[player[Agent]]):
            set CustomPlayer.AllCounterTexts = array{CustomPlayer.RockCounterText,CustomPlayer.MobCounterText,CustomPlayer.BossCounterText}
            if(not DateData.IsSameDay(Today, CustomPlayer.LastCompletedDate)?):
                Print("NEW DAY DONT SHOW UI")
            else:
                for(i:=0..CustomPlayer.QuestsStarted.Length-1):
                    if (QuestStarted := CustomPlayer.QuestsStarted[i], QuestStarted = true):
                        if(QuestCompleted := CustomPlayer.QuestsCompleted[i], QuestCompleted = true):
                        else:
                            if (i = 0):
                                OpenRockCounterUI(Agent)
                            else if (i = 1):
                                OpenMobCounterUI(Agent)
                            else if (i = 2):
                                OpenBossCounterUI(Agent)

    OpenRockCounterUI(Agent:agent): void=
        if(CustomPlayer := gm.PlayersMap[Agent]):
            if(MyUI:=CustomPlayer.UIMaps["RockQuestCounter"]?):
            else:
                var MyRockCounterCanvas : canvas = MakeRockCounterCanvas(CustomPlayer)
                CustomPlayer.AddUI(MyRockCounterCanvas)
                if(set CustomPlayer.UIMaps["RockQuestCounter"] = option{MyRockCounterCanvas}) {}
                UpdateCounterUI(CustomPlayer)

    OpenMobCounterUI(Agent:agent): void=
        if(CustomPlayer := gm.PlayersMap[Agent]):
            if(MyUI:=CustomPlayer.UIMaps["MobQuestCounter"]?):
            else:
                var MyMobCounterCanvas : canvas = MakeMobCounterCanvas(CustomPlayer)
                CustomPlayer.AddUI(MyMobCounterCanvas)
                if(set CustomPlayer.UIMaps["MobQuestCounter"] = option{MyMobCounterCanvas}) {}
                UpdateCounterUI(CustomPlayer)

    OpenBossCounterUI(Agent:agent): void=
        if(CustomPlayer := gm.PlayersMap[Agent]):
            if(MyUI:=CustomPlayer.UIMaps["BossQuestCounter"]?):
            else:
                var MyBossCounterCanvas : canvas = MakeBossCounterCanvas(CustomPlayer)
                CustomPlayer.AddUI(MyBossCounterCanvas)
                if(set CustomPlayer.UIMaps["BossQuestCounter"] = option{MyBossCounterCanvas}) {}
                UpdateCounterUI(CustomPlayer)

    CloseRockCounterUI(Agent:agent):void=
        if:
            CustomPlayer := gm.PlayersMap[Agent]
            PlayerUI:=GetPlayerUI[CustomPlayer.GetPlayer()]
            MyUI:=CustomPlayer.UIMaps["RockQuestCounter"]?
        then:
            PlayerUI.RemoveWidget(MyUI)
            MyUI.SetVisibility(widget_visibility.Collapsed)
            if(set CustomPlayer.UIMaps["RockQuestCounter"] = false){}

    CloseMobCounterUI(Agent:agent):void=
        if:
            CustomPlayer := gm.PlayersMap[Agent]
            PlayerUI:=GetPlayerUI[CustomPlayer.GetPlayer()]
            MyUI:=CustomPlayer.UIMaps["MobQuestCounter"]?
        then:
            PlayerUI.RemoveWidget(MyUI)
            MyUI.SetVisibility(widget_visibility.Collapsed)
            if(set CustomPlayer.UIMaps["MobQuestCounter"] = false){}

    CloseBossCounterUI(Agent:agent):void=
        if:
            CustomPlayer := gm.PlayersMap[Agent]
            PlayerUI:=GetPlayerUI[CustomPlayer.GetPlayer()]
            MyUI:=CustomPlayer.UIMaps["BossQuestCounter"]?
        then:
            PlayerUI.RemoveWidget(MyUI)
            MyUI.SetVisibility(widget_visibility.Collapsed)
            if(set CustomPlayer.UIMaps["BossQuestCounter"] = false){}

    MakeRockCounterCanvas(CustomPlayer:Custom_Player):canvas=
        MyCanvas: canvas = canvas {
            Slots := array:
                canvas_slot:
                    Widget:=CustomPlayer.RockCounterBG
                    ZOrder:=0
                    SizeToContent:=false
                    Alignment := vector2{X := 0.0, Y := 0.0}
                    Offsets:=margin{Top:=335.0,Left:=0.0,Bottom:=64.0,Right:=215.0}
                canvas_slot:
                    Widget:=CustomPlayer.RockCounterText
                    ZOrder:=9
                    SizeToContent:=true
                    Offsets:=margin{Top:=348.0,Left:=64.0}
        }

    MakeMobCounterCanvas(CustomPlayer:Custom_Player):canvas=
        MyCanvas: canvas = canvas {
            Slots := array:
                canvas_slot:
                    Widget:=CustomPlayer.MobCounterBG
                    ZOrder:=0
                    SizeToContent:=false
                    Alignment := vector2{X := 0.0, Y := 0.0}
                    Offsets:=margin{Top:=435.0,Left:=0.0,Bottom:=64.0,Right:=215.0}
                canvas_slot:
                    Widget:=CustomPlayer.MobCounterText
                    ZOrder:=9
                    SizeToContent:=true
                    Offsets:=margin{Top:=448.0,Left:=64.0}
        }

    MakeBossCounterCanvas(CustomPlayer:Custom_Player):canvas=
        MyCanvas: canvas = canvas {
            Slots := array:
                canvas_slot:
                    Widget:=CustomPlayer.BossCounterBG
                    ZOrder:=0
                    SizeToContent:=false
                    Alignment := vector2{X := 0.0, Y := 0.0}
                    Offsets:=margin{Top:=535.0,Left:=0.0,Bottom:=64.0,Right:=215.0}
                canvas_slot:
                    Widget:=CustomPlayer.BossCounterText
                    ZOrder:=9
                    SizeToContent:=true
                    Offsets:=margin{Top:=548.0,Left:=64.0}
        }
    
    UpdateCounterUI(CustomPlayer:Custom_Player):void=
        if:
            RockTarget := QuestTargets[0]
            MobTarget := QuestTargets[1]
            BossTarget := QuestTargets[2]
        then:
            if(CustomPlayer.QuestRocksDestroyed >= RockTarget):
                CustomPlayer.RockCounterText.SetTextColor(MakeColorFromHex("39FF39"))
                CustomPlayer.RockCounterText.SetText(StringToMessage("Done!"))
            else:
                CustomPlayer.RockCounterText.SetTextColor(MakeColorFromHex("FFFFFF"))
                CustomPlayer.RockCounterText.SetText(StringToMessage("{CustomPlayer.QuestRocksDestroyed}/{RockTarget}"))

            if(CustomPlayer.QuestMobsDefeated >= MobTarget):
                CustomPlayer.MobCounterText.SetTextColor(MakeColorFromHex("39FF39"))
                CustomPlayer.MobCounterText.SetText(StringToMessage("Done!"))
            else:
                CustomPlayer.MobCounterText.SetTextColor(MakeColorFromHex("FFFFFF"))
                CustomPlayer.MobCounterText.SetText(StringToMessage("{CustomPlayer.QuestMobsDefeated}/{MobTarget}"))

            if(CustomPlayer.QuestBossesDefeated >= BossTarget):
                CustomPlayer.BossCounterText.SetTextColor(MakeColorFromHex("39FF39"))
                CustomPlayer.BossCounterText.SetText(StringToMessage("Done!"))
            else:
                CustomPlayer.BossCounterText.SetTextColor(MakeColorFromHex("FFFFFF"))
                CustomPlayer.BossCounterText.SetText(StringToMessage("{CustomPlayer.QuestBossesDefeated}/{BossTarget}"))


    # Runs when the device is started in a running game
    OnBegin<override>()<suspends>:void=
        QuestPopUpFinal.RespondingButtonEvent.Subscribe(CheckFinalQuestStatus)
        for(Spawner : gm.PlayerSpawners):    
            Spawner.SpawnedEvent.Subscribe(OnPlayerJoined)
        #ChangeData.InteractedWithEvent.Subscribe(ChangeDatePress)
        for(i:=0..QuestBtns.Length - 1):
            if (Btn:= QuestBtns[i]):
                Btn.InteractedWithEvent.SubscribeAgent(OnQuestPress, i)
        for(i:=0..QuestPopupsAccept.Length-1):
            if(Popup := QuestPopupsAccept[i]):
                Popup.RespondingButtonEvent.SubscribeAgent7(OnQuestAccept, i)
        for(i:=0..QuestPopupsComplete.Length-1):
            if(Popup := QuestPopupsComplete[i]):
                Popup.RespondingButtonEvent.SubscribeAgent7(OnQuestComplete, i)
