using { /Fortnite.com/AI }
using { /Verse.org/Simulation }
using {/Fortnite.com/Characters }
using { /Fortnite.com/Animation/PlayAnimation }
using { /Verse.org/Assets }
using { /UnrealEngine.com/Temporary/SpatialMath }

npc_behaviour_gup := class(npc_behavior):

    var StareLocation : vector3 = vector3{X:=3072.990721, Y:=3312.0, Z:=100.0}
    var WepStareLocation : vector3 = vector3{X:=3432.0, Y:=4192.0, Z:=480.0}
    var PrestigeStareLocation : vector3 = vector3{X:=-2864.0, Y:=-1592.0, Z:=150.0}
    OnBegin<override>()<suspends>:void=
        if:
            NPCAgent := GetAgent[]
            NPCChar := NPCAgent.GetFortCharacter[]
            NPCNav := NPCChar.GetNavigatable[]
            NPCFocus := NPCChar.GetFocusInterface[]
            NPCAnim := NPCChar.GetPlayAnimationController[]
        then:
            NPCSpawnPoint := NPCChar.GetTransform().Translation
            CharLocation := NPCChar.GetTransform().Translation
            if(CharLocation.Y > 1023.0):
                if(CharLocation.Z < 350.0):
                    spawn{NPCFocus.MaintainFocus(StareLocation)}
                if(CharLocation.Z >= 350.0):
                    spawn{NPCFocus.MaintainFocus(WepStareLocation)}
            else:
                spawn{NPCFocus.MaintainFocus(PrestigeStareLocation)}
                    
    OnEnd<override>():void=
        block:
        #Print("Goodbye, NPC!")