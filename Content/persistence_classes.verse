using { /Fortnite.com/Devices }
using { /Verse.org/Simulation }
using { /UnrealEngine.com/Temporary/Diagnostics }

<# 
WHEN ADDING ANYTHING TO CUSTOM PLAYER IT MUST BE ADDED HERE (IDEALLY IN SAME PLACE)
MUST BE ADDED TO "MakePlayerSavedData"
MUST BE ADDED TO "SaveAllData"
MUST BE ADDED TO GAME MANAGER "LoadPlayerData"
#>



        <#
            DO NOT CHANGE THESE VALUES ANYMORE, JUST COMMENT/UNCOMMENT THE VALUES AT THE BOTTOM OF GAME_MANAGER
        #>


        
        #REVERT THIS BACK TO 9 PLEASE FOR HTE LOVE OF GOD DO NOT PUSH WITH ANYTHING OTHER THAN 9 VERSION
player_profile_data := class<final><persistable>:
    Version : int = 9 #CHANGE TO 9 EVERY TIME game_manager TOO

    #RESOURCES
    StoneAmount : float = 0.0 #CHANGE TO 0.0   #2.43e200 FOR TESTING    #2.43e51 For ~40 strength tiers
    ShardAmount : float = 0.0 #CHANGE TO 0.0 #PRESTIGE POINTS
    EggAmounts : []float = array{0.0,0.0,0.0,0.0,0.0,0.0} #CHANGE TO {0.0,0.0,0.0,0.0,0.0,0.0}
    WepMatAmount : float = 0.0 #CHANGE TO 0.0
    BossKeyAmount : float = 0.0  #CHANGE TO 0.0
    StardustAmount : float = 0.0

    #CAVE EXPLORATION
    StagesUnlocked : int = -1 #CHANGE TO -1
    MaxHP : float = 100.0 #CHANGE TO 100.0
    WepMatDropRate : float = 0.02 #CHANGE TO 0.02

    #AFK ZONE
    AFKStoneBoost : float = 0.0
    AFKPetsUnlocked : logic = false

    #STONE HARVEST UPGRADES
    Power : float = 1.0
    HarvestBuff : float = 0.0 #CHANGE TO 0.0  percentage bonus added to player power
    CritRate : float = 0.0  #CHANGE TO 0.0 || 1.0 IS 100%
    CritDMG : float = 1.5 #CHANGE TO 1.5 (2.0 IS TOO HIGH BASE BEACUSE OF HOW MANY MULTIPLIERS THERE ARE)
    FinalHitBonus : float = 0.20 #CHANGE TO 0.20
    MultiCrit : float = 1.0 #CHANGE TO 1.0

    ComboStrikeUnlocked : logic = false #CHANGE TO FALSE
    ComboStrikeMultiplier : float = 5.0
    ComboAccelerator : float = 1.0
    ComboStrikeStreak : float = 1.0

    RoomUnlockGather : float = 0.0

    #MOVEMENT SPEED STATS
    MovementSpeedLevel : int = 0 #CHANGE TO 0
    DoubleJumpUnlocked : logic = false
    DoubleJumpAmount : int = 0

    #WEAPON UPGRADES
    WeaponTier : int = 0 #CHANGE TO 0   #MAX LVL 9
    WeaponPerkTier : int = 0 #CHANGE TO 0

    WeaponStatLevel : float = 0.0 #HOW MANY TIMES UPGRADED STATS
    WeaponBonusLevel : float = 0.0 #HOW MANY TIMES UPGRADED BONUSES
    WeaponPerkLevel : float = 0.0 #HOW MANY TIMES UPGRADED PERKS
    
    WeaponStats : []float = array{0.0,0.0,0.0}
    WeaponBonuses : []float = array{0.0,0.0,0.0}

    StatCost : []float = array{2000000.0, 0.0} #STONE, EGG
    BonusCost : float = 10000000.0 #STONE
    PerkCost : float = 40000000.0 #STONE
    EvolveCost : []float = array{100.0, 2500000.0, 50.0} #LV. REQ, STONE, WEP MAT #CHANGE TO {40.0, 2500000.0, 50.0}

    PerksUnlocked : logic = false
    ActivePerks : []tuple(string, float) = array{
        ("Perk 1", 0.0),
        ("Perk 2", 0.0),
        ("Perk 3", 0.0)
    }
    
    #HUT UPGRADE INFO
    Hut1Levels : []int = array{0,0,0,0}
    Hut1Costs : []float = array{15000.0,250000.0,400000.0,80000000.0}
    Hut1Upgrades : []float = array{0.0,0.0,0.0,1.0} #THIS IS PURELY FOR THE HUD

    Hut2Levels : []int = array{0,0,0,0}
    Hut2Costs : []float = array{1.24e14,3.12e16,4.68e18,4.68e20}
    Hut2Upgrades : []float = array{0.0,0.0,1.0,1.0} #THIS IS PURELY FOR THE HUD

    Hut3Levels : []int = array{0,0,0,0}
    Hut3Costs : []float = array{1.95e25,3.95e26,2.95e25,1.95e25}
    Hut3Upgrades : []float = array{0.0,600.0,0.30,0.50} #THIS IS PURELY FOR THE HUD

    Hut4Levels : []int = array{0,0,0,0}

    Hut5Levels : []int = array{0,0,0,0}
    
    #GENERAL UPGRADE INFO
    GUP1Levels : []int = array{0,0}
    GUP1Costs : []float = array{2.50e05,2.34e19}
    GUP1Upgrades : []float = array{0.0,0.0} #THIS IS PURELY FOR THE HUD

    GUP2Levels : []int = array{0,0}
    GUP2Costs : []float = array{2.50e07,1.41e13}
    GUP2Upgrades : []float = array{0.0,0.0} #THIS IS PURELY FOR THE HUD

    GUP3Levels : []int = array{0,0}
    GUP3Costs : []float = array{7.50e10,3.12e17}
    GUP3Upgrades : []float = array{0.0,0.0} #THIS IS PURELY FOR THE HUD

    #PRESTIGE UPGRADE INFO
    PrestigeLevel : int = 0
    PrestigeCost : float = 1.12e16
    PrestigeHarvestBuff : float = 0.0 #THIS IS FOR INFO/HUD ONLY, BUFF IS ADDED TO HARVEST BUFF AUTOMATICALLY
    PrestigeStoneAmount : float = 0.0 #CHANGE TO 0.0
    PrestigeFix : logic = false
    
    PUPLevels : []int = array{0, 0, 0} #0 INDEX IS STRENGTH SURGE LEVELS
    PUPCosts : []float = array{1.0, 2.0, 4.0}
    PUPUpgrades : []float = array{0.0, 0.0, 0.0} #THIS IS PURELY FOR THE HUD

    EggPrestigeLevels : []int = array{0, 0, 0} 
    EggPrestigeCosts : []float = array{2.0, 1.0, 5.0}
    EggPrestigeUpgrades : []float = array{0.0, 0.0, 0.0} 

    PUPQOLLevels : []int = array{0, 0, 0}

    AutoGatherRate : float = 0.0
    BossRewardMult : float = 1.0

    EggAmountMultiplier : float = 0.0
    EggDropRateMultiplier : float = 0.0
    #STRENGTH UPGRADES

    #CHANGE IN GAME MANAGER
    BaseStrengthCosts : []float = array{
        15.0,   150.0,  800.0,  4.80e03,3.00e05,
        1.80e06,3.00e07,1.80e08,4.50e09,2.70e10,
        4.50e11,2.70e12,3.38e13,2.03e14,1.87e16,
        1.12e17,3.75e18,2.25e19,1.12e21,6.74e21,
        2.25e23,1.35e24,3.37e25,2.02e26,3.75e28,
        2.25e29,1.12e31,6.74e31,5.06e33,3.03e34,
        1.52e36,9.10e36,3.41e38,2.05e39,5.31e40,
        3.18e41,2.12e42,1.27e43,1.27e44,2.64e45, #7.64e45

        3.41e49, 2.05e51, 1.23e53, 7.37e54, 4.42e56, 
        2.65e58, 3.47e59, 2.08e61, 1.25e62, 7.50e63, 
        4.50e64, 2.70e65, 9.71e65, 4.63e66, 2.78e68, 
        1.67e70, 9.99e71, 6.00e72, 1.45e74, 8.70e76, 

        5.22e78, 3.13e79, 1.88e81, 1.13e83, 1.67e84, 
        3.00e85, 8.01e86, 7.61e88, 8.16e90, 9.30e91, 
        7.09e92, 8.51e93, 6.91e94, 5.34e95, 8.41e96, 
        8.44e97, 5.06e98, 3.04e99, 1.82e100, 1.09e101,
        
        9.81e105, 8.83e107, 7.95e109, 7.15e111, 6.44e113, 
        5.79e115, 5.21e117, 4.69e119, 4.22e121, 3.80e123, 
        3.42e125, 3.08e127, 2.77e129, 2.49e131, 2.24e133, 
        2.02e135, 1.82e137, 1.64e139, 1.47e141, 1.33e143, 
        1.19e145, 1.07e147, 9.66e148, 8.69e150, 7.83e152, 
        7.04e154, 6.34e156, 5.70e158, 5.13e160, 4.62e162     
        } 

    #NEVER USED
    StrengthUpgrades : []float = array{
        0.0, 0.0, 0.0, 0.0, 0.0,
        0.0, 0.0, 0.0, 0.0, 0.0,
        0.0, 0.0, 0.0, 0.0, 0.0,
        0.0, 0.0, 0.0, 0.0, 0.0,
        0.0, 0.0, 0.0, 0.0, 0.0,
        0.0, 0.0, 0.0, 0.0, 0.0,
        0.0, 0.0, 0.0, 0.0, 0.0,
        0.0, 0.0, 0.0, 0.0, 0.0,

        0.0, 0.0, 0.0, 0.0, 0.0,
        0.0, 0.0, 0.0, 0.0, 0.0,
        0.0, 0.0, 0.0, 0.0, 0.0,
        0.0, 0.0, 0.0, 0.0, 0.0,
        0.0, 0.0, 0.0, 0.0, 0.0,
        0.0, 0.0, 0.0, 0.0, 0.0,
        0.0, 0.0, 0.0, 0.0, 0.0,
        0.0, 0.0, 0.0, 0.0, 0.0,
        
        0.0, 0.0, 0.0, 0.0, 0.0,
        0.0, 0.0, 0.0, 0.0, 0.0,
        0.0, 0.0, 0.0, 0.0, 0.0,
        0.0, 0.0, 0.0, 0.0, 0.0,
        0.0, 0.0, 0.0, 0.0, 0.0,
        0.0, 0.0, 0.0, 0.0, 0.0
        }

    #CHANGE THESE BACK PLEASE
    StrengthEfficiencies : []int = array{
        1, 1, 1, 1, 1,
        1, 1, 1, 1, 1,
        1, 1, 1, 1, 1,
        1, 1, 1, 1, 1,
        1, 1, 1, 1, 1,
        1, 1, 1, 1, 1,
        1, 1, 1, 1, 1,
        1, 1, 1, 1, 1,


        1, 1, 1, 1, 1,
        1, 1, 1, 1, 1,
        1, 1, 1, 1, 1,
        1, 1, 1, 1, 1,
        1, 1, 1, 1, 1,
        1, 1, 1, 1, 1,
        1, 1, 1, 1, 1,
        1, 1, 1, 1, 1,
        
        1, 1, 1, 1, 1,
        1, 1, 1, 1, 1,
        1, 1, 1, 1, 1,
        1, 1, 1, 1, 1,
        1, 1, 1, 1, 1,
        1, 1, 1, 1, 1
        }

    StrengthUpgradeLevels : []int = array{
        0,0,0,0,0,
        0,0,0,0,0,
        0,0,0,0,0,
        0,0,0,0,0,
        0,0,0,0,0,
        0,0,0,0,0,
        0,0,0,0,0,
        0,0,0,0,0,

        0,0,0,0,0,
        0,0,0,0,0,
        0,0,0,0,0,
        0,0,0,0,0,
        0,0,0,0,0,
        0,0,0,0,0,
        0,0,0,0,0,
        0,0,0,0,0,
        
        0,0,0,0,0,
        0,0,0,0,0,
        0,0,0,0,0,
        0,0,0,0,0,
        0,0,0,0,0,
        0,0,0,0,0
        }

    #EARTHQUAKE STATS
    EarthquakeUnlocked : logic = false #CHANGE TO FALSE
    EarthquakeDistance : float = 600.0 #CHANGE TO 600
    EarthquakeGatherAmount : float = 0.20
    EarthquakeDamage : float = 0.60 #% BASED ON PLAYER DAMAGE

    #VOID FRACTURE STATS
    VoidUnlocked : logic = false
    VoidDamage : float = 0.50 #% BASED ON PLAYER DAMAGE
    VoidDuration : float = 3.0 #SECONDS
    VoidStacks : float = 1.0

    A4Unlocked : logic = false #CHANGE TO FALSE #If you've ever been to A4
    A5Unlocked : logic = false #CHANGE TO FALSE #If you've ever been to A5

    #INFINITE SINKS

    SinkPrestigePower : float = 0.0
    SinkWepMatPower : float = 0.0
    SinkBossKeyPower : float = 0.0

    #REAL TIME DATA
    CurrentDay : int = -1
    LastClaimed : time_and_date = time_and_date{Year:=-1,Month:=-1,Day:=-1,Hours:=-1,Minutes:=-1,Seconds:=-1}
    HatPrestiges : int = 0
    HatsUnlocked : [int]logic = map{
        0 => false,
        1 => false,
        2 => false,
        3 => false,
        4 => false,
        5 => false,
        6 => false,
        7 => false,
        8 => false,
        9 => false,
        10 => false,
        11 => false,
        12 => false,
        13 => false,
        14 => false,
        15 => false,
        16 => false,
        17 => false,
        18 => false,
        19 => false,
        20 => false,
        21 => false,
        22 => false,
        23 => false,
        24 => false,
        25 => false,
        26 => false,
        27 => false,
        28 => false,
        29 => false,
        30 => false,
        31 => false,
        32 => false,
        33 => false,
        34 => false,
        35 => false,
        36 => false,
        37 => false,
        38 => false,
        39 => false,
        40 => false,
        41 => false,
        42 => false,
        43 => false,
        44 => false,
        45 => false,
        46 => false,
        47 => false,
        48 => false,
        49 => false,
        50 => false,
        51 => false,
        52 => false,
        53 => false,
        54 => false,
        55 => false,
        56 => false,
        57 => false,
        58 => false,
        59 => false,
        60 => false,
        61 => false,
        62 => false,
        63 => false,
        64 => false,
        65 => false,
        66 => false,
        67 => false,
        68 => false,
        69 => false,
        70 => false,
        71 => false,
        72 => false,
        73 => false,
        74 => false,
        75 => false,
        76 => false,
        77 => false,
        78 => false,
        79 => false
    }
    CodeHatsUnlocked : [int]logic = map{
        0=>false,
        1=>false
    }

    #NEW DAILY REWARDS

    QuestRocksDestroyed : int = 0
    QuestMobsDefeated : int = 0
    QuestBossesDefeated : int = 0
    QuestFinalStatus : logic = false
    QuestsStarted : []logic = array{false,false,false}
    QuestsCompleted : []logic = array{false,false,false}
    LastCompletedDate : date_data = date_data{}

    #CHRISTMAS VILLAGE

    LastCompletedHolidayDate : date_data = date_data{}

    #EXPEDITIONS
    NewEXPPetsLevel : [int]float = map: #CHANGE TO 0.0, 0.0 #0 IS DAVE, ALWAYS 0.0
        0 => 0.0,
        1 => 0.0,
        2 => 0.0
    EXPPetsLevel : []float = array{0.0, 0.0, 0.0} #CHANGE TO 0.0, 0.0 #0 IS DAVE, ALWAYS 0.0
    CurrentExpedition : int = -1
    EXPCurrentPet : int = 0
    ExpeditionTimeLeft : float = 0.0
    EXPTotalRewards : []float = array{0.0,0.0,0.0,0.0}

    #GOLDEN DRILL
    GoldenDrillUnlocked : logic = false
    DrillActivations : int = 0 #HOW MANY TIMES TOTAL
    BuffMultiplier : float = 2.0 #100% BONUS TO START (1.0 wouldn't do anything)
    BuffTimeSeconds : float = 30.0
    BuffChargeTimer : float = 0.0 # CHANGE TO 0.0  #15 MINUTE CHARGE TIME
    BuffCurrentCharges : float = 1.0
    BuffMaxCharges : float = 1.0
    GoldenDrillLevels : []int = array{0,0,0,0}
    GoldenDrillCosts : []float = array{7.50e07,5.63e11,7.50e07,4.68e18}
    GoldenDrillUpgrades : []float = array{0.0,1.0,30.0,1.0} #THIS IS PURELY FOR THE HUD

    #SECRETS
    #DO NOT USE THIS ONE
    SecretsClaimed : []logic = array{false,false,false,false,false,false}

    #USE THIS ONE
    NewSecretsClaimed : [int]logic = map{
        0 => false,
        1 => false,
        2 => false,
        3 => false,
        4 => false,
        5 => false,
        6 => false,
        7 => false,
        8 => false,
        9 => false,
        10 => false,
        11 => false,
        12 => false,
        13 => false,
        14 => false,
        15 => false,
        16 => false,
        17 => false,
        18 => false,
        19 => false,
        20 => false,
        21 => false,
        22 => false,
        23 => false,
        24 => false,
        25 => false,
        26 => false,
        27 => false,
        28 => false,
        29 => false
    }
    #PET INFO
    EggsUnlocked : logic = false #CHANGE TO FALSE
    EggDropAmount : []float = array{1.0, 2.0}
    RockEggDropChance : float = 1.0 #CHANGE TO 1.0
    MobEggDropChance : float = 5.0 #CHANGE TO 5.0
    PetDataReset : logic = false

    PetEquipSlots : []logic = array{true, false, false, false, false, false} #CHANGE TO 1st TRUE REST FALSE

    PetSlots : []int = array{-1, -1, -1, -1, -1, -1} #CHANGE TO array{-1,-1,-1,-1,-1,-1}
    PetUnlocked : []logic = array{false,false,false,false,false,false}
    PetLevels : []int = array{0,0,0,0,0,0} #MAX TEST: 5000,7500,8000,10000,12500,3000   #PLAYTEST: 2500,3500,4000,8000,4000,1000
    PetEvolveLevels : []int = array{0,0,0,0,0,0}
    PetGatherRates : []float = array{0.0,0.0,0.0,0.0,0.0,0.0}
    PetAttackRates : []float = array{0.0,0.0,0.0,0.0,0.0,0.0}
    PetCritRates : []float = array{0.0,0.0,0.0,0.0,0.0,0.0}
    PetCritDMG : []float = array{0.0,0.0,0.0,0.0,0.0,0.0}
    PetFinalStrike : []float = array{0.0,0.0,0.0,0.0,0.0,0.0}

    EggsHatched : []float = array{0.0,0.0,0.0,0.0,0.0,0.0}

    HidePets : logic = false

    #BOSS INFO
    HighestFBDamage : float = 0.0 #CHANGE TO 0.0

    AllBossesDefeated : [int]logic = map:
        0 => false #Mushroom King
        1 => false #Razor Fang
        2 => false #Cactoro
        3 => false #Prestige Boss
        4 => false #Brambleback
        5 => false #Grimroot (Halloween Seasonal)

    DemonPetObtained : logic = false

    #TUTORIAL
    TutorialLevel : int = 0 # CHANGE TO 0
    TutorialProgress : float = 0.0

    CurrentMusic : string = "Default"

    #ACHIEVEMENTS
    AchLevels : []int = array{1, 1, 1, 1, 1, 1, 1, 1}
    AchBuffTotal : float = 0.0 #VISUAL FOR HUD ONLY, USE "HarvestBuff"
    Ach1 : []float = array{0.0,1000000.0} #Total Stone Gathered #CHANGE TO 1000000.0
    Ach2 : []float = array{0.0,1000.0} #Max Strength
    Ach3 : []float = array{0.0,100.0} #Total Mobs Defeated
    Ach4 : []float = array{0.0,10.0} #Prestige Points Obtained
    Ach5 : []float = array{0.0,1000.0} #Total Rocks Destroyed #CHANGE TO 1000.0
    Ach6 : []float = array{0.0,10.0} #Total Pet Level
    Ach7 : []float = array{0.0,1.0} #Total Bosses Defeated
    Ach8 : []float = array{0.0,1800.0} #Total Time Played IN SECONDS, CHANGE TO 1800 = 30 MINUTES

    PlayTimeSeconds : float = 0.0

    OneTimeAnalytics : [string]logic = map{
        "Play10Min"=>false,
        "Play30Min"=>false,
        "Play60Min"=>false,
        "Play120Min"=>false,
        "Play240Min"=>false,
        "Play6000Min"=>false,
        "Play30000Min"=>false,
        "DestroyMeteor"=>false,
        "FightPrestigeBoss"=>false,
        "Reach80StrTiers"=>false,
        "Reach80SurgeTiers"=>false,
        "FightHalloweenBoss"=>false
    }

var MySavedPlayerData:weak_map(player, player_profile_data) = map{}

MakePlayerSavedData<constructor>(Src:player_profile_data)<transacts> := player_profile_data:
    Version:= Src.Version
    #FloatMap := Src.FloatMap
    #IntMap := Src.IntMap
    StoneAmount:= Src.StoneAmount
    ShardAmount:= Src.ShardAmount
    EggAmounts:= Src.EggAmounts
    WepMatAmount:= Src.WepMatAmount
    StagesUnlocked:= Src.StagesUnlocked
    StardustAmount := Src.StardustAmount
    MaxHP:= Src.MaxHP
    WepMatDropRate:= Src.WepMatDropRate
    AFKStoneBoost:= Src.AFKStoneBoost
    AFKPetsUnlocked:= Src.AFKPetsUnlocked
    Power:= Src.Power
    HarvestBuff:= Src.HarvestBuff
    CritRate:= Src.CritRate
    CritDMG:= Src.CritDMG
    FinalHitBonus:= Src.FinalHitBonus
    MultiCrit:= Src.MultiCrit
    ComboStrikeUnlocked:= Src.ComboStrikeUnlocked
    ComboStrikeMultiplier:= Src.ComboStrikeMultiplier
    ComboAccelerator:= Src.ComboAccelerator
    ComboStrikeStreak:= Src.ComboStrikeStreak
    RoomUnlockGather:= Src.RoomUnlockGather
    MovementSpeedLevel:= Src.MovementSpeedLevel
    DoubleJumpUnlocked:= Src.DoubleJumpUnlocked
    DoubleJumpAmount:= Src.DoubleJumpAmount
    WeaponTier:= Src.WeaponTier
    WeaponPerkTier:= Src.WeaponPerkTier
    WeaponStatLevel:= Src.WeaponStatLevel
    WeaponBonusLevel:= Src.WeaponBonusLevel
    WeaponPerkLevel:= Src.WeaponPerkLevel
    WeaponStats:= Src.WeaponStats
    WeaponBonuses:= Src.WeaponBonuses
    StatCost:= Src.StatCost
    BonusCost:= Src.BonusCost
    PerkCost:= Src.PerkCost
    EvolveCost:= Src.EvolveCost
    PerksUnlocked:= Src.PerksUnlocked
    ActivePerks:= Src.ActivePerks
    Hut1Levels:= Src.Hut1Levels
    Hut1Costs:= Src.Hut1Costs
    Hut1Upgrades:= Src.Hut1Upgrades
    Hut2Levels:= Src.Hut2Levels
    Hut2Costs:= Src.Hut2Costs
    Hut2Upgrades:= Src.Hut2Upgrades
    Hut3Levels:= Src.Hut3Levels
    Hut3Costs:= Src.Hut3Costs
    Hut3Upgrades:= Src.Hut3Upgrades
    Hut4Levels:= Src.Hut4Levels
    Hut5Levels:= Src.Hut5Levels
    GUP1Levels:= Src.GUP1Levels
    GUP1Costs:= Src.GUP1Costs
    GUP1Upgrades:= Src.GUP1Upgrades
    GUP2Levels:= Src.GUP2Levels
    GUP2Costs:= Src.GUP2Costs
    GUP2Upgrades:= Src.GUP2Upgrades
    GUP3Levels:= Src.GUP3Levels
    GUP3Costs:= Src.GUP3Costs
    GUP3Upgrades:= Src.GUP3Upgrades
    PrestigeLevel:= Src.PrestigeLevel
    PrestigeFix:= Src.PrestigeFix
    PrestigeCost:= Src.PrestigeCost
    PrestigeHarvestBuff:= Src.PrestigeHarvestBuff
    PrestigeStoneAmount:= Src.PrestigeStoneAmount
    PUPLevels:= Src.PUPLevels
    PUPCosts:= Src.PUPCosts
    PUPUpgrades:= Src.PUPUpgrades
    EggPrestigeLevels:= Src.EggPrestigeLevels
    EggPrestigeCosts:= Src.EggPrestigeCosts
    EggPrestigeUpgrades:= Src.EggPrestigeUpgrades
    PUPQOLLevels:= Src.PUPQOLLevels
    PetEquipSlots:= Src.PetEquipSlots
    AutoGatherRate:= Src.AutoGatherRate
    BossRewardMult:= Src.BossRewardMult
    EggAmountMultiplier:= Src.EggAmountMultiplier
    EggDropRateMultiplier:= Src.EggDropRateMultiplier
    BaseStrengthCosts:= Src.BaseStrengthCosts
    StrengthUpgrades:= Src.StrengthUpgrades
    StrengthEfficiencies:= Src.StrengthEfficiencies
    StrengthUpgradeLevels:= Src.StrengthUpgradeLevels
    EarthquakeUnlocked:= Src.EarthquakeUnlocked
    EarthquakeDistance:= Src.EarthquakeDistance
    EarthquakeGatherAmount:= Src.EarthquakeGatherAmount
    EarthquakeDamage:= Src.EarthquakeDamage
    VoidUnlocked:= Src.VoidUnlocked
    VoidDamage:= Src.VoidDamage
    VoidDuration:= Src.VoidDuration
    VoidStacks:= Src.VoidStacks
    A4Unlocked:= Src.A4Unlocked
    A5Unlocked:= Src.A5Unlocked
    CurrentDay := Src.CurrentDay
    SinkPrestigePower := Src.SinkPrestigePower
    SinkWepMatPower := Src.SinkWepMatPower
    SinkBossKeyPower := Src.SinkBossKeyPower
    LastClaimed := Src.LastClaimed
    HatsUnlocked := Src.HatsUnlocked
    HatPrestiges := Src.HatPrestiges
    CodeHatsUnlocked :=Src.CodeHatsUnlocked
    QuestRocksDestroyed := Src.QuestRocksDestroyed
    QuestMobsDefeated := Src.QuestMobsDefeated
    QuestBossesDefeated := Src.QuestBossesDefeated
    QuestFinalStatus := Src.QuestFinalStatus
    QuestsStarted := Src.QuestsStarted
    QuestsCompleted := Src.QuestsCompleted
    LastCompletedDate := Src.LastCompletedDate
    LastCompletedHolidayDate := Src.LastCompletedHolidayDate
    EXPPetsLevel:= Src.EXPPetsLevel
    NewEXPPetsLevel:= Src.NewEXPPetsLevel
    CurrentExpedition := Src.CurrentExpedition
    EXPCurrentPet := Src.EXPCurrentPet
    ExpeditionTimeLeft := Src.ExpeditionTimeLeft
    EXPTotalRewards := Src.EXPTotalRewards
    GoldenDrillUnlocked := Src.GoldenDrillUnlocked
    DrillActivations := Src.DrillActivations
    BuffMultiplier := Src.BuffMultiplier
    BuffTimeSeconds := Src.BuffTimeSeconds
    BuffChargeTimer := Src.BuffChargeTimer
    BuffCurrentCharges := Src.BuffCurrentCharges
    BuffMaxCharges := Src.BuffMaxCharges
    GoldenDrillLevels := Src.GoldenDrillLevels
    GoldenDrillCosts := Src.GoldenDrillCosts
    GoldenDrillUpgrades := Src.GoldenDrillUpgrades
    SecretsClaimed := Src.SecretsClaimed
    NewSecretsClaimed := Src.NewSecretsClaimed
    EggsUnlocked:= Src.EggsUnlocked
    RockEggDropChance:= Src.RockEggDropChance
    EggDropAmount:= Src.EggDropAmount
    MobEggDropChance:= Src.MobEggDropChance
    PetDataReset := Src.PetDataReset
    PetUnlocked:= Src.PetUnlocked
    PetSlots:= Src.PetSlots
    HidePets:= Src.HidePets
    PetLevels:= Src.PetLevels
    PetEvolveLevels:= Src.PetEvolveLevels
    PetGatherRates:= Src.PetGatherRates
    PetAttackRates:= Src.PetAttackRates
    PetCritRates:= Src.PetCritRates
    PetCritDMG:= Src.PetCritDMG
    PetFinalStrike:= Src.PetFinalStrike
    EggsHatched:= Src.EggsHatched
    HighestFBDamage:= Src.HighestFBDamage
    AllBossesDefeated:= Src.AllBossesDefeated
    DemonPetObtained:= Src.DemonPetObtained
    BossKeyAmount:= Src.BossKeyAmount
    TutorialLevel:= Src.TutorialLevel
    CurrentMusic:= Src.CurrentMusic
    TutorialProgress:= Src.TutorialProgress
    AchLevels:= Src.AchLevels
    AchBuffTotal:= Src.AchBuffTotal
    Ach1 := Src.Ach1
    Ach2 := Src.Ach2
    Ach3 := Src.Ach3
    Ach4 := Src.Ach4
    Ach5 := Src.Ach5
    Ach6 := Src.Ach6
    Ach7 := Src.Ach7
    Ach8 := Src.Ach8
    PlayTimeSeconds:= Src.PlayTimeSeconds
    OneTimeAnalytics:= Src.OneTimeAnalytics
        
SaveVersion(Agent:agent, NewValue:int):void=
    if(Player := player[Agent], Player.IsActive[]):
        if:
            CheckSaveDataForPlayer[Player]
            SourceData := MySavedPlayerData[Player]
        
            set MySavedPlayerData[Player] = player_profile_data:
                MakePlayerSavedData<constructor>(SourceData)
                Version := NewValue
    else:
        Print("SAVE VERSION PLAYER NOT ACTIVE")

<#SaveResources(Agent:agent, NewStone:float, NewEggs:[]float, NewWepMats:float, NewShards:float, NewBossKeys:float):void=
    if(Player := player[Agent], Player.IsActive[]):
        if:
            CheckSaveDataForPlayer[Player]
            SourceData := MySavedPlayerData[Player]

            set MySavedPlayerData[Player] = player_profile_data:
                MakePlayerSavedData<constructor>(SourceData)
                StoneAmount := NewStone
                EggAmounts := NewEggs
                WepMatAmount := NewWepMats
                ShardAmount := NewShards
                BossKeyAmount := NewBossKeys
    else:
        Print("SAVE RESOURCES PLAYER NOT ACTIVE")#>

<#SaveAchievements(Agent:agent, NewAchLevels:[]int, NewAchBuffTotal:float, NewAllAchs:[][]float):void=
    if(Player := player[Agent], Player.IsActive[]):
        if:
            CheckSaveDataForPlayer[Player]
            SourceData := MySavedPlayerData[Player]
            NewAch1 := NewAllAchs[0]
            NewAch2 := NewAllAchs[1]
            NewAch3 := NewAllAchs[2]
            NewAch4 := NewAllAchs[3]
            NewAch5 := NewAllAchs[4]
            NewAch6 := NewAllAchs[5]
            NewAch7 := NewAllAchs[6]
            NewAch8 := NewAllAchs[7]
 
            set MySavedPlayerData[Player] = player_profile_data:
                MakePlayerSavedData<constructor>(SourceData)
                AchLevels := NewAchLevels
                AchBuffTotal := NewAchBuffTotal
                Ach1:= NewAch1
                Ach2:= NewAch2
                Ach3:= NewAch3
                Ach4:= NewAch4
                Ach5:= NewAch5
                Ach6:= NewAch6
                Ach7:= NewAch7
                Ach8:= NewAch8
    else:
        Print("SAVE ACHIEVEMENTS PLAYER NOT ACTIVE")#>

SaveAllData(CustomPlayer:Custom_Player):void=
    if:
        Player := CustomPlayer.Player
        Player.IsActive[]
        SourceData := MySavedPlayerData[Player]
        #CheckSaveDataForPlayer[Player]
        NewAch1 := CustomPlayer.AllAchs[0]
        NewAch2 := CustomPlayer.AllAchs[1]
        NewAch3 := CustomPlayer.AllAchs[2]
        NewAch4 := CustomPlayer.AllAchs[3]
        NewAch5 := CustomPlayer.AllAchs[4]
        NewAch6 := CustomPlayer.AllAchs[5]
        NewAch7 := CustomPlayer.AllAchs[6]
        NewAch8 := CustomPlayer.AllAchs[7]
        SaveData := player_profile_data:
            MakePlayerSavedData<constructor>(SourceData)
            StoneAmount:= CustomPlayer.StoneAmount
            ShardAmount:= CustomPlayer.ShardAmount
            EggAmounts:= CustomPlayer.EggAmounts
            WepMatAmount:= CustomPlayer.WepMatAmount
            StagesUnlocked:= CustomPlayer.StagesUnlocked
            StardustAmount:= CustomPlayer.StardustAmount
            MaxHP:= CustomPlayer.MaxHP
            WepMatDropRate:= CustomPlayer.WepMatDropRate
            AFKStoneBoost:= CustomPlayer.AFKStoneBoost
            AFKPetsUnlocked:= CustomPlayer.AFKPetsUnlocked
            Power:= CustomPlayer.Power
            HarvestBuff:= CustomPlayer.HarvestBuff
            CritRate:= CustomPlayer.CritRate
            CritDMG:= CustomPlayer.CritDMG
            FinalHitBonus:= CustomPlayer.FinalHitBonus
            MultiCrit:= CustomPlayer.MultiCrit
            ComboStrikeUnlocked:= CustomPlayer.ComboStrikeUnlocked
            ComboStrikeMultiplier:= CustomPlayer.ComboStrikeMultiplier
            ComboAccelerator:= CustomPlayer.ComboAccelerator
            ComboStrikeStreak:= CustomPlayer.ComboStrikeStreak
            RoomUnlockGather:= CustomPlayer.RoomUnlockGather
            MovementSpeedLevel:= CustomPlayer.MovementSpeedLevel
            DoubleJumpUnlocked:= CustomPlayer.DoubleJumpUnlocked
            DoubleJumpAmount:= CustomPlayer.DoubleJumpAmount
            WeaponTier:= CustomPlayer.WeaponTier
            WeaponPerkTier:= CustomPlayer.WeaponPerkTier
            WeaponStatLevel:= CustomPlayer.WeaponStatLevel
            WeaponBonusLevel:= CustomPlayer.WeaponBonusLevel
            WeaponPerkLevel:= CustomPlayer.WeaponPerkLevel
            WeaponStats:= CustomPlayer.WeaponStats
            WeaponBonuses:= CustomPlayer.WeaponBonuses
            <#StatCost:= CustomPlayer.StatCost
            BonusCost:= CustomPlayer.BonusCost
            PerkCost:= CustomPlayer.PerkCost
            EvolveCost:= CustomPlayer.EvolveCost#>
            PerksUnlocked:= CustomPlayer.PerksUnlocked
            ActivePerks:= CustomPlayer.ActivePerks
            Hut1Levels:= CustomPlayer.Hut1Levels
            Hut1Costs:= CustomPlayer.Hut1Costs
            #Hut1Upgrades:= CustomPlayer.Hut1Upgrades
            Hut2Levels:= CustomPlayer.Hut2Levels
            Hut2Costs:= CustomPlayer.Hut2Costs
            #Hut2Upgrades:= CustomPlayer.Hut2Upgrades
            Hut3Levels:= CustomPlayer.Hut3Levels
            Hut3Costs:= CustomPlayer.Hut3Costs
            Hut4Levels:= CustomPlayer.Hut4Levels
            Hut5Levels:= CustomPlayer.Hut5Levels
            #Hut3Upgrades:= CustomPlayer.Hut3Upgrades
            GUP1Levels:= CustomPlayer.GUP1Levels
            #GUP1Costs:= CustomPlayer.GUP1Costs
            #GUP1Upgrades:= CustomPlayer.GUP1Upgrades
            GUP2Levels:= CustomPlayer.GUP2Levels
            #GUP2Costs:= CustomPlayer.GUP2Costs
            #GUP2Upgrades:= CustomPlayer.GUP2Upgrades
            GUP3Levels:= CustomPlayer.GUP3Levels
            #GUP3Costs:= CustomPlayer.GUP3Costs
            #GUP3Upgrades:= CustomPlayer.GUP3Upgrades
            PrestigeLevel:= CustomPlayer.PrestigeLevel
            PrestigeFix:= CustomPlayer.PrestigeFix
            #PrestigeCost:= CustomPlayer.PrestigeCost
            PrestigeHarvestBuff:= CustomPlayer.PrestigeHarvestBuff
            PrestigeStoneAmount:= CustomPlayer.PrestigeStoneAmount
            PUPLevels:= CustomPlayer.PUPLevels
            #PUPCosts:= CustomPlayer.PUPCosts
            #PUPUpgrades:= CustomPlayer.PUPUpgrades
            EggPrestigeLevels:= CustomPlayer.EggPrestigeLevels
            #EggPrestigeCosts:= CustomPlayer.EggPrestigeCosts
            #EggPrestigeUpgrades:= CustomPlayer.EggPrestigeUpgrades
            PUPQOLLevels:= CustomPlayer.PUPQOLLevels
            PetEquipSlots:= CustomPlayer.PetEquipSlots
            AutoGatherRate:= CustomPlayer.AutoGatherRate
            BossRewardMult:= CustomPlayer.BossRewardMult
            EggAmountMultiplier:= CustomPlayer.EggAmountMultiplier
            EggDropRateMultiplier:= CustomPlayer.EggDropRateMultiplier
            BaseStrengthCosts:= CustomPlayer.BaseStrengthCosts
            #StrengthUpgrades:= CustomPlayer.StrengthUpgrades
            StrengthEfficiencies:= CustomPlayer.StrengthEfficiencies
            StrengthUpgradeLevels:= CustomPlayer.StrengthUpgradeLevels
            EarthquakeUnlocked:= CustomPlayer.EarthquakeUnlocked
            EarthquakeDistance:= CustomPlayer.EarthquakeDistance
            EarthquakeGatherAmount:= CustomPlayer.EarthquakeGatherAmount
            EarthquakeDamage:= CustomPlayer.EarthquakeDamage
            VoidUnlocked:= CustomPlayer.VoidUnlocked
            #VoidDamage:= CustomPlayer.VoidDamage
            VoidDuration:= CustomPlayer.VoidDuration
            VoidStacks:= CustomPlayer.VoidStacks
            A4Unlocked:= CustomPlayer.A4Unlocked
            A5Unlocked:= CustomPlayer.A5Unlocked
            CurrentDay := CustomPlayer.CurrentDay
            SinkPrestigePower := CustomPlayer.SinkPrestigePower
            SinkWepMatPower := CustomPlayer.SinkWepMatPower
            SinkBossKeyPower := CustomPlayer.SinkBossKeyPower
            LastClaimed := CustomPlayer.LastClaimed
            HatsUnlocked := CustomPlayer.HatsUnlocked
            HatPrestiges := CustomPlayer.HatPrestiges
            CodeHatsUnlocked :=CustomPlayer.CodeHatsUnlocked
            EXPPetsLevel:= CustomPlayer.EXPPetsLevel
            NewEXPPetsLevel:= CustomPlayer.NewEXPPetsLevel
            QuestRocksDestroyed := CustomPlayer.QuestRocksDestroyed
            QuestMobsDefeated := CustomPlayer.QuestMobsDefeated
            QuestBossesDefeated := CustomPlayer.QuestBossesDefeated
            QuestFinalStatus := CustomPlayer.QuestFinalStatus
            QuestsStarted := CustomPlayer.QuestsStarted
            QuestsCompleted := CustomPlayer.QuestsCompleted
            LastCompletedDate := CustomPlayer.LastCompletedDate
            LastCompletedHolidayDate := CustomPlayer.LastCompletedHolidayDate
            CurrentExpedition:= CustomPlayer.CurrentExpedition
            EXPCurrentPet:= CustomPlayer.EXPCurrentPet
            ExpeditionTimeLeft:= CustomPlayer.ExpeditionTimeLeft
            EXPTotalRewards:= CustomPlayer.EXPTotalRewards
            GoldenDrillUnlocked := CustomPlayer.GoldenDrillUnlocked
            DrillActivations := CustomPlayer.DrillActivations
            BuffMultiplier := CustomPlayer.BuffMultiplier
            BuffTimeSeconds := CustomPlayer.BuffTimeSeconds
            BuffChargeTimer := CustomPlayer.BuffChargeTimer
            BuffCurrentCharges := CustomPlayer.BuffCurrentCharges
            BuffMaxCharges := CustomPlayer.BuffMaxCharges
            GoldenDrillLevels := CustomPlayer.GoldenDrillLevels
            #GoldenDrillCosts := CustomPlayer.GoldenDrillCosts
            #GoldenDrillUpgrades := CustomPlayer.GoldenDrillUpgrades
            SecretsClaimed := CustomPlayer.SecretsClaimed
            NewSecretsClaimed := CustomPlayer.NewSecretsClaimed
            EggsUnlocked:= CustomPlayer.EggsUnlocked
            RockEggDropChance:= CustomPlayer.RockEggDropChance
            EggDropAmount:= CustomPlayer.EggDropAmount
            MobEggDropChance:= CustomPlayer.MobEggDropChance
            PetDataReset := CustomPlayer.PetDataReset
            PetUnlocked:= CustomPlayer.PetUnlocked
            PetSlots:= CustomPlayer.PetSlots
            HidePets:= CustomPlayer.HidePets
            PetLevels:= CustomPlayer.PetLevels
            PetEvolveLevels:= CustomPlayer.PetEvolveLevels
            PetGatherRates:= CustomPlayer.PetGatherRates
            PetAttackRates:= CustomPlayer.PetAttackRates
            PetCritRates:= CustomPlayer.PetCritRates
            PetCritDMG:= CustomPlayer.PetCritDMG
            PetFinalStrike:= CustomPlayer.PetFinalStrike
            EggsHatched:= CustomPlayer.EggsHatched
            HighestFBDamage:= CustomPlayer.HighestFBDamage
            AllBossesDefeated:= CustomPlayer.AllBossesDefeated
            DemonPetObtained:= CustomPlayer.DemonPetObtained
            BossKeyAmount:= CustomPlayer.BossKeyAmount
            TutorialLevel:= CustomPlayer.TutorialLevel
            CurrentMusic:= CustomPlayer.CurrentMusic
            TutorialProgress:= CustomPlayer.TutorialProgress
            AchLevels:= CustomPlayer.AchLevels
            AchBuffTotal:= CustomPlayer.AchBuffTotal
            Ach1:= NewAch1
            Ach2:= NewAch2
            Ach3:= NewAch3
            Ach4:= NewAch4
            Ach5:= NewAch5
            Ach6:= NewAch6
            Ach7:= NewAch7
            Ach8:= NewAch8
            PlayTimeSeconds:= CustomPlayer.PlayTimeSeconds
            OneTimeAnalytics:= CustomPlayer.OneTimeAnalytics
        FitsInPlayerMap[SaveData]
        set MySavedPlayerData[Player] = SaveData
    else:
        Print("SOMETHING WENT WRONG")
    
    <#if:
        Player := CustomPlayer.Player
        Player.IsActive[]
        CheckSaveDataForPlayer[Player]
        SourceData := MySavedPlayerData[Player]
        NewAch1 := CustomPlayer.AllAchs[0]
        NewAch2 := CustomPlayer.AllAchs[1]
        NewAch3 := CustomPlayer.AllAchs[2]
        NewAch4 := CustomPlayer.AllAchs[3]
        NewAch5 := CustomPlayer.AllAchs[4]
        NewAch6 := CustomPlayer.AllAchs[5]
        NewAch7 := CustomPlayer.AllAchs[6]
        NewAch8 := CustomPlayer.AllAchs[7]

        set MySavedPlayerData[Player] = player_profile_data:
            MakePlayerSavedData<constructor>(SourceData)
            StoneAmount:= CustomPlayer.StoneAmount
            ShardAmount:= CustomPlayer.ShardAmount
            EggAmounts:= CustomPlayer.EggAmounts
            WepMatAmount:= CustomPlayer.WepMatAmount
            StagesUnlocked:= CustomPlayer.StagesUnlocked
            MaxHP:= CustomPlayer.MaxHP
            WepMatDropRate:= CustomPlayer.WepMatDropRate
            AFKStoneBoost:= CustomPlayer.AFKStoneBoost
            AFKPetsUnlocked:= CustomPlayer.AFKPetsUnlocked
            Power:= CustomPlayer.Power
            HarvestBuff:= CustomPlayer.HarvestBuff
            CritRate:= CustomPlayer.CritRate
            CritDMG:= CustomPlayer.CritDMG
            FinalHitBonus:= CustomPlayer.FinalHitBonus
            MultiCrit:= CustomPlayer.MultiCrit
            ComboStrikeUnlocked:= CustomPlayer.ComboStrikeUnlocked
            ComboStrikeMultiplier:= CustomPlayer.ComboStrikeMultiplier
            ComboAccelerator:= CustomPlayer.ComboAccelerator
            ComboStrikeStreak:= CustomPlayer.ComboStrikeStreak
            RoomUnlockGather:= CustomPlayer.RoomUnlockGather
            MovementSpeedLevel:= CustomPlayer.MovementSpeedLevel
            DoubleJumpUnlocked:= CustomPlayer.DoubleJumpUnlocked
            DoubleJumpAmount:= CustomPlayer.DoubleJumpAmount
            WeaponTier:= CustomPlayer.WeaponTier
            WeaponPerkTier:= CustomPlayer.WeaponPerkTier
            WeaponStatLevel:= CustomPlayer.WeaponStatLevel
            WeaponBonusLevel:= CustomPlayer.WeaponBonusLevel
            WeaponPerkLevel:= CustomPlayer.WeaponPerkLevel
            WeaponStats:= CustomPlayer.WeaponStats
            WeaponBonuses:= CustomPlayer.WeaponBonuses
            <#StatCost:= CustomPlayer.StatCost
            BonusCost:= CustomPlayer.BonusCost
            PerkCost:= CustomPlayer.PerkCost
            EvolveCost:= CustomPlayer.EvolveCost#>
            PerksUnlocked:= CustomPlayer.PerksUnlocked
            ActivePerks:= CustomPlayer.ActivePerks
            Hut1Levels:= CustomPlayer.Hut1Levels
            Hut1Costs:= CustomPlayer.Hut1Costs
            #Hut1Upgrades:= CustomPlayer.Hut1Upgrades
            Hut2Levels:= CustomPlayer.Hut2Levels
            Hut2Costs:= CustomPlayer.Hut2Costs
            #Hut2Upgrades:= CustomPlayer.Hut2Upgrades
            Hut3Levels:= CustomPlayer.Hut3Levels
            Hut3Costs:= CustomPlayer.Hut3Costs
            Hut4Levels:= CustomPlayer.Hut4Levels
            #Hut3Upgrades:= CustomPlayer.Hut3Upgrades
            GUP1Levels:= CustomPlayer.GUP1Levels
            #GUP1Costs:= CustomPlayer.GUP1Costs
            #GUP1Upgrades:= CustomPlayer.GUP1Upgrades
            GUP2Levels:= CustomPlayer.GUP2Levels
            #GUP2Costs:= CustomPlayer.GUP2Costs
            #GUP2Upgrades:= CustomPlayer.GUP2Upgrades
            GUP3Levels:= CustomPlayer.GUP3Levels
            #GUP3Costs:= CustomPlayer.GUP3Costs
            #GUP3Upgrades:= CustomPlayer.GUP3Upgrades
            PrestigeLevel:= CustomPlayer.PrestigeLevel
            #PrestigeCost:= CustomPlayer.PrestigeCost
            PrestigeHarvestBuff:= CustomPlayer.PrestigeHarvestBuff
            PrestigeStoneAmount:= CustomPlayer.PrestigeStoneAmount
            PUPLevels:= CustomPlayer.PUPLevels
            #PUPCosts:= CustomPlayer.PUPCosts
            #PUPUpgrades:= CustomPlayer.PUPUpgrades
            EggPrestigeLevels:= CustomPlayer.EggPrestigeLevels
            #EggPrestigeCosts:= CustomPlayer.EggPrestigeCosts
            #EggPrestigeUpgrades:= CustomPlayer.EggPrestigeUpgrades
            PUPQOLLevels:= CustomPlayer.PUPQOLLevels
            PetEquipSlots:= CustomPlayer.PetEquipSlots
            AutoGatherRate:= CustomPlayer.AutoGatherRate
            BossRewardMult:= CustomPlayer.BossRewardMult
            EggAmountMultiplier:= CustomPlayer.EggAmountMultiplier
            EggDropRateMultiplier:= CustomPlayer.EggDropRateMultiplier
            BaseStrengthCosts:= CustomPlayer.BaseStrengthCosts
            #StrengthUpgrades:= CustomPlayer.StrengthUpgrades
            StrengthEfficiencies:= CustomPlayer.StrengthEfficiencies
            StrengthUpgradeLevels:= CustomPlayer.StrengthUpgradeLevels
            EarthquakeUnlocked:= CustomPlayer.EarthquakeUnlocked
            EarthquakeDistance:= CustomPlayer.EarthquakeDistance
            EarthquakeGatherAmount:= CustomPlayer.EarthquakeGatherAmount
            EarthquakeDamage:= CustomPlayer.EarthquakeDamage
            VoidUnlocked:= CustomPlayer.VoidUnlocked
            #VoidDamage:= CustomPlayer.VoidDamage
            VoidDuration:= CustomPlayer.VoidDuration
            VoidStacks:= CustomPlayer.VoidStacks
            A4Unlocked:= CustomPlayer.A4Unlocked
            CurrentDay := CustomPlayer.CurrentDay
            LastCompletedDate := CustomPlayer.LastCompletedDate
            LastCompletedHolidayDate := CustomPlayer.LastCompletedHolidayDate
            SinkPrestigePower := CustomPlayer.SinkPrestigePower
            SinkWepMatPower := CustomPlayer.SinkWepMatPower
            LastClaimed := CustomPlayer.LastClaimed
            HatsUnlocked := CustomPlayer.HatsUnlocked
            EXPPetsLevel:= CustomPlayer.EXPPetsLevel
            NewEXPPetsLevel:= CustomPlayer.NewEXPPetsLevel
            CurrentExpedition:= CustomPlayer.CurrentExpedition
            EXPCurrentPet:= CustomPlayer.EXPCurrentPet
            ExpeditionTimeLeft:= CustomPlayer.ExpeditionTimeLeft
            EXPTotalRewards:= CustomPlayer.EXPTotalRewards
            GoldenDrillUnlocked := CustomPlayer.GoldenDrillUnlocked
            DrillActivations := CustomPlayer.DrillActivations
            BuffMultiplier := CustomPlayer.BuffMultiplier
            BuffTimeSeconds := CustomPlayer.BuffTimeSeconds
            BuffChargeTimer := CustomPlayer.BuffChargeTimer
            BuffCurrentCharges := CustomPlayer.BuffCurrentCharges
            BuffMaxCharges := CustomPlayer.BuffMaxCharges
            GoldenDrillLevels := CustomPlayer.GoldenDrillLevels
            #GoldenDrillCosts := CustomPlayer.GoldenDrillCosts
            #GoldenDrillUpgrades := CustomPlayer.GoldenDrillUpgrades
            SecretsClaimed := CustomPlayer.SecretsClaimed
            NewSecretsClaimed := CustomPlayer.NewSecretsClaimed
            EggsUnlocked:= CustomPlayer.EggsUnlocked
            RockEggDropChance:= CustomPlayer.RockEggDropChance
            EggDropAmount:= CustomPlayer.EggDropAmount
            MobEggDropChance:= CustomPlayer.MobEggDropChance
            PetDataReset := CustomPlayer.PetDataReset
            PetUnlocked:= CustomPlayer.PetUnlocked
            PetSlots:= CustomPlayer.PetSlots
            HidePets:= CustomPlayer.HidePets
            PetLevels:= CustomPlayer.PetLevels
            PetEvolveLevels:= CustomPlayer.PetEvolveLevels
            PetGatherRates:= CustomPlayer.PetGatherRates
            PetAttackRates:= CustomPlayer.PetAttackRates
            PetCritRates:= CustomPlayer.PetCritRates
            PetCritDMG:= CustomPlayer.PetCritDMG
            PetFinalStrike:= CustomPlayer.PetFinalStrike
            EggsHatched:= CustomPlayer.EggsHatched
            HighestFBDamage:= CustomPlayer.HighestFBDamage
            BossKeyAmount:= CustomPlayer.BossKeyAmount
            TutorialLevel:= CustomPlayer.TutorialLevel
            CurrentMusic:= CustomPlayer.CurrentMusic
            TutorialProgress:= CustomPlayer.TutorialProgress
            AchLevels:= CustomPlayer.AchLevels
            AchBuffTotal:= CustomPlayer.AchBuffTotal
            Ach1:= NewAch1
            Ach2:= NewAch2
            Ach3:= NewAch3
            Ach4:= NewAch4
            Ach5:= NewAch5
            Ach6:= NewAch6
            Ach7:= NewAch7
            Ach8:= NewAch8
            PlayTimeSeconds:= CustomPlayer.PlayTimeSeconds
    else:
        Print("SOMETHING WENT WRONG")#>

CheckSaveDataForPlayer(Agent:agent)<decides><transacts>:void=
    if(Player := player[Agent]):
        if(Player.IsActive[]):
            if(MySavedPlayerData[Player]):
            else:
                set MySavedPlayerData[Player] = player_profile_data{}
        else:
            Print("NOT ACTIVE PLAYER")
    else:
        Print("NOT PLAYER")

persistence_classes := class(creative_device):

    OnBegin<override>()<suspends>:void=
        block:
