using { /Fortnite.com/AI }
using { /Verse.org/Simulation }
using {/Fortnite.com/Characters }
using { /Fortnite.com/Animation/PlayAnimation }
using { /Verse.org/Assets }
using { /UnrealEngine.com/Temporary/SpatialMath }
using { /Verse.org/Random }

pet1_behaviour := class(npc_behavior):

    var TargetAgent : ?agent = false

    OnBegin<override>()<suspends>:void=
        #Print("PET 1 BEHAVIOUR")
        if:
            NPCAgent := GetAgent[]
            NPCChar := NPCAgent.GetFortCharacter[]
            NPCNav := NPCChar.GetNavigatable[]
            NPCFocus := NPCChar.GetFocusInterface[]
            NPCAnim := NPCChar.GetPlayAnimationController[]
        then:
            spawn{TeleportPet(TargetAgent, NPCChar, NPCNav)}
            loop:
                Sleep(0.1)
                if(not NPCChar.IsActive[]):
                    break:
                if(Target := TargetAgent?,FC := Target.GetFortCharacter[]):
                    if(Distance(NPCChar.GetTransform().Translation, FC.GetTransform().Translation) < 1500.0):
                        spawn{NPCFocus.MaintainFocus(Target)}
                        
                        #Variables to make sure pet stays behind player as much as possible
                        FCRotation := FC.GetTransform().Rotation
                        ForwardVector := FCRotation.GetLocalForward()
                        OffsetDirection := -ForwardVector
                        OffsetPosition := FC.GetTransform().Translation + OffsetDirection * 250.0

                        #Set target + Navigation to follow player
                        NavTarget := MakeNavigationTarget(OffsetPosition)
                        NavResult:=NPCNav.NavigateTo(NavTarget, ?MovementType := movement_types.Running, ?ReachRadius := 300.0, ?AllowPartialPath:=true)
                        if(NavResult <> navigation_result.Reached):
                            NPCNav.StopNavigation()
                            NPCRotation := NPCChar.GetTransform().Rotation
                            if(Destination:=NPCNav.GetCurrentDestination[]):
                                if(NPCChar.TeleportTo[Destination, NPCRotation]) {}
                                Sleep(0.2)
                            else:
                                if(NPCChar.TeleportTo[OffsetPosition, NPCRotation]) {}
                        else:
                            NPCNav.StopNavigation()
                
    TeleportPet(MaybeTarget:?agent, NPCChar:fort_character, NPCNav:navigatable)<suspends>:void=
        loop:
            Sleep(1.0)
            if(not NPCChar.IsActive[]):
                break:
            if(Target := TargetAgent?,FC := Target.GetFortCharacter[]):
                if(Distance(NPCChar.GetTransform().Translation, FC.GetTransform().Translation) < 1500.0):
                    #Print("In Range")
                else:
                    NPCNav.StopNavigation()
                    NPCRotation := NPCChar.GetTransform().Rotation
                    FCRotation := FC.GetTransform().Rotation
                    ForwardVector := FCRotation.GetLocalForward()
                    OffsetDirection := -ForwardVector
                    IDXTP := GetRandomFloat(250.0,400.0)
                    OffsetPosition := FC.GetTransform().Translation + OffsetDirection * IDXTP
                    if(NPCChar.TeleportTo[OffsetPosition, NPCRotation]) {}

    SetTargetAgent(Agent:?agent):void=
        set TargetAgent = Agent

    OnEnd<override>():void=
        if:
            NPCAgent := GetAgent[]
            NPCChar := NPCAgent.GetFortCharacter[]
            NPCNav := NPCChar.GetNavigatable[]
        then:
            NPCNav.StopNavigation()
