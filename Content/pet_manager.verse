using { /Fortnite.com/Devices }
using { /Verse.org/Simulation }
using { /Verse.org/Colors }
using { /Verse.org/Verse }
using { /Verse.org/Random }
using { /Fortnite.com/Teams }
using { /UnrealEngine.com/Assets }
using { /Fortnite.com/Game }
using { /Fortnite.com/Characters }
using { /Fortnite.com/Playspaces }
using { /Fortnite.com/UI }
using { /UnrealEngine.com/Temporary/Diagnostics }
using { /UnrealEngine.com/Temporary/SpatialMath }
using { /UnrealEngine.com/Temporary/UI }
using { /Fortnite.com/AI }
using { /Verse.org/Assets }

pet_manager := class(creative_device):

    @editable
    RemoteManager : signal_remote_manager_device = signal_remote_manager_device{}
    @editable
    gm : game_manager = game_manager{}
    @editable
    AudioManager : audio_manager = audio_manager{}
    @editable
    menu : ui_main_menu = ui_main_menu{}
    @editable
    Tutorial : tutorial_manager = tutorial_manager{}
    @editable
    EggGranters : []item_granter_device = array{}
    @editable
    PetSpawners : []npc_spawner_device = array{}
    @editable
    ForceSpawnPets : mutator_zone_device = mutator_zone_device{}
    @editable
    RoomUnlockBarriers : []barrier_device = array{}
    @editable
    UnequipPetsButton : button_device = button_device{}

    var DamageNumbers<private>: damage_numbers = damage_numbers{}

    var petimages : []texture = array{UI.pet1, UI.pet2, UI.pet3, UI.pet4, UI.pet5, UI.pet6}

    var NewestPetOwner : ?agent = false

   <#  UnequipPets(Agent:agent):void=
        if(CustomPlayer := gm.PlayersMap[player[Agent]]):
            for(i:int=0..CustomPlayer.PetFortChars.Length):
                if(PetChar := CustomPlayer.PetFortChars[i],Fort:=PetChar?):
                    Fort.Damage(9999999.9)
                    if(set CustomPlayer.PetFortChars[CustomPlayer.PetSlots[i]] = false) {}
            for(i:int=0..CustomPlayer.PetSlots.Length):
                if(set CustomPlayer.PetSlots[i] = -1){}
            SaveAllData(CustomPlayer)
            gm.ShowMessage(CustomPlayer.Player, "Pets unequipped!", 0) #>

    OnBegin<override>()<suspends>:void=
        GetPlayspace().PlayerRemovedEvent().Subscribe(ElimAllPets)

        #UnequipPetsButton.InteractedWithEvent.Subscribe(UnequipPets)

        #for(x:int=0..5):
        #    for(i:int=0..PetSpawners.Length):
        #        Print("Looping through {i} spawner")
        #        if(PetSpawner:=PetSpawners[i]):
        #            PetSpawner.Spawn()
        #            PetSpawner.DespawnAll(false)

        for(i:int=0..PetSpawners.Length):
            if(PetSpawner:=PetSpawners[i]):
                PetSpawner.SpawnedEvent.SubscribeAgent(OnPetSpawned, i)
                

        for(Spawner : gm.PlayerSpawners):     
            Spawner.SpawnedEvent.Subscribe(OnPlayerJoined)

    OnPlayerJoined(Agent:agent):void=
        spawn{OnPlayerJoinedBuffer(Agent)}

    OnPlayerJoinedBuffer(Agent:agent)<suspends>:void=
        Sleep(3.0)
        if(CustomPlayer := gm.PlayersMap[player[Agent]]):
            set NewestPetOwner = option{CustomPlayer.GetPlayer()}

            if(CustomPlayer.PlayTimeSessionSeconds = 0.0):
                if(set CustomPlayer.PetFortChars[0] = false) {}
                if(set CustomPlayer.PetFortChars[1] = false) {}
                if(set CustomPlayer.PetFortChars[2] = false) {}
                if(set CustomPlayer.PetFortChars[3] = false) {}
                if(set CustomPlayer.PetFortChars[4] = false) {}
                if(set CustomPlayer.PetFortChars[5] = false) {}

                CustomPlayer.EquipSlot1Btn.OnClick().SubscribeAgent3(EquipHandler, 0)
                CustomPlayer.Pet1Btn.OnClick().SubscribeAgent3(SelectPetHandler, 0)

                CustomPlayer.EquipSlot2Btn.OnClick().SubscribeAgent3(EquipHandler, 1)
                CustomPlayer.Pet2Btn.OnClick().SubscribeAgent3(SelectPetHandler, 1)

                CustomPlayer.EquipSlot3Btn.OnClick().SubscribeAgent3(EquipHandler, 2)
                CustomPlayer.Pet3Btn.OnClick().SubscribeAgent3(SelectPetHandler, 2)

                CustomPlayer.EquipSlot4Btn.OnClick().SubscribeAgent3(EquipHandler, 3)
                CustomPlayer.Pet4Btn.OnClick().SubscribeAgent3(SelectPetHandler, 3)

                CustomPlayer.EquipSlot5Btn.OnClick().SubscribeAgent3(EquipHandler, 4)
                CustomPlayer.Pet5Btn.OnClick().SubscribeAgent3(SelectPetHandler, 4)

                CustomPlayer.EquipSlot6Btn.OnClick().SubscribeAgent3(EquipHandler, 5)
                CustomPlayer.Pet6Btn.OnClick().SubscribeAgent3(SelectPetHandler, 5)

                CustomPlayer.UnequipPetsBtn.OnClick().Subscribe(UnequipAllPets)
                CustomPlayer.HidePetsBtn.OnClick().Subscribe(HideAllPets)
            
            SpawnAllPets(Agent)


#PET MANAGEMENT REWORK
    #NEEDS TO SPAWN ALL APPLICABLE PETS ON PLAYER LOAD
    #NEEDS TO ELIM ALL PETS ON PLAYER LEAVE

    #NEED TO HIDE ALL PETS THROUGH UI BUTTON (Sets HidePets = true)
    #NEED TO CALL PETS THROUGH UI BUTTON (Sets HidePets = false)

    SpawnAllPets(Agent:agent):void=
        if(CustomPlayer := gm.PlayersMap[player[Agent]]):
            if(CustomPlayer.HidePets = false):
                set NewestPetOwner = option{CustomPlayer.GetPlayer()}
                
                for(i:int=0..CustomPlayer.PetSlots.Length):
                    if(PetSlot:=CustomPlayer.PetSlots[i], PetSlot <> -1):
                        if(PetChar := CustomPlayer.PetFortChars[PetSlot],Fort:=PetChar?):
                            #Print("EXISTS ALREADY {i}")
                        else:
                            #Print("SPAWN {i} - {PetSlot}")
                            SpawnPets(CustomPlayer, PetSlot)
            else:
                #Print("PETS ARE HIDDEN")

    UnequipAllPets(Message : widget_message):void=
        if(CustomPlayer := gm.PlayersMap[Message.Player]):
            set CustomPlayer.HidePets = false
            Print("CALL PET")
            if(CustomPlayer.SpawningPets = false):
                Print("SPAWNING PETS IS FALSE")
                set CustomPlayer.SpawningPets = true
                CustomPlayer.UnequipPetsBtn.SetEnabled(false)
                set NewestPetOwner = option{CustomPlayer.GetPlayer()}
                for(i:int=0..CustomPlayer.PetSlots.Length):
                    if:
                        PetSlot:=CustomPlayer.PetSlots[i]
                        PetSlot <> -1
                        PetChar := CustomPlayer.PetFortChars[PetSlot]
                        Fort:=PetChar?
                    then:
                        Fort.Damage(9999999.9)
                        if(set CustomPlayer.PetFortChars[PetSlot] = false) {}

                for(i:int=0..CustomPlayer.PetSlots.Length):
                    if(PetSlot:=CustomPlayer.PetSlots[i], PetSlot <> -1):
                        if(PetChar := CustomPlayer.PetFortChars[PetSlot],Fort:=PetChar?):
                        else:
                            SpawnPets(CustomPlayer, PetSlot)
                spawn{CallPetBuffer(CustomPlayer)}
            SaveAllData(CustomPlayer)

    CallPetBuffer(CustomPlayer:Custom_Player)<suspends>:void=
        Sleep(3.0)
        set CustomPlayer.SpawningPets = false
        Print("SPAWNING PETS IS NOW FALSE")

    ElimAllPets(PlayerLeaving : player):void=
        if(CustomPlayer := gm.PlayersMap[PlayerLeaving]):
            for(i:int=0..CustomPlayer.PetFortChars.Length):
                if(PetChar := CustomPlayer.PetFortChars[i],Fort:=PetChar?):
                    Fort.Damage(9999999.9)
                    if(set CustomPlayer.PetFortChars[CustomPlayer.PetSlots[i]] = false) {}

    HideAllPets(Message : widget_message):void=
        if(CustomPlayer := gm.PlayersMap[Message.Player]):
            set CustomPlayer.HidePets = true
            for(i:int=0..CustomPlayer.PetFortChars.Length):
                if(PetChar := CustomPlayer.PetFortChars[i],Fort:=PetChar?):
                    Fort.Damage(9999999.9)
                    if(set CustomPlayer.PetFortChars[CustomPlayer.PetSlots[i]] = false) {}
            SaveAllData(CustomPlayer)

    SpawnPets(CustomPlayer : Custom_Player, WhichPet:int):void=
        if:
            CustomPlayer.HidePets = false or WhichPet = 6 or WhichPet = 7
            PetSpawner := PetSpawners[WhichPet]
        then:
            if:
                PetChar := CustomPlayer.PetFortChars[WhichPet]
                Fort:=PetChar?
            then:
                #Print("PET EXISTS")
            else:
                set NewestPetOwner = option{CustomPlayer.GetPlayer()}
                #Print("SPAWN THE PET")
                PetSpawner.Spawn()
        else:

    OnPetSpawned(Agent:agent, WhichSpawner: int):void=
        if(Pet := Agent.GetFortCharacter[],Owner := NewestPetOwner?,CustomPlayer := gm.PlayersMap[Owner],Fort:=Agent.GetFortCharacter[],NPCBehave:=pet1_behaviour[Agent.GetNPCBehavior[]]):
            NPCBehave.SetTargetAgent(NewestPetOwner)
            Pet.DamagedEvent().SubscribeAgent5(OnPetDamaged, Pet)
            Pet.EliminatedEvent().SubscribeAgentElim(OnPetElimed, CustomPlayer)
            
            for(RoomUnlockBarrier:RoomUnlockBarriers):
                RoomUnlockBarrier.AddToIgnoreList(Agent)
                
            if(set CustomPlayer.PetFortChars[WhichSpawner] = option{Fort}) {}

    OnPetDamaged(DmgResult : damage_result, Pet:fort_character):void=
        Pet.Heal(1000.0)
        MaybeInstigator := DmgResult.Instigator
        if:
            Instigator := MaybeInstigator?
            InstagatingAgent := Instigator.GetInstigatorAgent[]
            Player:=player[InstagatingAgent]
            FC:=Player.GetFortCharacter[]
        then:
            NPCRotation := Pet.GetTransform().Rotation
            FCRotation := FC.GetTransform().Rotation
            ForwardVector := FCRotation.GetLocalForward()
            OffsetDirection := -ForwardVector
            IDX := GetRandomFloat(400.0,1000.0)
            OffsetPosition := FC.GetTransform().Translation + OffsetDirection * IDX
            if(Pet.TeleportTo[OffsetPosition, NPCRotation]) {}

    OnPetElimed(ElimResult:elimination_result, CustomPlayer:Custom_Player):void=
        if(set CustomPlayer.PetFortChars[6] = false) {}

    EquipHandler(Message: widget_message, whichSlot : int): void=
        if(CustomPlayer := gm.PlayersMap[Message.Player]):
            if(CustomPlayer.PetEquipSlots[whichSlot] = true):
                set CustomPlayer.CurrentPetSlot = whichSlot
            else:
                AudioManager.Denied.Play(CustomPlayer.Player)
            menu.UpdatePetMenu(CustomPlayer)
    
    SelectPetHandler(Message : widget_message, whichPet : int): void=
        var AlreadyEquipped : logic = false
        if(CustomPlayer := gm.PlayersMap[Message.Player]):
            if (CustomPlayer.PetUnlocked[whichPet] = true):
                for(PetSlot:CustomPlayer.PetSlots):
                    if(PetSlot = whichPet):
                        set AlreadyEquipped = true
                Tutorial.UpdateTutorialUI(CustomPlayer, "EquipPet", 1.0)
                if(AlreadyEquipped = false):
                    if(PetChar := CustomPlayer.PetFortChars[CustomPlayer.PetSlots[CustomPlayer.CurrentPetSlot]],Fort:=PetChar?):
                        if(set CustomPlayer.PetFortChars[CustomPlayer.PetSlots[CustomPlayer.CurrentPetSlot]] = false) {}
                        Fort.Damage(9999999.9)
                    set NewestPetOwner = option{CustomPlayer.GetPlayer()}
                    if(set CustomPlayer.PetSlots[CustomPlayer.CurrentPetSlot] = whichPet){}
                    SpawnPets(CustomPlayer, whichPet)
                    SaveAllData(CustomPlayer)
                    if: 
                        PetImage := petimages[whichPet]
                        EquipSlot := CustomPlayer.EquipImages[CustomPlayer.CurrentPetSlot] 
                    then:
                        EquipSlot.SetImage(PetImage)
                else:
                    AudioManager.Denied.Play(CustomPlayer.Player)
            menu.UpdatePetMenu(CustomPlayer) 
                
    DoesEggDrop(CustomPlayer : Custom_Player, whichStage:int, MobOrStone:string):void=
        var whichEgg : int = 0
        var DropChance : float = 0.0
        if(MobOrStone = "Stone"):
            set DropChance = CustomPlayer.RockEggDropChance + (CustomPlayer.EggDropRateMultiplier*20.0) #1/5 the full bonus for rocks
        else:
            set DropChance = CustomPlayer.MobEggDropChance + (CustomPlayer.EggDropRateMultiplier*100.0)
            
        if(CustomPlayer.StagesUnlocked >= 1):
            IDX:=GetRandomFloat(1.0,100.0)
            if(DropChance >= IDX,whichDropTable:=gm.EggDropOdds[whichStage]):
                set whichEgg = SelectItem(whichDropTable)
                if(EggDropMin := Floor[CustomPlayer.EggDropAmount[0] + (CustomPlayer.EggDropAmount[0] * CustomPlayer.EggAmountMultiplier)],EggDropMax := Floor[CustomPlayer.EggDropAmount[1] + (CustomPlayer.EggDropAmount[1] * CustomPlayer.EggAmountMultiplier)]):
                    DropAmount := GetRandomInt(EggDropMin, EggDropMax)*1.0
                    AudioManager.ItemObtained.Play(CustomPlayer.Player)
                    CustomPlayer.Collect(DropAmount, "Egg", whichEgg)
                    DamageNumbers.ShowDamageUIForPlayer(CustomPlayer, DropAmount, false, 0, false, "Egg{whichEgg}", -1.0)

                SaveAllData(CustomPlayer)

                #VISUAL FX
                if(EggGranter:=EggGranters[whichEgg]):
                    #EggGranter.GrantItem(CustomPlayer.Player)