
using { /Fortnite.com/Devices }
using { /Verse.org/Simulation }
using { /UnrealEngine.com/Temporary/Diagnostics }
using { /UnrealEngine.com/Temporary/UI }
using { /UnrealEngine.com/Temporary/SpatialMath }
using { /Verse.org/Assets }
using { /Verse.org/Colors }

# See https://dev.epicgames.com/documentation/en-us/uefn/create-your-own-device-in-verse for how to create a verse device.

# A Verse-authored creative device that can be placed in a level
pet_mutation_manager := class(creative_device):

    @editable
    gm : game_manager = game_manager{}

    @editable
    AudioManager : audio_manager = audio_manager{}

    @editable
    PetMutatorButton : button_device = button_device{}

    var PetTextures : []texture = array{UI.pet1, UI.pet2}
    var EggTextures : []texture = array{UI.Egg1, UI.Egg2}

    var PetMutationBonusValues : []float = array{44.0,66.0}
    var StoneCosts : []float = array{5.00e100, 7.50e125}
    var PrestigeCosts : []float = array{5000.0, 15000.0}

    var EggInfo : tuple(int,int) = (0,0)
    var EggMoveOn : logic = false

    # Runs when the device is started in a running game
    OnBegin<override>()<suspends>:void=
        for(Spawner : gm.PlayerSpawners):     
            Spawner.SpawnedEvent.Subscribe(OnPlayerJoined)

        PetMutatorButton.InteractedWithEvent.Subscribe(OpenMutateUI)

    OnPlayerJoined(Agent:agent):void=
        spawn{OnPlayerJoinedBuffer(Agent)}

    OnPlayerJoinedBuffer(Agent:agent)<suspends>:void=
        Sleep(1.0)
        if(CustomPlayer := gm.PlayersMap[player[Agent]]):
            if(CustomPlayer.PlayTimeSessionSeconds = 0.0):
                set CustomPlayer.MutateEggAmount = array{CustomPlayer.PetMEgg1Amount,CustomPlayer.PetMEgg2Amount}

                CustomPlayer.PetMCloseButton.OnClick().Subscribe(CloseMutateUI)
                CustomPlayer.PetM1ViewButton.OnClick().SubscribeAgent3(HandleMPetSelection, 0)
                CustomPlayer.PetM2ViewButton.OnClick().SubscribeAgent3(HandleMPetSelection, 1)
                CustomPlayer.PetMBackButton.OnClick().SubscribeAgent3(HandlePageSelection, 0)
                CustomPlayer.PetMForwardButton.OnClick().SubscribeAgent3(HandlePageSelection, 1)
                CustomPlayer.PetMHatchButton.OnClick().SubscribeAgent3(HandlePeMutate, false)
                CustomPlayer.PetMHatchAllButton.OnClick().Subscribe(PetMutataAllHelper)

    OpenMutateUI(Agent:agent): void=
        if(CustomPlayer := gm.PlayersMap[Agent]):
            if(MyUI:=CustomPlayer.UIMaps["PetMutate"]?):
            else:
                var MyPetMutateCanvas : canvas = MakePetCanvas(CustomPlayer)
                CustomPlayer.AddFullHUD(MyPetMutateCanvas)
                if(set CustomPlayer.UIMaps["PetMutate"] = option{MyPetMutateCanvas}) {}
                spawn{UpdatePetMutatorUI(CustomPlayer)}

    CloseMutateUI(Message : widget_message):void=
        if:
            CustomPlayer := gm.PlayersMap[Message.Player]
            Player:=Message.Player
            PlayerUI:=GetPlayerUI[Player]
            MyUI:=CustomPlayer.UIMaps["PetMutate"]?
        then:
            PlayerUI.RemoveWidget(MyUI)
            MyUI.SetVisibility(widget_visibility.Collapsed)
            if(set CustomPlayer.UIMaps["PetMutate"] = false){}
            
    UpdatePetMutatorUI(CustomPlayer : Custom_Player)<suspends>:void=
        CustomPlayer.PetMStoneAmount.SetText(StringToMessage("{NumberCondensor(CustomPlayer.StoneAmount)}"))
        CustomPlayer.PetMPrestigeAmount.SetText(StringToMessage("{NumberCondensor(CustomPlayer.ShardAmount)}"))
        if (Pet1Level := CustomPlayer.PetLevels[0], Pet2Level := CustomPlayer.PetLevels[1]):
            CustomPlayer.PetM1LevelTitle.SetText(StringToMessage("Lv. {NumberCondensor(Pet1Level*1.0)}"))
            CustomPlayer.PetM2LevelTitle.SetText(StringToMessage("Lv. {NumberCondensor(Pet2Level*1.0)}"))

        for (i:int=0..CustomPlayer.MutateEggAmount.Length-1):
            if(EggAmount := CustomPlayer.EggAmounts[i], EggAmountBlock := CustomPlayer.MutateEggAmount[i]):
                EggAmountBlock.SetText(StringToMessage("{NumberCondensor(EggAmount)}"))

        for(i:int=0..PetTextures.Length):
            if (i = CustomPlayer.MPetSelected):
                if:
                    PetImage := PetTextures[i]
                    PetLevel := CustomPlayer.PetLevels[i]
                    PetGather := CustomPlayer.PetGatherRates[i]
                    PetAttack := CustomPlayer.PetAttackRates[i]
                    PetCritRate := CustomPlayer.PetCritRates[i]
                    PetCritDMG := CustomPlayer.PetCritDMG[i]
                    PetFinalStrike :=  CustomPlayer.PetFinalStrike[i]
                then:
                    CustomPlayer.PetMSPetStatsIcon.SetImage(PetImage)
                    CustomPlayer.PetMGatherAmount.SetText(StringToMessage("{NumberCondensor(PetGather)}%"))
                    CustomPlayer.PetMStrengthAmount.SetText(StringToMessage("{NumberCondensor(PetAttack)}%"))
                    CustomPlayer.PetMCritRateAmount.SetText(StringToMessage("{NumberCondensor(PetCritRate)}%"))
                    CustomPlayer.PetMCritDMGAmount.SetText(StringToMessage("{NumberCondensor(PetCritDMG)}%"))
                    CustomPlayer.PetMFSAmounnt.SetText(StringToMessage("{NumberCondensor(PetFinalStrike)}%"))
                    if(PetFinalStrike >= 1000.0):
                        CustomPlayer.PetMFSTitle.SetText(StringToMessage("FS:"))

        for(i:int=0..EggTextures.Length):
            if (i = CustomPlayer.MPetCurrentPage):
                set EggInfo  = (0,0)
                set EggMoveOn = false
                spawn{NewTotalEggsNeeded(CustomPlayer)}
                loop:
                    if (EggMoveOn = true):
                        break:
                    Sleep(0.1)
                Print("Eggs Used : {EggInfo(0)} Hatches : {EggInfo(1)}")
                if:
                    EggAmount := CustomPlayer.EggAmounts[i]
                    EggCost:=Floor[1.0 + (CustomPlayer.EggsHatched[CustomPlayer.MPetCurrentPage]/25.0)]
                    EggSelected := EggTextures[i]
                    StoneCost := StoneCosts[i]
                    PrestigeCost := PrestigeCosts[i]
                then:
                    CustomPlayer.PetMEggCostIcon.SetImage(EggSelected)
                    CustomPlayer.PetMAllEggCostIcon.SetImage(EggSelected)
                    CustomPlayer.PetMWhichEggIcon.SetImage(EggSelected)
                    CustomPlayer.PetMEggCostTitle.SetText(StringToMessage("{EggCost}"))
                    CustomPlayer.PetMAllEggCostTitle.SetText(StringToMessage(" {NumberCondensor(EggInfo(0) * 1.0)}"))
                    CustomPlayer.PetMStoneCostTitle.SetText(StringToMessage(" {NumberCondensor(StoneCost)}"))
                    CustomPlayer.PetMAllStoneCostTitle.SetText(StringToMessage(" {NumberCondensor(StoneCost * EggInfo(1))}"))
                    CustomPlayer.PetMPrestigeCostTitle.SetText(StringToMessage(" {NumberCondensor(PrestigeCost)}"))
                    CustomPlayer.PetMAllPrestigeCostTitle.SetText(StringToMessage(" {NumberCondensor(PrestigeCost * EggInfo(1))}"))
                    
                    #CHANGE TEXT COLOR IF YOU CAN OR CAN'T MAKE PURCHASE
                    if(CustomPlayer.StoneAmount >= StoneCost):
                        CustomPlayer.PetMStoneCostTitle.SetTextColor(MakeColorFromHex("39FF39"))
                    else:    
                        CustomPlayer.PetMStoneCostTitle.SetTextColor(MakeColorFromHex("FA4D56"))
                    if(CustomPlayer.EggAmounts[i] >= EggCost*1.0):
                        CustomPlayer.PetMEggCostTitle.SetTextColor(MakeColorFromHex("39FF39"))
                    else:    
                        CustomPlayer.PetMEggCostTitle.SetTextColor(MakeColorFromHex("FA4D56"))
                    if(CustomPlayer.ShardAmount >= PrestigeCost):
                        CustomPlayer.PetMPrestigeCostTitle.SetTextColor(MakeColorFromHex("39FF39"))
                    else:    
                        CustomPlayer.PetMPrestigeCostTitle.SetTextColor(MakeColorFromHex("FA4D56"))

                    if(CustomPlayer.StoneAmount >= StoneCost * EggInfo(1)):
                        CustomPlayer.PetMAllStoneCostTitle.SetTextColor(MakeColorFromHex("39FF39"))
                    else:    
                        CustomPlayer.PetMAllStoneCostTitle.SetTextColor(MakeColorFromHex("FA4D56"))
                    if(CustomPlayer.EggAmounts[i] >= EggInfo(0) *1.0):
                        CustomPlayer.PetMAllEggCostTitle.SetTextColor(MakeColorFromHex("39FF39"))
                    else:    
                        CustomPlayer.PetMAllEggCostTitle.SetTextColor(MakeColorFromHex("FA4D56"))
                    if(CustomPlayer.ShardAmount >= PrestigeCost * EggInfo(1)):
                        CustomPlayer.PetMAllPrestigeCostTitle.SetTextColor(MakeColorFromHex("39FF39"))
                    else:    
                        CustomPlayer.PetMAllPrestigeCostTitle.SetTextColor(MakeColorFromHex("FA4D56"))

    HandleMPetSelection(Message : widget_message, WhichPet : int):void=
        if(CustomPlayer := gm.PlayersMap[Message.Player]):
            set CustomPlayer.MPetSelected = WhichPet
            spawn{UpdatePetMutatorUI(CustomPlayer)}

    HandlePageSelection(Message: widget_message, WhichDirection : int):void=
        if(CustomPlayer := gm.PlayersMap[Message.Player]):
            if (WhichDirection = 0):
                if(CustomPlayer.MPetCurrentPage > 0):
                    set CustomPlayer.MPetCurrentPage -= 1
                    set CustomPlayer.MPetSelected = CustomPlayer.MPetCurrentPage
                else:
                    AudioManager.Denied.Play(CustomPlayer.Player)
            if (WhichDirection = 1):
                if(CustomPlayer.MPetCurrentPage < 1):
                    set CustomPlayer.MPetCurrentPage += 1
                    set CustomPlayer.MPetSelected = CustomPlayer.MPetCurrentPage
                else:
                    AudioManager.Denied.Play(CustomPlayer.Player)
                    
            spawn{UpdatePetMutatorUI(CustomPlayer)}

    HandlePeMutate(Message: widget_message, HatchAll:logic):void=
        if(CustomPlayer := gm.PlayersMap[Message.Player]):
            if:
                PetLevel := CustomPlayer.PetLevels[CustomPlayer.MPetCurrentPage]
                PetGather := (PetMutationBonusValues[CustomPlayer.MPetCurrentPage] * 1.0) * ((PetLevel + 1) * 0.2)
                PetAttack := (PetMutationBonusValues[CustomPlayer.MPetCurrentPage] * 1.0) * ((PetLevel + 1) * 0.2)
                PetCritRate := (PetMutationBonusValues[CustomPlayer.MPetCurrentPage] * 0.1) * ((PetLevel + 1) * 0.1)
                PetCritDMG := (PetMutationBonusValues[CustomPlayer.MPetCurrentPage] * 0.1) * ((PetLevel + 1) * 0.1)
                PetFinalStrike :=  (PetMutationBonusValues[CustomPlayer.MPetCurrentPage] * 0.15) * ((PetLevel + 1) * 0.1)
                EggCost:=Floor[1.0 + (CustomPlayer.EggsHatched[CustomPlayer.MPetCurrentPage]/25.0)]
                AmountOfEggs := CustomPlayer.EggAmounts[CustomPlayer.MPetCurrentPage]
                AmountOfEggs >= EggCost*1.0
                StoneCost:=StoneCosts[CustomPlayer.MPetCurrentPage]
                PrestigeCost := PrestigeCosts[CustomPlayer.MPetCurrentPage]
                CustomPlayer.StoneAmount >= StoneCost
                CustomPlayer.ShardAmount >= PrestigeCost
            then:
                if (set CustomPlayer.PetLevels[CustomPlayer.MPetCurrentPage] += 1){}
                if(set CustomPlayer.PetGatherRates[CustomPlayer.MPetCurrentPage] = PetGather):
                if(set CustomPlayer.PetAttackRates[CustomPlayer.MPetCurrentPage] = PetAttack) {}
                if(set CustomPlayer.PetCritRates[CustomPlayer.MPetCurrentPage] = PetCritRate) {}
                if(set CustomPlayer.PetCritDMG[CustomPlayer.MPetCurrentPage] = PetCritDMG) {}
                if(set CustomPlayer.PetFinalStrike[CustomPlayer.MPetCurrentPage] = PetFinalStrike) {}
                if(set CustomPlayer.EggsHatched[CustomPlayer.MPetCurrentPage] += 1.0) {}
                CustomPlayer.UpAchValue(5, 1.0)
                CustomPlayer.Spend(EggCost*1.0, "Egg", CustomPlayer.MPetCurrentPage)
                CustomPlayer.Spend(PrestigeCost, "Shard", 0)
                CustomPlayer.Spend(StoneCost, "Stone", 0)               
                if(Analytic:=AudioManager.GeneralAnalytics[3]): #ANALYTIC
                    Analytic.Submit(CustomPlayer.Player)
                else:
                    if(HatchAll = false):
                        AudioManager.Denied.Play(CustomPlayer.Player)

            if(HatchAll = false):
                SaveAllData(CustomPlayer)
                spawn{UpdatePetMutatorUI(CustomPlayer)}

    TotalEggsNeeded(CustomPlayer:Custom_Player):tuple(int,int)=
        var TotalEggCost : int = 0
        var Hatches : int = 0
        var EggAmount : int = 0
        var CurrentCost : int = 0
        var CurrentHatched : int = 0
        var UseAbleFloorEggs : int = 0
        if:
            EggsHatched := Floor[CustomPlayer.EggsHatched[CustomPlayer.MPetCurrentPage]]
            FlooredCost := Floor[(CustomPlayer.EggsHatched[CustomPlayer.MPetCurrentPage] / 25.0) + 1.0]
            FlooredEggs := Floor[CustomPlayer.EggAmounts[CustomPlayer.MPetCurrentPage]]
        then:
            set UseAbleFloorEggs = FlooredEggs
            if (UseAbleFloorEggs >= 1000000000):
                set UseAbleFloorEggs = 1000000000
            set CurrentHatched = EggsHatched
            set CurrentCost = FlooredCost
            set EggAmount = UseAbleFloorEggs
            loop:
                if(EggAmount > 0):
                    if(EggsToNextBlock := (25 - Mod[CurrentHatched, 25]) * CurrentCost):
                        if(EggAmount >= EggsToNextBlock):
                            set TotalEggCost += EggsToNextBlock
                            set Hatches += 25
                            set EggAmount -= EggsToNextBlock
                            set CurrentHatched += EggsToNextBlock
                            set CurrentCost += 1
                        else:
                            if(HatchesLeft:=Floor[(EggAmount*1.0) / (CurrentCost*1.0)]):
                                set TotalEggCost += HatchesLeft * CurrentCost
                                set Hatches += HatchesLeft                            
                            set EggAmount = 0
                            break:
                else:
                    break:

        return (TotalEggCost, Hatches)

    NewTotalEggsNeeded(CustomPlayer:Custom_Player)<suspends>:void=
        var TotalHatches : int = 0
        var TotalEggCost : int = 0
        var InitalHatched : int = 0
        var InitialEggCost : int = 0
        var EggAmount : int = 0
        var TotalStone : float = 0.0
        var TotalShards : float = 0.0
        var MaxHatches : int = 0
        var MaxHatchesEggs : int =0
        var UseAbleFloorEggs : int = 0
        if:
            EggsHatched := Floor[CustomPlayer.EggsHatched[CustomPlayer.MPetCurrentPage]]
            FlooredCost := Floor[(CustomPlayer.EggsHatched[CustomPlayer.MPetCurrentPage] / 25.0) + 1.0]
            FlooredEggs := Floor[CustomPlayer.EggAmounts[CustomPlayer.MPetCurrentPage]]
        then:
            set UseAbleFloorEggs = FlooredEggs
            if (UseAbleFloorEggs >= 1000000000):
                set UseAbleFloorEggs = 1000000000
            set InitalHatched = EggsHatched
            set InitialEggCost = FlooredCost
            set EggAmount = UseAbleFloorEggs
            set TotalStone = CustomPlayer.StoneAmount
            set TotalShards = CustomPlayer.ShardAmount
            loop:
                if:
                    EggAmount >= InitialEggCost
                then:
                    if (CustomPlayer.MPetCurrentPage = 0):
                        if:
                            TotalStone >= 5.00e100
                            TotalShards >= 5000.0
                        then:
                            var HatchesUnitlIncrease : int = 0
                            if(Hatches:= 25 - Mod[EggsHatched,25]):
                                set HatchesUnitlIncrease = Hatches

                            if(Hatches:=Floor(EggAmount / InitialEggCost)):
                                set MaxHatchesEggs = Hatches
                            MaxHatchesStone := TotalStone / 5.00e100
                            MaxHatchesShards := TotalShards / 5000.0

                            set MaxHatches = HatchesUnitlIncrease
                            if (MaxHatchesEggs < MaxHatches ):
                                set MaxHatches = MaxHatchesEggs
                            if (MaxHatchesStone < MaxHatches * 1.0):
                                if (Hatches := Floor[MaxHatchesStone]):
                                    set MaxHatches =  Hatches
                            if (MaxHatchesShards < MaxHatches *1.0):
                                if (Hatches := Floor[MaxHatchesShards]):
                                    set MaxHatches = Hatches
                            
                            set TotalHatches += MaxHatches
                            set TotalEggCost += MaxHatches * InitialEggCost
                            set EggAmount -= MaxHatches * InitialEggCost
                            set TotalStone -= MaxHatches * 5.00e100
                            set TotalShards -= MaxHatches * 5000.0
                            set InitalHatched += MaxHatches

                            if(Mod[EggsHatched,25] = 0):
                                set InitialEggCost +=1
                                                       
                            <# set EggAmount -= 1
                            set TotalStone -= 5.00e100
                            set TotalShards -= 5000.0

                            set TotalEggCost += InitialEggCost
                            set TotalHatches += 1
                            set InitalHatched += 1

                            if (Mod[InitalHatched,25] = 0):
                                set InitialEggCost += 1 #>
                        else:
                            break:
                    else if (CustomPlayer.MPetCurrentPage = 1):
                        if:
                            TotalStone >= 7.50e125
                            TotalShards >= 15000.0
                        then:
                            var HatchesUnitlIncrease : int = 0
                            if(Hatches:= 25 - Mod[EggsHatched,25]):
                                set HatchesUnitlIncrease = Hatches

                            if(Hatches:=Floor(EggAmount / InitialEggCost)):
                                set MaxHatchesEggs = Hatches
                            MaxHatchesStone := TotalStone / 7.50e125
                            MaxHatchesShards := TotalShards / 15000.0

                            set MaxHatches = HatchesUnitlIncrease
                            if (MaxHatchesEggs < MaxHatches ):
                                set MaxHatches = MaxHatchesEggs
                            if (MaxHatchesStone < MaxHatches * 1.0):
                                if (Hatches := Floor[MaxHatchesStone]):
                                    set MaxHatches =  Hatches
                            if (MaxHatchesShards < MaxHatches *1.0):
                                if (Hatches := Floor[MaxHatchesShards]):
                                    set MaxHatches = Hatches

                            if (MaxHatches > HatchesUnitlIncrease):
                                set MaxHatches = HatchesUnitlIncrease
                            
                            set TotalHatches += MaxHatches
                            set TotalEggCost += MaxHatches * InitialEggCost
                            set EggAmount -= MaxHatches * InitialEggCost
                            set TotalStone -= MaxHatches * 7.50e125
                            set TotalShards -= MaxHatches * 15000.0
                            set InitalHatched += MaxHatches

                            if(Mod[EggsHatched,25] = 0):
                                set InitialEggCost +=1
                            <# set EggAmount -= 1
                            set TotalStone -= 7.50e125
                            set TotalShards -= 15000.0

                            set TotalEggCost += InitialEggCost
                            set TotalHatches += 1
                            set InitalHatched += 1

                            if (Mod[InitalHatched,25] = 0):
                                set InitialEggCost += 1 #>
                        else:
                            break:
                else:
                    break:
        set EggInfo = (TotalEggCost, TotalHatches)
        set EggMoveOn = true


    PetMutataAllHelper(Message:widget_message):void=
        spawn{HandleMPetHatchAll(Message)}

    
    HandleMPetHatchAll(Message: widget_message)<suspends>:void=
        set EggInfo = (0,0)
        set EggMoveOn = false
        if(CustomPlayer := gm.PlayersMap[Message.Player]):
            CustomPlayer.ForwardArrow.SetEnabled(false)
            CustomPlayer.BackArrow.SetEnabled(false)
            spawn{NewTotalEggsNeeded(CustomPlayer)}
            loop:
                if (EggMoveOn?):
                    break:
                Sleep(0.1)
            Print("Total Egg Cost : {EggInfo(0)} Hatches : {EggInfo(1)}")
            #for(i:int=0..EggInfo(1)):
            if:
                EggInfo(1) >= 1
                StoneCost:=StoneCosts[CustomPlayer.MPetCurrentPage] * EggInfo(1)
                PrestigeCost:=PrestigeCosts[CustomPlayer.MPetCurrentPage] * EggInfo(1)
                CustomPlayer.StoneAmount >= StoneCost
                CustomPlayer.ShardAmount >= PrestigeCost
                set CustomPlayer.PetLevels[CustomPlayer.MPetCurrentPage] += EggInfo(1)
                PetLevel := CustomPlayer.PetLevels[CustomPlayer.MPetCurrentPage]
                PetGather := (PetMutationBonusValues[CustomPlayer.MPetCurrentPage] * 1.0) * (PetLevel * 0.2)
                PetAttack := (PetMutationBonusValues[CustomPlayer.MPetCurrentPage] * 1.0) * (PetLevel * 0.2)
                PetCritRate := (PetMutationBonusValues[CustomPlayer.MPetCurrentPage] * 0.1) * (PetLevel * 0.1)
                PetCritDMG := (PetMutationBonusValues[CustomPlayer.MPetCurrentPage] * 0.1) * (PetLevel * 0.1)
                PetFinalStrike :=  (PetMutationBonusValues[CustomPlayer.MPetCurrentPage] * 0.15) * (PetLevel * 0.1)
            then:
                CustomPlayer.UpAchValue(5, EggInfo(1)*1.0)
                CustomPlayer.Spend(StoneCost, "Stone", 0)
                CustomPlayer.Spend(PrestigeCost, "Shard", 0)
                CustomPlayer.Spend(EggInfo(0)*1.0, "Egg", CustomPlayer.MPetCurrentPage)
                if(set CustomPlayer.EggsHatched[CustomPlayer.MPetCurrentPage] += EggInfo(1)*1.0) {}
                if(set CustomPlayer.PetGatherRates[CustomPlayer.MPetCurrentPage] = PetGather) {}
                if(set CustomPlayer.PetAttackRates[CustomPlayer.MPetCurrentPage] = PetAttack) {}
                if(set CustomPlayer.PetCritRates[CustomPlayer.MPetCurrentPage] = PetCritRate) {}
                if(set CustomPlayer.PetCritDMG[CustomPlayer.MPetCurrentPage] = PetCritDMG) {}
                if(set CustomPlayer.PetFinalStrike[CustomPlayer.MPetCurrentPage] = PetFinalStrike) {}
                if(set CustomPlayer.EggsHatched[CustomPlayer.MPetCurrentPage] += 1.0) {}
                if(Analytic:=AudioManager.GeneralAnalytics[3]):
                    Analytic.Submit(CustomPlayer.Player)
            else:
                AudioManager.Denied.Play(CustomPlayer.Player)

            SaveAllData(CustomPlayer) 
            CustomPlayer.ForwardArrow.SetEnabled(true)
            CustomPlayer.BackArrow.SetEnabled(true)
            spawn{UpdatePetMutatorUI(CustomPlayer)}
                    
    MakePetCanvas(CustomPlayer:Custom_Player):canvas=
        #Default Text
        CustomPlayer.PetMCloseButton.SetText(StringToMessage("X"))
        CustomPlayer.PetMGatherTitle.SetText(StringToMessage("Gather:"))
        CustomPlayer.PetMStrengthTitle.SetText(StringToMessage("Strength %:"))
        CustomPlayer.PetMCritRateTitle.SetText(StringToMessage("Crit Rate:"))
        CustomPlayer.PetMCritDMGTitle.SetText(StringToMessage("Crit Dmg:"))
        CustomPlayer.PetMFSTitle.SetText(StringToMessage("Final Strike:"))
        CustomPlayer.PetMCostTitle.SetText(StringToMessage("Cost:"))
        CustomPlayer.PetMHatchButton.SetText(StringToMessage("Hatch"))
        CustomPlayer.PetMHatchAllButton.SetText(StringToMessage("Hatch All"))
        CustomPlayer.PetM1ViewButton.SetText(StringToMessage("View"))
        CustomPlayer.PetM2ViewButton.SetText(StringToMessage("View"))
        CustomPlayer.PetM1TierTitle.SetText(StringToMessage("Common"))
        CustomPlayer.PetM2TierTitle.SetText(StringToMessage("Uncommon"))

        MyCanvas: canvas = canvas {
            Slots := array:
                # General Stuff
                canvas_slot:
                    Widget:=CustomPlayer.PetMBG
                    ZOrder:=0
                    SizeToContent:=false
                    Alignment := vector2{X := 0.5, Y := 0.5}
                    Anchors := anchors{Minimum := vector2{X:= 0.5, Y:= 0.5}, Maximum := vector2{X:= 0.5, Y:= 0.5}}
                    Offsets:=margin{Left:=0.0,Top:=0.0,Right:=1920.0,Bottom:=1080.0}
                canvas_slot:
                    Widget:=CustomPlayer.PetMCloseButton
                    ZOrder:=1
                    SizeToContent:=false
                    Alignment := vector2{X := 0.5, Y := 0.5}
                    Anchors := anchors{Minimum := vector2{X:= 0.5, Y:= 0.5}, Maximum := vector2{X:= 0.5, Y:= 0.5}}
                    Offsets:=margin{Left:=576.0,Top:=-432.0,Right:=76.0,Bottom:=56.0}
                canvas_slot:
                    Widget:=CustomPlayer.PetMTitle
                    ZOrder:=1
                    SizeToContent:=false
                    Alignment := vector2{X := 0.0, Y := 0.0}
                    Anchors := anchors{Minimum := vector2{X:= 0.5, Y:= 0.5}, Maximum := vector2{X:= 0.5, Y:= 0.5}}
                    Offsets:=margin{Left:=-48.0,Top:=-432.0,Right:=489.0,Bottom:=61.0}
                
                # Left Section
                canvas_slot:
                    Widget:=CustomPlayer.PetMStoneIcon
                    ZOrder:=1
                    SizeToContent:=false
                    Alignment := vector2{X := 0.0, Y := 0.0}
                    Anchors := anchors{Minimum := vector2{X:= 0.5, Y:= 0.5}, Maximum := vector2{X:= 0.5, Y:= 0.5}}
                    Offsets:=margin{Left:=-752.0,Top:=-312.0,Right:=50.0,Bottom:=50.0}
                canvas_slot:
                    Widget:=CustomPlayer.PetMStoneAmount
                    ZOrder:=1
                    SizeToContent:=true
                    Alignment := vector2{X := 0.0, Y := 0.0}
                    Anchors := anchors{Minimum := vector2{X:= 0.5, Y:= 0.5}, Maximum := vector2{X:= 0.5, Y:= 0.5}}
                    Offsets:=margin{Left:=-684.0,Top:=-312.0}
                canvas_slot:
                    Widget:=CustomPlayer.PetMPrestigeIcon
                    ZOrder:=1
                    SizeToContent:=false
                    Alignment := vector2{X := 0.0, Y := 0.0}
                    Anchors := anchors{Minimum := vector2{X:= 0.5, Y:= 0.5}, Maximum := vector2{X:= 0.5, Y:= 0.5}}
                    Offsets:=margin{Left:=-752.0,Top:=-228.0,Right:=50.0,Bottom:=50.0}
                canvas_slot:
                    Widget:=CustomPlayer.PetMPrestigeAmount
                    ZOrder:=1
                    SizeToContent:=true
                    Alignment := vector2{X := 0.0, Y := 0.0}
                    Anchors := anchors{Minimum := vector2{X:= 0.5, Y:= 0.5}, Maximum := vector2{X:= 0.5, Y:= 0.5}}
                    Offsets:=margin{Left:=-684.0,Top:=-228.0}
                canvas_slot:
                    Widget:=CustomPlayer.PetEgg1con
                    ZOrder:=1
                    SizeToContent:=false
                    Alignment := vector2{X := 0.0, Y := 0.0}
                    Anchors := anchors{Minimum := vector2{X:= 0.5, Y:= 0.5}, Maximum := vector2{X:= 0.5, Y:= 0.5}}
                    Offsets:=margin{Left:=-752.0,Top:=-144.0,Right:=50.0,Bottom:=50.0}
                canvas_slot:
                    Widget:=CustomPlayer.PetMEgg1Amount
                    ZOrder:=1
                    SizeToContent:=true
                    Alignment := vector2{X := 0.0, Y := 0.0}
                    Anchors := anchors{Minimum := vector2{X:= 0.5, Y:= 0.5}, Maximum := vector2{X:= 0.5, Y:= 0.5}}
                    Offsets:=margin{Left:=-684.0,Top:=-144.0}
                canvas_slot:
                    Widget:=CustomPlayer.PetMEgg2Icon
                    ZOrder:=1
                    SizeToContent:=false
                    Alignment := vector2{X := 0.0, Y := 0.0}
                    Anchors := anchors{Minimum := vector2{X:= 0.5, Y:= 0.5}, Maximum := vector2{X:= 0.5, Y:= 0.5}}
                    Offsets:=margin{Left:=-752.0,Top:=-56.0,Right:=50.0,Bottom:=50.0}
                canvas_slot:
                    Widget:=CustomPlayer.PetMEgg2Amount
                    ZOrder:=1
                    SizeToContent:=true
                    Alignment := vector2{X := 0.0, Y := 0.0}
                    Anchors := anchors{Minimum := vector2{X:= 0.5, Y:= 0.5}, Maximum := vector2{X:= 0.5, Y:= 0.5}}
                    Offsets:=margin{Left:=-684.0,Top:=-56.0}

                # Middle Section
                canvas_slot:
                    Widget:=CustomPlayer.PetMPet1Icon
                    ZOrder:=1
                    SizeToContent:=false
                    Alignment := vector2{X := 0.0, Y := 0.0}
                    Anchors := anchors{Minimum := vector2{X:= 0.5, Y:= 0.5}, Maximum := vector2{X:= 0.5, Y:= 0.5}}
                    Offsets:=margin{Left:=-436.0,Top:=-300.0,Right:=256.0,Bottom:=256.0}
                canvas_slot:
                    Widget:=CustomPlayer.PetMPet2Icon
                    ZOrder:=1
                    SizeToContent:=false
                    Alignment := vector2{X := 0.0, Y := 0.0}
                    Anchors := anchors{Minimum := vector2{X:= 0.5, Y:= 0.5}, Maximum := vector2{X:= 0.5, Y:= 0.5}}
                    Offsets:=margin{Left:=-436.0,Top:=24.0,Right:=256.0,Bottom:=256.0}
                canvas_slot:
                    Widget:=CustomPlayer.PetM1LevelTitle
                    ZOrder:=1
                    SizeToContent:=true
                    Alignment := vector2{X := 0.0, Y := 0.0}
                    Anchors := anchors{Minimum := vector2{X:= 0.5, Y:= 0.5}, Maximum := vector2{X:= 0.5, Y:= 0.5}}
                    Offsets:=margin{Left:=-157.0,Top:=-265.0}
                canvas_slot:
                    Widget:=CustomPlayer.PetM1TierTitle
                    ZOrder:=1
                    SizeToContent:=true
                    Alignment := vector2{X := 0.0, Y := 0.0}
                    Anchors := anchors{Minimum := vector2{X:= 0.5, Y:= 0.5}, Maximum := vector2{X:= 0.5, Y:= 0.5}}
                    Offsets:=margin{Left:=-157.0,Top:=-185.0}
                canvas_slot:
                    Widget:=CustomPlayer.PetM2LevelTitle
                    ZOrder:=1
                    SizeToContent:=true
                    Alignment := vector2{X := 0.0, Y := 0.0}
                    Anchors := anchors{Minimum := vector2{X:= 0.5, Y:= 0.5}, Maximum := vector2{X:= 0.5, Y:= 0.5}}
                    Offsets:=margin{Left:=-157.0,Top:=60.0}
                canvas_slot:
                    Widget:=CustomPlayer.PetM2TierTitle
                    ZOrder:=1
                    SizeToContent:=true
                    Alignment := vector2{X := 0.0, Y := 0.0}
                    Anchors := anchors{Minimum := vector2{X:= 0.5, Y:= 0.5}, Maximum := vector2{X:= 0.5, Y:= 0.5}}
                    Offsets:=margin{Left:=-157.0,Top:=140.0}
                canvas_slot:
                    Widget:=CustomPlayer.PetM1ViewButton
                    ZOrder:=1
                    SizeToContent:=false
                    Alignment := vector2{X := 0.0, Y := 0.0}
                    Anchors := anchors{Minimum := vector2{X:= 0.5, Y:= 0.5}, Maximum := vector2{X:= 0.5, Y:= 0.5}}
                    Offsets:=margin{Left:=-161.0,Top:=-100.0,Right:=200.0,Bottom:=60.0}
                canvas_slot:
                    Widget:=CustomPlayer.PetM2ViewButton
                    ZOrder:=1
                    SizeToContent:=false
                    Alignment := vector2{X := 0.0, Y := 0.0}
                    Anchors := anchors{Minimum := vector2{X:= 0.5, Y:= 0.5}, Maximum := vector2{X:= 0.5, Y:= 0.5}}
                    Offsets:=margin{Left:=-161.0,Top:=224.0,Right:=200.0,Bottom:=60.0}
                
                # Top Right Section
                canvas_slot:
                    Widget:=CustomPlayer.PetMSPetStatsIcon
                    ZOrder:=1
                    SizeToContent:=false
                    Alignment := vector2{X := 0.0, Y := 0.0}
                    Anchors := anchors{Minimum := vector2{X:= 0.5, Y:= 0.5}, Maximum := vector2{X:= 0.5, Y:= 0.5}}
                    Offsets:=margin{Left:=140.0,Top:=-272.0,Right:=200.0,Bottom:=200.0}
                canvas_slot:
                    Widget:=CustomPlayer.PetMGatherIcon
                    ZOrder:=1
                    SizeToContent:=false
                    Alignment := vector2{X := 0.0, Y := 0.0}
                    Anchors := anchors{Minimum := vector2{X:= 0.5, Y:= 0.5}, Maximum := vector2{X:= 0.5, Y:= 0.5}}
                    Offsets:=margin{Left:=380.0,Top:=-320.0,Right:=55.0,Bottom:=55.0}
                canvas_slot:
                    Widget:=CustomPlayer.PetMStrengthIcon
                    ZOrder:=1
                    SizeToContent:=false
                    Alignment := vector2{X := 0.0, Y := 0.0}
                    Anchors := anchors{Minimum := vector2{X:= 0.5, Y:= 0.5}, Maximum := vector2{X:= 0.5, Y:= 0.5}}
                    Offsets:=margin{Left:=380.0,Top:=-260.0,Right:=55.0,Bottom:=55.0}
                canvas_slot:
                    Widget:=CustomPlayer.PetMCritRateIcon
                    ZOrder:=1
                    SizeToContent:=false
                    Alignment := vector2{X := 0.0, Y := 0.0}
                    Anchors := anchors{Minimum := vector2{X:= 0.5, Y:= 0.5}, Maximum := vector2{X:= 0.5, Y:= 0.5}}
                    Offsets:=margin{Left:=380.0,Top:=-200.0,Right:=55.0,Bottom:=55.0}
                canvas_slot:
                    Widget:=CustomPlayer.PetMCritDMGIcon
                    ZOrder:=1
                    SizeToContent:=false
                    Alignment := vector2{X := 0.0, Y := 0.0}
                    Anchors := anchors{Minimum := vector2{X:= 0.5, Y:= 0.5}, Maximum := vector2{X:= 0.5, Y:= 0.5}}
                    Offsets:=margin{Left:=380.0,Top:=-140.0,Right:=55.0,Bottom:=55.0}
                canvas_slot:
                    Widget:=CustomPlayer.PetMFSIcon
                    ZOrder:=1
                    SizeToContent:=false
                    Alignment := vector2{X := 0.0, Y := 0.0}
                    Anchors := anchors{Minimum := vector2{X:= 0.5, Y:= 0.5}, Maximum := vector2{X:= 0.5, Y:= 0.5}}
                    Offsets:=margin{Left:=380.0,Top:=-80.0,Right:=55.0,Bottom:=55.0}
                canvas_slot:
                    Widget:=CustomPlayer.PetMGatherTitle
                    ZOrder:=1
                    SizeToContent:=true
                    Alignment := vector2{X := 0.0, Y := 0.0}
                    Anchors := anchors{Minimum := vector2{X:= 0.5, Y:= 0.5}, Maximum := vector2{X:= 0.5, Y:= 0.5}}
                    Offsets:=margin{Left:=436.0,Top:=-312.0}
                canvas_slot:
                    Widget:=CustomPlayer.PetMGatherAmount
                    ZOrder:=1
                    SizeToContent:=true
                    Alignment := vector2{X := 1.0, Y := 0.0}
                    Anchors := anchors{Minimum := vector2{X:= 0.5, Y:= 0.5}, Maximum := vector2{X:= 0.5, Y:= 0.5}}
                    Offsets:=margin{Left:=780.0,Top:=-312.0}
                canvas_slot:
                    Widget:=CustomPlayer.PetMStrengthTitle
                    ZOrder:=1
                    SizeToContent:=true
                    Alignment := vector2{X := 0.0, Y := 0.0}
                    Anchors := anchors{Minimum := vector2{X:= 0.5, Y:= 0.5}, Maximum := vector2{X:= 0.5, Y:= 0.5}}
                    Offsets:=margin{Left:=436.0,Top:=-252.0}
                canvas_slot:
                    Widget:=CustomPlayer.PetMStrengthAmount
                    ZOrder:=1
                    SizeToContent:=true
                    Alignment := vector2{X := 1.0, Y := 0.0}
                    Anchors := anchors{Minimum := vector2{X:= 0.5, Y:= 0.5}, Maximum := vector2{X:= 0.5, Y:= 0.5}}
                    Offsets:=margin{Left:=780.0,Top:=-252.0}
                canvas_slot:
                    Widget:=CustomPlayer.PetMCritRateTitle
                    ZOrder:=1
                    SizeToContent:=true
                    Alignment := vector2{X := 0.0, Y := 0.0}
                    Anchors := anchors{Minimum := vector2{X:= 0.5, Y:= 0.5}, Maximum := vector2{X:= 0.5, Y:= 0.5}}
                    Offsets:=margin{Left:=436.0,Top:=-192.0}
                canvas_slot:
                    Widget:=CustomPlayer.PetMCritRateAmount
                    ZOrder:=1
                    SizeToContent:=true
                    Alignment := vector2{X := 1.0, Y := 0.0}
                    Anchors := anchors{Minimum := vector2{X:= 0.5, Y:= 0.5}, Maximum := vector2{X:= 0.5, Y:= 0.5}}
                    Offsets:=margin{Left:=780.0,Top:=-192.0}
                canvas_slot:
                    Widget:=CustomPlayer.PetMCritDMGTitle
                    ZOrder:=1
                    SizeToContent:=true
                    Alignment := vector2{X := 0.0, Y := 0.0}
                    Anchors := anchors{Minimum := vector2{X:= 0.5, Y:= 0.5}, Maximum := vector2{X:= 0.5, Y:= 0.5}}
                    Offsets:=margin{Left:=436.0,Top:=-132.0}
                canvas_slot:
                    Widget:=CustomPlayer.PetMCritDMGAmount
                    ZOrder:=1
                    SizeToContent:=true
                    Alignment := vector2{X := 1.0, Y := 0.0}
                    Anchors := anchors{Minimum := vector2{X:= 0.5, Y:= 0.5}, Maximum := vector2{X:= 0.5, Y:= 0.5}}
                    Offsets:=margin{Left:=780.0,Top:=-132.0}
                canvas_slot:
                    Widget:=CustomPlayer.PetMFSTitle
                    ZOrder:=1
                    SizeToContent:=true
                    Alignment := vector2{X := 0.0, Y := 0.0}
                    Anchors := anchors{Minimum := vector2{X:= 0.5, Y:= 0.5}, Maximum := vector2{X:= 0.5, Y:= 0.5}}
                    Offsets:=margin{Left:=436.0,Top:=-72.0}
                canvas_slot:
                    Widget:=CustomPlayer.PetMFSAmounnt
                    ZOrder:=1
                    SizeToContent:=true
                    Alignment := vector2{X := 1.0, Y := 0.0}
                    Anchors := anchors{Minimum := vector2{X:= 0.5, Y:= 0.5}, Maximum := vector2{X:= 0.5, Y:= 0.5}}
                    Offsets:=margin{Left:=780.0,Top:=-72.0}

                # Bottom Right Section
                canvas_slot:
                    Widget:=CustomPlayer.PetMStoneCostIcon
                    ZOrder:=1
                    SizeToContent:=false
                    Alignment := vector2{X := 0.0, Y := 0.0}
                    Anchors := anchors{Minimum := vector2{X:= 0.5, Y:= 0.5}, Maximum := vector2{X:= 0.5, Y:= 0.5}}
                    Offsets:=margin{Left:=216.0,Top:=28.0,Right:=50.0,Bottom:=50.0}
                canvas_slot:
                    Widget:=CustomPlayer.PetMPrestigeCostIcon
                    ZOrder:=1
                    SizeToContent:=false
                    Alignment := vector2{X := 0.0, Y := 0.0}
                    Anchors := anchors{Minimum := vector2{X:= 0.5, Y:= 0.5}, Maximum := vector2{X:= 0.5, Y:= 0.5}}
                    Offsets:=margin{Left:=216.0,Top:=88.0,Right:=50.0,Bottom:=50.0}
                canvas_slot:
                    Widget:=CustomPlayer.PetMEggCostIcon
                    ZOrder:=1
                    SizeToContent:=false
                    Alignment := vector2{X := 0.0, Y := 0.0}
                    Anchors := anchors{Minimum := vector2{X:= 0.5, Y:= 0.5}, Maximum := vector2{X:= 0.5, Y:= 0.5}}
                    Offsets:=margin{Left:=216.0,Top:=148.0,Right:=50.0,Bottom:=50.0}
                canvas_slot:
                    Widget:=CustomPlayer.PetMAllStoneCostIcon
                    ZOrder:=1
                    SizeToContent:=false
                    Alignment := vector2{X := 0.0, Y := 0.0}
                    Anchors := anchors{Minimum := vector2{X:= 0.5, Y:= 0.5}, Maximum := vector2{X:= 0.5, Y:= 0.5}}
                    Offsets:=margin{Left:=492.0,Top:=28.0,Right:=50.0,Bottom:=50.0}
                canvas_slot:
                    Widget:=CustomPlayer.PetMAllPrestigeCostIcon
                    ZOrder:=1
                    SizeToContent:=false
                    Alignment := vector2{X := 0.0, Y := 0.0}
                    Anchors := anchors{Minimum := vector2{X:= 0.5, Y:= 0.5}, Maximum := vector2{X:= 0.5, Y:= 0.5}}
                    Offsets:=margin{Left:=492.0,Top:=88.0,Right:=50.0,Bottom:=50.0}
                canvas_slot:
                    Widget:=CustomPlayer.PetMAllEggCostIcon
                    ZOrder:=1
                    SizeToContent:=false
                    Alignment := vector2{X := 0.0, Y := 0.0}
                    Anchors := anchors{Minimum := vector2{X:= 0.5, Y:= 0.5}, Maximum := vector2{X:= 0.5, Y:= 0.5}}
                    Offsets:=margin{Left:=492.0,Top:=148.0,Right:=50.0,Bottom:=50.0}
                canvas_slot:
                    Widget:=CustomPlayer.PetMWhichEggIcon
                    ZOrder:=1
                    SizeToContent:=false
                    Alignment := vector2{X := 0.0, Y := 0.0}
                    Anchors := anchors{Minimum := vector2{X:= 0.5, Y:= 0.5}, Maximum := vector2{X:= 0.5, Y:= 0.5}}
                    Offsets:=margin{Left:=420.0,Top:=216.0,Right:=50.0,Bottom:=50.0}
                canvas_slot:
                    Widget:=CustomPlayer.PetMCostTitle
                    ZOrder:=1
                    SizeToContent:=true
                    Alignment := vector2{X := 0.0, Y := 0.0}
                    Anchors := anchors{Minimum := vector2{X:= 0.5, Y:= 0.5}, Maximum := vector2{X:= 0.5, Y:= 0.5}}
                    Offsets:=margin{Left:=123.0,Top:=32.0}
                canvas_slot:
                    Widget:=CustomPlayer.PetMStoneCostTitle
                    ZOrder:=1
                    SizeToContent:=true
                    Alignment := vector2{X := 0.5, Y := 0.0}
                    Anchors := anchors{Minimum := vector2{X:= 0.5, Y:= 0.5}, Maximum := vector2{X:= 0.5, Y:= 0.5}}
                    Offsets:=margin{Left:=344.0,Top:=36.0}
                canvas_slot:
                    Widget:=CustomPlayer.PetMPrestigeCostTitle
                    ZOrder:=1
                    SizeToContent:=true
                    Alignment := vector2{X := 0.5, Y := 0.0}
                    Anchors := anchors{Minimum := vector2{X:= 0.5, Y:= 0.5}, Maximum := vector2{X:= 0.5, Y:= 0.5}}
                    Offsets:=margin{Left:=344.0,Top:=96.0}
                canvas_slot:
                    Widget:=CustomPlayer.PetMEggCostTitle
                    ZOrder:=1
                    SizeToContent:=true
                    Alignment := vector2{X := 0.5, Y := 0.0}
                    Anchors := anchors{Minimum := vector2{X:= 0.5, Y:= 0.5}, Maximum := vector2{X:= 0.5, Y:= 0.5}}
                    Offsets:=margin{Left:=344.0,Top:=156.0}
                canvas_slot:
                    Widget:=CustomPlayer.PetMAllStoneCostTitle
                    ZOrder:=1
                    SizeToContent:=true
                    Alignment := vector2{X := 0.5, Y := 0.0}
                    Anchors := anchors{Minimum := vector2{X:= 0.5, Y:= 0.5}, Maximum := vector2{X:= 0.5, Y:= 0.5}}
                    Offsets:=margin{Left:=644.0,Top:=36.0}
                canvas_slot:
                    Widget:=CustomPlayer.PetMAllPrestigeCostTitle
                    ZOrder:=1
                    SizeToContent:=true
                    Alignment := vector2{X := 0.5, Y := 0.0}
                    Anchors := anchors{Minimum := vector2{X:= 0.5, Y:= 0.5}, Maximum := vector2{X:= 0.5, Y:= 0.5}}
                    Offsets:=margin{Left:=644.0,Top:=96.0}
                canvas_slot:
                    Widget:=CustomPlayer.PetMAllEggCostTitle
                    ZOrder:=1
                    SizeToContent:=true
                    Alignment := vector2{X := 0.5, Y := 0.0}
                    Anchors := anchors{Minimum := vector2{X:= 0.5, Y:= 0.5}, Maximum := vector2{X:= 0.5, Y:= 0.5}}
                    Offsets:=margin{Left:=644.0,Top:=156.0}
                canvas_slot:
                    Widget:=CustomPlayer.PetMBackButton
                    ZOrder:=7
                    SizeToContent:=false
                    Alignment := vector2{X := 0.0, Y := 0.0}
                    Anchors := anchors{Minimum := vector2{X:= 0.5, Y:= 0.5}, Maximum := vector2{X:= 0.5, Y:= 0.5}}
                    Offsets:=margin{Left:=124.0,Top:=220.0,Right:=275.0,Bottom:=40.0}
                canvas_slot:
                    Widget:=CustomPlayer.BackArrow
                    ZOrder:=4
                    SizeToContent:=false
                    Anchors := anchors{Minimum := vector2{X:= 0.5, Y:= 0.0}, Maximum := vector2{X:= 0.5, Y:= 0.0}}
                    Offsets:=margin{Top:=752.0,Left:=240.0, Right:=50.0, Bottom:=50.0}
                canvas_slot:
                    Widget:=CustomPlayer.PetMForwardButton
                    ZOrder:=7
                    SizeToContent:=false
                    Alignment := vector2{X := 0.0, Y := 0.0}
                    Anchors := anchors{Minimum := vector2{X:= 0.5, Y:= 0.5}, Maximum := vector2{X:= 0.5, Y:= 0.5}}
                    Offsets:=margin{Left:=492.0,Top:=220.0,Right:=275.0,Bottom:=40.0}
                canvas_slot:
                    Widget:=CustomPlayer.ForwardArrow
                    ZOrder:=4
                    SizeToContent:=false
                    Anchors := anchors{Minimum := vector2{X:= 0.5, Y:= 0.0}, Maximum := vector2{X:= 0.5, Y:= 0.0}}
                    Offsets:=margin{Top:=752.0,Left:=624.0, Right:=50.0, Bottom:=50.0}
                <# canvas_slot:
                    Widget:=CustomPlayer.PetMHatchButton
                    ZOrder:=1
                    SizeToContent:=false
                    Alignment := vector2{X := 0.0, Y := 0.0}
                    Anchors := anchors{Minimum := vector2{X:= 0.5, Y:= 0.5}, Maximum := vector2{X:= 0.5, Y:= 0.5}}
                    Offsets:=margin{Left:=124.0,Top:=287.0,Right:=275.0,Bottom:=55.0} #>
                canvas_slot:
                    Widget:=CustomPlayer.PetMHatchAllButton
                    ZOrder:=1
                    SizeToContent:=false
                    Alignment := vector2{X := 0.0, Y := 0.0}
                    Anchors := anchors{Minimum := vector2{X:= 0.5, Y:= 0.5}, Maximum := vector2{X:= 0.5, Y:= 0.5}}
                    Offsets:=margin{Left:=492.0,Top:=287.0,Right:=275.0,Bottom:=55.0}
        }