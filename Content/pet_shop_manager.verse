using { /Fortnite.com/Devices }
using { /Verse.org/Simulation }
using { /Verse.org/Colors }
using { /Verse.org/Verse }
using { /Verse.org/Random }
using { /Fortnite.com/Teams }
using { /UnrealEngine.com/Assets }
using { /Fortnite.com/Game }
using { /Fortnite.com/Characters }
using { /Fortnite.com/Playspaces }
using { /Fortnite.com/UI }
using { /UnrealEngine.com/Temporary/Diagnostics }
using { /UnrealEngine.com/Temporary/SpatialMath }
using { /UnrealEngine.com/Temporary/UI }
using { /Verse.org/Assets}

# A Verse-authored creative device that can be placed in a level
pet_shop_manager := class(creative_device):

    @editable
    gm : game_manager = game_manager{}

    @editable
    Tutorial : tutorial_manager = tutorial_manager{}

    @editable
    AudioManager : audio_manager = audio_manager{}

    @editable
    ShopButton : button_device = button_device{}

    var PetTextures : []texture = array{UI.pet1, UI.pet2, UI.pet3, UI.pet4, UI.pet5, UI.pet6}
    var EggTextures : []texture = array{UI.Egg1, UI.Egg2, UI.Egg3, UI.Egg4, UI.Egg5, UI.Egg6}
    var TiersTexts : []string = array{
        "Common - Uncommon",
        "Common - Rare",
        "Uncommon - Epic",
        "Rare - Legendary",
        "Epic - Mythic",
        "Epic - Mythic"
    }
    var StoneCosts : []float = array {5.00e02, 7.50e06, 3.12e13, 4.68e17, 1.95e24, 2.92e28} #change these values
    var PetInitialBonusValues : []float = array{2.0,4.0,8.0,16.0,28.0,44.0}


    var PetMUpgradeAmounts : []float = array{0.1, 0.1, 0.1, 0.1, 0.1, 0.1}
    var PetMMaxLevels : []int = array{10, 10, 10, 10, 10, 10}
    var PetMCostMult : []float = array{3.0, 3.0, 3.0, 3.0, 3.0, 3.0}
    var PetMStoneCosts : []float = array{1e3, 1e3, 1e3, 1e3, 1e3, 1e3}
    var PetMStardustCosts : []float = array{1e3, 1e3, 1e3, 1e3, 1e3, 1e3}


    var EggOutcomes : [int][]int = map{ # 0 = common, 1 = uncommon, 2 = rare, 3 = epic, 4 = legendary, 5 = mythic
        0 => array{0, 1},
        1 => array{0, 1, 2},
        2 => array{1, 2, 3},
        3 => array{2, 3, 4},
        4 => array{3, 4, 5},
        5 => array{3, 4, 5}
    }

    var A5EggOutcomes : [int][]int = map{ # 0 = common, 1 = uncommon, 2 = rare, 3 = epic, 4 = legendary, 5 = mythic
        0 => array{0, 1},
        1 => array{0, 1, 2},
        2 => array{2, 3, 4},
        3 => array{2, 3, 4},
        4 => array{3, 4, 5},
        5 => array{3, 4, 5}
    }

    var EggProbabilities : [int][]float = map{
        0 => array{0.7, 0.3},
        1 => array{0.5, 0.4, 0.1},
        2 => array{0.4, 0.3, 0.2},
        3 => array{0.4, 0.3, 0.2},
        4 => array{0.5, 0.4, 0.1},
        5 => array{0.4, 0.3, 0.2}
    }

    var A5EggProbabilities : [int][]float = map{
        0 => array{0.7, 0.3},
        1 => array{0.5, 0.4, 0.1},
        2 => array{0.6, 0.3, 0.1},
        3 => array{0.4, 0.3, 0.2},
        4 => array{0.5, 0.4, 0.1},
        5 => array{0.4, 0.3, 0.2}
    }

    SelectWhichEgg(CustomPlayer:Custom_Player, WhichPage : int):void=
        var CurrentChance : float = 0.0
        if:
            Outcomes := EggOutcomes[WhichPage]
            Probabilities := EggProbabilities[WhichPage]
            A5Outcomes := A5EggOutcomes[WhichPage]
            A5Probabilities := A5EggProbabilities[WhichPage]
        then:
            IDX := GetRandomFloat(0.0,1.0)
            for(i:int=0..Probabilities.Length-1):
                if (not CustomPlayer.A5Unlocked?):
                    if (set CurrentChance += Probabilities[i]){}
                    if (IDX <= CurrentChance):
                        if:
                            Egg := Outcomes[i]
                        then:
                            set CustomPlayer.SelectedEgg = Egg
                            return
                else:
                    if (set CurrentChance += A5Probabilities[i]){}
                    if (IDX <= CurrentChance):
                        if:
                            Egg := A5Outcomes[i]
                        then:
                            set CustomPlayer.SelectedEgg = Egg
                            return
    
    SelectWhichEggAll(CustomPlayer:Custom_Player, WhichPage : int, Hatches:int):[int]int=
        var CurrentChance : float = 0.0
        var Output : [int]int = map{
            0 => 0,
            1 => 0,
            2 => 0,
            3 => 0,
            4 => 0,
            5 => 0
        }
        if:
            Outcomes := EggOutcomes[WhichPage]
            Probabilities := EggProbabilities[WhichPage]
            A5Outcomes := A5EggOutcomes[WhichPage]
            A5Probabilities := A5EggProbabilities[WhichPage]
        then:
            for(i:int=0..Hatches-1):
                IDX := GetRandomFloat(0.0,1.0)
                for(x:int=0..Probabilities.Length-1):
                    if (not CustomPlayer.A5Unlocked?):
                        if (set CurrentChance += Probabilities[x]){}
                        if (IDX <= CurrentChance):
                            if:
                                Egg := Outcomes[x]
                            then:
                                if (set Output[Egg] += 1){}
                                set CurrentChance = 0.0
                    else:
                        if (set CurrentChance += A5Probabilities[x]){}
                        if (IDX <= CurrentChance):
                            if:
                                Egg := A5Outcomes[x]
                            then:
                                if (set Output[Egg] += 1){}
                                set CurrentChance = 0.0
        return Output

    # Runs when the device is started in a running game
    OnBegin<override>()<suspends>:void=
        ShopButton.InteractedWithEvent.Subscribe(OpenShopHUD)

        for(Spawner : gm.PlayerSpawners):     
            Spawner.SpawnedEvent.Subscribe(OnPlayerJoined)

    OnPlayerJoined(Agent:agent):void=
        spawn{OnPlayerJoinedBuffer(Agent)}

    OnPlayerJoinedBuffer(Agent:agent)<suspends>:void=
        Sleep(1.0)
        if(CustomPlayer := gm.PlayersMap[player[Agent]]):
            if(CustomPlayer.PlayTimeSessionSeconds = 0.0):
                set CustomPlayer.EggAmountsText = array{CustomPlayer.InvEgg1, CustomPlayer.InvEgg2, CustomPlayer.InvEgg3, CustomPlayer.InvEgg4, CustomPlayer.InvEgg5, CustomPlayer.InvEgg6}
                set CustomPlayer.PetLevelsText = array{CustomPlayer.Pet1Level, CustomPlayer.Pet2Level, CustomPlayer.Pet3Level, CustomPlayer.Pet4Level, CustomPlayer.Pet5Level, CustomPlayer.Pet6Level}
                CustomPlayer.PetShopCloseButton.OnClick().Subscribe(HandleExit)
                CustomPlayer.Pet1MenuButton.OnClick().SubscribeAgent3(HandlePetSelection, 0)
                CustomPlayer.Pet2MenuButton.OnClick().SubscribeAgent3(HandlePetSelection, 1)
                CustomPlayer.Pet3MenuButton.OnClick().SubscribeAgent3(HandlePetSelection, 2)
                CustomPlayer.Pet4MenuButton.OnClick().SubscribeAgent3(HandlePetSelection, 3)
                CustomPlayer.Pet5MenuButton.OnClick().SubscribeAgent3(HandlePetSelection, 4)
                CustomPlayer.Pet6MenuButton.OnClick().SubscribeAgent3(HandlePetSelection, 5)
                CustomPlayer.BackButton.OnClick().SubscribeAgent3(HandlePageSelection, 0)
                CustomPlayer.ForwardButton.OnClick().SubscribeAgent3(HandlePageSelection, 1)
                CustomPlayer.HatchButton.OnClick().SubscribeAgent3(HandlePetHatch, false)
                CustomPlayer.HatchAllButton.OnClick().Subscribe(HandlePetHatchAll)

                set CustomPlayer.PSAllTexts = array{CustomPlayer.InvStone,CustomPlayer.InvEgg1,CustomPlayer.InvEgg2,CustomPlayer.InvEgg3,CustomPlayer.InvEgg4,CustomPlayer.InvEgg5,CustomPlayer.InvEgg6,CustomPlayer.InvWepMat,
                CustomPlayer.PetsShopTitle,CustomPlayer.Pet1Level,CustomPlayer.Pet1Tier,CustomPlayer.Pet2Level,CustomPlayer.Pet2Tier,CustomPlayer.Pet3Level,CustomPlayer.Pet3Tier,CustomPlayer.Pet4Level,CustomPlayer.Pet4Tier,CustomPlayer.Pet5Level,CustomPlayer.Pet5Tier,CustomPlayer.Pet6Level,CustomPlayer.Pet6Tier,
                CustomPlayer.FinalStrikeText,CustomPlayer.FinalStrikeAmount,CustomPlayer.CritDMGAmount,CustomPlayer.CritDMGText,CustomPlayer.CritRateAmount,CustomPlayer.CritRateText,CustomPlayer.AttackAmount,CustomPlayer.StrengthText,CustomPlayer.PetGatherAmount,CustomPlayer.PetGatherText,
                CustomPlayer.TiersTitle,CustomPlayer.TiersText,CustomPlayer.CostTitle,CustomPlayer.StoneCostText,CustomPlayer.EggCostText,CustomPlayer.AllStoneCostText,CustomPlayer.AllEggCostText}
            
                if (not CustomPlayer.PetDataReset?):
                    Print("Resetting Data")
                    set CustomPlayer.PetDataReset = true

                    for(i:int=0..CustomPlayer.PetLevels.Length-1):
                        Print("Resetting Pet {i}")
                        if:
                            PetLevel := CustomPlayer.PetLevels[i]
                            PetGather := (PetInitialBonusValues[i] * 1.0) * (PetLevel * 0.2)
                            PetAttack := (PetInitialBonusValues[i] * 1.0) * (PetLevel * 0.2)
                            PetCritRate := (PetInitialBonusValues[i] * 0.1) * (PetLevel * 0.1)
                            PetCritDMG := (PetInitialBonusValues[i] * 0.1) * (PetLevel * 0.1)
                            PetFinalStrike :=  (PetInitialBonusValues[i] * 0.15) * (PetLevel  * 0.1)
                        then:
                            if(set CustomPlayer.PetGatherRates[i] = 0.0){}
                            if(set CustomPlayer.PetAttackRates[i] = 0.0){}
                            if(set CustomPlayer.PetCritRates[i] = 0.0){}
                            if(set CustomPlayer.PetCritDMG[i] = 0.0){}
                            if(set CustomPlayer.PetFinalStrike[i] = 0.0){}
                            if(set CustomPlayer.PetGatherRates[i] = PetGather){}
                            if(set CustomPlayer.PetAttackRates[i] = PetAttack){}
                            if (i >= 2):
                                if(set CustomPlayer.PetCritRates[i] = PetCritRate){}
                                if(set CustomPlayer.PetCritDMG[i] = PetCritDMG){}
                            if (i >= 4):
                                if(set CustomPlayer.PetFinalStrike[i] = PetFinalStrike){}
                    SaveAllData(CustomPlayer)

    #PET SHOP UI    
    OpenShopHUD(Agent:agent): void=
        if(CustomPlayer := gm.PlayersMap[Agent]):
            if(MyUI:=CustomPlayer.UIMaps["PetShop"]?):
            else:
                var MyPetCanvas : canvas = MakePetShopCanvas(CustomPlayer)
                set CustomPlayer.PetCurrentPage = 0
                set CustomPlayer.PetSelected = 0
                CustomPlayer.AddFullHUD(MyPetCanvas)
                if(set CustomPlayer.UIMaps["PetShop"] = option{MyPetCanvas}) {}
                UpdateShopHUD(CustomPlayer)

    HandleExit(Message : widget_message):void=
        if(CustomPlayer := gm.PlayersMap[Message.Player]):
            if(Player:=Message.Player, PlayerUI:=GetPlayerUI[Player]):
                if(MyUI:=CustomPlayer.UIMaps["PetShop"]?):
                    PlayerUI.RemoveWidget(MyUI)
                    MyUI.SetVisibility(widget_visibility.Collapsed)
                    if(set CustomPlayer.UIMaps["PetShop"] = false){}

    HandlePetSelection(Message : widget_message, WhichPet : int):void=
        if(CustomPlayer := gm.PlayersMap[Message.Player]):
            set CustomPlayer.PetSelected = WhichPet
            set CustomPlayer.PetCurrentPage = WhichPet
            UpdateShopHUD(CustomPlayer)
    
    HandlePageSelection(Message: widget_message, WhichDirection : int):void=
        if(CustomPlayer := gm.PlayersMap[Message.Player]):
            if (WhichDirection = 0):
                if(CustomPlayer.PetCurrentPage > 0):
                    set CustomPlayer.PetCurrentPage -= 1
                    set CustomPlayer.PetSelected = CustomPlayer.PetCurrentPage
                else:
                    AudioManager.Denied.Play(CustomPlayer.Player)
            if (WhichDirection = 1):
                if(CustomPlayer.PetCurrentPage < 5):
                    set CustomPlayer.PetCurrentPage += 1
                    set CustomPlayer.PetSelected = CustomPlayer.PetCurrentPage
                else:
                    AudioManager.Denied.Play(CustomPlayer.Player)
                    
            UpdateShopHUD(CustomPlayer)

    HandlePetHatch(Message: widget_message, HatchAll:logic):void=
        if(CustomPlayer := gm.PlayersMap[Message.Player]):
            if (CustomPlayer.A5Unlocked? and (CustomPlayer.PetCurrentPage = 0 or CustomPlayer.PetCurrentPage = 1)):
                CustomPlayer.PetHatchMessage.SetText(StringToMessage("Must use Pet Mutations in Area 5!"))
            else:
                CustomPlayer.PetHatchMessage.SetText(StringToMessage(""))
                SelectWhichEgg(CustomPlayer, CustomPlayer.PetCurrentPage)
                Print("{CustomPlayer.SelectedEgg}")
                if:
                    EggCost:=Floor[1.0 + (CustomPlayer.EggsHatched[CustomPlayer.PetCurrentPage]/25.0)]
                    AmountOfEggs := CustomPlayer.EggAmounts[CustomPlayer.PetCurrentPage]
                    AmountOfEggs >= EggCost*1.0
                    StoneCost:=StoneCosts[CustomPlayer.SelectedEgg]
                    CustomPlayer.StoneAmount >= StoneCost
                    
                    PetLevel := CustomPlayer.PetLevels[CustomPlayer.SelectedEgg]
                    PetGather := (PetInitialBonusValues[CustomPlayer.SelectedEgg] * 1.0) * ((PetLevel + 1) * 0.2)
                    PetAttack := (PetInitialBonusValues[CustomPlayer.SelectedEgg] * 1.0) * ((PetLevel + 1) * 0.2)
                    PetCritRate := (PetInitialBonusValues[CustomPlayer.SelectedEgg] * 0.1) * ((PetLevel + 1) * 0.1)
                    PetCritDMG := (PetInitialBonusValues[CustomPlayer.SelectedEgg] * 0.1) * ((PetLevel + 1) * 0.1)
                    PetFinalStrike :=  (PetInitialBonusValues[CustomPlayer.SelectedEgg] * 0.15) * ((PetLevel + 1) * 0.1)
                then:
                    if (set CustomPlayer.PetLevels[CustomPlayer.SelectedEgg] += 1){}
                    if(set CustomPlayer.EggsHatched[CustomPlayer.PetCurrentPage] += 1.0) {}
                    Tutorial.UpdateTutorialUI(CustomPlayer, "HatchPet", 1.0)
                    CustomPlayer.UpAchValue(5, 1.0)
                    CustomPlayer.Spend(EggCost*1.0, "Egg", CustomPlayer.PetCurrentPage)
                    CustomPlayer.Spend(StoneCost, "Stone", 0)
                    if(set CustomPlayer.PetGatherRates[CustomPlayer.SelectedEgg] = PetGather){}
                    if(set CustomPlayer.PetAttackRates[CustomPlayer.SelectedEgg] = PetAttack) {}
                    if (CustomPlayer.SelectedEgg >= 2):
                        if(set CustomPlayer.PetCritRates[CustomPlayer.SelectedEgg] = PetCritRate) {}
                        if(set CustomPlayer.PetCritDMG[CustomPlayer.SelectedEgg] = PetCritDMG) {}
                    if (CustomPlayer.SelectedEgg >= 4):
                        if(set CustomPlayer.PetFinalStrike[CustomPlayer.SelectedEgg] = PetFinalStrike) {}
                    if (PetUnlocked := CustomPlayer.PetUnlocked[CustomPlayer.SelectedEgg], PetUnlocked = false): #IF YOU DON'T OWN PET YET
                        if(Accolade:=AudioManager.Accolades[5]): #AWARD PLAYER EXP
                            Accolade.Award(CustomPlayer.Player)
                        if (set CustomPlayer.PetUnlocked[CustomPlayer.SelectedEgg] = true){}
                    if(Analytic:=AudioManager.GeneralAnalytics[3]): #ANALYTIC
                        Analytic.Submit(CustomPlayer.Player)
                else:
                    if(HatchAll = false):
                        AudioManager.Denied.Play(CustomPlayer.Player)

                if(HatchAll = false):
                    SaveAllData(CustomPlayer)
                UpdateShopHUD(CustomPlayer)

    TotalEggsNeeded(CustomPlayer:Custom_Player):tuple(int,int)=
        var TotalEggCost : int = 0
        var Hatches : int = 0
        var EggAmount : int = 0
        var CurrentCost : int = 0
        var CurrentHatched : int = 0
        if:
            EggsHatched := Floor[CustomPlayer.EggsHatched[CustomPlayer.PetCurrentPage]]
            FlooredCost := Floor[(CustomPlayer.EggsHatched[CustomPlayer.PetCurrentPage] / 25.0) + 1.0]
            FlooredEggs := Floor[CustomPlayer.EggAmounts[CustomPlayer.PetCurrentPage]]
        then:
            set CurrentHatched = EggsHatched
            set CurrentCost = FlooredCost
            set EggAmount = FlooredEggs
            loop:
                if(EggAmount > 0):
                    if(EggsToNextBlock := (25 - Mod[CurrentHatched, 25]) * CurrentCost):
                        if(EggAmount >= EggsToNextBlock):
                            set TotalEggCost += EggsToNextBlock
                            set Hatches += 25
                            set EggAmount -= EggsToNextBlock
                            set CurrentHatched += EggsToNextBlock
                            set CurrentCost += 1
                        else:
                            if(HatchesLeft:=Floor[(EggAmount*1.0) / (CurrentCost*1.0)]):
                                set TotalEggCost += HatchesLeft * CurrentCost
                                set Hatches += HatchesLeft
                            set EggAmount = 0
                            break:
                else:
                    break:

        return (TotalEggCost, Hatches)
    
    HandlePetHatchAll(Message: widget_message):void=
        if(CustomPlayer := gm.PlayersMap[Message.Player]):
            if (CustomPlayer.A5Unlocked? and (CustomPlayer.PetCurrentPage = 0 or CustomPlayer.PetCurrentPage = 1)):
                CustomPlayer.PetHatchMessage.SetText(StringToMessage("Must use Pet Mutations in Area 5!"))
            else:
                CustomPlayer.PetHatchMessage.SetText(StringToMessage(""))
                CustomPlayer.ForwardArrow.SetEnabled(false)
                CustomPlayer.BackArrow.SetEnabled(false)
                EggInfo := TotalEggsNeeded(CustomPlayer)
                Print("{EggInfo(1)}")
                WhichEggsToHatch := SelectWhichEggAll(CustomPlayer,CustomPlayer.PetCurrentPage,EggInfo(1))
                Print("{WhichEggsToHatch.Length}")
                if:
                    EggInfo(1) >= 1
                    StoneCost:=StoneCosts[CustomPlayer.PetCurrentPage] * EggInfo(1)
                    CustomPlayer.StoneAmount > StoneCost
                then:
                    Print("StoneCost: {StoneCost}")
                    CustomPlayer.Spend(EggInfo(0)*1.0, "Egg", CustomPlayer.PetCurrentPage)
                    CustomPlayer.Spend(StoneCost, "Stone", 0)
                    CustomPlayer.UpAchValue(5, EggInfo(1)*1.0)
                    Tutorial.UpdateTutorialUI(CustomPlayer, "HatchPet", 1.0)
                    if(set CustomPlayer.EggsHatched[CustomPlayer.PetCurrentPage] += EggInfo(1)*1.0) {}
                    for (Key->Value:WhichEggsToHatch):
                        if(set CustomPlayer.PetLevels[Key] += WhichEggsToHatch[Key]){}
                        if:
                            PetUnlocked := CustomPlayer.PetUnlocked[Key]
                            PetUnlocked = false
                        then:
                            if(Accolade:=AudioManager.Accolades[5]):
                                Accolade.Award(CustomPlayer.Player)
                            if (set CustomPlayer.PetUnlocked[Key] = true){}
                        if:
                            PetLevel := CustomPlayer.PetLevels[Key]
                            PetGather := (PetInitialBonusValues[Key] * 1.0) * ((PetLevel) * 0.2)
                            PetAttack := (PetInitialBonusValues[Key] * 1.0) * ((PetLevel) * 0.2)
                            PetCritRate := (PetInitialBonusValues[Key] * 0.1) * ((PetLevel) * 0.1)
                            PetCritDMG := (PetInitialBonusValues[Key] * 0.1) * ((PetLevel) * 0.1)
                            PetFinalStrike :=  (PetInitialBonusValues[Key] * 0.15) * ((PetLevel) * 0.1)
                        then:
                            if(CustomPlayer.A5Unlocked? and (Key=0 or Key =1)):
                            else:
                                if(set CustomPlayer.PetGatherRates[Key] = PetGather){}
                                if(set CustomPlayer.PetAttackRates[Key] = PetAttack) {}
                                if (Key >= 2):
                                    if(set CustomPlayer.PetCritRates[Key] = PetCritRate) {}
                                    if(set CustomPlayer.PetCritDMG[Key] = PetCritDMG) {}
                                if (Key >= 4):
                                    if(set CustomPlayer.PetFinalStrike[Key] = PetFinalStrike) {}
        
                        if(Analytic:=AudioManager.GeneralAnalytics[3]):
                            Analytic.Submit(CustomPlayer.Player)
                else:
                    AudioManager.Denied.Play(CustomPlayer.Player)
                SaveAllData(CustomPlayer)
                CustomPlayer.ForwardArrow.SetEnabled(true)
                CustomPlayer.BackArrow.SetEnabled(true)
                UpdateShopHUD(CustomPlayer)
                

    UpdateShopHUD(CustomPlayer : Custom_Player):void=
        var EggInfo : tuple(int,int) = (0, 0)
        CustomPlayer.InvStone.SetText(StringToMessage("{NumberCondensor(CustomPlayer.StoneAmount)}"))
        if:
            CustomPlayer.A5Unlocked?
            (CustomPlayer.PetCurrentPage = 0 or CustomPlayer.PetCurrentPage = 1)        
        then:
            CustomPlayer.PetHatchMessage.SetText(StringToMessage("Must use Pet Mutations in Area 5!"))
            CustomPlayer.PetHatchBG.SetImage(UI.ButtonOff)
        else:
            CustomPlayer.PetHatchMessage.SetText(StringToMessage(""))
            CustomPlayer.PetHatchBG.SetImage(UI.TransparentButton)

        for(i:int=0..CustomPlayer.PetLevelsText.Length):
            if:
                TextBlock := CustomPlayer.PetLevelsText[i]
                PetLevel := CustomPlayer.PetLevels[i]
            then:
                TextBlock.SetText(StringToMessage("Lv. {NumberCondensor(PetLevel*1.0)}"))

        if:
            PetImage := PetTextures[CustomPlayer.PetSelected]
            PetLevel := CustomPlayer.PetLevels[CustomPlayer.PetSelected]
            PetGather := CustomPlayer.PetGatherRates[CustomPlayer.PetSelected]
            PetAttack := CustomPlayer.PetAttackRates[CustomPlayer.PetSelected]
            PetCritRate := CustomPlayer.PetCritRates[CustomPlayer.PetSelected]
            PetCritDMG := CustomPlayer.PetCritDMG[CustomPlayer.PetSelected]
            PetFinalStrike :=  CustomPlayer.PetFinalStrike[CustomPlayer.PetSelected]
        then:
            CustomPlayer.PetDisplayIMG.SetImage(PetImage)
            CustomPlayer.PetGatherAmount.SetText(StringToMessage("{NumberCondensor(PetGather)}%"))
            CustomPlayer.AttackAmount.SetText(StringToMessage("{NumberCondensor(PetAttack)}%"))
            CustomPlayer.CritRateAmount.SetText(StringToMessage("{NumberCondensor(PetCritRate)}%"))
            CustomPlayer.CritDMGAmount.SetText(StringToMessage("{NumberCondensor(PetCritDMG)}%"))
            CustomPlayer.FinalStrikeAmount.SetText(StringToMessage("{NumberCondensor(PetFinalStrike)}%"))
            if(PetFinalStrike >= 1000.0):
                CustomPlayer.FinalStrikeText.SetText(StringToMessage("FS:"))
                        
        if:
            CustomPlayer.A5Unlocked?
            (CustomPlayer.PetCurrentPage = 0 or CustomPlayer.PetCurrentPage = 1)        
        then:
            Print("YOU HAVE PET MUTATION")
        else:
            Print("NO PET MUTATION")
            set EggInfo = TotalEggsNeeded(CustomPlayer)

        if:
            EggAmount := CustomPlayer.EggAmounts[CustomPlayer.PetCurrentPage]
            EggCost:=Floor[1.0 + (CustomPlayer.EggsHatched[CustomPlayer.PetCurrentPage]/25.0)]
            EggSelected := EggTextures[CustomPlayer.PetCurrentPage]
            TierSelected := TiersTexts[CustomPlayer.PetCurrentPage]
            StoneCost := StoneCosts[CustomPlayer.PetCurrentPage]
        then:
            CustomPlayer.TiersText.SetText(StringToMessage(TierSelected))
            CustomPlayer.EggCostIMG.SetImage(EggSelected)
            CustomPlayer.AllEggCostIMG.SetImage(EggSelected)
            CustomPlayer.EggSelectedIMG.SetImage(EggSelected)
            CustomPlayer.EggCostText.SetText(StringToMessage("{EggCost}"))
            CustomPlayer.AllEggCostText.SetText(StringToMessage(" {NumberCondensor(EggInfo(0) * 1.0)}"))
            CustomPlayer.StoneCostText.SetText(StringToMessage(" {NumberCondensor(StoneCost)}"))
            CustomPlayer.AllStoneCostText.SetText(StringToMessage(" {NumberCondensor(StoneCost * EggInfo(1))}"))
            #CHANGE TEXT COLOR IF YOU CAN OR CAN'T MAKE PURCHASE
            if(CustomPlayer.StoneAmount >= StoneCost):
                CustomPlayer.StoneCostText.SetTextColor(MakeColorFromHex("39FF39"))
            else:    
                CustomPlayer.StoneCostText.SetTextColor(MakeColorFromHex("FA4D56"))
            if(CustomPlayer.EggAmounts[CustomPlayer.PetCurrentPage] >= EggCost*1.0):
                CustomPlayer.EggCostText.SetTextColor(MakeColorFromHex("39FF39"))
            else:    
                CustomPlayer.EggCostText.SetTextColor(MakeColorFromHex("FA4D56"))

            if(CustomPlayer.StoneAmount >= StoneCost * EggInfo(1)):
                CustomPlayer.AllStoneCostText.SetTextColor(MakeColorFromHex("39FF39"))
            else:    
                CustomPlayer.AllStoneCostText.SetTextColor(MakeColorFromHex("FA4D56"))
            if(CustomPlayer.EggAmounts[CustomPlayer.PetCurrentPage] >= EggInfo(0) * 1.0):
                CustomPlayer.AllEggCostText.SetTextColor(MakeColorFromHex("39FF39"))
            else:    
                CustomPlayer.AllEggCostText.SetTextColor(MakeColorFromHex("FA4D56"))

        for(i:int=0..CustomPlayer.EggAmountsText.Length):
            if(EggAmount := CustomPlayer.EggAmounts[i], EggAmountBlock := CustomPlayer.EggAmountsText[i]):
                EggAmountBlock.SetText(StringToMessage("{NumberCondensor(EggAmount)}"))
            
    MakePetShopCanvas(CustomPlayer:Custom_Player):canvas=
        CustomPlayer.PetShopCloseButton.SetText(StringToMessage("X"))
        CustomPlayer.Pet1MenuButton.SetText(StringToMessage("Select"))
        CustomPlayer.Pet2MenuButton.SetText(StringToMessage("Select"))
        CustomPlayer.Pet3MenuButton.SetText(StringToMessage("Select"))
        CustomPlayer.Pet4MenuButton.SetText(StringToMessage("Select"))
        CustomPlayer.Pet5MenuButton.SetText(StringToMessage("Select"))
        CustomPlayer.Pet6MenuButton.SetText(StringToMessage("Select"))

        CustomPlayer.Pet1Tier.SetText(StringToMessage("Common"))
        CustomPlayer.Pet2Tier.SetText(StringToMessage("Uncommon"))
        CustomPlayer.Pet3Tier.SetText(StringToMessage("Rare"))
        CustomPlayer.Pet4Tier.SetText(StringToMessage("Epic"))
        CustomPlayer.Pet5Tier.SetText(StringToMessage("Legendary"))
        CustomPlayer.Pet6Tier.SetText(StringToMessage("Mythic"))

        CustomPlayer.PetGatherText.SetText(StringToMessage("Gather:"))
        CustomPlayer.StrengthText.SetText(StringToMessage("Attack %:"))
        CustomPlayer.CritRateText.SetText(StringToMessage("Crit Rate:"))
        CustomPlayer.CritDMGText.SetText(StringToMessage("Crit Dmg:"))
        CustomPlayer.FinalStrikeText.SetText(StringToMessage("Final Strike:"))

        CustomPlayer.TiersTitle.SetText(StringToMessage("Tiers:"))
        CustomPlayer.CostTitle.SetText(StringToMessage("Cost:"))

        CustomPlayer.ForwardArrow.SetEnabled(true)
        CustomPlayer.BackArrow.SetEnabled(true)

        CustomPlayer.HatchButton.SetText(StringToMessage("Hatch"))
        CustomPlayer.HatchAllButton.SetText(StringToMessage("Hatch All"))

        for(TextBlock : CustomPlayer.PSAllTexts):
            TextBlock.SetShadowColor(CustomPlayer.ShadowDefaults.DefaultShadowColor)
            TextBlock.SetShadowOffset(CustomPlayer.ShadowDefaults.DefaultShadowOffset)
            TextBlock.SetShadowOpacity(1.0)

        MyCanvas: canvas = canvas {
            Slots := array:
                canvas_slot:
                    Widget:=CustomPlayer.petShopBackGround
                    ZOrder:=0
                    SizeToContent:=false
                    Alignment := vector2{X := 0.5, Y := 0.0}
                    Anchors := anchors{Minimum := vector2{X:= 0.5, Y:= 0.0}, Maximum := vector2{X:= 0.5, Y:= 0.0}}
                    Offsets:=margin{Top:=0.0,Left:=0.0,Bottom:=CustomPlayer.petShopBackGround.DefaultDesiredSize.Y,Right:=CustomPlayer.petShopBackGround.DefaultDesiredSize.X}
                canvas_slot:
                    Widget:=CustomPlayer.PetShopCloseButton
                    ZOrder:=9
                    SizeToContent:=false
                    Anchors := anchors{Minimum := vector2{X:= 0.5, Y:= 0.0}, Maximum := vector2{X:= 0.5, Y:= 0.0}}
                    Offsets:=margin{Top:=122.0,Left:=504.0,Bottom:=CustomPlayer.PetShopClose_Size.Y,Right:=CustomPlayer.PetShopClose_Size.X}
                canvas_slot:
                    Widget:=CustomPlayer.PetShopTitle
                    ZOrder:=9
                    SizeToContent:=false
                    Anchors := anchors{Minimum := vector2{X:= 0.5, Y:= 0.0}, Maximum := vector2{X:= 0.5, Y:= 0.0}}
                    Offsets:=margin{Top:=112.0,Left:=108.0,Bottom:=CustomPlayer.PetShopTitle.DefaultDesiredSize.Y,Right:=CustomPlayer.PetShopTitle.DefaultDesiredSize.X}
               
                canvas_slot:
                    Widget:=CustomPlayer.InvTitle
                    ZOrder:=7
                    SizeToContent:=true
                    Anchors := anchors{Minimum := vector2{X:= 0.5, Y:= 0.0}, Maximum := vector2{X:= 0.5, Y:= 0.0}}
                    Offsets:=margin{Top:=156.0,Left:=-676.0}
                canvas_slot:
                    Widget:=CustomPlayer.InvStoneIcon
                    ZOrder:=7
                    SizeToContent:=true
                    Anchors := anchors{Minimum := vector2{X:= 0.5, Y:= 0.0}, Maximum := vector2{X:= 0.5, Y:= 0.0}}
                    Offsets:=margin{Top:=232.0,Left:=-756.0}
                canvas_slot:
                    Widget:=CustomPlayer.InvEgg1Icon
                    ZOrder:=7
                    SizeToContent:=true
                    Anchors := anchors{Minimum := vector2{X:= 0.5, Y:= 0.0}, Maximum := vector2{X:= 0.5, Y:= 0.0}}
                    Offsets:=margin{Top:=316.0,Left:=-756.0}
                canvas_slot:
                    Widget:=CustomPlayer.InvEgg2Icon
                    ZOrder:=7
                    SizeToContent:=true
                    Anchors := anchors{Minimum := vector2{X:= 0.5, Y:= 0.0}, Maximum := vector2{X:= 0.5, Y:= 0.0}}
                    Offsets:=margin{Top:=400.0,Left:=-756.0}
                canvas_slot:
                    Widget:=CustomPlayer.InvEgg3Icon
                    ZOrder:=7
                    SizeToContent:=true
                    Anchors := anchors{Minimum := vector2{X:= 0.5, Y:= 0.0}, Maximum := vector2{X:= 0.5, Y:= 0.0}}
                    Offsets:=margin{Top:=484.0,Left:=-756.0}
                canvas_slot:
                    Widget:=CustomPlayer.InvEgg4Icon
                    ZOrder:=7
                    SizeToContent:=true
                    Anchors := anchors{Minimum := vector2{X:= 0.5, Y:= 0.0}, Maximum := vector2{X:= 0.5, Y:= 0.0}}
                    Offsets:=margin{Top:=568.0,Left:=-756.0}
                canvas_slot:
                    Widget:=CustomPlayer.InvEgg5Icon
                    ZOrder:=7
                    SizeToContent:=true
                    Anchors := anchors{Minimum := vector2{X:= 0.5, Y:= 0.0}, Maximum := vector2{X:= 0.5, Y:= 0.0}}
                    Offsets:=margin{Top:=652.0,Left:=-756.0}
                canvas_slot:
                    Widget:=CustomPlayer.InvEgg6Icon
                    ZOrder:=7
                    SizeToContent:=true
                    Anchors := anchors{Minimum := vector2{X:= 0.5, Y:= 0.0}, Maximum := vector2{X:= 0.5, Y:= 0.0}}
                    Offsets:=margin{Top:=732.0,Left:=-756.0}


                canvas_slot:
                    Widget:=CustomPlayer.InvStone
                    ZOrder:=7
                    SizeToContent:=true
                    Anchors := anchors{Minimum := vector2{X:= 0.5, Y:= 0.0}, Maximum := vector2{X:= 0.5, Y:= 0.0}}
                    Offsets:=margin{Top:=232.0,Left:=-700.0}
                canvas_slot:
                    Widget:=CustomPlayer.InvEgg1
                    ZOrder:=7
                    SizeToContent:=true
                    Anchors := anchors{Minimum := vector2{X:= 0.5, Y:= 0.0}, Maximum := vector2{X:= 0.5, Y:= 0.0}}
                    Offsets:=margin{Top:=318.0,Left:=-700.0}
                canvas_slot:
                    Widget:=CustomPlayer.InvEgg2
                    ZOrder:=7
                    SizeToContent:=true
                    Anchors := anchors{Minimum := vector2{X:= 0.5, Y:= 0.0}, Maximum := vector2{X:= 0.5, Y:= 0.0}}
                    Offsets:=margin{Top:=400.0,Left:=-700.0}
                canvas_slot:
                    Widget:=CustomPlayer.InvEgg3
                    ZOrder:=7
                    SizeToContent:=true
                    Anchors := anchors{Minimum := vector2{X:= 0.5, Y:= 0.0}, Maximum := vector2{X:= 0.5, Y:= 0.0}}
                    Offsets:=margin{Top:=484.0,Left:=-700.0}
                canvas_slot:
                    Widget:=CustomPlayer.InvEgg4
                    ZOrder:=7
                    SizeToContent:=true
                    Anchors := anchors{Minimum := vector2{X:= 0.5, Y:= 0.0}, Maximum := vector2{X:= 0.5, Y:= 0.0}}
                    Offsets:=margin{Top:=568.0,Left:=-700.0}
                canvas_slot:
                    Widget:=CustomPlayer.InvEgg5
                    ZOrder:=7
                    SizeToContent:=true
                    Anchors := anchors{Minimum := vector2{X:= 0.5, Y:= 0.0}, Maximum := vector2{X:= 0.5, Y:= 0.0}}
                    Offsets:=margin{Top:=652.0,Left:=-700.0}
                canvas_slot:
                    Widget:=CustomPlayer.InvEgg6
                    ZOrder:=7
                    SizeToContent:=true
                    Anchors := anchors{Minimum := vector2{X:= 0.5, Y:= 0.0}, Maximum := vector2{X:= 0.5, Y:= 0.0}}
                    Offsets:=margin{Top:=736.0,Left:=-700.0}
                    
                <#canvas_slot:
                    Widget:=CustomPlayer.PetsTitle
                    ZOrder:=7
                    SizeToContent:=true
                    Offsets:=margin{Top:=208.0,Left:=736.0}#>
                canvas_slot:
                    Widget:=CustomPlayer.Pet1Icon
                    ZOrder:=7
                    SizeToContent:=true
                    Anchors := anchors{Minimum := vector2{X:= 0.5, Y:= 0.0}, Maximum := vector2{X:= 0.5, Y:= 0.0}}
                    Offsets:=margin{Top:=232.0,Left:=-452.0}
                canvas_slot:
                    Widget:=CustomPlayer.Pet1Level
                    ZOrder:=7
                    SizeToContent:=true
                    Anchors := anchors{Minimum := vector2{X:= 0.5, Y:= 0.0}, Maximum := vector2{X:= 0.5, Y:= 0.0}}
                    Offsets:=margin{Top:=256.0,Left:=-320.0}
                canvas_slot:
                    Widget:=CustomPlayer.Pet1Tier
                    ZOrder:=7
                    SizeToContent:=true
                    Anchors := anchors{Minimum := vector2{X:= 0.5, Y:= 0.0}, Maximum := vector2{X:= 0.5, Y:= 0.0}}
                    Offsets:=margin{Top:=296.0,Left:=-320.0}
                canvas_slot:
                    Widget:=CustomPlayer.Pet1MenuButton
                    ZOrder:=7
                    SizeToContent:=false
                    Anchors := anchors{Minimum := vector2{X:= 0.5, Y:= 0.0}, Maximum := vector2{X:= 0.5, Y:= 0.0}}
                    Offsets:=margin{Top:=264.0,Left:=-128.0, Bottom:=56.0, Right:=200.0}
                canvas_slot:
                    Widget:=CustomPlayer.Pet2Icon
                    ZOrder:=7
                    SizeToContent:=true
                    Anchors := anchors{Minimum := vector2{X:= 0.5, Y:= 0.0}, Maximum := vector2{X:= 0.5, Y:= 0.0}}
                    Offsets:=margin{Top:=344.0,Left:=-452.0}
                canvas_slot:
                    Widget:=CustomPlayer.Pet2Level
                    ZOrder:=7
                    SizeToContent:=true
                    Anchors := anchors{Minimum := vector2{X:= 0.5, Y:= 0.0}, Maximum := vector2{X:= 0.5, Y:= 0.0}}
                    Offsets:=margin{Top:=364.0,Left:=-320.0}
                canvas_slot:
                    Widget:=CustomPlayer.Pet2Tier
                    ZOrder:=7
                    SizeToContent:=true
                    Anchors := anchors{Minimum := vector2{X:= 0.5, Y:= 0.0}, Maximum := vector2{X:= 0.5, Y:= 0.0}}
                    Offsets:=margin{Top:=404.0,Left:=-320.0}
                canvas_slot:
                    Widget:=CustomPlayer.Pet2MenuButton
                    ZOrder:=7
                    SizeToContent:=false
                    Anchors := anchors{Minimum := vector2{X:= 0.5, Y:= 0.0}, Maximum := vector2{X:= 0.5, Y:= 0.0}}
                    Offsets:=margin{Top:=372.0,Left:=-128.0, Bottom:=56.0, Right:=200.0}
                canvas_slot:
                    Widget:=CustomPlayer.Pet3Icon
                    ZOrder:=7
                    SizeToContent:=true
                    Anchors := anchors{Minimum := vector2{X:= 0.5, Y:= 0.0}, Maximum := vector2{X:= 0.5, Y:= 0.0}}
                    Offsets:=margin{Top:=456.0,Left:=-452.0}
                canvas_slot:
                    Widget:=CustomPlayer.Pet3Level
                    ZOrder:=7
                    SizeToContent:=true
                    Anchors := anchors{Minimum := vector2{X:= 0.5, Y:= 0.0}, Maximum := vector2{X:= 0.5, Y:= 0.0}}
                    Offsets:=margin{Top:=468.0,Left:=-320.0}
                canvas_slot:
                    Widget:=CustomPlayer.Pet3Tier
                    ZOrder:=7
                    SizeToContent:=true
                    Anchors := anchors{Minimum := vector2{X:= 0.5, Y:= 0.0}, Maximum := vector2{X:= 0.5, Y:= 0.0}}
                    Offsets:=margin{Top:=508.0,Left:=-320.0}
                canvas_slot:
                    Widget:=CustomPlayer.Pet3MenuButton
                    ZOrder:=7
                    SizeToContent:=false
                    Anchors := anchors{Minimum := vector2{X:= 0.5, Y:= 0.0}, Maximum := vector2{X:= 0.5, Y:= 0.0}}
                    Offsets:=margin{Top:=476.0,Left:=-128.0, Bottom:=56.0, Right:=200.0}
                canvas_slot:
                    Widget:=CustomPlayer.Pet4Icon
                    ZOrder:=7
                    SizeToContent:=true
                    Anchors := anchors{Minimum := vector2{X:= 0.5, Y:= 0.0}, Maximum := vector2{X:= 0.5, Y:= 0.0}}
                    Offsets:=margin{Top:=568.0,Left:=-452.0}
                canvas_slot:
                    Widget:=CustomPlayer.Pet4Level
                    ZOrder:=7
                    SizeToContent:=true
                    Anchors := anchors{Minimum := vector2{X:= 0.5, Y:= 0.0}, Maximum := vector2{X:= 0.5, Y:= 0.0}}
                    Offsets:=margin{Top:=576.0,Left:=-320.0}
                canvas_slot:
                    Widget:=CustomPlayer.Pet4Tier
                    ZOrder:=7
                    SizeToContent:=true
                    Anchors := anchors{Minimum := vector2{X:= 0.5, Y:= 0.0}, Maximum := vector2{X:= 0.5, Y:= 0.0}}
                    Offsets:=margin{Top:=616.0,Left:=-320.0}
                canvas_slot:
                    Widget:=CustomPlayer.Pet4MenuButton
                    ZOrder:=7
                    SizeToContent:=false
                    Anchors := anchors{Minimum := vector2{X:= 0.5, Y:= 0.0}, Maximum := vector2{X:= 0.5, Y:= 0.0}}
                    Offsets:=margin{Top:=584.0,Left:=-128.0, Bottom:=56.0, Right:=200.0}
                canvas_slot:
                    Widget:=CustomPlayer.Pet5Icon
                    ZOrder:=7
                    SizeToContent:=true
                    Anchors := anchors{Minimum := vector2{X:= 0.5, Y:= 0.0}, Maximum := vector2{X:= 0.5, Y:= 0.0}}
                    Offsets:=margin{Top:=680.0,Left:=-452.0}
                canvas_slot:
                    Widget:=CustomPlayer.Pet5Level
                    ZOrder:=7
                    SizeToContent:=true
                    Anchors := anchors{Minimum := vector2{X:= 0.5, Y:= 0.0}, Maximum := vector2{X:= 0.5, Y:= 0.0}}
                    Offsets:=margin{Top:=692.0,Left:=-320.0}
                canvas_slot:
                    Widget:=CustomPlayer.Pet5Tier
                    ZOrder:=7
                    SizeToContent:=true
                    Anchors := anchors{Minimum := vector2{X:= 0.5, Y:= 0.0}, Maximum := vector2{X:= 0.5, Y:= 0.0}}
                    Offsets:=margin{Top:=732.0,Left:=-320.0}
                canvas_slot:
                    Widget:=CustomPlayer.Pet5MenuButton
                    ZOrder:=7
                    SizeToContent:=false
                    Anchors := anchors{Minimum := vector2{X:= 0.5, Y:= 0.0}, Maximum := vector2{X:= 0.5, Y:= 0.0}}
                    Offsets:=margin{Top:=700.0,Left:=-128.0, Bottom:=56.0, Right:=200.0}
                canvas_slot:
                    Widget:=CustomPlayer.Pet6Icon
                    ZOrder:=7
                    SizeToContent:=true
                    Anchors := anchors{Minimum := vector2{X:= 0.5, Y:= 0.0}, Maximum := vector2{X:= 0.5, Y:= 0.0}}
                    Offsets:=margin{Top:=792.0,Left:=-452.0}
                canvas_slot:
                    Widget:=CustomPlayer.Pet6Level
                    ZOrder:=7
                    SizeToContent:=true
                    Anchors := anchors{Minimum := vector2{X:= 0.5, Y:= 0.0}, Maximum := vector2{X:= 0.5, Y:= 0.0}}
                    Offsets:=margin{Top:=808.0,Left:=-320.0}
                canvas_slot:
                    Widget:=CustomPlayer.Pet6Tier
                    ZOrder:=7
                    SizeToContent:=true
                    Anchors := anchors{Minimum := vector2{X:= 0.5, Y:= 0.0}, Maximum := vector2{X:= 0.5, Y:= 0.0}}
                    Offsets:=margin{Top:=848.0,Left:=-320.0}
                canvas_slot:
                    Widget:=CustomPlayer.Pet6MenuButton
                    ZOrder:=7
                    SizeToContent:=false
                    Anchors := anchors{Minimum := vector2{X:= 0.5, Y:= 0.0}, Maximum := vector2{X:= 0.5, Y:= 0.0}}
                    Offsets:=margin{Top:=816.0,Left:=-128.0, Bottom:=56.0, Right:=200.0}

                canvas_slot:
                    Widget:=CustomPlayer.PetDisplayIMG
                    ZOrder:=8
                    SizeToContent:=false
                    Anchors := anchors{Minimum := vector2{X:= 0.5, Y:= 0.0}, Maximum := vector2{X:= 0.5, Y:= 0.0}}
                    Offsets:=margin{Top:=276.0,Left:=144.0, Bottom:=200.0, Right:=200.0}
                    
                canvas_slot:
                    Widget:=CustomPlayer.PetGatherIcon
                    ZOrder:=7
                    SizeToContent:=true
                    Anchors := anchors{Minimum := vector2{X:= 0.5, Y:= 0.0}, Maximum := vector2{X:= 0.5, Y:= 0.0}}
                    Offsets:=margin{Top:=224.0,Left:=388.0, Bottom:=55.0, Right:=55.0}
                canvas_slot:
                    Widget:=CustomPlayer.PetGatherText
                    ZOrder:=7
                    SizeToContent:=true
                    Anchors := anchors{Minimum := vector2{X:= 0.5, Y:= 0.0}, Maximum := vector2{X:= 0.5, Y:= 0.0}}
                    Offsets:=margin{Top:=232.0,Left:=444.0}
                canvas_slot:
                    Widget:=CustomPlayer.PetGatherAmount
                    ZOrder:=7
                    SizeToContent:=true
                    Anchors := anchors{Minimum := vector2{X:= 0.5, Y:= 0.0}, Maximum := vector2{X:= 0.5, Y:= 0.0}}
                    Offsets:=margin{Top:=232.0,Left:=764.0}
                    Alignment := vector2{X := 1.0, Y := 0.0}
                canvas_slot:
                    Widget:=CustomPlayer.StrengthIcon
                    ZOrder:=7
                    SizeToContent:=true
                    Anchors := anchors{Minimum := vector2{X:= 0.5, Y:= 0.0}, Maximum := vector2{X:= 0.5, Y:= 0.0}}
                    Offsets:=margin{Top:=284.0,Left:=388.0, Bottom:=55.0, Right:=55.0}
                canvas_slot:
                    Widget:=CustomPlayer.StrengthText
                    ZOrder:=7
                    SizeToContent:=true
                    Anchors := anchors{Minimum := vector2{X:= 0.5, Y:= 0.0}, Maximum := vector2{X:= 0.5, Y:= 0.0}}
                    Offsets:=margin{Top:=292.0,Left:=444.0}
                canvas_slot:
                    Widget:=CustomPlayer.AttackAmount
                    ZOrder:=7
                    SizeToContent:=true
                    Anchors := anchors{Minimum := vector2{X:= 0.5, Y:= 0.0}, Maximum := vector2{X:= 0.5, Y:= 0.0}}
                    Offsets:=margin{Top:=292.0,Left:=764.0}
                    Alignment := vector2{X := 1.0, Y := 0.0}
                canvas_slot:
                    Widget:=CustomPlayer.CritRateIcon
                    ZOrder:=7
                    SizeToContent:=true
                    Anchors := anchors{Minimum := vector2{X:= 0.5, Y:= 0.0}, Maximum := vector2{X:= 0.5, Y:= 0.0}}
                    Offsets:=margin{Top:=344.0,Left:=388.0, Bottom:=50.0, Right:=50.0}
                canvas_slot:
                    Widget:=CustomPlayer.CritRateText
                    ZOrder:=7
                    SizeToContent:=true
                    Anchors := anchors{Minimum := vector2{X:= 0.5, Y:= 0.0}, Maximum := vector2{X:= 0.5, Y:= 0.0}}
                    Offsets:=margin{Top:=352.0,Left:=444.0}
                canvas_slot:
                    Widget:=CustomPlayer.CritRateAmount
                    ZOrder:=7
                    SizeToContent:=true
                    Anchors := anchors{Minimum := vector2{X:= 0.5, Y:= 0.0}, Maximum := vector2{X:= 0.5, Y:= 0.0}}
                    Offsets:=margin{Top:=352.0,Left:=764.0}
                    Alignment := vector2{X := 1.0, Y := 0.0}
                canvas_slot:
                    Widget:=CustomPlayer.CritDMGIcon
                    ZOrder:=7
                    SizeToContent:=true
                    Anchors := anchors{Minimum := vector2{X:= 0.5, Y:= 0.0}, Maximum := vector2{X:= 0.5, Y:= 0.0}}
                    Offsets:=margin{Top:=400.0,Left:=388.0, Bottom:=55.0, Right:=55.0}
                canvas_slot:
                    Widget:=CustomPlayer.CritDMGText
                    ZOrder:=7
                    SizeToContent:=true
                    Anchors := anchors{Minimum := vector2{X:= 0.5, Y:= 0.0}, Maximum := vector2{X:= 0.5, Y:= 0.0}}
                    Offsets:=margin{Top:=412.0,Left:=444.0}
                canvas_slot:
                    Widget:=CustomPlayer.CritDMGAmount
                    ZOrder:=7
                    SizeToContent:=true
                    Anchors := anchors{Minimum := vector2{X:= 0.5, Y:= 0.0}, Maximum := vector2{X:= 0.5, Y:= 0.0}}
                    Offsets:=margin{Top:=412.0,Left:=764.0}
                    Alignment := vector2{X := 1.0, Y := 0.0}
                canvas_slot:
                    Widget:=CustomPlayer.FinalStrikeIcon
                    ZOrder:=7
                    SizeToContent:=true
                    Anchors := anchors{Minimum := vector2{X:= 0.5, Y:= 0.0}, Maximum := vector2{X:= 0.5, Y:= 0.0}}
                    Offsets:=margin{Top:=464.0,Left:=388.0, Bottom:=55.0, Right:=55.0}
                canvas_slot:
                    Widget:=CustomPlayer.FinalStrikeText
                    ZOrder:=7
                    SizeToContent:=true
                    Anchors := anchors{Minimum := vector2{X:= 0.5, Y:= 0.0}, Maximum := vector2{X:= 0.5, Y:= 0.0}}
                    Offsets:=margin{Top:=472.0,Left:=444.0}
                canvas_slot:
                    Widget:=CustomPlayer.FinalStrikeAmount
                    ZOrder:=7
                    SizeToContent:=true
                    Anchors := anchors{Minimum := vector2{X:= 0.5, Y:= 0.0}, Maximum := vector2{X:= 0.5, Y:= 0.0}}
                    Offsets:=margin{Top:=472.0,Left:=764.0}
                    Alignment := vector2{X := 1.0, Y := 0.0}

                canvas_slot:
                    Widget:=CustomPlayer.TiersTitle
                    ZOrder:=7
                    SizeToContent:=true
                    Anchors := anchors{Minimum := vector2{X:= 0.5, Y:= 0.0}, Maximum := vector2{X:= 0.5, Y:= 0.0}}
                    Offsets:=margin{Top:=580.0,Left:=128.0}
                canvas_slot:
                    Widget:=CustomPlayer.TiersText
                    ZOrder:=7
                    SizeToContent:=true
                    Anchors := anchors{Minimum := vector2{X:= 0.5, Y:= 0.0}, Maximum := vector2{X:= 0.5, Y:= 0.0}}
                    Offsets:=margin{Top:=580.0,Left:=348.0}

                canvas_slot:
                    Widget:=CustomPlayer.CostTitle
                    ZOrder:=7
                    SizeToContent:=true
                    Anchors := anchors{Minimum := vector2{X:= 0.5, Y:= 0.0}, Maximum := vector2{X:= 0.5, Y:= 0.0}}
                    Offsets:=margin{Top:=640.0,Left:=128.0}
                canvas_slot:
                    ZOrder:=1
                    SizeToContent:=true
                    Offsets:=margin{Top:=640.0,Left:=336.0}
                    Anchors := anchors{Minimum := vector2{X:= 0.5, Y:= 0.0}, Maximum := vector2{X:= 0.5, Y:= 0.0}}
                    Alignment := vector2{X:= 0.5, Y:= 0.0}
                    Widget := stack_box:
                        Orientation := orientation.Horizontal
                        Slots := array:
                            stack_box_slot:
                                Widget := CustomPlayer.StoneCostIMG  
                            stack_box_slot:
                                Widget := CustomPlayer.StoneCostText
                canvas_slot:
                    ZOrder:=1
                    SizeToContent:=true
                    Offsets:=margin{Top:=692.0,Left:=336.0}
                    Anchors := anchors{Minimum := vector2{X:= 0.5, Y:= 0.0}, Maximum := vector2{X:= 0.5, Y:= 0.0}}
                    Alignment := vector2{X:= 0.5, Y:= 0.0}
                    Widget := stack_box:
                        Orientation := orientation.Horizontal
                        Slots := array:
                            stack_box_slot:
                                Widget := CustomPlayer.EggCostIMG  
                            stack_box_slot:
                                Widget := CustomPlayer.EggCostText
                canvas_slot:
                    ZOrder:=1
                    SizeToContent:=true
                    Offsets:=margin{Top:=640.0,Left:=636.0}
                    Anchors := anchors{Minimum := vector2{X:= 0.5, Y:= 0.0}, Maximum := vector2{X:= 0.5, Y:= 0.0}}
                    Alignment := vector2{X:= 0.5, Y:= 0.0}
                    Widget := stack_box:
                        Orientation := orientation.Horizontal
                        Slots := array:
                            stack_box_slot:
                                Widget := CustomPlayer.AllStoneCostIMG  
                            stack_box_slot:
                                Widget := CustomPlayer.AllStoneCostText
                <#canvas_slot:
                    ZOrder:=1
                    SizeToContent:=true
                    Offsets:=margin{Top:=692.0,Left:=1600.0}
                    Alignment := vector2{X:= 0.5, Y:= 0.0}
                    Widget := stack_box:
                        Orientation := orientation.Horizontal
                        Slots := array:
                            stack_box_slot:
                                Widget := CustomPlayer.AllEggCostIMG  
                            stack_box_slot:
                                Widget := CustomPlayer.AllEggCostText#>
                canvas_slot:
                    Widget:=CustomPlayer.EggSelectedIMG
                    ZOrder:=7
                    SizeToContent:=true
                    Anchors := anchors{Minimum := vector2{X:= 0.5, Y:= 0.0}, Maximum := vector2{X:= 0.5, Y:= 0.0}}
                    Offsets:=margin{Top:=752.0,Left:=424.0, Bottom:=50.0, Right:=50.0}
                canvas_slot:
                    Widget:=CustomPlayer.BackButton
                    ZOrder:=7
                    SizeToContent:=false
                    Anchors := anchors{Minimum := vector2{X:= 0.5, Y:= 0.0}, Maximum := vector2{X:= 0.5, Y:= 0.0}}
                    Offsets:=margin{Top:=756.0,Left:=124.0, Bottom:=40.0, Right:=275.0}
                canvas_slot:
                    Widget:=CustomPlayer.BackArrow
                    ZOrder:=4
                    SizeToContent:=false
                    Anchors := anchors{Minimum := vector2{X:= 0.5, Y:= 0.0}, Maximum := vector2{X:= 0.5, Y:= 0.0}}
                    Offsets:=margin{Top:=752.0,Left:=240.0, Bottom:=50.0, Right:=50.0}
                canvas_slot:
                    Widget:=CustomPlayer.ForwardButton
                    ZOrder:=7
                    SizeToContent:=false
                    Anchors := anchors{Minimum := vector2{X:= 0.5, Y:= 0.0}, Maximum := vector2{X:= 0.5, Y:= 0.0}}
                    Offsets:=margin{Top:=756.0,Left:=504.0, Bottom:=40.0, Right:=275.0}
                canvas_slot:
                    Widget:=CustomPlayer.ForwardArrow
                    ZOrder:=4
                    SizeToContent:=false
                    Anchors := anchors{Minimum := vector2{X:= 0.5, Y:= 0.0}, Maximum := vector2{X:= 0.5, Y:= 0.0}}
                    Offsets:=margin{Top:=752.0,Left:=624.0, Bottom:=50.0, Right:=50.0}
                canvas_slot:
                    Widget:=CustomPlayer.HatchButton
                    ZOrder:=7
                    SizeToContent:=false
                    Anchors := anchors{Minimum := vector2{X:= 0.5, Y:= 0.0}, Maximum := vector2{X:= 0.5, Y:= 0.0}}
                    Offsets:=margin{Top:=832.0,Left:=120.0, Bottom:=55.0, Right:=340.0}
                canvas_slot:
                    Widget:=CustomPlayer.HatchAllButton
                    ZOrder:=7
                    SizeToContent:=false
                    Alignment := vector2{X:= 0.5, Y:= 0.0}
                    Anchors := anchors{Minimum := vector2{X:= 0.5, Y:= 0.0}, Maximum := vector2{X:= 0.5, Y:= 0.0}}
                    Offsets:=margin{Top:=832.0,Left:=636.0, Bottom:=55.0, Right:=275.0}
                canvas_slot:
                    Widget:=CustomPlayer.PetHatchBG
                    ZOrder:=6
                    SizeToContent:=false
                    Anchors := anchors{Minimum := vector2{X:= 0.5, Y:= 0.0}, Maximum := vector2{X:= 0.5, Y:= 0.0}}
                    Offsets:=margin{Top:=832.0,Left:=120.0, Bottom:=55.0, Right:=340.0}
                canvas_slot:
                    Widget:=CustomPlayer.PetHatchAllBG
                    ZOrder:=6
                    SizeToContent:=false
                    Alignment := vector2{X:= 0.5, Y:= 0.0}
                    Anchors := anchors{Minimum := vector2{X:= 0.5, Y:= 0.0}, Maximum := vector2{X:= 0.5, Y:= 0.0}}
                    Offsets:=margin{Top:=832.0,Left:=636.0, Bottom:=55.0, Right:=275.0}
                canvas_slot:
                    Widget:=CustomPlayer.PetHatchMessage
                    ZOrder:=6
                    SizeToContent:=true
                    Alignment := vector2{X:= 0.5, Y:= 0.0}
                    Anchors := anchors{Minimum := vector2{X:= 0.5, Y:= 0.0}, Maximum := vector2{X:= 0.5, Y:= 0.0}}
                    Offsets:=margin{Top:=920.0, Left:=452.0}
        }
