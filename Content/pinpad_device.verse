using { /Fortnite.com/Devices }
using { /Verse.org/Simulation }
using { /UnrealEngine.com/Temporary/Diagnostics }
using { /UnrealEngine.com/Temporary/UI}
using { /UnrealEngine.com/Temporary/SpatialMath}
using { /Fortnite.com/UI }
using { /Verse.org/Colors }
using { /Verse.org/Colors/NamedColors }
 
pinpad_device := class(creative_device):
 
    var StuffPerAgent : [agent]?tuple(canvas,text_block,string) = map{}
 
    @editable
    AssignPinPadButton : button_device = button_device{}
    @editable
    ReturnToEarth : []button_device = array{}
    @editable
    TPToHiddenRoom : teleporter_device = teleporter_device{}
    @editable
    TPToQuarry : teleporter_device = teleporter_device{}
    @editable
    TPToExclusiveRoom : teleporter_device = teleporter_device{}

    @editable PIN : int = 0
    @editable ExclusivePIN : int = 0
    @editable SecretPIN : int = 0
    @editable
    gm : game_manager = game_manager{}
 
    # Runs when the device is started in a running game
    OnBegin<override>()<suspends>:void=
        AssignPinPadButton.InteractedWithEvent.Subscribe(AddPinPad)

        for(Button:ReturnToEarth):
            Button.InteractedWithEvent.Subscribe(ReturnToEarthHelper)
            
        for(Spawner : gm.PlayerSpawners):     
            Spawner.SpawnedEvent.Subscribe(OnPlayerJoined)

    OnPlayerJoined(Agent:agent):void=
        spawn{OnPlayerJoinedBuffer(Agent)}

    OnPlayerJoinedBuffer(Agent:agent)<suspends>:void=
        Sleep(1.0)
        if(CustomPlayer := gm.PlayersMap[player[Agent]]):
            if(CustomPlayer.PlayTimeSessionSeconds = 0.0):
                CustomPlayer.PPButton0.OnClick().Subscribe(Add0)
                CustomPlayer.PPButton1.OnClick().Subscribe(Add1)
                CustomPlayer.PPButton2.OnClick().Subscribe(Add2)
                CustomPlayer.PPButton3.OnClick().Subscribe(Add3)
                CustomPlayer.PPButton4.OnClick().Subscribe(Add4)
                CustomPlayer.PPButton5.OnClick().Subscribe(Add5)
                CustomPlayer.PPButton6.OnClick().Subscribe(Add6)
                CustomPlayer.PPButton7.OnClick().Subscribe(Add7)
                CustomPlayer.PPButton8.OnClick().Subscribe(Add8)
                CustomPlayer.PPButton9.OnClick().Subscribe(Add9)
                CustomPlayer.PPButtonEnter.OnClick().Subscribe(EnterFunc)
                CustomPlayer.PPButtonDelete.OnClick().Subscribe(DeleteLastChar)

    ReturnToEarthHelper(A:agent):void=
        TPToQuarry.Teleport(A)
 
    AddPinPad(A:agent):void=
        if(CustomPlayer := gm.PlayersMap[A]):
            if(Stuff:=StuffPerAgent[A]?):
                #Do not assign
            else:
                Val:=MakeUICanvas(CustomPlayer)
                if:
                    PlayerUI:=GetPlayerUI[player[A]]
                then:
                    PlayerUI.AddWidget(Val(0),player_ui_slot{InputMode:=ui_input_mode.All})
                    if(set StuffPerAgent[A] = option{(Val(0),Val(1),"")}){}
 
    RemovePinPad(A:agent):void=
        if(Stuff:=StuffPerAgent[A]?):
            if:
                PlayerUI:=GetPlayerUI[player[A]]
            then:
                PlayerUI.RemoveWidget(Stuff(0))
                if(set StuffPerAgent[A] = false){}
 
    MakeUICanvas(CustomPlayer : Custom_Player):tuple(canvas,text_block)=
        CustomPlayer.PPButton0.SetText(StringToMessage("0"))
        CustomPlayer.PPButton1.SetText(StringToMessage("1"))
        CustomPlayer.PPButton2.SetText(StringToMessage("2"))
        CustomPlayer.PPButton3.SetText(StringToMessage("3"))
        CustomPlayer.PPButton4.SetText(StringToMessage("4"))
        CustomPlayer.PPButton5.SetText(StringToMessage("5"))
        CustomPlayer.PPButton6.SetText(StringToMessage("6"))
        CustomPlayer.PPButton7.SetText(StringToMessage("7"))
        CustomPlayer.PPButton8.SetText(StringToMessage("8"))
        CustomPlayer.PPButton9.SetText(StringToMessage("9"))
        CustomPlayer.PPButtonDelete.SetText(StringToMessage("DELETE"))
        CustomPlayer.PPButtonEnter.SetText(StringToMessage("ENTER"))
        CustomPlayer.NumTB.SetText(StringToMessage(""))
 
        MyCanvas:= canvas:
            Slots:=array:
                canvas_slot:
                    Anchors:=anchors{Minimum:=vector2{X:=0.5,Y:=0.5},Maximum:=vector2{X:=0.5,Y:=0.5}}
                    Alignment:=vector2{X:=0.5,Y:=0.5}
                    SizeToContent:=true
                    Widget:=overlay:
                        Slots:=array:
                            overlay_slot:
                                #Outside ColorBlock
                                HorizontalAlignment:=horizontal_alignment.Fill
                                VerticalAlignment:=vertical_alignment.Fill
                                Widget:=color_block{DefaultColor:=MakeColorFromSRGB(0.05,0.05,0.05),DefaultDesiredSize:=vector2{X:=655.0,Y:=660.0}}
                            overlay_slot:  
                                HorizontalAlignment:=horizontal_alignment.Center
                                VerticalAlignment:=vertical_alignment.Bottom
                                Padding:=margin{Bottom:=20.0}
                                Widget:=overlay:
                                    Slots:=array:
                                        #PinPad ColorBlock
                                        overlay_slot:
                                            HorizontalAlignment:=horizontal_alignment.Fill
                                            VerticalAlignment:=vertical_alignment.Fill
                                            Widget:=color_block{DefaultColor:=MakeColorFromSRGB(0.0,0.01,0.50)}
                                        #PinPad
                                        overlay_slot:
                                            HorizontalAlignment:=horizontal_alignment.Center
                                            VerticalAlignment:=vertical_alignment.Center
                                            Padding:=margin{Top:=20.0,Left:=20.0,Bottom:=20.0,Right:=20.0}
                                            Widget:=stack_box:
                                                Orientation:=orientation.Vertical
                                                Slots:=array:
                                                    #1-3 Stack
                                                    stack_box_slot:
                                                        HorizontalAlignment:=horizontal_alignment.Fill
                                                        VerticalAlignment:=vertical_alignment.Fill
                                                        Padding:=margin{Top:=25.0,Bottom:=25.0}
                                                        Widget:=stack_box:
                                                            Orientation:=orientation.Horizontal
                                                            Slots:=array:
                                                                stack_box_slot:
                                                                    Distribution:=option{1.0}
                                                                    HorizontalAlignment:=horizontal_alignment.Center
                                                                    VerticalAlignment:=vertical_alignment.Center
                                                                    Widget:=CustomPlayer.PPButton1
                                                                stack_box_slot:
                                                                    Widget:=CustomPlayer.PPButton2
                                                                    Distribution:=option{1.0}
                                                                    HorizontalAlignment:=horizontal_alignment.Center
                                                                    VerticalAlignment:=vertical_alignment.Center
                                                                    Padding:=margin{Left:=20.0,Right:=20.0}
                                                                stack_box_slot:
                                                                    Widget:=CustomPlayer.PPButton3
                                                                    Distribution:=option{1.0}
                                                                    HorizontalAlignment:=horizontal_alignment.Center
                                                                    VerticalAlignment:=vertical_alignment.Center
                                                    #4-6
                                                    stack_box_slot:
                                                        HorizontalAlignment:=horizontal_alignment.Fill
                                                        VerticalAlignment:=vertical_alignment.Fill
                                                        Padding:=margin{Top:=25.0,Bottom:=25.0}
                                                        Widget:=stack_box:
                                                            Orientation:=orientation.Horizontal
                                                            Slots:=array:
                                                                stack_box_slot:
                                                                    Distribution:=option{1.0}
                                                                    HorizontalAlignment:=horizontal_alignment.Center
                                                                    VerticalAlignment:=vertical_alignment.Center
                                                                    Widget:=CustomPlayer.PPButton4
                                                                stack_box_slot:
                                                                    Distribution:=option{1.0}
                                                                    Widget:=CustomPlayer.PPButton5
                                                                    HorizontalAlignment:=horizontal_alignment.Center
                                                                    VerticalAlignment:=vertical_alignment.Center
                                                                    Padding:=margin{Left:=20.0,Right:=20.0}
                                                                stack_box_slot:
                                                                    Distribution:=option{1.0}
                                                                    Widget:=CustomPlayer.PPButton6
                                                                    HorizontalAlignment:=horizontal_alignment.Center
                                                                    VerticalAlignment:=vertical_alignment.Center
                                                    #7-9
                                                    stack_box_slot:
                                                        HorizontalAlignment:=horizontal_alignment.Fill
                                                        VerticalAlignment:=vertical_alignment.Fill
                                                        Padding:=margin{Top:=25.0,Bottom:=25.0}
                                                        Widget:=stack_box:
                                                            Orientation:=orientation.Horizontal
                                                            Slots:=array:
                                                                stack_box_slot:
                                                                    Distribution:=option{1.0}
                                                                    HorizontalAlignment:=horizontal_alignment.Center
                                                                    VerticalAlignment:=vertical_alignment.Center
                                                                    Widget:=CustomPlayer.PPButton7
                                                                stack_box_slot:
                                                                    Widget:=CustomPlayer.PPButton8
                                                                    Distribution:=option{1.0}
                                                                    HorizontalAlignment:=horizontal_alignment.Center
                                                                    VerticalAlignment:=vertical_alignment.Center
                                                                    Padding:=margin{Left:=20.0,Right:=20.0}
                                                                stack_box_slot:
                                                                    Widget:=CustomPlayer.PPButton9
                                                                    Distribution:=option{1.0}
                                                                    HorizontalAlignment:=horizontal_alignment.Center
                                                                    VerticalAlignment:=vertical_alignment.Center
                                                    #Delete, 0 and Enter
                                                    stack_box_slot:
                                                        HorizontalAlignment:=horizontal_alignment.Fill
                                                        VerticalAlignment:=vertical_alignment.Fill
                                                        Padding:=margin{Top:=25.0,Bottom:=25.0}
                                                        Widget:=stack_box:
                                                            Orientation:=orientation.Horizontal
                                                            Slots:=array:
                                                                stack_box_slot:
                                                                    Distribution:=option{1.0}
                                                                    HorizontalAlignment:=horizontal_alignment.Center
                                                                    VerticalAlignment:=vertical_alignment.Center
                                                                    Widget:=CustomPlayer.PPButtonDelete
                                                                stack_box_slot:
                                                                    Widget:=CustomPlayer.PPButton0
                                                                    Distribution:=option{1.0}
                                                                    HorizontalAlignment:=horizontal_alignment.Center
                                                                    VerticalAlignment:=vertical_alignment.Center
                                                                    Padding:=margin{Left:=20.0,Right:=20.0}
                                                                stack_box_slot:
                                                                    Widget:=CustomPlayer.PPButtonEnter
                                                                    Distribution:=option{1.0}
                                                                    HorizontalAlignment:=horizontal_alignment.Center
                                                                    VerticalAlignment:=vertical_alignment.Center
                            #Text Block Stuff
                            overlay_slot:
                                HorizontalAlignment:=horizontal_alignment.Center
                                VerticalAlignment:=vertical_alignment.Top
                                Padding:=margin{Top:=20.0}
                                Widget:=overlay:
                                    Slots:=array:
                                        overlay_slot:
                                            Widget:=color_block{DefaultColor:=Black,DefaultDesiredSize:=vector2{X:=400.0,Y:=32.0}}
                                            HorizontalAlignment:=horizontal_alignment.Fill
                                            VerticalAlignment:=vertical_alignment.Fill
                                        overlay_slot:
                                            Widget:=CustomPlayer.NumTB
                                            HorizontalAlignment:=horizontal_alignment.Center
                                            VerticalAlignment:=vertical_alignment.Center
 
 
 
 
        return (MyCanvas,CustomPlayer.NumTB)
 
    Add0(WM:widget_message):void=AddNum(0,WM)
    Add1(WM:widget_message):void=AddNum(1,WM)
    Add2(WM:widget_message):void=AddNum(2,WM)
    Add3(WM:widget_message):void=AddNum(3,WM)
    Add4(WM:widget_message):void=AddNum(4,WM)
    Add5(WM:widget_message):void=AddNum(5,WM)
    Add6(WM:widget_message):void=AddNum(6,WM)
    Add7(WM:widget_message):void=AddNum(7,WM)
    Add8(WM:widget_message):void=AddNum(8,WM)
    Add9(WM:widget_message):void=AddNum(9,WM)
 
    EnterFunc(WM:widget_message):void=
        if(CustomPlayer := gm.PlayersMap[WM.Player]):
            Print("Enter")
            PinAsString:string="{PIN}"
            ExclusivePinAsString:string="{ExclusivePIN}"
            SecretPinAsString:string="{SecretPIN}"
            if(Stuff:=StuffPerAgent[WM.Player]?):
                if(PinAsString=Stuff(2)):
                    Print("SUCCESS: {PIN} = {Stuff(2)}")
                    TPToHiddenRoom.Teleport(CustomPlayer.Player)
                    gm.ShowMessage(CustomPlayer.Player, "SUCCESS", 0)
                    
                else if(ExclusivePinAsString=Stuff(2)):
                    Print("SUCCESS: {PIN} = {Stuff(2)}")
                    TPToExclusiveRoom.Teleport(CustomPlayer.Player)
                    gm.ShowMessage(CustomPlayer.Player, "SUCCESS", 0)

                else if(SecretPinAsString=Stuff(2)):
                    Print("SECRET SUCCESS: {PIN} = {Stuff(2)}")
                    CustomPlayer.Collect(2e120, "Stone", 0)
                    CustomPlayer.Collect(1000000000.0, "Shard", 0)
                    CustomPlayer.Collect(9e11, "Egg", 0)
                    CustomPlayer.Collect(9e11, "Egg", 1)
                    CustomPlayer.Collect(9e11, "Egg", 2)
                    CustomPlayer.Collect(9e11, "Egg", 3)
                    CustomPlayer.Collect(9e11, "Egg", 4)
                    CustomPlayer.Collect(9e11, "Egg", 5)
                    CustomPlayer.Collect(1000000.0, "WepMat", 0)
                    CustomPlayer.Collect(1000000.0, "BossKey", 0)
                    gm.ShowMessage(CustomPlayer.Player, "SUCCESS", 0)
                else:
                    Print("FAIL: {PIN} != {Stuff(2)}")
                    gm.ShowMessage(CustomPlayer.Player, "FAIL", 1)
    
                RemovePinPad(WM.Player)
 
 
 
    AddNum(Num:int,WM:widget_message):void=
        if(Stuff:=StuffPerAgent[WM.Player]?):
            var String:string=Stuff(2)
            set String += "{Num}" #Add a new number
            if(set StuffPerAgent[WM.Player] = option{(Stuff(0),Stuff(1),String)}){}
            UpdateUI(WM)
 
    UpdateUI(WM:widget_message):void=
        if(Stuff:=StuffPerAgent[WM.Player]?):
            var String:string=Stuff(2)
            TB:=Stuff(1)
            TB.SetText(StringToMessage("{String}"))
 
    DeleteLastChar(WM:widget_message):void=
        if(Stuff:=StuffPerAgent[WM.Player]?):
            var String: string = Stuff(2)
            var UpdatedString : string = ""
            for(Index->S:String):
                if(Index = String.Length-1):
                    #This the last character, remove it
                else:
                    set UpdatedString += "{S}"
            if(set StuffPerAgent[WM.Player] = option{(Stuff(0),Stuff(1),UpdatedString)}){}
            UpdateUI(WM)