using { /Fortnite.com/Devices }
using { /Verse.org/Simulation }
using { /Verse.org/Colors }
using { /Verse.org/Verse }
using { /Verse.org/Random }
using { /Fortnite.com/Teams }
using { /UnrealEngine.com/Assets }
using { /Fortnite.com/Game }
using { /Fortnite.com/Characters }
using { /Fortnite.com/Playspaces }
using { /Fortnite.com/UI }
using { /UnrealEngine.com/Temporary/Diagnostics }
using { /UnrealEngine.com/Temporary/SpatialMath }
using { /UnrealEngine.com/Temporary/UI }
using { /Verse.org/Assets}


room_unlocks := class(creative_device):

    @editable
    gm : game_manager = game_manager{}

    @editable
    MobRooms : mob_room_manager = mob_room_manager{}

    @editable
    Tutorial : tutorial_manager = tutorial_manager{}
    
    @editable
    AudioManager : audio_manager = audio_manager{}

    @editable
    A4UnlockButton : button_device = button_device{}
    @editable
    A5UnlockButton : button_device = button_device{}

    @editable
    RoomUnlockTriggers : []trigger_device = array{}
    @editable
    RoomUnlockTexts : []billboard_device = array{}
    @editable
    RoomUnlockBarriers : []barrier_device = array{}
    @editable
    RoomUnlockGrindRails : []grind_rail_device = array{}

    @editable
    LobbyTeleporters : []teleporter_device = array{}
    @editable
    LobbyTeleporterTriggers : []trigger_device = array{}
    @editable
    RoomTeleporters : []teleporter_device = array{}

    #RoomNumbers : []int = array{1, 3, 6, 8, 11, 13, 16, 18}

    RoomUnlockCosts : []float = array{
       1000.0, 3.75e05, 3.75e07, 5.63e09, 5.63e11,
       4.22e13,2.34e16,4.68e18,1.40e21,2.81e23,
       4.21e25,9.36e28,2.81e31,1.26e34,3.79e36,
       3.41e49,4.16e58,1.32e67,1.99e75,3.24e84,
       1000.0,4.48e121,7.24e130,1.08e140,1.75e149
       }

    var WeaponTierNames : []string = array{"Common","Uncommon","Rare","Epic","Legendary","Mythic", "Supreme","Paragon", "Transcendant"}

    OnBegin<override>()<suspends>:void=

        A4UnlockButton.InteractedWithEvent.Subscribe(OpenA4HUD)
        A5UnlockButton.InteractedWithEvent.SubscribeAgent(PurchaseArea5, 20)

        for(i:=15..5):
            if(RoomTeleporter:=RoomTeleporters[i],LobbyTeleporter:=LobbyTeleporters[i]):
                LobbyTeleporter.EnterEvent.SubscribeAgent(EnterSpace, "Space")
                RoomTeleporter.EnterEvent.SubscribeAgent(SwitchMusic, "Default")
            
        for(i:int=0..RoomUnlockTriggers.Length):
            if(RoomUnlockTrigger:=RoomUnlockTriggers[i]):
                RoomUnlockTrigger.TriggeredEvent.SubscribeAgent2(PurchaseRoomUnlock, i)

        for(i:int=0..LobbyTeleporterTriggers.Length):
            if(LobbyTeleporterTrigger:=LobbyTeleporterTriggers[i]):
                LobbyTeleporterTrigger.TriggeredEvent.SubscribeAgent2(TeleportPlayerToCave, i)

        for(i:int=0..RoomTeleporters.Length):
            if(RoomTeleporter:=RoomTeleporters[i]):
                RoomTeleporter.EnterEvent.Subscribe(TeleportToQuarry)

        for(i:int=0..RoomUnlockGrindRails.Length):
            if(RoomUnlockGrindRail:=RoomUnlockGrindRails[i]):
                RoomUnlockGrindRail.StartedGrindingEvent.SubscribeAgent(CanGrindRails, i)

        for(Spawner : gm.PlayerSpawners):     
            Spawner.SpawnedEvent.Subscribe(OnPlayerJoined)
            
    OnPlayerJoined(Agent:agent):void=
        spawn{OnPlayerJoinedBuffer(Agent)}

    OnPlayerJoinedBuffer(Agent:agent)<suspends>:void=
        Sleep(1.0)
        if(CustomPlayer := gm.PlayersMap[player[Agent]]):
            UnlockRooms(CustomPlayer)
            if(CustomPlayer.PlayTimeSessionSeconds = 0.0):
                set CustomPlayer.AllA4Texts = array{CustomPlayer.WhatsNewTitle,CustomPlayer.WhatsNewText,CustomPlayer.A4ReqTitle,CustomPlayer.A4Stone,CustomPlayer.A4ReqStone,CustomPlayer.A4WepTier,CustomPlayer.A4ReqWepTier,CustomPlayer.A4MaxStr,CustomPlayer.A4ReqMaxStr}
                CustomPlayer.A4UnlockButton.OnClick().SubscribeAgent3(PurchaseArea4, 15)
                CustomPlayer.A4CloseButton.OnClick().Subscribe(RemoveA4HUD)

    
    EnterSpace(Agent:agent, WhichMusic:string):void=
        if(Player:=player[Agent], CustomPlayer := gm.PlayersMap[Player]):
            Print("IN SPACE TRUE")
            set CustomPlayer.InSpace = true
            gm.SwitchMusic(Agent, WhichMusic)

    SwitchMusic(Agent:agent, WhichMusic:string):void=
        gm.SwitchMusic(Agent, WhichMusic)

    TeleportToQuarry(Agent:agent):void=
        if(Player:=player[Agent], CustomPlayer := gm.PlayersMap[Player]):
            Print("IN SPACE FALSE")
            set CustomPlayer.InSpace = false
    
    UnlockRooms(CustomPlayer:Custom_Player):void=
        for(i:int=0..RoomUnlockTexts.Length-1):
            if(RoomUnlockText:=RoomUnlockTexts[i]):
                if(Area := 1 + Floor(i/5), Stage:= 1 + Mod[i,5], RoomUnlockCost:=RoomUnlockCosts[i]):
                    RoomUnlockText.SetText(StringToMessage("Area {Area} - Stage {Stage}\nUnlock: {NumberCondensor(RoomUnlockCost)}"))

        for(i:int=0..RoomUnlockBarriers.Length):
            if(RoomUnlockBarrier:=RoomUnlockBarriers[i]):
                if(CustomPlayer.StagesUnlocked >= i):
                   RoomUnlockBarrier.AddToIgnoreList(CustomPlayer.Player)
                else:
                    RoomUnlockBarrier.RemoveFromIgnoreList(CustomPlayer.Player)

    CanGrindRails(Agent:agent, whichRoom:int):void=
        if(Player:=player[Agent], CustomPlayer := gm.PlayersMap[Player]):
            Tutorial.UpdateTutorialUI(CustomPlayer, "GrindRails", 1.0)


    TeleportPlayerToCave(Agent:?agent, whichRoom:int):void=
        if(IsAgent:= Agent?, Player:=player[IsAgent], CustomPlayer := gm.PlayersMap[Player]):
            if(CustomPlayer.StagesUnlocked >= whichRoom):
                Tutorial.UpdateTutorialUI(CustomPlayer, "Teleport", 1.0)
                if(RoomTeleporter:=RoomTeleporters[whichRoom]):
                    RoomTeleporter.Teleport(Player)
                    Print("whichRoom: {whichRoom}")
                    if(whichRoom >= 15):
                        Print("IN SPACE TRUE")
                        set CustomPlayer.InSpace = true
                    if(whichRoom >= 20):
                        Print("UNLOCK A5 IF NOT UNLCOKED")
                        set CustomPlayer.A5Unlocked = true
            else:
                gm.ShowMessage(CustomPlayer.Player, "You do not have this room unlocked yet!", 1)

    PurchaseRoomUnlock(Agent:?agent, whichRoom:int):void=
        if(IsAgent:= Agent?, CustomPlayer := gm.PlayersMap[player[IsAgent]], RoomUnlockCost:=RoomUnlockCosts[whichRoom]):
            if(CustomPlayer.StagesUnlocked < whichRoom):
                if(CustomPlayer.StoneAmount >= RoomUnlockCost):
                    Tutorial.UpdateTutorialUI(CustomPlayer, "UnlockRoom", 1000.0)
                    if(Accolade:=AudioManager.Accolades[3]): #AWARD PLAYER EXP
                        Accolade.Award(CustomPlayer.Player)
                    if(Analytic:=AudioManager.GeneralAnalytics[0]): #ANALYTIC
                        Analytic.Submit(CustomPlayer.Player)
                    if(whichRoom >= 1):
                        set CustomPlayer.EggsUnlocked = true
                    if(RoomUnlockBarrier:=RoomUnlockBarriers[whichRoom]):
                        RoomUnlockBarrier.AddToIgnoreList(CustomPlayer.Player)
                    CustomPlayer.Spend(RoomUnlockCost, "Stone", 0)
                    set CustomPlayer.RoomUnlockGather += 0.05
                    set CustomPlayer.StagesUnlocked += 1
                    gm.ShowMessage(CustomPlayer.Player, "Stage Unlocked!", 0)
                    SaveAllData(CustomPlayer)
                else:
                    gm.ShowMessage(CustomPlayer.Player, "You do not have enough Stone!", 1)
            else:
                gm.ShowMessage(CustomPlayer.Player, "Room Already Unlocked!", 1)

    PurchaseArea4(Wdata:widget_message, whichRoom:int):void=
        var StrTiersMaxed : int = 0
        if(CustomPlayer := gm.PlayersMap[Wdata.Player], RoomUnlockCost:=RoomUnlockCosts[whichRoom]):
            if(CustomPlayer.A4Unlocked = false):
                for(i:int=0..CustomPlayer.StrengthUpgradeLevels.Length):
                    if:
                        CustomPlayer.StrengthUpgradeLevels[i] >= 100
                        CustomPlayer.StrengthEfficiencies[i] >= 7
                    then:
                        set StrTiersMaxed += 1
                if(CustomPlayer.StagesUnlocked < whichRoom):
                    if:
                        CustomPlayer.StoneAmount >= RoomUnlockCost and
                        CustomPlayer.WeaponTier >= 6 and
                        StrTiersMaxed >= 40
                    then:
                        set CustomPlayer.A4Unlocked = true
                        if(Accolade:=AudioManager.Accolades[3]): #AWARD PLAYER EXP
                            Accolade.Award(CustomPlayer.Player)
                        if(Analytic:=AudioManager.GeneralAnalytics[0]): #ANALYTIC
                            Analytic.Submit(CustomPlayer.Player)
                        if(RoomUnlockBarrier:=RoomUnlockBarriers[whichRoom]):
                            CustomPlayer.Spend(RoomUnlockCost, "Stone", 0)
                            set CustomPlayer.RoomUnlockGather += 0.05
                            RoomUnlockBarrier.AddToIgnoreList(CustomPlayer.Player)
                            set CustomPlayer.StagesUnlocked += 1
                            AudioManager.LeveledUp.Play(CustomPlayer.Player)
                            SaveAllData(CustomPlayer)
                            RemoveA4HUD(Wdata)
                    else:
                        AudioManager.Denied.Play(CustomPlayer.Player)
                else:
                    AudioManager.Denied.Play(CustomPlayer.Player)
            else:
                if(CustomPlayer.StagesUnlocked < 15):
                    if(Accolade:=AudioManager.Accolades[3]): #AWARD PLAYER EXP
                        Accolade.Award(CustomPlayer.Player)
                    if(Analytic:=AudioManager.GeneralAnalytics[0]): #ANALYTIC
                        Analytic.Submit(CustomPlayer.Player)
                    if(RoomUnlockBarrier:=RoomUnlockBarriers[whichRoom]):
                        RoomUnlockBarrier.AddToIgnoreList(CustomPlayer.Player)
                    #CustomPlayer.Spend(RoomUnlockCost, "Stone", 0)
                    set CustomPlayer.RoomUnlockGather += 0.05
                    set CustomPlayer.StagesUnlocked += 1
                    AudioManager.LeveledUp.Play(CustomPlayer.Player)
                    SaveAllData(CustomPlayer)
                    RemoveA4HUD(Wdata)
                else:
                    AudioManager.Denied.Play(CustomPlayer.Player)

    PurchaseArea5(Agent:agent, whichRoom:int):void=
        if(CustomPlayer := gm.PlayersMap[Agent], RoomUnlockCost:=RoomUnlockCosts[whichRoom]):
            if(CustomPlayer.A5Unlocked = false):
                if(CustomPlayer.StagesUnlocked >= 19):
                    if:
                        CustomPlayer.OneTimeAnalytics["DestroyMeteor"] = true
                    then:
                        if(CustomPlayer.OneTimeAnalytics["UnlockArea5"] = false):
                            if(set CustomPlayer.OneTimeAnalytics["UnlockArea5"] = true) {}
                            if(Analytic:=AudioManager.OneTimeAnalytics[12]): #ANALYTIC
                                Analytic.Submit(CustomPlayer.Player)
                        set CustomPlayer.A5Unlocked = true
                        if(Accolade:=AudioManager.Accolades[3]): #AWARD PLAYER EXP
                            Accolade.Award(CustomPlayer.Player)
                        if(Analytic:=AudioManager.GeneralAnalytics[0]): #ANALYTIC
                            Analytic.Submit(CustomPlayer.Player)
                        if(CustomPlayer.StagesUnlocked = 19):
                            set CustomPlayer.RoomUnlockGather += 0.05
                            set CustomPlayer.StagesUnlocked += 1
                        if(RoomTeleporter:=RoomTeleporters[whichRoom]):
                            RoomTeleporter.Teleport(CustomPlayer.Player)
                        AudioManager.LeveledUp.Play(CustomPlayer.Player)
                        SaveAllData(CustomPlayer)
                        gm.ShowMessage(CustomPlayer.Player, "Area 5 Unlocked!", 0)
                    else:
                        gm.ShowMessage(CustomPlayer.Player, "You must destroy the meteor!", 1)
                        AudioManager.Denied.Play(CustomPlayer.Player)
                else:
                    gm.ShowMessage(CustomPlayer.Player, "You are not ready!", 1)
                    AudioManager.Denied.Play(CustomPlayer.Player)
            else:
                if(CustomPlayer.StagesUnlocked >= 19):
                    if(Accolade:=AudioManager.Accolades[3]): #AWARD PLAYER EXP
                        Accolade.Award(CustomPlayer.Player)
                    if(Analytic:=AudioManager.GeneralAnalytics[0]): #ANALYTIC
                        Analytic.Submit(CustomPlayer.Player)
                    if(CustomPlayer.StagesUnlocked = 19):
                        set CustomPlayer.RoomUnlockGather += 0.05
                        set CustomPlayer.StagesUnlocked += 1
                    if(RoomTeleporter:=RoomTeleporters[whichRoom]):
                        RoomTeleporter.Teleport(CustomPlayer.Player)
                    AudioManager.LeveledUp.Play(CustomPlayer.Player)
                    gm.ShowMessage(CustomPlayer.Player, "Area 5 Unlocked!", 0)
                    SaveAllData(CustomPlayer)
                else:
                    if(CustomPlayer.StagesUnlocked > 19):
                        if(RoomTeleporter:=RoomTeleporters[whichRoom]):
                            RoomTeleporter.Teleport(CustomPlayer.Player)
                    else:
                        gm.ShowMessage(CustomPlayer.Player, "You are not ready!", 1)
                        AudioManager.Denied.Play(CustomPlayer.Player)

    OpenA4HUD(Agent:agent): void=
        if(CustomPlayer := gm.PlayersMap[Agent]):
                if(MyUI:=CustomPlayer.UIMaps["A4"]?):
                else:
                    var MyA4Canvas : canvas = MakeA4Canvas(CustomPlayer)
                    CustomPlayer.AddFullHUD(MyA4Canvas)
                    if(set CustomPlayer.UIMaps["A4"] = option{MyA4Canvas}) {}
                    UpdateA4UnlockHUD(CustomPlayer)

    RemoveA4HUD(Wdata:widget_message):void=
        if(CustomPlayer := gm.PlayersMap[Wdata.Player]):
            if(Player:=Wdata.Player, PlayerUI:=GetPlayerUI[Player]):
                if(MyUI:=CustomPlayer.UIMaps["A4"]?):
                    PlayerUI.RemoveWidget(MyUI)
                    MyUI.SetVisibility(widget_visibility.Collapsed)
                    if(set CustomPlayer.UIMaps["A4"] = false){}

    MakeA4Canvas(CustomPlayer:Custom_Player):canvas=
        CustomPlayer.A4CloseButton.SetText(StringToMessage("X"))

        CustomPlayer.WhatsNewTitle.SetText(StringToMessage("What's Unlocked?"))
        CustomPlayer.WhatsNewText.SetText(StringToMessage("New area to explore\n\nNew Ability: Void Fracture\n\nNew 'Infinite' Rock with bonuses\n\nNew Unique Boss\n\nNew Mobs\n\nNew Weapon & Strength Tiers\n\n3x Strength Prestige Multiplier"))
        CustomPlayer.A4ReqTitle.SetText(StringToMessage("Requirements"))
        CustomPlayer.A4ReqWepTier.SetText(StringToMessage("Mythic Weapon"))
        CustomPlayer.A4ReqMaxStr.SetText(StringToMessage("40 Max Strength Tiers"))

        for(TextBlock : CustomPlayer.AllA4Texts):
            TextBlock.SetShadowColor(CustomPlayer.ShadowDefaults.DefaultShadowColor)
            TextBlock.SetShadowOffset(CustomPlayer.ShadowDefaults.DefaultShadowOffset)
            TextBlock.SetShadowOpacity(1.0)

        CustomPlayer.A4UnlockButton.SetText(StringToMessage("UNLOCK"))
        var MyA4Canvas: canvas = canvas:
            Slots := array:
                canvas_slot:
                    Widget:=CustomPlayer.A4UnlockHUDBG
                    ZOrder:=0
                    SizeToContent:=false
                    Alignment := vector2{X := 0.5, Y := 0.0}
                    Anchors := anchors{Minimum := vector2{X:= 0.5, Y:= 0.0}, Maximum := vector2{X:= 0.5, Y:= 0.0}}
                    Offsets:=margin{Top:=0.0,Left:=0.0,Bottom:=1024.0,Right:=2048.0}
                canvas_slot:
                    Widget:=CustomPlayer.A4UnlockHUDTitle
                    ZOrder:=1
                    SizeToContent:=false
                    Alignment := vector2{X := 0.5, Y := 0.0}
                    Anchors := anchors{Minimum := vector2{X:= 0.5, Y:= 0.0}, Maximum := vector2{X:= 0.5, Y:= 0.0}}
                    Offsets:=margin{Top:=130.0,Left:=0.0,Bottom:=56.0,Right:=495.0}
                canvas_slot:
                    Widget:=CustomPlayer.A4CloseButton
                    ZOrder:=9
                    SizeToContent:=false
                    Alignment := vector2{X:= 0.5, Y:= 0.0}
                    Anchors := anchors{Minimum := vector2{X:= 0.5, Y:= 0.0}, Maximum := vector2{X:= 0.5, Y:= 0.0}}
                    Offsets:=margin{Top:=130.0,Left:=500.0,Bottom:=56.0, Right:=76.0}
                
                canvas_slot:
                    Widget:=CustomPlayer.WhatsNewTitle
                    ZOrder:=1
                    SizeToContent:=true
                    Alignment := vector2{X:= 0.5, Y:= 0.0}
                    Anchors := anchors{Minimum := vector2{X:= 0.5, Y:= 0.0}, Maximum := vector2{X:= 0.5, Y:= 0.0}}
                    Offsets:=margin{Top:=296.0,Left:=-335.0}
                canvas_slot:
                    Widget:=CustomPlayer.WhatsNewText
                    ZOrder:=1
                    SizeToContent:=true
                    Alignment := vector2{X:= 0.0, Y:= 0.0}
                    Anchors := anchors{Minimum := vector2{X:= 0.5, Y:= 0.0}, Maximum := vector2{X:= 0.5, Y:= 0.0}}
                    Offsets:=margin{Top:=352.0,Left:=-600.0}
                canvas_slot:
                    Widget:=CustomPlayer.A4ReqTitle
                    ZOrder:=1
                    SizeToContent:=true
                    Alignment := vector2{X:= 0.5, Y:= 0.0}
                    Anchors := anchors{Minimum := vector2{X:= 0.5, Y:= 0.0}, Maximum := vector2{X:= 0.5, Y:= 0.0}}
                    Offsets:=margin{Top:=296.0,Left:=450.0}

                canvas_slot:
                    ZOrder:=1
                    SizeToContent:=true
                    Alignment := vector2{X:= 0.5, Y:= 0.0}
                    Anchors := anchors{Minimum := vector2{X:= 0.5, Y:= 0.0}, Maximum := vector2{X:= 0.5, Y:= 0.0}}
                    Offsets:=margin{Top:=412.0,Left:=450.0}
                    Widget := stack_box:
                        Orientation := orientation.Horizontal
                        Slots := array:
                            stack_box_slot:
                                Widget := CustomPlayer.A4StoneIcon  
                            stack_box_slot:
                                Widget := CustomPlayer.A4Stone
                            stack_box_slot:
                                Widget := CustomPlayer.A4ReqStone
                canvas_slot:
                    ZOrder:=1
                    SizeToContent:=true
                    Alignment := vector2{X:= 0.5, Y:= 0.0}
                    Anchors := anchors{Minimum := vector2{X:= 0.5, Y:= 0.0}, Maximum := vector2{X:= 0.5, Y:= 0.0}}
                    Offsets:=margin{Top:=520.0,Left:=450.0}
                    Widget := stack_box:
                        Orientation := orientation.Horizontal
                        Slots := array:
                            stack_box_slot:
                                Widget := CustomPlayer.A4WepTier  
                            stack_box_slot:
                                Widget := CustomPlayer.A4ReqWepTier
                canvas_slot:
                    ZOrder:=1
                    SizeToContent:=true
                    Alignment := vector2{X:= 0.5, Y:= 0.0}
                    Anchors := anchors{Minimum := vector2{X:= 0.5, Y:= 0.0}, Maximum := vector2{X:= 0.5, Y:= 0.0}}
                    Offsets:=margin{Top:=640.0,Left:=450.0}
                    Widget := stack_box:
                        Orientation := orientation.Horizontal
                        Slots := array:
                            stack_box_slot:
                                Widget := CustomPlayer.A4MaxStr  
                            stack_box_slot:
                                Widget := CustomPlayer.A4ReqMaxStr

                canvas_slot:
                    Widget:=CustomPlayer.A4UnlockButton
                    ZOrder:=9
                    SizeToContent:=false
                    Alignment := vector2{X:= 0.5, Y:= 0.0}
                    Anchors := anchors{Minimum := vector2{X:= 0.5, Y:= 0.0}, Maximum := vector2{X:= 0.5, Y:= 0.0}}
                    Offsets:=margin{Top:=820.0, Left:=450.0,Bottom:=60.0, Right:=400.0}
         
        return MyA4Canvas

    UpdateA4UnlockHUD(CustomPlayer:Custom_Player):void=
        var StrTiersMaxed : int = 0
        CustomPlayer.A4Stone.SetText(StringToMessage("{NumberCondensor(CustomPlayer.StoneAmount)}"))
        if(CustomPlayer.A4Unlocked = false):
            if(RoomCost:=RoomUnlockCosts[15], WepTier:=WeaponTierNames[CustomPlayer.WeaponTier-1]):
                CustomPlayer.A4ReqStone.SetText(StringToMessage(" / {NumberCondensor(RoomCost)}"))
                CustomPlayer.A4WepTier.SetText(StringToMessage("{WepTier} / "))

                if(CustomPlayer.StoneAmount >= RoomCost):
                    CustomPlayer.A4Stone.SetTextColor(MakeColorFromHex("39FF39"))
                    CustomPlayer.A4ReqStone.SetTextColor(MakeColorFromHex("39FF39"))
                else:
                    CustomPlayer.A4Stone.SetTextColor(MakeColorFromHex("fa4d56"))
                    CustomPlayer.A4ReqStone.SetTextColor(MakeColorFromHex("fa4d56"))

                if(CustomPlayer.WeaponTier >= 6):
                    CustomPlayer.A4WepTier.SetTextColor(MakeColorFromHex("39FF39"))
                    CustomPlayer.A4ReqWepTier.SetTextColor(MakeColorFromHex("39FF39"))
                else:
                    CustomPlayer.A4WepTier.SetTextColor(MakeColorFromHex("fa4d56"))
                    CustomPlayer.A4ReqWepTier.SetTextColor(MakeColorFromHex("fa4d56"))


            for(i:int=0..CustomPlayer.StrengthUpgradeLevels.Length):
                if:
                    CustomPlayer.StrengthUpgradeLevels[i] >= 100
                    CustomPlayer.StrengthEfficiencies[i] >= 7
                then:
                    set StrTiersMaxed += 1
                
            CustomPlayer.A4MaxStr.SetText(StringToMessage("{StrTiersMaxed} / "))

            if(StrTiersMaxed >= 40):
                CustomPlayer.A4MaxStr.SetTextColor(MakeColorFromHex("39FF39"))
                CustomPlayer.A4ReqMaxStr.SetTextColor(MakeColorFromHex("39FF39"))
            else:
                CustomPlayer.A4MaxStr.SetTextColor(MakeColorFromHex("fa4d56"))
                CustomPlayer.A4ReqMaxStr.SetTextColor(MakeColorFromHex("fa4d56"))
        else:
            CustomPlayer.A4ReqStone.SetText(StringToMessage("Press to Unlock!"))
            CustomPlayer.A4Stone.SetText(StringToMessage(""))
            CustomPlayer.A4WepTier.SetText(StringToMessage(""))
            CustomPlayer.A4MaxStr.SetText(StringToMessage(""))
            CustomPlayer.A4ReqWepTier.SetText(StringToMessage(""))
            CustomPlayer.A4ReqMaxStr.SetText(StringToMessage(""))
            CustomPlayer.A4Stone.SetTextColor(MakeColorFromHex("39FF39"))
            CustomPlayer.A4ReqStone.SetTextColor(MakeColorFromHex("39FF39"))
            CustomPlayer.A4WepTier.SetTextColor(MakeColorFromHex("39FF39"))
            CustomPlayer.A4ReqWepTier.SetTextColor(MakeColorFromHex("39FF39"))

                
                