using { /Fortnite.com/Devices }
using { /Verse.org/Simulation }
using { /Verse.org/Colors }
using { /Verse.org/Verse }
using { /Verse.org/Random }
using { /Fortnite.com/Teams }
using { /UnrealEngine.com/Assets }
using { /Fortnite.com/Game }
using { /Fortnite.com/Characters }
using { /Fortnite.com/Playspaces }
using { /Fortnite.com/UI }
using { /UnrealEngine.com/Temporary/Diagnostics }
using { /UnrealEngine.com/Temporary/SpatialMath }
using { /UnrealEngine.com/Temporary/UI }
using { /Verse.org/Assets}
using { /Verse.org/Colors/NamedColors }
using { /Fortnite.com/AI }

seasonal_halloween_manager := class(creative_device):

    @editable
    gm : game_manager = game_manager{}

    @editable
    AudioManager: audio_manager = audio_manager{}

    @editable
    pm: pet_manager = pet_manager{}

    @editable
    PetDoggoButton : button_device = button_device{}

    @editable
    PetDemonButton : button_device = button_device{}

    #@editable
    #PetDoggoSpawner : npc_spawner_device = npc_spawner_device{}

    @editable
    BossSpawner : npc_spawner_device = npc_spawner_device{}
    @editable
    MobSpawners : []npc_spawner_device = array{}

    var DamageNumbers<private>: damage_numbers = damage_numbers{}

    HPBars: health_bar_device = health_bar_device{}

    var QueuedPlayers : [int]?agent = map{}
    var BossHP : [fort_character]float = map{}
    var EachBoss : [int]fort_character = map{}
    var BossMaxHP : float = 1000.0 #100000.0 #CHANGE TO 100000.0
    var BossPhase : int = 1
    var CanBeDamaged : logic = true
    var BossDefeated : logic = false

    OnBegin<override>()<suspends>:void=
        PetDoggoButton.InteractedWithEvent.Subscribe(OnPetSpawnButton)
        PetDemonButton.InteractedWithEvent.Subscribe(OnDemonSpawnButton)
        #BossSpawner.SpawnedEvent.Subscribe(OnBossSpawned)
        
        <#for(i:int=0..MobSpawners.Length):
            if(MobSpawner:=MobSpawners[i]):
                MobSpawner.SpawnedEvent.SubscribeAgent(OnMobSpawned, i)#>

    OnPetSpawnButton(Agent:agent):void=
        Print("PET SPAWN BUTTON")
        if(CustomPlayer := gm.PlayersMap[Agent]):
            if:
                CustomPlayer.AllBossesDefeated[5] = true
            then:
                gm.ShowMessage(CustomPlayer.Player, "Here is your puppy!", 0)
                pm.SpawnPets(CustomPlayer, 6)
            else:
                gm.ShowMessage(CustomPlayer.Player, "You must defeat the seasonal boss first", 1)

    OnDemonSpawnButton(Agent:agent):void=
        Print("PET DEMON BUTTON")
        if(CustomPlayer := gm.PlayersMap[Agent]):
            if:
                CustomPlayer.AllBossesDefeated[7] = true #HARD MODE DEMON
                CustomPlayer.DemonPetObtained = true #GOT LUCKY
            then:
                pm.SpawnPets(CustomPlayer, 7)
            else:
                gm.ShowMessage(CustomPlayer.Player, "You do not own this pet!", 1)


    <#StartBossFight(Agents:[int]?agent):void=
        set QueuedPlayers = Agents
        Print("START BOSS FIGHT FROM SEASONAL BOSS")
        set BossPhase = 1
        set BossDefeated = false
        BossSpawner.Enable()
        BossSpawner.Spawn()

    UpdateBossTargets(Agent:agent, Agents:[int]?agent):void=
        Print("UPDATED BOSS TARGETS")#>
        
    <#OnBossSpawned(Agent:agent):void=
        if (Boss := Agent.GetFortCharacter[]):
            if(set BossHP[Boss] = BossMaxHP){}
            if(set EachBoss[0] = Boss) {}
            Boss.SetMaxHealth(2e200) #BossMaxHP
            UpdateBossTargets(Agent, QueuedPlayers)
            Boss.DamagedEvent().SubscribeAgent5(OnBossDamaged, Boss)#>

    OnMobSpawned(Agent:agent, whichRoom:int):void=
        if (Creep := Agent.GetFortCharacter[]):
            Creep.SetMaxHealth(1.0)
            if:
                Fort:=Agent.GetFortCharacter[]
                NPCBehave:=mob_halloween_behaviour[Agent.GetNPCBehavior[]]
            then:
                if:
                    whichRoom = 0 or
                    whichRoom = 3 or
                    whichRoom = 5
                then:
                    NPCBehave.SetTargetAgents(QueuedPlayers, 0)
                else:
                    NPCBehave.SetTargetAgents(QueuedPlayers, 1)

                Creep.EliminatedEvent().Subscribe(NPCBehave.SetIsElim)
            #Creep.DamagedEvent().SubscribeAgent5(OnMobDamaged, Creep)

    <#OnBossDamaged(DmgResult : damage_result, Boss:fort_character) : void=
        if:
            MaybeInstigator := DmgResult.Instigator
            Instigator := MaybeInstigator?
            InstagatingAgent := Instigator.GetInstigatorAgent[]
            Player:=player[InstagatingAgent]
            CustomPlayer := gm.PlayersMap[Player]
            #CanBeDamaged = true #CHANGE TO UNCOMMENTED
        then:
            var DamageResult : tuple(float, logic, float, int) = CustomPlayer.CalcTotalDamage("Mob")
            DmgResult.Target.Damage(100.0)
            if(set BossHP[Boss] -= 100.0) {}
            #DamageNumbers.ShowDamageUIForPlayer(CustomPlayer, 100.0, DamageResult(1), DamageResult(3), false, "Mob", -1.0)
            for(i:int=0..QueuedPlayers.Length):
                if:
                    PlayerAgent:=QueuedPlayers[i]?
                    CPlayer := gm.PlayersMap[PlayerAgent]
                then:
                    if(BossHP[Boss] <= 0.0):
                        Print("DEFEATED")
                            Boss.Damage(9e200)
                            BossSpawner.DespawnAll(false)
                            if(set BossHP[Boss] = 0.0){}
                            HPBars.TakeDamage(CPlayer, 0.0, ?IsPerc:=true)
                            AudioManager.BigTextHUD.SetText(StringToMessage("BOSS DEFEATED!"))
                            AudioManager.BigTextHUD.Show(CPlayer.Player)
                            AudioManager.BossDefeatedVictory.Play(CPlayer.Player)
                            if(ChestReward:=AudioManager.ChestWaiting[CPlayer.CurrentBoss], Beacon:=AudioManager.ChestBeacons[CPlayer.CurrentBoss]):
                                ChestReward.Play(CPlayer.Player)
                                Beacon.AddToShowList(CPlayer.Player)
                            CPlayer.RemoveRockHealthBarUI()
                            set BossDefeated = true
                            if(set CustomPlayer.AllBossesDefeated[CPlayer.CurrentBoss] = true) {Print("{CPlayer.CurrentBoss} BOSS DEFEATED")}
                    else:
                        if(BossHP[Boss] <= (BossMaxHP / 3.0)):
                            #Print("PHASE 3")
                            if:
                                BossPhase <> 3
                                BossAgent := Boss.GetAgent[]
                                NPCBehave := halloween_boss_behaviour[BossAgent.GetNPCBehavior[]]
                            then:
                                set BossPhase = 3
                                NPCBehave.SetBossPhase()
                                spawn{StartMobSpawn(QueuedPlayers)}

                        else if(BossHP[Boss] <= (BossMaxHP / 3.0) + (BossMaxHP / 3.0)):
                            #Print("PHASE 2")
                            if:
                                BossPhase <> 2
                                BossAgent := Boss.GetAgent[]
                                NPCBehave := halloween_boss_behaviour[BossAgent.GetNPCBehavior[]]
                            then:
                                set BossPhase = 2
                                NPCBehave.SetBossPhase()
                                spawn{StartMobSpawn(QueuedPlayers)}
                        else:
                            #Print("PHASE 1")
                            set BossPhase = 1

                        if:
                            NewBossHP:=BossHP[Boss]
                            BossHPPerc := (NewBossHP / BossMaxHP) * 100.0
                        then:
                            HPBars.TakeDamage(CPlayer, BossHPPerc, ?IsPerc:=true)#>
             
    <#StartMobSpawn(Agents:[int]?agent)<suspends>:void=
        Print("RELEASE THE MOBS")
        set CanBeDamaged = false
        for(MobSpawner:MobSpawners):
            MobSpawner.Enable()
        for(i:int=0..Agents.Length):
            if:
                PlayerAgent:=Agents[i]?
                CPlayer := gm.PlayersMap[PlayerAgent]
            then:
                CPlayer.RockBarColorTop.SetColor(MakeColorFromHex("474747"))
                CPlayer.RockBarColorTop2.SetColor(MakeColorFromHex("333333"))
                CPlayer.RockBarColorTopSmooth.SetColor(MakeColorFromHex("F5F5F5"))

        Sleep(45.0) #Make sure this sleep is same as behaviour

        Print("REMOVE THE MOBS")
        set CanBeDamaged = true
        for(MobSpawner:MobSpawners):
            MobSpawner.Disable()
        for(i:int=0..Agents.Length):
            if:
                PlayerAgent:=Agents[i]?
                CPlayer := gm.PlayersMap[PlayerAgent]
            then:
                CPlayer.RockBarColorTop.SetColor(MakeColorFromHex("D61F28"))
                CPlayer.RockBarColorTop2.SetColor(MakeColorFromHex("A1171E"))
                CPlayer.RockBarColorTopSmooth.SetColor(MakeColorFromHex("EB7076"))#>

    <#OnRewardButton(Agent:agent):void=
        if(CustomPlayer := gm.PlayersMap[Agent]):
            if(BossDefeated = true and CustomPlayer.CurrentBoss = 5):
                if(Analytic:=AudioManager.BossAnalytics[5]): #ANALYTIC
                    Analytic.Submit(CustomPlayer.Player)
                if(Accolade:=AudioManager.Accolades[2]): #AWARD PLAYER EXP
                    Accolade.Award(CustomPlayer.Player)
                #UpdateBossHUD(CustomPlayer)
                if:
                    ChestReward:=AudioManager.ChestWaiting[5]
                    OpenChest:=AudioManager.ChestOpen[5]
                    Beacon:=AudioManager.ChestBeacons[5]
                then:
                    ChestReward.Stop(CustomPlayer.Player)
                    OpenChest.Play(CustomPlayer.Player)
                    Beacon.RemoveFromShowList(CustomPlayer.Player)
                set CustomPlayer.CurrentBoss = -1
                CustomPlayer.UpAchValue(6, 1.0)
                spawn{GiveRewards(CustomPlayer)}#>