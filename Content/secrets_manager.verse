
using { /Fortnite.com/Devices }
using { /Verse.org/Simulation }
using { /UnrealEngine.com/Temporary/Diagnostics }


secrets_manager := class(creative_device):

    @editable
    gm : game_manager = game_manager{}

    @editable
    AudioManager : audio_manager = audio_manager{}

    @editable
    SecretButtons : []button_device = array{}

    OnBegin<override>()<suspends>:void=
        for(i:int=0..SecretButtons.Length-1):
            if(SecretButton:=SecretButtons[i]):
                SecretButton.InteractedWithEvent.SubscribeAgent(ClaimSecret, i)

    ClaimSecret(Agent:agent, WhichSecret:int):void=
        if(CustomPlayer := gm.PlayersMap[player[Agent]]):
            if(CustomPlayer.NewSecretsClaimed[WhichSecret] = false):
                AudioManager.ChestOpenGlobal.Play(CustomPlayer.Player)
                if(set CustomPlayer.NewSecretsClaimed[WhichSecret] = true) {}
                gm.ShowMessage(CustomPlayer.Player, "+5% Strength Obtained!", 0)
            else:
                gm.ShowMessage(CustomPlayer.Player, "Secret Already Claimed!", 1)

        