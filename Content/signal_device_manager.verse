using { /Fortnite.com/Devices }
using { /Verse.org/Simulation }
using { /Verse.org/Colors }
using { /Verse.org/Verse }
using { /Verse.org/Random }
using { /Fortnite.com/Teams }
using { /UnrealEngine.com/Assets }
using { /Fortnite.com/Game }
using { /Fortnite.com/Characters }
using { /Fortnite.com/Playspaces }
using { /Fortnite.com/UI }
using { /UnrealEngine.com/Temporary/Diagnostics }
using { /UnrealEngine.com/Temporary/SpatialMath }
using { /UnrealEngine.com/Temporary/UI }
using { /Verse.org/Assets}

signal_device_manager := class(creative_device):

    @editable
    gm : game_manager = game_manager{}

    @editable
    AudioManager : audio_manager = audio_manager{}

    @editable
    Tutorial : tutorial_manager = tutorial_manager{}

    @editable
    ShowSUP : signal_remote_manager_device = signal_remote_manager_device{}

    var EfficiencyCostMult : float = 8.0
    var BaseStrengthCostMult : float = 1.1

    var SUPIconImgs : []texture = array{
        UI.SUP1,UI.SUP2,UI.SUP3,UI.SUP4,UI.SUP5,
        UI.SUP6,UI.SUP7,UI.SUP8,UI.SUP9,UI.SUP10,
        UI.SUP11,UI.SUP12,UI.SUP13,UI.SUP14,UI.SUP15,
        UI.SUP16,UI.SUP17,UI.SUP18,UI.SUP19,UI.SUP20,
        UI.SUP21,UI.SUP22,UI.SUP23,UI.SUP24,UI.SUP25,
        UI.SUP26,UI.SUP27,UI.SUP28,UI.SUP29,UI.SUP30,
        UI.SUP31,UI.SUP32,UI.SUP33,UI.SUP34,UI.SUP35,
        UI.SUP36,UI.SUP37,UI.SUP38,UI.SUP39,UI.SUP40,
        
        UI.SUP41,UI.SUP42,UI.SUP43,UI.SUP44,UI.SUP45,
        UI.SUP46,UI.SUP47,UI.SUP48,UI.SUP49,UI.SUP50,
        UI.SUP41,UI.SUP42,UI.SUP43,UI.SUP44,UI.SUP45,
        UI.SUP46,UI.SUP47,UI.SUP48,UI.SUP49,UI.SUP50,
        UI.SUP41,UI.SUP42,UI.SUP43,UI.SUP44,UI.SUP45,
        UI.SUP46,UI.SUP47,UI.SUP48,UI.SUP49,UI.SUP50,
        UI.SUP41,UI.SUP42,UI.SUP43,UI.SUP44,UI.SUP45,
        UI.SUP46,UI.SUP47,UI.SUP48,UI.SUP49,UI.SUP50,

        UI.SUP41,UI.SUP42,UI.SUP43,UI.SUP44,UI.SUP45,
        UI.SUP46,UI.SUP47,UI.SUP48,UI.SUP49,UI.SUP50,
        UI.SUP41,UI.SUP42,UI.SUP43,UI.SUP44,UI.SUP45,
        UI.SUP46,UI.SUP47,UI.SUP48,UI.SUP49,UI.SUP50,
        UI.SUP41,UI.SUP42,UI.SUP43,UI.SUP44,UI.SUP45,
        UI.SUP46,UI.SUP47,UI.SUP48,UI.SUP49,UI.SUP50
        }

    OnBegin<override>()<suspends>:void=
        ShowSUP.PrimarySignalEvent.Subscribe(OpenSUPHUD)
        
        for(Spawner : gm.PlayerSpawners):     
            Spawner.SpawnedEvent.Subscribe(OnPlayerJoined)

    OnPlayerJoined(Agent:agent):void=
        spawn{PlayerJoinedBuffer(Agent)}

    PlayerJoinedBuffer(Agent:agent)<suspends>:void=
        Sleep(2.0)
        if(CustomPlayer := gm.PlayersMap[player[Agent]]):
            if(CustomPlayer.PlayTimeSessionSeconds = 0.0):
                set CustomPlayer.SUPIcons = array{CustomPlayer.SUP1Icon,CustomPlayer.SUP2Icon,CustomPlayer.SUP3Icon,CustomPlayer.SUP4Icon,CustomPlayer.SUP5Icon}
                set CustomPlayer.SUPSurges = array{CustomPlayer.SUP1Surge,CustomPlayer.SUP2Surge,CustomPlayer.SUP3Surge,CustomPlayer.SUP4Surge,CustomPlayer.SUP5Surge}
                set CustomPlayer.SUPRomanNums = array{CustomPlayer.SUP1RomanNum,CustomPlayer.SUP2RomanNum,CustomPlayer.SUP3RomanNum,CustomPlayer.SUP4RomanNum,CustomPlayer.SUP5RomanNum}
                set CustomPlayer.SUPLevels = array{CustomPlayer.SUP1Level,CustomPlayer.SUP2Level,CustomPlayer.SUP3Level,CustomPlayer.SUP4Level,CustomPlayer.SUP5Level}
                set CustomPlayer.SUPLevelEffs = array{CustomPlayer.SUP1LevelEff,CustomPlayer.SUP2LevelEff,CustomPlayer.SUP3LevelEff,CustomPlayer.SUP4LevelEff,CustomPlayer.SUP5LevelEff}
                set CustomPlayer.SUPBonuses = array{CustomPlayer.SUP1Bonus,CustomPlayer.SUP2Bonus,CustomPlayer.SUP3Bonus,CustomPlayer.SUP4Bonus,CustomPlayer.SUP5Bonus}
                set CustomPlayer.SUPNextBonuses = array{CustomPlayer.SUP1NextBonus,CustomPlayer.SUP2NextBonus,CustomPlayer.SUP3NextBonus,CustomPlayer.SUP4NextBonus,CustomPlayer.SUP5NextBonus}
                set CustomPlayer.SUPUpgradesTexts = array{CustomPlayer.SUP1UpgradeText,CustomPlayer.SUP2UpgradeText,CustomPlayer.SUP3UpgradeText,CustomPlayer.SUP4UpgradeText,CustomPlayer.SUP5UpgradeText}
                set CustomPlayer.SUPEfficienciesTexts = array{CustomPlayer.SUP1EfficiencyText,CustomPlayer.SUP2EfficiencyText,CustomPlayer.SUP3EfficiencyText,CustomPlayer.SUP4EfficiencyText,CustomPlayer.SUP5EfficiencyText}
                set CustomPlayer.SUPUpgrades = array{CustomPlayer.SUP1Upgrade,CustomPlayer.SUP2Upgrade,CustomPlayer.SUP3Upgrade,CustomPlayer.SUP4Upgrade,CustomPlayer.SUP5Upgrade}
                set CustomPlayer.SUPEfficiencies = array{CustomPlayer.SUP1Efficiency,CustomPlayer.SUP2Efficiency,CustomPlayer.SUP3Efficiency,CustomPlayer.SUP4Efficiency,CustomPlayer.SUP5Efficiency}
                set CustomPlayer.SUPUpgradesStones = array{CustomPlayer.SUP1StoneIcon1,CustomPlayer.SUP2StoneIcon1,CustomPlayer.SUP3StoneIcon1,CustomPlayer.SUP4StoneIcon1,CustomPlayer.SUP5StoneIcon1}
                set CustomPlayer.SUPEfficienciesStones = array{CustomPlayer.SUP1StoneIcon2,CustomPlayer.SUP2StoneIcon2,CustomPlayer.SUP3StoneIcon2,CustomPlayer.SUP4StoneIcon2,CustomPlayer.SUP5StoneIcon2}
                
                UpdatePower(CustomPlayer)

                for(i:int=0..CustomPlayer.SUPUpgrades.Length):
                    if(SUPUpgrade:=CustomPlayer.SUPUpgrades[i],SUPEfficiency:=CustomPlayer.SUPEfficiencies[i]):
                        SUPUpgrade.OnClick().SubscribeAgent3(OnUpgradeDamageButton, i)
                        SUPEfficiency.OnClick().SubscribeAgent3(UpgradeEfficiency, i)
                CustomPlayer.SUPAutoUpgrade.OnClick().Subscribe(OnAutoUpgrade)
                CustomPlayer.SUPEnableUpgrade.OnClick().Subscribe(OnEnableUpgrade)
                CustomPlayer.SUPCloseButton.OnClick().Subscribe(RemoveSUPHUD)
#UI BLOCKS

#CANVAS
    OpenSUPHUD(Agent:agent): void= #ONLY HUD THAT SHOWS/HIDES LIKE THIS SO KEEP AN EYE ON IF IT'S FASTER
        if(CustomPlayer := gm.PlayersMap[Agent]):
            if(CustomPlayer.CurrentBoss = -1):
                if(MyUI:=CustomPlayer.UIMaps["SUP"]?):
                    CustomPlayer.AddFullHUD(MyUI)
                    UpdateSUPHUD(CustomPlayer)
                else:
                    var MySUPCanvas : canvas = MakeSUPCanvas(CustomPlayer)
                    CustomPlayer.AddFullHUD(MySUPCanvas)
                    if(set CustomPlayer.UIMaps["SUP"] = option{MySUPCanvas}) {}
                    UpdateSUPHUD(CustomPlayer)
            else:
                gm.ShowMessage(CustomPlayer.Player, "Can't use while in boss fight!", 1)
                

    RemoveSUPHUD(Wdata:widget_message):void=
        if(CustomPlayer := gm.PlayersMap[Wdata.Player]):
            if(Player:=Wdata.Player, PlayerUI:=GetPlayerUI[Player]):
                if(MyUI:=CustomPlayer.UIMaps["SUP"]?):
                    PlayerUI.RemoveWidget(MyUI)
                    MyUI.SetVisibility(widget_visibility.Collapsed)
                    if(set CustomPlayer.UIMaps["SUP"] = false){}
                     
    MakeSUPCanvas(CustomPlayer:Custom_Player):canvas=
        CustomPlayer.SUPCloseButton.SetText(StringToMessage("X"))
        #CustomPlayer.SUPAutoUpgrade.SetEnabled(false)
        #CustomPlayer.SUPEnableUpgrade.SetEnabled(false)
        #CustomPlayer.SUPEnableBG.SetVisibility(widget_visibility.Collapsed)
        CustomPlayer.SUP1Icon.SetImage(UI.SUP1)
        CustomPlayer.SUP2Icon.SetImage(UI.SUP2)
        CustomPlayer.SUP3Icon.SetImage(UI.SUP3)
        CustomPlayer.SUP4Icon.SetImage(UI.SUP4)
        CustomPlayer.SUP5Icon.SetImage(UI.SUP5)

        CustomPlayer.SUP1Surge.SetVisibility(widget_visibility.Collapsed)
        CustomPlayer.SUP2Surge.SetVisibility(widget_visibility.Collapsed)
        CustomPlayer.SUP3Surge.SetVisibility(widget_visibility.Collapsed)
        CustomPlayer.SUP4Surge.SetVisibility(widget_visibility.Collapsed)
        CustomPlayer.SUP5Surge.SetVisibility(widget_visibility.Collapsed)

        CustomPlayer.SUPStoneAmountText.SetShadowColor(CustomPlayer.ShadowDefaults.DefaultShadowColor)
        CustomPlayer.SUPStoneAmountText.SetShadowOffset(CustomPlayer.ShadowDefaults.DefaultShadowOffset)
        CustomPlayer.SUPStoneAmountText.SetShadowOpacity(1.0)
        CustomPlayer.SUPTotalDMG.SetShadowColor(CustomPlayer.ShadowDefaults.DefaultShadowColor)
        CustomPlayer.SUPTotalDMG.SetShadowOffset(CustomPlayer.ShadowDefaults.DefaultShadowOffset)
        CustomPlayer.SUPTotalDMG.SetShadowOpacity(1.0)

        for(i:int=0..CustomPlayer.SUPUpgradesTexts.Length):
            if:
                Upgrade := CustomPlayer.SUPUpgradesTexts[i]
                Efficiency := CustomPlayer.SUPEfficienciesTexts[i]
                NextBonus := CustomPlayer.SUPNextBonuses[i]
                Bonus := CustomPlayer.SUPBonuses[i]
                Level := CustomPlayer.SUPLevels[i]
                LevelEff := CustomPlayer.SUPLevelEffs[i]
                RomanNum := CustomPlayer.SUPRomanNums[i]
            then:
                Upgrade.SetShadowColor(CustomPlayer.ShadowDefaults.DefaultShadowColor)
                Upgrade.SetShadowOffset(CustomPlayer.ShadowDefaults.DefaultShadowOffset)
                Upgrade.SetShadowOpacity(1.0)
                Efficiency.SetShadowColor(CustomPlayer.ShadowDefaults.DefaultShadowColor)
                Efficiency.SetShadowOffset(CustomPlayer.ShadowDefaults.DefaultShadowOffset)
                Efficiency.SetShadowOpacity(1.0)
                NextBonus.SetShadowColor(CustomPlayer.ShadowDefaults.DefaultShadowColor)
                NextBonus.SetShadowOffset(CustomPlayer.ShadowDefaults.DefaultShadowOffset)
                NextBonus.SetShadowOpacity(1.0)
                Bonus.SetShadowColor(CustomPlayer.ShadowDefaults.DefaultShadowColor)
                Bonus.SetShadowOffset(CustomPlayer.ShadowDefaults.DefaultShadowOffset)
                Bonus.SetShadowOpacity(1.0)
                Level.SetShadowColor(CustomPlayer.ShadowDefaults.DefaultShadowColor)
                Level.SetShadowOffset(CustomPlayer.ShadowDefaults.DefaultShadowOffset)
                Level.SetShadowOpacity(1.0)
                LevelEff.SetShadowColor(CustomPlayer.ShadowDefaults.DefaultShadowColor)
                LevelEff.SetShadowOffset(CustomPlayer.ShadowDefaults.DefaultShadowOffset)
                LevelEff.SetShadowOpacity(1.0)
                RomanNum.SetShadowColor(CustomPlayer.ShadowDefaults.DefaultShadowColor)
                RomanNum.SetShadowOffset(CustomPlayer.ShadowDefaults.DefaultShadowOffset)
                RomanNum.SetShadowOpacity(1.0)

        var MySUPCanvas: canvas = canvas:
            Slots := array:
                canvas_slot:
                    Widget:=CustomPlayer.SUPHUDBG
                    ZOrder:=0
                    SizeToContent:=false
                    Alignment := vector2{X:= 0.5, Y:= 0.0}
                    Offsets:=margin{Top:= 0.0,Left:= 960.0,Bottom:= 1025.0, Right:= 2048.0}
                canvas_slot:
                    Widget:=CustomPlayer.SUPStoneAmountIcon
                    ZOrder:=1
                    SizeToContent:=false
                    Offsets:=margin{Top:=104.0,Left:=172.0, Right:=45.0, Bottom:=45.0}
                canvas_slot:
                    Widget:=CustomPlayer.SUPStoneAmountText
                    ZOrder:=1
                    SizeToContent:=false
                    Offsets:=margin{Top:=108.0,Left:=228.0}
                canvas_slot:
                    Widget:=CustomPlayer.SUPTotalDMG
                    ZOrder:=1
                    SizeToContent:=false
                    Offsets:=margin{Top:=108.0,Left:=540.0}
                canvas_slot:
                    Widget:=CustomPlayer.SUPTotalDMGIcon
                    ZOrder:=1
                    SizeToContent:=false
                    Offsets:=margin{Top:=104.0,Left:=480.0, Right:=45.0, Bottom:=45.0}

                canvas_slot:
                    Widget:=CustomPlayer.SUPEnableUpgrade
                    ZOrder:=1
                    SizeToContent:=false
                    Alignment := vector2{X:= 0.5, Y:= 0.0}
                    Offsets:=margin{Top:=96.0,Left:=1244.0, Right:=200.0, Bottom:=50.0}
                canvas_slot:
                    Widget:=CustomPlayer.SUPEnableBG
                    ZOrder:=1
                    SizeToContent:=false
                    Alignment := vector2{X:= 0.5, Y:= 0.0}
                    Offsets:=margin{Top:=96.0,Left:=1244.0, Right:=200.0, Bottom:=50.0}

                canvas_slot:
                    Widget:=CustomPlayer.SUPAutoUpgrade
                    ZOrder:=1
                    SizeToContent:=false
                    Alignment := vector2{X:= 0.5, Y:= 0.0}
                    Offsets:=margin{Top:=96.0,Left:=1484.0, Right:=200.0, Bottom:=50.0}
                    
                canvas_slot:
                    Widget:=CustomPlayer.SUPLevelTitle
                    ZOrder:=1
                    SizeToContent:=false
                    Alignment := vector2{X:= 0.5, Y:= 0.0}
                    Offsets:=margin{Top:=188.0,Left:=488.0, Right:=94.0, Bottom:=28.0}
                canvas_slot:
                    Widget:=CustomPlayer.SUPBonusTitle
                    ZOrder:=1
                    SizeToContent:=false
                    Alignment := vector2{X:= 0.5, Y:= 0.0}
                    Offsets:=margin{Top:=188.0,Left:=796.0, Right:=113.0, Bottom:=27.0}
                canvas_slot:
                    Widget:=CustomPlayer.SUPUpgradeTitle
                    ZOrder:=1
                    SizeToContent:=false
                    Alignment := vector2{X:= 0.5, Y:= 0.0}
                    Offsets:=margin{Top:=188.0,Left:=1220.0, Right:=153.0, Bottom:=34.0}
                canvas_slot:
                    Widget:=CustomPlayer.SUPUpgradeEffTitle
                    ZOrder:=1
                    SizeToContent:=false
                    Alignment := vector2{X:= 0.5, Y:= 0.0}
                    Offsets:=margin{Top:=188.0,Left:=1580.0, Right:=338.0, Bottom:=36.0}


                canvas_slot:
                    Widget:=CustomPlayer.SUP1Icon
                    ZOrder:=1
                    SizeToContent:=true
                    Alignment := vector2{X:= 0.0, Y:= 0.25}
                    Offsets:=margin{Top:=288.0,Left:=188.0,Bottom:= 67.0, Right:= 67.0}
                canvas_slot:
                    Widget:=CustomPlayer.SUP1Surge
                    ZOrder:=1
                    SizeToContent:=true
                    Offsets:=margin{Top:=312.0,Left:=240.0,Bottom:= 45.0, Right:= 45.0}
                canvas_slot:
                    Widget:=CustomPlayer.SUP1RomanNum
                    ZOrder:=1
                    SizeToContent:=false
                    Alignment := vector2{X:= 0.5, Y:= 0.0}
                    Offsets:=margin{Top:=288.0,Left:=324.0}
                canvas_slot:
                    #Widget:=CustomPlayer.SUP1Level
                    ZOrder:=1
                    SizeToContent:=true
                    Alignment := vector2{X:= 0.5, Y:= 0.0}
                    Offsets:=margin{Top:=288.0,Left:=488.0}
                    Widget := stack_box:
                        Orientation := orientation.Horizontal
                        Slots := array:
                            stack_box_slot:
                                Widget := CustomPlayer.SUP1Level
                            stack_box_slot:
                                Widget := CustomPlayer.SUP1LevelEff
                canvas_slot:
                    #Widget:=CustomPlayer.SUP1Bonus
                    ZOrder:=1
                    SizeToContent:=true
                    Alignment := vector2{X:= 0.5, Y:= 0.0}
                    Offsets:=margin{Top:=288.0,Left:=792.0}
                    Widget := stack_box:
                        Orientation := orientation.Horizontal
                        Slots := array:
                            stack_box_slot:
                                Widget := CustomPlayer.SUP1Bonus
                            stack_box_slot:
                                Widget := CustomPlayer.SUP1NextBonus
                canvas_slot:
                    Widget:=CustomPlayer.SUP1Upgrade
                    ZOrder:=1
                    SizeToContent:=false
                    Alignment := vector2{X:= 0.5, Y:= 0.0}
                    Offsets:=margin{Top:=280.0,Left:=1220.0, Bottom:= 50.0, Right:= 330.0}
                canvas_slot:
                    ZOrder:=1
                    SizeToContent:=true
                    Offsets:=margin{Top:=288.0,Left:=1220.0}
                    Alignment := vector2{X:= 0.5, Y:= 0.0}
                    Widget := stack_box:
                        Orientation := orientation.Horizontal
                        Slots := array:
                            stack_box_slot:
                                Widget := CustomPlayer.SUP1StoneIcon1
                            stack_box_slot:
                                Widget := CustomPlayer.SUP1UpgradeText
                canvas_slot:
                    Widget:=CustomPlayer.SUP1Efficiency
                    ZOrder:=1
                    SizeToContent:=false
                    Alignment := vector2{X:= 0.5, Y:= 0.0}
                    Offsets:=margin{Top:=280.0,Left:=1580.0, Bottom:= 50.0, Right:= 330.0}
                canvas_slot:
                    ZOrder:=1
                    SizeToContent:=true
                    Offsets:=margin{Top:=288.0,Left:=1580.0}
                    Alignment := vector2{X:= 0.5, Y:= 0.0}
                    Widget := stack_box:
                        Orientation := orientation.Horizontal
                        Slots := array:
                            stack_box_slot:
                                Widget := CustomPlayer.SUP1StoneIcon2
                            stack_box_slot:
                                Widget := CustomPlayer.SUP1EfficiencyText
                                
                canvas_slot:
                    Widget:=CustomPlayer.SUP2Icon
                    ZOrder:=1
                    SizeToContent:=true
                    Alignment := vector2{X:= 0.0, Y:= 0.25}
                    Offsets:=margin{Top:=436.0,Left:=188.0,Bottom:= 67.0, Right:= 67.0}
                canvas_slot:
                    Widget:=CustomPlayer.SUP2Surge
                    ZOrder:=1
                    SizeToContent:=true
                    Offsets:=margin{Top:=460.0,Left:=240.0,Bottom:= 45.0, Right:= 45.0}
                canvas_slot:
                    Widget:=CustomPlayer.SUP2RomanNum
                    ZOrder:=1
                    SizeToContent:=false
                    Alignment := vector2{X:= 0.5, Y:= 0.0}
                    Offsets:=margin{Top:=436.0,Left:=324.0}
                canvas_slot:
                    #Widget:=CustomPlayer.SUP2Level
                    ZOrder:=1
                    SizeToContent:=true
                    Alignment := vector2{X:= 0.5, Y:= 0.0}
                    Offsets:=margin{Top:=436.0,Left:=488.0}
                    Widget := stack_box:
                        Orientation := orientation.Horizontal
                        Slots := array:
                            stack_box_slot:
                                Widget := CustomPlayer.SUP2Level
                            stack_box_slot:
                                Widget := CustomPlayer.SUP2LevelEff
                canvas_slot:
                    #Widget:=CustomPlayer.SUP2Bonus
                    ZOrder:=1
                    SizeToContent:=true
                    Alignment := vector2{X:= 0.5, Y:= 0.0}
                    Offsets:=margin{Top:=436.0,Left:=792.0}
                    Widget := stack_box:
                        Orientation := orientation.Horizontal
                        Slots := array:
                            stack_box_slot:
                                Widget := CustomPlayer.SUP2Bonus
                            stack_box_slot:
                                Widget := CustomPlayer.SUP2NextBonus
                canvas_slot:
                    Widget:=CustomPlayer.SUP2Upgrade
                    ZOrder:=1
                    SizeToContent:=false
                    Alignment := vector2{X:= 0.5, Y:= 0.0}
                    Offsets:=margin{Top:=428.0,Left:=1220.0, Bottom:= 50.0, Right:= 330.0}
                canvas_slot:
                    ZOrder:=1
                    SizeToContent:=true
                    Offsets:=margin{Top:=436.0,Left:=1220.0}
                    Alignment := vector2{X:= 0.5, Y:= 0.0}
                    Widget := stack_box:
                        Orientation := orientation.Horizontal
                        Slots := array:
                            stack_box_slot:
                                Widget := CustomPlayer.SUP2StoneIcon1
                            stack_box_slot:
                                Widget := CustomPlayer.SUP2UpgradeText
                canvas_slot:
                    Widget:=CustomPlayer.SUP2Efficiency
                    ZOrder:=1
                    SizeToContent:=false
                    Alignment := vector2{X:= 0.5, Y:= 0.0}
                    Offsets:=margin{Top:=428.0,Left:=1580.0, Bottom:= 50.0, Right:= 330.0}
                canvas_slot:
                    ZOrder:=1
                    SizeToContent:=true
                    Offsets:=margin{Top:=436.0,Left:=1580.0}
                    Alignment := vector2{X:= 0.5, Y:= 0.0}
                    Widget := stack_box:
                        Orientation := orientation.Horizontal
                        Slots := array:
                            stack_box_slot:
                                Widget := CustomPlayer.SUP2StoneIcon2
                            stack_box_slot:
                                Widget := CustomPlayer.SUP2EfficiencyText
                                
                canvas_slot:
                    Widget:=CustomPlayer.SUP3Icon
                    ZOrder:=1
                    SizeToContent:=true
                    Alignment := vector2{X:= 0.0, Y:= 0.25}
                    Offsets:=margin{Top:=576.0,Left:=188.0,Bottom:= 67.0, Right:= 67.0}
                canvas_slot:
                    Widget:=CustomPlayer.SUP3Surge
                    ZOrder:=1
                    SizeToContent:=true
                    Offsets:=margin{Top:=608.0,Left:=240.0,Bottom:= 45.0, Right:= 45.0}
                canvas_slot:
                    Widget:=CustomPlayer.SUP3RomanNum
                    ZOrder:=1
                    SizeToContent:=false
                    Alignment := vector2{X:= 0.5, Y:= 0.0}
                    Offsets:=margin{Top:=576.0,Left:=324.0}
                canvas_slot:
                    #Widget:=CustomPlayer.SUP3Level
                    ZOrder:=1
                    SizeToContent:=true
                    Alignment := vector2{X:= 0.5, Y:= 0.0}
                    Offsets:=margin{Top:=576.0,Left:=488.0}
                    Widget := stack_box:
                        Orientation := orientation.Horizontal
                        Slots := array:
                            stack_box_slot:
                                Widget := CustomPlayer.SUP3Level
                            stack_box_slot:
                                Widget := CustomPlayer.SUP3LevelEff
                canvas_slot:
                    #Widget:=CustomPlayer.SUP3Bonus
                    ZOrder:=1
                    SizeToContent:=true
                    Alignment := vector2{X:= 0.5, Y:= 0.0}
                    Offsets:=margin{Top:=576.0,Left:=792.0}
                    Widget := stack_box:
                        Orientation := orientation.Horizontal
                        Slots := array:
                            stack_box_slot:
                                Widget := CustomPlayer.SUP3Bonus
                            stack_box_slot:
                                Widget := CustomPlayer.SUP3NextBonus
                canvas_slot:
                    Widget:=CustomPlayer.SUP3Upgrade
                    ZOrder:=1
                    SizeToContent:=false
                    Alignment := vector2{X:= 0.5, Y:= 0.0}
                    Offsets:=margin{Top:=568.0,Left:=1220.0, Bottom:= 50.0, Right:= 330.0}
                canvas_slot:
                    ZOrder:=1
                    SizeToContent:=true
                    Offsets:=margin{Top:=576.0,Left:=1220.0}
                    Alignment := vector2{X:= 0.5, Y:= 0.0}
                    Widget := stack_box:
                        Orientation := orientation.Horizontal
                        Slots := array:
                            stack_box_slot:
                                Widget := CustomPlayer.SUP3StoneIcon1
                            stack_box_slot:
                                Widget := CustomPlayer.SUP3UpgradeText
                canvas_slot:
                    Widget:=CustomPlayer.SUP3Efficiency
                    ZOrder:=1
                    SizeToContent:=false
                    Alignment := vector2{X:= 0.5, Y:= 0.0}
                    Offsets:=margin{Top:=568.0,Left:=1580.0, Bottom:= 50.0, Right:= 330.0}
                canvas_slot:
                    ZOrder:=1
                    SizeToContent:=true
                    Offsets:=margin{Top:=576.0,Left:=1580.0}
                    Alignment := vector2{X:= 0.5, Y:= 0.0}
                    Widget := stack_box:
                        Orientation := orientation.Horizontal
                        Slots := array:
                            stack_box_slot:
                                Widget := CustomPlayer.SUP3StoneIcon2
                            stack_box_slot:
                                Widget := CustomPlayer.SUP3EfficiencyText
                                
                canvas_slot:
                    Widget:=CustomPlayer.SUP4Icon
                    ZOrder:=1
                    SizeToContent:=true
                    Alignment := vector2{X:= 0.0, Y:= 0.25}
                    Offsets:=margin{Top:=720.0,Left:=188.0,Bottom:= 67.0, Right:= 67.0}
                canvas_slot:
                    Widget:=CustomPlayer.SUP4Surge
                    ZOrder:=1
                    SizeToContent:=true
                    Offsets:=margin{Top:=748.0,Left:=240.0,Bottom:= 45.0, Right:= 45.0}
                canvas_slot:
                    Widget:=CustomPlayer.SUP4RomanNum
                    ZOrder:=1
                    SizeToContent:=false
                    Alignment := vector2{X:= 0.5, Y:= 0.0}
                    Offsets:=margin{Top:=720.0,Left:=324.0}
                canvas_slot:
                    #Widget:=CustomPlayer.SUP4Level
                    ZOrder:=1
                    SizeToContent:=true
                    Alignment := vector2{X:= 0.5, Y:= 0.0}
                    Offsets:=margin{Top:=720.0,Left:=488.0}
                    Widget := stack_box:
                        Orientation := orientation.Horizontal
                        Slots := array:
                            stack_box_slot:
                                Widget := CustomPlayer.SUP4Level
                            stack_box_slot:
                                Widget := CustomPlayer.SUP4LevelEff
                canvas_slot:
                    #Widget:=CustomPlayer.SUP4Bonus
                    ZOrder:=1
                    SizeToContent:=true
                    Alignment := vector2{X:= 0.5, Y:= 0.0}
                    Offsets:=margin{Top:=720.0,Left:=792.0}
                    Widget := stack_box:
                        Orientation := orientation.Horizontal
                        Slots := array:
                            stack_box_slot:
                                Widget := CustomPlayer.SUP4Bonus
                            stack_box_slot:
                                Widget := CustomPlayer.SUP4NextBonus
                canvas_slot:
                    Widget:=CustomPlayer.SUP4Upgrade
                    ZOrder:=1
                    SizeToContent:=false
                    Alignment := vector2{X:= 0.5, Y:= 0.0}
                    Offsets:=margin{Top:=712.0,Left:=1220.0, Bottom:= 50.0, Right:= 330.0}
                canvas_slot:
                    ZOrder:=1
                    SizeToContent:=true
                    Offsets:=margin{Top:=720.0,Left:=1220.0}
                    Alignment := vector2{X:= 0.5, Y:= 0.0}
                    Widget := stack_box:
                        Orientation := orientation.Horizontal
                        Slots := array:
                            stack_box_slot:
                                Widget := CustomPlayer.SUP4StoneIcon1
                            stack_box_slot:
                                Widget := CustomPlayer.SUP4UpgradeText
                canvas_slot:
                    Widget:=CustomPlayer.SUP4Efficiency
                    ZOrder:=1
                    SizeToContent:=false
                    Alignment := vector2{X:= 0.5, Y:= 0.0}
                    Offsets:=margin{Top:=712.0,Left:=1580.0, Bottom:= 50.0, Right:= 330.0}
                canvas_slot:
                    ZOrder:=1
                    SizeToContent:=true
                    Offsets:=margin{Top:=720.0,Left:=1580.0}
                    Alignment := vector2{X:= 0.5, Y:= 0.0}
                    Widget := stack_box:
                        Orientation := orientation.Horizontal
                        Slots := array:
                            stack_box_slot:
                                Widget := CustomPlayer.SUP4StoneIcon2
                            stack_box_slot:
                                Widget := CustomPlayer.SUP4EfficiencyText
                                
                canvas_slot:
                    Widget:=CustomPlayer.SUP5Icon
                    ZOrder:=1
                    SizeToContent:=true
                    Alignment := vector2{X:= 0.0, Y:= 0.25}
                    Offsets:=margin{Top:=860.0,Left:=188.0,Bottom:= 67.0, Right:= 67.0}
                canvas_slot:
                    Widget:=CustomPlayer.SUP5Surge
                    ZOrder:=1
                    SizeToContent:=true
                    Offsets:=margin{Top:=892.0,Left:=240.0,Bottom:= 45.0, Right:= 45.0}
                canvas_slot:
                    Widget:=CustomPlayer.SUP5RomanNum
                    ZOrder:=1
                    SizeToContent:=false
                    Alignment := vector2{X:= 0.5, Y:= 0.0}
                    Offsets:=margin{Top:=860.0,Left:=324.0}
                canvas_slot:
                    #Widget:=CustomPlayer.SUP5Level
                    ZOrder:=1
                    SizeToContent:=true
                    Alignment := vector2{X:= 0.5, Y:= 0.0}
                    Offsets:=margin{Top:=860.0,Left:=488.0}
                    Widget := stack_box:
                        Orientation := orientation.Horizontal
                        Slots := array:
                            stack_box_slot:
                                Widget := CustomPlayer.SUP5Level
                            stack_box_slot:
                                Widget := CustomPlayer.SUP5LevelEff
                canvas_slot:
                    #Widget:=CustomPlayer.SUP5Bonus
                    ZOrder:=1
                    SizeToContent:=true
                    Alignment := vector2{X:= 0.5, Y:= 0.0}
                    Offsets:=margin{Top:=860.0,Left:=792.0}
                    Widget := stack_box:
                        Orientation := orientation.Horizontal
                        Slots := array:
                            stack_box_slot:
                                Widget := CustomPlayer.SUP5Bonus
                            stack_box_slot:
                                Widget := CustomPlayer.SUP5NextBonus
                canvas_slot:
                    Widget:=CustomPlayer.SUP5Upgrade
                    ZOrder:=1
                    SizeToContent:=false
                    Alignment := vector2{X:= 0.5, Y:= 0.0}
                    Offsets:=margin{Top:=852.0,Left:=1220.0, Bottom:= 50.0, Right:= 330.0}
                canvas_slot:
                    ZOrder:=1
                    SizeToContent:=true
                    Offsets:=margin{Top:=860.0,Left:=1220.0}
                    Alignment := vector2{X:= 0.5, Y:= 0.0}
                    Widget := stack_box:
                        Orientation := orientation.Horizontal
                        Slots := array:
                            stack_box_slot:
                                Widget := CustomPlayer.SUP5StoneIcon1
                            stack_box_slot:
                                Widget := CustomPlayer.SUP5UpgradeText
                canvas_slot:
                    Widget:=CustomPlayer.SUP5Efficiency
                    ZOrder:=1
                    SizeToContent:=false
                    Alignment := vector2{X:= 0.5, Y:= 0.0}
                    Offsets:=margin{Top:=852.0,Left:=1580.0, Bottom:= 50.0, Right:= 330.0}
                canvas_slot:
                    ZOrder:=1
                    SizeToContent:=true
                    Offsets:=margin{Top:=860.0,Left:=1580.0}
                    Alignment := vector2{X:= 0.5, Y:= 0.0}
                    Widget := stack_box:
                        Orientation := orientation.Horizontal
                        Slots := array:
                            stack_box_slot:
                                Widget := CustomPlayer.SUP5StoneIcon2
                            stack_box_slot:
                                Widget := CustomPlayer.SUP5EfficiencyText

                
                canvas_slot:
                    Widget:=CustomPlayer.SUPCloseButton
                    ZOrder:=1
                    SizeToContent:=false
                    Alignment := vector2{X:= 0.5, Y:= 0.0}
                    Offsets:=margin{Top:= 96.0,Left:= 1700.0,Bottom:= 56.0,Right:= 76.0}

        return MySUPCanvas
#CODE

    #CALL THIS AFTER ANY UPGRADE TO MAKE SURE THE HUD IS UP TO DATE
    UpdateSUPHUD(CustomPlayer: Custom_Player):void=
        var slimit : int = 35
        CustomPlayer.SUPAutoUpgrade.SetText(StringToMessage("AUTO"))
        CustomPlayer.SUPEnableUpgrade.SetText(StringToMessage("TOGGLE"))
        #CustomPlayer.SUPEnableUpgrade.SetEnabled(false)
        #CustomPlayer.SUPEnableUpgrade.SetVisibility(widget_visibility.Collapsed)
        #CustomPlayer.SUPEnableBG.SetVisibility(widget_visibility.Collapsed)

        if(CustomPlayer.PUPQOLLevels[0] = 2):
            #CustomPlayer.SUPEnableBG.SetVisibility(widget_visibility.Visible)
            #CustomPlayer.SUPEnableUpgrade.SetVisibility(widget_visibility.Visible)
            #CustomPlayer.SUPEnableUpgrade.SetEnabled(true)
            if(CustomPlayer.AutoUpgradeOnHit = true):
                CustomPlayer.SUPEnableBG.SetImage(UI.ButtonOn)
                CustomPlayer.SUPEnableUpgrade.SetText(StringToMessage("DISABLE"))
            else if(CustomPlayer.AutoUpgradeOnHit = false):
                CustomPlayer.SUPEnableBG.SetImage(UI.ButtonOff)
                CustomPlayer.SUPEnableUpgrade.SetText(StringToMessage("ENABLE"))
                
        if(CustomPlayer.PUPQOLLevels[0] >= 1):
            #CustomPlayer.SUPAutoUpgrade.SetEnabled(true)

        CustomPlayer.SUPTotalDMG.SetText(StringToMessage("{NumberCondensor(CustomPlayer.GetPlayerDamage())}"))
        CustomPlayer.SUPStoneAmountText.SetText(StringToMessage("{NumberCondensor(CustomPlayer.StoneAmount)}"))

        for(i:int=0..CustomPlayer.SUPLevels.Length):
            var s : int = i
            if:
                SUPSurge:=CustomPlayer.SUPSurges[i]
                SUPIcon:=CustomPlayer.SUPIcons[i]
                SUPRomanNum:=CustomPlayer.SUPRomanNums[i]
                SUPLevel := CustomPlayer.SUPLevels[i]
                SUPLevelEff := CustomPlayer.SUPLevelEffs[i]
                Bonus := CustomPlayer.SUPBonuses[i]
                NextBonus := CustomPlayer.SUPNextBonuses[i]
                Upgrade := CustomPlayer.SUPUpgradesTexts[i]
                Efficiency := CustomPlayer.SUPEfficienciesTexts[i]
                UpgradeStone:=CustomPlayer.SUPUpgradesStones[i]
                EffStone:=CustomPlayer.SUPEfficienciesStones[i]
            then:
                if(CustomPlayer.A4Unlocked = true):
                    set slimit = 75
                if(CustomPlayer.A5Unlocked = true):
                    set slimit = 105
                loop:
                    if:
                        WhichStrengthLevel:=CustomPlayer.StrengthUpgradeLevels[s]
                        WhichStrengthEfficientLvl:=CustomPlayer.StrengthEfficiencies[s]
                    then:
                        if(WhichStrengthLevel>=100 and WhichStrengthEfficientLvl>=7 and s < slimit):
                            set s += 5
                            Tutorial.UpdateTutorialUI(CustomPlayer, "MaxStr", 1.0)
                            if(NewSUPIcon:=SUPIconImgs[s]):
                                SUPIcon.SetImage(NewSUPIcon)
                        else:
                            break:
                    else:
                        break:
                if(CustomPlayer.PUPLevels[0]-1 >= s): #PRESTIGE POWER SURGE MULTIPLIER
                    SUPSurge.SetVisibility(widget_visibility.Visible)
                else:
                    SUPSurge.SetVisibility(widget_visibility.Collapsed)
                if:
                    StrengthBaseUpgrade:=CustomPlayer.StrengthBaseUpgrades[s]
                    StrengthLevel:=CustomPlayer.StrengthUpgradeLevels[s]
                    StrengthEfficientLvl:=CustomPlayer.StrengthEfficiencies[s]
                    StrengthCost:=CustomPlayer.BaseStrengthCosts[s]
                then:
                    SUPRomanNum.SetText(StringToMessage("{(s+1)}"))
                    SUPLevel.SetText(StringToMessage("{StrengthLevel}"))
                    SUPLevelEff.SetText(StringToMessage(" | {StrengthEfficientLvl}"))

                    if(CustomPlayer.PUPLevels[0]-1 >= s): #PRESTIGE POWER SURGE MULTIPLIER
                        set CustomPlayer.EfficiencyDMGBonus = Pow(CustomPlayer.SurgeMultiplier, (StrengthEfficientLvl-1)*1.0)
                    else:
                        set CustomPlayer.EfficiencyDMGBonus = Pow(2.0, (StrengthEfficientLvl-1)*1.0)
                    NextUpgrade := (StrengthBaseUpgrade * CustomPlayer.EfficiencyDMGBonus)
                    if(StrengthLevel = 0):
                        Bonus.SetText(StringToMessage("0"))
                    else:

                        if(CustomPlayer.PUPLevels[0]-1 >= s): #PRESTIGE POWER SURGE MULTIPLIER
                            StrengthUpgrade := (StrengthBaseUpgrade * StrengthLevel) * Pow(CustomPlayer.SurgeMultiplier, (StrengthEfficientLvl*1.0)-1.0)
                            Bonus.SetText(StringToMessage("{NumberCondensor(StrengthUpgrade)}"))
                        else:
                            StrengthUpgrade := (StrengthBaseUpgrade * StrengthLevel) * Pow(2.0, (StrengthEfficientLvl*1.0)-1.0)
                            Bonus.SetText(StringToMessage("{NumberCondensor(StrengthUpgrade)}"))
                            
                        
                    if(StrengthLevel < 100):
                        NextBonus.SetText(StringToMessage(" (+{NumberCondensor(NextUpgrade)})"))
                        
                    if(StrengthLevel = 100):
                        UpgradeStone.SetVisibility(widget_visibility.Collapsed)
                        Upgrade.SetText(StringToMessage("MAX LEVEL"))
                        Upgrade.SetTextColor(MakeColorFromHex("FFFFFF"))
                    else:
                        UpgradeStone.SetVisibility(widget_visibility.Visible)
                        LevelsLeft := (100 - StrengthLevel) * 1.0
                        MaxCost := StrengthCost * Pow(BaseStrengthCostMult, StrengthLevel*1.0) * Pow(BaseStrengthCostMult, LevelsLeft)
                        ActualCost := StrengthCost * Pow(BaseStrengthCostMult, StrengthLevel*1.0)
                        if(CustomPlayer.StoneAmount >= MaxCost):
                            Upgrade.SetText(StringToMessage(" {NumberCondensor(MaxCost)} (MAX)"))
                        else:
                            Upgrade.SetText(StringToMessage(" {NumberCondensor(ActualCost)}"))
                        if(CustomPlayer.StoneAmount >= ActualCost):
                            Upgrade.SetTextColor(MakeColorFromHex("00F500"))
                        else:
                            Upgrade.SetTextColor(MakeColorFromHex("F9101C"))
                            

                    var x : int = 0
                    if(StrengthEfficientLvl <= 6):
                        loop:
                            var ax : int = CustomPlayer.EfficiencyLevelRqs.Length - x
                            if(ax > 0):
                                if(LevelRq:=CustomPlayer.EfficiencyLevelRqs[ax-1]):
                                    if(StrengthLevel >= LevelRq and StrengthEfficientLvl = ax):
                                        NewCost := Pow(EfficiencyCostMult, (StrengthEfficientLvl)*1.0)
                                        EffStone.SetVisibility(widget_visibility.Visible)
                                        Efficiency.SetText(StringToMessage(" {NumberCondensor(StrengthCost*NewCost)}"))
                                        if(CustomPlayer.StoneAmount >= StrengthCost*NewCost):
                                            Efficiency.SetTextColor(MakeColorFromHex("00F500"))
                                        else:
                                            Efficiency.SetTextColor(MakeColorFromHex("F9101C"))
                                        break:
                                    else if(StrengthEfficientLvl < ax):
                                    else:
                                        EffStone.SetVisibility(widget_visibility.Collapsed)
                                        Efficiency.SetText(StringToMessage("Requires Lv. {LevelRq}"))
                                        Efficiency.SetTextColor(MakeColorFromHex("FFFFFF"))
                                        break:
                                set x += 1
                            else:
                                break:
                    else:
                        EffStone.SetVisibility(widget_visibility.Collapsed)
                        Efficiency.SetText(StringToMessage("MAX EFFICIENCY")) 
                        Efficiency.SetTextColor(MakeColorFromHex("FFFFFF"))
                    
                    if(StrengthEfficientLvl = 7):
                        EffStone.SetVisibility(widget_visibility.Collapsed)
                        Efficiency.SetText(StringToMessage("MAX EFFICIENCY")) 
                        Efficiency.SetTextColor(MakeColorFromHex("FFFFFF"))
                
    #TO UPGRADE PLAYERS DAMAGE
    OnUpgradeDamageButton(Wdata:widget_message, which:int):void=
        var WhichUpgrade : int = which
        var slimit : int = 35
        if(CustomPlayer := gm.PlayersMap[Wdata.Player]):
            if(CustomPlayer.A4Unlocked = true):
                set slimit = 75
            if(CustomPlayer.A5Unlocked = true):
                Print("AREA 5 UNLOCKED")
                set slimit = 105
            loop:
                if(WhichStrengthLevel:=CustomPlayer.StrengthUpgradeLevels[WhichUpgrade],WhichStrengthEfficientLvl:=CustomPlayer.StrengthEfficiencies[WhichUpgrade]):

                    if(WhichStrengthLevel>=100 and WhichStrengthEfficientLvl>=7 and WhichUpgrade < slimit):
                        set WhichUpgrade += 5
                    else:
                        break:
                else:
                    break:
            if:
                StrengthLevel:=CustomPlayer.StrengthUpgradeLevels[WhichUpgrade]
                StrengthEfficientLvl:=CustomPlayer.StrengthEfficiencies[WhichUpgrade]
                StrengthCost:=CustomPlayer.BaseStrengthCosts[WhichUpgrade]
            then:
                LevelsLeft := (100 - StrengthLevel) * 1.0
                MaxCost := StrengthCost * Pow(BaseStrengthCostMult, StrengthLevel*1.0) * Pow(BaseStrengthCostMult, LevelsLeft)
                ActualCost := StrengthCost * Pow(BaseStrengthCostMult, StrengthLevel*1.0)
                if(CustomPlayer.StoneAmount >= MaxCost and StrengthLevel < 100):
                    UpgradeDamage(CustomPlayer, WhichUpgrade, LevelsLeft, MaxCost)
                else if(CustomPlayer.StoneAmount >= ActualCost and StrengthLevel < 100):
                    UpgradeDamage(CustomPlayer, WhichUpgrade, 1.0, ActualCost)
                #SaveAllData(CustomPlayer)
                UpdateSUPHUD(CustomPlayer)

    UpgradeDamage(CustomPlayer:Custom_Player, WhichUpgrade:int, Levels:float, Cost:float):void=
        if:
            StrengthLevel:=CustomPlayer.StrengthUpgradeLevels[WhichUpgrade]
            StrengthEfficientLvl:=CustomPlayer.StrengthEfficiencies[WhichUpgrade]
            StrengthCost:=CustomPlayer.BaseStrengthCosts[WhichUpgrade]
        then:
            Tutorial.UpdateTutorialUI(CustomPlayer, "UpStr", Levels)

            CustomPlayer.Spend(Cost, "Stone", 0)

            AudioManager.LeveledUp.Play(CustomPlayer.Player)

            if(set CustomPlayer.StrengthUpgradeLevels[WhichUpgrade] += Floor[Levels+0.01]) {}

            UpdatePower(CustomPlayer)

    OnAutoUpgrade(Wdata:widget_message):void=
        if(CustomPlayer := gm.PlayersMap[Wdata.Player]):
            if(CustomPlayer.PUPQOLLevels[0] >= 1):
                spawn{AutoUpgrade(CustomPlayer)}
            else:
                gm.ShowMessage(CustomPlayer.Player, "Unlock Button In Quarry!", 1)
                AudioManager.Denied.Play(CustomPlayer.Player)
                
    
    OnEnableUpgrade(Wdata:widget_message):void=
        if(CustomPlayer := gm.PlayersMap[Wdata.Player]):
            if(CustomPlayer.PUPQOLLevels[0] = 2):
                if(CustomPlayer.AutoUpgradeOnHit = false):
                    CustomPlayer.SUPEnableBG.SetImage(UI.ButtonOn)
                    CustomPlayer.SUPEnableUpgrade.SetText(StringToMessage("DISABLE"))
                    set CustomPlayer.AutoUpgradeOnHit = true
                else:
                    CustomPlayer.SUPEnableBG.SetImage(UI.ButtonOff)
                    CustomPlayer.SUPEnableUpgrade.SetText(StringToMessage("ENABLE"))
                    set CustomPlayer.AutoUpgradeOnHit = false
            else:
                gm.ShowMessage(CustomPlayer.Player, "Unlock Button In Quarry!", 1)
                AudioManager.Denied.Play(CustomPlayer.Player)

    AutoUpgrade(CustomPlayer:Custom_Player)<suspends>:void=
        var StrTiers : [int]int = map{}
        var StrengthLevels : [int]int = map{}
        var StrengthEffLevels : [int]int = map{}
        
        var UpgradeCosts : [int]float = map{}
        var UpgradeEffCosts : [int]float= map{}

        var UpgradeAmounts : [int]float = map{}
        var UpgradeNextAmount : [int]float = map{}

        var UpgradeRatios : [int]float = map{}
        var EfficiencyRatios : [int]float = map{}

        var MostEfficientUpgrades : []float = array{0.0,0.0,0.0,0.0,0.0,0.0,0.0,0.0,0.0,0.0}
        if(CustomPlayer.AutoUpgrading = false):
            set CustomPlayer.AutoUpgrading = true
            for(i:int=0..CustomPlayer.SUPUpgrades.Length):
                if(SUPUpgrade:=CustomPlayer.SUPUpgrades[i],SUPEfficiency:=CustomPlayer.SUPEfficiencies[i]):
                    SUPUpgrade.SetEnabled(false)
                    SUPEfficiency.SetEnabled(false)
            loop:
                var LowestCost : float = 9.99e200
                for(i:int=0..CustomPlayer.SUPRomanNums.Length):
                    if:
                        SUPLevel := CustomPlayer.SUPRomanNums[i]
                    then:
                        StrTier := StringToInt(SUPLevel.GetText()) - 1
                        if(set StrTiers[i] = StrTier) {}
                        if:
                            StrengthLevel:=CustomPlayer.StrengthUpgradeLevels[StrTier]
                            StrengthEfficientLvl:=CustomPlayer.StrengthEfficiencies[StrTier]
                            StrengthCost:=CustomPlayer.BaseStrengthCosts[StrTier]
                            StrengthBaseUpgrade:=CustomPlayer.StrengthBaseUpgrades[StrTier]
                        then:
                            if(set StrengthLevels[StrTier] = StrengthLevel) {}
                            if(set StrengthEffLevels[StrTier] = StrengthEfficientLvl) {}
                            if(set UpgradeCosts[StrTier] = StrengthCost) {}
                            if(set UpgradeEffCosts[StrTier] = StrengthCost * Pow(EfficiencyCostMult, (StrengthEfficientLvl)*1.0)) {}
                            if(set UpgradeAmounts[StrTier] = StrengthCost * Pow(BaseStrengthCostMult, StrengthLevel*1.0)) {}

                            if(CustomPlayer.PUPLevels[0]-1 >= StrTier):
                                set CustomPlayer.EfficiencyDMGBonus = Pow(CustomPlayer.SurgeMultiplier, (StrengthEfficientLvl-1)*1.0)
                            else:
                                set CustomPlayer.EfficiencyDMGBonus = Pow(2.0, (StrengthEfficientLvl-1)*1.0)

                            if(set UpgradeNextAmount[i] = StrengthBaseUpgrade * CustomPlayer.EfficiencyDMGBonus) {}

                            if(StrengthLevel < 100):
                                if(set UpgradeRatios[StrTier] = StrengthCost / (StrengthBaseUpgrade * CustomPlayer.EfficiencyDMGBonus)) {}
                            else:
                                if(set UpgradeRatios[StrTier] = -1.0) {}

                            if(StrengthLevel >= CustomPlayer.EfficiencyLevelRqs[StrengthEfficientLvl-1] and StrengthEfficientLvl <= 6):
                                if(set EfficiencyRatios[StrTier] =  (StrengthCost * (Pow(EfficiencyCostMult, (StrengthEfficientLvl)*1.0))) / ((StrengthLevel * StrengthBaseUpgrade) * CustomPlayer.EfficiencyDMGBonus)) {}
                            else:
                                if(set EfficiencyRatios[StrTier] = -1.0) {}
                            if:
                                UpgradeRatio:= UpgradeRatios[StrTier]
                                EffRatio:= EfficiencyRatios[StrTier]
                            then:
                                if(set MostEfficientUpgrades[i] = UpgradeRatio) {}
                                if(set MostEfficientUpgrades[i+5] = EffRatio) {}

                set MostEfficientUpgrades = BubbleSort(MostEfficientUpgrades)

                for(i:int=0..CustomPlayer.SUPRomanNums.Length):
                    if(SUPLevel := CustomPlayer.SUPRomanNums[i]):
                        StrTier := StringToInt(SUPLevel.GetText())# - 1
                        if:
                            StrengthLevel:=StrengthLevels[StrTier]
                            StrengthCost:=UpgradeCosts[StrTier]
                            UpgradeEffCost:=UpgradeEffCosts[StrTier]
                            StrengthEfficientLvl:=StrengthEffLevels[StrTier]
                        then:
                            LevelsLeft := (100 - StrengthLevel) * 1.0
                            MaxCost := StrengthCost * Pow(BaseStrengthCostMult, StrengthLevel*1.0) * Pow(BaseStrengthCostMult, LevelsLeft)
                            ActualCost := StrengthCost * Pow(BaseStrengthCostMult, StrengthLevel*1.0)
                            if(StrengthLevel < 100):
                                if(MaxCost < LowestCost):
                                    set LowestCost = MaxCost
                                else if(ActualCost < LowestCost):
                                    set LowestCost = ActualCost

                            if(StrengthEfficientLvl <= 6):
                                if(UpgradeEffCost < LowestCost):
                                    set LowestCost = UpgradeEffCost

                Sleep(0.1)
                for(s:int=0..MostEfficientUpgrades.Length-1):
                    var hasPurchased : logic = false
                    if(MostEfficient:= MostEfficientUpgrades[s], MostEfficient > 0.0):
                        for(i:int=0..CustomPlayer.SUPRomanNums.Length):
                            if(SUPLevel := CustomPlayer.SUPRomanNums[i]):
                                StrTier := StringToInt(SUPLevel.GetText()) - 1
                                if:
                                    StrengthLevel:=StrengthLevels[StrTier]
                                    StrengthCost:=UpgradeCosts[StrTier]
                                    UpgradeRatio:=UpgradeRatios[StrTier]
                                
                                    StrengthEfficientLvl:=StrengthEffLevels[StrTier]
                                    UpgradeEffCost:=UpgradeEffCosts[StrTier]
                                    EfficiencyRatio:=EfficiencyRatios[StrTier]
                                    hasPurchased = false
                                then:
                                    if(UpgradeRatio = MostEfficient):
                                        LevelsLeft := (100 - StrengthLevel) * 1.0
                                        MaxCost := StrengthCost * Pow(BaseStrengthCostMult, StrengthLevel*1.0) * Pow(BaseStrengthCostMult, LevelsLeft)
                                        ActualCost := StrengthCost * Pow(BaseStrengthCostMult, StrengthLevel*1.0)
                                        if(CustomPlayer.StoneAmount >= MaxCost):
                                            if(StrengthLevel < 100):
                                                set hasPurchased = true
                                                UpgradeDamage(CustomPlayer, StrTier, LevelsLeft, MaxCost)
                                                if(set StrengthLevels[StrTier] += Floor[LevelsLeft]) {}

                                        else if(CustomPlayer.StoneAmount >= ActualCost):
                                            if(StrengthLevel < 100):
                                                set hasPurchased = true
                                                UpgradeDamage(CustomPlayer, StrTier, 1.0, ActualCost)
                                                if(set StrengthLevels[StrTier] += 1) {}

                                    else if(EfficiencyRatio = MostEfficient):
                                        if(CustomPlayer.StoneAmount >= UpgradeEffCost):
                                            set hasPurchased = true
                                            CustomPlayer.Spend(UpgradeEffCost, "Stone", 0)
                                            if(set CustomPlayer.StrengthEfficiencies[StrTier] += 1) {}
                                            AudioManager.LeveledUp.Play(CustomPlayer.Player)
                                            if(Accolade:=AudioManager.Accolades[7]): #AWARD PLAYER EXP
                                                Accolade.Award(CustomPlayer.Player)
                                            UpdatePower(CustomPlayer)

                UpdateSUPHUD(CustomPlayer)

                if(CustomPlayer.StoneAmount < LowestCost):
                    set CustomPlayer.AutoUpgrading = false
                    SaveAllData(CustomPlayer)     
                    for(i:int=0..CustomPlayer.SUPUpgrades.Length):
                        if(SUPUpgrade:=CustomPlayer.SUPUpgrades[i],SUPEfficiency:=CustomPlayer.SUPEfficiencies[i]):
                            SUPUpgrade.SetEnabled(true)
                            SUPEfficiency.SetEnabled(true)           
                    Print("YOU ARE POOR")
                    break:

    # TO UPGRADE EFFICIENCY (2x bonus)
    UpgradeEfficiency(Wdata:widget_message, which:int):void=
        var WhichUpgrade : int = which
        if(CustomPlayer := gm.PlayersMap[Wdata.Player]):
            loop:
                if(WhichStrengthLevel:=CustomPlayer.StrengthUpgradeLevels[WhichUpgrade],WhichStrengthEfficientLvl:=CustomPlayer.StrengthEfficiencies[WhichUpgrade]):
                    if(WhichStrengthLevel>=100 and WhichStrengthEfficientLvl>=7):
                        set WhichUpgrade += 5
                    else:
                        break:
                else:
                    break:
            if:
                StrengthLevel:=CustomPlayer.StrengthUpgradeLevels[WhichUpgrade]
                StrengthEfficientLvl:=CustomPlayer.StrengthEfficiencies[WhichUpgrade]
                StrengthCost:=CustomPlayer.BaseStrengthCosts[WhichUpgrade]
            then:
                if(StrengthLevel >= CustomPlayer.EfficiencyLevelRqs[StrengthEfficientLvl-1] and StrengthEfficientLvl <= 6):
                    NewCost := Pow(EfficiencyCostMult, (StrengthEfficientLvl)*1.0)
                    if(CustomPlayer.StoneAmount >= StrengthCost*NewCost):
                        Tutorial.UpdateTutorialUI(CustomPlayer, "UpStr", 1.0)
                        if(Accolade:=AudioManager.Accolades[7]): #AWARD PLAYER EXP
                            Accolade.Award(CustomPlayer.Player)
                        CustomPlayer.Spend(StrengthCost*NewCost, "Stone", 0)
                        AudioManager.LeveledUp.Play(CustomPlayer.Player)

                        if(set CustomPlayer.StrengthEfficiencies[WhichUpgrade] += 1) {}

                        UpdatePower(CustomPlayer)
            UpdateSUPHUD(CustomPlayer)

    UpdatePower(CustomPlayer:Custom_Player):void=
        for(i:int=0..CustomPlayer.StrengthUpgradeLevels.Length):
            if(CustomPlayer.StrengthUpgradeLevels[i] > 0):
                if(CustomPlayer.PUPLevels[0]-1 >= i): #PRESTIGE POWER SURGE MULTIPLIER
                    if(TempBuff:=(CustomPlayer.StrengthUpgradeLevels[i] * CustomPlayer.StrengthBaseUpgrades[i]) * Pow(CustomPlayer.SurgeMultiplier, (CustomPlayer.StrengthEfficiencies[i]*1.0)-1.0)):
                        set CustomPlayer.TempBuffCalc += TempBuff
                else:
                    if(TempBuff:=(CustomPlayer.StrengthUpgradeLevels[i] * CustomPlayer.StrengthBaseUpgrades[i]) * Pow(2.0, (CustomPlayer.StrengthEfficiencies[i]*1.0)-1.0)):
                        set CustomPlayer.TempBuffCalc += TempBuff 
        set CustomPlayer.Power = CustomPlayer.BaseDamage + CustomPlayer.TempBuffCalc
        CustomPlayer.SetAchValue(1, CustomPlayer.Power)
        set CustomPlayer.TempBuffCalc = 0.0

        
        if:
            CustomPlayer.StrengthUpgradeLevels[79] = 100
            CustomPlayer.OneTimeAnalytics["Reach80StrTiers"] = false
        then:
            if(set CustomPlayer.OneTimeAnalytics["Reach80StrTiers"] = true) {}
            if(Analytic:=AudioManager.OneTimeAnalytics[9]): #ANALYTIC
                Print("REACH 80 STRENGTH TIERS")
                Analytic.Submit(CustomPlayer.Player)

        if:
            CustomPlayer.PUPLevels[0] = 80
            CustomPlayer.OneTimeAnalytics["Reach80SurgeTiers"] = false
        then:
            if(set CustomPlayer.OneTimeAnalytics["Reach80SurgeTiers"] = true) {}
            if(Analytic:=AudioManager.OneTimeAnalytics[10]): #ANALYTIC
                Print("REACH 80 SURGE TIERS")
                Analytic.Submit(CustomPlayer.Player)

        SaveAllData(CustomPlayer)
        

    