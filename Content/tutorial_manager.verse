using { /Fortnite.com/Devices }
using { /Verse.org/Simulation }
using { /Verse.org/Colors }
using { /Verse.org/Verse }
using { /Verse.org/Random }
using { /Fortnite.com/Teams }
using { /UnrealEngine.com/Assets }
using { /Fortnite.com/Game }
using { /Fortnite.com/Characters }
using { /Fortnite.com/Playspaces }
using { /Fortnite.com/UI }
using { /UnrealEngine.com/Temporary/Diagnostics }
using { /UnrealEngine.com/Temporary/SpatialMath }
using { /UnrealEngine.com/Temporary/UI }
using { /Verse.org/Assets}

tutorial_manager := class(creative_device):

    @editable
    gm : game_manager = game_manager{}

    @editable
    AudioManager: audio_manager = audio_manager{}
    
    @editable
    Objectives : []objective_device = array{}

    @editable #DON'T USE UNLESS YOU HAVE TO FOR SOME REASON
    ObjectivePulses : []map_indicator_device = array{}

    @editable
    JumpStarts : []objective_device = array{}

    @editable
    OnTriggers : []trigger_device = array{}

    @editable
    OffTriggers : []trigger_device = array{}

    
    var WhichTutorials : []string = array{
        "HitMainRock",
        "UpStr",
        "UnlockRoom",
        "GrindRails",
        "Teleport",
        "AFKZone",
        "RockDestroyed",
        "OpenHut",
        "MovementSpeed",
        "UnlockWeapon",
        "MobElim",
        "HatchPet",
        "EquipPet",
        "Drill",
        "MaxStr",
        "Boss",
        "EvolveWeapon",
        "Prestige"
    }

    var TutorialTitles : []string = array{
        "Strike the Quarry Rock",
        "Strength Upgrade",
        "Unlock Caves",
        "Grind Rails",
        "Use the Teleporters",
        "Enter the AFK Zone",
        "Destroy Rocks",
        "Check Cave Huts",
        "Upgrade Movement Speed",
        "Unlock your Weapon",
        "Defeat Mobs",
        "Hatch an Egg",
        "Equip a pet",
        "Unlock the Drill Buff",
        "Max a Strength Tier",
        "Defeat a Boss",
        "Evolve your Weapon",
        "Prestige"
    }

    var TutorialDescs : []string = array{
        "Strike the Quarry Rock\nto gather stone",
        "Use Signal Remote (Fire)\nTo Upgrade Strength",
        "Run over to the Cave\nEntrance to begin\nyour journey!",
        "Use the Grind Rails\nto move around\nthe caves quicker",
        "Use Teleporters to\nget from the caves\nto Quarry quicker",
        "Enter the AFK Zone\nto the right of cave\nto get stone passively",
        "Destroy rocks to get\n even more stone",
        "Each cave area has\nits own hut\nand unique buffs",
        "Unlock Buffs such as\n movement speed in\n the Upgrades Area",
        "Unlock your Weapon in\nthe Upgrades Area\nin the Quarry",
        "Mobs drop Stone,\nEggs, Boss Keys,\nand Weapon Materials!",
        "Pets grant passive\nbuffs to your\nstats when equipped!",
        "Use Signal Remote\n(Target) to Equip\na Pet",
        "Head over to the \nDrill to unlock its\npowerful buff",
        "Max tier upgrades\nand efficiences to\nunlock the next tier",
        "Bosses are always\non stage 5\nof each area",
        "Evolve your Weapon\nfor more damage",
        "The more you prestige\nthe more shards\nyou will receive"
    }
    
    var TutorialAmounts : []float = array{15.0,10.0,1000.0,1.0,1.0,1.0,5.0,1.0,2.50e05,2.50e06,20.0,1.0,1.0,7.50e07,1.0,1.0,50.0,1.12e16}
    var Destroyed : int = 0

    OnBegin<override>()<suspends>:void=
        for(Spawner : gm.PlayerSpawners):     
            Spawner.SpawnedEvent.Subscribe(OnPlayerJoined)
        
    OnPlayerJoined(Agent:agent):void=
        spawn{ShowTutorial(Agent)}
            
    ShowTutorial(Agent:agent)<suspends>:void=
        Sleep(2.0)
        Print("Destroyed: {Destroyed}")
        if(CustomPlayer := gm.PlayersMap[player[Agent]]):
            if(CustomPlayer.TutorialLevel <= 17):
                if(JumpStart:=JumpStarts[Destroyed]):
                    for(i:=0..20):
                        JumpStart.ActivateObjectivePulse(Agent)
                        Sleep(0.1)
                        JumpStart.DeactivateObjectivePulse(Agent)
                    JumpStart.Destroy(Agent)
                    set Destroyed += 1
            if(MyUI:=CustomPlayer.UIMaps["Tutorial"]?):
            else:
                Print("SHOW TUTORIAL")
                CustomPlayer.TutorialTitleText.SetText(StringToMessage(""))
                CustomPlayer.TutorialDescText.SetText(StringToMessage(""))
                CustomPlayer.TutorialProgressText.SetText(StringToMessage(""))
                CustomPlayer.ProgressIcon.SetVisibility(widget_visibility.Collapsed)
                if(CustomPlayer.TutorialLevel <= 17): #TUTORIAL CHANGE THIS TO NEW MAX
                    OpenTutorialHUD(Agent)
#UI BLOCKS

#CANVAS
    MakeTutorialCanvas(CustomPlayer:Custom_Player):canvas=
        CustomPlayer.TutorialTitleText.SetShadowColor(CustomPlayer.AutoStoneText.DefaultShadowColor)
        CustomPlayer.TutorialTitleText.SetShadowOffset(CustomPlayer.AutoStoneText.DefaultShadowOffset)
        CustomPlayer.TutorialTitleText.SetShadowOpacity(1.0)
        CustomPlayer.TutorialDescText.SetShadowColor(CustomPlayer.AutoStoneText.DefaultShadowColor)
        CustomPlayer.TutorialDescText.SetShadowOffset(CustomPlayer.AutoStoneText.DefaultShadowOffset)
        CustomPlayer.TutorialDescText.SetShadowOpacity(1.0)
        CustomPlayer.TutorialProgressText.SetShadowColor(CustomPlayer.AutoStoneText.DefaultShadowColor)
        CustomPlayer.TutorialProgressText.SetShadowOffset(CustomPlayer.AutoStoneText.DefaultShadowOffset)
        CustomPlayer.TutorialProgressText.SetShadowOpacity(1.0)
        var MyTutorialCanvas: canvas = canvas:
            Slots := array:
                canvas_slot:
                    Widget:=CustomPlayer.TutorialHUDBG
                    ZOrder:=0
                    SizeToContent:=false
                    Anchors := anchors{Minimum := vector2{X := 1.0, Y := 0.0}, Maximum := vector2{X := 1.0, Y := 0.0}}
                    Alignment := vector2{X := 0.5, Y := 0.0}
                    Offsets:=margin{Top:=120.0,Left:=-264.0,Bottom:=300.0,Right:=512.0}

                canvas_slot:
                    Widget:=CustomPlayer.TutorialTitleText
                    ZOrder:=1
                    SizeToContent:=true
                    Anchors := anchors{Minimum := vector2{X := 1.0, Y := 0.0}, Maximum := vector2{X := 1.0, Y := 0.0}}
                    Offsets:=margin{Top:=156.0,Left:=-268.0}
                    Alignment := vector2{X:= 0.5, Y:= 0.0}
                canvas_slot:
                    Widget:=CustomPlayer.TutorialDescText
                    ZOrder:=1
                    SizeToContent:=true
                    Anchors := anchors{Minimum := vector2{X := 1.0, Y := 0.0}, Maximum := vector2{X := 1.0, Y := 0.0}}
                    Offsets:=margin{Top:=220.0,Left:=-472.0}
                    Alignment := vector2{X:= 0.0, Y:= 0.0}
                canvas_slot:
                    ZOrder:=1
                    SizeToContent:=true
                    Anchors := anchors{Minimum := vector2{X := 1.0, Y := 0.0}, Maximum := vector2{X := 1.0, Y := 0.0}}
                    Offsets:=margin{Top:=348.0,Left:=-268.0}
                    Alignment := vector2{X:= 0.5, Y:= 0.0}
                    Widget := stack_box:
                        Orientation := orientation.Horizontal
                        Slots := array:
                            stack_box_slot:
                                Widget := CustomPlayer.ProgressIcon
                            stack_box_slot:
                                Widget := CustomPlayer.TutorialProgressText

        return MyTutorialCanvas
        
#CODE
    OpenTutorialHUD(Agent:agent): void=
        Print("OPEN HUD")
        if(CustomPlayer := gm.PlayersMap[Agent]):
            Print("Get past CustomPlayer")
            if(MyUI:=CustomPlayer.UIMaps["Tutorial"]?):
            else:
                if (Objective:=OnTriggers[CustomPlayer.TutorialLevel]):
                    Objective.Trigger(CustomPlayer.Player)
                var MyTutorialCanvas : canvas = MakeTutorialCanvas(CustomPlayer)
                CustomPlayer.AddUI(MyTutorialCanvas)
                if(set CustomPlayer.UIMaps["Tutorial"] = option{MyTutorialCanvas}) {}
                UpdateTutorialUI(CustomPlayer, "", 0.0)

    RemoveTutorialHUD(CustomPlayer:Custom_Player):void=
        if(PlayerUI:=GetPlayerUI[CustomPlayer.Player]):
            if(MyUI:=CustomPlayer.UIMaps["Tutorial"]?):
                PlayerUI.RemoveWidget(MyUI)
                MyUI.SetVisibility(widget_visibility.Collapsed)
                if(set CustomPlayer.UIMaps["Tutorial"] = false){}

    UpdateTutorialUI(CustomPlayer: Custom_Player, WhichTutorial:string, Amount:float):void= 
        TL := CustomPlayer.TutorialLevel
        for(i:int=0..WhichTutorials.Length):
            if(WhichTutorials[i] = WhichTutorial and TL = i):
                set CustomPlayer.TutorialProgress += Amount

        if(TutorialTitle:=TutorialTitles[TL],TutorialDesc:=TutorialDescs[TL],TutorialAmount:=TutorialAmounts[TL]):
            if:
                (TL = 2 and CustomPlayer.StagesUnlocked >= 0) or
                (TL = 8 and CustomPlayer.MovementSpeedLevel > 0) or
                (TL = 9 and CustomPlayer.WeaponTier >= 1) or
                (TL = 13 and CustomPlayer.GoldenDrillUnlocked = true) or
                (TL = 15 and CustomPlayer.AllAchs[6][0] >= 1.0) or
                (TL = 16 and CustomPlayer.WeaponTier >= 2) or
                (TL = 17 and CustomPlayer.PrestigeLevel >= 1)
            then:
                spawn{TutorialComplete(CustomPlayer)}
            else:
                if(CustomPlayer.TutorialProgress < TutorialAmount):
                    CustomPlayer.TutorialTitleText.SetText(StringToMessage("{TutorialTitle}"))
                    CustomPlayer.TutorialDescText.SetText(StringToMessage("{TutorialDesc}"))
                    if(TL = 2 or TL = 8 or TL = 9 or TL = 11 or TL = 13 or TL = 16 or TL = 17):
                        if(TL = 11): #EGG AMOUNT
                            var totalEggAmount : float = 0.0
                            for(i:int=0..CustomPlayer.EggAmounts.Length):
                                if(CustomPlayer.EggAmounts[i] >= 1.0):
                                    if(set totalEggAmount += CustomPlayer.EggAmounts[i]) {}
                            CustomPlayer.ProgressIcon.SetImage(UI.Egg1)
                            CustomPlayer.TutorialProgressText.SetText(StringToMessage(" {NumberCondensor(totalEggAmount)}/{NumberCondensor(TutorialAmount)}"))
                            if(totalEggAmount >= TutorialAmount):
                                CustomPlayer.TutorialProgressText.SetTextColor(MakeColorFromHex("39FF39"))
                            else:
                                CustomPlayer.TutorialProgressText.SetTextColor(MakeColorFromHex("fa4d56"))
                        else if(TL = 17): #PRESTIGE TOTAL STONE
                            CustomPlayer.ProgressIcon.SetImage(UI.StoneIcon)
                            CustomPlayer.TutorialProgressText.SetText(StringToMessage(" {NumberCondensor(CustomPlayer.PrestigeStoneAmount)}/{NumberCondensor(TutorialAmount)}"))
                            if(CustomPlayer.PrestigeStoneAmount >= TutorialAmount):
                                CustomPlayer.TutorialProgressText.SetTextColor(MakeColorFromHex("39FF39"))
                            else:
                                CustomPlayer.TutorialProgressText.SetTextColor(MakeColorFromHex("fa4d56"))
                        else if(TL = 16): #WEP MAT
                            CustomPlayer.ProgressIcon.SetImage(UI.WepMat)
                            CustomPlayer.TutorialProgressText.SetText(StringToMessage(" {NumberCondensor(CustomPlayer.WepMatAmount)}/{NumberCondensor(TutorialAmount)}"))
                            if(CustomPlayer.WepMatAmount >= TutorialAmount):
                                CustomPlayer.TutorialProgressText.SetTextColor(MakeColorFromHex("39FF39"))
                            else:
                                CustomPlayer.TutorialProgressText.SetTextColor(MakeColorFromHex("fa4d56"))
                        else: #ALL OTHER STONE REQs
                            CustomPlayer.ProgressIcon.SetImage(UI.StoneIcon)
                            CustomPlayer.TutorialProgressText.SetText(StringToMessage(" {NumberCondensor(CustomPlayer.StoneAmount)}/{NumberCondensor(TutorialAmount)}"))
                            if(CustomPlayer.StoneAmount >= TutorialAmount):
                                CustomPlayer.TutorialProgressText.SetTextColor(MakeColorFromHex("39FF39"))
                            else:
                                CustomPlayer.TutorialProgressText.SetTextColor(MakeColorFromHex("fa4d56"))
                    else: #JUST COUNTS
                        CustomPlayer.ProgressIcon.SetVisibility(widget_visibility.Collapsed)
                        CustomPlayer.TutorialProgressText.SetText(StringToMessage("{NumberCondensor(CustomPlayer.TutorialProgress)}/{NumberCondensor(TutorialAmount)}"))
                        if(CustomPlayer.TutorialProgress >= TutorialAmount):
                            CustomPlayer.TutorialProgressText.SetTextColor(MakeColorFromHex("39FF39"))
                        else:
                            CustomPlayer.TutorialProgressText.SetTextColor(MakeColorFromHex("fa4d56"))
                    SaveAllData(CustomPlayer)
                else:
                    spawn{TutorialComplete(CustomPlayer)}


    TutorialComplete(CustomPlayer : Custom_Player)<suspends>:void= 
        if(TutorialTitle:=TutorialTitles[CustomPlayer.TutorialLevel],TutorialDesc:=TutorialDescs[CustomPlayer.TutorialLevel],TutorialAmount:=TutorialAmounts[CustomPlayer.TutorialLevel]):
            CustomPlayer.TutorialTitleText.SetText(StringToMessage("Tutorial Completed!"))
            CustomPlayer.TutorialDescText.SetText(StringToMessage("Congratulations on\ncompleting this tutorial!"))
            CustomPlayer.TutorialProgressText.SetText(StringToMessage(""))
            CustomPlayer.ProgressIcon.SetVisibility(widget_visibility.Collapsed)
            
        if (Objective:=OffTriggers[CustomPlayer.TutorialLevel]):
            Objective.Trigger(CustomPlayer.Player)
        set CustomPlayer.TutorialLevel += 1
        set CustomPlayer.TutorialProgress = 0.0
        SaveAllData(CustomPlayer)
        AudioManager.AchCompleted.Play(CustomPlayer.Player)
        Sleep(5.0)
        if (Objective:=OnTriggers[CustomPlayer.TutorialLevel]):
            Objective.Trigger(CustomPlayer.Player)
        if(CustomPlayer.TutorialLevel > 17):  #TUTORIAL CHANGE THIS TO NEW MAX
            RemoveTutorialHUD(CustomPlayer)
        else:
            UpdateTutorialUI(CustomPlayer, "", 0.0)