using { /Fortnite.com/Devices }
using { /Verse.org/Simulation }
using { /Verse.org/Colors }
using { /Verse.org/Verse }
using { /Verse.org/Random }
using { /Fortnite.com/Teams }
using { /UnrealEngine.com/Assets }
using { /Fortnite.com/Game }
using { /Fortnite.com/Characters }
using { /Fortnite.com/Playspaces }
using { /Fortnite.com/UI }
using { /UnrealEngine.com/Temporary/Diagnostics }
using { /UnrealEngine.com/Temporary/SpatialMath }
using { /UnrealEngine.com/Temporary/UI }
using { /Verse.org/Assets}
using { /Verse.org/Colors/NamedColors }

damage_numbers := class:
    
    MoveNewBox(CustomPlayer:Custom_Player, CurrentBlock:color_block, CurrentText:text_block, CurrentImg:texture_block, whichNum:int)<suspends>:void=
        var YIncrease : float = CurrentBlock.DefaultDesiredSize.Y
        var isDone : logic = true
        var OpacityFloat : float = 1.0
        var ImgX : float = CurrentImg.DefaultDesiredSize.X
        var ImgY : float = CurrentImg.DefaultDesiredSize.Y

        for(i:int=0..40):
            CurrentBlock.SetDesiredSize(vector2{X:=CurrentBlock.DefaultDesiredSize.X,Y:=YIncrease})
            if(Opacity:=type{_X:float where 0.000000 <= _X, _X <= 1.000000}[OpacityFloat]):
                CurrentImg.SetDesiredSize(vector2{X:=ImgX,Y:=ImgY})
                CurrentText.SetTextOpacity(Opacity)
                CurrentText.SetShadowOpacity(Opacity)
            set YIncrease -= 5.0
            set OpacityFloat -= 0.05
            set ImgX -= 1.0
            set ImgY -= 1.0
            Sleep(0.01)
        if:
            Player:=CustomPlayer.Player
            PlayerUI := GetPlayerUI[Player]
            MyUI:=CustomPlayer.DamageMap[whichNum]?
        then:
            MyUI.SetVisibility(widget_visibility.Collapsed)

        for(i:int=0..CustomPlayer.DamageMap.Length-1): #Keeping total until it needs to be reset
            if:
                Map :=CustomPlayer.DamageMap[i]?
            then:
                IsCollapsed := Map.GetVisibility()
                if:
                    IsCollapsed = widget_visibility.Visible
                then:
                    set isDone = false

        if:
            isDone = true
            TotalNum := CustomPlayer.UIMaps["Total"]?
        then:
            TotalNum.SetVisibility(widget_visibility.Collapsed)
            CustomPlayer.TextureTotal.SetVisibility(widget_visibility.Collapsed)
            CustomPlayer.TextBlockTotal.SetText(StringToMessage(""))
            set CustomPlayer.TotalAmountGained = 0.0

    UpdateDamageUI(CustomPlayer:Custom_Player, Damage:float, whichNum:int, isCrit:logic, CritCount:int, FinalStrike:logic, MobOrRock:string,Combo:float)<suspends>:void=
        var NewImage : string = MobOrRock
        if:
            NewBlockDMG:= CustomPlayer.DmgBlockDMGs[whichNum]
            NewColorBlockDMG:= CustomPlayer.DmgColorBlockDMGs[whichNum]
            DmgTexture:= CustomPlayer.DmgTextures[whichNum]
        then:
            if(MobOrRock = "Rock" or MobOrRock = "AFK"):
                if(FinalStrike = true):
                    set NewImage = "FinalStrikeRock"
                    NewBlockDMG.SetTextColor(MakeColorFromHex("E6894C"))
                else if(isCrit = true):
                    set NewImage = "CritRock"
                    NewBlockDMG.SetTextColor(MakeColorFromHex("fa4d56"))
                else if(CustomPlayer.DrillActive = true):
                    NewBlockDMG.SetTextColor(MakeColorFromHex("73ABFF"))
                else:
                    set NewImage = "StoneIcon"
                    NewBlockDMG.SetTextColor(NewBlockDMG.DefaultTextColor)

                #TOTAL GAINS
                set CustomPlayer.TotalAmountGained += Damage
                CustomPlayer.TextureTotal.SetVisibility(widget_visibility.Visible)
                CustomPlayer.TextBlockTotal.SetText(StringToMessage(" +{NumberCondensor(CustomPlayer.TotalAmountGained)}"))
            else:
                <#CustomPlayer.TextureTotal.SetVisibility(widget_visibility.Collapsed)
                CustomPlayer.TextBlockTotal.SetText(StringToMessage(""))
                set CustomPlayer.TotalAmountGained = 0.0#>
                NewBlockDMG.SetTextColor(NewBlockDMG.DefaultTextColor)

                #FOR SETTING IMAGE CORRECTLY
            if(NewImage <> CustomPlayer.CurrentImages[whichNum]):
                if(NewImage = "FinalStrikeRock"):
                    DmgTexture.SetImage(UI.FinalStrikeRock)
                    if(set CustomPlayer.CurrentImages[whichNum] = "FinalStrikeRock"){}
                if(NewImage = "CritRock"):
                    DmgTexture.SetImage(UI.CritRock)
                    if(set CustomPlayer.CurrentImages[whichNum] = "CritRock"){}
                if(NewImage = "StoneIcon"):
                    DmgTexture.SetImage(UI.StoneIcon)
                    if(set CustomPlayer.CurrentImages[whichNum] = "StoneIcon"){}
                if(NewImage = "WepMat"):
                    DmgTexture.SetImage(UI.WepMat)
                    if(set CustomPlayer.CurrentImages[whichNum] = NewImage){}
                if(NewImage = "BossKey"):
                    DmgTexture.SetImage(UI.BossKeyIcon)
                    if(set CustomPlayer.CurrentImages[whichNum] = NewImage){}
                if(NewImage = "Shard"):
                    DmgTexture.SetImage(UI.ShardIcon)
                    if(set CustomPlayer.CurrentImages[whichNum] = NewImage){}
                if(NewImage = "Egg0"):
                    DmgTexture.SetImage(UI.Egg1)
                    if(set CustomPlayer.CurrentImages[whichNum] = NewImage){}
                if(NewImage = "Egg1"):
                    DmgTexture.SetImage(UI.Egg2)
                    if(set CustomPlayer.CurrentImages[whichNum] = NewImage){}
                if(NewImage = "Egg2"):
                    DmgTexture.SetImage(UI.Egg3)
                    if(set CustomPlayer.CurrentImages[whichNum] = NewImage){}
                if(NewImage = "Egg3"):
                    DmgTexture.SetImage(UI.Egg4)
                    if(set CustomPlayer.CurrentImages[whichNum] = NewImage){}
                if(NewImage = "Egg4"):
                    DmgTexture.SetImage(UI.Egg5)
                    if(set CustomPlayer.CurrentImages[whichNum] = NewImage){}
                if(NewImage = "Egg5"):
                    DmgTexture.SetImage(UI.Egg6)
                    if(set CustomPlayer.CurrentImages[whichNum] = NewImage){}
                #Print("CURRENT IMAGE SET TO {NewImage}")

    
            #COMBO STRIKE 
            if(Combo >= 0.0):
                if(Combo >= 100.0):
                    CustomPlayer.TextBlockCombo.SetTextColor(MakeColorFromHex("fa4d56"))
                    CustomPlayer.TextBlockCombo.SetText(StringToMessage(" {NumberCondensor(CustomPlayer.CurrentComboStrikeStreak)}/{NumberCondensor(CustomPlayer.ComboStrikeStreak)}"))
                else:
                    CustomPlayer.TextBlockCombo.SetTextColor(MakeColorFromHex("FFFFFF"))
                    CustomPlayer.TextBlockCombo.SetText(StringToMessage(" {NumberCondensor(Combo)}/100"))
                CustomPlayer.TextureCombo.SetVisibility(widget_visibility.Visible)
            else:
                CustomPlayer.TextBlockCombo.SetText(StringToMessage(""))
                CustomPlayer.TextureCombo.SetVisibility(widget_visibility.Collapsed)
                

            #CRIT MULTIPLIER
            if(isCrit = true):
                NewBlockDMG.SetText(StringToMessage(" +{NumberCondensor(Damage)} (x{CritCount})"))
            else:
                NewBlockDMG.SetText(StringToMessage(" +{NumberCondensor(Damage)}"))

            #ADD SHADOWS
            NewBlockDMG.SetShadowColor(NewBlockDMG.DefaultShadowColor)
            NewBlockDMG.SetShadowOffset(NewBlockDMG.DefaultShadowOffset)
            NewBlockDMG.SetShadowOpacity(1.0)
            CustomPlayer.TextBlockCombo.SetShadowColor(NewBlockDMG.DefaultShadowColor)
            CustomPlayer.TextBlockCombo.SetShadowOffset(NewBlockDMG.DefaultShadowOffset)
            CustomPlayer.TextBlockCombo.SetShadowOpacity(1.0)
            CustomPlayer.TextBlockTotal.SetShadowColor(NewBlockDMG.DefaultShadowColor)
            CustomPlayer.TextBlockTotal.SetShadowOffset(NewBlockDMG.DefaultShadowOffset)
            CustomPlayer.TextBlockTotal.SetShadowOpacity(1.0)

            spawn{MoveNewBox(CustomPlayer, NewColorBlockDMG, NewBlockDMG, DmgTexture, whichNum)}

    ShowDamageUIForPlayer(CustomPlayer:Custom_Player,Damage:float, isCrit:logic, CritCount:int, FinalStrike:logic, MobOrRock:string, Combo:float):void=
        if:
            Player:=CustomPlayer.Player
            PlayerUI := GetPlayerUI[Player]
            CustomPlayer.ShowNumbers = true
        then:
            var NumPushed : logic = false
            if(MobOrRock <> "Missile" and MobOrRock <> "Mob" and MobOrRock <> "Earthquake" and MobOrRock <> "Void"):
                for(i:int=0..7): #CustomPlayer.DamageMap.Length+1):
                    if(NumPushed = false):
                        if(DmgNum:=CustomPlayer.DamageMap[i]?):
                            IsCollapsed := DmgNum.GetVisibility()
                            if:
                                IsCollapsed = widget_visibility.Collapsed
                                NewBlockDMG:= CustomPlayer.DmgBlockDMGs[i]
                                NewColorBlockDMG:= CustomPlayer.DmgColorBlockDMGs[i]
                                DmgTexture:= CustomPlayer.DmgTextures[i]
                            then:
                                if(TotalNum := CustomPlayer.UIMaps["Total"]?):
                                    TotalNum.SetVisibility(widget_visibility.Visible)
                                else:
                                    #Print("NEW TOTAL AFTER DMGNUMS")
                                    MyTotalUI := CreateTotalUI(CustomPlayer, Damage, isCrit, CritCount, MobOrRock, Combo)
                                    PlayerUI.AddWidget(MyTotalUI)
                                    if (set CustomPlayer.UIMaps["Total"] = option{MyTotalUI}):
                                set NumPushed = true
                                DmgNum.SetVisibility(widget_visibility.Visible)
                                NewColorBlockDMG.SetDesiredSize(vector2{X:=NewColorBlockDMG.DefaultDesiredSize.X,Y:=NewColorBlockDMG.DefaultDesiredSize.Y})
                                spawn{UpdateDamageUI(CustomPlayer,Damage,i,isCrit,CritCount,FinalStrike,MobOrRock,Combo)}
                        else:
                            set NumPushed = true
                            MyUI := CreateDamageUI(CustomPlayer, Damage, i, isCrit, CritCount, FinalStrike, MobOrRock, Combo)
                            PlayerUI.AddWidget(MyUI)
                            if (set CustomPlayer.DamageMap[i] = option{MyUI}) {}
                            if(MobOrRock = "Rock" or MobOrRock = "AFK"):
                                MyTotalUI := CreateTotalUI(CustomPlayer, Damage, isCrit, CritCount, MobOrRock, Combo)
                                PlayerUI.AddWidget(MyTotalUI)
                                if (set CustomPlayer.UIMaps["Total"] = option{MyTotalUI}):
                            spawn{UpdateDamageUI(CustomPlayer,Damage,i,isCrit,CritCount,FinalStrike,MobOrRock,Combo)}
            else:
                if(MobOrRock = "Earthquake"):
                    if(MaybeUI:=CustomPlayer.UIMaps["Earthquake"]?):
                        MaybeUI.SetVisibility(widget_visibility.Visible)
                        CustomPlayer.TextureEQ.SetVisibility(widget_visibility.Visible)
                        set CustomPlayer.EQTotalAmountGained += Damage
                        CustomPlayer.TextBlockEQ.SetText(StringToMessage(" +{NumberCondensor(CustomPlayer.EQTotalAmountGained)}"))
                    else:
                        EarthquakeUI := CreateEarthquakeUI(CustomPlayer, Damage, isCrit, CritCount, MobOrRock, Combo)
                        PlayerUI.AddWidget(EarthquakeUI)
                        if (set CustomPlayer.UIMaps["Earthquake"] = option{EarthquakeUI}){}
                    spawn{MaybeRemoveUI(CustomPlayer, CustomPlayer.EQTotalAmountGained, "Earthquake")}

                else if(MobOrRock = "Void"):
                    if(MaybeUI:=CustomPlayer.UIMaps["Void"]?):
                        MaybeUI.SetVisibility(widget_visibility.Visible)
                        CustomPlayer.TextureVoid.SetVisibility(widget_visibility.Visible)
                        set CustomPlayer.VoidTotalAmountGained += Damage
                        CustomPlayer.TextBlockVoid.SetText(StringToMessage(" +{NumberCondensor(CustomPlayer.VoidTotalAmountGained)}"))
                    else:
                        VoidUI := CreateVoidUI(CustomPlayer, Damage, isCrit, CritCount, MobOrRock, Combo)
                        PlayerUI.AddWidget(VoidUI)
                        if (set CustomPlayer.UIMaps["Void"] = option{VoidUI}){}
                    spawn{MaybeRemoveUI(CustomPlayer, CustomPlayer.VoidTotalAmountGained, "Void")}
                else if(MobOrRock = "Missile"):
                    if(MaybeUI:=CustomPlayer.UIMaps["Missile"]?):
                        MaybeUI.SetVisibility(widget_visibility.Visible)
                        CustomPlayer.TextureMissile.SetVisibility(widget_visibility.Visible)
                        set CustomPlayer.MissileTotalAmountGained += Damage
                        CustomPlayer.TextBlockMissile.SetText(StringToMessage(" +{NumberCondensor(CustomPlayer.MissileTotalAmountGained)}"))
                    else:
                        MissileUI := CreateMissileUI(CustomPlayer, Damage)
                        PlayerUI.AddWidget(MissileUI)
                        if (set CustomPlayer.UIMaps["Missile"] = option{MissileUI}){}
                    spawn{MaybeRemoveUI(CustomPlayer, CustomPlayer.MissileTotalAmountGained, "Missile")}
                else:
                    if(MaybeUI:=CustomPlayer.UIMaps["Mob"]?):
                        MaybeUI.SetVisibility(widget_visibility.Visible)
                        set CustomPlayer.WepTotalAmountGained += Damage
                        CustomPlayer.WepTextBlockTotal.SetText(StringToMessage(" {NumberCondensor(CustomPlayer.WepTotalAmountGained)}"))
                    else:
                        WepUI := CreateWepUI(CustomPlayer, Damage, isCrit, CritCount, MobOrRock, Combo)
                        PlayerUI.AddWidget(WepUI)
                        if (set CustomPlayer.UIMaps["Mob"] = option{WepUI}){}
                    spawn{MaybeRemoveUI(CustomPlayer, CustomPlayer.WepTotalAmountGained, "")}


    MaybeRemoveUI(CustomPlayer:Custom_Player, OldAmount:float, Type:string)<suspends>:void=
        Sleep(1.0)
        if(Type = "Earthquake"):
            if(OldAmount = CustomPlayer.EQTotalAmountGained):
                if(MyUI:=CustomPlayer.UIMaps["Earthquake"]?):
                    set CustomPlayer.EQTotalAmountGained = 0.0
                    MyUI.SetVisibility(widget_visibility.Collapsed)
        else if(Type = "Void"):
            if(OldAmount = CustomPlayer.VoidTotalAmountGained):
                if(MyUI:=CustomPlayer.UIMaps["Void"]?):
                    set CustomPlayer.VoidTotalAmountGained = 0.0
                    MyUI.SetVisibility(widget_visibility.Collapsed)
        else if(Type = "Missile"):
            if(OldAmount = CustomPlayer.MissileTotalAmountGained):
                if(MyUI:=CustomPlayer.UIMaps["Missile"]?):
                    set CustomPlayer.MissileTotalAmountGained = 0.0
                    MyUI.SetVisibility(widget_visibility.Collapsed)
        else:
            if(OldAmount = CustomPlayer.WepTotalAmountGained):
                if(MyUI:=CustomPlayer.UIMaps["Mob"]?):
                    set CustomPlayer.WepTotalAmountGained = 0.0
                    MyUI.SetVisibility(widget_visibility.Collapsed)


    CreateDamageUI<private>(CustomPlayer:Custom_Player, Damage:float, whichNum:int, isCrit:logic, CritCount:int, FinalStrike:logic, MobOrRock:string,Combo:float) : canvas=
        Default : canvas = canvas:
            Slots := array:
                canvas_slot:
                    Anchors := anchors{Minimum := vector2{X := 0.55, Y := 0.4}, Maximum := vector2{X := 0.55, Y := 0.4}}
                    Offsets := margin{Top := 0.0, Left := 0.0, Right := 0.0, Bottom := 0.0}
                    Alignment := vector2{X := 0.0, Y := 0.0}
                    Widget := stack_box:
                        Orientation := orientation.Horizontal
                        Slots := array:
                            stack_box_slot:
                                Widget := CustomPlayer.DmgColorBlockDMG1    
                            stack_box_slot:
                                Widget := CustomPlayer.DmgTexture1     
                            stack_box_slot:
                                Widget := CustomPlayer.DmgBlockDMG1 
        if:
            NewBlockDMG:= CustomPlayer.DmgBlockDMGs[whichNum]
            NewColorBlockDMG:= CustomPlayer.DmgColorBlockDMGs[whichNum]
            DmgTexture:= CustomPlayer.DmgTextures[whichNum]
        then:
            TheCanvas : canvas = canvas:
                Slots := array:
                    canvas_slot:
                        Anchors := anchors{Minimum := vector2{X := 0.55, Y := 0.4}, Maximum := vector2{X := 0.55, Y := 0.4}}
                        Offsets := margin{Top := 0.0, Left := 0.0, Right := 0.0, Bottom := 0.0}
                        Alignment := vector2{X := 0.0, Y := 0.0}
                        Widget := stack_box:
                            Orientation := orientation.Horizontal
                            Slots := array:
                                stack_box_slot:
                                    Widget := NewColorBlockDMG    
                                stack_box_slot:
                                    Widget := DmgTexture     
                                stack_box_slot:
                                    Widget := NewBlockDMG 
            return TheCanvas
        else:
            Print("SENDING DEFAULT??")
            return Default

    CreateTotalUI<private>(CustomPlayer:Custom_Player, Damage:float, isCrit:logic, CritCount:int, MobOrRock:string, Combo:float) : canvas=
        TheCanvas : canvas = canvas:
            Slots := array:  
                canvas_slot:
                    Anchors := anchors{Minimum := vector2{X := 0.55, Y := 0.4}, Maximum := vector2{X := 0.55, Y := 0.4}}
                    Offsets := margin{Top := 150.0, Left := 0.0, Right := 0.0, Bottom := 0.0}
                    Alignment := vector2{X := 0.0, Y := 0.0}
                    Widget := stack_box:
                        Orientation := orientation.Horizontal
                        Slots := array:
                            stack_box_slot:
                                Widget := CustomPlayer.TextureTotal     
                            stack_box_slot:
                                Widget := CustomPlayer.TextBlockTotal
                canvas_slot:
                    Anchors := anchors{Minimum := vector2{X := 0.55, Y := 0.4}, Maximum := vector2{X := 0.55, Y := 0.4}}
                    Offsets := margin{Top := 210.0, Left := 0.0, Right := 0.0, Bottom := 0.0}
                    Alignment := vector2{X := 0.0, Y := 0.0}
                    Widget := stack_box:
                        Orientation := orientation.Horizontal
                        Slots := array:
                            stack_box_slot:
                                Widget := CustomPlayer.TextureCombo     
                            stack_box_slot:
                                Widget := CustomPlayer.TextBlockCombo

        return TheCanvas
    CreateWepUI<private>(CustomPlayer:Custom_Player, Damage:float, isCrit:logic, CritCount:int, MobOrRock:string, Combo:float) : canvas=
        #NewBlockDMG := text_block{DefaultTextColor := MakeColorFromHex("FFFFFF"),DefaultShadowOffset := option{vector2{X:=3.0, Y:=3.0}},DefaultShadowColor := MakeColorFromHex("000000")}
        CustomPlayer.TextBlockCombo.SetText(StringToMessage(""))
        CustomPlayer.TextureCombo.SetVisibility(widget_visibility.Collapsed)
        CustomPlayer.TextBlockEQ.SetText(StringToMessage(""))
        CustomPlayer.TextureEQ.SetVisibility(widget_visibility.Collapsed)
        CustomPlayer.TextBlockVoid.SetText(StringToMessage(""))
        CustomPlayer.TextureVoid.SetVisibility(widget_visibility.Collapsed)

        set CustomPlayer.WepTotalAmountGained += Damage
        CustomPlayer.WepTextBlockTotal.SetText(StringToMessage(" {NumberCondensor(CustomPlayer.WepTotalAmountGained)}"))
        
        spawn{MaybeRemoveUI(CustomPlayer, CustomPlayer.WepTotalAmountGained, "")}

        CustomPlayer.WepTextBlockTotal.SetShadowColor(CustomPlayer.ShadowDefaults.DefaultShadowColor)
        CustomPlayer.WepTextBlockTotal.SetShadowOffset(CustomPlayer.ShadowDefaults.DefaultShadowOffset)
        CustomPlayer.WepTextBlockTotal.SetShadowOpacity(1.0)

        TheCanvas : canvas = canvas:
            Slots := array:  
                canvas_slot:
                    Anchors := anchors{Minimum := vector2{X := 0.55, Y := 0.4}, Maximum := vector2{X := 0.55, Y := 0.4}}
                    Offsets := margin{Top := 150.0, Left := 0.0, Right := 0.0, Bottom := 0.0}
                    Alignment := vector2{X := 0.0, Y := 0.0}
                    Widget := stack_box:
                        Orientation := orientation.Horizontal
                        Slots := array:
                            stack_box_slot:
                                Widget := CustomPlayer.WepTextureTotal     
                            stack_box_slot:
                                Widget := CustomPlayer.WepTextBlockTotal

        return TheCanvas

    CreateEarthquakeUI<private>(CustomPlayer:Custom_Player, Damage:float, isCrit:logic, CritCount:int, MobOrRock:string, Combo:float) : canvas=
        #NewBlockDMG := text_block{DefaultTextColor := MakeColorFromHex("FFFFFF"),DefaultShadowOffset := option{vector2{X:=3.0, Y:=3.0}},DefaultShadowColor := MakeColorFromHex("000000")}
        set CustomPlayer.EQTotalAmountGained += Damage
        CustomPlayer.TextBlockEQ.SetTextColor(MakeColorFromHex("cd7f57"))
        CustomPlayer.TextBlockEQ.SetText(StringToMessage(" +{NumberCondensor(CustomPlayer.EQTotalAmountGained)}"))
        CustomPlayer.TextureEQ.SetVisibility(widget_visibility.Visible)

        CustomPlayer.TextBlockEQ.SetShadowColor(CustomPlayer.ShadowDefaults.DefaultShadowColor)
        CustomPlayer.TextBlockEQ.SetShadowOffset(CustomPlayer.ShadowDefaults.DefaultShadowOffset)
        CustomPlayer.TextBlockEQ.SetShadowOpacity(1.0)

        TheCanvas : canvas = canvas:
            Slots := array:  
                canvas_slot:
                    Anchors := anchors{Minimum := vector2{X := 0.55, Y := 0.4}, Maximum := vector2{X := 0.55, Y := 0.4}}
                    Offsets := margin{Top := 270.0, Left := 0.0, Right := 0.0, Bottom := 0.0}
                    Alignment := vector2{X := 0.0, Y := 0.0}
                    Widget := stack_box:
                        Orientation := orientation.Horizontal
                        Slots := array:
                            stack_box_slot:
                                Widget := CustomPlayer.TextureEQ
                            stack_box_slot:
                                Widget := CustomPlayer.TextBlockEQ

        return TheCanvas

    CreateVoidUI<private>(CustomPlayer:Custom_Player, Damage:float, isCrit:logic, CritCount:int, MobOrRock:string, Combo:float) : canvas=
        #NewBlockDMG := text_block{DefaultTextColor := MakeColorFromHex("FFFFFF"),DefaultShadowOffset := option{vector2{X:=3.0, Y:=3.0}},DefaultShadowColor := MakeColorFromHex("000000")}
        set CustomPlayer.VoidTotalAmountGained += Damage
        CustomPlayer.TextBlockVoid.SetTextColor(MakeColorFromHex("8e5ff9"))
        CustomPlayer.TextBlockVoid.SetText(StringToMessage(" +{NumberCondensor(CustomPlayer.VoidTotalAmountGained)}"))
        CustomPlayer.TextureVoid.SetVisibility(widget_visibility.Visible)

        CustomPlayer.TextBlockVoid.SetShadowColor(CustomPlayer.ShadowDefaults.DefaultShadowColor)
        CustomPlayer.TextBlockVoid.SetShadowOffset(CustomPlayer.ShadowDefaults.DefaultShadowOffset)
        CustomPlayer.TextBlockVoid.SetShadowOpacity(1.0)
        TheCanvas : canvas = canvas:
            Slots := array:  
                canvas_slot:
                    Anchors := anchors{Minimum := vector2{X := 0.55, Y := 0.4}, Maximum := vector2{X := 0.55, Y := 0.4}}
                    Offsets := margin{Top := 330.0, Left := 0.0, Right := 0.0, Bottom := 0.0}
                    Alignment := vector2{X := 0.0, Y := 0.0}
                    Widget := stack_box:
                        Orientation := orientation.Horizontal
                        Slots := array:
                            stack_box_slot:
                                Widget := CustomPlayer.TextureVoid
                            stack_box_slot:
                                Widget := CustomPlayer.TextBlockVoid
        return TheCanvas

    CreateMissileUI<private>(CustomPlayer:Custom_Player, Damage:float) : canvas=
        #NewBlockDMG := text_block{DefaultTextColor := MakeColorFromHex("FFFFFF"),DefaultShadowOffset := option{vector2{X:=3.0, Y:=3.0}},DefaultShadowColor := MakeColorFromHex("000000")}
        set CustomPlayer.MissileTotalAmountGained += Damage
        CustomPlayer.TextBlockMissile.SetTextColor(MakeColorFromHex("FFF873"))
        CustomPlayer.TextBlockMissile.SetText(StringToMessage(" +{NumberCondensor(CustomPlayer.MissileTotalAmountGained)}"))
        CustomPlayer.TextureMissile.SetVisibility(widget_visibility.Visible)

        CustomPlayer.TextBlockMissile.SetShadowColor(CustomPlayer.ShadowDefaults.DefaultShadowColor)
        CustomPlayer.TextBlockMissile.SetShadowOffset(CustomPlayer.ShadowDefaults.DefaultShadowOffset)
        CustomPlayer.TextBlockMissile.SetShadowOpacity(1.0)
        TheCanvas : canvas = canvas:
            Slots := array:  
                canvas_slot:
                    Anchors := anchors{Minimum := vector2{X := 0.55, Y := 0.4}, Maximum := vector2{X := 0.55, Y := 0.4}}
                    Offsets := margin{Top := 210.0, Left := 0.0, Right := 0.0, Bottom := 0.0}
                    Alignment := vector2{X := 0.0, Y := 0.0}
                    Widget := stack_box:
                        Orientation := orientation.Horizontal
                        Slots := array:
                            stack_box_slot:
                                Widget := CustomPlayer.TextureMissile    
                            stack_box_slot:
                                Widget := CustomPlayer.TextBlockMissile 
        return TheCanvas
