using { /Fortnite.com/Devices }
using { /Verse.org/Simulation }
using { /Verse.org/Colors }
using { /Verse.org/Verse }
using { /Verse.org/Random }
using { /Fortnite.com/Teams }
using { /UnrealEngine.com/Assets }
using { /Fortnite.com/Game }
using { /Fortnite.com/Characters }
using { /Fortnite.com/Playspaces }
using { /Fortnite.com/UI }
using { /UnrealEngine.com/Temporary/Diagnostics }
using { /UnrealEngine.com/Temporary/SpatialMath }
using { /UnrealEngine.com/Temporary/UI }
using { /Verse.org/Assets}

ui_hut_upgrades := class(creative_device):

    @editable
    gm : game_manager = game_manager{}

    @editable
    AudioManager : audio_manager = audio_manager{}

    @editable
    Tutorial : tutorial_manager = tutorial_manager{}
    
    @editable
    HutUpgradeButtons : []button_device = array{}

    var AllHutUpgradeAmounts : [][]float = array{}
    var Hut1UpgradeAmounts : []float = array{0.05, 0.10, 0.10, 1.0}
    var Hut2UpgradeAmounts : []float = array{1.0, 2.0, 1.0, 1.0}
    var Hut3UpgradeAmounts : []float = array{1.0, 100.0, 0.10, 0.05}
    var Hut4UpgradeAmounts : []float = array{1.0, 2.0, 0.02, 1.0}
    var Hut5UpgradeAmounts : []float = array{1.0, 0.10, 0.1, 2.0}

    var AllHutMaxLevels : [][]int = array{}
    var Hut1MaxLevels : []int = array{10, 10, 10, 5}
    var Hut2MaxLevels : []int = array{1, 5, 9, 4}
    var Hut3MaxLevels : []int = array{1, 10, 10, 10}
    var Hut4MaxLevels : []int = array{1, 20, 5, 4}
    var Hut5MaxLevels : []int = array{1, 20, 10, 5}

    #IF THESE ARE CHANGING ALSO CHANGE IN CUSTOM_PLAYER
    var AllHutCostMults : [][]float = array{}
    var Hut1Costs : []float = array{15000.0,250000.0,400000.0,8000000.0}
    var Hut2Costs : []float = array{1.24e14,3.12e16,4.68e18,4.68e20}
    var Hut3Costs : []float = array{1.95e25,2.95e25,2.95e25,1.95e25}
    var Hut1CostMult : []float = array{350.0,600.0,1000.0,80000.0} #THIS IS PURELY FOR THE HUD
    var Hut2CostMult : []float = array{1.0,10000.0,3000.0,120000.0} #THIS IS PURELY FOR THE HUD
    var Hut3CostMult : []float = array{1.0,1000.0,500.0,250.0} #THIS IS PURELY FOR THE HUD

    var Hut4Costs : []float = array{5.69e36,7.11e36,1.71e56,6.83e70} #CHANGE FOR BALANCE
    var Hut4CostMult : []float = array{1.0,1000.0,50000000.0,10000000.0} #THIS IS PURELY FOR THE HUD
    var Hut4ShardCost : []float = array{100.0,500.0,1000.0,200.0} #CHANGE FOR BALANCE
    var Hut4ShardMult : []float = array{1.0,1.2,2.0,2.0} #THIS IS PURELY FOR THE HUD

    var Hut5Costs : []float = array{1.12e110,1.81e122,2.71e131,4.37e140} #CHANGE FOR BALANCE
    var Hut5CostMult : []float = array{1.0,250.0,11000.0,6000000.0} #THIS IS PURELY FOR THE HUD
    var Hut5ShardCost : []float = array{1000000.0,50000.0,12500.0,20000.0} #CHANGE FOR BALANCE
    var Hut5ShardMult : []float = array{1.0,2.0,5.0,40.0} #THIS IS PURELY FOR THE HUD
    
    var Hut1UpgradeTitles : []string = array{
        "Critical Strike Chance",
        "Critical Damage",
        "Final Strike",
        "Multi-Critical Strike"
        }
    var Hut2UpgradeTitles : []string = array{
        "Combo Strike",
        "Combo Multiplier",
        "Combo Accelerator",
        "Multi-Combo Strike"
        }
    var Hut3UpgradeTitles : []string = array{
        "Earthquake",
        "Earthquake Distance",
        "Earthquake Gather",
        "Earthquake Damage"
        }
    var Hut4UpgradeTitles : []string = array{
        "Void Fracture",
        "Void Damage",
        "Void Wither",
        "Void Stacks"
        }
    var Hut5UpgradeTitles : []string = array{
        "Fire Missile",
        "Missile Damage",
        "Missile Cooldown",
        "Heal"
        }
        
    var Hut1UpgradeDescs : []string = array{
        "Chance to Critical Strike",
        "Critical Strike multiplier",
        "Stone Reward Multiplier\nfor destroying a rock",
        "Critical Strikes can trigger\nmultiple times per strike"
        }
    var Hut2UpgradeDescs : []string = array{
        "Chain attacks to unleash\na powerful strike",
        "Increase multiplier\nfor Combo Strike",
        "Increase combo chain\ngained per strike",
        "Combo Strike triggers\nmultiple times"
        }
    var Hut3UpgradeDescs : []string = array{
        "Cause an Earthquake on hit\nhits rocks around you",
        "Increase Earthquake\ndistance",
        "Increase gather amount\nfrom rocks hit",
        "Increase Earthquake\ndamage"
        }
    var Hut4UpgradeDescs : []string = array{
        "Unleash void energy\nthat fractures the rock,\ndealing damage over time",
        "Increase Void\nFracture damage",
        "Increase Damage dealt\nwhile Void is active",
        "Increase the amount of\nstacks that can be applied"
        }
    var Hut5UpgradeDescs : []string = array{
        "Send Fire Missiles\nat your targets dealing\ndamage",
        "Increase Fire Missile\ndamage",
        "Decrease Cooldown for\nFire Missiles",
        "Fire Missiles heal\nfor a flat amount"
        }

    var Hut1Textures :[]texture = array{UI.CritChance,UI.CritDamage,UI.FinalStrike,UI.MultiCrit}
    var Hut2Textures :[]texture = array{UI.ComboStrike,UI.ComboMultiplier,UI.ComboAccelerator,UI.ComboStreak}
    var Hut3Textures :[]texture = array{UI.Earthquake,UI.EarthquakeDistance,UI.EarthquakeGather,UI.EarthquakeDamage}
    var Hut4Textures :[]texture = array{UI.VoidFracture,UI.VoidDamage,UI.VoidDuration,UI.VoidStacks}
    var Hut5Textures :[]texture = array{UI.FireUnlock,UI.FireDamage,UI.FireCooldown,UI.FireHeal}

    var AllHutUpgradeTitles : [][]string = array{}
    var AllHutUpgradeDescs : [][]string = array{}
    var AllHutTextures : [][]texture = array{}

    OnBegin<override>()<suspends>:void=
        for(i:int=0..HutUpgradeButtons.Length):
            if(HutUpgradeButton:=HutUpgradeButtons[i]):
                HutUpgradeButton.InteractedWithEvent.SubscribeAgent(OpenHUPHUD, i)

        set AllHutUpgradeAmounts = array{Hut1UpgradeAmounts,Hut2UpgradeAmounts,Hut3UpgradeAmounts,Hut4UpgradeAmounts,Hut5UpgradeAmounts}
        set AllHutMaxLevels = array{Hut1MaxLevels,Hut2MaxLevels,Hut3MaxLevels,Hut4MaxLevels,Hut5MaxLevels}
        set AllHutCostMults = array{Hut1CostMult,Hut2CostMult,Hut3CostMult,Hut4CostMult,Hut5CostMult}

        set AllHutUpgradeTitles = array{Hut1UpgradeTitles,Hut2UpgradeTitles,Hut3UpgradeTitles,Hut4UpgradeTitles,Hut5UpgradeTitles}
        set AllHutUpgradeDescs = array{Hut1UpgradeDescs,Hut2UpgradeDescs,Hut3UpgradeDescs,Hut4UpgradeDescs,Hut5UpgradeDescs}
        set AllHutTextures = array{Hut1Textures,Hut2Textures,Hut3Textures,Hut4Textures,Hut5Textures}

        for(Spawner : gm.PlayerSpawners):     
            Spawner.SpawnedEvent.Subscribe(OnPlayerJoined)

    OnPlayerJoined(Agent:agent):void=
        spawn{PlayerJoinedBuffer(Agent)}

    PlayerJoinedBuffer(Agent:agent)<suspends>:void=
        Sleep(1.0)
        if(CustomPlayer := gm.PlayersMap[player[Agent]]):
            if(CustomPlayer.PlayTimeSessionSeconds = 0.0):
                CustomPlayer.HUPCloseButton.OnClick().Subscribe(RemoveHUPHUD)
                CustomPlayer.HUPUpgradeAllButton.OnClick().Subscribe(OnUpgradeAll)
                CustomPlayer.HUPEnableButton.OnClick().Subscribe(EnableHut)
                CustomPlayer.HUPEnableButton.SetText(StringToMessage("DISABLE"))

                set CustomPlayer.HUPAllTexts = array{CustomPlayer.StoneAmountText, CustomPlayer.Upgrade1Title, CustomPlayer.Upgrade1Level,CustomPlayer.Upgrade1Desc,CustomPlayer.Upgrade1Bonus,CustomPlayer.Upgrade1NextBonus,CustomPlayer.Upgrade1CostText,CustomPlayer.Upgrade1CostAmount,CustomPlayer.Upgrade2Title,CustomPlayer.Upgrade2Level,CustomPlayer.Upgrade2Desc,CustomPlayer.Upgrade2Bonus,CustomPlayer.Upgrade2NextBonus,CustomPlayer.Upgrade2CostText,CustomPlayer.Upgrade2CostAmount,CustomPlayer.Upgrade3Title,CustomPlayer.Upgrade3Level,CustomPlayer.Upgrade3Desc,CustomPlayer.Upgrade3Bonus,CustomPlayer.Upgrade3NextBonus,CustomPlayer.Upgrade3CostText,CustomPlayer.Upgrade3CostAmount,CustomPlayer.Upgrade4Title,CustomPlayer.Upgrade4Level,CustomPlayer.Upgrade4Desc,CustomPlayer.Upgrade4Bonus,CustomPlayer.Upgrade4NextBonus,CustomPlayer.Upgrade4CostText,CustomPlayer.Upgrade4CostAmount}
       
                set CustomPlayer.HUPStoneIcons = array{CustomPlayer.StoneIcon1,CustomPlayer.StoneIcon2,CustomPlayer.StoneIcon3,CustomPlayer.StoneIcon4}
                set CustomPlayer.HUPUpgradeTitles = array{CustomPlayer.Upgrade1Title,CustomPlayer.Upgrade2Title,CustomPlayer.Upgrade3Title,CustomPlayer.Upgrade4Title}
                set CustomPlayer.HUPUpgradeLevels = array{CustomPlayer.Upgrade1Level,CustomPlayer.Upgrade2Level,CustomPlayer.Upgrade3Level,CustomPlayer.Upgrade4Level}
                set CustomPlayer.HUPUpgradeIcons = array{CustomPlayer.Upgrade1Icon,CustomPlayer.Upgrade2Icon,CustomPlayer.Upgrade3Icon,CustomPlayer.Upgrade4Icon}
                set CustomPlayer.HUPUpgradeDescTexts = array{CustomPlayer.Upgrade1Desc,CustomPlayer.Upgrade2Desc,CustomPlayer.Upgrade3Desc,CustomPlayer.Upgrade4Desc}
                set CustomPlayer.HUPUpgradeBonuses = array{CustomPlayer.Upgrade1Bonus,CustomPlayer.Upgrade2Bonus,CustomPlayer.Upgrade3Bonus,CustomPlayer.Upgrade4Bonus}
                set CustomPlayer.HUPUpgradeNextBonuses = array{CustomPlayer.Upgrade1NextBonus,CustomPlayer.Upgrade2NextBonus,CustomPlayer.Upgrade3NextBonus,CustomPlayer.Upgrade4NextBonus}
                set CustomPlayer.HUPUpgradeCosts = array{CustomPlayer.Upgrade1CostAmount,CustomPlayer.Upgrade2CostAmount,CustomPlayer.Upgrade3CostAmount,CustomPlayer.Upgrade4CostAmount}
                set CustomPlayer.HUPUpgradeCostTexts = array{CustomPlayer.Upgrade1CostText,CustomPlayer.Upgrade2CostText,CustomPlayer.Upgrade3CostText,CustomPlayer.Upgrade4CostText}
                set CustomPlayer.HUPUpgradeCostIcons = array{CustomPlayer.Upgrade1CostIcon,CustomPlayer.Upgrade2CostIcon,CustomPlayer.Upgrade3CostIcon,CustomPlayer.Upgrade4CostIcon}
                set CustomPlayer.HUPUpgradeButtons = array{CustomPlayer.HUPUpgradeButton1,CustomPlayer.HUPUpgradeButton2,CustomPlayer.HUPUpgradeButton3,CustomPlayer.HUPUpgradeButton4}

                set CustomPlayer.AllHutLevels = array{CustomPlayer.Hut1Levels,CustomPlayer.Hut2Levels,CustomPlayer.Hut3Levels,CustomPlayer.Hut4Levels, CustomPlayer.Hut5Levels}
                set CustomPlayer.AllHutCosts = array{Hut1Costs,Hut2Costs,Hut3Costs,Hut4Costs, Hut5Costs}
                set CustomPlayer.AllHutUpgrades = array{Hut1UpgradeAmounts,Hut2UpgradeAmounts,Hut3UpgradeAmounts,Hut4UpgradeAmounts, Hut5UpgradeAmounts}

                for(i:int=0..CustomPlayer.HUPUpgradeButtons.Length):
                    if(UpgradeButton:=CustomPlayer.HUPUpgradeButtons[i]):
                        UpgradeButton.OnClick().SubscribeAgent3(OnUpgradeButton, i)
            

    EnableHut(Wdata:widget_message):void=
        if(CustomPlayer := gm.PlayersMap[Wdata.Player]):
            if(CustomPlayer.HutEnabled[CustomPlayer.CurrentHut] = true):
                CustomPlayer.HUPEnableBG.SetImage(UI.ButtonOff)
                CustomPlayer.HUPEnableButton.SetText(StringToMessage("ENABLE"))
                if(set CustomPlayer.HutEnabled[CustomPlayer.CurrentHut] = false) {}
            else:
                CustomPlayer.HUPEnableBG.SetImage(UI.ButtonOn)
                CustomPlayer.HUPEnableButton.SetText(StringToMessage("DISABLE"))
                if(set CustomPlayer.HutEnabled[CustomPlayer.CurrentHut] = true) {}

    OnUpgradeButton(Wdata:widget_message, WhichUpgrade:int):void=
        if:
            CustomPlayer := gm.PlayersMap[Wdata.Player]
            HutLevels:=CustomPlayer.AllHutLevels[CustomPlayer.CurrentHut]
            HutLevel:=HutLevels[WhichUpgrade]
            HutCost:=CustomPlayer.AllHutCosts[CustomPlayer.CurrentHut][WhichUpgrade]*(Pow(AllHutCostMults[CustomPlayer.CurrentHut][WhichUpgrade], HutLevel*1.0))
            HutCostMult:=AllHutCostMults[CustomPlayer.CurrentHut]
            HutMaxLevels:=AllHutMaxLevels[CustomPlayer.CurrentHut]
            HutUpgradeAmounts:=AllHutUpgradeAmounts[CustomPlayer.CurrentHut]
            HutCosts:=CustomPlayer.AllHutCosts[CustomPlayer.CurrentHut]
            CustomPlayer.AutoUpgrading = false
        then:
            if(HutLevels[WhichUpgrade] < HutMaxLevels[WhichUpgrade]):
                if(CustomPlayer.StoneAmount >= HutCost): #If you don't have enough stone nothing else matters
                    if(CustomPlayer.CurrentHut < 3): #MINER UPGRADES I (Crits)
                        CustomPlayer.Spend(HutCost, "Stone", 0)
                        LevelHutUpgrade(CustomPlayer, CustomPlayer.CurrentHut, HutUpgradeAmounts, WhichUpgrade)
                    else if:
                        ShardCost:=Hut4ShardCost[WhichUpgrade]*(Pow(Hut4ShardMult[WhichUpgrade], HutLevel*1.0))
                        CustomPlayer.CurrentHut = 3 and CustomPlayer.ShardAmount >= ShardCost
                    then:
                        CustomPlayer.Spend(HutCost, "Stone", 0)
                        CustomPlayer.Spend(ShardCost, "Shard", 0)
                        LevelHutUpgrade(CustomPlayer, CustomPlayer.CurrentHut, HutUpgradeAmounts, WhichUpgrade)

                    else if:
                        ShardCost:=Hut5ShardCost[WhichUpgrade]*(Pow(Hut5ShardMult[WhichUpgrade], HutLevel*1.0))
                        CustomPlayer.CurrentHut = 4 and CustomPlayer.ShardAmount >= ShardCost
                    then:
                        CustomPlayer.Spend(HutCost, "Stone", 0)
                        CustomPlayer.Spend(ShardCost, "Shard", 0)
                        LevelHutUpgrade(CustomPlayer, CustomPlayer.CurrentHut, HutUpgradeAmounts, WhichUpgrade)
                    else:
                        AudioManager.Denied.Play(CustomPlayer.Player)

                    SaveAllData(CustomPlayer)
                    UpdateHUPHUD(CustomPlayer)
                else:
                    AudioManager.Denied.Play(CustomPlayer.Player)
            else:
                AudioManager.Denied.Play(CustomPlayer.Player)

    LevelHutUpgrade(CustomPlayer:Custom_Player, CurrentHut:int, HutUpgradeAmounts:[]float, WhichUpgrade:int):void=
        AudioManager.LeveledUp.Play(CustomPlayer.Player)

        if(Accolade:=AudioManager.Accolades[11]): #AWARD PLAYER EXP
            Accolade.Award(CustomPlayer.Player)
            
        if(CurrentHut = 0):
            if(WhichUpgrade = 0): #Crit Strike Chance
                if(set CustomPlayer.CritRate += HutUpgradeAmounts[WhichUpgrade]) {}
            else if(WhichUpgrade = 1): #Crit Damage
                if(set CustomPlayer.CritDMG += HutUpgradeAmounts[WhichUpgrade]) {}
            else if(WhichUpgrade = 2): #Final Strike
                if(set CustomPlayer.FinalHitBonus += HutUpgradeAmounts[WhichUpgrade]) {Print("{CustomPlayer.FinalHitBonus}")}
            else if(WhichUpgrade = 3): #Multi-Crit Strike
                if(set CustomPlayer.MultiCrit += HutUpgradeAmounts[WhichUpgrade]) {}

            if(set CustomPlayer.Hut1Levels[WhichUpgrade] += 1){} #Up level

        else if(CurrentHut = 1):
            if(WhichUpgrade = 0): #Combo Strike Unlock
                set CustomPlayer.ComboStrikeUnlocked = true
            else if(WhichUpgrade = 1): #Combo Damage
                if(set CustomPlayer.ComboStrikeMultiplier += HutUpgradeAmounts[WhichUpgrade]) {}
            else if(WhichUpgrade = 2): #Combo Accelerator
                if(set CustomPlayer.ComboAccelerator += HutUpgradeAmounts[WhichUpgrade]) {}
            else if(WhichUpgrade = 3): #Combo Conservation
                if(set CustomPlayer.ComboStrikeStreak += HutUpgradeAmounts[WhichUpgrade]) {}
            
            if(set CustomPlayer.Hut2Levels[WhichUpgrade] += 1){} #Up level

        else if(CurrentHut = 2):
            if(WhichUpgrade = 0): #Earthquake Unlock
                set CustomPlayer.EarthquakeUnlocked = true
            else if(WhichUpgrade = 1): #Earthquake Distance
                if(set CustomPlayer.EarthquakeDistance += HutUpgradeAmounts[WhichUpgrade]) {}
            else if(WhichUpgrade = 2): #Earthquake Gather Amount %
                if(set CustomPlayer.EarthquakeGatherAmount += HutUpgradeAmounts[WhichUpgrade]) {}
            else if(WhichUpgrade = 3): #Earthquake Damage %
                if(set CustomPlayer.EarthquakeDamage += HutUpgradeAmounts[WhichUpgrade]) {}
            
            if(set CustomPlayer.Hut3Levels[WhichUpgrade] += 1){} #Up level

        else if(CurrentHut = 3):
            if(WhichUpgrade = 0): #Unlock
                set CustomPlayer.VoidUnlocked = true
            else if(WhichUpgrade = 1): #Void Damage
                #if(set CustomPlayer.VoidDamage += HutUpgradeAmounts[WhichUpgrade]) {}
            else if(WhichUpgrade = 2): #Void Duration
                if(set CustomPlayer.VoidDuration += HutUpgradeAmounts[WhichUpgrade]) {}
            else if(WhichUpgrade = 3): #Void Stacks
                if(set CustomPlayer.VoidStacks += HutUpgradeAmounts[WhichUpgrade]) {}

            if(set CustomPlayer.Hut4Levels[WhichUpgrade] += 1){} #Up level

        else if(CurrentHut = 4):
            if(set CustomPlayer.Hut5Levels[WhichUpgrade] += 1){} #Up level

    OnUpgradeAll(Wdata:widget_message):void=
        if(CustomPlayer := gm.PlayersMap[Wdata.Player]):
            if:
                CustomPlayer.AutoUpgrading = false
                CustomPlayer.PUPQOLLevels[1] >= 1
                CustomPlayer.CurrentHut < 3
            then:
                set CustomPlayer.AutoUpgrading = true
                spawn{UpgradeAll(CustomPlayer)}
            else:
                gm.ShowMessage(CustomPlayer.Player, "Unlock Button In Quarry!", 1)
                AudioManager.Denied.Play(CustomPlayer.Player)

    UpgradeAll(CustomPlayer:Custom_Player)<suspends>:void=
        #for(HUPUpgradeButton:CustomPlayer.HUPUpgradeButtons):
            #HUPUpgradeButton.SetEnabled(false)   
        loop:
            var LowestCost : float = 9.99e200
            for(i:int=0..CustomPlayer.HUPUpgradeTitles.Length-1):
                if:
                    HutLevels:=CustomPlayer.AllHutLevels[CustomPlayer.CurrentHut]
                    HutLevel:=HutLevels[i]
                    HutCost:=CustomPlayer.AllHutCosts[CustomPlayer.CurrentHut][i]*(Pow(AllHutCostMults[CustomPlayer.CurrentHut][i], HutLevel*1.0))
                    HutMaxLevels:=AllHutMaxLevels[CustomPlayer.CurrentHut]
                    HutUpgradeAmounts:=AllHutUpgradeAmounts[CustomPlayer.CurrentHut]
                then:
                    if:
                        HutLevels[i] < HutMaxLevels[i]
                        HutCost < LowestCost
                    then:
                        Print("LOWEST COST IS {i} - {HutCost}")
                        set LowestCost = HutCost
                    else:
                        Print("CANT UPGRADE THIS ONE")
            Sleep(0.1)
            for(i:int=0..CustomPlayer.HUPUpgradeTitles.Length-1):
                n := 3 - i
                if:
                    HutLevels:=CustomPlayer.AllHutLevels[CustomPlayer.CurrentHut]
                    HutLevel:=HutLevels[n]
                    HutCost:=CustomPlayer.AllHutCosts[CustomPlayer.CurrentHut][n]*(Pow(AllHutCostMults[CustomPlayer.CurrentHut][n], HutLevel*1.0))
                    HutMaxLevels:=AllHutMaxLevels[CustomPlayer.CurrentHut]
                    HutUpgradeAmounts:=AllHutUpgradeAmounts[CustomPlayer.CurrentHut]
                then:
                    #FIRST 3
                    if(CustomPlayer.CurrentHut < 3):
                        if(CustomPlayer.CurrentHut = 0):
                            if(HutLevels[n] < HutMaxLevels[n]):
                                if(CustomPlayer.StoneAmount >= HutCost):
                                    CustomPlayer.Spend(HutCost, "Stone", 0)
                                    LevelHutUpgrade(CustomPlayer, CustomPlayer.CurrentHut, HutUpgradeAmounts, n)
                        else:
                            #UNLOCK FIRST FOR COMBO AND EQ
                            if(i = 0 and HutLevels[0] = 0):
                                if:
                                    NewHutCost:=CustomPlayer.AllHutCosts[CustomPlayer.CurrentHut][i]
                                    CustomPlayer.StoneAmount >= NewHutCost
                                then:
                                    CustomPlayer.Spend(NewHutCost, "Stone", 0)
                                    LevelHutUpgrade(CustomPlayer, CustomPlayer.CurrentHut, HutUpgradeAmounts, i)
                            else:
                                if(HutLevels[0] = 1, HutLevels[n] < HutMaxLevels[n]):
                                    if(CustomPlayer.StoneAmount >= HutCost):
                                        CustomPlayer.Spend(HutCost, "Stone", 0)
                                        LevelHutUpgrade(CustomPlayer, CustomPlayer.CurrentHut, HutUpgradeAmounts, n)
                    else if(CustomPlayer.CurrentHut = 3):
                        #UNLOCK FIRST FOR VOID
                        if(i = 0 and HutLevels[0] = 0):
                            if:
                                ShardCost:=Hut4ShardCost[i]*(Pow(Hut4ShardMult[i], HutLevel*1.0))
                                NewHutCost:=CustomPlayer.AllHutCosts[CustomPlayer.CurrentHut][i]
                                CustomPlayer.ShardAmount >= ShardCost
                                CustomPlayer.StoneAmount >= NewHutCost
                            then:
                                CustomPlayer.Spend(NewHutCost, "Stone", 0)
                                CustomPlayer.Spend(ShardCost, "Shard", 0)
                                LevelHutUpgrade(CustomPlayer, CustomPlayer.CurrentHut, HutUpgradeAmounts, i)
                        else:
                            if:
                                ShardCost:=Hut4ShardCost[i]*(Pow(Hut4ShardMult[i], HutLevel*1.0))
                                CustomPlayer.ShardAmount >= ShardCost
                                CustomPlayer.StoneAmount >= HutCost
                                HutLevels[0] = 1
                                HutLevels[n] < HutMaxLevels[n]
                            then:
                                CustomPlayer.Spend(HutCost, "Stone", 0)
                                CustomPlayer.Spend(ShardCost, "Shard", 0)
                                LevelHutUpgrade(CustomPlayer, CustomPlayer.CurrentHut, HutUpgradeAmounts, n)

                    else if(CustomPlayer.CurrentHut = 4):
                        #UNLOCK FIRST FOR FIRE MISSILE
                        if(i = 0 and HutLevels[0] = 0):
                            if:
                                ShardCost:=Hut5ShardCost[i]*(Pow(Hut5ShardMult[i], HutLevel*1.0))
                                NewHutCost:=CustomPlayer.AllHutCosts[CustomPlayer.CurrentHut][i]
                                CustomPlayer.ShardAmount >= ShardCost
                                CustomPlayer.StoneAmount >= NewHutCost
                            then:
                                CustomPlayer.Spend(NewHutCost, "Stone", 0)
                                CustomPlayer.Spend(ShardCost, "Shard", 0)
                                LevelHutUpgrade(CustomPlayer, CustomPlayer.CurrentHut, HutUpgradeAmounts, i)
                        else:
                            if:
                                ShardCost:=Hut5ShardCost[i]*(Pow(Hut5ShardMult[i], HutLevel*1.0))
                                CustomPlayer.ShardAmount >= ShardCost
                                CustomPlayer.StoneAmount >= HutCost
                                HutLevels[0] = 1
                                HutLevels[n] < HutMaxLevels[n]
                            then:
                                CustomPlayer.Spend(HutCost, "Stone", 0)
                                CustomPlayer.Spend(ShardCost, "Shard", 0)
                                LevelHutUpgrade(CustomPlayer, CustomPlayer.CurrentHut, HutUpgradeAmounts, n)

            UpdateHUPHUD(CustomPlayer)
            if(CustomPlayer.StoneAmount < LowestCost):
                set CustomPlayer.AutoUpgrading = false
                SaveAllData(CustomPlayer)     
                for(HUPUpgradeButton:CustomPlayer.HUPUpgradeButtons):
                    HUPUpgradeButton.SetEnabled(true)   
                Print("YOU ARE POOR")
                break:

    OpenHUPHUD(Agent:agent, whichHut:int): void=
        if(CustomPlayer := gm.PlayersMap[Agent]):
            if(MyUI:=CustomPlayer.UIMaps["HUP"]?):
            else:
                Tutorial.UpdateTutorialUI(CustomPlayer, "OpenHut", 1.0)
                set CustomPlayer.CurrentHut = whichHut
                var MyHUPCanvas : canvas = MakeHUPCanvas(CustomPlayer)
                CustomPlayer.AddFullHUD(MyHUPCanvas)
                if(set CustomPlayer.UIMaps["HUP"] = option{MyHUPCanvas}) {}
                UpdateHUPHUD(CustomPlayer)

    RemoveHUPHUD(Wdata:widget_message):void=
        if(CustomPlayer := gm.PlayersMap[Wdata.Player]):
            if(Player:=Wdata.Player, PlayerUI:=GetPlayerUI[Player]):
                if(MyUI:=CustomPlayer.UIMaps["HUP"]?):
                    PlayerUI.RemoveWidget(MyUI)
                    MyUI.SetVisibility(widget_visibility.Collapsed)
                    if(set CustomPlayer.UIMaps["HUP"] = false){}

    #CALL THIS AFTER ANY UPGRADE TO MAKE SURE THE HUD IS UP TO DATE
    UpdateHUPHUD(CustomPlayer: Custom_Player):void=
        Visible := widget_visibility.Visible
        Collapsed := widget_visibility.Collapsed
        #if(CustomPlayer.PUPQOLLevels[1] >= 1):
        CustomPlayer.HUPUpgradeAllButton.SetEnabled(true)
        #if(CustomPlayer.PUPQOLLevels[1] >= 2):
        CustomPlayer.HUPEnableButton.SetVisibility(widget_visibility.Visible)
        CustomPlayer.HUPEnableBG.SetVisibility(widget_visibility.Visible)
        #else:
            #CustomPlayer.HUPEnableButton.SetVisibility(widget_visibility.Collapsed)
            #CustomPlayer.HUPEnableBG.SetVisibility(widget_visibility.Collapsed)
        

        if(CustomPlayer.HutEnabled[CustomPlayer.CurrentHut] = true):
            CustomPlayer.HUPEnableBG.SetImage(UI.ButtonOn)
            CustomPlayer.HUPEnableButton.SetText(StringToMessage("DISABLE"))
        else:
            CustomPlayer.HUPEnableBG.SetImage(UI.ButtonOff)
            CustomPlayer.HUPEnableButton.SetText(StringToMessage("ENABLE"))

        set CustomPlayer.AllHutLevels = array{CustomPlayer.Hut1Levels,CustomPlayer.Hut2Levels,CustomPlayer.Hut3Levels,CustomPlayer.Hut4Levels, CustomPlayer.Hut5Levels}
        set CustomPlayer.AllHutCosts = array{Hut1Costs,Hut2Costs,Hut3Costs,Hut4Costs, Hut5Costs}
        set CustomPlayer.AllHutUpgrades = array{Hut1UpgradeAmounts,Hut2UpgradeAmounts,Hut3UpgradeAmounts,Hut4UpgradeAmounts,Hut5UpgradeAmounts}
        if:
            HutCostMult:=AllHutCostMults[CustomPlayer.CurrentHut]
            HutMaxLevels:=AllHutMaxLevels[CustomPlayer.CurrentHut]
            HutUpgrades:=CustomPlayer.AllHutUpgrades[CustomPlayer.CurrentHut]
            HutUpgradeAmounts:=AllHutUpgradeAmounts[CustomPlayer.CurrentHut]
            HutCosts:=CustomPlayer.AllHutCosts[CustomPlayer.CurrentHut]
            HutLevels:=CustomPlayer.AllHutLevels[CustomPlayer.CurrentHut]
            HutUpgradeTitles:=AllHutUpgradeTitles[CustomPlayer.CurrentHut]
            HutUpgradeDescs:=AllHutUpgradeDescs[CustomPlayer.CurrentHut]
            HutTextures:=AllHutTextures[CustomPlayer.CurrentHut]
        then:
            if (CustomPlayer.CurrentHut = 0):
                CustomPlayer.HUPHUDBGCover.SetDesiredSize(vector2{X:=0.0,Y:=0.0})
                
                if(CustomPlayer.AutoUpgrading = false):
                    CustomPlayer.HUPUpgradeButton2.SetEnabled(true)
                    CustomPlayer.HUPUpgradeButton3.SetEnabled(true)
                    CustomPlayer.HUPUpgradeButton4.SetEnabled(true)
                    CustomPlayer.HUPUpgradeButton2.SetVisibility(widget_visibility.Visible)
                    CustomPlayer.HUPUpgradeButton3.SetVisibility(widget_visibility.Visible)
                    CustomPlayer.HUPUpgradeButton4.SetVisibility(widget_visibility.Visible)
            else:
                if(HutLevels[0] >= 1):
                    CustomPlayer.HUPHUDBGCover.SetDesiredSize(vector2{X:=0.0,Y:=0.0})
                    if(CustomPlayer.AutoUpgrading = false):
                        CustomPlayer.HUPUpgradeButton2.SetEnabled(true)
                        CustomPlayer.HUPUpgradeButton3.SetEnabled(true)
                        CustomPlayer.HUPUpgradeButton4.SetEnabled(true)
                        CustomPlayer.HUPUpgradeButton2.SetVisibility(widget_visibility.Visible)
                        CustomPlayer.HUPUpgradeButton3.SetVisibility(widget_visibility.Visible)
                        CustomPlayer.HUPUpgradeButton4.SetVisibility(widget_visibility.Visible)
                else:
                    CustomPlayer.HUPHUDBGCover.SetDesiredSize(vector2{X:=2140.5,Y:=1076.0})
                    CustomPlayer.HUPUpgradeButton2.SetEnabled(false)
                    CustomPlayer.HUPUpgradeButton3.SetEnabled(false)
                    CustomPlayer.HUPUpgradeButton4.SetEnabled(false)
                    CustomPlayer.HUPUpgradeButton2.SetVisibility(widget_visibility.Collapsed)
                    CustomPlayer.HUPUpgradeButton3.SetVisibility(widget_visibility.Collapsed)
                    CustomPlayer.HUPUpgradeButton4.SetVisibility(widget_visibility.Collapsed)

            CustomPlayer.StoneAmountText.SetText(StringToMessage(NumberCondensor(CustomPlayer.StoneAmount)))

            for(i:int=0..CustomPlayer.HUPUpgradeTitles.Length):
                if:
                    StoneIcon:=CustomPlayer.StoneIcons[i]
                    UpgradeNextBonus:=CustomPlayer.HUPUpgradeNextBonuses[i]
                    UpgradeCostText:=CustomPlayer.HUPUpgradeCostTexts[i]
                    UpgradeCostIcon:=CustomPlayer.HUPUpgradeCostIcons[i]
                    HutLevel:=HutLevels[i]
                    MaxLevel:=HutMaxLevels[i]
                    HutUpgrade:=HutUpgrades[i]*HutLevel
                    HutUpgradeAmount:=HutUpgradeAmounts[i]
                    HutCost:=HutCosts[i]*(Pow(HutCostMult[i], HutLevel*1.0))
                    HutShard4Cost:=Hut4ShardCost[i]*(Pow(Hut4ShardMult[i], HutLevel*1.0))
                    HutShard5Cost:=Hut5ShardCost[i]*(Pow(Hut5ShardMult[i], HutLevel*1.0))
                    HutUpgradeTitle:=HutUpgradeTitles[i]
                    HutUpgradeDesc:=HutUpgradeDescs[i]
                    HutUpgradeIcon:=HutTextures[i]
                    UpgradeTitle:=CustomPlayer.HUPUpgradeTitles[i]
                    UpgradeLevel:=CustomPlayer.HUPUpgradeLevels[i]
                    UpgradeIcon:=CustomPlayer.HUPUpgradeIcons[i]
                    UpgradeDesc:=CustomPlayer.HUPUpgradeDescTexts[i]
                    UpgradeBonus:=CustomPlayer.HUPUpgradeBonuses[i]
                    UpgradeCost:=CustomPlayer.HUPUpgradeCosts[i]
                    UpgradeButton:=CustomPlayer.HUPUpgradeButtons[i]
                then:
                    UpgradeTitle.SetText(StringToMessage(HutUpgradeTitle))
                    UpgradeLevel.SetText(StringToMessage("Lv. {HutLevel}"))
                    UpgradeIcon.SetImage(HutUpgradeIcon)
                    UpgradeDesc.SetText(StringToMessage(HutUpgradeDesc))
                    UpgradeCost.SetText(StringToMessage(" {NumberCondensor(HutCost)}"))

                    if(CustomPlayer.StoneAmount >= HutCost):
                        UpgradeCost.SetTextColor(MakeColorFromHex("39FF39"))
                    else:
                        UpgradeCost.SetTextColor(MakeColorFromHex("fa4d56"))
                    if(CustomPlayer.CurrentHut = 3):
                        if(CustomPlayer.ShardAmount >= HutShard4Cost):
                            UpgradeCostText.SetTextColor(MakeColorFromHex("39FF39"))
                        else:
                            UpgradeCostText.SetTextColor(MakeColorFromHex("fa4d56"))
                    else if(CustomPlayer.CurrentHut = 4):
                        if(CustomPlayer.ShardAmount >= HutShard5Cost):
                            UpgradeCostText.SetTextColor(MakeColorFromHex("39FF39"))
                        else:
                            UpgradeCostText.SetTextColor(MakeColorFromHex("fa4d56"))
                    else:
                        UpgradeCostText.SetTextColor(MakeColorFromHex("DDDDDD"))

                        
                    var isPerc : logic = true
                    if(CustomPlayer.CurrentHut = 0):
                        if(i = 3):
                            set isPerc = false
                    else if(CustomPlayer.CurrentHut = 1):
                        if(i = 0 or i = 2 or i = 3):
                            set isPerc = false
                    else if(CustomPlayer.CurrentHut = 2):
                        if(i = 0 or i = 1):
                            set isPerc = false
                    else if(CustomPlayer.CurrentHut = 3):
                        if(i = 0 or i = 3):
                            set isPerc = false
                    else if(CustomPlayer.CurrentHut = 4):
                        if(i = 0 or i = 3):
                            set isPerc = false
                            
                    if(isPerc = true):
                        UpgradeBonus.SetText(StringToMessage("{NumberCondensor(HutUpgrade*100.0)}% "))
                    else:
                        UpgradeBonus.SetText(StringToMessage("{NumberCondensor(HutUpgrade)} "))

                    if(HutLevel < MaxLevel):
                        UpgradeButton.SetText(StringToMessage("UPGRADE"))
                        StoneIcon.SetVisibility(widget_visibility.Visible)
                        if(CustomPlayer.CurrentHut = 3 or CustomPlayer.CurrentHut = 4):
                            UpgradeCostIcon.SetVisibility(widget_visibility.Visible)
                            if(CustomPlayer.CurrentHut = 3):
                                UpgradeCostText.SetText(StringToMessage(" {NumberCondensor(HutShard4Cost)}"))
                            else if(CustomPlayer.CurrentHut = 4):
                                UpgradeCostText.SetText(StringToMessage(" {NumberCondensor(HutShard5Cost)}"))
                        else:
                            UpgradeCostIcon.SetVisibility(widget_visibility.Collapsed)
                            UpgradeCostText.SetText(StringToMessage("COST:"))
                        if(isPerc = true):
                            UpgradeNextBonus.SetText(StringToMessage("(+{NumberCondensor(HutUpgradeAmount*100.0)}%)"))
                        else:
                            UpgradeNextBonus.SetText(StringToMessage("(+{NumberCondensor(HutUpgradeAmount)})"))
                    else:
                        UpgradeButton.SetText(StringToMessage("MAX"))
                        StoneIcon.SetVisibility(widget_visibility.Collapsed)
                        UpgradeCostText.SetText(StringToMessage(""))
                        UpgradeCostIcon.SetVisibility(widget_visibility.Collapsed)
                        UpgradeCost.SetText(StringToMessage(""))
                        UpgradeNextBonus.SetText(StringToMessage(""))

    MakeHUPCanvas(CustomPlayer:Custom_Player):canvas=
        CustomPlayer.HUPCloseButton.SetText(StringToMessage("CLOSE"))
        CustomPlayer.HUPUpgradeAllButton.SetText(StringToMessage("AUTO"))

        CustomPlayer.Upgrade1Desc.SetJustification(text_justification.Center)
        CustomPlayer.Upgrade2Desc.SetJustification(text_justification.Center)
        CustomPlayer.Upgrade3Desc.SetJustification(text_justification.Center)
        CustomPlayer.Upgrade4Desc.SetJustification(text_justification.Center)
        #CustomPlayer.HUPUpgradeAllButton.SetEnabled(false)

        #if(CustomPlayer.CurrentHut = 3):
        #    CustomPlayer.HUPUpgradeAllButton.SetVisibility(widget_visibility.Collapsed)
        #else:
        #    CustomPlayer.HUPUpgradeAllButton.SetVisibility(widget_visibility.Visible)


        for(TextBlock : CustomPlayer.HUPAllTexts):
            TextBlock.SetShadowColor(CustomPlayer.ShadowDefaults.DefaultShadowColor)
            TextBlock.SetShadowOffset(CustomPlayer.ShadowDefaults.DefaultShadowOffset)
            TextBlock.SetShadowOpacity(1.0)

        var MyHUPCanvas: canvas = canvas:
            Slots := array:
                canvas_slot:
                    Widget:=CustomPlayer.HUPHUDBG
                    ZOrder:=0
                    SizeToContent:=false
                    Alignment := vector2{X := 0.5, Y := 0.0}
                    Anchors := anchors{Minimum := vector2{X:= 0.5, Y:= 0.0}, Maximum := vector2{X:= 0.5, Y:= 0.0}}
                    Offsets:=margin{Top:=0.0,Left:=0.0,Bottom:=1076.0,Right:=2140.5}
                canvas_slot:
                    Widget:=CustomPlayer.HUPHUDBGCover
                    ZOrder:=2
                    SizeToContent:=true
                    Alignment := vector2{X := 0.5, Y := 0.0}
                    Anchors := anchors{Minimum := vector2{X:= 0.5, Y:= 0.0}, Maximum := vector2{X:= 0.5, Y:= 0.0}}
                    Offsets:=margin{Top:=0.0,Left:=0.0,Bottom:=1076.0,Right:=2140.5}
                
                canvas_slot:
                    Widget:=CustomPlayer.HUPUpgradeAllButton
                    ZOrder:=3
                    SizeToContent:=false
                    Alignment := vector2{X:= 0.5, Y:= 0.0}
                    Anchors := anchors{Minimum := vector2{X:= 0.5, Y:= 0.0}, Maximum := vector2{X:= 0.5, Y:= 0.0}}
                    Offsets:=margin{Top:=976.0,Left:=-348.0, Bottom:=50.0, Right:=150.0}
                canvas_slot:
                    Widget:=CustomPlayer.HUPEnableButton
                    ZOrder:=3
                    SizeToContent:=false
                    Alignment := vector2{X:= 0.5, Y:= 0.0}
                    Anchors := anchors{Minimum := vector2{X:= 0.5, Y:= 0.0}, Maximum := vector2{X:= 0.5, Y:= 0.0}}
                    Offsets:=margin{Top:=976.0,Left:=-148.0, Bottom:=50.0, Right:=200.0}
                canvas_slot:
                    Widget:=CustomPlayer.HUPEnableBG
                    ZOrder:=3
                    SizeToContent:=false
                    Alignment := vector2{X:= 0.5, Y:= 0.0}
                    Anchors := anchors{Minimum := vector2{X:= 0.5, Y:= 0.0}, Maximum := vector2{X:= 0.5, Y:= 0.0}}
                    Offsets:=margin{Top:=976.0,Left:=-148.0, Bottom:=50.0, Right:=200.0}

                canvas_slot:
                    Widget:=CustomPlayer.HUPHUDTitle
                    ZOrder:=3
                    SizeToContent:=false
                    Anchors := anchors{Minimum := vector2{X:= 0.5, Y:= 0.0}, Maximum := vector2{X:= 0.5, Y:= 0.0}}
                    Offsets:=margin{Top:=120.0,Left:=-100.0, Bottom:=48.0, Right:=537.0}

                canvas_slot:
                    Widget:=CustomPlayer.StoneAmountIcon
                    ZOrder:=3
                    SizeToContent:=false
                    Anchors := anchors{Minimum := vector2{X:= 0.5, Y:= 0.0}, Maximum := vector2{X:= 0.5, Y:= 0.0}}
                    Offsets:=margin{Top:=162.0,Left:=-828.0, Bottom:=40.0, Right:=40.0}
                canvas_slot:
                    Widget:=CustomPlayer.StoneAmountText
                    ZOrder:=3
                    SizeToContent:=true
                    Anchors := anchors{Minimum := vector2{X:= 0.5, Y:= 0.0}, Maximum := vector2{X:= 0.5, Y:= 0.0}}
                    Offsets:=margin{Top:=162.0,Left:=-776.0}
                    Alignment := vector2{X:= 0.0, Y:= 0.0}

                canvas_slot:
                    Widget:=CustomPlayer.Upgrade1Title
                    ZOrder:=3
                    SizeToContent:=true
                    Anchors := anchors{Minimum := vector2{X:= 0.5, Y:= 0.0}, Maximum := vector2{X:= 0.5, Y:= 0.0}}
                    Offsets:=margin{Top:=282.0,Left:=-660.0}
                    Alignment := vector2{X:= 0.5, Y:= 0.0}
                canvas_slot:
                    Widget:=CustomPlayer.Upgrade1Level
                    ZOrder:=3
                    SizeToContent:=true
                    Anchors := anchors{Minimum := vector2{X:= 0.5, Y:= 0.0}, Maximum := vector2{X:= 0.5, Y:= 0.0}}
                    Offsets:=margin{Top:=341.0,Left:=-660.0}
                    Alignment := vector2{X:= 0.5, Y:= 0.0}
                canvas_slot:
                    Widget:=CustomPlayer.Upgrade1Icon
                    ZOrder:=3
                    SizeToContent:=false
                    Anchors := anchors{Minimum := vector2{X:= 0.5, Y:= 0.0}, Maximum := vector2{X:= 0.5, Y:= 0.0}}
                    Offsets:=margin{Top:=421.0,Left:=-660.0, Bottom:=140.0,Right:=140.0}
                    Alignment := vector2{X:= 0.5, Y:= 0.0}
                canvas_slot:
                    Widget:=CustomPlayer.Upgrade1Desc
                    ZOrder:=3
                    SizeToContent:=true
                    Anchors := anchors{Minimum := vector2{X:= 0.5, Y:= 0.0}, Maximum := vector2{X:= 0.5, Y:= 0.0}}
                    Offsets:=margin{Top:=592.0,Left:=-660.0}
                    Alignment := vector2{X:= 0.5, Y:= 0.0}
                canvas_slot:
                    #Widget:=Upgrade1Bonus
                    ZOrder:=3
                    SizeToContent:=true
                    Anchors := anchors{Minimum := vector2{X:= 0.5, Y:= 0.0}, Maximum := vector2{X:= 0.5, Y:= 0.0}}
                    Offsets:=margin{Top:=740.0,Left:=-660.0}
                    Alignment := vector2{X:= 0.5, Y:= 0.0}
                    Widget := stack_box:
                        Orientation := orientation.Horizontal
                        Slots := array:
                            stack_box_slot:
                                Widget := CustomPlayer.Upgrade1Bonus   
                            stack_box_slot:
                                Widget := CustomPlayer.Upgrade1NextBonus
                canvas_slot:
                    #Widget:=CustomPlayer.Upgrade1CostText
                    ZOrder:=3
                    SizeToContent:=true
                    Anchors := anchors{Minimum := vector2{X:= 0.5, Y:= 0.0}, Maximum := vector2{X:= 0.5, Y:= 0.0}}
                    Offsets:=margin{Top:=800.0,Left:=-776.0}
                    Alignment := vector2{X:= 0.5, Y:= 0.0}
                    Widget := stack_box:
                        Orientation := orientation.Horizontal
                        Slots := array:
                            stack_box_slot:
                                Widget := CustomPlayer.Upgrade1CostIcon   
                            stack_box_slot:
                                Widget := CustomPlayer.Upgrade1CostText
                canvas_slot:
                    ZOrder:=3
                    SizeToContent:=true
                    Anchors := anchors{Minimum := vector2{X:= 0.5, Y:= 0.0}, Maximum := vector2{X:= 0.5, Y:= 0.0}}
                    Offsets:=margin{Top:=800.0,Left:=-484.0}
                    Alignment := vector2{X:= 1.0, Y:= 0.0}
                    Widget := stack_box:
                        Orientation := orientation.Horizontal
                        Slots := array:
                            stack_box_slot:
                                Widget := CustomPlayer.StoneIcon1    
                            stack_box_slot:
                                Widget := CustomPlayer.Upgrade1CostAmount
                canvas_slot:
                    Widget:=CustomPlayer.HUPUpgradeButton1
                    ZOrder:=3
                    SizeToContent:=true
                    Anchors := anchors{Minimum := vector2{X:= 0.5, Y:= 0.0}, Maximum := vector2{X:= 0.5, Y:= 0.0}}
                    Offsets:=margin{Top:=870.0,Left:=-660.0, Bottom:=60.0, Right:=220.0}
                    Alignment := vector2{X:= 0.5, Y:= 0.0}
                    
                canvas_slot:
                    Widget:=CustomPlayer.Upgrade2Title
                    ZOrder:=1
                    SizeToContent:=true
                    Anchors := anchors{Minimum := vector2{X:= 0.5, Y:= 0.0}, Maximum := vector2{X:= 0.5, Y:= 0.0}}
                    Offsets:=margin{Top:=282.0,Left:=-220.0}
                    Alignment := vector2{X:= 0.5, Y:= 0.0}
                canvas_slot:
                    Widget:=CustomPlayer.Upgrade2Level
                    ZOrder:=1
                    SizeToContent:=true
                    Anchors := anchors{Minimum := vector2{X:= 0.5, Y:= 0.0}, Maximum := vector2{X:= 0.5, Y:= 0.0}}
                    Offsets:=margin{Top:=341.0,Left:=-220.0}
                    Alignment := vector2{X:= 0.5, Y:= 0.0}
                canvas_slot:
                    Widget:=CustomPlayer.Upgrade2Icon
                    ZOrder:=1
                    SizeToContent:=false
                    Anchors := anchors{Minimum := vector2{X:= 0.5, Y:= 0.0}, Maximum := vector2{X:= 0.5, Y:= 0.0}}
                    Offsets:=margin{Top:=421.0,Left:=-220.0, Bottom:=140.0,Right:=140.0}
                    Alignment := vector2{X:= 0.5, Y:= 0.0}
                canvas_slot:
                    Widget:=CustomPlayer.Upgrade2Desc
                    ZOrder:=1
                    SizeToContent:=true
                    Anchors := anchors{Minimum := vector2{X:= 0.5, Y:= 0.0}, Maximum := vector2{X:= 0.5, Y:= 0.0}}
                    Offsets:=margin{Top:=592.0,Left:=-220.0}
                    Alignment := vector2{X:= 0.5, Y:= 0.0}
                canvas_slot:
                    #Widget:=Upgrade2Bonus
                    ZOrder:=1
                    SizeToContent:=true
                    Anchors := anchors{Minimum := vector2{X:= 0.5, Y:= 0.0}, Maximum := vector2{X:= 0.5, Y:= 0.0}}
                    Offsets:=margin{Top:=740.0,Left:=-220.0}
                    Alignment := vector2{X:= 0.5, Y:= 0.0}
                    Widget := stack_box:
                        Orientation := orientation.Horizontal
                        Slots := array:
                            stack_box_slot:
                                Widget := CustomPlayer.Upgrade2Bonus   
                            stack_box_slot:
                                Widget := CustomPlayer.Upgrade2NextBonus
                canvas_slot:
                    #Widget:=CustomPlayer.Upgrade2CostText
                    ZOrder:=1
                    SizeToContent:=true
                    Anchors := anchors{Minimum := vector2{X:= 0.5, Y:= 0.0}, Maximum := vector2{X:= 0.5, Y:= 0.0}}
                    Offsets:=margin{Top:=800.0,Left:=-336.0}
                    Alignment := vector2{X:= 0.5, Y:= 0.0}
                    Widget := stack_box:
                        Orientation := orientation.Horizontal
                        Slots := array:
                            stack_box_slot:
                                Widget := CustomPlayer.Upgrade2CostIcon   
                            stack_box_slot:
                                Widget := CustomPlayer.Upgrade2CostText
                canvas_slot:
                    ZOrder:=1
                    SizeToContent:=true
                    Anchors := anchors{Minimum := vector2{X:= 0.5, Y:= 0.0}, Maximum := vector2{X:= 0.5, Y:= 0.0}}
                    Offsets:=margin{Top:=800.0,Left:=-44.0}
                    Alignment := vector2{X:= 1.0, Y:= 0.0}
                    Widget := stack_box:
                        Orientation := orientation.Horizontal
                        Slots := array:
                            stack_box_slot:
                                Widget := CustomPlayer.StoneIcon2   
                            stack_box_slot:
                                Widget := CustomPlayer.Upgrade2CostAmount
                canvas_slot:
                    Widget:=CustomPlayer.HUPUpgradeButton2
                    ZOrder:=1
                    SizeToContent:=true
                    Anchors := anchors{Minimum := vector2{X:= 0.5, Y:= 0.0}, Maximum := vector2{X:= 0.5, Y:= 0.0}}
                    Offsets:=margin{Top:=870.0,Left:=-220.0, Bottom:=60.0, Right:=220.0}
                    Alignment := vector2{X:= 0.5, Y:= 0.0}
                    
                canvas_slot:
                    Widget:=CustomPlayer.Upgrade3Title
                    ZOrder:=1
                    SizeToContent:=true
                    Anchors := anchors{Minimum := vector2{X:= 0.5, Y:= 0.0}, Maximum := vector2{X:= 0.5, Y:= 0.0}}
                    Offsets:=margin{Top:=282.0,Left:=216.0}
                    Alignment := vector2{X:= 0.5, Y:= 0.0}
                canvas_slot:
                    Widget:=CustomPlayer.Upgrade3Level
                    ZOrder:=1
                    SizeToContent:=true
                    Anchors := anchors{Minimum := vector2{X:= 0.5, Y:= 0.0}, Maximum := vector2{X:= 0.5, Y:= 0.0}}
                    Offsets:=margin{Top:=341.0,Left:=216.0}
                    Alignment := vector2{X:= 0.5, Y:= 0.0}
                canvas_slot:
                    Widget:=CustomPlayer.Upgrade3Icon
                    ZOrder:=1
                    SizeToContent:=false
                    Anchors := anchors{Minimum := vector2{X:= 0.5, Y:= 0.0}, Maximum := vector2{X:= 0.5, Y:= 0.0}}
                    Offsets:=margin{Top:=421.0,Left:=216.0, Bottom:=140.0,Right:=140.0}
                    Alignment := vector2{X:= 0.5, Y:= 0.0}
                canvas_slot:
                    Widget:=CustomPlayer.Upgrade3Desc
                    ZOrder:=1
                    SizeToContent:=true
                    Anchors := anchors{Minimum := vector2{X:= 0.5, Y:= 0.0}, Maximum := vector2{X:= 0.5, Y:= 0.0}}
                    Offsets:=margin{Top:=592.0,Left:=216.0}
                    Alignment := vector2{X:= 0.5, Y:= 0.0}
                canvas_slot:
                    #Widget:=Upgrade3Bonus
                    ZOrder:=1
                    SizeToContent:=true
                    Anchors := anchors{Minimum := vector2{X:= 0.5, Y:= 0.0}, Maximum := vector2{X:= 0.5, Y:= 0.0}}
                    Offsets:=margin{Top:=740.0,Left:=216.0}
                    Alignment := vector2{X:= 0.5, Y:= 0.0}
                    Widget := stack_box:
                        Orientation := orientation.Horizontal
                        Slots := array:
                            stack_box_slot:
                                Widget := CustomPlayer.Upgrade3Bonus   
                            stack_box_slot:
                                Widget := CustomPlayer.Upgrade3NextBonus
                canvas_slot:
                    #Widget:=CustomPlayer.Upgrade3CostText
                    ZOrder:=1
                    SizeToContent:=true
                    Anchors := anchors{Minimum := vector2{X:= 0.5, Y:= 0.0}, Maximum := vector2{X:= 0.5, Y:= 0.0}}
                    Offsets:=margin{Top:=800.0,Left:=100.0}
                    Alignment := vector2{X:= 0.5, Y:= 0.0}
                    Widget := stack_box:
                        Orientation := orientation.Horizontal
                        Slots := array:
                            stack_box_slot:
                                Widget := CustomPlayer.Upgrade3CostIcon   
                            stack_box_slot:
                                Widget := CustomPlayer.Upgrade3CostText
                canvas_slot:
                    ZOrder:=1
                    SizeToContent:=true
                    Anchors := anchors{Minimum := vector2{X:= 0.5, Y:= 0.0}, Maximum := vector2{X:= 0.5, Y:= 0.0}}
                    Offsets:=margin{Top:=800.0,Left:=392.0}
                    Alignment := vector2{X:= 1.0, Y:= 0.0}
                    Widget := stack_box:
                        Orientation := orientation.Horizontal
                        Slots := array:
                            stack_box_slot:
                                Widget := CustomPlayer.StoneIcon3    
                            stack_box_slot:
                                Widget := CustomPlayer.Upgrade3CostAmount
                canvas_slot:
                    Widget:=CustomPlayer.HUPUpgradeButton3
                    ZOrder:=1
                    SizeToContent:=true
                    Anchors := anchors{Minimum := vector2{X:= 0.5, Y:= 0.0}, Maximum := vector2{X:= 0.5, Y:= 0.0}}
                    Offsets:=margin{Top:=870.0,Left:=216.0, Bottom:=60.0, Right:=220.0}
                    Alignment := vector2{X:= 0.5, Y:= 0.0}
                    
                canvas_slot:
                    Widget:=CustomPlayer.Upgrade4Title
                    ZOrder:=1
                    SizeToContent:=true
                    Anchors := anchors{Minimum := vector2{X:= 0.5, Y:= 0.0}, Maximum := vector2{X:= 0.5, Y:= 0.0}}
                    Offsets:=margin{Top:=282.0,Left:=664.0}
                    Alignment := vector2{X:= 0.5, Y:= 0.0}
                canvas_slot:
                    Widget:=CustomPlayer.Upgrade4Level
                    ZOrder:=1
                    SizeToContent:=true
                    Anchors := anchors{Minimum := vector2{X:= 0.5, Y:= 0.0}, Maximum := vector2{X:= 0.5, Y:= 0.0}}
                    Offsets:=margin{Top:=341.0,Left:=664.0}
                    Alignment := vector2{X:= 0.5, Y:= 0.0}
                canvas_slot:
                    Widget:=CustomPlayer.Upgrade4Icon
                    ZOrder:=1
                    SizeToContent:=false
                    Anchors := anchors{Minimum := vector2{X:= 0.5, Y:= 0.0}, Maximum := vector2{X:= 0.5, Y:= 0.0}}
                    Offsets:=margin{Top:=421.0,Left:=664.0, Bottom:=140.0,Right:=140.0}
                    Alignment := vector2{X:= 0.5, Y:= 0.0}
                canvas_slot:
                    Widget:=CustomPlayer.Upgrade4Desc
                    ZOrder:=1
                    SizeToContent:=true
                    Anchors := anchors{Minimum := vector2{X:= 0.5, Y:= 0.0}, Maximum := vector2{X:= 0.5, Y:= 0.0}}
                    Offsets:=margin{Top:=592.0,Left:=664.0}
                    Alignment := vector2{X:= 0.5, Y:= 0.0}
                canvas_slot:
                    #Widget:=Upgrade4Bonus
                    ZOrder:=1
                    SizeToContent:=true
                    Anchors := anchors{Minimum := vector2{X:= 0.5, Y:= 0.0}, Maximum := vector2{X:= 0.5, Y:= 0.0}}
                    Offsets:=margin{Top:=740.0,Left:=664.0}
                    Alignment := vector2{X:= 0.5, Y:= 0.0}
                    Widget := stack_box:
                        Orientation := orientation.Horizontal
                        Slots := array:
                            stack_box_slot:
                                Widget := CustomPlayer.Upgrade4Bonus   
                            stack_box_slot:
                                Widget := CustomPlayer.Upgrade4NextBonus
                canvas_slot:
                    #Widget:=CustomPlayer.Upgrade4CostText
                    ZOrder:=1
                    SizeToContent:=true
                    Anchors := anchors{Minimum := vector2{X:= 0.5, Y:= 0.0}, Maximum := vector2{X:= 0.5, Y:= 0.0}}
                    Offsets:=margin{Top:=800.0,Left:=552.0}
                    Alignment := vector2{X:= 0.5, Y:= 0.0}
                    Widget := stack_box:
                        Orientation := orientation.Horizontal
                        Slots := array:
                            stack_box_slot:
                                Widget := CustomPlayer.Upgrade4CostIcon   
                            stack_box_slot:
                                Widget := CustomPlayer.Upgrade4CostText
                canvas_slot:
                    ZOrder:=1
                    SizeToContent:=true
                    Anchors := anchors{Minimum := vector2{X:= 0.5, Y:= 0.0}, Maximum := vector2{X:= 0.5, Y:= 0.0}}
                    Offsets:=margin{Top:=800.0,Left:=840.0}
                    Alignment := vector2{X:= 1.0, Y:= 0.0}
                    Widget := stack_box:
                        Orientation := orientation.Horizontal
                        Slots := array:
                            stack_box_slot:
                                Widget := CustomPlayer.StoneIcon4    
                            stack_box_slot:
                                Widget := CustomPlayer.Upgrade4CostAmount
                canvas_slot:
                    Widget:=CustomPlayer.HUPUpgradeButton4
                    ZOrder:=1
                    SizeToContent:=true
                    Anchors := anchors{Minimum := vector2{X:= 0.5, Y:= 0.0}, Maximum := vector2{X:= 0.5, Y:= 0.0}}
                    Offsets:=margin{Top:=870.0,Left:=664.0, Bottom:=60.0, Right:=220.0}
                    Alignment := vector2{X:= 0.5, Y:= 0.0}


                canvas_slot:
                    Widget:=CustomPlayer.HUPCloseButton
                    ZOrder:=9
                    SizeToContent:=false
                    Alignment := vector2{X:= 0.5, Y:= 0.0}
                    Anchors := anchors{Minimum := vector2{X:= 0.5, Y:= 0.0}, Maximum := vector2{X:= 0.5, Y:= 0.0}}
                    Offsets:=margin{Top:=976.0,Left:=-548.0, Bottom:=50.0, Right:=200.0}
        return MyHUPCanvas



        