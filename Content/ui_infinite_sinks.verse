using { /Fortnite.com/Devices }
using { /Verse.org/Simulation }
using { /Verse.org/Colors }
using { /Verse.org/Verse }
using { /Verse.org/Random }
using { /Fortnite.com/Teams }
using { /UnrealEngine.com/Assets }
using { /Fortnite.com/Game }
using { /Fortnite.com/Characters }
using { /Fortnite.com/Playspaces }
using { /Fortnite.com/UI }
using { /UnrealEngine.com/Temporary/Diagnostics }
using { /UnrealEngine.com/Temporary/SpatialMath }
using { /UnrealEngine.com/Temporary/UI }
using { /Verse.org/Assets}

ui_infinite_sinks := class(creative_device):

    @editable
    gm : game_manager = game_manager{}

    @editable
    AudioManager : audio_manager = audio_manager{}

    @editable
    SinkButtons : []button_device = array{}

    var ShardConversionAmount : float = 1000000.0
    var WepMatConversionAmount : float = 25000.0
    var BossKeyConversionAmount : float = 100000000.0

    OnBegin<override>()<suspends>:void=
        for(Spawner : gm.PlayerSpawners):     
            Spawner.SpawnedEvent.Subscribe(OnPlayerJoined)
        for(i:int=0..SinkButtons.Length-1):
            if (SinkButton:=SinkButtons[i]):
                SinkButton.InteractedWithEvent.Subscribe(OpenSinkUI)

    OnPlayerJoined(Agent:agent):void=
        spawn{OnPlayerJoinedBuffer(Agent)}

    OnPlayerJoinedBuffer(Agent:agent)<suspends>:void=
        Sleep(1.0)
        if(CustomPlayer := gm.PlayersMap[player[Agent]]):
            if(CustomPlayer.PlayTimeSessionSeconds = 0.0):
                CustomPlayer.ISShardButton.OnClick().SubscribeAgent3(HandleConversion, 0)
                CustomPlayer.ISWepMatButton.OnClick().SubscribeAgent3(HandleConversion, 1)
                CustomPlayer.ISBossKeyButton.OnClick().SubscribeAgent3(HandleConversion, 2)
                CustomPlayer.ISCloseButton.OnClick().Subscribe(CloseSinkUI)
                CustomPlayer.ISAllShardButton.OnClick().SubscribeAgent3(HandelConvertAll, 0)
                CustomPlayer.ISAllWepMatButton.OnClick().SubscribeAgent3(HandelConvertAll, 1)
                CustomPlayer.ISAllBossKeyButton.OnClick().SubscribeAgent3(HandelConvertAll, 2)

    HandleConversion(Message:widget_message,Which:int):void=
        if(CustomPlayer := gm.PlayersMap[Message.Player]):
            if (Which = 0):
                SliderValue:= CustomPlayer.ISShardSlider.GetValue()
                ShardsToSink := ShardConversionAmount * SliderValue #Get Amount of shards spent
                if:
                    SliderValue >= 1.0 #Must be more than 1% to transfer
                    CustomPlayer.ShardAmount >= ShardsToSink #Must have enough to do
                then:
                    set CustomPlayer.SinkPrestigePower += SliderValue
                    set CustomPlayer.PrestigeHarvestBuff += SliderValue
                    CustomPlayer.Spend(ShardsToSink, "Shard", 0)
                    AudioManager.LeveledUp.Play(CustomPlayer.Player)

            if (Which = 1):
                SliderValue:= CustomPlayer.ISWepMatSlider.GetValue()
                SilverToSink := WepMatConversionAmount * SliderValue #Get Amount of wepmat spent
                if:
                    SliderValue >= 1.0
                    CustomPlayer.WepMatAmount >= SilverToSink
                then:
                    set CustomPlayer.SinkWepMatPower += SliderValue
                    CustomPlayer.Spend(SilverToSink, "WepMat", 0)
                    AudioManager.LeveledUp.Play(CustomPlayer.Player)

            if (Which = 2):
                SliderValue:= CustomPlayer.ISBossKeySlider.GetValue()
                BossKeyToSink := BossKeyConversionAmount * SliderValue #Get Amount of bosskeys spent
                if:
                    SliderValue >= 1.0
                    CustomPlayer.BossKeyAmount >= BossKeyToSink
                then:
                    set CustomPlayer.SinkBossKeyPower += SliderValue
                    CustomPlayer.Spend(BossKeyToSink, "BossKey", 0)
                    AudioManager.LeveledUp.Play(CustomPlayer.Player)
            UpdateSinkUI(CustomPlayer)

    HandelConvertAll(Message:widget_message,Which:int):void=
        if(CustomPlayer := gm.PlayersMap[Message.Player]):
            if (Which = 0):
                if (CustomPlayer.ShardAmount >= ShardConversionAmount):
                    set CustomPlayer.SinkPrestigePower += CustomPlayer.ShardAmount / ShardConversionAmount
                    set CustomPlayer.PrestigeHarvestBuff += CustomPlayer.ShardAmount / ShardConversionAmount
                    CustomPlayer.Spend(CustomPlayer.ShardAmount, "Shard", 0)
                    AudioManager.LeveledUp.Play(CustomPlayer.Player)
            if (Which = 1):
                if (CustomPlayer.WepMatAmount >= WepMatConversionAmount):
                    set CustomPlayer.SinkWepMatPower += CustomPlayer.WepMatAmount / WepMatConversionAmount
                    CustomPlayer.Spend(CustomPlayer.WepMatAmount, "WepMat", 0)
                    AudioManager.LeveledUp.Play(CustomPlayer.Player)
            if (Which = 2):
                if (CustomPlayer.BossKeyAmount >= BossKeyConversionAmount):
                    set CustomPlayer.SinkBossKeyPower += CustomPlayer.BossKeyAmount / BossKeyConversionAmount
                    CustomPlayer.Spend(CustomPlayer.BossKeyAmount, "BossKey", 0)
                    AudioManager.LeveledUp.Play(CustomPlayer.Player)
            UpdateSinkUI(CustomPlayer)
    
    OpenSinkUI(Agent:agent): void=
        if(CustomPlayer := gm.PlayersMap[Agent]):
            if(CustomPlayer.StagesUnlocked >= 19):
                if(MyUI:=CustomPlayer.UIMaps["InfSink"]?):
                else:
                    var MySinkCanvas : canvas = MakeCanvas(CustomPlayer)
                    CustomPlayer.AddFullHUD(MySinkCanvas)
                    if(set CustomPlayer.UIMaps["InfSink"] = option{MySinkCanvas}) {}
                    UpdateSinkUI(CustomPlayer)
            else:
                gm.ShowMessage(CustomPlayer.Player, "You are not ready!", 1)

    CloseSinkUI(Message : widget_message):void=
        if:
            CustomPlayer := gm.PlayersMap[Message.Player]
            Player:=Message.Player
            PlayerUI:=GetPlayerUI[Player]
            MyUI:=CustomPlayer.UIMaps["InfSink"]?
        then:
            PlayerUI.RemoveWidget(MyUI)
            MyUI.SetVisibility(widget_visibility.Collapsed)
            if(set CustomPlayer.UIMaps["InfSink"] = false){}

    UpdateSinkUI(CustomPlayer:Custom_Player):void=
        CustomPlayer.ISShardSlider.SetMinValue(1.0)
        CustomPlayer.ISShardSlider.SetMaxValue(100.0)
        CustomPlayer.ISWepMatSlider.SetMinValue(1.0)
        CustomPlayer.ISWepMatSlider.SetMaxValue(100.0)
        CustomPlayer.ISBossKeySlider.SetMinValue(1.0)
        CustomPlayer.ISBossKeySlider.SetMaxValue(100.0)
        CustomPlayer.ISConverstionRateTitle.SetText(StringToMessage("Silver Conversion: {NumberCondensor(WepMatConversionAmount)} = 1% Attack\nBoss Key Conversion: {NumberCondensor(BossKeyConversionAmount)} = 1% Boss Drop Amount\nShard Conversion: {NumberCondensor(ShardConversionAmount)} = 1% Strength & 0.01% Prestige Rewards"))
        CustomPlayer.ISShardAmountTitle.SetText(StringToMessage("{NumberCondensor(CustomPlayer.ShardAmount)}"))
        CustomPlayer.ISWeppMatAmountTitle.SetText(StringToMessage("{NumberCondensor(CustomPlayer.WepMatAmount)}"))
        CustomPlayer.ISBossKeyAmountTitle.SetText(StringToMessage("{NumberCondensor(CustomPlayer.BossKeyAmount)}"))

        #IF LESS THAN 1% IN EITHER BUT NOT 0
        if(CustomPlayer.SinkPrestigePower < 1.0 and CustomPlayer.SinkPrestigePower > 0.0):
            CustomPlayer.ISPrestigePercentTitle.SetText(StringToMessage("\<1% | \<1%"))
        else if((CustomPlayer.SinkPrestigePower / 100.0) < 1.0 and CustomPlayer.SinkPrestigePower > 0.0):
            CustomPlayer.ISPrestigePercentTitle.SetText(StringToMessage("{NumberCondensor(CustomPlayer.SinkPrestigePower)}% & \<1%"))
        else:
            CustomPlayer.ISPrestigePercentTitle.SetText(StringToMessage("{NumberCondensor(CustomPlayer.SinkPrestigePower)}% & {NumberCondensor(CustomPlayer.SinkPrestigePower/100.0)}%"))

        #IF LESS THAN 1% BUT NOT 0
        if(CustomPlayer.SinkWepMatPower < 1.0 and CustomPlayer.SinkWepMatPower > 0.0):
            CustomPlayer.ISWeppMatPercentTitle.SetText(StringToMessage("\<1%"))
        else:
            CustomPlayer.ISWeppMatPercentTitle.SetText(StringToMessage("{NumberCondensor(CustomPlayer.SinkWepMatPower)}%"))

        #IF LESS THAN 1% BUT NOT 0
        if(CustomPlayer.SinkBossKeyPower < 1.0 and CustomPlayer.SinkBossKeyPower > 0.0):
            CustomPlayer.ISBossKeyPercentTitle.SetText(StringToMessage("\<1%"))
        else:
            CustomPlayer.ISBossKeyPercentTitle.SetText(StringToMessage("{NumberCondensor(CustomPlayer.SinkBossKeyPower)}%"))

    MakeCanvas(CustomPlayer:Custom_Player):canvas=
        CustomPlayer.ISCloseButton.SetText(StringToMessage("Close"))
        CustomPlayer.ISShardButton.SetText(StringToMessage("Convert"))
        CustomPlayer.ISWepMatButton.SetText(StringToMessage("Convert"))
        CustomPlayer.ISBossKeyButton.SetText(StringToMessage("Convert"))
        CustomPlayer.ISAllShardButton.SetText(StringToMessage("Convert All"))
        CustomPlayer.ISAllWepMatButton.SetText(StringToMessage("Convert All"))
        CustomPlayer.ISAllBossKeyButton.SetText(StringToMessage("Convert All"))

        MyCanvas : canvas = canvas {
            Slots := array:
                # BACKGROUND
                canvas_slot:
                    Widget:=CustomPlayer.ISBG
                    ZOrder:=0
                    SizeToContent:=false
                    Anchors := anchors{Minimum := vector2{X:= 0.5, Y:= 0.5}, Maximum := vector2{X:= 0.5, Y:= 0.5}}
                    Offsets:=margin{Left:=-960.0,Top:=-540.0,Right:=1920.0,Bottom:=1080.0}
                # TITLE
                canvas_slot:
                    Widget:=CustomPlayer.ISTitle
                    ZOrder:=1
                    SizeToContent:=false
                    Anchors := anchors{Minimum := vector2{X:= 0.5, Y:= 0.5}, Maximum := vector2{X:= 0.5, Y:= 0.5}}
                    Offsets:=margin{Left:=0.0,Top:=-400.0,Right:=458.0,Bottom:=96.0}
                # BUTTONS
                canvas_slot:
                    Widget:=CustomPlayer.ISCloseButton
                    ZOrder:=1
                    SizeToContent:=false
                    Anchors := anchors{Minimum := vector2{X:= 0.5, Y:= 0.5}, Maximum := vector2{X:= 0.5, Y:= 0.5}}
                    Offsets:=margin{Left:=560.0,Top:=-380.0,Right:=260.0,Bottom:=50.0}
                canvas_slot:
                    Widget:=CustomPlayer.ISWepMatButton
                    ZOrder:=1
                    SizeToContent:=false
                    Anchors := anchors{Minimum := vector2{X:= 0.5, Y:= 0.5}, Maximum := vector2{X:= 0.5, Y:= 0.5}}
                    Offsets:=margin{Left:=220.0,Top:=-92.0,Right:=260.0,Bottom:=50.0} 
                canvas_slot:
                    Widget:=CustomPlayer.ISAllWepMatButton
                    ZOrder:=1
                    SizeToContent:=false
                    Anchors := anchors{Minimum := vector2{X:= 0.5, Y:= 0.5}, Maximum := vector2{X:= 0.5, Y:= 0.5}}
                    Offsets:=margin{Left:=560.0,Top:=-92.0,Right:=260.0,Bottom:=50.0}
                canvas_slot:
                    Widget:=CustomPlayer.ISShardButton
                    ZOrder:=1
                    SizeToContent:=false
                    Anchors := anchors{Minimum := vector2{X:= 0.5, Y:= 0.5}, Maximum := vector2{X:= 0.5, Y:= 0.5}}
                    Offsets:=margin{Left:=220.0,Top:=120.0,Right:=260.0,Bottom:=50.0}
                canvas_slot:
                    Widget:=CustomPlayer.ISAllShardButton
                    ZOrder:=1
                    SizeToContent:=false
                    Anchors := anchors{Minimum := vector2{X:= 0.5, Y:= 0.5}, Maximum := vector2{X:= 0.5, Y:= 0.5}}
                    Offsets:=margin{Left:=560.0,Top:=120.0,Right:=260.0,Bottom:=50.0}
                canvas_slot:
                    Widget:=CustomPlayer.ISBossKeyButton
                    ZOrder:=1
                    SizeToContent:=false
                    Anchors := anchors{Minimum := vector2{X:= 0.5, Y:= 0.5}, Maximum := vector2{X:= 0.5, Y:= 0.5}}
                    Offsets:=margin{Left:=220.0,Top:=12.0,Right:=260.0,Bottom:=50.0}
                canvas_slot:
                    Widget:=CustomPlayer.ISAllBossKeyButton
                    ZOrder:=1
                    SizeToContent:=false
                    Anchors := anchors{Minimum := vector2{X:= 0.5, Y:= 0.5}, Maximum := vector2{X:= 0.5, Y:= 0.5}}
                    Offsets:=margin{Left:=560.0,Top:=12.0,Right:=260.0,Bottom:=50.0}
                # SLIDERS
                canvas_slot:
                    Widget:=CustomPlayer.ISWepMatSlider
                    ZOrder:=1
                    SizeToContent:=false
                    Anchors := anchors{Minimum := vector2{X:= 0.5, Y:= 0.5}, Maximum := vector2{X:= 0.5, Y:= 0.5}}
                    Offsets:=margin{Left:=-368.0,Top:=-92.0,Right:=500.0,Bottom:=78.0}
                canvas_slot:
                    Widget:=CustomPlayer.ISShardSlider
                    ZOrder:=1
                    SizeToContent:=false
                    Anchors := anchors{Minimum := vector2{X:= 0.5, Y:= 0.5}, Maximum := vector2{X:= 0.5, Y:= 0.5}}
                    Offsets:=margin{Left:=-368.0,Top:=120.0,Right:=500.0,Bottom:=78.0}
                canvas_slot:
                    Widget:=CustomPlayer.ISBossKeySlider
                    ZOrder:=1
                    SizeToContent:=false
                    Anchors := anchors{Minimum := vector2{X:= 0.5, Y:= 0.5}, Maximum := vector2{X:= 0.5, Y:= 0.5}}
                    Offsets:=margin{Left:=-368.0,Top:=12.0,Right:=500.0,Bottom:=78.0}
                # ICONS
                canvas_slot:
                    Widget:=CustomPlayer.ISWepMatIcon
                    ZOrder:=1
                    SizeToContent:=false
                    Anchors := anchors{Minimum := vector2{X:= 0.5, Y:= 0.5}, Maximum := vector2{X:= 0.5, Y:= 0.5}}
                    Offsets:=margin{Left:=-788.0,Top:=-92.0,Right:=50.0,Bottom:=50.0}
                canvas_slot:
                    Widget:=CustomPlayer.ISShardIcon
                    ZOrder:=1
                    SizeToContent:=false
                    Anchors := anchors{Minimum := vector2{X:= 0.5, Y:= 0.5}, Maximum := vector2{X:= 0.5, Y:= 0.5}}
                    Offsets:=margin{Left:=-788.0,Top:=120.0,Right:=50.0,Bottom:=50.0}
                canvas_slot:
                    Widget:=CustomPlayer.ISBossKeyIcon
                    ZOrder:=1
                    SizeToContent:=false
                    Anchors := anchors{Minimum := vector2{X:= 0.5, Y:= 0.5}, Maximum := vector2{X:= 0.5, Y:= 0.5}}
                    Offsets:=margin{Left:=-788.0,Top:=12.0,Right:=50.0,Bottom:=50.0}
                canvas_slot:
                    Widget:=CustomPlayer.ISWepMatPercentIcon
                    ZOrder:=1
                    SizeToContent:=false
                    Anchors := anchors{Minimum := vector2{X:= 0.5, Y:= 0.5}, Maximum := vector2{X:= 0.5, Y:= 0.5}}
                    Offsets:=margin{Left:=-328.0,Top:=232.0,Right:=50.0,Bottom:=50.0}
                canvas_slot:
                    Widget:=CustomPlayer.ISBossKeyPercentIcon
                    ZOrder:=1
                    SizeToContent:=false
                    Anchors := anchors{Minimum := vector2{X:= 0.5, Y:= 0.5}, Maximum := vector2{X:= 0.5, Y:= 0.5}}
                    Offsets:=margin{Left:=-16.0,Top:=232.0,Right:=50.0,Bottom:=50.0}
                canvas_slot:
                    Widget:=CustomPlayer.ISShardPercentIcon
                    ZOrder:=1
                    SizeToContent:=false
                    Anchors := anchors{Minimum := vector2{X:= 0.5, Y:= 0.5}, Maximum := vector2{X:= 0.5, Y:= 0.5}}
                    Offsets:=margin{Left:=268.0,Top:=232.0,Right:=50.0,Bottom:=50.0}
                # TEXT
                canvas_slot:
                    Widget:=CustomPlayer.ISConverstionRateTitle
                    ZOrder:=1
                    SizeToContent:=true
                    Alignment := vector2{X := 0.0, Y := 0.0}
                    Anchors := anchors{Minimum := vector2{X:= 0.5, Y:= 0.5}, Maximum := vector2{X:= 0.5, Y:= 0.5}}
                    Offsets:=margin{Left:=-360.0,Top:=-244.0}
                canvas_slot:
                    Widget:=CustomPlayer.ISWeppMatAmountTitle
                    ZOrder:=1
                    SizeToContent:=true
                    Alignment := vector2{X := 0.0, Y := 0.0}
                    Anchors := anchors{Minimum := vector2{X:= 0.5, Y:= 0.5}, Maximum := vector2{X:= 0.5, Y:= 0.5}}
                    Offsets:=margin{Left:=-708.0,Top:=-92.0}
                canvas_slot:
                    Widget:=CustomPlayer.ISShardAmountTitle
                    ZOrder:=1
                    SizeToContent:=true
                    Alignment := vector2{X := 0.0, Y := 0.0}
                    Anchors := anchors{Minimum := vector2{X:= 0.5, Y:= 0.5}, Maximum := vector2{X:= 0.5, Y:= 0.5}}
                    Offsets:=margin{Left:=-708.0,Top:=120.0}
                canvas_slot:
                    Widget:=CustomPlayer.ISBossKeyAmountTitle
                    ZOrder:=1
                    SizeToContent:=true
                    Alignment := vector2{X := 0.0, Y := 0.0}
                    Anchors := anchors{Minimum := vector2{X:= 0.5, Y:= 0.5}, Maximum := vector2{X:= 0.5, Y:= 0.5}}
                    Offsets:=margin{Left:=-708.0,Top:=12.0}
                canvas_slot:
                    Widget:=CustomPlayer.ISWeppMatPercentTitle
                    ZOrder:=1
                    SizeToContent:=true
                    Alignment := vector2{X := 0.0, Y := 0.0}
                    Anchors := anchors{Minimum := vector2{X:= 0.5, Y:= 0.5}, Maximum := vector2{X:= 0.5, Y:= 0.5}}
                    Offsets:=margin{Left:=-252.0,Top:=236.0}
                canvas_slot:
                    Widget:=CustomPlayer.ISBossKeyPercentTitle
                    ZOrder:=1
                    SizeToContent:=true
                    Alignment := vector2{X := 0.0, Y := 0.0}
                    Anchors := anchors{Minimum := vector2{X:= 0.5, Y:= 0.5}, Maximum := vector2{X:= 0.5, Y:= 0.5}}
                    Offsets:=margin{Left:=60.0,Top:=236.0}
                canvas_slot:
                    Widget:=CustomPlayer.ISPrestigePercentTitle
                    ZOrder:=1
                    SizeToContent:=true
                    Alignment := vector2{X := 0.0, Y := 0.0}
                    Anchors := anchors{Minimum := vector2{X:= 0.5, Y:= 0.5}, Maximum := vector2{X:= 0.5, Y:= 0.5}}
                    Offsets:=margin{Left:=344.0,Top:=236.0}
        }