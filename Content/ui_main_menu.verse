using { /Fortnite.com/Devices }
using { /Verse.org/Simulation }
using { /Verse.org/Colors }
using { /Verse.org/Verse }
using { /Verse.org/Random }
using { /Fortnite.com/Teams }
using { /UnrealEngine.com/Assets }
using { /Fortnite.com/Game }
using { /Fortnite.com/Characters }
using { /Fortnite.com/Playspaces }
using { /Fortnite.com/UI }
using { /UnrealEngine.com/Temporary/Diagnostics }
using { /UnrealEngine.com/Temporary/SpatialMath }
using { /UnrealEngine.com/Temporary/UI }
using { /Verse.org/Assets}
using { /Verse.org/Colors/NamedColors }

ui_main_menu := class(creative_device):

    @editable
    gm : game_manager = game_manager{}

    @editable
    AudioManager : audio_manager = audio_manager{}

    @editable
    Drill : golden_drill_manager = golden_drill_manager{}

    @editable
    RemoteManager : signal_remote_manager_device = signal_remote_manager_device{}

    @editable
    TPToQuarry : teleporter_device = teleporter_device{}


    OnBegin<override>()<suspends>:void=        
        RemoteManager.SecondarySignalEvent.Subscribe(OpenMainMenuUI)

        #THESE NEED TO BE HERE AND NOWHERE ELSE
        <#if you're subscribing every time you open the subscribe function
        will stack so it will call the same function multiple times for no reason#>

        for(Spawner : gm.PlayerSpawners):     
            Spawner.SpawnedEvent.Subscribe(OnPlayerJoined)

    OnPlayerJoined(Agent:agent):void=
        spawn{OnPlayerJoinedBuffer(Agent)}

    OnPlayerJoinedBuffer(Agent:agent)<suspends>:void=
        Sleep(1.0)
        if(CustomPlayer := gm.PlayersMap[player[Agent]]):
            
            if(CustomPlayer.PlayTimeSessionSeconds = 0.0):
                CustomPlayer.GSMenuTitle.SetText(StringToMessage("Menu"))
                CustomPlayer.GBMenuTitle.SetText(StringToMessage("Menu"))

                CustomPlayer.AchievementsBtn.OnClick().Subscribe(HandleAchButtonClick)
                CustomPlayer.PlayerStatsBtn.OnClick().Subscribe(HandleStatsButtonClick)
                CustomPlayer.PetsBtn.OnClick().Subscribe(HandlePetButtonClick)
                CustomPlayer.InventoryBtn.OnClick().Subscribe(HandleInvButtonClick)

                CustomPlayer.NumbersBtn.OnClick().Subscribe(HandleNNButtonClick)
                CustomPlayer.MapBtn.OnClick().Subscribe(HandleMapButtonClick)
                CustomPlayer.TeleportBtn.OnClick().Subscribe(TeleportToQuarry)
                CustomPlayer.PrestigeBtn.OnClick().Subscribe(HandlePrestigeButtonClick)

                CustomPlayer.HideNumbersBtn.OnClick().Subscribe(HideNumbers)
                CustomPlayer.ReloadHUDBtn.OnClick().Subscribe(ReloadHUD)
                CustomPlayer.MuteMusicBtn.OnClick().Subscribe(MuteMusic)
                CustomPlayer.DrillBtn.OnClick().Subscribe(HandleDrillButtonClick)


                CustomPlayer.CloseBtn.OnClick().SubscribeAgent3(HandleMenuExit, "Menu") 
                CustomPlayer.CloseAchButton.OnClick().SubscribeAgent3(HandleMenuExit, "ACH")
                CustomPlayer.ClosePetMenuBtn.OnClick().SubscribeAgent3(HandleMenuExit, "Pets")
                CustomPlayer.CloseInvBtn.OnClick().SubscribeAgent3(HandleMenuExit, "INV")
                CustomPlayer.GBCloseBtn.OnClick().SubscribeAgent3(HandleMenuExit, "STAT")
                CustomPlayer.GSCloseBtn.OnClick().SubscribeAgent3(HandleMenuExit, "GS")
                CustomPlayer.NNCloseButton.OnClick().SubscribeAgent3(HandleMenuExit, "Numbers")
                CustomPlayer.InfoCloseButton.OnClick().SubscribeAgent3(HandleMenuExit, "Info")
                CustomPlayer.MapCloseButton.OnClick().SubscribeAgent3(HandleMenuExit, "CaveMap")
                CustomPlayer.PrestigeCloseButton.OnClick().SubscribeAgent3(HandleMenuExit, "Prestige")

                CustomPlayer.UnequipPetsBtn.OnClick().SubscribeAgent3(HandleMenuExit, "Pets")
                CustomPlayer.HidePetsBtn.OnClick().SubscribeAgent3(HandleMenuExit, "Pets")

                #Populating My Arrays so I can use for loops instead of manual code
                set CustomPlayer.AchHUDTitles = array{CustomPlayer.Ach1Title,CustomPlayer.Ach2Title,CustomPlayer.Ach3Title,CustomPlayer.Ach4Title,CustomPlayer.Ach5Title,CustomPlayer.Ach6Title,CustomPlayer.Ach7Title,CustomPlayer.Ach8Title}
                set CustomPlayer.AchHUDDescs = array{CustomPlayer.Ach1Desc,CustomPlayer.Ach2Desc,CustomPlayer.Ach3Desc,CustomPlayer.Ach4Desc,CustomPlayer.Ach5Desc,CustomPlayer.Ach6Desc,CustomPlayer.Ach7Desc,CustomPlayer.Ach8Desc}
                set CustomPlayer.AchHUDProgress = array{CustomPlayer.Ach1Progress,CustomPlayer.Ach2Progress,CustomPlayer.Ach3Progress,CustomPlayer.Ach4Progress,CustomPlayer.Ach5Progress,CustomPlayer.Ach6Progress,CustomPlayer.Ach7Progress,CustomPlayer.Ach8Progress}
                set CustomPlayer.AchBarColorTops = array{CustomPlayer.Ach1BarColorTop,CustomPlayer.Ach2BarColorTop,CustomPlayer.Ach3BarColorTop,CustomPlayer.Ach4BarColorTop,CustomPlayer.Ach5BarColorTop,CustomPlayer.Ach6BarColorTop,CustomPlayer.Ach7BarColorTop,CustomPlayer.Ach8BarColorTop}
                set CustomPlayer.EquipImages = {CustomPlayer.Equip1Image, CustomPlayer.Equip2Image, CustomPlayer.Equip3Image, CustomPlayer.Equip4Image, CustomPlayer.Equip5Image, CustomPlayer.Equip6Image}
                set CustomPlayer.PetImageTextureBlocks = array{CustomPlayer.Equip1Image,CustomPlayer.Equip2Image,CustomPlayer.Equip3Image,CustomPlayer.Equip4Image,CustomPlayer.Equip5Image,CustomPlayer.Equip6Image}
                set CustomPlayer.EquipButtons = array{CustomPlayer.EquipSlot1Btn,CustomPlayer.EquipSlot2Btn,CustomPlayer.EquipSlot3Btn,CustomPlayer.EquipSlot4Btn,CustomPlayer.EquipSlot5Btn,CustomPlayer.EquipSlot6Btn}
                set CustomPlayer.PetButtons = array{CustomPlayer.Pet1Btn,CustomPlayer.Pet2Btn,CustomPlayer.Pet3Btn,CustomPlayer.Pet4Btn,CustomPlayer.Pet5Btn,CustomPlayer.Pet6Btn}
                set CustomPlayer.InvEggTexts = array{CustomPlayer.InvEgg1Text,CustomPlayer.InvEgg2Text,CustomPlayer.InvEgg3Text,CustomPlayer.InvEgg4Text,CustomPlayer.InvEgg5Text,CustomPlayer.InvEgg6Text}
                set CustomPlayer.GBTitles = array{CustomPlayer.GBTitle1,CustomPlayer.GBTitle2,CustomPlayer.GBTitle3,CustomPlayer.GBTitle4,CustomPlayer.GBTitle5,CustomPlayer.GBTitle6,CustomPlayer.GBTitle7,CustomPlayer.GBTitle8,CustomPlayer.GBTitle9,CustomPlayer.GBTitle10}
                set CustomPlayer.GBBtns = array{CustomPlayer.GBGBBtn,CustomPlayer.GBGenStatsBtn,CustomPlayer.GBGameStatsBtn}
                set CustomPlayer.GSBtns = array{CustomPlayer.GSGBBtn,CustomPlayer.GSGenStatsBtn,CustomPlayer.GSGameStatsBtn}
                set CustomPlayer.PetLevelsText = array{CustomPlayer.Pet1Level, CustomPlayer.Pet2Level, CustomPlayer.Pet3Level, CustomPlayer.Pet4Level, CustomPlayer.Pet5Level, CustomPlayer.Pet6Level}
                set CustomPlayer.GSTitles = array{CustomPlayer.GSTitle1,CustomPlayer.GSTitle2,CustomPlayer.GSTitle3,CustomPlayer.GSTitle4,CustomPlayer.GSTitle5,CustomPlayer.GSTitle6,CustomPlayer.GSTitle7,CustomPlayer.GSTitle8,CustomPlayer.GSTitle9, CustomPlayer.GSTitle10}
                set CustomPlayer.GSIconBlock = array{CustomPlayer.GSIcon1,CustomPlayer.GSIcon2,CustomPlayer.GSIcon3,CustomPlayer.GSIcon4,CustomPlayer.GSIcon5,CustomPlayer.GSIcon6,CustomPlayer.GSIcon7,CustomPlayer.GSIcon8}
                
                set CustomPlayer.SelectButtons = array{CustomPlayer.EquipSlot1Btn,CustomPlayer.EquipSlot2Btn,CustomPlayer.EquipSlot3Btn,CustomPlayer.EquipSlot4Btn,CustomPlayer.EquipSlot5Btn,CustomPlayer.EquipSlot6Btn}

                set CustomPlayer.AllPetTexts = array{CustomPlayer.UnequipPetsText,CustomPlayer.PetTotalTitle,CustomPlayer.GatherTotalText,CustomPlayer.GatherTotalAmount,CustomPlayer.StrengthText,CustomPlayer.StrengthAmount,CustomPlayer.CritRateText,CustomPlayer.CritRateAmount,CustomPlayer.CritDmgText,CustomPlayer.CritDmgAmount,CustomPlayer.FinalStrikeText,CustomPlayer.FinalStrikeAmount,CustomPlayer.Pet1Level,CustomPlayer.Pet1Tier,CustomPlayer.Pet2Level,CustomPlayer.Pet2Tier,CustomPlayer.Pet3Level,CustomPlayer.Pet3Tier,CustomPlayer.Pet4Level,CustomPlayer.Pet4Tier,CustomPlayer.Pet5Level,CustomPlayer.Pet5Tier,CustomPlayer.Pet6Level,CustomPlayer.Pet6Tier}
                set CustomPlayer.AllPrestigeTexts = array{CustomPlayer.InfoText,CustomPlayer.KeepTitle,CustomPlayer.KeepText,CustomPlayer.ResetTitle,CustomPlayer.ResetText,CustomPlayer.NextPrestigeTitle,CustomPlayer.NextPrestigeAmount,CustomPlayer.CurrentPrestigeLevelText, CustomPlayer.PrestigeNowTitle,CustomPlayer.PrestigeNowShards,CustomPlayer.PrestigeNowGather,CustomPlayer.PrestigeNowText}
                set CustomPlayer.AllNNTexts = array{CustomPlayer.Row1Letter,CustomPlayer.Row1Number,CustomPlayer.Row2Letter,CustomPlayer.Row2Number,CustomPlayer.Row3Letter,CustomPlayer.Row3Number,CustomPlayer.Row4Letter,CustomPlayer.Row4Number}
                set CustomPlayer.AllGSTexts = array{CustomPlayer.GSTitleText,CustomPlayer.GSHeadingTitle1,CustomPlayer.GSHeadingTitle2,CustomPlayer.GSTitle1,CustomPlayer.GSTitle2,CustomPlayer.GSTitle3,CustomPlayer.GSTitle4,CustomPlayer.GSTitle5,CustomPlayer.GSTitle6,CustomPlayer.GSTitle7,CustomPlayer.GSTitle8,CustomPlayer.GSTitle9, CustomPlayer.GSTitle10, CustomPlayer.GSAmount1,CustomPlayer.GSAmount2,CustomPlayer.GSAmount3,CustomPlayer.GSAmount4,CustomPlayer.GSAmount5,CustomPlayer.GSAmount6,CustomPlayer.GSAmount7,CustomPlayer.GSAmount8,CustomPlayer.GSAmount9, CustomPlayer.GSAmount10}
                set CustomPlayer.AllGBTexts = array{CustomPlayer.GBTitleText,CustomPlayer.GBTitle1,CustomPlayer.GBTitle2,CustomPlayer.GBTitle3,CustomPlayer.GBTitle4,CustomPlayer.GBTitle5,CustomPlayer.GBTitle6,CustomPlayer.GBTitle7,CustomPlayer.GBTitle8,CustomPlayer.GBTitle9,CustomPlayer.GBTitle10,CustomPlayer.GBAmount1,CustomPlayer.GBAmount2,CustomPlayer.GBAmount3,CustomPlayer.GBAmount4,CustomPlayer.GBAmount5,CustomPlayer.GBAmount6,CustomPlayer.GBAmount7,CustomPlayer.GBAmount8,CustomPlayer.GBAmount9,CustomPlayer.GBAmount10}
                set CustomPlayer.AllInvTexts = array{CustomPlayer.InvStoneText,CustomPlayer.InvShardText,CustomPlayer.InvWepMatText,CustomPlayer.InvBossKeyText,CustomPlayer.InvEgg1Text,CustomPlayer.InvEgg2Text,CustomPlayer.InvEgg3Text,CustomPlayer.InvEgg4Text,CustomPlayer.InvEgg5Text,CustomPlayer.InvEgg6Text}
    
                for(i:int=0..CustomPlayer.GBBtns.Length):
                    if(Button:=CustomPlayer.GBBtns[i], Text:=GBBtnTitles[i]):
                        Button.SetText(StringToMessage("{Text}"))
                        Button.OnClick().SubscribeAgent3(HandlePSButtonClick,i)

                for(i:int=0..CustomPlayer.GSBtns.Length):
                    if(Button:=CustomPlayer.GSBtns[i], Text:=GBBtnTitles[i]):
                        Button.SetText(StringToMessage("{Text}"))
                        Button.OnClick().SubscribeAgent3(HandlePSButtonClick,i)
            #PlayerLoadUIs(CustomPlayer)

# MAIN MENU UI
    ToMessage<localizes>(String : string) : message = "{String}"

    OpenMainMenuUI(Agent : agent):void=
        if(CustomPlayer := gm.PlayersMap[Agent]):
            if(CustomPlayer.CurrentBoss = -1):
                AddMainMenuUI(CustomPlayer)
            else:
                gm.ShowMessage(CustomPlayer.Player, "Can't use while in boss fight!", 1)
    
    HandleMenuExit(Message: widget_message, WhichMenu:string) : void=
        if(CustomPlayer := gm.PlayersMap[Message.Player]):
            if(Player:=Message.Player, PlayerUI:=GetPlayerUI[Player]):
                if(MyUI:=CustomPlayer.UIMaps[WhichMenu]?):
                    PlayerUI.RemoveWidget(MyUI)
                    MyUI.SetVisibility(widget_visibility.Collapsed)
                    if(set CustomPlayer.UIMaps[WhichMenu] = false){}

    HandlePetButtonClick(Message: widget_message):void=
        HandleMenuExit(Message, "Menu")
        if(CustomPlayer := gm.PlayersMap[Message.Player]):
            AddPetMenuUI(CustomPlayer)

    HandleInvButtonClick(Message: widget_message):void=
        HandleMenuExit(Message, "Menu")
        if(CustomPlayer := gm.PlayersMap[Message.Player]):
            AddInvMenuUI(CustomPlayer)
    
    HandleStatsButtonClick(Message: widget_message):void=
        HandleMenuExit(Message, "Menu")
        if(CustomPlayer := gm.PlayersMap[Message.Player]):
            AddGatherBreakdownMenuUI(CustomPlayer)

    HandleInfoButtonClick(Message: widget_message):void=
        HandleMenuExit(Message, "Menu")
        if(CustomPlayer := gm.PlayersMap[Message.Player]):
            AddInfoMenuUI(CustomPlayer)

    HandleMapButtonClick(Message: widget_message):void=
        HandleMenuExit(Message, "Menu")
        if(CustomPlayer := gm.PlayersMap[Message.Player]):
            if(CustomPlayer.ShowVFX = true):
                CustomPlayer.MapBtnBG.SetImage(UI.ButtonOff)
                gm.ShowMessage(CustomPlayer.Player, "VFX Disabled!", 0)
                set CustomPlayer.ShowVFX = false
            else:
                CustomPlayer.MapBtnBG.SetImage(UI.ButtonOn)
                gm.ShowMessage(CustomPlayer.Player, "VFX Enabled!", 0)
                set CustomPlayer.ShowVFX = true
        #    AddMapMenuUI(CustomPlayer)
    
    HandlePSButtonClick(Message: widget_message, whichButton : int):void=
        if(CustomPlayer := gm.PlayersMap[Message.Player]):
            if (whichButton = 0):
                if (CustomPlayer.CurrentStat = 0):
                else:
                    HandleMenuExit(Message, "GS")       
                    AddGatherBreakdownMenuUI(CustomPlayer)
            if (whichButton = 1):
                if (CustomPlayer.CurrentStat = 0):
                    HandleMenuExit(Message, "STAT")
                    AddGSMenuUI(CustomPlayer, whichButton)
                else if (CustomPlayer.CurrentStat = 1):
                else if (CustomPlayer.CurrentStat = 2):
                    HandleMenuExit(Message, "GS")     
                    AddGSMenuUI(CustomPlayer, whichButton)
            if (whichButton = 2):
                if (CustomPlayer.CurrentStat = 0):
                    HandleMenuExit(Message, "STAT")
                    AddGSMenuUI(CustomPlayer, whichButton)
                else if (CustomPlayer.CurrentStat = 1):
                    HandleMenuExit(Message, "GS")     
                    AddGSMenuUI(CustomPlayer,whichButton)
                else if (CustomPlayer.CurrentStat = 2):                    


    HandlePrestigeButtonClick(Message: widget_message):void=
        HandleMenuExit(Message, "Menu")     
        if(CustomPlayer := gm.PlayersMap[Message.Player]):
            AddPrestigeMenuUI(CustomPlayer)

    HandleAchButtonClick(Message: widget_message):void=
        HandleMenuExit(Message, "Menu")     
        if(CustomPlayer := gm.PlayersMap[Message.Player]):
            AddAchMenuUI(CustomPlayer)

    HandleNNButtonClick(Message: widget_message):void=
        HandleMenuExit(Message, "Menu")     
        if(CustomPlayer := gm.PlayersMap[Message.Player]):
            AddNNMenuUI(CustomPlayer)

    HandleDrillButtonClick(Message: widget_message):void=
        HandleMenuExit(Message, "Menu")     
        Drill.ActivateButton(Message.Player)


    TeleportToQuarry(Message: widget_message):void=
        HandleMenuExit(Message, "Menu")     
        if(CustomPlayer := gm.PlayersMap[Message.Player]):
            if(CustomPlayer.CurrentBoss = -1):
                TPToQuarry.Teleport(CustomPlayer.Player)
                Print("IN SPACE FALSE")
                set CustomPlayer.InSpace = false
                gm.SwitchMusic(CustomPlayer.Player, "Default")

    HideNumbers(Message: widget_message):void=
        HandleMenuExit(Message, "Menu")     
        if(CustomPlayer := gm.PlayersMap[Message.Player]):
            if(CustomPlayer.ShowNumbers = true):
                CustomPlayer.HideNumbersBtnBG.SetImage(UI.ButtonOff)
                set CustomPlayer.ShowNumbers = false
            else:
                CustomPlayer.HideNumbersBtnBG.SetImage(UI.ButtonOn)
                set CustomPlayer.ShowNumbers = true

    ReloadHUD(Message: widget_message):void=
        HandleMenuExit(Message, "Menu")     
        if(CustomPlayer := gm.PlayersMap[Message.Player]):
            CustomPlayer.RemoveUI()
            CustomPlayer.InitiateUI()

    MuteMusic(Message: widget_message):void=
        HandleMenuExit(Message, "Menu")   
        if(CustomPlayer := gm.PlayersMap[Message.Player]):  
            if(CustomPlayer.CurrentMusic <> "Muted"):
                CustomPlayer.MuteMusicBtnBG.SetImage(UI.ButtonOff)
                set CustomPlayer.CurrentMusic = "Muted"
                for(i:int=0..AudioManager.BGMs.Length):
                    if(BGM:=AudioManager.BGMs[i]):
                        BGM.Stop(CustomPlayer.Player)
            else:
                CustomPlayer.MuteMusicBtnBG.SetImage(UI.ButtonOn)
                set CustomPlayer.CurrentMusic = "Default"
                gm.SwitchMusic(CustomPlayer.Player, "Default")

    AddMainMenuUI(CustomPlayer : Custom_Player):void=
        if(PlayerUI := CustomPlayer.UIMaps["Menu"]?):
        else:
            MainMenuCanvas := CreateMainMenuUI(CustomPlayer)
            CustomPlayer.AddFullHUD(MainMenuCanvas)
            if(set CustomPlayer.UIMaps["Menu"] = option{MainMenuCanvas}) {}

    AddPrestigeMenuUI(CustomPlayer : Custom_Player): void=
        if(MyUI:=CustomPlayer.UIMaps["Prestige"]?):
        else:
            MyPrestigeCanvas := MakePrestigeCanvas(CustomPlayer)
            CustomPlayer.AddFullHUD(MyPrestigeCanvas)
            if(set CustomPlayer.UIMaps["Prestige"] = option{MyPrestigeCanvas}) {}
            UpdatePrestigeHUD(CustomPlayer)

    AddPetMenuUI(CustomPlayer : Custom_Player): void=
        if(PlayerUI := CustomPlayer.UIMaps["Pets"]?):
        else:
            PetMenuCanvas := CreatePetMenuUI(CustomPlayer)
            CustomPlayer.AddFullHUD(PetMenuCanvas)
            if(set CustomPlayer.UIMaps["Pets"] = option{PetMenuCanvas}) {}
            UpdatePetMenu(CustomPlayer)

    AddInvMenuUI(CustomPlayer : Custom_Player): void=
        if(PlayerUI := CustomPlayer.UIMaps["INV"]?):
        else:
            InvMenuCanvas := CreateInvUI(CustomPlayer)
            CustomPlayer.AddFullHUD(InvMenuCanvas)
            if(set CustomPlayer.UIMaps["INV"] = option{InvMenuCanvas}) {}
            UpdateInv(CustomPlayer)

    AddGatherBreakdownMenuUI(CustomPlayer : Custom_Player): void=
        set CustomPlayer.CurrentStat = 0
        if(PlayerUI := CustomPlayer.UIMaps["STAT"]?):
        else:
            StatMenuCanvas := CreateGBUI(CustomPlayer)
            CustomPlayer.AddFullHUD(StatMenuCanvas)
            if(set CustomPlayer.UIMaps["STAT"] = option{StatMenuCanvas}) {}
            UpdateGBUI(CustomPlayer)
    
    AddGSMenuUI(CustomPlayer : Custom_Player, whichPage : int): void=
        if (whichPage = 1):
            set CustomPlayer.CurrentStat = 1
            if(PlayerUI := CustomPlayer.UIMaps["GS"]?):
            else:
                StatMenuCanvas := CreateGSUI(CustomPlayer)
                CustomPlayer.AddFullHUD(StatMenuCanvas)
                if(set CustomPlayer.UIMaps["GS"] = option{StatMenuCanvas}) {}
                UpdateGSUI(CustomPlayer, whichPage)
        if (whichPage = 2):
            set CustomPlayer.CurrentStat = 2
            if(PlayerUI := CustomPlayer.UIMaps["GS"]?):
            else:
                StatMenuCanvas := CreateGSUI(CustomPlayer)
                CustomPlayer.AddFullHUD(StatMenuCanvas)
                if(set CustomPlayer.UIMaps["GS"] = option{StatMenuCanvas}) {}
                UpdateGSUI(CustomPlayer, whichPage)

    AddAchMenuUI(CustomPlayer : Custom_Player): void=
        if(PlayerUI := CustomPlayer.UIMaps["ACH"]?):
        else:
            AchMenuCanvas := MakeAchCanvas(CustomPlayer)
            CustomPlayer.AddFullHUD(AchMenuCanvas)
            if(set CustomPlayer.UIMaps["ACH"] = option{AchMenuCanvas}) {}
            UpdateAchievementHUD(CustomPlayer)

    AddNNMenuUI(CustomPlayer : Custom_Player): void=
        if(PlayerUI := CustomPlayer.UIMaps["Numbers"]?):
        else:
            NNMenuCanvas := MakeNNCanvas(CustomPlayer)
            CustomPlayer.AddFullHUD(NNMenuCanvas)
            if(set CustomPlayer.UIMaps["Numbers"] = option{NNMenuCanvas}) {}
            
    AddInfoMenuUI(CustomPlayer : Custom_Player): void=
        if(PlayerUI := CustomPlayer.UIMaps["Info"]?):
        else:
            InfoMenuCanvas := MakeInfoCanvas(CustomPlayer)
            CustomPlayer.AddFullHUD(InfoMenuCanvas)
            if(set CustomPlayer.UIMaps["Info"] = option{InfoMenuCanvas}) {}

    AddMapMenuUI(CustomPlayer : Custom_Player): void=
        if(PlayerUI := CustomPlayer.UIMaps["CaveMap"]?):
        else:
            MapMenuCanvas := MakeMapCanvas(CustomPlayer)
            CustomPlayer.AddFullHUD(MapMenuCanvas)
            if(set CustomPlayer.UIMaps["CaveMap"] = option{MapMenuCanvas}) {}

    CreateMainMenuUI(CustomPlayer:Custom_Player) : canvas=
        CustomPlayer.CloseBtn.SetText(StringToMessage("X"))

        CustomPlayer.AchievementsBtn.SetText(StringToMessage("Achievements"))
        CustomPlayer.PlayerStatsBtn.SetText(StringToMessage("Player Stats"))
        CustomPlayer.PetsBtn.SetText(StringToMessage("Pets"))
        CustomPlayer.InventoryBtn.SetText(StringToMessage("Inventory"))
        
        CustomPlayer.NumbersBtn.SetText(StringToMessage("E Notations"))
        CustomPlayer.MapBtn.SetText(StringToMessage("Toggle VFX"))
        CustomPlayer.TeleportBtn.SetText(StringToMessage("To Quarry"))
        CustomPlayer.PrestigeBtn.SetText(StringToMessage("Prestige"))

        CustomPlayer.HideNumbersBtn.SetText(StringToMessage("Toggle Numbers"))
        CustomPlayer.ReloadHUDBtn.SetText(StringToMessage("Reload HUD"))
        CustomPlayer.MuteMusicBtn.SetText(StringToMessage("Toggle Music"))
        CustomPlayer.DrillBtn.SetText(StringToMessage("Activate"))

        NewUI : canvas = canvas{
            Slots := array:
                canvas_slot:
                    Widget:=CustomPlayer.MainMenuBackGround
                    ZOrder:=0
                    SizeToContent:=false
                    Alignment := vector2{X := 0.5, Y := 0.0}
                    Anchors := anchors{Minimum := vector2{X:= 0.5, Y:= 0.0}, Maximum := vector2{X:= 0.5, Y:= 0.0}}
                    Offsets:=margin{Top:=0.0,Left:=0.0,Bottom:=1076.0,Right:=2048.0}
                canvas_slot:
                    Widget:=CustomPlayer.CloseBtn
                    ZOrder:=2
                    SizeToContent:=false
                    Anchors := anchors{Minimum := vector2{X:= 0.5, Y:= 0.0}, Maximum := vector2{X:= 0.5, Y:= 0.0}}
                    Offsets:=margin{Top:=120.0,Left:=500.0,Bottom:=56.0,Right:=76.0}
                canvas_slot:
                    Widget:=CustomPlayer.MainMenuTitle
                    ZOrder:=2
                    SizeToContent:=false
                    Alignment := vector2{X := 0.5, Y := 0.0}
                    Anchors := anchors{Minimum := vector2{X:= 0.5, Y:= 0.0}, Maximum := vector2{X:= 0.5, Y:= 0.0}}
                    Offsets:=margin{Top:=120.0,Left:=0.0,Bottom:=55.0,Right:=356.0}
               
                canvas_slot:
                    Widget:=CustomPlayer.AchievementsIcon
                    ZOrder:=2
                    SizeToContent:=false
                    Alignment := vector2{X := 0.5, Y := 0.0}
                    Anchors := anchors{Minimum := vector2{X:= 0.5, Y:= 0.0}, Maximum := vector2{X:= 0.5, Y:= 0.0}}
                    Offsets:=margin{Top:=212.0,Left:=-600.0, Bottom:=128.0, Right:=128.0}
                canvas_slot:
                    Widget:=CustomPlayer.AchievementsBtn
                    ZOrder:=2
                    SizeToContent:=false
                    Alignment := vector2{X := 0.5, Y := 0.0}
                    Anchors := anchors{Minimum := vector2{X:= 0.5, Y:= 0.0}, Maximum := vector2{X:= 0.5, Y:= 0.0}}
                    Offsets:=margin{Top:=372.0,Left:=-600.0, Bottom:= 50.0,Right:=300.0}
                canvas_slot:
                    Widget:=CustomPlayer.PlayerStatsIcon
                    ZOrder:=2
                    SizeToContent:=false
                    Alignment := vector2{X := 0.5, Y := 0.0}
                    Anchors := anchors{Minimum := vector2{X:= 0.5, Y:= 0.0}, Maximum := vector2{X:= 0.5, Y:= 0.0}}
                    Offsets:=margin{Top:=212.0,Left:=-200.0, Bottom:=128.0, Right:=128.0}
                canvas_slot:
                    Widget:=CustomPlayer.PlayerStatsBtn
                    ZOrder:=2
                    SizeToContent:=false
                    Alignment := vector2{X := 0.5, Y := 0.0}
                    Anchors := anchors{Minimum := vector2{X:= 0.5, Y:= 0.0}, Maximum := vector2{X:= 0.5, Y:= 0.0}}
                    Offsets:=margin{Top:=372.0,Left:=-200.0, Bottom:= 50.0,Right:=300.0}
                canvas_slot:
                    Widget:=CustomPlayer.PetsIcon
                    ZOrder:=2
                    SizeToContent:=false
                    Alignment := vector2{X := 0.5, Y := 0.0}
                    Anchors := anchors{Minimum := vector2{X:= 0.5, Y:= 0.0}, Maximum := vector2{X:= 0.5, Y:= 0.0}}
                    Offsets:=margin{Top:=212.0,Left:=200.0, Bottom:=128.0, Right:=128.0}
                canvas_slot:
                    Widget:=CustomPlayer.PetsBtn
                    ZOrder:=2
                    SizeToContent:=false
                    Alignment := vector2{X := 0.5, Y := 0.0}
                    Anchors := anchors{Minimum := vector2{X:= 0.5, Y:= 0.0}, Maximum := vector2{X:= 0.5, Y:= 0.0}}
                    Offsets:=margin{Top:=372.0,Left:=200.0, Bottom:= 50.0,Right:=300.0}
                canvas_slot:
                    Widget:=CustomPlayer.InventoryIcon
                    ZOrder:=2
                    SizeToContent:=false
                    Alignment := vector2{X := 0.5, Y := 0.0}
                    Anchors := anchors{Minimum := vector2{X:= 0.5, Y:= 0.0}, Maximum := vector2{X:= 0.5, Y:= 0.0}}
                    Offsets:=margin{Top:=212.0,Left:=600.0, Bottom:=128.0, Right:=128.0}
                canvas_slot:
                    Widget:=CustomPlayer.InventoryBtn
                    ZOrder:=2
                    SizeToContent:=false
                    Alignment := vector2{X := 0.5, Y := 0.0}
                    Anchors := anchors{Minimum := vector2{X:= 0.5, Y:= 0.0}, Maximum := vector2{X:= 0.5, Y:= 0.0}}
                    Offsets:=margin{Top:=372.0,Left:=600.0, Bottom:= 50.0,Right:=300.0}

                canvas_slot:
                    Widget:=CustomPlayer.NumbersIcon
                    ZOrder:=2
                    SizeToContent:=false
                    Alignment := vector2{X := 0.5, Y := 0.0}
                    Anchors := anchors{Minimum := vector2{X:= 0.5, Y:= 0.0}, Maximum := vector2{X:= 0.5, Y:= 0.0}}
                    Offsets:=margin{Top:=516.0,Left:=-600.0, Bottom:=128.0, Right:=128.0}
                canvas_slot:
                    Widget:=CustomPlayer.NumbersBtn
                    ZOrder:=2
                    SizeToContent:=false
                    Alignment := vector2{X := 0.5, Y := 0.0}
                    Anchors := anchors{Minimum := vector2{X:= 0.5, Y:= 0.0}, Maximum := vector2{X:= 0.5, Y:= 0.0}}
                    Offsets:=margin{Top:=672.0,Left:=-600.0, Bottom:= 50.0,Right:=300.0}
                canvas_slot:
                    Widget:=CustomPlayer.ReloadHUDIcon
                    ZOrder:=2
                    SizeToContent:=false
                    Alignment := vector2{X := 0.5, Y := 0.0}
                    Anchors := anchors{Minimum := vector2{X:= 0.5, Y:= 0.0}, Maximum := vector2{X:= 0.5, Y:= 0.0}}
                    Offsets:=margin{Top:=516.0,Left:=-200.0, Bottom:=128.0, Right:=128.0}
                canvas_slot:
                    Widget:=CustomPlayer.ReloadHUDBtn
                    ZOrder:=2
                    SizeToContent:=false
                    Alignment := vector2{X := 0.5, Y := 0.0}
                    Anchors := anchors{Minimum := vector2{X:= 0.5, Y:= 0.0}, Maximum := vector2{X:= 0.5, Y:= 0.0}}
                    Offsets:=margin{Top:=672.0,Left:=-200.0, Bottom:= 50.0,Right:=300.0}
                canvas_slot:
                    Widget:=CustomPlayer.TeleportIcon
                    ZOrder:=2
                    SizeToContent:=false
                    Alignment := vector2{X := 0.5, Y := 0.0}
                    Anchors := anchors{Minimum := vector2{X:= 0.5, Y:= 0.0}, Maximum := vector2{X:= 0.5, Y:= 0.0}}
                    Offsets:=margin{Top:=516.0,Left:=200.0, Bottom:=128.0, Right:=128.0}
                canvas_slot:
                    Widget:=CustomPlayer.TeleportBtn
                    ZOrder:=2
                    SizeToContent:=false
                    Alignment := vector2{X := 0.5, Y := 0.0}
                    Anchors := anchors{Minimum := vector2{X:= 0.5, Y:= 0.0}, Maximum := vector2{X:= 0.5, Y:= 0.0}}
                    Offsets:=margin{Top:=672.0,Left:=200.0, Bottom:= 50.0,Right:=300.0}
                canvas_slot:
                    Widget:=CustomPlayer.DrillIcon
                    ZOrder:=2
                    SizeToContent:=false
                    Alignment := vector2{X := 0.5, Y := 0.0}
                    Anchors := anchors{Minimum := vector2{X:= 0.5, Y:= 0.0}, Maximum := vector2{X:= 0.5, Y:= 0.0}}
                    Offsets:=margin{Top:=516.0,Left:=600.0, Bottom:=128.0, Right:=128.0}
                canvas_slot:
                    Widget:=CustomPlayer.DrillBtn
                    ZOrder:=2
                    SizeToContent:=false
                    Alignment := vector2{X := 0.5, Y := 0.0}
                    Anchors := anchors{Minimum := vector2{X:= 0.5, Y:= 0.0}, Maximum := vector2{X:= 0.5, Y:= 0.0}}
                    Offsets:=margin{Top:=672.0,Left:=600.0, Bottom:= 50.0,Right:=300.0}


                canvas_slot:
                    Widget:=CustomPlayer.HideNumbersIcon
                    ZOrder:=2
                    SizeToContent:=false
                    Alignment := vector2{X := 0.5, Y := 0.0}
                    Anchors := anchors{Minimum := vector2{X:= 0.5, Y:= 0.0}, Maximum := vector2{X:= 0.5, Y:= 0.0}}
                    Offsets:=margin{Top:=816.0,Left:=-600.0, Bottom:=128.0, Right:=128.0}
                canvas_slot:
                    Widget:=CustomPlayer.HideNumbersBtn
                    ZOrder:=2
                    SizeToContent:=false
                    Alignment := vector2{X := 0.5, Y := 0.0}
                    Anchors := anchors{Minimum := vector2{X:= 0.5, Y:= 0.0}, Maximum := vector2{X:= 0.5, Y:= 0.0}}
                    Offsets:=margin{Top:=972.0,Left:=-600.0, Bottom:= 50.0,Right:=300.0}
                canvas_slot:
                    Widget:=CustomPlayer.HideNumbersBtnBG
                    ZOrder:=2
                    SizeToContent:=false
                    Alignment := vector2{X := 0.5, Y := 0.0}
                    Anchors := anchors{Minimum := vector2{X:= 0.5, Y:= 0.0}, Maximum := vector2{X:= 0.5, Y:= 0.0}}
                    Offsets:=margin{Top:=972.0,Left:=-600.0, Bottom:= 50.0,Right:=300.0}
                canvas_slot:
                    Widget:=CustomPlayer.MapIcon
                    ZOrder:=2
                    SizeToContent:=false
                    Alignment := vector2{X := 0.5, Y := 0.0}
                    Anchors := anchors{Minimum := vector2{X:= 0.5, Y:= 0.0}, Maximum := vector2{X:= 0.5, Y:= 0.0}}
                    Offsets:=margin{Top:=816.0,Left:=-200.0, Bottom:=128.0, Right:=128.0}
                canvas_slot:
                    Widget:=CustomPlayer.MapBtn
                    ZOrder:=2
                    SizeToContent:=false
                    Alignment := vector2{X := 0.5, Y := 0.0}
                    Anchors := anchors{Minimum := vector2{X:= 0.5, Y:= 0.0}, Maximum := vector2{X:= 0.5, Y:= 0.0}}
                    Offsets:=margin{Top:=972.0,Left:=-200.0, Bottom:= 50.0,Right:=300.0}
                canvas_slot:
                    Widget:=CustomPlayer.MapBtnBG
                    ZOrder:=2
                    SizeToContent:=false
                    Alignment := vector2{X := 0.5, Y := 0.0}
                    Anchors := anchors{Minimum := vector2{X:= 0.5, Y:= 0.0}, Maximum := vector2{X:= 0.5, Y:= 0.0}}
                    Offsets:=margin{Top:=972.0,Left:=-200.0, Bottom:= 50.0,Right:=300.0}
                canvas_slot:
                    Widget:=CustomPlayer.MuteMusicIcon
                    ZOrder:=2
                    SizeToContent:=false
                    Alignment := vector2{X := 0.5, Y := 0.0}
                    Anchors := anchors{Minimum := vector2{X:= 0.5, Y:= 0.0}, Maximum := vector2{X:= 0.5, Y:= 0.0}}
                    Offsets:=margin{Top:=816.0,Left:=200.0, Bottom:=128.0, Right:=128.0}
                canvas_slot:
                    Widget:=CustomPlayer.MuteMusicBtn
                    ZOrder:=2
                    SizeToContent:=false
                    Alignment := vector2{X := 0.5, Y := 0.0}
                    Anchors := anchors{Minimum := vector2{X:= 0.5, Y:= 0.0}, Maximum := vector2{X:= 0.5, Y:= 0.0}}
                    Offsets:=margin{Top:=972.0,Left:=200.0, Bottom:= 50.0,Right:=300.0}
                canvas_slot:
                    Widget:=CustomPlayer.MuteMusicBtnBG
                    ZOrder:=2
                    SizeToContent:=false
                    Alignment := vector2{X := 0.5, Y := 0.0}
                    Anchors := anchors{Minimum := vector2{X:= 0.5, Y:= 0.0}, Maximum := vector2{X:= 0.5, Y:= 0.0}}
                    Offsets:=margin{Top:=972.0,Left:=200.0, Bottom:= 50.0,Right:=300.0}
                canvas_slot:
                    Widget:=CustomPlayer.PrestigeIcon
                    ZOrder:=2
                    SizeToContent:=false
                    Alignment := vector2{X := 0.5, Y := 0.0}
                    Anchors := anchors{Minimum := vector2{X:= 0.5, Y:= 0.0}, Maximum := vector2{X:= 0.5, Y:= 0.0}}
                    Offsets:=margin{Top:=816.0,Left:=600.0, Bottom:=128.0, Right:=128.0}
                canvas_slot:
                    Widget:=CustomPlayer.PrestigeBtn
                    ZOrder:=2
                    SizeToContent:=false
                    Alignment := vector2{X := 0.5, Y := 0.0}
                    Anchors := anchors{Minimum := vector2{X:= 0.5, Y:= 0.0}, Maximum := vector2{X:= 0.5, Y:= 0.0}}
                    Offsets:=margin{Top:=972.0,Left:=600.0, Bottom:= 50.0,Right:=300.0}
        }

# PETS MENU UI

    var petimages : []texture = array{UI.pet1, UI.pet2, UI.pet3, UI.pet4, UI.pet5, UI.pet6}

    UpdatePetMenu(CustomPlayer : Custom_Player):void=
        var PetGather : float = 0.0
        var PetStrength : float = 0.0
        var PetCritRate : float = 0.0
        var PetCritDmg : float = 0.0
        var PetFinalStrike : float = 0.0
        var HasCrit : logic = false
        var HasFinalStrike : logic = false

        for (i:int=0..CustomPlayer.PetEquipSlots.Length):
            if(EquipButton := CustomPlayer.EquipButtons[i]):
                if(CustomPlayer.PetEquipSlots[i] = true):
                    if(CustomPlayer.CurrentPetSlot = i):
                        EquipButton.SetText(ToMessage("Selected"))
                    else:
                        EquipButton.SetText(ToMessage("Select"))
                else:
                        EquipButton.SetText(ToMessage("Locked"))
        for (i:int=0..CustomPlayer.PetUnlocked.Length):
            if (PetButton := CustomPlayer.PetButtons[i]):
                if(CustomPlayer.PetUnlocked[i] = true):
                        PetButton.SetText(ToMessage("Equip"))
                else:
                        PetButton.SetText(ToMessage("Locked"))

        for (i:int=0..CustomPlayer.PetLevelsText.Length):
            if(PetLevelBlock := CustomPlayer.PetLevelsText[i], PetLevel := CustomPlayer.PetLevels[i]):
                PetLevelBlock.SetText(StringToMessage("Lv. {NumberCondensor(PetLevel*1.0)}"))

        for(i:int=0..CustomPlayer.PetSlots.Length):
            if(CustomPlayer.PetSlots[i] <> -1):
                if:
                    PetSlot := CustomPlayer.PetSlots[i]
                    PetLevel := CustomPlayer.PetLevels[PetSlot]
                    NewPetGather := CustomPlayer.PetGatherRates[PetSlot]
                    NewPetAttack := CustomPlayer.PetAttackRates[PetSlot]
                    NewPetCritRate := CustomPlayer.PetCritRates[PetSlot]
                    NewPetCritDMG := CustomPlayer.PetCritDMG[PetSlot]
                    NewPetFinalStrike :=  CustomPlayer.PetFinalStrike[PetSlot]
                then:
                    set PetGather += NewPetGather
                    set PetStrength += NewPetAttack
                    set PetCritRate += NewPetCritRate
                    set PetCritDmg += NewPetCritDMG
                    set PetFinalStrike += NewPetFinalStrike
  
                if (PetImage:=petimages[CustomPlayer.PetSlots[i]], EquipImage:=CustomPlayer.EquipImages[i]):
                    EquipImage.SetImage(PetImage)
                    #if(CustomPlayer.PetSlots[i] >= 2):
                    #    set HasCrit = true
                    #if(CustomPlayer.PetSlots[i] >= 4):
                    #    set HasFinalStrike = true
            else:
                if (EquipImage:=CustomPlayer.EquipImages[i]):
                    EquipImage.SetImage(UI.Blank)


        CustomPlayer.GatherTotalAmount.SetText(StringToMessage("{NumberCondensor(PetGather)}%"))
        CustomPlayer.StrengthAmount.SetText(StringToMessage("{NumberCondensor(PetStrength)}%"))
        CustomPlayer.CritRateAmount.SetText(StringToMessage("{NumberCondensor(PetCritRate)}%"))
        CustomPlayer.CritDmgAmount.SetText(StringToMessage("{NumberCondensor(PetCritDmg)}%"))
        CustomPlayer.FinalStrikeAmount.SetText(StringToMessage("{NumberCondensor(PetFinalStrike)}%"))
        if(PetFinalStrike >= 1000.0):
            CustomPlayer.FinalStrikeText.SetText(StringToMessage("FS:"))

    CreatePetMenuUI(CustomPlayer:Custom_Player) : canvas=
        CustomPlayer.ClosePetMenuBtn.SetText(ToMessage("X"))

        CustomPlayer.Pet1Tier.SetText(StringToMessage("Common"))
        CustomPlayer.Pet2Tier.SetText(StringToMessage("Uncommon"))
        CustomPlayer.Pet3Tier.SetText(StringToMessage("Rare"))
        CustomPlayer.Pet4Tier.SetText(StringToMessage("Epic"))
        CustomPlayer.Pet5Tier.SetText(StringToMessage("Legendary"))
        CustomPlayer.Pet6Tier.SetText(StringToMessage("Mythic"))

        CustomPlayer.UnequipPetsText.SetText(StringToMessage("Pets acting weird?"))
        CustomPlayer.UnequipPetsBtn.SetText(StringToMessage("Call Pets"))
        CustomPlayer.UnequipPetsBtn.SetEnabled(true)
        CustomPlayer.HidePetsBtn.SetText(StringToMessage("Hide Pets"))

        CustomPlayer.PetTotalTitle.SetText(StringToMessage("Total Pet buffs"))
        CustomPlayer.GatherTotalText.SetText(StringToMessage("Gather:"))
        CustomPlayer.StrengthText.SetText(StringToMessage("Attack %:"))
        CustomPlayer.CritRateText.SetText(StringToMessage("Crit Rate:"))
        CustomPlayer.CritDmgText.SetText(StringToMessage("Crit Dmg:"))
        CustomPlayer.FinalStrikeText.SetText(StringToMessage("Final Strike:"))
        
        CustomPlayer.GatherTotalAmount.SetText(StringToMessage("0%"))
        CustomPlayer.StrengthAmount.SetText(StringToMessage("0%"))
        CustomPlayer.CritRateAmount.SetText(StringToMessage("0%"))
        CustomPlayer.CritDmgAmount.SetText(StringToMessage("0%"))
        CustomPlayer.FinalStrikeAmount.SetText(StringToMessage("0%"))

        for(TextBlock : CustomPlayer.AllPetTexts):
            TextBlock.SetShadowColor(CustomPlayer.ShadowDefaults.DefaultShadowColor)
            TextBlock.SetShadowOffset(CustomPlayer.ShadowDefaults.DefaultShadowOffset)
            TextBlock.SetShadowOpacity(1.0)

        NewUI : canvas = canvas{
            Slots := array:
                canvas_slot:
                    Widget := CustomPlayer.PetMenuBackGround
                    ZOrder := 1
                    SizeToContent := false
                    Anchors := anchors{Minimum := vector2{X:= 0.5, Y:= 0.0}, Maximum := vector2{X:= 0.5, Y:= 0.0}}
                    Offsets := margin{Left := 0.0, Top:= 0.0, Right:= 2048.0, Bottom:= 1076.0}
                    Alignment := vector2{X:= 0.5, Y:= 0.0}
                canvas_slot:
                    Widget := CustomPlayer.ClosePetMenuBtn
                    ZOrder := 2
                    SizeToContent := false
                    Anchors := anchors{Minimum := vector2{X:= 0.5, Y:= 0.0}, Maximum := vector2{X:= 0.5, Y:= 0.0}}
                    Offsets := margin{Left := 680.0, Top:=676.0, Right:= 300.0, Bottom:= 56.0}
                    Alignment := vector2{X:= 0.5, Y:= 0.0}
                canvas_slot:
                    Widget := CustomPlayer.PetsTitle
                    ZOrder := 2
                    SizeToContent := false
                    Anchors := anchors{Minimum := vector2{X:= 0.5, Y:= 0.0}, Maximum := vector2{X:= 0.5, Y:= 0.0}}
                    Offsets := margin{Left := 0.0, Top:= 120.0, Right:= 159.0, Bottom:= 56.0}
                    Alignment := vector2{X:= 0.5, Y:= 0.0}

                canvas_slot:
                    Widget := CustomPlayer.EquipSlot1Btn
                    ZOrder := 2
                    SizeToContent := false
                    Anchors := anchors{Minimum := vector2{X:= 0.5, Y:= 0.0}, Maximum := vector2{X:= 0.5, Y:= 0.0}}
                    Offsets := margin{Left := -792.0, Top:= 268.0, Right:= 250.0, Bottom:= 50.0}
                canvas_slot:
                    Widget := CustomPlayer.Equip1Image
                    ZOrder := 2
                    SizeToContent := false
                    Anchors := anchors{Minimum := vector2{X:= 0.5, Y:= 0.0}, Maximum := vector2{X:= 0.5, Y:= 0.0}}
                    Offsets := margin{Left := -508.0, Top:= 252.0, Right:= 80.0, Bottom:= 80.0}
                canvas_slot:
                    Widget := CustomPlayer.EquipSlot2Btn
                    ZOrder := 2
                    SizeToContent := false
                    Anchors := anchors{Minimum := vector2{X:= 0.5, Y:= 0.0}, Maximum := vector2{X:= 0.5, Y:= 0.0}}
                    Offsets := margin{Left := -792.0, Top:= 392.0, Right:= 250.0, Bottom:= 50.0}
                canvas_slot:
                    Widget := CustomPlayer.Equip2Image
                    ZOrder := 2
                    SizeToContent := false
                    Anchors := anchors{Minimum := vector2{X:= 0.5, Y:= 0.0}, Maximum := vector2{X:= 0.5, Y:= 0.0}}
                    Offsets := margin{Left := -508.0, Top:= 380.0, Right:= 80.0, Bottom:= 80.0}
                canvas_slot:
                    Widget := CustomPlayer.EquipSlot3Btn
                    ZOrder := 2
                    SizeToContent := false
                    Anchors := anchors{Minimum := vector2{X:= 0.5, Y:= 0.0}, Maximum := vector2{X:= 0.5, Y:= 0.0}}
                    Offsets := margin{Left := -792.0, Top:= 528.0, Right:= 250.0, Bottom:= 50.0}
                canvas_slot:
                    Widget := CustomPlayer.Equip3Image
                    ZOrder := 2
                    SizeToContent := false
                    Anchors := anchors{Minimum := vector2{X:= 0.5, Y:= 0.0}, Maximum := vector2{X:= 0.5, Y:= 0.0}}
                    Offsets := margin{Left := -508.0, Top:= 512.0, Right:= 80.0, Bottom:= 80.0}
                canvas_slot:
                    Widget := CustomPlayer.EquipSlot4Btn
                    ZOrder := 2
                    SizeToContent := false
                    Anchors := anchors{Minimum := vector2{X:= 0.5, Y:= 0.0}, Maximum := vector2{X:= 0.5, Y:= 0.0}}
                    Offsets := margin{Left := -792.0, Top:= 656.0, Right:= 250.0, Bottom:= 50.0}
                canvas_slot:
                    Widget := CustomPlayer.Equip4Image
                    ZOrder := 2
                    SizeToContent := false
                    Anchors := anchors{Minimum := vector2{X:= 0.5, Y:= 0.0}, Maximum := vector2{X:= 0.5, Y:= 0.0}}
                    Offsets := margin{Left := -508.0, Top:= 640.0, Right:= 80.0, Bottom:= 80.0}
                canvas_slot:
                    Widget := CustomPlayer.EquipSlot5Btn
                    ZOrder := 2
                    SizeToContent := false
                    Anchors := anchors{Minimum := vector2{X:= 0.5, Y:= 0.0}, Maximum := vector2{X:= 0.5, Y:= 0.0}}
                    Offsets := margin{Left := -792.0, Top:= 792.0, Right:= 250.0, Bottom:= 50.0}
                canvas_slot:
                    Widget := CustomPlayer.Equip5Image
                    ZOrder := 2
                    SizeToContent := false
                    Anchors := anchors{Minimum := vector2{X:= 0.5, Y:= 0.0}, Maximum := vector2{X:= 0.5, Y:= 0.0}}
                    Offsets := margin{Left := -508.0, Top:= 772.0, Right:= 80.0, Bottom:= 80.0}
                canvas_slot:
                    Widget := CustomPlayer.EquipSlot6Btn
                    ZOrder := 2
                    SizeToContent := false
                    Anchors := anchors{Minimum := vector2{X:= 0.5, Y:= 0.0}, Maximum := vector2{X:= 0.5, Y:= 0.0}}
                    Offsets := margin{Left := -792.0, Top:= 920.0, Right:= 250.0, Bottom:= 50.0}
                canvas_slot:
                    Widget := CustomPlayer.Equip6Image
                    ZOrder := 2
                    SizeToContent := false
                    Anchors := anchors{Minimum := vector2{X:= 0.5, Y:= 0.0}, Maximum := vector2{X:= 0.5, Y:= 0.0}}
                    Offsets := margin{Left := -508.0, Top:= 904.0, Right:= 80.0, Bottom:= 80.0}

                canvas_slot:
                    Widget:=CustomPlayer.Pet1Level
                    ZOrder:=2
                    SizeToContent:=true
                    Anchors := anchors{Minimum := vector2{X:= 0.5, Y:= 0.0}, Maximum := vector2{X:= 0.5, Y:= 0.0}}
                    Offsets:=margin{Left:=-328.0,Top:=280.0}
                canvas_slot:
                    Widget:=CustomPlayer.Pet1Tier
                    ZOrder:=2
                    SizeToContent:=true
                    Anchors := anchors{Minimum := vector2{X:= 0.5, Y:= 0.0}, Maximum := vector2{X:= 0.5, Y:= 0.0}}
                    Offsets:=margin{Left:=-328.0,Top:=332.0}
                canvas_slot:
                    Widget := CustomPlayer.Pet1Btn
                    ZOrder := 2
                    SizeToContent := false
                    Anchors := anchors{Minimum := vector2{X:= 0.5, Y:= 0.0}, Maximum := vector2{X:= 0.5, Y:= 0.0}}
                    Offsets := margin{Left := -344.0, Top:= 400.0, Right:= 200.0, Bottom:= 50.0}
                canvas_slot:
                    Widget := CustomPlayer.Pet1Image
                    ZOrder := 2
                    SizeToContent := false
                    Anchors := anchors{Minimum := vector2{X:= 0.5, Y:= 0.0}, Maximum := vector2{X:= 0.5, Y:= 0.0}}
                    Offsets := margin{Left := -112.0, Top:= 280.0, Right:= 140.0, Bottom:= 140.0}
                canvas_slot:
                    Widget:=CustomPlayer.Pet2Level
                    ZOrder:=2
                    SizeToContent:=true
                    Anchors := anchors{Minimum := vector2{X:= 0.5, Y:= 0.0}, Maximum := vector2{X:= 0.5, Y:= 0.0}}
                    Offsets:=margin{Left:=-328.0,Top:=552.0}
                canvas_slot:
                    Widget:=CustomPlayer.Pet2Tier
                    ZOrder:=2
                    SizeToContent:=true
                    Anchors := anchors{Minimum := vector2{X:= 0.5, Y:= 0.0}, Maximum := vector2{X:= 0.5, Y:= 0.0}}
                    Offsets:=margin{Left:=-328.0,Top:=604.0}
                canvas_slot:
                    Widget := CustomPlayer.Pet2Btn
                    ZOrder := 2
                    SizeToContent := false
                    Anchors := anchors{Minimum := vector2{X:= 0.5, Y:= 0.0}, Maximum := vector2{X:= 0.5, Y:= 0.0}}
                    Offsets := margin{Left := -344.0, Top:= 672.0, Right:= 200.0, Bottom:= 50.0}
                canvas_slot:
                    Widget := CustomPlayer.Pet2Image
                    ZOrder := 2
                    SizeToContent := false
                    Anchors := anchors{Minimum := vector2{X:= 0.5, Y:= 0.0}, Maximum := vector2{X:= 0.5, Y:= 0.0}}
                    Offsets := margin{Left := -112.0, Top:= 548.0, Right:= 140.0, Bottom:= 140.0}
                canvas_slot:
                    Widget:=CustomPlayer.Pet3Level
                    ZOrder:=2
                    SizeToContent:=true
                    Anchors := anchors{Minimum := vector2{X:= 0.5, Y:= 0.0}, Maximum := vector2{X:= 0.5, Y:= 0.0}}
                    Offsets:=margin{Left:=-328.0,Top:=820.0}
                canvas_slot:
                    Widget:=CustomPlayer.Pet3Tier
                    ZOrder:=2
                    SizeToContent:=true
                    Anchors := anchors{Minimum := vector2{X:= 0.5, Y:= 0.0}, Maximum := vector2{X:= 0.5, Y:= 0.0}}
                    Offsets:=margin{Left:=-328.0,Top:=872.0}
                canvas_slot:
                    Widget := CustomPlayer.Pet3Btn
                    ZOrder := 2
                    SizeToContent := false
                    Anchors := anchors{Minimum := vector2{X:= 0.5, Y:= 0.0}, Maximum := vector2{X:= 0.5, Y:= 0.0}}
                    Offsets := margin{Left := -344.0, Top:= 936.0, Right:= 200.0, Bottom:= 50.0}
                canvas_slot:
                    Widget := CustomPlayer.Pet3Image
                    ZOrder := 2
                    SizeToContent := false
                    Anchors := anchors{Minimum := vector2{X:= 0.5, Y:= 0.0}, Maximum := vector2{X:= 0.5, Y:= 0.0}}
                    Offsets := margin{Left := -112.0, Top:= 816.0, Right:= 140.0, Bottom:= 140.0}
                canvas_slot:
                    Widget:=CustomPlayer.Pet4Level
                    ZOrder:=2
                    SizeToContent:=true
                    Anchors := anchors{Minimum := vector2{X:= 0.5, Y:= 0.0}, Maximum := vector2{X:= 0.5, Y:= 0.0}}
                    Offsets:=margin{Left:=96.0,Top:=280.0}
                canvas_slot:
                    Widget:=CustomPlayer.Pet4Tier
                    ZOrder:=2
                    SizeToContent:=true
                    Anchors := anchors{Minimum := vector2{X:= 0.5, Y:= 0.0}, Maximum := vector2{X:= 0.5, Y:= 0.0}}
                    Offsets:=margin{Left:=96.0,Top:=332.0}
                canvas_slot:
                    Widget := CustomPlayer.Pet4Btn
                    ZOrder := 2
                    SizeToContent := false
                    Anchors := anchors{Minimum := vector2{X:= 0.5, Y:= 0.0}, Maximum := vector2{X:= 0.5, Y:= 0.0}}
                    Offsets := margin{Left := 84.0, Top:= 400.0, Right:= 200.0, Bottom:= 50.0}
                canvas_slot:
                    Widget := CustomPlayer.Pet4Image
                    ZOrder := 2
                    SizeToContent := false
                    Anchors := anchors{Minimum := vector2{X:= 0.5, Y:= 0.0}, Maximum := vector2{X:= 0.5, Y:= 0.0}}
                    Offsets := margin{Left := 312.0, Top:= 280.0, Right:= 140.0, Bottom:= 140.0}
                canvas_slot:
                    Widget:=CustomPlayer.Pet5Level
                    ZOrder:=2
                    SizeToContent:=true
                    Anchors := anchors{Minimum := vector2{X:= 0.5, Y:= 0.0}, Maximum := vector2{X:= 0.5, Y:= 0.0}}
                    Offsets:=margin{Left:=96.0,Top:=552.0}
                canvas_slot:
                    Widget:=CustomPlayer.Pet5Tier
                    ZOrder:=2
                    SizeToContent:=true
                    Anchors := anchors{Minimum := vector2{X:= 0.5, Y:= 0.0}, Maximum := vector2{X:= 0.5, Y:= 0.0}}
                    Offsets:=margin{Left:=96.0,Top:=604.0}
                canvas_slot:
                    Widget := CustomPlayer.Pet5Btn
                    ZOrder := 2
                    SizeToContent := false
                    Anchors := anchors{Minimum := vector2{X:= 0.5, Y:= 0.0}, Maximum := vector2{X:= 0.5, Y:= 0.0}}
                    Offsets := margin{Left := 84.0, Top:= 676.0, Right:= 200.0, Bottom:= 50.0}
                canvas_slot:
                    Widget := CustomPlayer.Pet5Image
                    ZOrder := 2
                    SizeToContent := false
                    Anchors := anchors{Minimum := vector2{X:= 0.5, Y:= 0.0}, Maximum := vector2{X:= 0.5, Y:= 0.0}}
                    Offsets := margin{Left := 312.0, Top:= 548.0, Right:= 140.0, Bottom:= 140.0}
                canvas_slot:
                    Widget:=CustomPlayer.Pet6Level
                    ZOrder:=2
                    SizeToContent:=true
                    Anchors := anchors{Minimum := vector2{X:= 0.5, Y:= 0.0}, Maximum := vector2{X:= 0.5, Y:= 0.0}}
                    Offsets:=margin{Left:=96.0,Top:=820.0}
                canvas_slot:
                    Widget:=CustomPlayer.Pet6Tier
                    ZOrder:=2
                    SizeToContent:=true
                    Anchors := anchors{Minimum := vector2{X:= 0.5, Y:= 0.0}, Maximum := vector2{X:= 0.5, Y:= 0.0}}
                    Offsets:=margin{Left:=96.0,Top:=872.0}
                canvas_slot:
                    Widget := CustomPlayer.Pet6Btn
                    ZOrder := 2
                    SizeToContent := false
                    Anchors := anchors{Minimum := vector2{X:= 0.5, Y:= 0.0}, Maximum := vector2{X:= 0.5, Y:= 0.0}}
                    Offsets := margin{Left := 84.0, Top:= 936.0, Right:= 200.0, Bottom:= 50.0}
                canvas_slot:
                    Widget := CustomPlayer.Pet6Image
                    ZOrder := 2
                    SizeToContent := false
                    Anchors := anchors{Minimum := vector2{X:= 0.5, Y:= 0.0}, Maximum := vector2{X:= 0.5, Y:= 0.0}}
                    Offsets := margin{Left := 312.0, Top:= 816.0, Right:= 140.0, Bottom:= 140.0}
                
                canvas_slot:
                    Widget:=CustomPlayer.PetTotalTitle
                    ZOrder:=2
                    SizeToContent:=true
                    Anchors := anchors{Minimum := vector2{X:= 0.5, Y:= 0.0}, Maximum := vector2{X:= 0.5, Y:= 0.0}}
                    Offsets:=margin{Left:=684.0,Top:=220.0}
                    Alignment := vector2{X:= 0.5, Y:= 0.0}
                canvas_slot:
                    Widget:=CustomPlayer.GatherTotalText
                    ZOrder:=2
                    SizeToContent:=true
                    Anchors := anchors{Minimum := vector2{X:= 0.5, Y:= 0.0}, Maximum := vector2{X:= 0.5, Y:= 0.0}}
                    Offsets:=margin{Left:=524.0,Top:=284.0}
                canvas_slot:
                    Widget:=CustomPlayer.GatherTotalAmount
                    ZOrder:=2
                    SizeToContent:=true
                    Anchors := anchors{Minimum := vector2{X:= 0.5, Y:= 0.0}, Maximum := vector2{X:= 0.5, Y:= 0.0}}
                    Offsets:=margin{Left:=832.0,Top:=284.0}
                    Alignment := vector2{X:= 1.0, Y:= 0.0}
                canvas_slot:
                    Widget:=CustomPlayer.StrengthText
                    ZOrder:=2
                    SizeToContent:=true
                    Anchors := anchors{Minimum := vector2{X:= 0.5, Y:= 0.0}, Maximum := vector2{X:= 0.5, Y:= 0.0}}
                    Offsets:=margin{Left:=524.0,Top:=352.0}
                canvas_slot:
                    Widget:=CustomPlayer.StrengthAmount
                    ZOrder:=2
                    SizeToContent:=true
                    Anchors := anchors{Minimum := vector2{X:= 0.5, Y:= 0.0}, Maximum := vector2{X:= 0.5, Y:= 0.0}}
                    Offsets:=margin{Left:=832.0,Top:=352.0}
                    Alignment := vector2{X:= 1.0, Y:= 0.0}
                canvas_slot:
                    Widget:=CustomPlayer.CritRateText
                    ZOrder:=2
                    SizeToContent:=true
                    Anchors := anchors{Minimum := vector2{X:= 0.5, Y:= 0.0}, Maximum := vector2{X:= 0.5, Y:= 0.0}}
                    Offsets:=margin{Left:=524.0,Top:=420.0}
                canvas_slot:
                    Widget:=CustomPlayer.CritRateAmount
                    ZOrder:=2
                    SizeToContent:=true
                    Anchors := anchors{Minimum := vector2{X:= 0.5, Y:= 0.0}, Maximum := vector2{X:= 0.5, Y:= 0.0}}
                    Offsets:=margin{Left:=832.0,Top:=420.0}
                    Alignment := vector2{X:= 1.0, Y:= 0.0}
                canvas_slot:
                    Widget:=CustomPlayer.CritDmgText
                    ZOrder:=2
                    SizeToContent:=true
                    Anchors := anchors{Minimum := vector2{X:= 0.5, Y:= 0.0}, Maximum := vector2{X:= 0.5, Y:= 0.0}}
                    Offsets:=margin{Left:=524.0,Top:=488.0}
                canvas_slot:
                    Widget:=CustomPlayer.CritDmgAmount
                    ZOrder:=2
                    SizeToContent:=true
                    Anchors := anchors{Minimum := vector2{X:= 0.5, Y:= 0.0}, Maximum := vector2{X:= 0.5, Y:= 0.0}}
                    Offsets:=margin{Left:=832.0,Top:=488.0}
                    Alignment := vector2{X:= 1.0, Y:= 0.0}
                canvas_slot:
                    Widget:=CustomPlayer.FinalStrikeText
                    ZOrder:=2
                    SizeToContent:=true
                    Anchors := anchors{Minimum := vector2{X:= 0.5, Y:= 0.0}, Maximum := vector2{X:= 0.5, Y:= 0.0}}
                    Offsets:=margin{Left:=524.0,Top:=556.0}
                canvas_slot:
                    Widget:=CustomPlayer.FinalStrikeAmount
                    ZOrder:=2
                    SizeToContent:=true
                    Anchors := anchors{Minimum := vector2{X:= 0.5, Y:= 0.0}, Maximum := vector2{X:= 0.5, Y:= 0.0}}
                    Offsets:=margin{Left:=832.0,Top:=556.0}
                    Alignment := vector2{X:= 1.0, Y:= 0.0}
                canvas_slot:
                    Widget:=CustomPlayer.UnequipPetsBtn
                    ZOrder:=2
                    SizeToContent:=false
                    Anchors := anchors{Minimum := vector2{X:= 0.5, Y:= 0.0}, Maximum := vector2{X:= 0.5, Y:= 0.0}}
                    Offsets:=margin{Left:=680.0,Top:=852.0, Right:= 250.0, Bottom:= 50.0}
                    Alignment := vector2{X:= 0.5, Y:= 0.0}
                canvas_slot:
                    Widget:=CustomPlayer.UnequipPetsText
                    ZOrder:=2
                    SizeToContent:=true
                    Anchors := anchors{Minimum := vector2{X:= 0.5, Y:= 0.0}, Maximum := vector2{X:= 0.5, Y:= 0.0}}
                    Offsets:=margin{Left:=680.0,Top:=800.0}
                    Alignment := vector2{X:= 0.5, Y:= 0.0}
                canvas_slot:
                    Widget:=CustomPlayer.HidePetsBtn
                    ZOrder:=2
                    SizeToContent:=false
                    Anchors := anchors{Minimum := vector2{X:= 0.5, Y:= 0.0}, Maximum := vector2{X:= 0.5, Y:= 0.0}}
                    Offsets:=margin{Left:=680.0,Top:=940.0, Right:= 250.0, Bottom:= 50.0}
                    Alignment := vector2{X:= 0.5, Y:= 0.0}

        }
#INVENTORY UI

    UpdateInv(CustomPlayer:Custom_Player):void=
        CustomPlayer.InvStoneText.SetText(StringToMessage("{NumberCondensor(CustomPlayer.StoneAmount)}"))
        CustomPlayer.InvShardText.SetText(StringToMessage("{NumberCondensor(CustomPlayer.ShardAmount)}"))
        CustomPlayer.InvWepMatText.SetText(StringToMessage("{NumberCondensor(CustomPlayer.WepMatAmount)}"))
        CustomPlayer.InvBossKeyText.SetText(StringToMessage("{NumberCondensor(CustomPlayer.BossKeyAmount)}"))
        for(i:int=0..CustomPlayer.InvEggTexts.Length):
            if(InvEggTextBlock:=CustomPlayer.InvEggTexts[i], EggAmount:=CustomPlayer.EggAmounts[i]):
                InvEggTextBlock.SetText(StringToMessage("{NumberCondensor(EggAmount)}"))

    CreateInvUI(CustomPlayer:Custom_Player) : canvas=
        CustomPlayer.CloseInvBtn.SetText(StringToMessage("Close"))

        for(TextBlock : CustomPlayer.AllInvTexts):
            TextBlock.SetShadowColor(CustomPlayer.ShadowDefaults.DefaultShadowColor)
            TextBlock.SetShadowOffset(CustomPlayer.ShadowDefaults.DefaultShadowOffset)
            TextBlock.SetShadowOpacity(1.0)
        NewUI : canvas = canvas{
            Slots := array:
                canvas_slot:
                    Widget:=CustomPlayer.InvMenuBackGround
                    ZOrder:=0
                    SizeToContent:=false
                    Offsets:=margin{Left:=-961.0,Top:=-541.0,Right:=1920.0,Bottom:=1080.0}
                    Alignment := vector2{X := 0.0, Y := 0.0}
                    Anchors := anchors{Minimum := vector2{X:= 0.5, Y:= 0.5}, Maximum := vector2{X:= 0.5, Y:= 0.5}}
                canvas_slot:
                    Widget:=CustomPlayer.InventoryTitle
                    ZOrder:=1
                    SizeToContent:=false
                    Offsets:=margin{Left:=-205.0,Top:=-460.0,Right:=396.0,Bottom:=76.0}
                    Alignment := vector2{X := 0.0, Y := 0.0}
                    Anchors := anchors{Minimum := vector2{X:= 0.5, Y:= 0.5}, Maximum := vector2{X:= 0.5, Y:= 0.5}}
                canvas_slot:
                    Widget:=CustomPlayer.CurrencyTitle
                    ZOrder:=1
                    SizeToContent:=false
                    Offsets:=margin{Left:=-609.0,Top:=-393.0,Right:=360.0,Bottom:=232.0}
                    Alignment := vector2{X := 0.0, Y := 0.0}
                    Anchors := anchors{Minimum := vector2{X:= 0.5, Y:= 0.5}, Maximum := vector2{X:= 0.5, Y:= 0.5}}
                canvas_slot:
                    Widget:=CustomPlayer.EggTitle
                    ZOrder:=1
                    SizeToContent:=false
                    Offsets:=margin{Left:=71.0,Top:=-329.0,Right:=164.0,Bottom:=88.0}
                    Alignment := vector2{X := 0.0, Y := 0.0}
                    Anchors := anchors{Minimum := vector2{X:= 0.5, Y:= 0.5}, Maximum := vector2{X:= 0.5, Y:= 0.5}}
                canvas_slot:
                    Widget:=CustomPlayer.InvStoneIcon
                    ZOrder:=1
                    SizeToContent:=false
                    Offsets:=margin{Left:=-560.0,Top:=-184.0,Right:=50.0,Bottom:=50.0}
                    Alignment := vector2{X := 0.0, Y := 0.0}
                    Anchors := anchors{Minimum := vector2{X:= 0.5, Y:= 0.5}, Maximum := vector2{X:= 0.5, Y:= 0.5}}
                canvas_slot:
                    Widget:=CustomPlayer.InvShardIcon
                    ZOrder:=1
                    SizeToContent:=false
                    Offsets:=margin{Left:=-560.0,Top:=-104.0,Right:=50.0,Bottom:=50.0}
                    Alignment := vector2{X := 0.0, Y := 0.0}
                    Anchors := anchors{Minimum := vector2{X:= 0.5, Y:= 0.5}, Maximum := vector2{X:= 0.5, Y:= 0.5}}
                canvas_slot:
                    Widget:=CustomPlayer.InvWepMatIcon
                    ZOrder:=1
                    SizeToContent:=false
                    Offsets:=margin{Left:=-560.0,Top:=-12.0,Right:=50.0,Bottom:=50.0}
                    Alignment := vector2{X := 0.0, Y := 0.0}
                    Anchors := anchors{Minimum := vector2{X:= 0.5, Y:= 0.5}, Maximum := vector2{X:= 0.5, Y:= 0.5}}
                canvas_slot:
                    Widget:=CustomPlayer.InvBossKeyIcon
                    ZOrder:=1
                    SizeToContent:=false
                    Offsets:=margin{Left:=-560.0,Top:=80.0,Right:=50.0,Bottom:=50.0}
                    Alignment := vector2{X := 0.0, Y := 0.0}
                    Anchors := anchors{Minimum := vector2{X:= 0.5, Y:= 0.5}, Maximum := vector2{X:= 0.5, Y:= 0.5}}
                canvas_slot:
                    Widget:=CustomPlayer.InvEgg1Icon
                    ZOrder:=1
                    SizeToContent:=false
                    Offsets:=margin{Left:=87.0,Top:=-184.0,Right:=50.0,Bottom:=50.0}
                    Alignment := vector2{X := 0.0, Y := 0.0}
                    Anchors := anchors{Minimum := vector2{X:= 0.5, Y:= 0.5}, Maximum := vector2{X:= 0.5, Y:= 0.5}}
                canvas_slot:
                    Widget:=CustomPlayer.InvEgg2Icon
                    ZOrder:=1
                    SizeToContent:=false
                    Offsets:=margin{Left:=87.0,Top:=-108.0,Right:=50.0,Bottom:=50.0}
                    Alignment := vector2{X := 0.0, Y := 0.0}
                    Anchors := anchors{Minimum := vector2{X:= 0.5, Y:= 0.5}, Maximum := vector2{X:= 0.5, Y:= 0.5}}
                canvas_slot:
                    Widget:=CustomPlayer.InvEgg3Icon
                    ZOrder:=1
                    SizeToContent:=false
                    Offsets:=margin{Left:=87.0,Top:=-32.0,Right:=50.0,Bottom:=50.0}
                    Alignment := vector2{X := 0.0, Y := 0.0}
                    Anchors := anchors{Minimum := vector2{X:= 0.5, Y:= 0.5}, Maximum := vector2{X:= 0.5, Y:= 0.5}}
                canvas_slot:
                    Widget:=CustomPlayer.InvEgg4Icon
                    ZOrder:=1
                    SizeToContent:=false
                    Offsets:=margin{Left:=87.0,Top:=48.0,Right:=50.0,Bottom:=50.0}
                    Alignment := vector2{X := 0.0, Y := 0.0}
                    Anchors := anchors{Minimum := vector2{X:= 0.5, Y:= 0.5}, Maximum := vector2{X:= 0.5, Y:= 0.5}}
                canvas_slot:
                    Widget:=CustomPlayer.InvEgg5Icon
                    ZOrder:=1
                    SizeToContent:=false
                    Offsets:=margin{Left:=87.0,Top:=124.0,Right:=50.0,Bottom:=50.0}
                    Alignment := vector2{X := 0.0, Y := 0.0}
                    Anchors := anchors{Minimum := vector2{X:= 0.5, Y:= 0.5}, Maximum := vector2{X:= 0.5, Y:= 0.5}}
                canvas_slot:
                    Widget:=CustomPlayer.InvEgg6Icon
                    ZOrder:=1
                    SizeToContent:=false
                    Offsets:=margin{Left:=87.0,Top:=202.0,Right:=50.0,Bottom:=50.0}
                    Alignment := vector2{X := 0.0, Y := 0.0}
                    Anchors := anchors{Minimum := vector2{X:= 0.5, Y:= 0.5}, Maximum := vector2{X:= 0.5, Y:= 0.5}}
                canvas_slot:
                    Widget:=CustomPlayer.InvStoneText
                    ZOrder:=1
                    SizeToContent:=true
                    Offsets:=margin{Left:=-500.0,Top:=-185.0,Right:=165.0,Bottom:=41.0}
                    Alignment := vector2{X := 0.0, Y := 0.0}
                    Anchors := anchors{Minimum := vector2{X:= 0.5, Y:= 0.5}, Maximum := vector2{X:= 0.5, Y:= 0.5}}
                canvas_slot:
                    Widget:=CustomPlayer.InvShardText
                    ZOrder:=1
                    SizeToContent:=true
                    Offsets:=margin{Left:=-500.0,Top:=-100.0,Right:=165.0,Bottom:=41.0}
                    Alignment := vector2{X := 0.0, Y := 0.0}
                    Anchors := anchors{Minimum := vector2{X:= 0.5, Y:= 0.5}, Maximum := vector2{X:= 0.5, Y:= 0.5}}
                canvas_slot:
                    Widget:=CustomPlayer.InvWepMatText
                    ZOrder:=1
                    SizeToContent:=true
                    Offsets:=margin{Left:=-500.0,Top:=-8.0,Right:=165.0,Bottom:=41.0}
                    Alignment := vector2{X := 0.0, Y := 0.0}
                    Anchors := anchors{Minimum := vector2{X:= 0.5, Y:= 0.5}, Maximum := vector2{X:= 0.5, Y:= 0.5}}
                canvas_slot:
                    Widget:=CustomPlayer.InvBossKeyText
                    ZOrder:=1
                    SizeToContent:=true
                    Offsets:=margin{Left:=-500.0,Top:=88.0,Right:=165.0,Bottom:=41.0}
                    Alignment := vector2{X := 0.0, Y := 0.0}
                    Anchors := anchors{Minimum := vector2{X:= 0.5, Y:= 0.5}, Maximum := vector2{X:= 0.5, Y:= 0.5}}
                canvas_slot:
                    Widget:=CustomPlayer.InvEgg1Text
                    ZOrder:=1
                    SizeToContent:=true
                    Offsets:=margin{Left:=143.0,Top:=-181.0,Right:=165.0,Bottom:=41.0}
                    Alignment := vector2{X := 0.0, Y := 0.0}
                    Anchors := anchors{Minimum := vector2{X:= 0.5, Y:= 0.5}, Maximum := vector2{X:= 0.5, Y:= 0.5}}
                canvas_slot:
                    Widget:=CustomPlayer.InvEgg2Text
                    ZOrder:=1
                    SizeToContent:=true
                    Offsets:=margin{Left:=143.0,Top:=-104.0,Right:=165.0,Bottom:=41.0}
                    Alignment := vector2{X := 0.0, Y := 0.0}
                    Anchors := anchors{Minimum := vector2{X:= 0.5, Y:= 0.5}, Maximum := vector2{X:= 0.5, Y:= 0.5}}
                canvas_slot:
                    Widget:=CustomPlayer.InvEgg3Text
                    ZOrder:=1
                    SizeToContent:=true
                    Offsets:=margin{Left:=143.0,Top:=-28.0,Right:=165.0,Bottom:=41.0}
                    Alignment := vector2{X := 0.0, Y := 0.0}
                    Anchors := anchors{Minimum := vector2{X:= 0.5, Y:= 0.5}, Maximum := vector2{X:= 0.5, Y:= 0.5}}
                canvas_slot:
                    Widget:=CustomPlayer.InvEgg4Text
                    ZOrder:=1
                    SizeToContent:=true
                    Offsets:=margin{Left:=143.0,Top:=51.0,Right:=165.0,Bottom:=41.0}
                    Alignment := vector2{X := 0.0, Y := 0.0}
                    Anchors := anchors{Minimum := vector2{X:= 0.5, Y:= 0.5}, Maximum := vector2{X:= 0.5, Y:= 0.5}}
                canvas_slot:
                    Widget:=CustomPlayer.InvEgg5Text
                    ZOrder:=1
                    SizeToContent:=true
                    Offsets:=margin{Left:=143.0,Top:=128.0,Right:=165.0,Bottom:=41.0}
                    Alignment := vector2{X := 0.0, Y := 0.0}
                    Anchors := anchors{Minimum := vector2{X:= 0.5, Y:= 0.5}, Maximum := vector2{X:= 0.5, Y:= 0.5}}
                canvas_slot:
                    Widget:=CustomPlayer.InvEgg6Text
                    ZOrder:=1
                    SizeToContent:=true
                    Offsets:=margin{Left:=143.0,Top:=207.0,Right:=165.0,Bottom:=41.0}
                    Alignment := vector2{X := 0.0, Y := 0.0}
                    Anchors := anchors{Minimum := vector2{X:= 0.5, Y:= 0.5}, Maximum := vector2{X:= 0.5, Y:= 0.5}}
                canvas_slot:
                    Anchors := anchors{Minimum := vector2{X:= 0.5, Y:= 0.0}, Maximum := vector2{X:= 0.5, Y:= 0.0}}
                    Offsets:=margin{Top:=120.0,Left:=500.0,Bottom:=50.0,Right:=250.0}
                    Alignment := vector2{X:= 0.0, Y:= 0.0}
                    ZOrder := 1
                    SizeToContent := false
                    Widget := CustomPlayer.CloseInvBtn
        }
    
#PLAYER STATS MENU UI
    var GBTextTitles : []string = array{
        "Base Strength",
        "Pet Strength",
        "Weapon Strength",
        "Total Strength",
        "Area Multiplier",
        "Achievement Multiplier",
        "Prestige Multiplier",
        "Total Strength On Hit",
        "Auto Gather",
        "Auto Gather Amount"
    }
    var GBBtnTitles : []string = array{
        "Gather",
        "General",
        "Stats"
    }

    UpdateGBUI(CustomPlayer:Custom_Player):void=
        var PetBaseStrength : float = 0.0
        CustomPlayer.GBTitle.SetImage(UI.PLAYER_STATS)
        CustomPlayer.GBAmount1.SetText(StringToMessage("{NumberCondensor(CustomPlayer.Power)}"))

        
        for(i:int=0..CustomPlayer.PetSlots.Length):
            if:
                PetSlot := CustomPlayer.PetSlots[i]
                PetSlot <> -1
                PetLevel := CustomPlayer.PetLevels[PetSlot]
                PetGather := CustomPlayer.PetGatherRates[PetSlot]
            then:
                set PetBaseStrength += CustomPlayer.Power * (PetGather/100.0)
                #Print("{PetBaseStrength}")
        CustomPlayer.GBAmount2.SetText(StringToMessage("{NumberCondensor(PetBaseStrength)}"))

        if(StrStat:=CustomPlayer.WeaponStats[2],StrBonus:=CustomPlayer.WeaponBonuses[2]): #STR UPGRADE
            CustomPlayer.GBAmount3.SetText(StringToMessage("{NumberCondensor(CustomPlayer.Power * (StrStat + (StrStat*StrBonus)))}"))
            CustomPlayer.GBAmount4.SetText(StringToMessage("{NumberCondensor(CustomPlayer.Power + PetBaseStrength + (CustomPlayer.Power * (StrStat + (StrStat*StrBonus))))}"))
        CustomPlayer.GBAmount5.SetText(StringToMessage("{NumberCondensor(CustomPlayer.RoomUnlockGather * 100)}%"))
        CustomPlayer.GBAmount6.SetText(StringToMessage("{NumberCondensor(CustomPlayer.AchBuffTotal * 100)}%"))
        CustomPlayer.GBAmount7.SetText(StringToMessage("{NumberCondensor(CustomPlayer.PrestigeHarvestBuff)}%"))
        CustomPlayer.GBAmount8.SetText(StringToMessage("{NumberCondensor(CustomPlayer.GetPlayerDamage())}/Hit"))
        CustomPlayer.GBAmount9.SetText(StringToMessage("{NumberCondensor(CustomPlayer.AutoGatherRate * 100)}%"))
        CustomPlayer.GBAmount10.SetText(StringToMessage("{NumberCondensor(CustomPlayer.GetPlayerDamage() * CustomPlayer.AutoGatherRate)}/Sec"))

    CreateGBUI(CustomPlayer:Custom_Player) : canvas=
        CustomPlayer.GBTitleText.SetText(StringToMessage("Gather Break Down"))
        CustomPlayer.GBCloseBtn.SetText(StringToMessage("Close"))

        for(TextBlock : CustomPlayer.AllGBTexts):
            TextBlock.SetShadowColor(CustomPlayer.ShadowDefaults.DefaultShadowColor)
            TextBlock.SetShadowOffset(CustomPlayer.ShadowDefaults.DefaultShadowOffset)
            TextBlock.SetShadowOpacity(1.0)

        for(i:int=0..GBTextTitles.Length):
            if(TextBlock:=CustomPlayer.GBTitles[i], Text:=GBTextTitles[i]):
                TextBlock.SetText(StringToMessage("{Text}"))
        NewUI : canvas = canvas{
            Slots := array:
                canvas_slot:
                    Widget:=CustomPlayer.GBBackground
                    ZOrder:=0
                    SizeToContent:=false
                    Offsets:=margin{Left:=-961.0,Top:=-541.0,Right:=1920.0,Bottom:=1080.0}
                    Alignment := vector2{X := 0.0, Y := 0.0}
                    Anchors := anchors{Minimum := vector2{X:= 0.5, Y:= 0.5}, Maximum := vector2{X:= 0.5, Y:= 0.5}}
                canvas_slot:
                    Widget:=CustomPlayer.GBIcon1
                    ZOrder:=1
                    SizeToContent:=false
                    Offsets:=margin{Left:=-461.0,Top:=-265.0,Right:=55.0,Bottom:=55.0}
                    Alignment := vector2{X := 0.0, Y := 0.0}
                    Anchors := anchors{Minimum := vector2{X:= 0.5, Y:= 0.5}, Maximum := vector2{X:= 0.5, Y:= 0.5}}
                canvas_slot:
                    Widget:=CustomPlayer.GBIcon2
                    ZOrder:=1
                    SizeToContent:=false
                    Offsets:=margin{Left:=-461.0,Top:=-204.0,Right:=55.0,Bottom:=55.0}
                    Alignment := vector2{X := 0.0, Y := 0.0}
                    Anchors := anchors{Minimum := vector2{X:= 0.5, Y:= 0.5}, Maximum := vector2{X:= 0.5, Y:= 0.5}}
                canvas_slot:
                    Widget:=CustomPlayer.GBIcon3
                    ZOrder:=1
                    SizeToContent:=false
                    Offsets:=margin{Left:=-461.0,Top:=-144.0,Right:=55.0,Bottom:=55.0}
                    Alignment := vector2{X := 0.0, Y := 0.0}
                    Anchors := anchors{Minimum := vector2{X:= 0.5, Y:= 0.5}, Maximum := vector2{X:= 0.5, Y:= 0.5}}
                canvas_slot:
                    Widget:=CustomPlayer.GBIcon4
                    ZOrder:=1
                    SizeToContent:=false
                    Offsets:=margin{Left:=-461.0,Top:=-85.0,Right:=55.0,Bottom:=55.0}
                    Alignment := vector2{X := 0.0, Y := 0.0}
                    Anchors := anchors{Minimum := vector2{X:= 0.5, Y:= 0.5}, Maximum := vector2{X:= 0.5, Y:= 0.5}}
                canvas_slot:
                    Widget:=CustomPlayer.GBIcon5
                    ZOrder:=1
                    SizeToContent:=false
                    Offsets:=margin{Left:=-461.0,Top:=23.0,Right:=55.0,Bottom:=55.0}
                    Alignment := vector2{X := 0.0, Y := 0.0}
                    Anchors := anchors{Minimum := vector2{X:= 0.5, Y:= 0.5}, Maximum := vector2{X:= 0.5, Y:= 0.5}}
                canvas_slot:
                    Widget:=CustomPlayer.GBIcon6
                    ZOrder:=1
                    SizeToContent:=false
                    Offsets:=margin{Left:=-461.0,Top:=87.0,Right:=55.0,Bottom:=55.0}
                    Alignment := vector2{X := 0.0, Y := 0.0}
                    Anchors := anchors{Minimum := vector2{X:= 0.5, Y:= 0.5}, Maximum := vector2{X:= 0.5, Y:= 0.5}}
                canvas_slot:
                    Widget:=CustomPlayer.GBIcon7
                    ZOrder:=1
                    SizeToContent:=false
                    Offsets:=margin{Left:=-461.0,Top:=147.0,Right:=55.0,Bottom:=55.0}
                    Alignment := vector2{X := 0.0, Y := 0.0}
                    Anchors := anchors{Minimum := vector2{X:= 0.5, Y:= 0.5}, Maximum := vector2{X:= 0.5, Y:= 0.5}}
                canvas_slot:
                    Widget:=CustomPlayer.GBIcon8
                    ZOrder:=1
                    SizeToContent:=false
                    Offsets:=margin{Left:=-461.0,Top:=211.0,Right:=55.0,Bottom:=55.0}
                    Alignment := vector2{X := 0.0, Y := 0.0}
                    Anchors := anchors{Minimum := vector2{X:= 0.5, Y:= 0.5}, Maximum := vector2{X:= 0.5, Y:= 0.5}}
                canvas_slot:
                    Widget:=CustomPlayer.GBIcon9
                    ZOrder:=1
                    SizeToContent:=false
                    Offsets:=margin{Left:=-461.0,Top:=303.0,Right:=55.0,Bottom:=55.0}
                    Alignment := vector2{X := 0.0, Y := 0.0}
                    Anchors := anchors{Minimum := vector2{X:= 0.5, Y:= 0.5}, Maximum := vector2{X:= 0.5, Y:= 0.5}}
                canvas_slot:
                    Widget:=CustomPlayer.GBIcon10
                    ZOrder:=1
                    SizeToContent:=false
                    Offsets:=margin{Left:=-461.0,Top:=367.0,Right:=55.0,Bottom:=55.0}
                    Alignment := vector2{X := 0.0, Y := 0.0}
                    Anchors := anchors{Minimum := vector2{X:= 0.5, Y:= 0.5}, Maximum := vector2{X:= 0.5, Y:= 0.5}}
                canvas_slot:
                    Widget:=CustomPlayer.GBTitle1
                    ZOrder:=1
                    SizeToContent:=true
                    Offsets:=margin{Left:=-386.0,Top:=-261.0,Right:=165.0,Bottom:=41.0}
                    Alignment := vector2{X := 0.0, Y := 0.0}
                    Anchors := anchors{Minimum := vector2{X:= 0.5, Y:= 0.5}, Maximum := vector2{X:= 0.5, Y:= 0.5}}
                canvas_slot:
                    Widget:=CustomPlayer.GBAmount1
                    ZOrder:=1
                    SizeToContent:=true
                    Offsets:=margin{Left:=239.0,Top:=-261.0,Right:=165.0,Bottom:=41.0}
                    Alignment := vector2{X := 0.0, Y := 0.0}
                    Anchors := anchors{Minimum := vector2{X:= 0.5, Y:= 0.5}, Maximum := vector2{X:= 0.5, Y:= 0.5}}
                canvas_slot:
                    Widget:=CustomPlayer.GBTitle2
                    ZOrder:=1
                    SizeToContent:=true
                    Offsets:=margin{Left:=-386.0,Top:=-197.0,Right:=165.0,Bottom:=41.0}
                    Alignment := vector2{X := 0.0, Y := 0.0}
                    Anchors := anchors{Minimum := vector2{X:= 0.5, Y:= 0.5}, Maximum := vector2{X:= 0.5, Y:= 0.5}}
                canvas_slot:
                    Widget:=CustomPlayer.GBAmount2
                    ZOrder:=1
                    SizeToContent:=true
                    Offsets:=margin{Left:=239.0,Top:=-200.0,Right:=165.0,Bottom:=41.0}
                    Alignment := vector2{X := 0.0, Y := 0.0}
                    Anchors := anchors{Minimum := vector2{X:= 0.5, Y:= 0.5}, Maximum := vector2{X:= 0.5, Y:= 0.5}}
                canvas_slot:
                    Widget:=CustomPlayer.GBTitle3
                    ZOrder:=1
                    SizeToContent:=true
                    Offsets:=margin{Left:=-386.0,Top:=-141.0,Right:=165.0,Bottom:=41.0}
                    Alignment := vector2{X := 0.0, Y := 0.0}
                    Anchors := anchors{Minimum := vector2{X:= 0.5, Y:= 0.5}, Maximum := vector2{X:= 0.5, Y:= 0.5}}
                canvas_slot:
                    Widget:=CustomPlayer.GBAmount3
                    ZOrder:=1
                    SizeToContent:=true
                    Offsets:=margin{Left:=239.0,Top:=-141.0,Right:=165.0,Bottom:=41.0}
                    Alignment := vector2{X := 0.0, Y := 0.0}
                    Anchors := anchors{Minimum := vector2{X:= 0.5, Y:= 0.5}, Maximum := vector2{X:= 0.5, Y:= 0.5}}
                canvas_slot:
                    Widget:=CustomPlayer.GBTitle4
                    ZOrder:=1
                    SizeToContent:=true
                    Offsets:=margin{Left:=-386.0,Top:=-75.0,Right:=165.0,Bottom:=41.0}
                    Alignment := vector2{X := 0.0, Y := 0.0}
                    Anchors := anchors{Minimum := vector2{X:= 0.5, Y:= 0.5}, Maximum := vector2{X:= 0.5, Y:= 0.5}}
                canvas_slot:
                    Widget:=CustomPlayer.GBAmount4
                    ZOrder:=1
                    SizeToContent:=true
                    Offsets:=margin{Left:=239.0,Top:=-75.0,Right:=165.0,Bottom:=41.0}
                    Alignment := vector2{X := 0.0, Y := 0.0}
                    Anchors := anchors{Minimum := vector2{X:= 0.5, Y:= 0.5}, Maximum := vector2{X:= 0.5, Y:= 0.5}}
                canvas_slot:
                    Widget:=CustomPlayer.GBTitle5
                    ZOrder:=1
                    SizeToContent:=true
                    Offsets:=margin{Left:=-386.0,Top:=35.0,Right:=165.0,Bottom:=41.0}
                    Alignment := vector2{X := 0.0, Y := 0.0}
                    Anchors := anchors{Minimum := vector2{X:= 0.5, Y:= 0.5}, Maximum := vector2{X:= 0.5, Y:= 0.5}}
                canvas_slot:
                    Widget:=CustomPlayer.GBAmount5
                    ZOrder:=1
                    SizeToContent:=true
                    Offsets:=margin{Left:=239.0,Top:=35.0,Right:=165.0,Bottom:=41.0}
                    Alignment := vector2{X := 0.0, Y := 0.0}
                    Anchors := anchors{Minimum := vector2{X:= 0.5, Y:= 0.5}, Maximum := vector2{X:= 0.5, Y:= 0.5}}
                canvas_slot:
                    Widget:=CustomPlayer.GBTitle6
                    ZOrder:=1
                    SizeToContent:=true
                    Offsets:=margin{Left:=-386.0,Top:=99.0,Right:=165.0,Bottom:=41.0}
                    Alignment := vector2{X := 0.0, Y := 0.0}
                    Anchors := anchors{Minimum := vector2{X:= 0.5, Y:= 0.5}, Maximum := vector2{X:= 0.5, Y:= 0.5}}
                canvas_slot:
                    Widget:=CustomPlayer.GBAmount6
                    ZOrder:=1
                    SizeToContent:=true
                    Offsets:=margin{Left:=239.0,Top:=99.0,Right:=165.0,Bottom:=41.0}
                    Alignment := vector2{X := 0.0, Y := 0.0}
                    Anchors := anchors{Minimum := vector2{X:= 0.5, Y:= 0.5}, Maximum := vector2{X:= 0.5, Y:= 0.5}}
                canvas_slot:
                    Widget:=CustomPlayer.GBTitle7
                    ZOrder:=1
                    SizeToContent:=true
                    Offsets:=margin{Left:=-386.0,Top:=155.0,Right:=165.0,Bottom:=41.0}
                    Alignment := vector2{X := 0.0, Y := 0.0}
                    Anchors := anchors{Minimum := vector2{X:= 0.5, Y:= 0.5}, Maximum := vector2{X:= 0.5, Y:= 0.5}}
                canvas_slot:
                    Widget:=CustomPlayer.GBAmount7
                    ZOrder:=1
                    SizeToContent:=true
                    Offsets:=margin{Left:=239.0,Top:=155.0,Right:=165.0,Bottom:=41.0}
                    Alignment := vector2{X := 0.0, Y := 0.0}
                    Anchors := anchors{Minimum := vector2{X:= 0.5, Y:= 0.5}, Maximum := vector2{X:= 0.5, Y:= 0.5}}
                canvas_slot:
                    Widget:=CustomPlayer.GBTitle8
                    ZOrder:=1
                    SizeToContent:=true
                    Offsets:=margin{Left:=-386.0,Top:=215.0,Right:=165.0,Bottom:=41.0}
                    Alignment := vector2{X := 0.0, Y := 0.0}
                    Anchors := anchors{Minimum := vector2{X:= 0.5, Y:= 0.5}, Maximum := vector2{X:= 0.5, Y:= 0.5}}
                canvas_slot:
                    Widget:=CustomPlayer.GBAmount8
                    ZOrder:=1
                    SizeToContent:=true
                    Offsets:=margin{Left:=239.0,Top:=215.0,Right:=165.0,Bottom:=41.0}
                    Alignment := vector2{X := 0.0, Y := 0.0}
                    Anchors := anchors{Minimum := vector2{X:= 0.5, Y:= 0.5}, Maximum := vector2{X:= 0.5, Y:= 0.5}}
                canvas_slot:
                    Widget:=CustomPlayer.GBTitle9
                    ZOrder:=1
                    SizeToContent:=true
                    Offsets:=margin{Left:=-386.0,Top:=315.0,Right:=165.0,Bottom:=41.0}
                    Alignment := vector2{X := 0.0, Y := 0.0}
                    Anchors := anchors{Minimum := vector2{X:= 0.5, Y:= 0.5}, Maximum := vector2{X:= 0.5, Y:= 0.5}}
                canvas_slot:
                    Widget:=CustomPlayer.GBAmount9
                    ZOrder:=1
                    SizeToContent:=true
                    Offsets:=margin{Left:=239.0,Top:=315.0,Right:=165.0,Bottom:=41.0}
                    Alignment := vector2{X := 0.0, Y := 0.0}
                    Anchors := anchors{Minimum := vector2{X:= 0.5, Y:= 0.5}, Maximum := vector2{X:= 0.5, Y:= 0.5}}
                canvas_slot:
                    Widget:=CustomPlayer.GBTitle10
                    ZOrder:=1
                    SizeToContent:=true
                    Offsets:=margin{Left:=-386.0,Top:=375.0,Right:=165.0,Bottom:=41.0}
                    Alignment := vector2{X := 0.0, Y := 0.0}
                    Anchors := anchors{Minimum := vector2{X:= 0.5, Y:= 0.5}, Maximum := vector2{X:= 0.5, Y:= 0.5}}
                canvas_slot:
                    Widget:=CustomPlayer.GBAmount10
                    ZOrder:=1
                    SizeToContent:=true
                    Offsets:=margin{Left:=239.0,Top:=375.0,Right:=165.0,Bottom:=41.0}
                    Alignment := vector2{X := 0.0, Y := 0.0}
                    Anchors := anchors{Minimum := vector2{X:= 0.5, Y:= 0.5}, Maximum := vector2{X:= 0.5, Y:= 0.5}}
                canvas_slot:
                    Widget:=CustomPlayer.GBGBBtn
                    ZOrder:=1
                    SizeToContent:=false
                    Offsets:=margin{Left:=-801.0,Top:=-225.0,Right:=286.0,Bottom:=41.0}
                    Alignment := vector2{X := 0.0, Y := 0.0}
                    Anchors := anchors{Minimum := vector2{X:= 0.5, Y:= 0.5}, Maximum := vector2{X:= 0.5, Y:= 0.5}}
                canvas_slot:
                    Widget:=CustomPlayer.GBGenStatsBtn
                    ZOrder:=1
                    SizeToContent:=false
                    Offsets:=margin{Left:=-801.0,Top:=-157.0,Right:=286.0,Bottom:=41.0}
                    Alignment := vector2{X := 0.0, Y := 0.0}
                    Anchors := anchors{Minimum := vector2{X:= 0.5, Y:= 0.5}, Maximum := vector2{X:= 0.5, Y:= 0.5}}
                canvas_slot:
                    Widget:=CustomPlayer.GBGameStatsBtn
                    ZOrder:=1
                    SizeToContent:=false
                    Offsets:=margin{Left:=-801.0,Top:=-89.0,Right:=286.0,Bottom:=41.0}
                    Alignment := vector2{X := 0.0, Y := 0.0}
                    Anchors := anchors{Minimum := vector2{X:= 0.5, Y:= 0.5}, Maximum := vector2{X:= 0.5, Y:= 0.5}}
                canvas_slot:
                    Widget:=CustomPlayer.GBTitle
                    ZOrder:=1
                    SizeToContent:=false
                    Offsets:=margin{Left:=-841.0,Top:=-429.0,Right:=520.0,Bottom:=72.0}
                    Alignment := vector2{X := 0.0, Y := 0.0}
                    Anchors := anchors{Minimum := vector2{X:= 0.5, Y:= 0.5}, Maximum := vector2{X:= 0.5, Y:= 0.5}}
                canvas_slot:
                    Widget:=CustomPlayer.GBMenuTitle
                    ZOrder:=1
                    SizeToContent:=false
                    Offsets:=margin{Left:=-713.0,Top:=-289.0,Right:=100.0,Bottom:=40.0}
                    Alignment := vector2{X := 0.0, Y := 0.0}
                    Anchors := anchors{Minimum := vector2{X:= 0.5, Y:= 0.5}, Maximum := vector2{X:= 0.5, Y:= 0.5}}
                canvas_slot:
                    Widget:=CustomPlayer.GBTitleText
                    ZOrder:=1
                    SizeToContent:=false
                    Offsets:=margin{Left:=-461.0,Top:=-321.0,Right:=346.0,Bottom:=40.0}
                    Alignment := vector2{X := 0.0, Y := 0.0}
                    Anchors := anchors{Minimum := vector2{X:= 0.5, Y:= 0.5}, Maximum := vector2{X:= 0.5, Y:= 0.5}}
                canvas_slot:
                    Widget:=CustomPlayer.GBCloseBtn
                    ZOrder:=1
                    SizeToContent:=false
                    Offsets:=margin{Left:=-801.0,Top:=-28.0,Right:=286.0,Bottom:=41.0}
                    Alignment := vector2{X := 0.0, Y := 0.0}
                    Anchors := anchors{Minimum := vector2{X:= 0.5, Y:= 0.5}, Maximum := vector2{X:= 0.5, Y:= 0.5}}
        }

#GENERAL STATS UI

    var GSTitleText1 : []string = array {
        "Attack",
        "Attack Average (incl. Crit)",
        "Critical Rate",
        "Critical Damage",
        "Final Strike",
        "Move Speed",
        "Egg Drop Rate",
        "Egg Drop Amount",
        "Hats Collected",
        "Hat Prestige Level"
    }
    var GSTitleText2 : []string = array {
        "Session Time",
        "Game Time (All Time)",
        "Stages Unlocked",
        "Rocks Destroyed",
        "Drill Activations",
        "Bosses Defeated",
        "Stone Collected (All Time)",
        "Prestige Points Collected (All Time)",
        "Secret Chests Collected",
        ""
    }

    UpdateGSUI(CustomPlayer : Custom_Player, WhichUI : int):void=
        CustomPlayer.GSTitle.SetImage(UI.PLAYER_STATS)
        if (WhichUI = 1):
            var PetCrit : float = 0.0
            CustomPlayer.GSHeadingTitle1.SetText(StringToMessage("Stats"))
            CustomPlayer.GSHeadingTitle2.SetText(StringToMessage("Egg Drop"))
            for(i:int=0..CustomPlayer.GSTitles.Length):
                if(TextBlock:=CustomPlayer.GSTitles[i], Text:=GSTitleText1[i]):
                    TextBlock.SetText(StringToMessage("{Text}"))
            for (i:int=0..CustomPlayer.GSIconBlock.Length):
                if(TextureBlock:=CustomPlayer.GSIconBlock[i], Image:=CustomPlayer.GSIcons1[i]):
                    TextureBlock.SetImage(Image)
            CustomPlayer.GSAmount1.SetText(StringToMessage("{NumberCondensor(CustomPlayer.GetPlayerWeaponDamage())}"))
            CustomPlayer.GSAmount2.SetText(StringToMessage("{NumberCondensor(CustomPlayer.GetPlayerWeaponDamage() + ((CustomPlayer.GetPlayerWeaponDamage() * CustomPlayer.CritDMG)*CustomPlayer.CritRate))}"))

            var CurrentCritChance : float = CustomPlayer.CritRate
            var CurrentCritDmg : float = CustomPlayer.CritDMG
            var CurrentFinalStrike : float = CustomPlayer.FinalHitBonus * 100.0
            for(i:int=0..CustomPlayer.PetSlots.Length):
                if(CustomPlayer.PetSlots[i] <> -1):
                    if:
                        PetLevel := CustomPlayer.PetLevels[CustomPlayer.PetSlots[i]]
                        PetCritRate := CustomPlayer.PetCritRates[CustomPlayer.PetSlots[i]]
                        PetCritDmg := CustomPlayer.PetCritDMG[CustomPlayer.PetSlots[i]]
                        FinalHit := CustomPlayer.PetFinalStrike[CustomPlayer.PetSlots[i]]
                    then:
                        set CurrentCritChance += PetCritRate/100.0
                        set CurrentCritDmg += PetCritDmg/100.0
                        set CurrentFinalStrike += FinalHit
                        
            CustomPlayer.GSAmount3.SetText(StringToMessage("{NumberCondensor(CurrentCritChance*100.0)}%"))
            CustomPlayer.GSAmount4.SetText(StringToMessage("{NumberCondensor(CurrentCritDmg * 100.0)}%"))
            CustomPlayer.GSAmount5.SetText(StringToMessage("{NumberCondensor(CurrentFinalStrike)}%"))
            CustomPlayer.GSAmount6.SetText(StringToMessage("{CustomPlayer.MovementSpeedLevel * 10}%"))
            CustomPlayer.GSAmount7.SetText(StringToMessage("{NumberCondensor(CustomPlayer.MobEggDropChance + (CustomPlayer.EggDropRateMultiplier*100.0))}%"))
            if:
                EggAmountMin:=CustomPlayer.EggDropAmount[0] + (CustomPlayer.EggDropAmount[0] * CustomPlayer.EggAmountMultiplier)
                EggAmountMax:=CustomPlayer.EggDropAmount[1] + (CustomPlayer.EggDropAmount[1] * CustomPlayer.EggAmountMultiplier)
            then:
                CustomPlayer.GSAmount8.SetText(StringToMessage("{NumberCondensor(EggAmountMin)} - {NumberCondensor(EggAmountMax)}"))
            
            var HatsCollected : int = 0
            for(i:int=0..CustomPlayer.HatsUnlocked.Length):
                if(CustomPlayer.HatsUnlocked[i] = true):
                    set HatsCollected += 1
            CustomPlayer.GSAmount9.SetText(StringToMessage("{HatsCollected} / {CustomPlayer.HatsUnlocked.Length}"))
            CustomPlayer.GSAmount10.SetText(StringToMessage("{CustomPlayer.HatPrestiges}"))
        if(WhichUI = 2):                
            CustomPlayer.GSHeadingTitle1.SetText(StringToMessage("Game Statistics"))
            CustomPlayer.GSHeadingTitle2.SetText(StringToMessage("Currency"))
            for(i:int=0..CustomPlayer.GSTitles.Length):
                if(TextBlock:=CustomPlayer.GSTitles[i], Text:=GSTitleText2[i]):
                    TextBlock.SetText(StringToMessage("{Text}"))
            for (i:int=0..CustomPlayer.GSIconBlock.Length):
                if(TextureBlock:=CustomPlayer.GSIconBlock[i], Image:=CustomPlayer.GSIcons2[i]):
                    TextureBlock.SetImage(Image)
            CustomPlayer.GSAmount1.SetText(StringToMessage(ConvertToTime(gm.SessionTime,"Hours")))
            if(TimePlayedAch:=CustomPlayer.AllAchs[7],TimePlayed:=TimePlayedAch[0]):
                CustomPlayer.GSAmount2.SetText(StringToMessage(ConvertToTime(TimePlayed,"Hours")))
            CustomPlayer.GSAmount3.SetText(StringToMessage("{CustomPlayer.StagesUnlocked +1}/25"))
            if(RocksDestroyed:=CustomPlayer.AllAchs[4][0]):
                CustomPlayer.GSAmount4.SetText(StringToMessage("{NumberCondensor(RocksDestroyed)}"))
            CustomPlayer.GSAmount5.SetText(StringToMessage("{CustomPlayer.DrillActivations}"))
            if(BossesDefeated:=CustomPlayer.AllAchs[6][0]):
                CustomPlayer.GSAmount6.SetText(StringToMessage("{NumberCondensor(BossesDefeated)}"))
            if(StoneCollected:=CustomPlayer.AllAchs[0][0]):
                CustomPlayer.GSAmount7.SetText(StringToMessage("{NumberCondensor(StoneCollected)}"))
            if(prestigeCollected:=CustomPlayer.AllAchs[3][0]):
                CustomPlayer.GSAmount8.SetText(StringToMessage("{NumberCondensor(prestigeCollected)}"))

            var SecretGather : int = 0
            for(i:int=0..CustomPlayer.NewSecretsClaimed.Length):
                if(CustomPlayer.NewSecretsClaimed[i] = true):
                    set SecretGather += 1
            CustomPlayer.GSAmount9.SetText(StringToMessage("{SecretGather} / {CustomPlayer.NewSecretsClaimed.Length}"))
            CustomPlayer.GSAmount10.SetText(StringToMessage(""))

    CreateGSUI(CustomPlayer:Custom_Player) : canvas=  
        CustomPlayer.GSCloseBtn.SetText(StringToMessage("Close"))      

        for(TextBlock : CustomPlayer.AllGSTexts):
            TextBlock.SetShadowColor(CustomPlayer.ShadowDefaults.DefaultShadowColor)
            TextBlock.SetShadowOffset(CustomPlayer.ShadowDefaults.DefaultShadowOffset)
            TextBlock.SetShadowOpacity(1.0)

        NewUI : canvas = canvas{
            Slots := array:
                canvas_slot:
                    Widget:=CustomPlayer.GSBackground
                    ZOrder:=0
                    SizeToContent:=false
                    Offsets:=margin{Left:=-961.0,Top:=-541.0,Right:=1920.0,Bottom:=1080.0}
                    Alignment := vector2{X := 0.0, Y := 0.0}
                    Anchors := anchors{Minimum := vector2{X:= 0.5, Y:= 0.5}, Maximum := vector2{X:= 0.5, Y:= 0.5}}
                canvas_slot:
                    Widget:=CustomPlayer.GSTitle
                    ZOrder:=1
                    SizeToContent:=false
                    Offsets:=margin{Left:=-841.0,Top:=-429.0,Right:=520.0,Bottom:=72.0}
                    Alignment := vector2{X := 0.0, Y := 0.0}
                    Anchors := anchors{Minimum := vector2{X:= 0.5, Y:= 0.5}, Maximum := vector2{X:= 0.5, Y:= 0.5}}
                canvas_slot:
                    Widget:=CustomPlayer.GSCloseBtn
                    ZOrder:=0
                    SizeToContent:=false
                    Offsets:=margin{Left:=-801.0,Top:=-28.0,Right:=286.0,Bottom:=41.0}
                    Alignment := vector2{X := 0.0, Y := 0.0}
                    Anchors := anchors{Minimum := vector2{X:= 0.5, Y:= 0.5}, Maximum := vector2{X:= 0.5, Y:= 0.5}}
                canvas_slot:
                    Widget:=CustomPlayer.GSIcon1
                    ZOrder:=1
                    SizeToContent:=false
                    Offsets:=margin{Left:=-464.0,Top:=-273.0,Right:=55.0,Bottom:=55.0}
                    Alignment := vector2{X := 0.0, Y := 0.0}
                    Anchors := anchors{Minimum := vector2{X:= 0.5, Y:= 0.5}, Maximum := vector2{X:= 0.5, Y:= 0.5}}
                canvas_slot:
                    Widget:=CustomPlayer.GSIcon2
                    ZOrder:=1
                    SizeToContent:=false
                    Offsets:=margin{Left:=-464.0,Top:=-213.0,Right:=55.0,Bottom:=55.0}
                    Alignment := vector2{X := 0.0, Y := 0.0}
                    Anchors := anchors{Minimum := vector2{X:= 0.5, Y:= 0.5}, Maximum := vector2{X:= 0.5, Y:= 0.5}}
                canvas_slot:
                    Widget:=CustomPlayer.GSIcon3
                    ZOrder:=1
                    SizeToContent:=false
                    Offsets:=margin{Left:=-464.0,Top:=-152.0,Right:=55.0,Bottom:=55.0}
                    Alignment := vector2{X := 0.0, Y := 0.0}
                    Anchors := anchors{Minimum := vector2{X:= 0.5, Y:= 0.5}, Maximum := vector2{X:= 0.5, Y:= 0.5}}
                canvas_slot:
                    Widget:=CustomPlayer.GSIcon4
                    ZOrder:=1
                    SizeToContent:=false
                    Offsets:=margin{Left:=-464.0,Top:=-93.0,Right:=55.0,Bottom:=55.0}
                    Alignment := vector2{X := 0.0, Y := 0.0}
                    Anchors := anchors{Minimum := vector2{X:= 0.5, Y:= 0.5}, Maximum := vector2{X:= 0.5, Y:= 0.5}}
                canvas_slot:
                    Widget:=CustomPlayer.GSIcon5
                    ZOrder:=1
                    SizeToContent:=false
                    Offsets:=margin{Left:=-464.0,Top:=-32.0,Right:=55.0,Bottom:=55.0}
                    Alignment := vector2{X := 0.0, Y := 0.0}
                    Anchors := anchors{Minimum := vector2{X:= 0.5, Y:= 0.5}, Maximum := vector2{X:= 0.5, Y:= 0.5}}
                canvas_slot:
                    Widget:=CustomPlayer.GSIcon6
                    ZOrder:=1
                    SizeToContent:=false
                    Offsets:=margin{Left:=-464.0,Top:=27.0,Right:=55.0,Bottom:=55.0}
                    Alignment := vector2{X := 0.0, Y := 0.0}
                    Anchors := anchors{Minimum := vector2{X:= 0.5, Y:= 0.5}, Maximum := vector2{X:= 0.5, Y:= 0.5}}
                canvas_slot:
                    Widget:=CustomPlayer.GSIcon7
                    ZOrder:=1
                    SizeToContent:=false
                    Offsets:=margin{Left:=-464.0,Top:=191.0,Right:=55.0,Bottom:=55.0}
                    Alignment := vector2{X := 0.0, Y := 0.0}
                    Anchors := anchors{Minimum := vector2{X:= 0.5, Y:= 0.5}, Maximum := vector2{X:= 0.5, Y:= 0.5}}
                canvas_slot:
                    Widget:=CustomPlayer.GSIcon8
                    ZOrder:=1
                    SizeToContent:=false
                    Offsets:=margin{Left:=-464.0,Top:=255.0,Right:=55.0,Bottom:=55.0}
                    Alignment := vector2{X := 0.0, Y := 0.0}
                    Anchors := anchors{Minimum := vector2{X:= 0.5, Y:= 0.5}, Maximum := vector2{X:= 0.5, Y:= 0.5}}
                canvas_slot:
                    Widget:=CustomPlayer.GSTitle1
                    ZOrder:=1
                    SizeToContent:=true
                    Offsets:=margin{Left:=-386.0,Top:=-261.0,Right:=165.0,Bottom:=41.0}
                    Alignment := vector2{X := 0.0, Y := 0.0}
                    Anchors := anchors{Minimum := vector2{X:= 0.5, Y:= 0.5}, Maximum := vector2{X:= 0.5, Y:= 0.5}}
                canvas_slot:
                    Widget:=CustomPlayer.GSAmount1
                    ZOrder:=1
                    SizeToContent:=true
                    Offsets:=margin{Left:=239.0,Top:=-261.0,Right:=165.0,Bottom:=41.0}
                    Alignment := vector2{X := 0.0, Y := 0.0}
                    Anchors := anchors{Minimum := vector2{X:= 0.5, Y:= 0.5}, Maximum := vector2{X:= 0.5, Y:= 0.5}}
                canvas_slot:
                    Widget:=CustomPlayer.GSTitle2
                    ZOrder:=1
                    SizeToContent:=true
                    Offsets:=margin{Left:=-386.0,Top:=-197.0,Right:=165.0,Bottom:=41.0}
                    Alignment := vector2{X := 0.0, Y := 0.0}
                    Anchors := anchors{Minimum := vector2{X:= 0.5, Y:= 0.5}, Maximum := vector2{X:= 0.5, Y:= 0.5}}
                canvas_slot:
                    Widget:=CustomPlayer.GSAmount2
                    ZOrder:=1
                    SizeToContent:=true
                    Offsets:=margin{Left:=239.0,Top:=-200.0,Right:=165.0,Bottom:=41.0}
                    Alignment := vector2{X := 0.0, Y := 0.0}
                    Anchors := anchors{Minimum := vector2{X:= 0.5, Y:= 0.5}, Maximum := vector2{X:= 0.5, Y:= 0.5}}
                canvas_slot:
                    Widget:=CustomPlayer.GSTitle3
                    ZOrder:=1
                    SizeToContent:=true
                    Offsets:=margin{Left:=-386.0,Top:=-141.0,Right:=165.0,Bottom:=41.0}
                    Alignment := vector2{X := 0.0, Y := 0.0}
                    Anchors := anchors{Minimum := vector2{X:= 0.5, Y:= 0.5}, Maximum := vector2{X:= 0.5, Y:= 0.5}}
                canvas_slot:
                    Widget:=CustomPlayer.GSAmount3
                    ZOrder:=1
                    SizeToContent:=true
                    Offsets:=margin{Left:=239.0,Top:=-141.0,Right:=165.0,Bottom:=41.0}
                    Alignment := vector2{X := 0.0, Y := 0.0}
                    Anchors := anchors{Minimum := vector2{X:= 0.5, Y:= 0.5}, Maximum := vector2{X:= 0.5, Y:= 0.5}}
                canvas_slot:
                    Widget:=CustomPlayer.GSTitle4
                    ZOrder:=1
                    SizeToContent:=true
                    Offsets:=margin{Left:=-386.0,Top:=-75.0,Right:=165.0,Bottom:=41.0}
                    Alignment := vector2{X := 0.0, Y := 0.0}
                    Anchors := anchors{Minimum := vector2{X:= 0.5, Y:= 0.5}, Maximum := vector2{X:= 0.5, Y:= 0.5}}
                canvas_slot:
                    Widget:=CustomPlayer.GSAmount4
                    ZOrder:=1
                    SizeToContent:=true
                    Offsets:=margin{Left:=239.0,Top:=-75.0,Right:=165.0,Bottom:=41.0}
                    Alignment := vector2{X := 0.0, Y := 0.0}
                    Anchors := anchors{Minimum := vector2{X:= 0.5, Y:= 0.5}, Maximum := vector2{X:= 0.5, Y:= 0.5}}
                canvas_slot:
                    Widget:=CustomPlayer.GSTitle5
                    ZOrder:=1
                    SizeToContent:=true
                    Offsets:=margin{Left:=-386.0,Top:=-20.0,Right:=165.0,Bottom:=41.0}
                    Alignment := vector2{X := 0.0, Y := 0.0}
                    Anchors := anchors{Minimum := vector2{X:= 0.5, Y:= 0.5}, Maximum := vector2{X:= 0.5, Y:= 0.5}}
                canvas_slot:
                    Widget:=CustomPlayer.GSAmount5
                    ZOrder:=1
                    SizeToContent:=true
                    Offsets:=margin{Left:=239.0,Top:=-20.0,Right:=165.0,Bottom:=41.0}
                    Alignment := vector2{X := 0.0, Y := 0.0}
                    Anchors := anchors{Minimum := vector2{X:= 0.5, Y:= 0.5}, Maximum := vector2{X:= 0.5, Y:= 0.5}}
                canvas_slot:
                    Widget:=CustomPlayer.GSTitle6
                    ZOrder:=1
                    SizeToContent:=true
                    Offsets:=margin{Left:=-386.0,Top:=35.0,Right:=165.0,Bottom:=41.0}
                    Alignment := vector2{X := 0.0, Y := 0.0}
                    Anchors := anchors{Minimum := vector2{X:= 0.5, Y:= 0.5}, Maximum := vector2{X:= 0.5, Y:= 0.5}}
                canvas_slot:
                    Widget:=CustomPlayer.GSAmount6
                    ZOrder:=1
                    SizeToContent:=true
                    Offsets:=margin{Left:=239.0,Top:=35.0,Right:=165.0,Bottom:=41.0}
                    Alignment := vector2{X := 0.0, Y := 0.0}
                    Anchors := anchors{Minimum := vector2{X:= 0.5, Y:= 0.5}, Maximum := vector2{X:= 0.5, Y:= 0.5}}
                canvas_slot:
                    Widget:=CustomPlayer.GSTitle7
                    ZOrder:=1
                    SizeToContent:=true
                    Offsets:=margin{Left:=-386.0,Top:=200.0,Right:=165.0,Bottom:=41.0}
                    Alignment := vector2{X := 0.0, Y := 0.0}
                    Anchors := anchors{Minimum := vector2{X:= 0.5, Y:= 0.5}, Maximum := vector2{X:= 0.5, Y:= 0.5}}
                canvas_slot:
                    Widget:=CustomPlayer.GSAmount7
                    ZOrder:=1
                    SizeToContent:=true
                    Offsets:=margin{Left:=239.0,Top:=200.0,Right:=165.0,Bottom:=41.0}
                    Alignment := vector2{X := 0.0, Y := 0.0}
                    Anchors := anchors{Minimum := vector2{X:= 0.5, Y:= 0.5}, Maximum := vector2{X:= 0.5, Y:= 0.5}}
                canvas_slot:
                    Widget:=CustomPlayer.GSTitle8
                    ZOrder:=1
                    SizeToContent:=true
                    Offsets:=margin{Left:=-386.0,Top:=263.0,Right:=165.0,Bottom:=41.0}
                    Alignment := vector2{X := 0.0, Y := 0.0}
                    Anchors := anchors{Minimum := vector2{X:= 0.5, Y:= 0.5}, Maximum := vector2{X:= 0.5, Y:= 0.5}}
                canvas_slot:
                    Widget:=CustomPlayer.GSAmount8
                    ZOrder:=1
                    SizeToContent:=true
                    Offsets:=margin{Left:=239.0,Top:=263.0,Right:=165.0,Bottom:=41.0}
                    Alignment := vector2{X := 0.0, Y := 0.0}
                    Anchors := anchors{Minimum := vector2{X:= 0.5, Y:= 0.5}, Maximum := vector2{X:= 0.5, Y:= 0.5}}
                canvas_slot:
                    Widget:=CustomPlayer.GSTitle9
                    ZOrder:=1
                    SizeToContent:=true
                    Offsets:=margin{Left:=-386.0,Top:=372.0,Right:=165.0,Bottom:=41.0}
                    Alignment := vector2{X := 0.0, Y := 0.0}
                    Anchors := anchors{Minimum := vector2{X:= 0.5, Y:= 0.5}, Maximum := vector2{X:= 0.5, Y:= 0.5}}
                canvas_slot:
                    Widget:=CustomPlayer.GSAmount9
                    ZOrder:=1
                    SizeToContent:=true
                    Offsets:=margin{Left:=239.0,Top:=372.0,Right:=165.0,Bottom:=41.0}
                    Alignment := vector2{X := 0.0, Y := 0.0}
                    Anchors := anchors{Minimum := vector2{X:= 0.5, Y:= 0.5}, Maximum := vector2{X:= 0.5, Y:= 0.5}}
                canvas_slot:
                    Widget:=CustomPlayer.GSTitle10
                    ZOrder:=1
                    SizeToContent:=true
                    Offsets:=margin{Left:=-386.0,Top:=424.0,Right:=165.0,Bottom:=41.0}
                    Alignment := vector2{X := 0.0, Y := 0.0}
                    Anchors := anchors{Minimum := vector2{X:= 0.5, Y:= 0.5}, Maximum := vector2{X:= 0.5, Y:= 0.5}}
                canvas_slot:
                    Widget:=CustomPlayer.GSAmount10
                    ZOrder:=1
                    SizeToContent:=true
                    Offsets:=margin{Left:=239.0,Top:=424.0,Right:=165.0,Bottom:=41.0}
                    Alignment := vector2{X := 0.0, Y := 0.0}
                    Anchors := anchors{Minimum := vector2{X:= 0.5, Y:= 0.5}, Maximum := vector2{X:= 0.5, Y:= 0.5}}
                canvas_slot:
                    Widget:=CustomPlayer.GSGBBtn
                    ZOrder:=1
                    SizeToContent:=false
                    Offsets:=margin{Left:=-801.0,Top:=-225.0,Right:=286.0,Bottom:=41.0}
                    Alignment := vector2{X := 0.0, Y := 0.0}
                    Anchors := anchors{Minimum := vector2{X:= 0.5, Y:= 0.5}, Maximum := vector2{X:= 0.5, Y:= 0.5}}
                canvas_slot:
                    Widget:=CustomPlayer.GSGenStatsBtn
                    ZOrder:=1
                    SizeToContent:=false
                    Offsets:=margin{Left:=-801.0,Top:=-157.0,Right:=286.0,Bottom:=41.0}
                    Alignment := vector2{X := 0.0, Y := 0.0}
                    Anchors := anchors{Minimum := vector2{X:= 0.5, Y:= 0.5}, Maximum := vector2{X:= 0.5, Y:= 0.5}}
                canvas_slot:
                    Widget:=CustomPlayer.GSGameStatsBtn
                    ZOrder:=1
                    SizeToContent:=false
                    Offsets:=margin{Left:=-801.0,Top:=-89.0,Right:=286.0,Bottom:=41.0}
                    Alignment := vector2{X := 0.0, Y := 0.0}
                    Anchors := anchors{Minimum := vector2{X:= 0.5, Y:= 0.5}, Maximum := vector2{X:= 0.5, Y:= 0.5}}
                canvas_slot:
                    Widget:=CustomPlayer.GSMenuTitle
                    ZOrder:=1
                    SizeToContent:=true
                    Offsets:=margin{Left:=-713.0,Top:=-289.0,Right:=100.0,Bottom:=40.0}
                    Alignment := vector2{X := 0.0, Y := 0.0}
                    Anchors := anchors{Minimum := vector2{X:= 0.5, Y:= 0.5}, Maximum := vector2{X:= 0.5, Y:= 0.5}}
                canvas_slot:
                    Widget:=CustomPlayer.GSHeadingTitle1
                    ZOrder:=1
                    SizeToContent:=true
                    Offsets:=margin{Left:=-457.0,Top:=-320.0,Right:=165.0,Bottom:=41.0}
                    Alignment := vector2{X := 0.0, Y := 0.0}
                    Anchors := anchors{Minimum := vector2{X:= 0.5, Y:= 0.5}, Maximum := vector2{X:= 0.5, Y:= 0.5}}
                canvas_slot:
                    Widget:=CustomPlayer.GSHeadingTitle2
                    ZOrder:=1
                    SizeToContent:=true
                    Offsets:=margin{Left:=-457.0,Top:=127.0,Right:=165.0,Bottom:=41.0}
                    Alignment := vector2{X := 0.0, Y := 0.0}
                    Anchors := anchors{Minimum := vector2{X:= 0.5, Y:= 0.5}, Maximum := vector2{X:= 0.5, Y:= 0.5}}
        }

#ACHIEVEMENTS MENU UI
    var AchTitles : []string = array{
        "Total Stone Gathered",
        "Max Strength",
        "Total Mobs Defeated",
        "Prestige Points Gathered",
        "Total Rocks Destroyed",
        "Total Pet Level",
        "Total Bosses Defeated",
        "Total Time Played"
        }

 ## UI BLOCKS


 ## CANVAS
    MakeAchCanvas(CustomPlayer:Custom_Player):canvas=
        CustomPlayer.CloseAchButton.SetText(StringToMessage("X"))

        var MyCanvas: canvas = canvas:
            Slots := array:
                canvas_slot:
                    Widget:=CustomPlayer.AchHUDBG
                    ZOrder:=0
                    SizeToContent:=false
                    Offsets:=margin{Top:=CustomPlayer.AchHUD_Position.Y,Left:=960.0,Bottom:=CustomPlayer.AchHUDBG.DefaultDesiredSize.Y,Right:=CustomPlayer.AchHUDBG.DefaultDesiredSize.X}
                    Alignment := vector2{X := 0.5, Y := 0.0}
                canvas_slot:
                    Widget:=CustomPlayer.CloseAchButton
                    ZOrder:=1
                    SizeToContent:=false
                    Offsets:=margin{Top:=CustomPlayer.CloseAch_Position.Y,Left:=CustomPlayer.CloseAch_Position.X,Bottom:=CustomPlayer.CloseAch_Size.Y,Right:=CustomPlayer.CloseAch_Size.X}

                canvas_slot:
                    Widget:=CustomPlayer.AchTitleText
                    ZOrder:=1
                    SizeToContent:=false
                    Offsets:=margin{Top:= 84.0, Left:= 1068.0, Bottom:=56.0, Right:=475.0}
                    Alignment := vector2{X := 0.0, Y := 0.0}

                canvas_slot:
                    Widget:=CustomPlayer.AchBuffTotalText
                    ZOrder:=1
                    SizeToContent:=true
                    Offsets:=margin{Top:= 104.0, Left:= 300.0}
                    Alignment := vector2{X := 0.5, Y := 0.0}

                canvas_slot:
                    Widget:=CustomPlayer.Ach1Title
                    ZOrder:=1
                    SizeToContent:=true
                    Offsets:=margin{Top:= 216.0, Left:= 244.0}
                canvas_slot:
                    Widget:=CustomPlayer.Ach1Desc
                    ZOrder:=1
                    SizeToContent:=true
                    Offsets:=margin{Top:= 268.0, Left:= 244.0}
                canvas_slot:
                    Widget:=CustomPlayer.Ach1Icon
                    ZOrder:=1
                    SizeToContent:=false
                    Offsets:=margin{Top:= 200.0, Left:= 80.0, Bottom:=140.0, Right:=140.0}
                canvas_slot:
                    Widget:=CustomPlayer.Ach1Progress
                    ZOrder:=3
                    SizeToContent:=true
                    Offsets:=margin{Top:= 344.0, Left:= 504.0}
                    Alignment := vector2{X := 0.5, Y := 0.0}
                canvas_slot:
                    Widget:=CustomPlayer.Ach1BarColorTop
                    ZOrder:=2
                    Offsets := margin{Top := 344.0, Left := 76.0}
                canvas_slot:
                    Offsets := margin{Top := 344.0, Left := 76.0, Bottom := 41.0, Right := 876.0}
                    Widget := stack_box:
                        Orientation := orientation.Vertical
                        Slots := array:
                            stack_box_slot:
                                Widget := CustomPlayer.AchBarColorBase1   
                                HorizontalAlignment := horizontal_alignment.Left  

                canvas_slot:
                    Widget:=CustomPlayer.Ach2Title
                    ZOrder:=1
                    SizeToContent:=true
                    Offsets:=margin{Top:= 420.0, Left:= 244.0}
                canvas_slot:
                    Widget:=CustomPlayer.Ach2Desc
                    ZOrder:=1
                    SizeToContent:=true
                    Offsets:=margin{Top:= 472.0, Left:= 244.0}
                canvas_slot:
                    Widget:=CustomPlayer.Ach2Icon
                    ZOrder:=1
                    SizeToContent:=false
                    Offsets:=margin{Top:= 400.0, Left:= 80.0, Bottom:=140.0, Right:=140.0}
                canvas_slot:
                    Widget:=CustomPlayer.Ach2Progress
                    ZOrder:=3
                    SizeToContent:=true
                    Offsets:=margin{Top:= 544.0, Left:= 504.0}
                    Alignment := vector2{X := 0.5, Y := 0.0}
                canvas_slot:
                    Widget:=CustomPlayer.Ach2BarColorTop
                    ZOrder:=2
                    Offsets := margin{Top := 544.0, Left := 76.0}
                canvas_slot:
                    Offsets := margin{Top := 544.0, Left := 76.0, Bottom := 41.0, Right := 876.0}
                    Widget := stack_box:
                        Orientation := orientation.Vertical
                        Slots := array:
                            stack_box_slot:
                                Widget := CustomPlayer.AchBarColorBase2
                                HorizontalAlignment := horizontal_alignment.Left  
                    
                canvas_slot:
                    Widget:=CustomPlayer.Ach3Title
                    ZOrder:=1
                    SizeToContent:=true
                    Offsets:=margin{Top:= 620.0, Left:= 244.0}
                canvas_slot:
                    Widget:=CustomPlayer.Ach3Desc
                    ZOrder:=1
                    SizeToContent:=true
                    Offsets:=margin{Top:= 672.0, Left:= 244.0}
                canvas_slot:
                    Widget:=CustomPlayer.Ach3Icon
                    ZOrder:=1
                    SizeToContent:=false
                    Offsets:=margin{Top:= 600.0, Left:= 80.0, Bottom:=140.0, Right:=140.0}
                canvas_slot:
                    Widget:=CustomPlayer.Ach3Progress
                    ZOrder:=3
                    SizeToContent:=true
                    Offsets:=margin{Top:= 744.0, Left:= 504.0}
                    Alignment := vector2{X := 0.5, Y := 0.0}
                canvas_slot:
                    Widget:=CustomPlayer.Ach3BarColorTop
                    ZOrder:=2
                    Offsets := margin{Top := 744.0, Left := 76.0}
                canvas_slot:
                    Offsets := margin{Top := 744.0, Left := 76.0, Bottom := 41.0, Right := 876.0}
                    Widget := stack_box:
                        Orientation := orientation.Vertical
                        Slots := array:
                            stack_box_slot:
                                Widget := CustomPlayer.AchBarColorBase3
                                HorizontalAlignment := horizontal_alignment.Left  

                canvas_slot:
                    Widget:=CustomPlayer.Ach4Title
                    ZOrder:=1
                    SizeToContent:=true
                    Offsets:=margin{Top:= 820.0, Left:= 244.0}
                canvas_slot:
                    Widget:=CustomPlayer.Ach4Desc
                    ZOrder:=1
                    SizeToContent:=true
                    Offsets:=margin{Top:= 872.0, Left:= 244.0}
                canvas_slot:
                    Widget:=CustomPlayer.Ach4Icon
                    ZOrder:=1
                    SizeToContent:=false
                    Offsets:=margin{Top:= 800.0, Left:= 80.0, Bottom:=140.0, Right:=140.0}
                canvas_slot:
                    Widget:=CustomPlayer.Ach4Progress
                    ZOrder:=3
                    SizeToContent:=true
                    Offsets:=margin{Top:= 944.0, Left:= 504.0}
                    Alignment := vector2{X := 0.5, Y := 0.0}
                canvas_slot:
                    Widget:=CustomPlayer.Ach4BarColorTop
                    ZOrder:=2
                    Offsets := margin{Top := 944.0, Left := 76.0}
                canvas_slot:
                    Offsets := margin{Top := 944.0, Left := 76.0, Bottom := 41.0, Right := 876.0}
                    Widget := stack_box:
                        Orientation := orientation.Vertical
                        Slots := array:
                            stack_box_slot:
                                Widget := CustomPlayer.AchBarColorBase4 
                                HorizontalAlignment := horizontal_alignment.Left  

                canvas_slot:
                    Widget:=CustomPlayer.Ach5Title
                    ZOrder:=1
                    SizeToContent:=true
                    Offsets:=margin{Top:= 216.0, Left:= 1136.0}
                canvas_slot:
                    Widget:=CustomPlayer.Ach5Desc
                    ZOrder:=1
                    SizeToContent:=true
                    Offsets:=margin{Top:= 268.0, Left:= 1136.0}
                canvas_slot:
                    Widget:=CustomPlayer.Ach5Icon
                    ZOrder:=1
                    SizeToContent:=false
                    Offsets:=margin{Top:= 200.0, Left:= 976.0, Bottom:=140.0, Right:=140.0}
                canvas_slot:
                    Widget:=CustomPlayer.Ach5Progress
                    ZOrder:=3
                    SizeToContent:=true
                    Offsets:=margin{Top:= 344.0, Left:= 1398.0}
                    Alignment := vector2{X := 0.5, Y := 0.0}
                canvas_slot:
                    Widget:=CustomPlayer.Ach5BarColorTop
                    ZOrder:=2
                    Offsets := margin{Top := 344.0, Left := 968.0}
                canvas_slot:
                    Offsets := margin{Top := 344.0, Left := 968.0, Bottom := 41.0, Right := 876.0}
                    Widget := stack_box:
                        Orientation := orientation.Vertical
                        Slots := array:
                            stack_box_slot:
                                Widget := CustomPlayer.AchBarColorBase5
                                HorizontalAlignment := horizontal_alignment.Left  

                canvas_slot:
                    Widget:=CustomPlayer.Ach6Title
                    ZOrder:=1
                    SizeToContent:=true
                    Offsets:=margin{Top:= 420.0, Left:= 1136.0}
                canvas_slot:
                    Widget:=CustomPlayer.Ach6Desc
                    ZOrder:=1
                    SizeToContent:=true
                    Offsets:=margin{Top:= 472.0, Left:= 1136.0}
                canvas_slot:
                    Widget:=CustomPlayer.Ach6Icon
                    ZOrder:=1
                    SizeToContent:=false
                    Offsets:=margin{Top:= 400.0, Left:= 976.0, Bottom:=140.0, Right:=140.0}
                canvas_slot:
                    Widget:=CustomPlayer.Ach6Progress
                    ZOrder:=3
                    SizeToContent:=true
                    Offsets:=margin{Top:= 544.0, Left:= 1398.0}
                    Alignment := vector2{X := 0.5, Y := 0.0}
                canvas_slot:
                    Widget:=CustomPlayer.Ach6BarColorTop
                    ZOrder:=2
                    Offsets := margin{Top := 544.0, Left := 968.0}
                canvas_slot:
                    Offsets := margin{Top := 544.0, Left := 968.0, Bottom := 41.0, Right := 876.0}
                    Widget := stack_box:
                        Orientation := orientation.Vertical
                        Slots := array:
                            stack_box_slot:
                                Widget := CustomPlayer.AchBarColorBase6
                                HorizontalAlignment := horizontal_alignment.Left  

                canvas_slot:
                    Widget:=CustomPlayer.Ach7Title
                    ZOrder:=1
                    SizeToContent:=true
                    Offsets:=margin{Top:= 620.0, Left:= 1136.0}
                canvas_slot:
                    Widget:=CustomPlayer.Ach7Desc
                    ZOrder:=1
                    SizeToContent:=true
                    Offsets:=margin{Top:= 672.0, Left:= 1136.0}
                canvas_slot:
                    Widget:=CustomPlayer.Ach7Icon
                    ZOrder:=1
                    SizeToContent:=false
                    Offsets:=margin{Top:= 600.0, Left:= 976.0, Bottom:=140.0, Right:=140.0}
                canvas_slot:
                    Widget:=CustomPlayer.Ach7Progress
                    ZOrder:=3
                    SizeToContent:=true
                    Offsets:=margin{Top:= 744.0, Left:= 1398.0}
                    Alignment := vector2{X := 0.5, Y := 0.0}
                canvas_slot:
                    Widget:=CustomPlayer.Ach7BarColorTop
                    ZOrder:=2
                    Offsets := margin{Top := 744.0, Left := 968.0}
                canvas_slot:
                    Offsets := margin{Top := 744.0, Left := 968.0, Bottom := 41.0, Right := 876.0}
                    Widget := stack_box:
                        Orientation := orientation.Vertical
                        Slots := array:
                            stack_box_slot:
                                Widget := CustomPlayer.AchBarColorBase7
                                HorizontalAlignment := horizontal_alignment.Left  

                canvas_slot:
                    Widget:=CustomPlayer.Ach8Title
                    ZOrder:=1
                    SizeToContent:=true
                    Offsets:=margin{Top:= 820.0, Left:= 1136.0}
                canvas_slot:
                    Widget:=CustomPlayer.Ach8Desc
                    ZOrder:=1
                    SizeToContent:=true
                    Offsets:=margin{Top:= 872.0, Left:= 1136.0}
                canvas_slot:
                    Widget:=CustomPlayer.Ach8Icon
                    ZOrder:=1
                    SizeToContent:=false
                    Offsets:=margin{Top:= 800.0, Left:= 976.0, Bottom:=140.0, Right:=140.0}
                canvas_slot:
                    Widget:=CustomPlayer.Ach8Progress
                    ZOrder:=3
                    SizeToContent:=true
                    Offsets:=margin{Top:= 944.0, Left:= 1398.0}
                    Alignment := vector2{X := 0.5, Y := 0.0}
                canvas_slot:
                    Widget:=CustomPlayer.Ach8BarColorTop
                    ZOrder:=2
                    Offsets := margin{Top := 944.0, Left := 968.0}
                canvas_slot:
                    Offsets := margin{Top := 944.0, Left := 968.0, Bottom := 41.0, Right := 876.0}
                    Widget := stack_box:
                        Orientation := orientation.Vertical
                        Slots := array:
                            stack_box_slot:
                                Widget := CustomPlayer.AchBarColorBase8
                                HorizontalAlignment := horizontal_alignment.Left  

        return MyCanvas

 ## CODE
   
    UpdateAchievementHUD(CustomPlayer:Custom_Player):void=
        CustomPlayer.AchBuffTotalText.SetText(StringToMessage("{NumberCondensor(CustomPlayer.AchBuffTotal*100.0)}% Strength"))
        CustomPlayer.AchBuffTotalText.SetShadowColor(CustomPlayer.ShadowDefaults.DefaultShadowColor)
        CustomPlayer.AchBuffTotalText.SetShadowOffset(CustomPlayer.ShadowDefaults.DefaultShadowOffset)
        CustomPlayer.AchBuffTotalText.SetShadowOpacity(1.0)
        for(i:int=0..CustomPlayer.AchHUDTitles.Length):
            if(AchTitle:=CustomPlayer.AchHUDTitles[i],AchDesc:=CustomPlayer.AchHUDDescs[i],AchProgress:=CustomPlayer.AchHUDProgress[i],CurrentTitle:=AchTitles[i]):
                AchTitle.SetShadowColor(CustomPlayer.ShadowDefaults.DefaultShadowColor)
                AchTitle.SetShadowOffset(CustomPlayer.ShadowDefaults.DefaultShadowOffset)
                AchTitle.SetShadowOpacity(1.0)
                AchDesc.SetShadowColor(CustomPlayer.ShadowDefaults.DefaultShadowColor)
                AchDesc.SetShadowOffset(CustomPlayer.ShadowDefaults.DefaultShadowOffset)
                AchDesc.SetShadowOpacity(1.0)
                AchProgress.SetShadowColor(CustomPlayer.ShadowDefaults.DefaultShadowColor)
                AchProgress.SetShadowOffset(CustomPlayer.ShadowDefaults.DefaultShadowOffset)
                AchProgress.SetShadowOpacity(1.0)

                if(CQP:=CustomPlayer.AllAchs[i]):
                    if(CQP1:=CQP[0],CQP2:=CQP[1],AchLevel:=CustomPlayer.AchLevels[i]):
                        AchTitle.SetText(StringToMessage("{CurrentTitle} {ConvertToRomanNumeral(AchLevel)}"))

                        AdjustProgress(CustomPlayer, CQP1, CQP2, i)

                        CondensedNumber1:=NumberCondensor(CQP1*1.0)
                        CondensedNumber2:=NumberCondensor(CQP2*1.0)
                        if(i = 0):
                            AchDesc.SetText(StringToMessage("Gather {CondensedNumber2} Stone"))
                        else if(i = 1):
                            AchDesc.SetText(StringToMessage("Increase Max Strength to {CondensedNumber2}"))
                        else if(i = 2):
                            AchDesc.SetText(StringToMessage("Defeat {CondensedNumber2} Mobs"))
                        else if(i = 3):
                            AchDesc.SetText(StringToMessage("Gather {CondensedNumber2} Prestige Points"))
                        else if(i = 4):
                            AchDesc.SetText(StringToMessage("Destroy {CondensedNumber2} Rocks"))
                        else if(i = 5):
                            AchDesc.SetText(StringToMessage("Level Pets {CondensedNumber2} Times"))
                        else if(i = 6):
                            AchDesc.SetText(StringToMessage("Defeat {CondensedNumber2} Bosses"))
                        else if(i = 7):
                            AchDesc.SetText(StringToMessage("Play for {ConvertToTime(CQP2,"Hours")}"))

                        if(i <> 7):
                            AchProgress.SetText(StringToMessage("{CondensedNumber1}/{CondensedNumber2}"))
                        else:
                            AchProgress.SetText(StringToMessage("{ConvertToTime(CQP1,"Hours")}/{ConvertToTime(CQP2,"Hours")}"))


    AdjustProgress(CustomPlayer:Custom_Player, Value:float,Target:float, whichQuest:int):void=
        var CompletionPercentage : float = 0.0
        LeftToComplete := Target - Value
        ProgressSoFar := Target - LeftToComplete
        set CompletionPercentage = ProgressSoFar / Target

        BarSize := CustomPlayer.AchBarColorBase1.GetDesiredSize()
        NewProgressBarWidth := BarSize.X * CompletionPercentage
        for(i:int=0..CustomPlayer.AchBarColorTops.Length):
            if(i = whichQuest):
                if(AchBarColorTop:=CustomPlayer.AchBarColorTops[i],AchProgress:=CustomPlayer.AchHUDProgress[i]):
                    if(NewProgressBarWidth >= 876.0):
                        AchBarColorTop.SetColor(MakeColorFromHex("22AA76"))
                        AchBarColorTop.SetDesiredSize(vector2{X:= 876.0, Y := 41.0})
                    else:
                        AchBarColorTop.SetColor(MakeColorFromHex("ffa800"))
                        AchBarColorTop.SetDesiredSize(vector2{X:= NewProgressBarWidth, Y := 41.0})
            
                    CondensedNumber1:=NumberCondensor(Value)
                    CondensedNumber2:=NumberCondensor(Target)
                    AchProgress.SetText(StringToMessage("{CondensedNumber1}/{CondensedNumber2}"))
    
#PRESTIGE INFO
    #UI BLOCKS
    #CANVAS
    MakePrestigeCanvas(CustomPlayer:Custom_Player):canvas=
        CustomPlayer.PrestigeCloseButton.SetText(StringToMessage("X"))
        CustomPlayer.InfoText.SetText(StringToMessage("Restart your progress to earn prestige shards"))
        CustomPlayer.KeepTitle.SetText(StringToMessage("KEEP"))
        CustomPlayer.KeepText.SetText(StringToMessage("- General Upgrades\n- Weapon Upgrades\n- Pets (Lvs. & Evos)\n- Drill Buffs\n- Achievements\n- Area 4 Upgrades\n- Inventory Items*\n\n*You will lose your stone"))
        CustomPlayer.ResetTitle.SetText(StringToMessage("RESET"))
        CustomPlayer.ResetText.SetText(StringToMessage("- Strength Upgrades\n- Stages Unlocked\n- Cave Hut Skills\n- Current Stone"))
        CustomPlayer.NextPrestigeTitle.SetText(StringToMessage("Stone for next Prestige"))
        CustomPlayer.PrestigeNowTitle.SetText(StringToMessage("Prestiging now will grant:"))
        CustomPlayer.PrestigeNowText.SetText(StringToMessage("Prestige at the Quarry!"))

        for(TextBlock : CustomPlayer.AllPrestigeTexts):
            TextBlock.SetShadowColor(CustomPlayer.ShadowDefaults.DefaultShadowColor)
            TextBlock.SetShadowOffset(CustomPlayer.ShadowDefaults.DefaultShadowOffset)
            TextBlock.SetShadowOpacity(1.0)

        CustomPlayer.PrestigeNowText.SetShadowColor(CustomPlayer.ShadowDefaults.DefaultShadowColor)
        CustomPlayer.PrestigeNowText.SetShadowOffset(CustomPlayer.ShadowDefaults.DefaultShadowOffset)
        CustomPlayer.PrestigeNowText.SetShadowOpacity(1.0)

        var MyPrestigeCanvas: canvas = canvas:
            Slots := array:
                canvas_slot:
                    Widget:=CustomPlayer.PrestigeHUDBG
                    ZOrder:=0
                    SizeToContent:=false
                    Alignment := vector2{X := 0.5, Y := 0.0}
                    Anchors := anchors{Minimum := vector2{X:= 0.5, Y:= 0.0}, Maximum := vector2{X:= 0.5, Y:= 0.0}}
                    Offsets:=margin{Top:=0.0,Left:=0.0,Bottom:=1076.0,Right:=2048.0}
                canvas_slot:
                    Widget:=CustomPlayer.PrestigeCloseButton
                    ZOrder:=9
                    SizeToContent:=false
                    Anchors := anchors{Minimum := vector2{X:= 0.5, Y:= 0.0}, Maximum := vector2{X:= 0.5, Y:= 0.0}}
                    Offsets:=margin{Top:=CustomPlayer.PrestigeClose_Position.Y,Left:=612.0,Bottom:=CustomPlayer.PrestigeClose_Size.Y,Right:=CustomPlayer.PrestigeClose_Size.X}
                
                canvas_slot:
                    Widget:=CustomPlayer.PrestigeHUDTitle
                    ZOrder:=1
                    SizeToContent:=false
                    Anchors := anchors{Minimum := vector2{X:= 0.5, Y:= 0.0}, Maximum := vector2{X:= 0.5, Y:= 0.0}}
                    Offsets:=margin{Top:=144.0,Left:=0.0, Bottom:=56.0, Right:=294.0}

                canvas_slot:
                    Widget:=CustomPlayer.InfoText
                    ZOrder:=1
                    SizeToContent:=true
                    Anchors := anchors{Minimum := vector2{X:= 0.5, Y:= 0.0}, Maximum := vector2{X:= 0.5, Y:= 0.0}}
                    Offsets:=margin{Top:=296.0,Left:=-628.0}
                    Alignment := vector2{X:= 0.0, Y:= 0.0}
                canvas_slot:
                    Widget:=CustomPlayer.KeepTitle
                    ZOrder:=1
                    SizeToContent:=true
                    Anchors := anchors{Minimum := vector2{X:= 0.5, Y:= 0.0}, Maximum := vector2{X:= 0.5, Y:= 0.0}}
                    Offsets:=margin{Top:=368.0,Left:=-492.0}
                    Alignment := vector2{X:= 0.5, Y:= 0.0}
                canvas_slot:
                    Widget:=CustomPlayer.KeepText
                    ZOrder:=1
                    SizeToContent:=true
                    Anchors := anchors{Minimum := vector2{X:= 0.5, Y:= 0.0}, Maximum := vector2{X:= 0.5, Y:= 0.0}}
                    Offsets:=margin{Top:=452.0,Left:=-664.0}
                    Alignment := vector2{X:= 0.0, Y:= 0.0}
                canvas_slot:
                    Widget:=CustomPlayer.ResetTitle
                    ZOrder:=1
                    SizeToContent:=true
                    Anchors := anchors{Minimum := vector2{X:= 0.5, Y:= 0.0}, Maximum := vector2{X:= 0.5, Y:= 0.0}}
                    Offsets:=margin{Top:=368.0,Left:=-8.0}
                    Alignment := vector2{X:= 0.5, Y:= 0.0}
                canvas_slot:
                    Widget:=CustomPlayer.ResetText
                    ZOrder:=1
                    SizeToContent:=true
                    Anchors := anchors{Minimum := vector2{X:= 0.5, Y:= 0.0}, Maximum := vector2{X:= 0.5, Y:= 0.0}}
                    Offsets:=margin{Top:=452.0,Left:=-188.0}
                    Alignment := vector2{X:= 0.0, Y:= 0.0}
                canvas_slot:
                    Widget:=CustomPlayer.CurrentPrestigeLevelText
                    ZOrder:=1
                    SizeToContent:=true
                    Anchors := anchors{Minimum := vector2{X:= 0.5, Y:= 0.0}, Maximum := vector2{X:= 0.5, Y:= 0.0}}
                    Offsets:=margin{Top:=300.0,Left:=552.0}
                    Alignment := vector2{X:= 0.5, Y:= 0.0}
                canvas_slot:
                    Widget:=CustomPlayer.NextPrestigeTitle
                    ZOrder:=1
                    SizeToContent:=true
                    Anchors := anchors{Minimum := vector2{X:= 0.5, Y:= 0.0}, Maximum := vector2{X:= 0.5, Y:= 0.0}}
                    Offsets:=margin{Top:=400.0,Left:=552.0}
                    Alignment := vector2{X:= 0.5, Y:= 0.0}
                canvas_slot:
                    ZOrder:=1
                    SizeToContent:=true
                    Anchors := anchors{Minimum := vector2{X:= 0.5, Y:= 0.0}, Maximum := vector2{X:= 0.5, Y:= 0.0}}
                    Offsets:=margin{Top:=464.0,Left:=552.0}
                    Alignment := vector2{X:= 0.5, Y:= 0.0}
                    Widget := stack_box:
                        Orientation := orientation.Horizontal
                        Slots := array:
                            stack_box_slot:
                                Widget := CustomPlayer.NextPrestigeStoneIcon   
                            stack_box_slot:
                                Widget := CustomPlayer.NextPrestigeAmount
                canvas_slot:
                    Widget:=CustomPlayer.PrestigeNowTitle
                    ZOrder:=1
                    SizeToContent:=true
                    Anchors := anchors{Minimum := vector2{X:= 0.5, Y:= 0.0}, Maximum := vector2{X:= 0.5, Y:= 0.0}}
                    Offsets:=margin{Top:=588.0,Left:=552.0}
                    Alignment := vector2{X:= 0.5, Y:= 0.0}
                canvas_slot:
                    ZOrder:=1
                    SizeToContent:=true
                    Anchors := anchors{Minimum := vector2{X:= 0.5, Y:= 0.0}, Maximum := vector2{X:= 0.5, Y:= 0.0}}
                    Offsets:=margin{Top:=656.0,Left:=552.0}
                    Alignment := vector2{X:= 0.5, Y:= 0.0}
                    Widget := stack_box:
                        Orientation := orientation.Horizontal
                        Slots := array:
                            stack_box_slot:
                                Widget := CustomPlayer.PrestigeNowShardIcon   
                            stack_box_slot:
                                Widget := CustomPlayer.PrestigeNowShards
                canvas_slot:
                    ZOrder:=1
                    SizeToContent:=true
                    Anchors := anchors{Minimum := vector2{X:= 0.5, Y:= 0.0}, Maximum := vector2{X:= 0.5, Y:= 0.0}}
                    Offsets:=margin{Top:=712.0,Left:=552.0}
                    Alignment := vector2{X:= 0.5, Y:= 0.0}
                    Widget := stack_box:
                        Orientation := orientation.Horizontal
                        Slots := array:
                            stack_box_slot:
                                Widget := CustomPlayer.PrestigeNowGatherIcon   
                            stack_box_slot:
                                Widget := CustomPlayer.PrestigeNowGather
                canvas_slot:
                    Widget:=CustomPlayer.PrestigeNowText
                    ZOrder:=1
                    SizeToContent:=true
                    Anchors := anchors{Minimum := vector2{X:= 0.5, Y:= 0.0}, Maximum := vector2{X:= 0.5, Y:= 0.0}}
                    Offsets:=margin{Top:=824.0,Left:=552.0, Bottom:=80.0, Right:=400.0}
                    Alignment := vector2{X:= 0.5, Y:= 0.0}
                
        return MyPrestigeCanvas

    UpdatePrestigeHUD(CustomPlayer: Custom_Player):void= 
        var PrestigeShardAmount : float = 0.0
        var PrestigeGatherAmount : float = 0.0
        var NewPrestigeCost : float = 0.0
        if(CustomPlayer.PrestigeCost * Pow(CustomPlayer.PrestigeCostMult, (CustomPlayer.PrestigeLevel)*1.0) > 1e300):
            set NewPrestigeCost = 1e300
        else:
            set NewPrestigeCost = CustomPlayer.PrestigeCost * Pow(CustomPlayer.PrestigeCostMult, (CustomPlayer.PrestigeLevel)*1.0)
        CustomPlayer.CurrentPrestigeLevelText.SetText(StringToMessage("Prestige Level: {CustomPlayer.PrestigeLevel}"))
        if(CustomPlayer.PrestigeStoneAmount >= NewPrestigeCost):
            CustomPlayer.NextPrestigeAmount.SetTextColor(MakeColorFromHex("39FF39"))
        else:
            CustomPlayer.NextPrestigeAmount.SetTextColor(MakeColorFromHex("fa4d56"))

        ratio := CustomPlayer.PrestigeStoneAmount / NewPrestigeCost
        if(CustomPlayer.PrestigeLevel <= 220):
            if(CustomPlayer.PrestigeLevel >= 1):
                set PrestigeShardAmount = ((Log(5.0, ratio)) * 5) * (1.5 * CustomPlayer.PrestigeLevel)
            else:
                set PrestigeShardAmount = ((Log(5.0, ratio)) * 5)
            
            if(CustomPlayer.StagesUnlocked >= 15):
                set PrestigeGatherAmount = (PrestigeShardAmount/5.0) * 3.0
            else:
                set PrestigeGatherAmount = (PrestigeShardAmount/5.0)
            if(PrestigeShardAmount < 1.0):
                set PrestigeShardAmount = 1.0
            if(PrestigeGatherAmount < 1.0):
                set PrestigeGatherAmount = 1.0

            set PrestigeShardAmount = PrestigeShardAmount * (1.0 + CustomPlayer.SinkPrestigePower / 10000.0)
            set PrestigeGatherAmount = PrestigeGatherAmount * (1.0 + CustomPlayer.SinkPrestigePower / 10000.0)
        else:
            set PrestigeShardAmount = CustomPlayer.EndGamePrestigeShardAmount * (1.0 + CustomPlayer.SinkPrestigePower / 10000.0)
            set PrestigeGatherAmount = CustomPlayer.EndGamePrestigeGatherAmount * (1.0 + CustomPlayer.SinkPrestigePower / 10000.0)

        CustomPlayer.NextPrestigeAmount.SetText(StringToMessage(" {NumberCondensor(CustomPlayer.PrestigeStoneAmount)} / {NumberCondensor(NewPrestigeCost)}"))

        if(FlooredPrestige:=Floor[PrestigeShardAmount]):
            CustomPlayer.PrestigeNowShards.SetText(StringToMessage(" +{FlooredPrestige}"))
        CustomPlayer.PrestigeNowGather.SetText(StringToMessage(" +{NumberCondensor(PrestigeGatherAmount)}% Strength"))


#NUMBER NOTATION
    #HUD STUFF
        
    #CANVAS
    MakeNNCanvas(CustomPlayer:Custom_Player):canvas=
        CustomPlayer.NNCloseButton.SetText(StringToMessage("X"))
        CustomPlayer.Row1Letter.SetText(StringToMessage("K\nM\nB\nT\nQa\nQt\nSx\nSp\nOc\nNo\nDc\nUDc\nDDc\nTDc\nQaDc\nQtDc\nSxDc"))
        CustomPlayer.Row1Number.SetText(StringToMessage("e3\ne6\ne9\ne12\ne15\ne18\ne21\ne24\ne27\ne30\ne33\ne36\ne39\ne42\ne45\ne48\ne51"))
        
        CustomPlayer.Row2Letter.SetText(StringToMessage("SpDc\nODc\nNDc\nVg\nUVg\nDVg\nTVg\nQaVg\nQtVg\nSxVg\nSpVg\nOVg\nNVg\nTg\nUTg\nDTg\nTTg"))
        CustomPlayer.Row2Number.SetText(StringToMessage("e54\ne57\ne60\ne63\ne66\ne69\ne72\ne75\ne78\ne81\ne84\ne87\ne90\ne93\ne96\ne99\ne102"))
        
        CustomPlayer.Row3Letter.SetText(StringToMessage("QaTg\nQtTg\nSxTg\nSpTg\nOTg\nNTg\nQd\nUQd\nDQd\nTQd\nQaQd\nQtQd\nSxQd\nSpQd\nOQd\nNQd\nQi"))
        CustomPlayer.Row3Number.SetText(StringToMessage("e105\ne108\ne111\ne114\ne117\ne120\ne123\ne126\ne129\ne132\ne135\ne138\ne141\ne144\ne147\ne150\ne153"))
        
        CustomPlayer.Row4Letter.SetText(StringToMessage("UQi\nDQi\nTQi\nQaQi\nQtQi\nSxQi\nSpQi\nOQi\nNQi\nSe\nUSe\nDSe\nTSe\nQaSe\nQtSe\nSxSe\nSpSe"))
        CustomPlayer.Row4Number.SetText(StringToMessage("e156\ne159\ne162\ne165\ne168\ne171\ne174\ne177\ne180\ne183\ne186\ne189\ne192\ne195\ne198\ne201\ne204"))


        for(TextBlock : CustomPlayer.AllNNTexts):
            TextBlock.SetShadowColor(CustomPlayer.ShadowDefaults.DefaultShadowColor)
            TextBlock.SetShadowOffset(CustomPlayer.ShadowDefaults.DefaultShadowOffset)
            TextBlock.SetShadowOpacity(1.0)

        var MyNNCanvas: canvas = canvas:
            Slots := array:
                canvas_slot:
                    Widget:=CustomPlayer.NNHUDBG
                    ZOrder:=0
                    SizeToContent:=false
                    Alignment := vector2{X := 0.5, Y := 0.0}
                    Anchors := anchors{Minimum := vector2{X:= 0.5, Y:= 0.0}, Maximum := vector2{X:= 0.5, Y:= 0.0}}
                    Offsets:=margin{Top:=0.0,Left:=0.0,Bottom:=1076.0,Right:=2048.0}
                canvas_slot:
                    Widget:=CustomPlayer.NNCloseButton
                    ZOrder:=9
                    SizeToContent:=false
                    Anchors := anchors{Minimum := vector2{X:= 0.5, Y:= 0.0}, Maximum := vector2{X:= 0.5, Y:= 0.0}}
                    Offsets:=margin{Top:=108.0,Left:=604.0,Bottom:=CustomPlayer.NNClose_Size.Y,Right:=CustomPlayer.NNClose_Size.X}
                canvas_slot:
                    Widget:=CustomPlayer.NNHUDTitle
                    ZOrder:=1
                    SizeToContent:=false
                    Alignment := vector2{X := 0.5, Y := 0.0}
                    Anchors := anchors{Minimum := vector2{X:= 0.5, Y:= 0.0}, Maximum := vector2{X:= 0.5, Y:= 0.0}}
                    Offsets:=margin{Top:=120.0,Left:=0.0, Bottom:=56.0, Right:=623.0}

                canvas_slot:
                    Widget:=CustomPlayer.Row1Letter
                    ZOrder:=1
                    SizeToContent:=true
                    Anchors := anchors{Minimum := vector2{X:= 0.5, Y:= 0.0}, Maximum := vector2{X:= 0.5, Y:= 0.0}}
                    Offsets:=margin{Top:=236.0,Left:=-656.0}
                canvas_slot:
                    Widget:=CustomPlayer.Row1Number
                    ZOrder:=1
                    SizeToContent:=true
                    Anchors := anchors{Minimum := vector2{X:= 0.5, Y:= 0.0}, Maximum := vector2{X:= 0.5, Y:= 0.0}}
                    Offsets:=margin{Top:=236.0,Left:=-424.0}
                    Alignment := vector2{X:= 1.0, Y:= 0.0}
                canvas_slot:
                    Widget:=CustomPlayer.Row2Letter
                    ZOrder:=1
                    SizeToContent:=true
                    Anchors := anchors{Minimum := vector2{X:= 0.5, Y:= 0.0}, Maximum := vector2{X:= 0.5, Y:= 0.0}}
                    Offsets:=margin{Top:=236.0,Left:=-292.0}
                canvas_slot:
                    Widget:=CustomPlayer.Row2Number
                    ZOrder:=1
                    SizeToContent:=true
                    Anchors := anchors{Minimum := vector2{X:= 0.5, Y:= 0.0}, Maximum := vector2{X:= 0.5, Y:= 0.0}}
                    Offsets:=margin{Top:=236.0,Left:=-60.0}
                    Alignment := vector2{X:= 1.0, Y:= 0.0}
                canvas_slot:
                    Widget:=CustomPlayer.Row3Letter
                    ZOrder:=1
                    SizeToContent:=true
                    Anchors := anchors{Minimum := vector2{X:= 0.5, Y:= 0.0}, Maximum := vector2{X:= 0.5, Y:= 0.0}}
                    Offsets:=margin{Top:=236.0,Left:=60.0}
                canvas_slot:
                    Widget:=CustomPlayer.Row3Number
                    ZOrder:=1
                    SizeToContent:=true
                    Anchors := anchors{Minimum := vector2{X:= 0.5, Y:= 0.0}, Maximum := vector2{X:= 0.5, Y:= 0.0}}
                    Offsets:=margin{Top:=236.0,Left:=292.0}
                    Alignment := vector2{X:= 1.0, Y:= 0.0}
                canvas_slot:
                    Widget:=CustomPlayer.Row4Letter
                    ZOrder:=1
                    SizeToContent:=true
                    Anchors := anchors{Minimum := vector2{X:= 0.5, Y:= 0.0}, Maximum := vector2{X:= 0.5, Y:= 0.0}}
                    Offsets:=margin{Top:=236.0,Left:=424.0}
                canvas_slot:
                    Widget:=CustomPlayer.Row4Number
                    ZOrder:=1
                    SizeToContent:=true
                    Anchors := anchors{Minimum := vector2{X:= 0.5, Y:= 0.0}, Maximum := vector2{X:= 0.5, Y:= 0.0}}
                    Offsets:=margin{Top:=236.0,Left:=656.0}
                    Alignment := vector2{X:= 1.0, Y:= 0.0}
                
        return MyNNCanvas      

#INFO BOARD
    #CANVAS
    MakeInfoCanvas(CustomPlayer:Custom_Player):canvas=
        CustomPlayer.InfoCloseButton.SetText(StringToMessage("X"))

        var MyInfoCanvas: canvas = canvas:
            Slots := array:
                canvas_slot:
                    Widget:=CustomPlayer.InfoBackground
                    ZOrder:=0
                    SizeToContent:=false
                    Alignment := vector2{X := 0.5, Y := 0.0}
                    Anchors := anchors{Minimum := vector2{X:= 0.5, Y:= 0.0}, Maximum := vector2{X:= 0.5, Y:= 0.0}}
                    Offsets:=margin{Top:=0.0,Left:=0.0,Bottom:=1076.0,Right:=2048.0}
                canvas_slot:
                    Widget:=CustomPlayer.InfoCloseButton
                    ZOrder:=9
                    SizeToContent:=false
                    Offsets:=margin{Top:=CustomPlayer.NNClose_Position.Y,Left:=CustomPlayer.NNClose_Position.X,Bottom:=CustomPlayer.NNClose_Size.Y,Right:=CustomPlayer.NNClose_Size.X}

                canvas_slot:
                    Widget:=CustomPlayer.AccoladesNoBG
                    ZOrder:=0
                    SizeToContent:=false
                    Alignment := vector2{X := 0.5, Y := 0.0}
                    Anchors := anchors{Minimum := vector2{X:= 0.5, Y:= 0.0}, Maximum := vector2{X:= 0.5, Y:= 0.0}}
                    Offsets:=margin{Top:=250.0,Left:=-550.0,Bottom:=600.0,Right:=561.0}
                canvas_slot:
                    Widget:=CustomPlayer.SignalNoBG
                    ZOrder:=0
                    SizeToContent:=false
                    Alignment := vector2{X := 0.5, Y := 0.0}
                    Anchors := anchors{Minimum := vector2{X:= 0.5, Y:= 0.0}, Maximum := vector2{X:= 0.5, Y:= 0.0}}
                    Offsets:=margin{Top:=250.0,Left:=0.0,Bottom:=600.0,Right:=561.0}
                canvas_slot:
                    Widget:=CustomPlayer.InfoNoBG
                    ZOrder:=0
                    SizeToContent:=false
                    Alignment := vector2{X := 0.5, Y := 0.0}
                    Anchors := anchors{Minimum := vector2{X:= 0.5, Y:= 0.0}, Maximum := vector2{X:= 0.5, Y:= 0.0}}
                    Offsets:=margin{Top:=250.0,Left:=550.0,Bottom:=600.0,Right:=561.0}
                
        return MyInfoCanvas      

#CAVE MAP BOARD
    #CANVAS
    MakeMapCanvas(CustomPlayer:Custom_Player):canvas=
        CustomPlayer.MapCloseButton.SetText(StringToMessage("X"))

        var MyMapCanvas: canvas = canvas:
            Slots := array:
                canvas_slot:
                    Widget:=CustomPlayer.InfoBackground
                    ZOrder:=0
                    SizeToContent:=false
                    Alignment := vector2{X := 0.5, Y := 0.0}
                    Anchors := anchors{Minimum := vector2{X:= 0.5, Y:= 0.0}, Maximum := vector2{X:= 0.5, Y:= 0.0}}
                    Offsets:=margin{Top:=0.0,Left:=0.0,Bottom:=1076.0,Right:=2048.0}
                canvas_slot:
                    Widget:=CustomPlayer.MapCloseButton
                    ZOrder:=9
                    SizeToContent:=false
                    Offsets:=margin{Top:=CustomPlayer.NNClose_Position.Y,Left:=CustomPlayer.NNClose_Position.X,Bottom:=CustomPlayer.NNClose_Size.Y,Right:=CustomPlayer.NNClose_Size.X}

                canvas_slot:
                    Widget:=CustomPlayer.CaveMapHUDBG
                    ZOrder:=0
                    SizeToContent:=false
                    Alignment := vector2{X := 0.5, Y := 0.0}
                    Anchors := anchors{Minimum := vector2{X:= 0.5, Y:= 0.0}, Maximum := vector2{X:= 0.5, Y:= 0.0}}
                    Offsets:=margin{Top:=150.0,Left:=0.0,Bottom:=800.0,Right:=748.0}
        return MyMapCanvas    
