using { /Fortnite.com/Devices }
using { /Verse.org/Simulation }
using { /Verse.org/Colors }
using { /Verse.org/Verse }
using { /Verse.org/Random }
using { /Fortnite.com/Teams }
using { /UnrealEngine.com/Assets }
using { /Fortnite.com/Game }
using { /Fortnite.com/Characters }
using { /Fortnite.com/Playspaces }
using { /Fortnite.com/UI }
using { /UnrealEngine.com/Temporary/Diagnostics }
using { /UnrealEngine.com/Temporary/SpatialMath }
using { /UnrealEngine.com/Temporary/UI }
using { /Verse.org/Assets}

ui_prestige_upgrades := class(creative_device):

    @editable
    gm : game_manager = game_manager{}

    @editable
    AudioManager : audio_manager = audio_manager{}

    @editable
    SUP : signal_device_manager = signal_device_manager{}

    @editable
    RoomUnlocks : room_unlocks = room_unlocks{}

    @editable
    OpenPUPButtons : []button_device = array{}

    @editable
    Tutorial : tutorial_manager = tutorial_manager{}

    @editable
    ResetAllDataButton : button_device = button_device{}

    var AllPrestigeUpgradeTitles : [][]string = array{}
    var AllPrestigeUpgradeDescs : [][]string = array{}
    var AllPrestigeTextures : [][]texture = array{}
    var AllPrestigeUpgradeAmounts : [][]float = array{}
    var AllPrestigeMaxLevels : [][]int = array{}

    var MaxLevels : []int = array{110,5,10}
    var EggMaxLevels : []int = {10,10,5}
    var UnlockMaxLevels : []int = {2,2,21}

    var UpgradeAmounts : []float = array{1.0, 0.05, 0.20}
    var EggUpgradeAmounts : []float = array{1.0, 0.02, 1.0}
    var UnlockUpgradeAmounts : []float = array{1.0, 1.0, 1.0}

    var PUPCostMult : []float = array{3.0,4.0,3.0} #THIS IS PURELY FOR THE HUD
    var PUPPetCostMult : []float = array{3.0,2.0,8.0} #THIS IS PURELY FOR THE HUD
    var PUPUnlockCostMult : []float = array{1.0,1.0,1.0} #THIS IS PURELY FOR THE HUD
    
    var PUPUnlockCosts : []float = array{10000.0, 2000.0, 1000.0}
    var PUPCosts : []float = array{1.0, 2.0, 4.0}
    var EggPrestigeCosts : []float = array{2.0, 1.0, 5.0}

    var AllPUPCostMults : [][]float = array{}

    var UpgradeTitles : []string = array{
        "Strength Surge",
        "Auto Gather",
        "Boss Reward Multiplier"
        }

    var EggUpgradeTitles : []string = array {
        "Egg Drop Amount",
        "Egg Drop Rate",
        "Pet Slot"
    }
    var UnlockUpgradeTitles : []string = array {
        "Auto Strength Upgrades",
        "Hut Upgrade All",
        "Permanent Room Unlock"
    }
        
    var UpgradeDescs : []string = array{
        "Significantly increases\nefficiency upgrades\nfor strength tier",
        "Automatically collect\nStone",
        "Increase Boss Rewards"
        }
    
    var EggDescriptions : []string = array{
        "Increase the amount\nof eggs drop",
        "Increase the chance\nthat eggs drop",
        "Unlock an additional pet slot"
    }
    var UnlockDescriptions : []string = array{
        "Unlock Auto Strength\n Upgrade button",
        "Unlock Hut Upgrade\nall button",
        "Permanently unlock each\nroom through prestige"
    }

    var Textures :[]texture = array{UI.StrengthSurge,UI.AutoGather,UI.BossRewards}
    var EggTextures : []texture = array{UI.EggDrop,UI.EggDropRate,UI.petslots}
    var UnlockTextures : []texture = array{UI.AutoStrength,UI.UpgradeAllHuts,UI.UnlockCaveRoom}


    OnBegin<override>()<suspends>:void=
        ResetAllDataButton.InteractedWithEvent.Subscribe(OpenResetAllHUD)

        for(i:int=0..OpenPUPButtons.Length):
            if(OpenPUPButton:=OpenPUPButtons[i]):
                OpenPUPButton.InteractedWithEvent.SubscribeAgent(OpenPUPHUD, i)

        for(Spawner : gm.PlayerSpawners):     
            Spawner.SpawnedEvent.Subscribe(OnPlayerJoined)

    OnPlayerJoined(Agent:agent):void=
        spawn{OnPlayerJoinedBuffer(Agent)}

    OnPlayerJoinedBuffer(Agent:agent)<suspends>:void=
        Sleep(1.0)
        if(CustomPlayer := gm.PlayersMap[player[Agent]]):
            
            set CustomPlayer.PUPShardIcons = array{CustomPlayer.ShardIcon1,CustomPlayer.ShardIcon2,CustomPlayer.ShardIcon3}
            set CustomPlayer.PUPUpgradeTitlesBlock = array{CustomPlayer.Upgrade1Title,CustomPlayer.Upgrade2Title,CustomPlayer.Upgrade3Title}
            set CustomPlayer.PUPUpgradeLevelsBlock = array{CustomPlayer.Upgrade1Level,CustomPlayer.Upgrade2Level,CustomPlayer.Upgrade3Level}
            set CustomPlayer.PUPUpgradeIconsBlock = array{CustomPlayer.Upgrade1Icon,CustomPlayer.Upgrade2Icon,CustomPlayer.Upgrade3Icon}
            set CustomPlayer.PUPUpgradeDescTextsBlock = array{CustomPlayer.Upgrade1Desc,CustomPlayer.Upgrade2Desc,CustomPlayer.Upgrade3Desc}
            set CustomPlayer.PUPUpgradeBonusesBlock = array{CustomPlayer.Upgrade1Bonus,CustomPlayer.Upgrade2Bonus,CustomPlayer.Upgrade3Bonus}
            set CustomPlayer.PUPUpgradeNextBonusesBlock = array{CustomPlayer.Upgrade1NextBonus,CustomPlayer.Upgrade2NextBonus,CustomPlayer.Upgrade3NextBonus}
            set CustomPlayer.PUPUpgradeCostsBlock = array{CustomPlayer.Upgrade1CostAmount,CustomPlayer.Upgrade2CostAmount,CustomPlayer.Upgrade3CostAmount}
            set CustomPlayer.PUPUpgradeCostTextsBlock = array{CustomPlayer.Upgrade1CostText,CustomPlayer.Upgrade2CostText,CustomPlayer.Upgrade3CostText}
            set CustomPlayer.PUPUpgradeButtonsBlock = array{CustomPlayer.PUPUpgradeButton1,CustomPlayer.PUPUpgradeButton2,CustomPlayer.PUPUpgradeButton3}
    
            set AllPrestigeUpgradeTitles = array{UpgradeTitles,EggUpgradeTitles, UnlockUpgradeTitles}
            set AllPrestigeUpgradeDescs = array{UpgradeDescs,EggDescriptions,UnlockDescriptions}
            set AllPrestigeTextures = array{Textures, EggTextures, UnlockTextures} 
            set AllPrestigeUpgradeAmounts = array{UpgradeAmounts,EggUpgradeAmounts,UnlockUpgradeAmounts}
            set AllPrestigeMaxLevels = array{MaxLevels,EggMaxLevels, UnlockMaxLevels}
            set AllPUPCostMults = array{PUPCostMult,PUPPetCostMult, PUPUnlockCostMult}
    
            set CustomPlayer.AllPrestigeTexts = array{CustomPlayer.InfoText,CustomPlayer.KeepTitle,CustomPlayer.KeepText,CustomPlayer.ResetTitle,CustomPlayer.ResetText,CustomPlayer.NextPrestigeTitle,CustomPlayer.NextPrestigeAmount,CustomPlayer.CurrentPrestigeLevelText,CustomPlayer.PrestigeNowTitle,CustomPlayer.PrestigeNowShards,CustomPlayer.PrestigeNowGather}
            set CustomPlayer.AllPUPTexts = array{CustomPlayer.ShardAmountText, CustomPlayer.Upgrade1Title, CustomPlayer.Upgrade1Level,CustomPlayer.Upgrade1Desc,CustomPlayer.Upgrade1Bonus,CustomPlayer.Upgrade1NextBonus,CustomPlayer.Upgrade1CostText,CustomPlayer.Upgrade1CostAmount,CustomPlayer.Upgrade2Title,CustomPlayer.Upgrade2Level,CustomPlayer.Upgrade2Desc,CustomPlayer.Upgrade2Bonus,CustomPlayer.Upgrade2NextBonus,CustomPlayer.Upgrade2CostText,CustomPlayer.Upgrade2CostAmount,CustomPlayer.Upgrade3Title,CustomPlayer.Upgrade3Level,CustomPlayer.Upgrade3Desc,CustomPlayer.Upgrade3Bonus,CustomPlayer.Upgrade3NextBonus,CustomPlayer.Upgrade3CostText,CustomPlayer.Upgrade3CostAmount}
       
            if(CustomPlayer.PlayTimeSessionSeconds = 0.0):
                CustomPlayer.PUPCloseButton.OnClick().Subscribe(RemovePUPHUD)
                CustomPlayer.PrestigeCloseButton.OnClick().Subscribe(RemovePrestigeHUD)

                CustomPlayer.ResetAllNowButton.OnClick().SubscribeAgent3(ResetAllPlayerData, 0)
                CustomPlayer.ResetAllConfirmButton.OnClick().SubscribeAgent3(ResetAllPlayerData, 1)
                CustomPlayer.ResetAllCloseButton.OnClick().Subscribe(RemoveOpenResetAllHUD)
    
                for(i:int=0..CustomPlayer.PUPUpgradeButtonsBlock.Length):
                    if(UpgradeButton:=CustomPlayer.PUPUpgradeButtonsBlock[i]):
                        UpgradeButton.OnClick().SubscribeAgent3(OnUpgradeButton, i)
    
                CustomPlayer.PrestigeNowButton.OnClick().Subscribe(OnPrestigeNow)

                spawn{AutoGather(CustomPlayer)}
    #CODE
    OpenPUPHUD(Agent:agent, which:int): void=
        if(CustomPlayer := gm.PlayersMap[Agent]):
            if(which <> 3):
                if(MyUI:=CustomPlayer.UIMaps["PUP"]?):
                    set CustomPlayer.CurrentPUP = which
                else:
                    var MyPUPCanvas : canvas = MakePUPCanvas(CustomPlayer)
                    set CustomPlayer.CurrentPUP = which
                    CustomPlayer.AddFullHUD(MyPUPCanvas)
                    if(set CustomPlayer.UIMaps["PUP"] = option{MyPUPCanvas}) {}
                    UpdatePUPHUD(CustomPlayer)
            else:
                if(MyUI:=CustomPlayer.UIMaps["Prestige"]?):
                else:
                    var MyPrestigeCanvas : canvas = MakePrestigeCanvas(CustomPlayer)
                    CustomPlayer.AddFullHUD(MyPrestigeCanvas)
                    if(set CustomPlayer.UIMaps["Prestige"] = option{MyPrestigeCanvas}) {}
                    UpdatePrestigeHUD(CustomPlayer)

    RemovePUPHUD(Wdata:widget_message):void=
        if(CustomPlayer := gm.PlayersMap[Wdata.Player]):
            if(Player:=Wdata.Player, PlayerUI:=GetPlayerUI[Player]):
                if(MyUI:=CustomPlayer.UIMaps["PUP"]?):
                    PlayerUI.RemoveWidget(MyUI)
                    MyUI.SetVisibility(widget_visibility.Collapsed)
                    if(set CustomPlayer.UIMaps["PUP"] = false){}

    RemovePrestigeHUD(Wdata:widget_message):void=
        if(CustomPlayer := gm.PlayersMap[Wdata.Player]):
            if(Player:=Wdata.Player, PlayerUI:=GetPlayerUI[Player]):
                if(MyUI:=CustomPlayer.UIMaps["Prestige"]?):
                    PlayerUI.RemoveWidget(MyUI)
                    MyUI.SetVisibility(widget_visibility.Collapsed)
                    if(set CustomPlayer.UIMaps["Prestige"] = false){}

    UpdatePUPHUD(CustomPlayer: Custom_Player):void= 
        set CustomPlayer.AllPrestigeLevels = array{CustomPlayer.PUPLevels,CustomPlayer.EggPrestigeLevels,CustomPlayer.PUPQOLLevels}
        set CustomPlayer.AllPrestigeCosts = array{PUPCosts,EggPrestigeCosts,PUPUnlockCosts}
        if:
            PrestigeMaxLevels:=AllPrestigeMaxLevels[CustomPlayer.CurrentPUP]
            PrestigeUpgradeAmounts:=AllPrestigeUpgradeAmounts[CustomPlayer.CurrentPUP]
            PrestigeCosts:=CustomPlayer.AllPrestigeCosts[CustomPlayer.CurrentPUP]
            PrestigeLevels:=CustomPlayer.AllPrestigeLevels[CustomPlayer.CurrentPUP]
            PrestigeTitles:=AllPrestigeUpgradeTitles[CustomPlayer.CurrentPUP]
            PrestigeDescs:=AllPrestigeUpgradeDescs[CustomPlayer.CurrentPUP]
            PrestigeTextures:=AllPrestigeTextures[CustomPlayer.CurrentPUP]
        then:
            CustomPlayer.ShardAmountText.SetText(StringToMessage(NumberCondensor(CustomPlayer.ShardAmount)))
            for(i:int=0..CustomPlayer.PUPUpgradeTitlesBlock.Length):
                if:
                    ShardIcon:=CustomPlayer.PUPShardIcons[i]
                    UpgradeNextBonus:=CustomPlayer.PUPUpgradeNextBonusesBlock[i]
                    UpgradeCostText:=CustomPlayer.PUPUpgradeCostTextsBlock[i]
                    MaxLevel:=PrestigeMaxLevels[i]
                    Level:=PrestigeLevels[i]
                    Upgrade:=PrestigeUpgradeAmounts[i]*Level
                    UpgradeAmount:=PrestigeUpgradeAmounts[i]
                    CostMults:=AllPUPCostMults[CustomPlayer.CurrentPUP]
                    UpgradeTitle:=PrestigeTitles[i]
                    UpgradeDesc:=PrestigeDescs[i]
                    UpgradeIcon:=PrestigeTextures[i]
                    UpgradeTitleBlock:=CustomPlayer.PUPUpgradeTitlesBlock[i]
                    UpgradeLevel:=CustomPlayer.PUPUpgradeLevelsBlock[i]
                    UpgradeIconBlock:=CustomPlayer.PUPUpgradeIconsBlock[i]
                    UpgradeDescBlock:=CustomPlayer.PUPUpgradeDescTextsBlock[i]
                    UpgradeBonus:=CustomPlayer.PUPUpgradeBonusesBlock[i]
                    UpgradeCost:=CustomPlayer.PUPUpgradeCostsBlock[i]
                    UpgradeButton:=CustomPlayer.PUPUpgradeButtonsBlock[i]
                then:
                    UpgradeTitleBlock.SetText(StringToMessage(UpgradeTitle))
                    UpgradeLevel.SetText(StringToMessage("Lv. {Level}"))
                    UpgradeIconBlock.SetImage(UpgradeIcon)
                    UpgradeDescBlock.SetText(StringToMessage(UpgradeDesc))
                    var Cost : float = 0.0
                    if(CustomPlayer.CurrentPUP = 0 and i=0):
                        set Cost = (1 + (100*Level)) * 1.0
                        #Print("Cost: {Cost}")
                        if(CustomPlayer.ShardAmount >= Cost):
                            UpgradeCost.SetTextColor(MakeColorFromHex("39FF39"))
                            UpgradeCost.SetText(StringToMessage(NumberCondensor(Cost)))
                        else:
                            UpgradeCost.SetTextColor(MakeColorFromHex("fa4d56"))
                        UpgradeCost.SetText(StringToMessage(NumberCondensor(Cost)))
                    else if(CustomPlayer.CurrentPUP = 2 and i=0 and CustomPlayer.PUPQOLLevels[0] = 1):
                        set Cost = 2000000.0
                        if(CustomPlayer.ShardAmount >= Cost):
                            UpgradeCost.SetTextColor(MakeColorFromHex("39FF39"))
                            UpgradeCost.SetText(StringToMessage(NumberCondensor(Cost)))
                        else:
                            UpgradeCost.SetTextColor(MakeColorFromHex("fa4d56"))
                        UpgradeCost.SetText(StringToMessage(NumberCondensor(Cost)))
                        UpgradeDescBlock.SetText(StringToMessage("Automatically Upgrades\nStrength on hit\n(Toggle)"))
                    else if(CustomPlayer.CurrentPUP = 2 and i=1 and CustomPlayer.PUPQOLLevels[1] = 1):
                        set Cost = 50000.0
                        if(CustomPlayer.ShardAmount >= Cost):
                            UpgradeCost.SetTextColor(MakeColorFromHex("39FF39"))
                            UpgradeCost.SetText(StringToMessage(NumberCondensor(Cost)))
                        else:
                            UpgradeCost.SetTextColor(MakeColorFromHex("fa4d56"))
                        UpgradeCost.SetText(StringToMessage(NumberCondensor(Cost)))
                        UpgradeDescBlock.SetText(StringToMessage("Keep Huts 1, 2 & 3\nLevels through prestige"))
                    else if(CustomPlayer.CurrentPUP = 2 and i=2):
                        if(Level = 20):
                            set Cost = 20000000.0
                        else:
                            set Cost = (1000 + (1000*Level)) * 1.0
                        if(CustomPlayer.ShardAmount >= Cost):
                            UpgradeCost.SetTextColor(MakeColorFromHex("39FF39"))
                            UpgradeCost.SetText(StringToMessage(NumberCondensor(Cost)))
                        else:
                            UpgradeCost.SetTextColor(MakeColorFromHex("fa4d56"))
                        UpgradeCost.SetText(StringToMessage(NumberCondensor(Cost)))
                    else:
                        if(TotalCost:=PrestigeCosts[i]*(Pow(CostMults[i], Level*1.0))):
                            if(CustomPlayer.ShardAmount >= TotalCost):
                                UpgradeCost.SetTextColor(MakeColorFromHex("39FF39"))
                                UpgradeCost.SetText(StringToMessage(NumberCondensor(TotalCost)))
                            else:
                                UpgradeCost.SetTextColor(MakeColorFromHex("fa4d56"))
                            UpgradeCost.SetText(StringToMessage(NumberCondensor(TotalCost)))
                    var isPerc : logic = true

                    if(CustomPlayer.CurrentPUP = 0 and i=0):
                        set isPerc = false
                    if(CustomPlayer.CurrentPUP = 1 and i=2):
                        set isPerc = false
                    if(CustomPlayer.CurrentPUP = 2):
                        set isPerc = false
                            
                    if(isPerc = true):
                        UpgradeBonus.SetText(StringToMessage("{NumberCondensor(Upgrade*100.0)}%"))
                    else:
                        if(CustomPlayer.CurrentPUP = 0):
                            if(FlooredUpgrade:=Floor[Upgrade]):
                                if(Level > 0):
                                    UpgradeBonus.SetText(StringToMessage("Surged Tiers: 1-{FlooredUpgrade}"))
                                else:
                                    UpgradeBonus.SetText(StringToMessage("Surged Tiers: None"))
                        else:
                            UpgradeBonus.SetText(StringToMessage("{NumberCondensor(Upgrade)}"))

                    if(Level < MaxLevel):
                        UpgradeButton.SetText(StringToMessage("UPGRADE"))
                        ShardIcon.SetVisibility(widget_visibility.Visible)
                        UpgradeCostText.SetText(StringToMessage("COST:"))
                        UpgradeCostText.SetTextColor(MakeColorFromHex("DDDDDD"))
                        if(isPerc = true):
                            UpgradeNextBonus.SetText(StringToMessage(" (+{NumberCondensor(UpgradeAmount*100.0)}%)"))
                        else:
                            if(CustomPlayer.CurrentPUP = 0):
                                if(FlooredUpgrade:=Floor[Upgrade+UpgradeAmount]):
                                    UpgradeNextBonus.SetText(StringToMessage(" ({FlooredUpgrade})"))
                            else:
                                UpgradeNextBonus.SetText(StringToMessage(" (+{NumberCondensor(UpgradeAmount)})"))
                    else:
                        UpgradeButton.SetText(StringToMessage("MAX"))
                        ShardIcon.SetVisibility(widget_visibility.Collapsed)
                        UpgradeCostText.SetText(StringToMessage(""))
                        UpgradeCost.SetText(StringToMessage(""))
                        UpgradeNextBonus.SetText(StringToMessage(""))


    OnUpgradeButton(Wdata:widget_message, WhichUpgrade:int):void=
        var Cost : float = 0.0
        if:
            CustomPlayer := gm.PlayersMap[Wdata.Player]
            PrestigeMaxLevels:=AllPrestigeMaxLevels[CustomPlayer.CurrentPUP]
            PrestigeUpgradeAmounts:=AllPrestigeUpgradeAmounts[CustomPlayer.CurrentPUP]
            PrestigeCosts:=CustomPlayer.AllPrestigeCosts[CustomPlayer.CurrentPUP][WhichUpgrade]
            PrestigeLevels:=CustomPlayer.AllPrestigeLevels[CustomPlayer.CurrentPUP]
            CostMults:=AllPUPCostMults[CustomPlayer.CurrentPUP]
            Level:=CustomPlayer.PUPLevels[WhichUpgrade]
            QOLLevel:=CustomPlayer.PUPQOLLevels[WhichUpgrade]
            CostMult:=Pow(CostMults[WhichUpgrade], PrestigeLevels[WhichUpgrade]*1.0)
            set Cost = PrestigeCosts * CostMult
        then:
            if(PrestigeLevels[WhichUpgrade] < PrestigeMaxLevels[WhichUpgrade]):
                if(CustomPlayer.CurrentPUP = 0 and WhichUpgrade = 0):
                    set Cost = (1 + (100*Level)) * 1.0
                else if(CustomPlayer.CurrentPUP = 2 and WhichUpgrade = 0 and CustomPlayer.PUPQOLLevels[0] = 1):
                    set Cost = 2000000.0
                else if(CustomPlayer.CurrentPUP = 2 and WhichUpgrade = 1 and CustomPlayer.PUPQOLLevels[1] = 1):
                    set Cost = 50000.0
                else if(CustomPlayer.CurrentPUP = 2 and WhichUpgrade = 2):
                    Print("HUT UPGRADE S")
                    if(QOLLevel = 20):
                        set Cost = 20000000.0
                    else:
                        set Cost = (1000 + (1000*QOLLevel)) * 1.0

                Print("Cost ACTUAL: {Cost}")
                if(CustomPlayer.ShardAmount >= Cost):
                    AudioManager.LeveledUp.Play(CustomPlayer.Player)
                    CustomPlayer.Spend(Cost, "Shard", 0)
                    if(CustomPlayer.CurrentPUP = 0):
                        if(WhichUpgrade = 0): #STRENGTH SURGE
                            SUP.UpdatePower(CustomPlayer)
                        else if(WhichUpgrade = 1): #AUTO GATHER
                            if(CustomPlayer.PUPLevels[WhichUpgrade] = 0):
                                spawn:
                                    AutoGather(CustomPlayer)
                            if(set CustomPlayer.AutoGatherRate += PrestigeUpgradeAmounts[WhichUpgrade]){}
                        else if(WhichUpgrade = 2): #BOSS REWARD MULTIPLIER
                            if(set CustomPlayer.BossRewardMult += PrestigeUpgradeAmounts[WhichUpgrade]){}
                            
                        if(set CustomPlayer.PUPLevels[WhichUpgrade] += 1){} #Up level

                    else if(CustomPlayer.CurrentPUP = 1):
                        if(WhichUpgrade = 0): #Egg Drop Amount
                            if(set CustomPlayer.EggAmountMultiplier += PrestigeUpgradeAmounts[WhichUpgrade]){}
                        else if (WhichUpgrade = 1): #Egg Drop Rate
                            if(set CustomPlayer.EggDropRateMultiplier += PrestigeUpgradeAmounts[WhichUpgrade]){}
                        else if (WhichUpgrade = 2): # Pet Slots
                            if(set CustomPlayer.PetEquipSlots[CustomPlayer.EggPrestigeLevels[WhichUpgrade]+1] = true){}
                        if(set CustomPlayer.EggPrestigeLevels[WhichUpgrade] += 1){} #Up level
                        
                    else if(CustomPlayer.CurrentPUP = 2):
                        if(set CustomPlayer.PUPQOLLevels[WhichUpgrade] += 1){} #Up level
                        if:
                            NewStagesUnlocked := CustomPlayer.PUPQOLLevels[2] - 1
                            CustomPlayer.StagesUnlocked < NewStagesUnlocked
                        then:
                            if(NewStagesUnlocked = 20):
                                Print("UNLOCK A5")
                                set CustomPlayer.A5Unlocked = true
                            set CustomPlayer.RoomUnlockGather += 0.05
                            set CustomPlayer.StagesUnlocked += 1
                            RoomUnlocks.UnlockRooms(CustomPlayer)
                            
                    SaveAllData(CustomPlayer)
                else:
                    AudioManager.Denied.Play(CustomPlayer.Player)
            else:
                AudioManager.Denied.Play(CustomPlayer.Player)
            UpdatePUPHUD(CustomPlayer)

    AutoGather(CustomPlayer:Custom_Player)<suspends>:void=
        loop:
            Sleep(3.0)
            if(CustomPlayer.Player.IsActive[]):
                CustomPlayer.Collect(CustomPlayer.GetPlayerDamage() * CustomPlayer.AutoGatherRate, "Stone", 0)
                if(MyUI:=CustomPlayer.UIMaps["SUP"]?):
                    SUP.UpdateSUPHUD(CustomPlayer)
            else:
                break:

#UI BLOCKS

    #HUD STUFF

        
#CANVAS
    MakePUPCanvas(CustomPlayer:Custom_Player):canvas=
        CustomPlayer.PUPCloseButton.SetText(StringToMessage("X"))
        CustomPlayer.Upgrade1Desc.SetJustification(text_justification.Center)
        CustomPlayer.Upgrade2Desc.SetJustification(text_justification.Center)
        CustomPlayer.Upgrade3Desc.SetJustification(text_justification.Center)
        CustomPlayer.Upgrade4Desc.SetJustification(text_justification.Center)

        for(TextBlock : CustomPlayer.AllPUPTexts):
            TextBlock.SetShadowColor(CustomPlayer.ShadowDefaults.DefaultShadowColor)
            TextBlock.SetShadowOffset(CustomPlayer.ShadowDefaults.DefaultShadowOffset)
            TextBlock.SetShadowOpacity(1.0)
        var MyPUPCanvas: canvas = canvas:
            Slots := array:
                canvas_slot:
                    Widget:=CustomPlayer.PUPHUDBG
                    ZOrder:=0
                    SizeToContent:=false
                    Alignment := vector2{X := 0.5, Y := 0.0}
                    Anchors := anchors{Minimum := vector2{X:= 0.5, Y:= 0.0}, Maximum := vector2{X:= 0.5, Y:= 0.0}}
                    Offsets:=margin{Top:=0.0,Left:=0.0,Bottom:=1076.0,Right:=2048.0}
                canvas_slot:
                    Widget:=CustomPlayer.PUPCloseButton
                    ZOrder:=9
                    SizeToContent:=false
                    Anchors := anchors{Minimum := vector2{X:= 0.5, Y:= 0.0}, Maximum := vector2{X:= 0.5, Y:= 0.0}}
                    Alignment := vector2{X:= 0.5, Y:= 0.0}
                    Offsets:=margin{Top:=148.0,Left:=560.0,Bottom:=CustomPlayer.PUPClose_Size.Y,Right:=CustomPlayer.PUPClose_Size.X}
                
                canvas_slot:
                    Widget:=CustomPlayer.PUPHUDTitle
                    ZOrder:=1
                    SizeToContent:=false
                    Anchors := anchors{Minimum := vector2{X:= 0.5, Y:= 0.0}, Maximum := vector2{X:= 0.5, Y:= 0.0}}
                    Offsets:=margin{Top:=144.0,Left:=-304.0, Bottom:=56.0, Right:=633.0}

                canvas_slot:
                    Widget:=CustomPlayer.ShardAmountIcon
                    ZOrder:=1
                    SizeToContent:=false
                    Anchors := anchors{Minimum := vector2{X:= 0.5, Y:= 0.0}, Maximum := vector2{X:= 0.5, Y:= 0.0}}
                    Offsets:=margin{Top:=162.0,Left:=-792.0, Bottom:=40.0, Right:=40.0}
                canvas_slot:
                    Widget:=CustomPlayer.ShardAmountText
                    ZOrder:=1
                    SizeToContent:=true
                    Anchors := anchors{Minimum := vector2{X:= 0.5, Y:= 0.0}, Maximum := vector2{X:= 0.5, Y:= 0.0}}
                    Offsets:=margin{Top:=162.0,Left:=-740.0}
                    Alignment := vector2{X:= 0.0, Y:= 0.0}

                canvas_slot:
                    Widget:=CustomPlayer.Upgrade1Title
                    ZOrder:=1
                    SizeToContent:=true
                    Anchors := anchors{Minimum := vector2{X:= 0.5, Y:= 0.0}, Maximum := vector2{X:= 0.5, Y:= 0.0}}
                    Offsets:=margin{Top:=276.0,Left:=-548.0}
                    Alignment := vector2{X:= 0.5, Y:= 0.0}
                canvas_slot:
                    Widget:=CustomPlayer.Upgrade1Level
                    ZOrder:=1
                    SizeToContent:=true
                    Anchors := anchors{Minimum := vector2{X:= 0.5, Y:= 0.0}, Maximum := vector2{X:= 0.5, Y:= 0.0}}
                    Offsets:=margin{Top:=336.0,Left:=-548.0}
                    Alignment := vector2{X:= 0.5, Y:= 0.0}
                canvas_slot:
                    Widget:=CustomPlayer.Upgrade1Icon
                    ZOrder:=1
                    SizeToContent:=false
                    Anchors := anchors{Minimum := vector2{X:= 0.5, Y:= 0.0}, Maximum := vector2{X:= 0.5, Y:= 0.0}}
                    Offsets:=margin{Top:=416.0,Left:=-548.0, Bottom:=140.0,Right:=140.0}
                    Alignment := vector2{X:= 0.5, Y:= 0.0}
                canvas_slot:
                    Widget:=CustomPlayer.Upgrade1Desc
                    ZOrder:=1
                    SizeToContent:=true
                    Anchors := anchors{Minimum := vector2{X:= 0.5, Y:= 0.0}, Maximum := vector2{X:= 0.5, Y:= 0.0}}
                    Offsets:=margin{Top:=588.0,Left:=-548.0}
                    Alignment := vector2{X:= 0.5, Y:= 0.0}
                canvas_slot:
                    ZOrder:=1
                    SizeToContent:=true
                    Anchors := anchors{Minimum := vector2{X:= 0.5, Y:= 0.0}, Maximum := vector2{X:= 0.5, Y:= 0.0}}
                    Offsets:=margin{Top:=736.0,Left:=-548.0}
                    Alignment := vector2{X:= 0.5, Y:= 0.0}
                    Widget := stack_box:
                        Orientation := orientation.Horizontal
                        Slots := array:
                            stack_box_slot:
                                Widget := CustomPlayer.Upgrade1Bonus   
                            stack_box_slot:
                                Widget := CustomPlayer.Upgrade1NextBonus
                canvas_slot:
                    Widget:=CustomPlayer.Upgrade1CostText
                    ZOrder:=1
                    SizeToContent:=true
                    Anchors := anchors{Minimum := vector2{X:= 0.5, Y:= 0.0}, Maximum := vector2{X:= 0.5, Y:= 0.0}}
                    Offsets:=margin{Top:=796.0,Left:=-644.0}
                    Alignment := vector2{X:= 0.5, Y:= 0.0}
                canvas_slot:
                    ZOrder:=1
                    SizeToContent:=true
                    Anchors := anchors{Minimum := vector2{X:= 0.5, Y:= 0.0}, Maximum := vector2{X:= 0.5, Y:= 0.0}}
                    Offsets:=margin{Top:=796.0,Left:=-388.0}
                    Alignment := vector2{X:= 1.0, Y:= 0.0}
                    Widget := stack_box:
                        Orientation := orientation.Horizontal
                        Slots := array:
                            stack_box_slot:
                                Widget := CustomPlayer.ShardIcon1    
                            stack_box_slot:
                                Widget := CustomPlayer.Upgrade1CostAmount
                canvas_slot:
                    Widget:=CustomPlayer.PUPUpgradeButton1
                    ZOrder:=3
                    SizeToContent:=true
                    Anchors := anchors{Minimum := vector2{X:= 0.5, Y:= 0.0}, Maximum := vector2{X:= 0.5, Y:= 0.0}}
                    Offsets:=margin{Top:=868.0,Left:=-548.0, Bottom:=60.0, Right:=220.0}
                    Alignment := vector2{X:= 0.5, Y:= 0.0}
                    
                canvas_slot:
                    Widget:=CustomPlayer.Upgrade2Title
                    ZOrder:=1
                    SizeToContent:=true
                    Anchors := anchors{Minimum := vector2{X:= 0.5, Y:= 0.0}, Maximum := vector2{X:= 0.5, Y:= 0.0}}
                    Offsets:=margin{Top:=276.0,Left:=8.0}
                    Alignment := vector2{X:= 0.5, Y:= 0.0}
                canvas_slot:
                    Widget:=CustomPlayer.Upgrade2Level
                    ZOrder:=1
                    SizeToContent:=true
                    Anchors := anchors{Minimum := vector2{X:= 0.5, Y:= 0.0}, Maximum := vector2{X:= 0.5, Y:= 0.0}}
                    Offsets:=margin{Top:=336.0,Left:=8.0}
                    Alignment := vector2{X:= 0.5, Y:= 0.0}
                canvas_slot:
                    Widget:=CustomPlayer.Upgrade2Icon
                    ZOrder:=1
                    SizeToContent:=false
                    Anchors := anchors{Minimum := vector2{X:= 0.5, Y:= 0.0}, Maximum := vector2{X:= 0.5, Y:= 0.0}}
                    Offsets:=margin{Top:=416.0,Left:=8.0, Bottom:=140.0,Right:=140.0}
                    Alignment := vector2{X:= 0.5, Y:= 0.0}
                canvas_slot:
                    Widget:=CustomPlayer.Upgrade2Desc
                    ZOrder:=1
                    SizeToContent:=true
                    Anchors := anchors{Minimum := vector2{X:= 0.5, Y:= 0.0}, Maximum := vector2{X:= 0.5, Y:= 0.0}}
                    Offsets:=margin{Top:=588.0,Left:=8.0}
                    Alignment := vector2{X:= 0.5, Y:= 0.0}
                canvas_slot:
                    ZOrder:=1
                    SizeToContent:=true
                    Anchors := anchors{Minimum := vector2{X:= 0.5, Y:= 0.0}, Maximum := vector2{X:= 0.5, Y:= 0.0}}
                    Offsets:=margin{Top:=736.0,Left:=8.0}
                    Alignment := vector2{X:= 0.5, Y:= 0.0}
                    Widget := stack_box:
                        Orientation := orientation.Horizontal
                        Slots := array:
                            stack_box_slot:
                                Widget := CustomPlayer.Upgrade2Bonus   
                            stack_box_slot:
                                Widget := CustomPlayer.Upgrade2NextBonus
                canvas_slot:
                    Widget:=CustomPlayer.Upgrade2CostText
                    ZOrder:=1
                    SizeToContent:=true
                    Anchors := anchors{Minimum := vector2{X:= 0.5, Y:= 0.0}, Maximum := vector2{X:= 0.5, Y:= 0.0}}
                    Offsets:=margin{Top:=796.0,Left:=-84.0}
                    Alignment := vector2{X:= 0.5, Y:= 0.0}
                canvas_slot:
                    ZOrder:=1
                    SizeToContent:=true
                    Anchors := anchors{Minimum := vector2{X:= 0.5, Y:= 0.0}, Maximum := vector2{X:= 0.5, Y:= 0.0}}
                    Offsets:=margin{Top:=796.0,Left:=168.0}
                    Alignment := vector2{X:= 1.0, Y:= 0.0}
                    Widget := stack_box:
                        Orientation := orientation.Horizontal
                        Slots := array:
                            stack_box_slot:
                                Widget := CustomPlayer.ShardIcon2   
                            stack_box_slot:
                                Widget := CustomPlayer.Upgrade2CostAmount
                canvas_slot:
                    Widget:=CustomPlayer.PUPUpgradeButton2
                    ZOrder:=1
                    SizeToContent:=true
                    Anchors := anchors{Minimum := vector2{X:= 0.5, Y:= 0.0}, Maximum := vector2{X:= 0.5, Y:= 0.0}}
                    Offsets:=margin{Top:=868.0,Left:=8.0, Bottom:=60.0, Right:=220.0}
                    Alignment := vector2{X:= 0.5, Y:= 0.0}

                canvas_slot:
                    Widget:=CustomPlayer.Upgrade3Title
                    ZOrder:=1
                    SizeToContent:=true
                    Anchors := anchors{Minimum := vector2{X:= 0.5, Y:= 0.0}, Maximum := vector2{X:= 0.5, Y:= 0.0}}
                    Offsets:=margin{Top:=276.0,Left:=560.0}
                    Alignment := vector2{X:= 0.5, Y:= 0.0}
                canvas_slot:
                    Widget:=CustomPlayer.Upgrade3Level
                    ZOrder:=1
                    SizeToContent:=true
                    Anchors := anchors{Minimum := vector2{X:= 0.5, Y:= 0.0}, Maximum := vector2{X:= 0.5, Y:= 0.0}}
                    Offsets:=margin{Top:=336.0,Left:=560.0}
                    Alignment := vector2{X:= 0.5, Y:= 0.0}
                canvas_slot:
                    Widget:=CustomPlayer.Upgrade3Icon
                    ZOrder:=1
                    SizeToContent:=false
                    Anchors := anchors{Minimum := vector2{X:= 0.5, Y:= 0.0}, Maximum := vector2{X:= 0.5, Y:= 0.0}}
                    Offsets:=margin{Top:=416.0,Left:=560.0, Bottom:=140.0,Right:=140.0}
                    Alignment := vector2{X:= 0.5, Y:= 0.0}
                canvas_slot:
                    Widget:=CustomPlayer.Upgrade3Desc
                    ZOrder:=1
                    SizeToContent:=true
                    Anchors := anchors{Minimum := vector2{X:= 0.5, Y:= 0.0}, Maximum := vector2{X:= 0.5, Y:= 0.0}}
                    Offsets:=margin{Top:=588.0,Left:=560.0}
                    Alignment := vector2{X:= 0.5, Y:= 0.0}
                canvas_slot:
                    ZOrder:=1
                    SizeToContent:=true
                    Anchors := anchors{Minimum := vector2{X:= 0.5, Y:= 0.0}, Maximum := vector2{X:= 0.5, Y:= 0.0}}
                    Offsets:=margin{Top:=736.0,Left:=560.0}
                    Alignment := vector2{X:= 0.5, Y:= 0.0}
                    Widget := stack_box:
                        Orientation := orientation.Horizontal
                        Slots := array:
                            stack_box_slot:
                                Widget := CustomPlayer.Upgrade3Bonus   
                            stack_box_slot:
                                Widget := CustomPlayer.Upgrade3NextBonus
                canvas_slot:
                    Widget:=CustomPlayer.Upgrade3CostText
                    ZOrder:=1
                    SizeToContent:=true
                    Anchors := anchors{Minimum := vector2{X:= 0.5, Y:= 0.0}, Maximum := vector2{X:= 0.5, Y:= 0.0}}
                    Offsets:=margin{Top:=796.0,Left:=464.0}
                    Alignment := vector2{X:= 0.5, Y:= 0.0}
                canvas_slot:
                    ZOrder:=1
                    SizeToContent:=true
                    Anchors := anchors{Minimum := vector2{X:= 0.5, Y:= 0.0}, Maximum := vector2{X:= 0.5, Y:= 0.0}}
                    Offsets:=margin{Top:=796.0,Left:=720.0}
                    Alignment := vector2{X:= 1.0, Y:= 0.0}
                    Widget := stack_box:
                        Orientation := orientation.Horizontal
                        Slots := array:
                            stack_box_slot:
                                Widget := CustomPlayer.ShardIcon3   
                            stack_box_slot:
                                Widget := CustomPlayer.Upgrade3CostAmount
                canvas_slot:
                    Widget:=CustomPlayer.PUPUpgradeButton3
                    ZOrder:=1
                    SizeToContent:=true
                    Anchors := anchors{Minimum := vector2{X:= 0.5, Y:= 0.0}, Maximum := vector2{X:= 0.5, Y:= 0.0}}
                    Offsets:=margin{Top:=868.0,Left:=560.0, Bottom:=60.0, Right:=220.0}
                    Alignment := vector2{X:= 0.5, Y:= 0.0}
        return MyPUPCanvas


#UI BLOCKS

    #HUD STUFF
        
#CANVAS
    MakePrestigeCanvas(CustomPlayer:Custom_Player):canvas=
        CustomPlayer.PrestigeCloseButton.SetText(StringToMessage("X"))
        CustomPlayer.InfoText.SetText(StringToMessage("Restart your progress to earn prestige shards"))
        CustomPlayer.KeepTitle.SetText(StringToMessage("KEEP"))
        CustomPlayer.KeepText.SetText(StringToMessage("- General Upgrades\n- Weapon Upgrades\n- Pets (Lvs. & Evos)\n- Drill Buffs\n- Achievements\n- Area 4 Upgrades\n- Inventory Items*\n\n*You will lose your stone"))
        CustomPlayer.ResetTitle.SetText(StringToMessage("RESET"))
        CustomPlayer.ResetText.SetText(StringToMessage("- Strength Upgrades\n- Stages Unlocked\n- Cave Hut Skills\n- Current Stone"))
        CustomPlayer.NextPrestigeTitle.SetText(StringToMessage("Stone for next Prestige"))
        CustomPlayer.PrestigeNowTitle.SetText(StringToMessage("Prestiging now will grant:"))
        CustomPlayer.PrestigeNowButton.SetText(StringToMessage("PRESTIGE NOW"))

        for(TextBlock : CustomPlayer.AllPrestigeTexts):
            TextBlock.SetShadowColor(CustomPlayer.ShadowDefaults.DefaultShadowColor)
            TextBlock.SetShadowOffset(CustomPlayer.ShadowDefaults.DefaultShadowOffset)
            TextBlock.SetShadowOpacity(1.0)

        var MyPrestigeCanvas: canvas = canvas:
            Slots := array:
                canvas_slot:
                    Widget:=CustomPlayer.PrestigeHUDBG
                    ZOrder:=0
                    SizeToContent:=false
                    Alignment := vector2{X := 0.5, Y := 0.0}
                    Anchors := anchors{Minimum := vector2{X:= 0.5, Y:= 0.0}, Maximum := vector2{X:= 0.5, Y:= 0.0}}
                    Offsets:=margin{Top:=0.0,Left:=0.0,Bottom:=1076.0,Right:=2048.0}
                canvas_slot:
                    Widget:=CustomPlayer.PrestigeCloseButton
                    ZOrder:=9
                    SizeToContent:=false
                    Anchors := anchors{Minimum := vector2{X:= 0.5, Y:= 0.0}, Maximum := vector2{X:= 0.5, Y:= 0.0}}
                    Offsets:=margin{Top:=CustomPlayer.PrestigeClose_Position.Y,Left:=612.0,Bottom:=CustomPlayer.PrestigeClose_Size.Y,Right:=CustomPlayer.PrestigeClose_Size.X}
                
                canvas_slot:
                    Widget:=CustomPlayer.PrestigeHUDTitle
                    ZOrder:=1
                    SizeToContent:=false
                    Anchors := anchors{Minimum := vector2{X:= 0.5, Y:= 0.0}, Maximum := vector2{X:= 0.5, Y:= 0.0}}
                    Offsets:=margin{Top:=144.0,Left:=0.0, Bottom:=56.0, Right:=294.0}

                canvas_slot:
                    Widget:=CustomPlayer.InfoText
                    ZOrder:=1
                    SizeToContent:=true
                    Anchors := anchors{Minimum := vector2{X:= 0.5, Y:= 0.0}, Maximum := vector2{X:= 0.5, Y:= 0.0}}
                    Offsets:=margin{Top:=296.0,Left:=-628.0}
                    Alignment := vector2{X:= 0.0, Y:= 0.0}
                canvas_slot:
                    Widget:=CustomPlayer.KeepTitle
                    ZOrder:=1
                    SizeToContent:=true
                    Anchors := anchors{Minimum := vector2{X:= 0.5, Y:= 0.0}, Maximum := vector2{X:= 0.5, Y:= 0.0}}
                    Offsets:=margin{Top:=368.0,Left:=-492.0}
                    Alignment := vector2{X:= 0.5, Y:= 0.0}
                canvas_slot:
                    Widget:=CustomPlayer.KeepText
                    ZOrder:=1
                    SizeToContent:=true
                    Anchors := anchors{Minimum := vector2{X:= 0.5, Y:= 0.0}, Maximum := vector2{X:= 0.5, Y:= 0.0}}
                    Offsets:=margin{Top:=452.0,Left:=-664.0}
                    Alignment := vector2{X:= 0.0, Y:= 0.0}
                canvas_slot:
                    Widget:=CustomPlayer.ResetTitle
                    ZOrder:=1
                    SizeToContent:=true
                    Anchors := anchors{Minimum := vector2{X:= 0.5, Y:= 0.0}, Maximum := vector2{X:= 0.5, Y:= 0.0}}
                    Offsets:=margin{Top:=368.0,Left:=-8.0}
                    Alignment := vector2{X:= 0.5, Y:= 0.0}
                canvas_slot:
                    Widget:=CustomPlayer.ResetText
                    ZOrder:=1
                    SizeToContent:=true
                    Anchors := anchors{Minimum := vector2{X:= 0.5, Y:= 0.0}, Maximum := vector2{X:= 0.5, Y:= 0.0}}
                    Offsets:=margin{Top:=452.0,Left:=-188.0}
                    Alignment := vector2{X:= 0.0, Y:= 0.0}
                canvas_slot:
                    Widget:=CustomPlayer.CurrentPrestigeLevelText
                    ZOrder:=1
                    SizeToContent:=true
                    Anchors := anchors{Minimum := vector2{X:= 0.5, Y:= 0.0}, Maximum := vector2{X:= 0.5, Y:= 0.0}}
                    Offsets:=margin{Top:=300.0,Left:=552.0}
                    Alignment := vector2{X:= 0.5, Y:= 0.0}
                canvas_slot:
                    Widget:=CustomPlayer.NextPrestigeTitle
                    ZOrder:=1
                    SizeToContent:=true
                    Anchors := anchors{Minimum := vector2{X:= 0.5, Y:= 0.0}, Maximum := vector2{X:= 0.5, Y:= 0.0}}
                    Offsets:=margin{Top:=400.0,Left:=552.0}
                    Alignment := vector2{X:= 0.5, Y:= 0.0}
                canvas_slot:
                    ZOrder:=1
                    SizeToContent:=true
                    Anchors := anchors{Minimum := vector2{X:= 0.5, Y:= 0.0}, Maximum := vector2{X:= 0.5, Y:= 0.0}}
                    Offsets:=margin{Top:=464.0,Left:=552.0}
                    Alignment := vector2{X:= 0.5, Y:= 0.0}
                    Widget := stack_box:
                        Orientation := orientation.Horizontal
                        Slots := array:
                            stack_box_slot:
                                Widget := CustomPlayer.NextPrestigeStoneIcon   
                            stack_box_slot:
                                Widget := CustomPlayer.NextPrestigeAmount
                canvas_slot:
                    Widget:=CustomPlayer.PrestigeNowTitle
                    ZOrder:=1
                    SizeToContent:=true
                    Anchors := anchors{Minimum := vector2{X:= 0.5, Y:= 0.0}, Maximum := vector2{X:= 0.5, Y:= 0.0}}
                    Offsets:=margin{Top:=588.0,Left:=552.0}
                    Alignment := vector2{X:= 0.5, Y:= 0.0}
                canvas_slot:
                    ZOrder:=1
                    SizeToContent:=true
                    Anchors := anchors{Minimum := vector2{X:= 0.5, Y:= 0.0}, Maximum := vector2{X:= 0.5, Y:= 0.0}}
                    Offsets:=margin{Top:=656.0,Left:=552.0}
                    Alignment := vector2{X:= 0.5, Y:= 0.0}
                    Widget := stack_box:
                        Orientation := orientation.Horizontal
                        Slots := array:
                            stack_box_slot:
                                Widget := CustomPlayer.PrestigeNowShardIcon   
                            stack_box_slot:
                                Widget := CustomPlayer.PrestigeNowShards
                canvas_slot:
                    ZOrder:=1
                    SizeToContent:=true
                    Anchors := anchors{Minimum := vector2{X:= 0.5, Y:= 0.0}, Maximum := vector2{X:= 0.5, Y:= 0.0}}
                    Offsets:=margin{Top:=712.0,Left:=552.0}
                    Alignment := vector2{X:= 0.5, Y:= 0.0}
                    Widget := stack_box:
                        Orientation := orientation.Horizontal
                        Slots := array:
                            stack_box_slot:
                                Widget := CustomPlayer.PrestigeNowGatherIcon   
                            stack_box_slot:
                                Widget := CustomPlayer.PrestigeNowGather
                canvas_slot:
                    Widget:=CustomPlayer.PrestigeNowButton
                    ZOrder:=1
                    SizeToContent:=true
                    Anchors := anchors{Minimum := vector2{X:= 0.5, Y:= 0.0}, Maximum := vector2{X:= 0.5, Y:= 0.0}}
                    Offsets:=margin{Top:=824.0,Left:=552.0, Bottom:=80.0, Right:=400.0}
                    Alignment := vector2{X:= 0.5, Y:= 0.0}
                
        return MyPrestigeCanvas


    UpdatePrestigeHUD(CustomPlayer: Custom_Player):void= 
        var PrestigeShardAmount : float = 0.0
        var PrestigeGatherAmount : float = 0.0
        var NewPrestigeCost : float = 0.0
        if(CustomPlayer.PrestigeCost * Pow(CustomPlayer.PrestigeCostMult, (CustomPlayer.PrestigeLevel)*1.0) > 1e300):
            set NewPrestigeCost = 1e300
        else:
            set NewPrestigeCost = CustomPlayer.PrestigeCost * Pow(CustomPlayer.PrestigeCostMult, (CustomPlayer.PrestigeLevel)*1.0)

        CustomPlayer.CurrentPrestigeLevelText.SetText(StringToMessage("Prestige Level: {CustomPlayer.PrestigeLevel}"))
        if(CustomPlayer.PrestigeStoneAmount >= NewPrestigeCost):
            CustomPlayer.NextPrestigeAmount.SetTextColor(MakeColorFromHex("39FF39"))
        else:
            CustomPlayer.NextPrestigeAmount.SetTextColor(MakeColorFromHex("fa4d56"))

        ratio := CustomPlayer.PrestigeStoneAmount / NewPrestigeCost

        if(CustomPlayer.PrestigeLevel <= 220):
            if(CustomPlayer.PrestigeLevel >= 1):
                set PrestigeShardAmount = ((Log(5.0, ratio)) * 5) * (1.5 * CustomPlayer.PrestigeLevel)
            else:
                set PrestigeShardAmount = ((Log(5.0, ratio)) * 5)

            if(CustomPlayer.StagesUnlocked >= 15):
                set PrestigeGatherAmount = (PrestigeShardAmount/5.0) * 3.0
            else:
                set PrestigeGatherAmount = (PrestigeShardAmount/5.0)

            if(PrestigeShardAmount < 1.0):
                set PrestigeShardAmount = 1.0
            if(PrestigeGatherAmount < 1.0):
                set PrestigeGatherAmount = 1.0

            set PrestigeShardAmount *= 1.0 + CustomPlayer.SinkPrestigePower / 10000.0
            set PrestigeGatherAmount *= 1.0 + CustomPlayer.SinkPrestigePower / 10000.0
        else:
            set PrestigeShardAmount = CustomPlayer.EndGamePrestigeShardAmount * (1.0 + CustomPlayer.SinkPrestigePower / 10000.0)
            set PrestigeGatherAmount = CustomPlayer.EndGamePrestigeGatherAmount * (1.0 + CustomPlayer.SinkPrestigePower / 10000.0)

        CustomPlayer.NextPrestigeAmount.SetText(StringToMessage(" {NumberCondensor(CustomPlayer.PrestigeStoneAmount)} / {NumberCondensor(NewPrestigeCost)}"))

        if(FlooredPrestige:=Floor[PrestigeShardAmount]):
            CustomPlayer.PrestigeNowShards.SetText(StringToMessage(" +{FlooredPrestige}"))
        CustomPlayer.PrestigeNowGather.SetText(StringToMessage(" +{NumberCondensor(PrestigeGatherAmount)}% Strength"))

    PrestigePlayer(CustomPlayer:Custom_Player):void=
        var PrestigeShardAmount : float = 0.0
        var PrestigeGatherAmount : float = 0.0

        var NewPrestigeCost : float = 0.0
        if(CustomPlayer.PrestigeCost * Pow(CustomPlayer.PrestigeCostMult, (CustomPlayer.PrestigeLevel)*1.0) > 1e300):
            set NewPrestigeCost = 1e300
        else:
            set NewPrestigeCost = CustomPlayer.PrestigeCost * Pow(CustomPlayer.PrestigeCostMult, (CustomPlayer.PrestigeLevel)*1.0)
            
        #Print("NewPrestigeCost: {NewPrestigeCost}")
        #Print("CustomPlayer.PrestigeStoneAmount: {CustomPlayer.PrestigeStoneAmount}")
        if(CustomPlayer.PrestigeStoneAmount >= NewPrestigeCost): # or CustomPlayer.PrestigeStoneAmount >= 1.0e154
            Tutorial.UpdateTutorialUI(CustomPlayer, "Prestige", 1.12e16)
            
            if(Accolade:=AudioManager.Accolades[6]): #AWARD PLAYER EXP
                Accolade.Award(CustomPlayer.Player)

            if: #AWARD PLAYER EXP
                Accolade:=AudioManager.Accolades[14]
                CustomPlayer.PUPQOLLevels[1] = 2
            then:
                Accolade.Award(CustomPlayer.Player)

            #RESET EVERYTHING THAT NEEDS TO BE RESET
            set CustomPlayer.Power = 1.0

            #IF YOU DO NOT HAVE PERMA HUT UNLOCK
            if(CustomPlayer.PUPQOLLevels[1] <> 2):
                Print("DOES NOT HAVE QOL LEVEL")
                set CustomPlayer.CritRate = 0.0
                set CustomPlayer.CritDMG = 1.5
                set CustomPlayer.FinalHitBonus = 0.20
                set CustomPlayer.MultiCrit = 1.0

                set CustomPlayer.ComboStrikeUnlocked = false
                set CustomPlayer.ComboStrikeMultiplier = 5.0
                set CustomPlayer.ComboAccelerator = 1.0
                set CustomPlayer.ComboCount = -1.0
                set CustomPlayer.ComboStrikeStreak = 1.0
                set CustomPlayer.CurrentComboStrikeStreak = 0.0

                set CustomPlayer.EarthquakeUnlocked = false
                set CustomPlayer.EarthquakeDistance = 600.0 
                set CustomPlayer.EarthquakeGatherAmount = 0.20
                set CustomPlayer.EarthquakeDamage = 0.50 #% BASED ON PLAYER DAMAGE
            
                set CustomPlayer.Hut1Levels = array{0,0,0,0}
                set CustomPlayer.Hut2Levels = array{0,0,0,0}
                set CustomPlayer.Hut3Levels = array{0,0,0,0}

                set CustomPlayer.AllHutLevels = array{CustomPlayer.Hut1Levels,CustomPlayer.Hut2Levels,CustomPlayer.Hut3Levels,CustomPlayer.Hut4Levels, CustomPlayer.Hut5Levels}
                
            set CustomPlayer.RoomUnlockGather = 0.0


            set CustomPlayer.StrengthEfficiencies = array{
                1, 1, 1, 1, 1,
                1, 1, 1, 1, 1,
                1, 1, 1, 1, 1,
                1, 1, 1, 1, 1,
                1, 1, 1, 1, 1,
                1, 1, 1, 1, 1,
                1, 1, 1, 1, 1,
                1, 1, 1, 1, 1,
        
                1, 1, 1, 1, 1,
                1, 1, 1, 1, 1,
                1, 1, 1, 1, 1,
                1, 1, 1, 1, 1,
                1, 1, 1, 1, 1,
                1, 1, 1, 1, 1,
                1, 1, 1, 1, 1,
                1, 1, 1, 1, 1,
                
                1, 1, 1, 1, 1,
                1, 1, 1, 1, 1,
                1, 1, 1, 1, 1,
                1, 1, 1, 1, 1,
                1, 1, 1, 1, 1,
                1, 1, 1, 1, 1
            }
            set CustomPlayer.StrengthUpgradeLevels = array{
                0,0,0,0,0,
                0,0,0,0,0,
                0,0,0,0,0,
                0,0,0,0,0,
                0,0,0,0,0,
                0,0,0,0,0,
                0,0,0,0,0,
                0,0,0,0,0,
        
                0,0,0,0,0,
                0,0,0,0,0,
                0,0,0,0,0,
                0,0,0,0,0,
                0,0,0,0,0,
                0,0,0,0,0,
                0,0,0,0,0,
                0,0,0,0,0,
                
                0,0,0,0,0,
                0,0,0,0,0,
                0,0,0,0,0,
                0,0,0,0,0,
                0,0,0,0,0,
                0,0,0,0,0
            }
            CustomPlayer.SUP1Icon.SetImage(UI.SUP1)
            CustomPlayer.SUP2Icon.SetImage(UI.SUP2)
            CustomPlayer.SUP3Icon.SetImage(UI.SUP3)
            CustomPlayer.SUP4Icon.SetImage(UI.SUP4)
            CustomPlayer.SUP5Icon.SetImage(UI.SUP5)

            if(CustomPlayer.PrestigeFix = false and CustomPlayer.PrestigeLevel > 220):
                Print("PLAYER NEEDS PRESTIGE FIXIN")
                PrestigeToFix := CustomPlayer.PrestigeLevel - 220

                ExtraPrestigeShards := PrestigeToFix * CustomPlayer.EndGamePrestigeShardAmount
                ExtraPrestigeGather := PrestigeToFix * CustomPlayer.EndGamePrestigeGatherAmount

                Print("ExtraPrestigeShards: {ExtraPrestigeShards}")
                Print("ExtraPrestigeGather: {ExtraPrestigeGather}")

                CustomPlayer.Collect(ExtraPrestigeShards, "Shard", 0)
                set CustomPlayer.HarvestBuff += ExtraPrestigeGather/100.0
                set CustomPlayer.PrestigeHarvestBuff += ExtraPrestigeGather
            else:
                Print("NOT OVER 400, NOT APPLICABLE, FIXING PRESTIGE NOW")

            set CustomPlayer.PrestigeFix = true
            #ADD PRESTIGE REWARDS
            ratio := CustomPlayer.PrestigeStoneAmount / NewPrestigeCost
            if(CustomPlayer.PrestigeLevel <= 220):
                Print("PLAYER PRESTIGE \<= 220")
                if(CustomPlayer.PrestigeLevel >= 1):
                    set PrestigeShardAmount = ((Log(5.0, ratio)) * 5) * (1.5 * CustomPlayer.PrestigeLevel)
                else:
                    set PrestigeShardAmount = ((Log(5.0, ratio)) * 5)

                if(CustomPlayer.StagesUnlocked >= 15):
                    set PrestigeGatherAmount = (PrestigeShardAmount/5.0) * 3.0
                else:
                    set PrestigeGatherAmount = (PrestigeShardAmount/5.0)

                if(PrestigeShardAmount < 1.0):
                    set PrestigeShardAmount = 1.0

                set PrestigeShardAmount = PrestigeShardAmount * (1.0 + CustomPlayer.SinkPrestigePower / 10000.0)
                set PrestigeGatherAmount = PrestigeGatherAmount * (1.0 + CustomPlayer.SinkPrestigePower / 10000.0)
            else:
                Print("PLAYER PRESTIGE \> 220")
                set PrestigeShardAmount = CustomPlayer.EndGamePrestigeShardAmount * (1.0 + CustomPlayer.SinkPrestigePower / 10000.0)
                set PrestigeGatherAmount = CustomPlayer.EndGamePrestigeGatherAmount * (1.0 + CustomPlayer.SinkPrestigePower / 10000.0)
                
            CustomPlayer.Collect(PrestigeShardAmount, "Shard", 0)
            set CustomPlayer.HarvestBuff += PrestigeGatherAmount/100.0
            set CustomPlayer.PrestigeHarvestBuff += PrestigeGatherAmount
                
            #set CustomPlayer.PrestigeCost *= PrestigeCostMult

            #INCREASE PRESTIGE LEVEL
            set CustomPlayer.PrestigeLevel += 1
            set CustomPlayer.StagesUnlocked = -1
            if:
                CustomPlayer.PUPQOLLevels[2] >= 1
                NewStagesUnlocked := CustomPlayer.PUPQOLLevels[2] - 1
            then:
                set CustomPlayer.RoomUnlockGather = (NewStagesUnlocked + 1) * 0.05
                set CustomPlayer.StagesUnlocked = NewStagesUnlocked

                for(i:int=0..NewStagesUnlocked):
                    Print("GIVE ACCOLADE FOR ROOM PERMA UNLOCKED {i}")
                    if(Accolade:=AudioManager.Accolades[3]): #AWARD PLAYER EXP
                        Accolade.Award(CustomPlayer.Player)

            RoomUnlocks.UnlockRooms(CustomPlayer)
            
            #DO THIS AT THE BOTTOM SO IT DOESN'T MESS UP THE GAINS
            CustomPlayer.Spend(CustomPlayer.StoneAmount, "Stone", 0)
            set CustomPlayer.PrestigeStoneAmount = 0.0
            AudioManager.Prestige.Play(CustomPlayer.Player)
            SaveAllData(CustomPlayer)
            #CINEMATIC OR SOMETHING TO SHOW PRESTIGE
        else:
            AudioManager.Denied.Play(CustomPlayer.Player)
        UpdatePrestigeHUD(CustomPlayer)

    OnPrestigeNow(Wdata:widget_message):void=
        if(CustomPlayer := gm.PlayersMap[Wdata.Player]):
            PrestigePlayer(CustomPlayer)

#THIS IS TO RESET ALL DATA TO THE START, THIS IS NOT PRESTIGE
    MakeResetAllCanvas(CustomPlayer:Custom_Player):canvas=
        CustomPlayer.ResetAllCloseButton.SetText(StringToMessage("CLOSE"))

        CustomPlayer.ResetInfoTitle.SetText(StringToMessage("RESET ALL DATA"))
        CustomPlayer.ResetWarning.SetText(StringToMessage("WARNING: THIS IS NOT PRESTIGE!\n\nTHIS IS ONLY IF YOU WISH TO COMPLETELY RESTART ALL OF YOUR PROGRESS TO THE BEGINNING\n\nTHIS IS NOT REVERSABLE"))
        CustomPlayer.ResetLoseAll.SetText(StringToMessage("YOU WILL LOSE EVERYTHING.\n\nALL STRENGTH UPGRADES, PETS & LEVELS, WEAPON + UPGRADES,\nSTONE, PRESTIGE POINTS AND MORE. EVERYTHING."))

        CustomPlayer.ResetAllNowButton.SetText(StringToMessage("RESET ALL DATA"))
        CustomPlayer.ResetAllConfirmButton.SetText(StringToMessage("CONFIRM RESET"))
        CustomPlayer.ResetAllConfirmButton.SetEnabled(false)
        var MyResetAllCanvas: canvas = canvas:
            Slots := array:
                canvas_slot:
                    Widget:=CustomPlayer.ResetAllHUDBG
                    ZOrder:=0
                    SizeToContent:=false
                    Alignment := vector2{X := 0.5, Y := 0.0}
                    Anchors := anchors{Minimum := vector2{X:= 0.5, Y:= 0.0}, Maximum := vector2{X:= 0.5, Y:= 0.0}}
                    Offsets:=margin{Top:=0.0,Left:=0.0,Bottom:=1076.0,Right:=2048.0}
                
                canvas_slot:
                    Widget:=CustomPlayer.ResetInfoTitle
                    ZOrder:=1
                    SizeToContent:=true
                    Alignment := vector2{X:= 0.5, Y:= 0.0}
                    Anchors := anchors{Minimum := vector2{X:= 0.5, Y:= 0.0}, Maximum := vector2{X:= 0.5, Y:= 0.0}}
                    Offsets:=margin{Top:=150.0,Left:=0.0}

                canvas_slot:
                    Widget:=CustomPlayer.ResetWarning
                    ZOrder:=1
                    SizeToContent:=true
                    Alignment := vector2{X:= 0.5, Y:= 0.0}
                    Anchors := anchors{Minimum := vector2{X:= 0.5, Y:= 0.0}, Maximum := vector2{X:= 0.5, Y:= 0.0}}
                    Offsets:=margin{Top:=250.0,Left:=0.0}
                canvas_slot:
                    Widget:=CustomPlayer.ResetLoseAll
                    ZOrder:=1
                    SizeToContent:=true
                    Alignment := vector2{X:= 0.5, Y:= 0.0}
                    Anchors := anchors{Minimum := vector2{X:= 0.5, Y:= 0.0}, Maximum := vector2{X:= 0.5, Y:= 0.0}}
                    Offsets:=margin{Top:=530.0,Left:=0.0}

                canvas_slot:
                    Widget:=CustomPlayer.ResetAllCloseButton
                    ZOrder:=9
                    SizeToContent:=false
                    Alignment := vector2{X:= 0.5, Y:= 0.0}
                    Anchors := anchors{Minimum := vector2{X:= 0.5, Y:= 0.0}, Maximum := vector2{X:= 0.5, Y:= 0.0}}
                    Offsets:=margin{Top:=825.0,Left:=-450.0,Bottom:=60.0, Right:=300.0}
                canvas_slot:
                    Widget:=CustomPlayer.ResetAllNowButton
                    ZOrder:=9
                    SizeToContent:=false
                    Alignment := vector2{X:= 0.5, Y:= 0.0}
                    Anchors := anchors{Minimum := vector2{X:= 0.5, Y:= 0.0}, Maximum := vector2{X:= 0.5, Y:= 0.0}}
                    Offsets:=margin{Top:=825.0,Left:=0.0,Bottom:=60.0, Right:=400.0}
                canvas_slot:
                    Widget:=CustomPlayer.ResetAllConfirmButton
                    ZOrder:=9
                    SizeToContent:=false
                    Alignment := vector2{X:= 0.5, Y:= 0.0}
                    Anchors := anchors{Minimum := vector2{X:= 0.5, Y:= 0.0}, Maximum := vector2{X:= 0.5, Y:= 0.0}}
                    Offsets:=margin{Top:=825.0,Left:=500.0,Bottom:=60.0, Right:=400.0}
         
        return MyResetAllCanvas

    OpenResetAllHUD(Agent:agent): void=
        if(CustomPlayer := gm.PlayersMap[Agent]):
                if(MyUI:=CustomPlayer.UIMaps["ResetAll"]?):
                else:
                    var MyResetAllCanvas : canvas = MakeResetAllCanvas(CustomPlayer)
                    CustomPlayer.AddFullHUD(MyResetAllCanvas)
                    if(set CustomPlayer.UIMaps["ResetAll"] = option{MyResetAllCanvas}) {}

    RemoveOpenResetAllHUD(Wdata:widget_message):void=
        if(CustomPlayer := gm.PlayersMap[Wdata.Player]):
            if(Player:=Wdata.Player, PlayerUI:=GetPlayerUI[Player]):
                if(MyUI:=CustomPlayer.UIMaps["ResetAll"]?):
                    PlayerUI.RemoveWidget(MyUI)
                    MyUI.SetVisibility(widget_visibility.Collapsed)
                    if(set CustomPlayer.UIMaps["ResetAll"] = false){}


    ResetAllPlayerData(Wdata:widget_message, ConfirmReset:int):void=
        if(CustomPlayer := gm.PlayersMap[Wdata.Player]):
            if(ConfirmReset = 0):
                CustomPlayer.ResetAllConfirmButton.SetEnabled(true)
            else if(ConfirmReset = 1):
                RemoveOpenResetAllHUD(Wdata)
                if(Analytic:=AudioManager.GeneralAnalytics[8]): #ANALYTIC
                    Analytic.Submit(CustomPlayer.Player)
                #SaveVersion(Wdata.Player, 9999)
                #gm.OnPlayerLeftServer(Wdata.Player)
                Tutorial.RemoveTutorialHUD(CustomPlayer)
                gm.ResetPlayerProgress(CustomPlayer.Player)
                spawn{ElimPlayer(CustomPlayer)}

    ElimPlayer(CustomPlayer:Custom_Player)<suspends>:void=
        if(Fort:=CustomPlayer.Player.GetFortCharacter[]):
            for(i:int=0..CustomPlayer.PetFortChars.Length):
                if(PetChar := CustomPlayer.PetFortChars[i],Pet:=PetChar?):
                    Pet.Damage(9999999.9)
            Fort.Damage(999999.0)




