using { /Fortnite.com/Devices }
using { /Verse.org/Simulation }
using { /Verse.org/Colors }
using { /Verse.org/Verse }
using { /Verse.org/Random }
using { /Fortnite.com/Teams }
using { /UnrealEngine.com/Assets }
using { /Fortnite.com/Game }
using { /Fortnite.com/Characters }
using { /Fortnite.com/Playspaces }
using { /Fortnite.com/UI }
using { /UnrealEngine.com/Temporary/Diagnostics }
using { /UnrealEngine.com/Temporary/SpatialMath }
using { /UnrealEngine.com/Temporary/UI }
using { /Verse.org/Assets}


ui_weapon_upgrades := class(creative_device):

    @editable
    gm : game_manager = game_manager{}

    @editable
    AudioManager : audio_manager = audio_manager{}

    @editable
    Tutorial : tutorial_manager = tutorial_manager{}

    @editable
    WeaponUpgradeButton : button_device = button_device{}

    @editable
    WeaponResetButton : button_device = button_device{}

    @editable
    WeaponUnlockGranters : []item_granter_device = array{}
    @editable
    SignalRemoteGranter : item_granter_device = item_granter_device{}

    var StatBaseUpgrade : float = 30.0
    var StatBaseMultiplier : float = 15.0
    var StatPercUpgrade : float = 0.02
    var StatPercAdditive : float = 0.00
    var BonusPerc : float = 0.02
    
    var PerkNames : []string = array{"Instant Reload","Crit Chance","Crit Damage","Attack","Final Strike","Strength"}
    var PerkBaseValues :[]float = array{1.0,0.06,0.05,0.10,0.05,0.10}
    var TierUpChance : int = 10 #OUT OF 100

    #CHANGE TO SAME AS MOB HPS * 5.0
    #var EvolveCosts : []float = array{2.50e06, 3.75e10, 1.56e17,2.34e21,9.75e27,1.46e32, 1000.0, 2000.0, 3000.0} 

    var WeaponTierNames : []string = array{"Common","Uncommon","Rare","Epic","Legendary","Mythic", "Supreme", "Paragon", "Transcendant"}
    var WeaponTierColors : []string = array{"DDDDDD","7bd75a","5de1e7","cb6ce7","ffbe5a","fada5a","E3772F","FB3434","2B2EDE"}

    #BASE COSTS
    var StatCost : []float = array{2000000.0, 0.0} #STONE, EGG
    var BonusCost : float = 10000000.0 #STONE
    var PerkCost : float = 40000000.0 #STONE
    var EvolveCost : [][]float = array{
        array{0.0, 2.50e06, 0.0, 0.0},
        array{100.0, 3.75e10, 50.0, 0.0},
        array{100.0, 1.56e17, 150.0, 0.0},
        array{100.0, 2.34e21, 450.0, 0.0},
        array{100.0, 9.75e27, 1350.0, 0.0},
        array{100.0, 1.46e32, 4050.0, 0.0},
        array{200.0, 1.56e45, 50000.0, 2000.0},
        array{200.0, 2.34e58, 125000.0, 4000.0},
        array{200.0, 9.75e61, 400000.0, 8000.0}
        }
    var ResetCost : float = 2000.0

    OnBegin<override>()<suspends>:void=
        WeaponUpgradeButton.InteractedWithEvent.Subscribe(OpenWUPHUD)

        for(Spawner : gm.PlayerSpawners):     
            Spawner.SpawnedEvent.Subscribe(OnPlayerJoined)

    OnPlayerJoined(Agent:agent):void=
        spawn{OnPlayerJoinedBuffer(Agent)}

    OnPlayerJoinedBuffer(Agent:agent)<suspends>:void=
        Sleep(1.0)
        if(CustomPlayer := gm.PlayersMap[player[Agent]]):
            if(CustomPlayer.PlayTimeSessionSeconds = 0.0):
                set CustomPlayer.InvTexts = array{CustomPlayer.InvStone,CustomPlayer.InvEgg1,CustomPlayer.InvEgg2,CustomPlayer.InvEgg3,CustomPlayer.InvEgg4,CustomPlayer.InvEgg5,CustomPlayer.InvEgg6,CustomPlayer.InvWepMat}
                set CustomPlayer.StatTexts = array{CustomPlayer.StatsAttackAmount,CustomPlayer.StatsAttackPercAmount,CustomPlayer.StatsStrPercAmount}
                set CustomPlayer.BonusTexts = array{CustomPlayer.BonusAttack,CustomPlayer.BonusAttackPerc,CustomPlayer.BonusStrPerc}
                set CustomPlayer.EggIcons = array{CustomPlayer.Egg1Icon,CustomPlayer.Egg2Icon,CustomPlayer.Egg3Icon,CustomPlayer.Egg4Icon,CustomPlayer.Egg5Icon,CustomPlayer.Egg6Icon}
    
                set CustomPlayer.PerkIcons = array{CustomPlayer.Perk1Icon, CustomPlayer.Perk2Icon, CustomPlayer.Perk3Icon}
                set CustomPlayer.AllWUPTexts = array{CustomPlayer.StatsTitle,CustomPlayer.StatsAttackTitle,CustomPlayer.StatsAttackAmount,CustomPlayer.StatsAttackPercTitle,CustomPlayer.StatsAttackPercAmount,CustomPlayer.StatsStrPercTitle,CustomPlayer.StatsStrPercAmount,CustomPlayer.BonusTitle,CustomPlayer.BonusAttack,CustomPlayer.BonusAttackPerc,CustomPlayer.BonusStrPerc,CustomPlayer.UnlockPerksTitle,CustomPlayer.PerksTitle,CustomPlayer.PerksTier,CustomPlayer.Perk1,CustomPlayer.Perk2,CustomPlayer.Perk3,CustomPlayer.WeaponDamage,CustomPlayer.WeaponDesc,CustomPlayer.WeaponTitle,CustomPlayer.UpgradeReqAmount3,CustomPlayer.UpgradeReqTitle3,CustomPlayer.UpgradeReqAmount2,CustomPlayer.UpgradeReqTitle2,CustomPlayer.UpgradeReqAmount1,CustomPlayer.UpgradeReqAmount4,CustomPlayer.UpgradeReqTitle1,CustomPlayer.UpgradeDesc,CustomPlayer.UpgradeTitle,CustomPlayer.InvWepMat,CustomPlayer.InvEgg6,CustomPlayer.InvEgg5,CustomPlayer.InvEgg4,CustomPlayer.InvEgg3,CustomPlayer.InvEgg2,CustomPlayer.InvEgg1,CustomPlayer.UnlockWUPCost,CustomPlayer.InvStone}
    
                CustomPlayer.WUPCloseButton.OnClick().Subscribe(RemoveWUPHUD)
    
                #CustomPlayer.UnlockWUPButton.OnClick().Subscribe(UnlockWUPHUD)
    
                #DISABLE THESE BUTTONS ON LOAD, ONLY ENABLE IF REQ
                <#CustomPlayer.ButtonStats.SetEnabled(false)
                CustomPlayer.ButtonBonus.SetEnabled(false)
                CustomPlayer.ButtonPerk.SetEnabled(false)
                CustomPlayer.ButtonEvolve.SetEnabled(false)
                CustomPlayer.ButtonReset.SetEnabled(false)
                CustomPlayer.ButtonUpgrade.SetEnabled(false)
                CustomPlayer.ButtonUpgradeAll.SetEnabled(false)
                CustomPlayer.ButtonPerk1.SetEnabled(false)
                CustomPlayer.ButtonPerk2.SetEnabled(false)
                CustomPlayer.ButtonPerk3.SetEnabled(false)#>

                CustomPlayer.ButtonStats.OnClick().SubscribeAgent3(ChangePage, 0)
                CustomPlayer.ButtonBonus.OnClick().SubscribeAgent3(ChangePage, 1)
                CustomPlayer.ButtonPerk.OnClick().SubscribeAgent3(ChangePage, 2)
                CustomPlayer.ButtonEvolve.OnClick().SubscribeAgent3(ChangePage, 3)
                CustomPlayer.ButtonReset.OnClick().SubscribeAgent3(ChangePage, 4)
    
                CustomPlayer.ButtonUpgrade.OnClick().Subscribe(OnUpgradeButton)
                CustomPlayer.ButtonUpgradeAll.OnClick().Subscribe(OnUpgradeAllButton)
    
                CustomPlayer.ButtonPerk1.OnClick().SubscribeAgent3(LockPerk, 0)
                CustomPlayer.ButtonPerk2.OnClick().SubscribeAgent3(LockPerk, 1)
                CustomPlayer.ButtonPerk3.OnClick().SubscribeAgent3(LockPerk, 2)
            GrantWeapon(CustomPlayer)
            
    OpenWUPHUD(Agent:agent): void=
        if(CustomPlayer := gm.PlayersMap[Agent]):
            if(MyUI:=CustomPlayer.UIMaps["WUP"]?):
            else:
                if(CustomPlayer.WeaponTier >= 1):
                    Tutorial.UpdateTutorialUI(CustomPlayer, "UnlockWeapon", 2500000.0)
                var MyWUPCanvas : canvas = MakeWUPCanvas(CustomPlayer)
                set CustomPlayer.CurrentPage = 0
                CustomPlayer.AddFullHUD(MyWUPCanvas)
                if(set CustomPlayer.UIMaps["WUP"] = option{MyWUPCanvas}) {}
                UpdateWUPHUD(CustomPlayer)

    RemoveWUPHUD(Wdata:widget_message):void=
        if(CustomPlayer := gm.PlayersMap[Wdata.Player]):
            if(Player:=Wdata.Player, PlayerUI:=GetPlayerUI[Player]):
                if(MyUI:=CustomPlayer.UIMaps["WUP"]?):
                    PlayerUI.RemoveWidget(MyUI)
                    MyUI.SetVisibility(widget_visibility.Collapsed)
                    if(set CustomPlayer.UIMaps["WUP"] = false){}

    MakeWUPCanvas(CustomPlayer:Custom_Player):canvas=
        CustomPlayer.WUPCloseButton.SetText(StringToMessage("CLOSE"))

        #CustomPlayer.UnlockWUPButton.SetText(StringToMessage("UNLOCK WEAPON"))
        
        CustomPlayer.ButtonStats.SetText(StringToMessage("STATS"))
        CustomPlayer.ButtonBonus.SetText(StringToMessage("BONUS"))
        CustomPlayer.ButtonPerk.SetText(StringToMessage("PERKS"))
        CustomPlayer.ButtonEvolve.SetText(StringToMessage("EVOLVE"))
        CustomPlayer.ButtonReset.SetText(StringToMessage("RESET STATS"))
        CustomPlayer.ButtonUpgradeAll.SetText(StringToMessage("UPGRADE ALL"))

        CustomPlayer.PerksTitle.SetText(StringToMessage("Perks"))

        CustomPlayer.StatsAttackTitle.SetText(StringToMessage("Attack:"))
        CustomPlayer.StatsAttackPercTitle.SetText(StringToMessage("Attack %:"))
        CustomPlayer.StatsStrPercTitle.SetText(StringToMessage("Strength %:"))

        CustomPlayer.Perk1Icon.SetImage(UI.UNLOCK_WEAPON)

        CustomPlayer.Perk1Icon.SetImage(UI.UnlockedIcon)
        CustomPlayer.Perk2Icon.SetImage(UI.UnlockedIcon)
        CustomPlayer.Perk3Icon.SetImage(UI.UnlockedIcon)

        CustomPlayer.UpgradeDesc.SetJustification(text_justification.Center)

        <#for(i:int=0..CustomPlayer.LockedPerks.Length):
            if(CustomPlayer.LockedPerks[i] = true):
                set CustomPlayer.PerkCost /= 10.0#>

        if(set CustomPlayer.LockedPerks[0] = false) {}
        if(set CustomPlayer.LockedPerks[1] = false) {}
        if(set CustomPlayer.LockedPerks[2] = false) {}

        for(TextBlock : CustomPlayer.AllWUPTexts):
            TextBlock.SetShadowColor(CustomPlayer.ShadowDefaults.DefaultShadowColor)
            TextBlock.SetShadowOffset(CustomPlayer.ShadowDefaults.DefaultShadowOffset)
            TextBlock.SetShadowOpacity(1.0)            

        var MyWUPCanvas: canvas = canvas:
            Slots := array:
                canvas_slot:
                    Widget:=CustomPlayer.WUPHUDBG
                    ZOrder:=0
                    SizeToContent:=false
                    Alignment := vector2{X := 0.5, Y := 0.0}
                    Anchors := anchors{Minimum := vector2{X:= 0.5, Y:= 0.0}, Maximum := vector2{X:= 0.5, Y:= 0.0}}
                    Offsets:=margin{Top:=0.0,Left:=0.0,Bottom:=CustomPlayer.WUPHUDBG.DefaultDesiredSize.Y,Right:=CustomPlayer.WUPHUDBG.DefaultDesiredSize.X}
                
                <#canvas_slot:
                    Widget:=CustomPlayer.UnlockWUPCover
                    ZOrder:=5
                    SizeToContent:=true
                    Alignment := vector2{X := 0.5, Y := 0.0}
                    Anchors := anchors{Minimum := vector2{X:= 0.5, Y:= 0.0}, Maximum := vector2{X:= 0.5, Y:= 0.0}}
                    Offsets:=margin{Top:=0.0,Left:=0.0,Bottom:=CustomPlayer.UnlockWUPCover.DefaultDesiredSize.Y,Right:=CustomPlayer.UnlockWUPCover.DefaultDesiredSize.X}
                canvas_slot:
                    ZOrder:=6
                    SizeToContent:=true
                    Anchors := anchors{Minimum := vector2{X:= 0.5, Y:= 0.0}, Maximum := vector2{X:= 0.5, Y:= 0.0}}
                    Offsets:=margin{Top:=732.0,Left:=-152.0}
                    Alignment := vector2{X:= 0.5, Y:= 0.0}
                    Widget := stack_box:
                        Orientation := orientation.Horizontal
                        Slots := array:
                            stack_box_slot:
                                Widget := CustomPlayer.UnlockWUPStone  
                            stack_box_slot:
                                Widget := CustomPlayer.UnlockWUPCost
                canvas_slot:
                    Widget:=CustomPlayer.UnlockWUPButton
                    ZOrder:=6
                    SizeToContent:=false
                    Anchors := anchors{Minimum := vector2{X:= 0.5, Y:= 0.0}, Maximum := vector2{X:= 0.5, Y:= 0.0}}
                    Offsets:=margin{Top:=828.0,Left:=-152.0, Bottom:=55.0, Right:=650.0}
                    Alignment := vector2{X:= 0.5, Y:= 0.0}


                canvas_slot:
                    Widget:=CustomPlayer.UnlockWUPTitle
                    ZOrder:=7
                    SizeToContent:=true
                    Anchors := anchors{Minimum := vector2{X:= 0.5, Y:= 0.0}, Maximum := vector2{X:= 0.5, Y:= 0.0}}
                    Offsets:=margin{Top:=248.0,Left:=-152.0}
                    Alignment := vector2{X:= 0.5, Y:= 0.0}
                canvas_slot:
                    Widget:=CustomPlayer.UnlockWUPIcon
                    ZOrder:=7
                    SizeToContent:=true
                    Anchors := anchors{Minimum := vector2{X:= 0.5, Y:= 0.0}, Maximum := vector2{X:= 0.5, Y:= 0.0}}
                    Offsets:=margin{Top:=392.0,Left:=-152.0}
                    Alignment := vector2{X:= 0.5, Y:= 0.0}#>
                <#canvas_slot:
                    Widget:=CustomPlayer.UpgradeTitleText
                    ZOrder:=1
                    SizeToContent:=true
                    Anchors := anchors{Minimum := vector2{X:= 0.5, Y:= 0.0}, Maximum := vector2{X:= 0.5, Y:= 0.0}}
                    Offsets:=margin{Top:=148.0,Left:=-248.0}#>
                canvas_slot:
                    Widget:=CustomPlayer.WUPTitle
                    ZOrder:=7
                    SizeToContent:=true
                    Anchors := anchors{Minimum := vector2{X:= 0.5, Y:= 0.0}, Maximum := vector2{X:= 0.5, Y:= 0.0}}
                    Offsets:=margin{Top:=112.0,Left:=228.0}
                canvas_slot:
                    Widget:=CustomPlayer.WeaponIcon
                    ZOrder:=1
                    SizeToContent:=true
                    Anchors := anchors{Minimum := vector2{X:= 0.5, Y:= 0.0}, Maximum := vector2{X:= 0.5, Y:= 0.0}}
                    Offsets:=margin{Top:=224.0,Left:=760.0}
                    Alignment := vector2{X:= 0.5, Y:= 0.0}
                canvas_slot:
                    Widget:=CustomPlayer.AttackIcon
                    ZOrder:=1
                    SizeToContent:=true
                    Anchors := anchors{Minimum := vector2{X:= 0.5, Y:= 0.0}, Maximum := vector2{X:= 0.5, Y:= 0.0}}
                    Offsets:=margin{Top:=536.0,Left:=232.0}
                canvas_slot:
                    Widget:=CustomPlayer.AttackPercIcon
                    ZOrder:=1
                    SizeToContent:=true
                    Anchors := anchors{Minimum := vector2{X:= 0.5, Y:= 0.0}, Maximum := vector2{X:= 0.5, Y:= 0.0}}
                    Offsets:=margin{Top:=588.0,Left:=232.0}
                canvas_slot:
                    Widget:=CustomPlayer.StrengthPercIcon
                    ZOrder:=1
                    SizeToContent:=true
                    Anchors := anchors{Minimum := vector2{X:= 0.5, Y:= 0.0}, Maximum := vector2{X:= 0.5, Y:= 0.0}}
                    Offsets:=margin{Top:=640.0,Left:=232.0}

                canvas_slot:
                    Widget:=CustomPlayer.WepInvTitle
                    ZOrder:=7
                    SizeToContent:=true
                    Anchors := anchors{Minimum := vector2{X:= 0.5, Y:= 0.0}, Maximum := vector2{X:= 0.5, Y:= 0.0}}
                    Offsets:=margin{Top:=148.0,Left:=-800.0}
                canvas_slot:
                    Widget:=CustomPlayer.InvStoneIcon
                    ZOrder:=7
                    SizeToContent:=true
                    Anchors := anchors{Minimum := vector2{X:= 0.5, Y:= 0.0}, Maximum := vector2{X:= 0.5, Y:= 0.0}}
                    Offsets:=margin{Top:=232.0,Left:=-852.0}
                canvas_slot:
                    Widget:=CustomPlayer.InvEgg1Icon
                    ZOrder:=7
                    SizeToContent:=true
                    Anchors := anchors{Minimum := vector2{X:= 0.5, Y:= 0.0}, Maximum := vector2{X:= 0.5, Y:= 0.0}}
                    Offsets:=margin{Top:=316.0,Left:=-852.0}
                canvas_slot:
                    Widget:=CustomPlayer.InvEgg2Icon
                    ZOrder:=7
                    SizeToContent:=true
                    Anchors := anchors{Minimum := vector2{X:= 0.5, Y:= 0.0}, Maximum := vector2{X:= 0.5, Y:= 0.0}}
                    Offsets:=margin{Top:=400.0,Left:=-852.0}
                canvas_slot:
                    Widget:=CustomPlayer.InvEgg3Icon
                    ZOrder:=7
                    SizeToContent:=true
                    Anchors := anchors{Minimum := vector2{X:= 0.5, Y:= 0.0}, Maximum := vector2{X:= 0.5, Y:= 0.0}}
                    Offsets:=margin{Top:=484.0,Left:=-852.0}
                canvas_slot:
                    Widget:=CustomPlayer.InvEgg4Icon
                    ZOrder:=7
                    SizeToContent:=true
                    Anchors := anchors{Minimum := vector2{X:= 0.5, Y:= 0.0}, Maximum := vector2{X:= 0.5, Y:= 0.0}}
                    Offsets:=margin{Top:=568.0,Left:=-852.0}
                canvas_slot:
                    Widget:=CustomPlayer.InvEgg5Icon
                    ZOrder:=7
                    SizeToContent:=true
                    Anchors := anchors{Minimum := vector2{X:= 0.5, Y:= 0.0}, Maximum := vector2{X:= 0.5, Y:= 0.0}}
                    Offsets:=margin{Top:=652.0,Left:=-852.0}
                canvas_slot:
                    Widget:=CustomPlayer.InvEgg6Icon
                    ZOrder:=7
                    SizeToContent:=true
                    Anchors := anchors{Minimum := vector2{X:= 0.5, Y:= 0.0}, Maximum := vector2{X:= 0.5, Y:= 0.0}}
                    Offsets:=margin{Top:=732.0,Left:=-852.0}
                canvas_slot:
                    Widget:=CustomPlayer.InvWepMatIcon
                    ZOrder:=7
                    SizeToContent:=true
                    Anchors := anchors{Minimum := vector2{X:= 0.5, Y:= 0.0}, Maximum := vector2{X:= 0.5, Y:= 0.0}}
                    Offsets:=margin{Top:=816.0,Left:=-852.0}


                canvas_slot:
                    Widget:=CustomPlayer.InvStone
                    ZOrder:=7
                    SizeToContent:=true
                    Anchors := anchors{Minimum := vector2{X:= 0.5, Y:= 0.0}, Maximum := vector2{X:= 0.5, Y:= 0.0}}
                    Offsets:=margin{Top:=232.0,Left:=-784.0}
                canvas_slot:
                    Widget:=CustomPlayer.InvEgg1
                    ZOrder:=7
                    SizeToContent:=true
                    Anchors := anchors{Minimum := vector2{X:= 0.5, Y:= 0.0}, Maximum := vector2{X:= 0.5, Y:= 0.0}}
                    Offsets:=margin{Top:=316.0,Left:=-784.0}
                canvas_slot:
                    Widget:=CustomPlayer.InvEgg2
                    ZOrder:=7
                    SizeToContent:=true
                    Anchors := anchors{Minimum := vector2{X:= 0.5, Y:= 0.0}, Maximum := vector2{X:= 0.5, Y:= 0.0}}
                    Offsets:=margin{Top:=400.0,Left:=-784.0}
                canvas_slot:
                    Widget:=CustomPlayer.InvEgg3
                    ZOrder:=7
                    SizeToContent:=true
                    Anchors := anchors{Minimum := vector2{X:= 0.5, Y:= 0.0}, Maximum := vector2{X:= 0.5, Y:= 0.0}}
                    Offsets:=margin{Top:=484.0,Left:=-784.0}
                canvas_slot:
                    Widget:=CustomPlayer.InvEgg4
                    ZOrder:=7
                    SizeToContent:=true
                    Anchors := anchors{Minimum := vector2{X:= 0.5, Y:= 0.0}, Maximum := vector2{X:= 0.5, Y:= 0.0}}
                    Offsets:=margin{Top:=568.0,Left:=-784.0}
                canvas_slot:
                    Widget:=CustomPlayer.InvEgg5
                    ZOrder:=7
                    SizeToContent:=true
                    Anchors := anchors{Minimum := vector2{X:= 0.5, Y:= 0.0}, Maximum := vector2{X:= 0.5, Y:= 0.0}}
                    Offsets:=margin{Top:=652.0,Left:=-784.0}
                canvas_slot:
                    Widget:=CustomPlayer.InvEgg6
                    ZOrder:=7
                    SizeToContent:=true
                    Anchors := anchors{Minimum := vector2{X:= 0.5, Y:= 0.0}, Maximum := vector2{X:= 0.5, Y:= 0.0}}
                    Offsets:=margin{Top:=732.0,Left:=-784.0}
                canvas_slot:
                    Widget:=CustomPlayer.InvWepMat
                    ZOrder:=7
                    SizeToContent:=true
                    Anchors := anchors{Minimum := vector2{X:= 0.5, Y:= 0.0}, Maximum := vector2{X:= 0.5, Y:= 0.0}}
                    Offsets:=margin{Top:=816.0,Left:=-784.0}
                    
                canvas_slot:
                    Widget:=CustomPlayer.ButtonStats
                    ZOrder:=1
                    SizeToContent:=false
                    Anchors := anchors{Minimum := vector2{X:= 0.5, Y:= 0.0}, Maximum := vector2{X:= 0.5, Y:= 0.0}}
                    Offsets:=margin{Top:=215.0,Left:=-536.0, Bottom:= 50.0, Right:=175.0}
                canvas_slot:
                    Widget:=CustomPlayer.ButtonBonus
                    ZOrder:=1
                    SizeToContent:=false
                    Anchors := anchors{Minimum := vector2{X:= 0.5, Y:= 0.0}, Maximum := vector2{X:= 0.5, Y:= 0.0}}
                    Offsets:=margin{Top:=215.0,Left:=-348.0, Bottom:= 50.0, Right:=175.0}
                canvas_slot:
                    Widget:=CustomPlayer.ButtonPerk
                    ZOrder:=1
                    SizeToContent:=false
                    Anchors := anchors{Minimum := vector2{X:= 0.5, Y:= 0.0}, Maximum := vector2{X:= 0.5, Y:= 0.0}}
                    Offsets:=margin{Top:=215.0,Left:=-160.0, Bottom:= 50.0, Right:=185.0}
                canvas_slot:
                    Widget:=CustomPlayer.ButtonEvolve
                    ZOrder:=1
                    SizeToContent:=false
                    Anchors := anchors{Minimum := vector2{X:= 0.5, Y:= 0.0}, Maximum := vector2{X:= 0.5, Y:= 0.0}}
                    Offsets:=margin{Top:=215.0,Left:=32.0, Bottom:= 50.0, Right:=175.0}
                canvas_slot:
                    Widget:=CustomPlayer.ButtonReset
                    ZOrder:=1
                    SizeToContent:=false
                    Alignment := vector2{X:= 0.5, Y:= 0.0}
                    Anchors := anchors{Minimum := vector2{X:= 0.5, Y:= 0.0}, Maximum := vector2{X:= 0.5, Y:= 0.0}}
                    Offsets:=margin{Top:=912.0,Left:=-704.0, Bottom:= 50.0, Right:=250.0}
                canvas_slot:
                    Widget:=CustomPlayer.ButtonUpgradeAll
                    ZOrder:=1
                    SizeToContent:=false
                    Anchors := anchors{Minimum := vector2{X:= 0.5, Y:= 0.0}, Maximum := vector2{X:= 0.5, Y:= 0.0}}
                    Offsets:=margin{Top:=780.0,Left:=-384.0, Bottom:= 50.0, Right:=450.0}
                canvas_slot:
                    Widget:=CustomPlayer.ButtonUpgrade
                    ZOrder:=1
                    SizeToContent:=false
                    Anchors := anchors{Minimum := vector2{X:= 0.5, Y:= 0.0}, Maximum := vector2{X:= 0.5, Y:= 0.0}}
                    Offsets:=margin{Top:=878.0,Left:=-432.0, Bottom:= 76.0, Right:=540.0}

                canvas_slot:
                    Widget:=CustomPlayer.UpgradeTitle
                    ZOrder:=1
                    SizeToContent:=true
                    Anchors := anchors{Minimum := vector2{X:= 0.5, Y:= 0.0}, Maximum := vector2{X:= 0.5, Y:= 0.0}}
                    Offsets:=margin{Top:=305.0,Left:=-160.0}
                    Alignment := vector2{X:= 0.5, Y:= 0.0}
                canvas_slot:
                    Widget:=CustomPlayer.UpgradeDesc
                    ZOrder:=1
                    SizeToContent:=true
                    Anchors := anchors{Minimum := vector2{X:= 0.5, Y:= 0.0}, Maximum := vector2{X:= 0.5, Y:= 0.0}}
                    Offsets:=margin{Top:=365.0,Left:=-160.0}
                    Alignment := vector2{X:= 0.5, Y:= 0.0}
                canvas_slot:
                    Widget:=CustomPlayer.UpgradeReqTitle1
                    ZOrder:=1
                    SizeToContent:=true
                    Anchors := anchors{Minimum := vector2{X:= 0.5, Y:= 0.0}, Maximum := vector2{X:= 0.5, Y:= 0.0}}
                    Offsets:=margin{Top:=535.0,Left:=-452.0}
                canvas_slot:
                    Widget:=CustomPlayer.UpgradeReqTitle2
                    ZOrder:=1
                    SizeToContent:=true
                    Anchors := anchors{Minimum := vector2{X:= 0.5, Y:= 0.0}, Maximum := vector2{X:= 0.5, Y:= 0.0}}
                    Offsets:=margin{Top:=595.0,Left:=-452.0}
                canvas_slot:
                    #Widget:=CustomPlayer.UpgradeReqAmount1
                    ZOrder:=1
                    SizeToContent:=true
                    Anchors := anchors{Minimum := vector2{X:= 0.5, Y:= 0.0}, Maximum := vector2{X:= 0.5, Y:= 0.0}}
                    Offsets:=margin{Top:=535.0,Left:=136.0}
                    Alignment := vector2{X:= 1.0, Y:= 0.0}
                    Widget := stack_box:
                        Orientation := orientation.Horizontal
                        Slots := array:
                            stack_box_slot:
                                Widget := CustomPlayer.UpgradeReqTexture1     
                            stack_box_slot:
                                Widget := CustomPlayer.UpgradeReqAmount1
                canvas_slot:
                    #Widget:=CustomPlayer.UpgradeReqAmount2
                    ZOrder:=1
                    SizeToContent:=true
                    Anchors := anchors{Minimum := vector2{X:= 0.5, Y:= 0.0}, Maximum := vector2{X:= 0.5, Y:= 0.0}}
                    Offsets:=margin{Top:=595.0,Left:=136.0}
                    Alignment := vector2{X:= 1.0, Y:= 0.0}
                    Widget := stack_box:
                        Orientation := orientation.Horizontal
                        Slots := array:
                            stack_box_slot:
                                Widget := CustomPlayer.UpgradeReqTexture2
                            stack_box_slot:
                                Widget := CustomPlayer.UpgradeReqAmount2
                canvas_slot:
                    #Widget:=CustomPlayer.UpgradeReqAmount3
                    ZOrder:=1
                    SizeToContent:=true
                    Anchors := anchors{Minimum := vector2{X:= 0.5, Y:= 0.0}, Maximum := vector2{X:= 0.5, Y:= 0.0}}
                    Offsets:=margin{Top:=655.0,Left:=136.0}
                    Alignment := vector2{X:= 1.0, Y:= 0.0}
                    Widget := stack_box:
                        Orientation := orientation.Horizontal
                        Slots := array:
                            stack_box_slot:
                                Widget := CustomPlayer.UpgradeReqTexture3  
                            stack_box_slot:
                                Widget := CustomPlayer.UpgradeReqAmount3
                canvas_slot:
                    #Widget:=CustomPlayer.UpgradeReqAmount3
                    ZOrder:=1
                    SizeToContent:=true
                    Anchors := anchors{Minimum := vector2{X:= 0.5, Y:= 0.0}, Maximum := vector2{X:= 0.5, Y:= 0.0}}
                    Offsets:=margin{Top:=715.0,Left:=136.0}
                    Alignment := vector2{X:= 1.0, Y:= 0.0}
                    Widget := stack_box:
                        Orientation := orientation.Horizontal
                        Slots := array:
                            stack_box_slot:
                                Widget := CustomPlayer.UpgradeReqTexture4
                            stack_box_slot:
                                Widget := CustomPlayer.UpgradeReqAmount4

                canvas_slot:
                    Widget:=CustomPlayer.WeaponTitle
                    ZOrder:=1
                    SizeToContent:=true
                    Anchors := anchors{Minimum := vector2{X:= 0.5, Y:= 0.0}, Maximum := vector2{X:= 0.5, Y:= 0.0}}
                    Offsets:=margin{Top:=225.0,Left:=256.0}
                canvas_slot:
                    Widget:=CustomPlayer.WeaponDesc
                    ZOrder:=1
                    SizeToContent:=true
                    Anchors := anchors{Minimum := vector2{X:= 0.5, Y:= 0.0}, Maximum := vector2{X:= 0.5, Y:= 0.0}}
                    Offsets:=margin{Top:=287.359314,Left:=256.0}
                canvas_slot:
                    Widget:=CustomPlayer.WeaponDamage
                    ZOrder:=1
                    SizeToContent:=true
                    Anchors := anchors{Minimum := vector2{X:= 0.5, Y:= 0.0}, Maximum := vector2{X:= 0.5, Y:= 0.0}}
                    Offsets:=margin{Top:=392.0,Left:=256.0}
                    
                canvas_slot:
                    Widget:=CustomPlayer.StatsTitle
                    ZOrder:=1
                    SizeToContent:=true
                    Anchors := anchors{Minimum := vector2{X:= 0.5, Y:= 0.0}, Maximum := vector2{X:= 0.5, Y:= 0.0}}
                    Offsets:=margin{Top:=485.0,Left:=248.0}
                canvas_slot:
                    Widget:=CustomPlayer.StatsAttackTitle
                    ZOrder:=1
                    SizeToContent:=true
                    Anchors := anchors{Minimum := vector2{X:= 0.5, Y:= 0.0}, Maximum := vector2{X:= 0.5, Y:= 0.0}}
                    Offsets:=margin{Top:=545.0,Left:=292.0}
                canvas_slot:
                    Widget:=CustomPlayer.StatsAttackPercTitle
                    ZOrder:=1
                    SizeToContent:=true
                    Anchors := anchors{Minimum := vector2{X:= 0.5, Y:= 0.0}, Maximum := vector2{X:= 0.5, Y:= 0.0}}
                    Offsets:=margin{Top:=600.0, Left:=292.0}
                canvas_slot:
                    Widget:=CustomPlayer.StatsStrPercTitle
                    ZOrder:=1
                    SizeToContent:=true
                    Anchors := anchors{Minimum := vector2{X:= 0.5, Y:= 0.0}, Maximum := vector2{X:= 0.5, Y:= 0.0}}
                    Offsets:=margin{Top:=655.0,Left:=292.0}
                canvas_slot:
                    Widget:=CustomPlayer.StatsAttackAmount
                    ZOrder:=1
                    SizeToContent:=true
                    Anchors := anchors{Minimum := vector2{X:= 0.5, Y:= 0.0}, Maximum := vector2{X:= 0.5, Y:= 0.0}}
                    Offsets:=margin{Top:=545.0,Left:=624.0}
                    Alignment := vector2{X:= 1.0, Y:= 0.0}
                canvas_slot:
                    Widget:=CustomPlayer.StatsAttackPercAmount
                    ZOrder:=1
                    SizeToContent:=true
                    Anchors := anchors{Minimum := vector2{X:= 0.5, Y:= 0.0}, Maximum := vector2{X:= 0.5, Y:= 0.0}}
                    Offsets:=margin{Top:=600.0, Left:=624.0}
                    Alignment := vector2{X:= 1.0, Y:= 0.0}
                canvas_slot:
                    Widget:=CustomPlayer.StatsStrPercAmount
                    ZOrder:=1
                    SizeToContent:=true
                    Anchors := anchors{Minimum := vector2{X:= 0.5, Y:= 0.0}, Maximum := vector2{X:= 0.5, Y:= 0.0}}
                    Offsets:=margin{Top:=655.0,Left:=624.0}
                    Alignment := vector2{X:= 1.0, Y:= 0.0}

                canvas_slot:
                    Widget:=CustomPlayer.BonusTitle
                    ZOrder:=1
                    SizeToContent:=true
                    Anchors := anchors{Minimum := vector2{X:= 0.5, Y:= 0.0}, Maximum := vector2{X:= 0.5, Y:= 0.0}}
                    Offsets:=margin{Top:=485.0,Left:=704.0}
                canvas_slot:
                    Widget:=CustomPlayer.BonusAttack
                    ZOrder:=1
                    SizeToContent:=true
                    Anchors := anchors{Minimum := vector2{X:= 0.5, Y:= 0.0}, Maximum := vector2{X:= 0.5, Y:= 0.0}}
                    Offsets:=margin{Top:=545.0,Left:=700.0}
                canvas_slot:
                    Widget:=CustomPlayer.BonusAttackPerc
                    ZOrder:=1
                    SizeToContent:=true
                    Anchors := anchors{Minimum := vector2{X:= 0.5, Y:= 0.0}, Maximum := vector2{X:= 0.5, Y:= 0.0}}
                    Offsets:=margin{Top:=600.0,Left:=700.0}
                canvas_slot:
                    Widget:=CustomPlayer.BonusStrPerc
                    ZOrder:=1
                    SizeToContent:=true
                    Anchors := anchors{Minimum := vector2{X:= 0.5, Y:= 0.0}, Maximum := vector2{X:= 0.5, Y:= 0.0}}
                    Offsets:=margin{Top:=655.0,Left:=700.0}
                    
                canvas_slot:
                    Widget:=CustomPlayer.UnlockPerksTitle
                    ZOrder:=4
                    SizeToContent:=true
                    Anchors := anchors{Minimum := vector2{X:= 0.5, Y:= 0.0}, Maximum := vector2{X:= 0.5, Y:= 0.0}}
                    Offsets:=margin{Top:=772.0,Left:=560.0}
                    Alignment := vector2{X:= 0.5, Y:= 0.0}
                canvas_slot:
                    Widget:=CustomPlayer.UnlockPerksLockIcon
                    ZOrder:=4
                    SizeToContent:=false
                    Anchors := anchors{Minimum := vector2{X:= 0.5, Y:= 0.0}, Maximum := vector2{X:= 0.5, Y:= 0.0}}
                    Offsets:=margin{Top:=848.0,Left:=520.0, Bottom:=CustomPlayer.UnlockPerksLockIcon.DefaultDesiredSize.Y,Right:=CustomPlayer.UnlockPerksLockIcon.DefaultDesiredSize.X}
                canvas_slot:
                    Widget:=CustomPlayer.UnlockPerksBG
                    ZOrder:=3
                    SizeToContent:=true
                    Anchors := anchors{Minimum := vector2{X:= 0.5, Y:= 0.0}, Maximum := vector2{X:= 0.5, Y:= 0.0}}
                    Offsets:=margin{Top:=732.0,Left:=212.0, Bottom:=251.575211,Right:=696.027771}
                  
                    
                canvas_slot:
                    Widget:=CustomPlayer.PerksTitle
                    ZOrder:=1
                    SizeToContent:=true
                    Anchors := anchors{Minimum := vector2{X:= 0.5, Y:= 0.0}, Maximum := vector2{X:= 0.5, Y:= 0.0}}
                    Offsets:=margin{Top:=744.0,Left:=248.0}
                canvas_slot:
                    Widget:=CustomPlayer.PerksTier
                    ZOrder:=1
                    SizeToContent:=true
                    Anchors := anchors{Minimum := vector2{X:= 0.5, Y:= 0.0}, Maximum := vector2{X:= 0.5, Y:= 0.0}}
                    Offsets:=margin{Top:=744.0,Left:=884.0}
                    Alignment := vector2{X:= 1.0, Y:= 0.0}
                canvas_slot:
                    Widget:=CustomPlayer.Perk1
                    ZOrder:=1
                    SizeToContent:=true
                    Anchors := anchors{Minimum := vector2{X:= 0.5, Y:= 0.0}, Maximum := vector2{X:= 0.5, Y:= 0.0}}
                    Offsets:=margin{Top:=800.0,Left:=388.0}
                canvas_slot:
                    Widget:=CustomPlayer.Perk2
                    ZOrder:=1
                    SizeToContent:=true
                    Anchors := anchors{Minimum := vector2{X:= 0.5, Y:= 0.0}, Maximum := vector2{X:= 0.5, Y:= 0.0}}
                    Offsets:=margin{Top:=860.0,Left:=388.0}
                canvas_slot:
                    Widget:=CustomPlayer.Perk3
                    ZOrder:=1
                    SizeToContent:=true
                    Anchors := anchors{Minimum := vector2{X:= 0.5, Y:= 0.0}, Maximum := vector2{X:= 0.5, Y:= 0.0}}
                    Offsets:=margin{Top:=920.0,Left:=388.0}
                canvas_slot:
                    Widget:=CustomPlayer.Perk1Icon
                    ZOrder:=1
                    SizeToContent:=true
                    Anchors := anchors{Minimum := vector2{X:= 0.5, Y:= 0.0}, Maximum := vector2{X:= 0.5, Y:= 0.0}}
                    Offsets:=margin{Top:=800.0,Left:=300.0}
                canvas_slot:
                    Widget:=CustomPlayer.Perk2Icon
                    ZOrder:=1
                    SizeToContent:=true
                    Anchors := anchors{Minimum := vector2{X:= 0.5, Y:= 0.0}, Maximum := vector2{X:= 0.5, Y:= 0.0}}
                    Offsets:=margin{Top:=860.0,Left:=300.0}
                canvas_slot:
                    Widget:=CustomPlayer.Perk3Icon
                    ZOrder:=1
                    SizeToContent:=true
                    Anchors := anchors{Minimum := vector2{X:= 0.5, Y:= 0.0}, Maximum := vector2{X:= 0.5, Y:= 0.0}}
                    Offsets:=margin{Top:=920.0,Left:=300.0}
                canvas_slot:
                    Widget:=CustomPlayer.ButtonPerk1
                    ZOrder:=2
                    SizeToContent:=false
                    Anchors := anchors{Minimum := vector2{X:= 0.5, Y:= 0.0}, Maximum := vector2{X:= 0.5, Y:= 0.0}}
                    Offsets:=margin{Top:=800.0,Left:=268.0, Bottom:= 40.0,Right:= 100.0}
                canvas_slot:
                    Widget:=CustomPlayer.ButtonPerk2
                    ZOrder:=2
                    SizeToContent:=false
                    Anchors := anchors{Minimum := vector2{X:= 0.5, Y:= 0.0}, Maximum := vector2{X:= 0.5, Y:= 0.0}}
                    Offsets:=margin{Top:=860.0,Left:=268.0, Bottom:= 40.0,Right:= 100.0}
                canvas_slot:
                    Widget:=CustomPlayer.ButtonPerk3
                    ZOrder:=2
                    SizeToContent:=false
                    Anchors := anchors{Minimum := vector2{X:= 0.5, Y:= 0.0}, Maximum := vector2{X:= 0.5, Y:= 0.0}}
                    Offsets:=margin{Top:=920.0,Left:=268.0, Bottom:= 40.0,Right:= 100.0}

                canvas_slot:
                    Widget:=CustomPlayer.WUPCloseButton
                    ZOrder:=9
                    SizeToContent:=false
                    Alignment := vector2{X:= 0.5, Y:= 0.0}
                    Anchors := anchors{Minimum := vector2{X:= 0.5, Y:= 0.0}, Maximum := vector2{X:= 0.5, Y:= 0.0}}
                    Offsets:=margin{Top:=140.0,Left:=-164.0,Bottom:=50.0,Right:=700.0}
        return MyWUPCanvas

    #CALL THIS AFTER ANY UPGRADE TO MAKE SURE THE HUD IS UP TO DATE
    UpdateWUPHUD(CustomPlayer: Custom_Player):void=
        Visible := widget_visibility.Visible
        Collapsed := widget_visibility.Collapsed

        for(i:int=0..CustomPlayer.InvTexts.Length):
            if(InvText:=CustomPlayer.InvTexts[i]):
                if(i=0): #STONE
                    InvText.SetText(StringToMessage("{NumberCondensor(CustomPlayer.StoneAmount)}"))
                else if(i=7): #WEP MAT
                    InvText.SetText(StringToMessage("{NumberCondensor(CustomPlayer.WepMatAmount)}"))
                else: #EGGS
                    if(EggAmount:=CustomPlayer.EggAmounts[i-1]): #OFFSET STONE BEING FIRST
                        InvText.SetText(StringToMessage("{NumberCondensor(EggAmount)}"))
        if(CustomPlayer.WeaponTier <= 0):
            CustomPlayer.UpgradeReqTitle2.SetText(StringToMessage(""))
            CustomPlayer.UpgradeReqTitle3.SetText(StringToMessage(""))
            CustomPlayer.UpgradeReqAmount2.SetText(StringToMessage(""))
            CustomPlayer.UpgradeReqAmount3.SetText(StringToMessage(""))
            CustomPlayer.UpgradeReqTexture2.SetImage(UI.Blank)
            CustomPlayer.UpgradeReqTexture3.SetImage(UI.Blank)
            CustomPlayer.UpgradeReqTexture4.SetImage(UI.Blank)
            CustomPlayer.UpgradeTitle.SetText(StringToMessage("Unlock Weapon"))
            CustomPlayer.UpgradeDesc.SetText(StringToMessage("Unlock your weapon\nto defeat mobs and bosses!"))
            CustomPlayer.UpgradeReqTitle1.SetText(StringToMessage("Cost:"))
            if(WeaponCost := EvolveCost[0][1]):
                CustomPlayer.UpgradeReqAmount1.SetText(StringToMessage(" {NumberCondensor(WeaponCost)}"))
            CustomPlayer.UpgradeReqTexture1.SetImage(UI.StoneIcon)
            CustomPlayer.ButtonUpgrade.SetText(StringToMessage("Unlock"))
            CustomPlayer.ButtonUpgradeAll.SetText(StringToMessage(""))
            <#if(CustomPlayer.StoneAmount >= EvolveCost[0][1]):
                CustomPlayer.UnlockWUPCost.SetTextColor(MakeColorFromHex("39FF39"))
            else:
                CustomPlayer.UnlockWUPCost.SetTextColor(MakeColorFromHex("fa4d56"))
            CustomPlayer.UnlockWUPTitle.SetDesiredSize(vector2{X:=CustomPlayer.UnlockWUPTitle.DefaultDesiredSize.X,Y:=CustomPlayer.UnlockWUPTitle.DefaultDesiredSize.Y})
            CustomPlayer.UnlockWUPIcon.SetDesiredSize(vector2{X:=200.0,Y:=200.0})
            CustomPlayer.UnlockWUPCover.SetDesiredSize(vector2{X:=CustomPlayer.UnlockWUPCover.DefaultDesiredSize.X,Y:=CustomPlayer.UnlockWUPCover.DefaultDesiredSize.Y})
            CustomPlayer.UnlockWUPStone.SetDesiredSize(vector2{X:=50.0,Y:=50.0})
            if(WeaponCost:=EvolveCost[0][1]):
                CustomPlayer.UnlockWUPCost.SetText(StringToMessage(" {NumberCondensor(WeaponCost)}"))
            CustomPlayer.UnlockWUPButton.SetEnabled(true)
            CustomPlayer.UnlockWUPButton.SetVisibility(Visible)#>
        else:
            <#CustomPlayer.UnlockWUPTitle.SetDesiredSize(vector2{X:=0.0,Y:=0.0})
            CustomPlayer.UnlockWUPIcon.SetDesiredSize(vector2{X:=0.0,Y:=0.0})
            CustomPlayer.UnlockWUPCover.SetDesiredSize(vector2{X:=0.0,Y:=0.0})
            CustomPlayer.UnlockWUPStone.SetDesiredSize(vector2{X:=0.0,Y:=0.0})
            CustomPlayer.UnlockWUPCost.SetText(StringToMessage(""))

            CustomPlayer.UnlockWUPButton.SetEnabled(false)
            CustomPlayer.UnlockWUPButton.SetVisibility(Collapsed)#>
            CustomPlayer.ButtonStats.SetEnabled(true)
            CustomPlayer.ButtonBonus.SetEnabled(true)
            CustomPlayer.ButtonPerk.SetEnabled(true)
            CustomPlayer.ButtonEvolve.SetEnabled(true)
            CustomPlayer.ButtonReset.SetEnabled(true)
            CustomPlayer.ButtonUpgrade.SetEnabled(true)
            CustomPlayer.ButtonUpgradeAll.SetEnabled(true)
            CustomPlayer.ButtonUpgradeAll.SetText(StringToMessage("UPGRADE ALL"))

            if(CurrentTitle:=CustomPlayer.UpgradeTitles[CustomPlayer.CurrentPage],CurrentDesc:=CustomPlayer.UpgradeDescs[CustomPlayer.CurrentPage],UpgradeBtn:=CustomPlayer.UpgradeBtns[CustomPlayer.CurrentPage]):
                CustomPlayer.UpgradeTitle.SetText(StringToMessage(CurrentTitle))
                CustomPlayer.UpgradeDesc.SetText(StringToMessage(CurrentDesc))

                #CLEAN SLATE EVERYTIME YOU MOVE PAGES
                CustomPlayer.UpgradeReqTitle1.SetText(StringToMessage(""))
                CustomPlayer.UpgradeReqTitle2.SetText(StringToMessage(""))
                CustomPlayer.UpgradeReqTitle3.SetText(StringToMessage(""))
                CustomPlayer.UpgradeReqAmount1.SetText(StringToMessage(""))
                CustomPlayer.UpgradeReqAmount2.SetText(StringToMessage(""))
                CustomPlayer.UpgradeReqAmount3.SetText(StringToMessage(""))
                CustomPlayer.UpgradeReqAmount4.SetText(StringToMessage(""))
                CustomPlayer.UpgradeReqTexture1.SetImage(UI.Blank)
                CustomPlayer.UpgradeReqTexture2.SetImage(UI.Blank)
                CustomPlayer.UpgradeReqTexture3.SetImage(UI.Blank)
                CustomPlayer.UpgradeReqTexture4.SetVisibility(widget_visibility.Collapsed)

                CustomPlayer.ButtonUpgrade.SetText(StringToMessage(UpgradeBtn))
                if(CustomPlayer.CurrentPage = 0): #STAT
                    CustomPlayer.ButtonUpgradeAll.SetVisibility(widget_visibility.Visible)
                    CustomPlayer.UpgradeReqTexture1.SetImage(UI.StoneIcon)
                    var WhichEgg : int = 0
                    if(CustomPlayer.WeaponTier-1 <= 5):
                        set WhichEgg = CustomPlayer.WeaponTier-1
                    else:
                        set WhichEgg = 5
                    if(EggIcon:=CustomPlayer.EggIcons[WhichEgg]):
                        CustomPlayer.UpgradeReqTexture2.SetImage(EggIcon.DefaultImage)

                    CustomPlayer.UpgradeReqTitle1.SetText(StringToMessage("Cost:"))
                    if(EggCost:= 1 + Floor[CustomPlayer.WeaponStatLevel/10.0], StoneCost := EvolveCost[CustomPlayer.WeaponTier-1][1] / 2.5):
                        if(CustomPlayer.StoneAmount >= StoneCost):
                            CustomPlayer.UpgradeReqAmount1.SetTextColor(MakeColorFromHex("39FF39"))
                        else:
                            CustomPlayer.UpgradeReqAmount1.SetTextColor(MakeColorFromHex("fa4d56"))
                        if(CustomPlayer.EggAmounts[WhichEgg] >= EggCost*1.0):
                            CustomPlayer.UpgradeReqAmount2.SetTextColor(MakeColorFromHex("39FF39"))
                        else:
                            CustomPlayer.UpgradeReqAmount2.SetTextColor(MakeColorFromHex("fa4d56"))

                        CustomPlayer.UpgradeReqAmount1.SetText(StringToMessage(" {NumberCondensor(StoneCost)}"))
                        CustomPlayer.UpgradeReqAmount2.SetText(StringToMessage(" {NumberCondensor(EggCost*1.0)}"))
                else if(CustomPlayer.CurrentPage = 1): #BONUS 
                    CustomPlayer.ButtonUpgradeAll.SetVisibility(widget_visibility.Visible)
                    if(NewBonusCost:= (EvolveCost[CustomPlayer.WeaponTier-1][1] * 2.0)):
                        CustomPlayer.UpgradeReqTexture1.SetImage(UI.StoneIcon)
                        CustomPlayer.UpgradeReqTitle1.SetText(StringToMessage("Cost:"))
                        if(CustomPlayer.StoneAmount >= NewBonusCost):
                            CustomPlayer.UpgradeReqAmount1.SetTextColor(MakeColorFromHex("39FF39"))
                        else:
                            CustomPlayer.UpgradeReqAmount1.SetTextColor(MakeColorFromHex("fa4d56"))
                        CustomPlayer.UpgradeReqAmount1.SetText(StringToMessage(" {NumberCondensor(NewBonusCost)}"))
                else if(CustomPlayer.CurrentPage = 2): #Perk 
                    CustomPlayer.ButtonUpgradeAll.SetVisibility(widget_visibility.Collapsed)
                        if(NewEvolveCost:=EvolveCost[CustomPlayer.WeaponTier-1][1]):
                            var NewPerkCost : float = NewEvolveCost * 10.0
                            for(i:int=0..CustomPlayer.LockedPerks.Length):
                                if(CustomPlayer.LockedPerks[i] = true):
                                    set NewPerkCost *= 10.0
                            CustomPlayer.UpgradeReqTexture1.SetImage(UI.StoneIcon)
                            CustomPlayer.UpgradeReqTitle1.SetText(StringToMessage("Cost:"))
                            if(CustomPlayer.StoneAmount >= NewPerkCost):
                                CustomPlayer.UpgradeReqAmount1.SetTextColor(MakeColorFromHex("39FF39"))
                            else:
                                CustomPlayer.UpgradeReqAmount1.SetTextColor(MakeColorFromHex("fa4d56"))
                            CustomPlayer.UpgradeReqAmount1.SetText(StringToMessage(" {NumberCondensor(NewPerkCost)}"))
                            if(CustomPlayer.PerksUnlocked = true):
                                CustomPlayer.ButtonUpgrade.SetText(StringToMessage("REROLL PERKS"))
                else if(CustomPlayer.CurrentPage = 3): #EVOLVE
                    CustomPlayer.ButtonUpgradeAll.SetVisibility(widget_visibility.Collapsed)
                    if(CustomPlayer.WeaponTier < 6 or CustomPlayer.A4Unlocked = true and CustomPlayer.WeaponTier < 9):
                        Print("CAN EVOLVE HUD UPDATE")
                        CustomPlayer.UpgradeReqTexture2.SetImage(UI.StoneIcon)
                        CustomPlayer.UpgradeReqTexture3.SetImage(UI.WepMat)
                        if(CustomPlayer.WeaponTier > 5):
                            CustomPlayer.UpgradeReqTexture4.SetVisibility(widget_visibility.Visible)
                        CustomPlayer.UpgradeReqTitle1.SetText(StringToMessage("Lv. Requirement:"))
                        CustomPlayer.UpgradeReqTitle2.SetText(StringToMessage("Cost:"))
                        
                        if:
                            LvlReq:=EvolveCost[CustomPlayer.WeaponTier][0]
                            StoneCost:=EvolveCost[CustomPlayer.WeaponTier][1]
                            MatCost:=EvolveCost[CustomPlayer.WeaponTier][2]
                            ShardCost:=EvolveCost[CustomPlayer.WeaponTier][3]
                        then:
                            if((CustomPlayer.WeaponStatLevel + CustomPlayer.WeaponBonusLevel + CustomPlayer.WeaponPerkLevel) >= LvlReq):
                                CustomPlayer.UpgradeReqAmount1.SetTextColor(MakeColorFromHex("39FF39"))
                            else:
                                CustomPlayer.UpgradeReqAmount1.SetTextColor(MakeColorFromHex("fa4d56"))
                            if(CustomPlayer.StoneAmount >= StoneCost):
                                CustomPlayer.UpgradeReqAmount2.SetTextColor(MakeColorFromHex("39FF39"))
                            else:
                                CustomPlayer.UpgradeReqAmount2.SetTextColor(MakeColorFromHex("fa4d56"))
                            if(CustomPlayer.WepMatAmount >= MatCost):
                                CustomPlayer.UpgradeReqAmount3.SetTextColor(MakeColorFromHex("39FF39"))
                            else:
                                CustomPlayer.UpgradeReqAmount3.SetTextColor(MakeColorFromHex("fa4d56"))
                            if(CustomPlayer.ShardAmount >= ShardCost):
                                CustomPlayer.UpgradeReqAmount4.SetTextColor(MakeColorFromHex("39FF39"))
                            else:
                                CustomPlayer.UpgradeReqAmount4.SetTextColor(MakeColorFromHex("fa4d56"))
                            CustomPlayer.UpgradeReqAmount1.SetText(StringToMessage("Lv. {NumberCondensor(LvlReq)}"))
                            CustomPlayer.UpgradeReqAmount2.SetText(StringToMessage(" {NumberCondensor(StoneCost)}"))
                            CustomPlayer.UpgradeReqAmount3.SetText(StringToMessage(" {NumberCondensor(MatCost)}"))
                            if(CustomPlayer.WeaponTier > 5):
                                CustomPlayer.UpgradeReqAmount4.SetText(StringToMessage(" {NumberCondensor(ShardCost)}"))
                    else:
                        CustomPlayer.UpgradeReqTexture2.SetImage(UI.Blank)
                        CustomPlayer.UpgradeReqTexture3.SetImage(UI.Blank)
                        CustomPlayer.UpgradeReqTexture4.SetVisibility(widget_visibility.Collapsed)
                        CustomPlayer.UpgradeReqAmount1.SetTextColor(MakeColorFromHex("FFFFFF"))
                        CustomPlayer.UpgradeReqAmount2.SetTextColor(MakeColorFromHex("FFFFFF"))
                        CustomPlayer.UpgradeReqAmount3.SetTextColor(MakeColorFromHex("FFFFFF"))
                        CustomPlayer.UpgradeReqAmount4.SetTextColor(MakeColorFromHex("FFFFFF"))
                        CustomPlayer.UpgradeReqAmount1.SetText(StringToMessage("Lv. --"))
                        CustomPlayer.UpgradeReqAmount2.SetText(StringToMessage(""))
                        CustomPlayer.UpgradeReqAmount3.SetText(StringToMessage(""))
                        CustomPlayer.UpgradeReqAmount4.SetText(StringToMessage(""))
                        CustomPlayer.ButtonUpgrade.SetText(StringToMessage("EVOLVE MAXED"))

                else if(CustomPlayer.CurrentPage = 4): #RESET STATS
                    CustomPlayer.ButtonUpgradeAll.SetVisibility(widget_visibility.Collapsed)
                    CustomPlayer.UpgradeReqTitle1.SetText(StringToMessage("Cost:"))
                    CustomPlayer.UpgradeReqTexture1.SetImage(UI.WepMat)
                    CustomPlayer.UpgradeReqAmount1.SetText(StringToMessage("{NumberCondensor(ResetCost)}"))
                    if(CustomPlayer.WepMatAmount >= ResetCost):
                        CustomPlayer.UpgradeReqAmount1.SetTextColor(MakeColorFromHex("39FF39"))
                    else:
                        CustomPlayer.UpgradeReqAmount1.SetTextColor(MakeColorFromHex("fa4d56"))

                    CustomPlayer.ButtonUpgrade.SetText(StringToMessage("RESET STATS"))
                else:
                    Print("You somehow missed all upgrades")

            CustomPlayer.WeaponTitle.SetText(StringToMessage("Weapon Lv. {NumberCondensor(CustomPlayer.WeaponStatLevel + CustomPlayer.WeaponBonusLevel + CustomPlayer.WeaponPerkLevel)}"))
            if(WeaponTierName:=WeaponTierNames[CustomPlayer.WeaponTier-1],WeaponTierColor:=WeaponTierColors[CustomPlayer.WeaponTier-1]):
                CustomPlayer.WeaponDesc.SetText(StringToMessage("{WeaponTierName} Assault Rifle"))
                CustomPlayer.WeaponDesc.SetTextColor(MakeColorFromHex(WeaponTierColor))

            CustomPlayer.WeaponDamage.SetText(StringToMessage("Total Damage: {NumberCondensor(CustomPlayer.GetPlayerWeaponDamage())}"))

            CustomPlayer.StatsTitle.SetText(StringToMessage("Stats (+{NumberCondensor(CustomPlayer.WeaponStatLevel)})"))
            CustomPlayer.BonusTitle.SetText(StringToMessage("Bonus (+{NumberCondensor(CustomPlayer.WeaponBonusLevel)})"))

            for(i:int=0..CustomPlayer.StatTexts.Length):
                if(StatText:=CustomPlayer.StatTexts[i],BonusText:=CustomPlayer.BonusTexts[i],WeaponBonus:=CustomPlayer.WeaponBonuses[i],WeaponStat:=CustomPlayer.WeaponStats[i]):
                    if(i=0): #Not a %
                        StatText.SetText(StringToMessage("{NumberCondensor(WeaponStat)}"))
                    else:
                        StatText.SetText(StringToMessage("{NumberCondensor(WeaponStat*100.0)}%"))
                    BonusText.SetText(StringToMessage("{NumberCondensor(WeaponBonus*100.0)}%"))
    
            #RESET PERKS AREA
            #UPDATED PERKS
            if(CustomPlayer.PerksUnlocked = true):
                CustomPlayer.UnlockPerksTitle.SetText(StringToMessage(""))
                CustomPlayer.UnlockPerksLockIcon.SetImage(UI.Blank)
                CustomPlayer.UnlockPerksBG.SetDesiredSize(vector2{X:=0.0,Y:=0.0})
                CustomPlayer.ButtonPerk1.SetEnabled(true)
                CustomPlayer.ButtonPerk2.SetEnabled(true)
                CustomPlayer.ButtonPerk3.SetEnabled(true)

                if(WeaponTierColor:=WeaponTierColors[CustomPlayer.WeaponPerkTier],PerkTier:=WeaponTierNames[CustomPlayer.WeaponPerkTier],ActivePerk1:=CustomPlayer.ActivePerks[0],ActivePerk2:=CustomPlayer.ActivePerks[1],ActivePerk3:=CustomPlayer.ActivePerks[2]):
                    CustomPlayer.PerksTier.SetTextColor(MakeColorFromHex(WeaponTierColor))
                    CustomPlayer.PerksTier.SetText(StringToMessage("{PerkTier} (+{NumberCondensor(CustomPlayer.WeaponPerkLevel)})"))
                    if(ActivePerk1(0) = "Instant Reload"):
                        CustomPlayer.Perk1.SetText(StringToMessage("{ActivePerk1(0)}"))
                    else:
                        CustomPlayer.Perk1.SetText(StringToMessage("+{NumberCondensor(ActivePerk1(1)*100.0)}% {ActivePerk1(0)}"))
                    if(ActivePerk2(0) = "Instant Reload"):
                        CustomPlayer.Perk2.SetText(StringToMessage("{ActivePerk2(0)}"))
                    else:
                        CustomPlayer.Perk2.SetText(StringToMessage("+{NumberCondensor(ActivePerk2(1)*100.0)}% {ActivePerk2(0)}"))
                    if(ActivePerk3(0) = "Instant Reload"):
                        CustomPlayer.Perk3.SetText(StringToMessage("{ActivePerk3(0)}"))
                    else:
                        CustomPlayer.Perk3.SetText(StringToMessage("+{NumberCondensor(ActivePerk3(1)*100.0)}% {ActivePerk3(0)}"))
            else:
                CustomPlayer.UnlockPerksTitle.SetText(StringToMessage("Unlock Perks"))
                CustomPlayer.UnlockPerksLockIcon.SetImage(UI.LockedIconWhite)
                CustomPlayer.UnlockPerksBG.SetDesiredSize(vector2{X:=696.027771,Y:=251.575211})
                #CustomPlayer.ButtonPerk1.SetEnabled(false)
                #CustomPlayer.ButtonPerk2.SetEnabled(false)
                #CustomPlayer.ButtonPerk3.SetEnabled(false)

            
    ChangePage(Wdata:widget_message, whichPage:int):void=
        if(CustomPlayer := gm.PlayersMap[Wdata.Player]):
            if(CustomPlayer.WeaponTier > 0):
                set CustomPlayer.CurrentPage = whichPage
                UpdateWUPHUD(CustomPlayer)
            else:
                AudioManager.Denied.Play(CustomPlayer.Player)

    UnlockWUPHUD(Wdata:widget_message):void=
        if(CustomPlayer := gm.PlayersMap[Wdata.Player]):
            if(UnlockCost:=EvolveCost[0][1], CustomPlayer.WeaponTier <= 0, CustomPlayer.StoneAmount >= UnlockCost):
                AudioManager.LeveledUp.Play(CustomPlayer.Player)
                CustomPlayer.Spend(UnlockCost, "Stone", 0)

                Tutorial.UpdateTutorialUI(CustomPlayer, "UnlockWeapon", 2500000.0)

                #INCREASE WEAPON TIER
                set CustomPlayer.WeaponTier += 1

                #INCREASE EVOLVE COST
                #if(NewEvolveCost:=EvolveCosts[CustomPlayer.WeaponTier]):
                    #if(set EvolveCost[1] = NewEvolveCost) {} 
                
                #GIVE NEW GUN
                if(CustomPlayer.WeaponTier <= 6):
                    spawn{GrantWeapon(CustomPlayer)}

                set CustomPlayer.CurrentPage = 0
                RemoveWUPHUD(Wdata)
                if (Agent := agent[Wdata.Player]):
                    OpenWUPHUD(Agent)
                UpdateWUPHUD(CustomPlayer)
                SaveAllData(CustomPlayer)
                RemoveWUPHUD(Wdata)
            else:
                AudioManager.Denied.Play(CustomPlayer.Player)


    OnUpgradeButton(Wdata:widget_message):void=
        if(CustomPlayer := gm.PlayersMap[Wdata.Player]):
            if(CustomPlayer.WeaponTier <= 0):
                UnlockWUPHUD(Wdata)
            else:
                if(CustomPlayer.CurrentPage = 0):
                    if(CustomPlayer.WeaponStatLevel < 100.0):
                        OnStatsUpgrade(CustomPlayer)
                    else:
                        AudioManager.Denied.Play(CustomPlayer.Player)
                else if(CustomPlayer.CurrentPage = 1):
                    if(CustomPlayer.WeaponBonusLevel < 100.0):
                        OnBonusUpgrade(CustomPlayer)
                    else:
                        AudioManager.Denied.Play(CustomPlayer.Player)
                else if(CustomPlayer.CurrentPage = 2):
                    OnPerkUpgrade(CustomPlayer)
                else if(CustomPlayer.CurrentPage = 3):
                    OnEvolveUpgrade(CustomPlayer)
                else if(CustomPlayer.CurrentPage = 4):
                    OnResetStats(CustomPlayer)
                else:
                    Print("You somehow missed all pages?")

    OnUpgradeAllButton(Wdata:widget_message):void=
        if(CustomPlayer := gm.PlayersMap[Wdata.Player]):
            if(CustomPlayer.WeaponTier > 0):
                if(CustomPlayer.CurrentPage = 0):
                    if:
                        CustomPlayer.WeaponStatLevel < 100.0
                        CustomPlayer.WeaponStatRolling = false
                    then:
                        set CustomPlayer.WeaponStatRolling = true
                        spawn{LoopUpgrades(CustomPlayer, "Stat")}
                    else:
                        AudioManager.Denied.Play(CustomPlayer.Player)
                else if(CustomPlayer.CurrentPage = 1):
                    if:
                        CustomPlayer.WeaponBonusLevel < 100.0
                        CustomPlayer.WeaponBonusRolling = false
                    then:
                        set CustomPlayer.WeaponBonusRolling = true
                        spawn{LoopUpgrades(CustomPlayer, "Bonus")}
                    else:
                        AudioManager.Denied.Play(CustomPlayer.Player)
                else:
                    Print("You somehow missed all pages?")
            else:
                AudioManager.Denied.Play(CustomPlayer.Player)

    LoopUpgrades(CustomPlayer:Custom_Player, WhichUpgrade:string)<suspends>:void=
        var WhichEgg : int = 0
        if(CustomPlayer.WeaponTier-1 <= 5):
            set WhichEgg = CustomPlayer.WeaponTier-1
        else:
            set WhichEgg = 5
        
        if(WhichUpgrade = "Stat"):
            var HowManyLevels : float = 100.0 - CustomPlayer.WeaponStatLevel
            for(i:=0..Floor[HowManyLevels]):
                if:
                    CustomPlayer.WeaponStatLevel < 100.0
                    CurrentEggs:=CustomPlayer.EggAmounts[WhichEgg]
                    StoneCost:=EvolveCost[CustomPlayer.WeaponTier-1][1] / 2.5
                    EggCost:= 1.0 * (1 + Floor[CustomPlayer.WeaponStatLevel/10.0])
                    CustomPlayer.StoneAmount >= StoneCost and CurrentEggs >= EggCost
                then:
                    OnStatsUpgrade(CustomPlayer)
                else:
                    Print("END LOOP")
                    set HowManyLevels = i*1.0
                Sleep(0.02)
            set CustomPlayer.WeaponStatRolling = false
        else if(WhichUpgrade = "Bonus"):
            var HowManyLevels : float = 100.0 - CustomPlayer.WeaponBonusLevel
            for(i:=0..Floor[HowManyLevels]):
                if:
                    CustomPlayer.WeaponBonusLevel < 100.0
                    NewBonusCost:= (EvolveCost[CustomPlayer.WeaponTier-1][1] * 2.0)
                    CustomPlayer.StoneAmount >= NewBonusCost
                then:
                    OnBonusUpgrade(CustomPlayer)
                else:
                    Print("END LOOP")
                    set HowManyLevels = i*1.0
                Sleep(0.02)
            set CustomPlayer.WeaponBonusRolling = false



    OnStatsUpgrade(CustomPlayer:Custom_Player):void=
        var WhichEgg : int = 0
        if(CustomPlayer.WeaponTier-1 <= 5):
            set WhichEgg = CustomPlayer.WeaponTier-1
        else:
            set WhichEgg = 5
        if(CurrentEggs:=CustomPlayer.EggAmounts[WhichEgg],StoneCost:=EvolveCost[CustomPlayer.WeaponTier-1][1] / 2.5, EggCost:= 1.0 * (1 + Floor[CustomPlayer.WeaponStatLevel/10.0])):
            Print("StoneCost: {StoneCost}")
            Print("EggCost: {EggCost}")
            if(CustomPlayer.StoneAmount >= StoneCost and CurrentEggs >= EggCost):

                if(Analytic:=AudioManager.WeaponAnalytics[0]): #ANALYTIC
                    Analytic.Submit(CustomPlayer.Player)

                #TAKE AWAY RESOURCES
                CustomPlayer.Spend(StoneCost, "Stone", 0)
                CustomPlayer.Spend(EggCost, "Egg", WhichEgg)

                AudioManager.LeveledUp.Play(CustomPlayer.Player)
                #UPGRADE WEAPON
                set CustomPlayer.WeaponStatLevel += 1.0
                IDX:=GetRandomInt(0,10)
                if(IDX <= 4):
                    UpgradeAmount := StatBaseUpgrade * Pow(StatBaseMultiplier, (CustomPlayer.WeaponTier-1)*1.0)
                    if(set CustomPlayer.WeaponStats[0] += UpgradeAmount) {}
                else if(IDX <= 8):
                    UpgradeAmount := StatPercUpgrade
                    if(set CustomPlayer.WeaponStats[1] += UpgradeAmount) {}
                else:
                    UpgradeAmount := StatPercUpgrade
                    if(set CustomPlayer.WeaponStats[2] += UpgradeAmount) {}
                SaveAllData(CustomPlayer)
            else:
                AudioManager.Denied.Play(CustomPlayer.Player)

            UpdateWUPHUD(CustomPlayer)


    OnBonusUpgrade(CustomPlayer:Custom_Player):void=
        if(NewBonusCost:= (EvolveCost[CustomPlayer.WeaponTier-1][1] * 2.0)):
            if(CustomPlayer.StoneAmount >= NewBonusCost):
                if(Analytic:=AudioManager.WeaponAnalytics[1]): #ANALYTIC
                    Analytic.Submit(CustomPlayer.Player)
                #TAKE AWAY RESOURCES
                CustomPlayer.Spend(NewBonusCost, "Stone", 0)
                #INCREASE COSTS
                AudioManager.LeveledUp.Play(CustomPlayer.Player)
                #UPGRADE WEAPON
                set CustomPlayer.WeaponBonusLevel += 1.0
                    #UpgradeAmount := BonusPerc + (BonusPerc*(CustomPlayer.WeaponTier-1))
                IDX:=GetRandomInt(0,10)
                if(IDX <= 4):
                    if(set CustomPlayer.WeaponBonuses[0] += BonusPerc) {}
                else if(IDX <= 8):
                    if(set CustomPlayer.WeaponBonuses[1] += BonusPerc) {}
                else:
                    if(set CustomPlayer.WeaponBonuses[2] += BonusPerc) {}
                SaveAllData(CustomPlayer)
            else:
                AudioManager.Denied.Play(CustomPlayer.Player)

        UpdateWUPHUD(CustomPlayer)

    OnPerkUpgrade(CustomPlayer:Custom_Player):void=
        if(NewEvolveCost:=EvolveCost[CustomPlayer.WeaponTier-1][1]):
            var NewPerkCost : float = NewEvolveCost * 10.0
            for(i:int=0..CustomPlayer.LockedPerks.Length):
                if(CustomPlayer.LockedPerks[i] = true):
                    set NewPerkCost *= 10.0
            var AreAllLocked : logic = true #Start by assuming all are locked
            if(CustomPlayer.PerksUnlocked = false):
                if(CustomPlayer.StoneAmount >= NewPerkCost):
                    if(Analytic:=AudioManager.WeaponAnalytics[2]): #ANALYTIC
                        Analytic.Submit(CustomPlayer.Player)
                    AudioManager.LeveledUp.Play(CustomPlayer.Player)
                    set CustomPlayer.PerksUnlocked = true
                    CustomPlayer.Spend(NewPerkCost, "Stone", 0)
                    RerollPerks(CustomPlayer)
                    SaveAllData(CustomPlayer)
            else:
                if(CustomPlayer.StoneAmount >= NewPerkCost):
                    if(Analytic:=AudioManager.WeaponAnalytics[2]): #ANALYTIC
                        Analytic.Submit(CustomPlayer.Player)
                    for(LockedPerk:CustomPlayer.LockedPerks):
                        if(LockedPerk = false):
                            set AreAllLocked = false #found one not locked, so turn false
                    if(AreAllLocked = false):
                        CustomPlayer.Spend(NewPerkCost, "Stone", 0)
                        AudioManager.LeveledUp.Play(CustomPlayer.Player)
                        if(CustomPlayer.WeaponPerkTier < 8): #CHANGE TO < 5 BUT LETS SEE
                            DoWeTierUp(CustomPlayer)
                        RerollPerks(CustomPlayer)
                        set CustomPlayer.WeaponPerkLevel += 1.0
                    else:
                        AudioManager.Denied.Play(CustomPlayer.Player)
                        #ALL ARE LOCKED
                    SaveAllData(CustomPlayer)
                else:
                    AudioManager.Denied.Play(CustomPlayer.Player)
        UpdateWUPHUD(CustomPlayer)

    DoWeTierUp(CustomPlayer:Custom_Player):void=
        IDX := GetRandomInt(1,100)
        var TierMultiplier : float = 0.5 + (0.25*CustomPlayer.WeaponPerkTier) + (0.5*CustomPlayer.WeaponTier) #CHANGE TO
        if(TierUpChance >= IDX):
            CustomPlayer.RemoveActivePerks() #Removes all active Perks
            set CustomPlayer.WeaponPerkTier += 1
            if:
                ActivePerk1:=CustomPlayer.ActivePerks[0],
                ActivePerk2:=CustomPlayer.ActivePerks[1],
                ActivePerk3:=CustomPlayer.ActivePerks[2]
            then:
                Perk1Value := ActivePerk1(1) * TierMultiplier
                Perk2Value := ActivePerk2(1) * TierMultiplier
                Perk3Value := ActivePerk3(1) * TierMultiplier
                if(set CustomPlayer.ActivePerks[0] = (ActivePerk1(0),Perk1Value)) {}
                if(set CustomPlayer.ActivePerks[1] = (ActivePerk2(0),Perk2Value)) {}
                if(set CustomPlayer.ActivePerks[2] = (ActivePerk3(0),Perk3Value)) {}
            CustomPlayer.AddActivePerks() #Adds all active Perks
            SaveAllData(CustomPlayer)

    RerollPerks(CustomPlayer:Custom_Player):void=
        var ActiveLockedNums : []int = {99, 99, 99}
        var TierMultiplier : float = 0.5 + (0.25*CustomPlayer.WeaponPerkTier) + (0.5*CustomPlayer.WeaponTier) #CHANGE TO
        if(ActivePerks:=array{CustomPlayer.ActivePerks[0],CustomPlayer.ActivePerks[1],CustomPlayer.ActivePerks[2]}):
            for(x:int=0..CustomPlayer.LockedPerks.Length):
                if(LockedPerk:=CustomPlayer.LockedPerks[x],ActivePerk:=ActivePerks[x]):
                    if(LockedPerk = true):
                        for(i:int=0..PerkNames.Length):
                            if(PerkNames[i] = ActivePerk(0)):
                                if(set ActiveLockedNums[x] = i) {}
            
        var Num1 : int  = GetRandomInt(0,PerkNames.Length-1)
        var Num2 : int  = GetRandomInt(0,PerkNames.Length-1)
        var Num3 : int  = GetRandomInt(0,PerkNames.Length-1)
        loop:
            if(Num1 <> Num2 and Num1 <> ActiveLockedNums[0] and Num1 <> ActiveLockedNums[1] and Num1 <> ActiveLockedNums[2]):
                if(Num1 <> Num3 and Num3 <> ActiveLockedNums[0] and Num3 <> ActiveLockedNums[1] and Num3 <> ActiveLockedNums[2]):
                    if(Num2 <> Num3 and Num2 <> ActiveLockedNums[0] and Num2 <> ActiveLockedNums[1] and Num2 <> ActiveLockedNums[2]):
                        break:
                    else:
                        set Num2 = GetRandomInt(0,PerkNames.Length-1)
                else:
                    set Num3 = GetRandomInt(0,PerkNames.Length-1)
            else:
                set Num1 = GetRandomInt(0,PerkNames.Length-1)
        if:
            NewPerkNames:=array{PerkNames[Num1],PerkNames[Num2],PerkNames[Num3]},
            NewPerkBases:=array{PerkBaseValues[Num1],PerkBaseValues[Num2],PerkBaseValues[Num3]},
            NewPerkValues:=array{NewPerkBases[0]*TierMultiplier,NewPerkBases[1]*TierMultiplier,NewPerkBases[2]*TierMultiplier}
        then:
            CustomPlayer.RemoveActivePerks() #Removes all active Perks
            for(i:int=0..NewPerkNames.Length-1):
                if(CustomPlayer.LockedPerks[i] = false,PerkName:=NewPerkNames[i],PerkValue:=NewPerkValues[i]):
                    if(set CustomPlayer.ActivePerks[i] = (PerkName,PerkValue)) {}
                else:
                    for(x:int=0..PerkNames.Length):
                        if(PerkNames[x] = CustomPlayer.ActivePerks[i](0),NewBase:=PerkBaseValues[x]*TierMultiplier):
                            if(set CustomPlayer.ActivePerks[i] = (CustomPlayer.ActivePerks[i](0),NewBase)) {}
                    
            CustomPlayer.AddActivePerks() #Adds all active Perks
            SaveAllData(CustomPlayer)
            gm.SetClass(CustomPlayer.Player,CustomPlayer) #Updates class, no harm if no change, updates if change

    LockPerk(Wdata:widget_message, whichPerk:int):void=
        if(CustomPlayer := gm.PlayersMap[Wdata.Player]):
            if(CustomPlayer.PerksUnlocked = true, PerkIcon:=CustomPlayer.PerkIcons[whichPerk]):
                if(CustomPlayer.LockedPerks[whichPerk] = true):
                    PerkIcon.SetImage(UI.UnlockedIcon)
                    if(set CustomPlayer.LockedPerks[whichPerk] = false) {}
                    #set CustomPlayer.PerkCost /= 10.0
                else:
                    PerkIcon.SetImage(UI.LockedIcon)
                    if(set CustomPlayer.LockedPerks[whichPerk] = true) {}
                    #set CustomPlayer.PerkCost *= 10.0
            UpdateWUPHUD(CustomPlayer)

    OnEvolveUpgrade(CustomPlayer:Custom_Player):void=
        if(CustomPlayer.WeaponTier < 6 or CustomPlayer.A4Unlocked = true and CustomPlayer.WeaponTier < 9):
            Print("CAN EVOLVE")
            if((CustomPlayer.WeaponStatLevel + CustomPlayer.WeaponBonusLevel + CustomPlayer.WeaponPerkLevel) >= EvolveCost[CustomPlayer.WeaponTier][0]):
                if(StoneCost:=EvolveCost[CustomPlayer.WeaponTier][1], WepMatCost:=EvolveCost[CustomPlayer.WeaponTier][2],ShardCost:=EvolveCost[CustomPlayer.WeaponTier][3]): #AWARD PLAYER EXP
                    if(CustomPlayer.ShardAmount >= ShardCost and CustomPlayer.StoneAmount >= StoneCost and CustomPlayer.WepMatAmount >= WepMatCost):
                        CustomPlayer.Spend(StoneCost, "Stone", 0)
                        CustomPlayer.Spend(WepMatCost, "WepMat", 0)
                        if(ShardCost > 0.0):
                            CustomPlayer.Spend(ShardCost, "Shard", 0)
                        Tutorial.UpdateTutorialUI(CustomPlayer, "EvolveWeapon", 50.0)
                        if(Accolade:=AudioManager.Accolades[4]): #AWARD PLAYER EXP
                            Accolade.Award(CustomPlayer.Player)
                        if(Analytic:=AudioManager.WeaponAnalytics[3]): #ANALYTIC
                            Analytic.Submit(CustomPlayer.Player)
                        
                        #INCREASE WEAPON TIER
                        AudioManager.Evolve.Play(CustomPlayer.Player)
                        set CustomPlayer.WeaponTier += 1
                
                        #RESET WEAPON UPGRADES
                        set CustomPlayer.WeaponStatLevel = 0.0
                        set CustomPlayer.WeaponBonusLevel = 0.0
                        for(i:int=0..CustomPlayer.WeaponStats.Length):
                            if(set CustomPlayer.WeaponStats[i] = 0.0) {}
                        for(i:int=0..CustomPlayer.WeaponBonuses.Length):
                            if(set CustomPlayer.WeaponBonuses[i] = 0.0) {}
                        
                        #RESET PERKS
                        set CustomPlayer.PerksUnlocked = false
                        CustomPlayer.RemoveActivePerks()
                        if(set CustomPlayer.ActivePerks[0] = ("Perk 1",0.0)) {}
                        if(set CustomPlayer.ActivePerks[1] = ("Perk 2",0.0)) {}
                        if(set CustomPlayer.ActivePerks[2] = ("Perk 3",0.0)) {}
                        set CustomPlayer.WeaponPerkTier = 0
                        set CustomPlayer.WeaponPerkLevel = 0.0
                        gm.SetClass(CustomPlayer.Player, CustomPlayer)
                        
                        SaveAllData(CustomPlayer)
                        #GIVE NEW GUN
                        spawn{GrantWeapon(CustomPlayer)}
                    else:
                        AudioManager.Denied.Play(CustomPlayer.Player)
                        #NOT ENOUGH RESOURCES
            else:
                AudioManager.Denied.Play(CustomPlayer.Player)
                #NOT HIGH ENOUGH LEVEL
        else:
            AudioManager.Denied.Play(CustomPlayer.Player)
            #MAX LEVEL

        UpdateWUPHUD(CustomPlayer)
        
    OnResetStats(CustomPlayer:Custom_Player):void=
        if(CustomPlayer.WepMatAmount >= ResetCost):
            CustomPlayer.Spend(ResetCost, "WepMat", 0)

            set CustomPlayer.WeaponStatLevel = 0.0
            set CustomPlayer.WeaponBonusLevel = 0.0

            if(set CustomPlayer.WeaponStats[0] = 0.0) {}
            if(set CustomPlayer.WeaponStats[1] = 0.0) {}
            if(set CustomPlayer.WeaponStats[2] = 0.0) {}

            if(set CustomPlayer.WeaponBonuses[0] = 0.0) {}
            if(set CustomPlayer.WeaponBonuses[1] = 0.0) {}
            if(set CustomPlayer.WeaponBonuses[2] = 0.0) {}

            AudioManager.LeveledUp.Play(CustomPlayer.Player)
            set CustomPlayer.CurrentPage = 0
            SaveAllData(CustomPlayer)
        else:
            AudioManager.Denied.Play(CustomPlayer.Player)

        UpdateWUPHUD(CustomPlayer)

    GrantWeapon(CustomPlayer:Custom_Player)<suspends>:void=
        var WhichWep : int = 0
        if(CustomPlayer.WeaponTier-1 <= 5):
            set WhichWep = CustomPlayer.WeaponTier-1
        else:
            set WhichWep = 5

        SignalRemoteGranter.GrantItem(CustomPlayer.Player)
        if(WeaponUnlock:=WeaponUnlockGranters[WhichWep]):
            Sleep(0.5)
            WeaponUnlock.GrantItem(CustomPlayer.Player)