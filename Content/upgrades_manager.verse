using { /Fortnite.com/Devices }
using { /Verse.org/Simulation }
using { /Verse.org/Colors }
using { /Verse.org/Verse }
using { /Verse.org/Random }
using { /Fortnite.com/Teams }
using { /UnrealEngine.com/Assets }
using { /Fortnite.com/Game }
using { /Fortnite.com/Characters }
using { /Fortnite.com/Playspaces }
using { /Fortnite.com/UI }
using { /UnrealEngine.com/Temporary/Diagnostics }
using { /UnrealEngine.com/Temporary/SpatialMath }
using { /UnrealEngine.com/Temporary/UI }
using { /Verse.org/Assets}


upgrades_manager := class(creative_device):

    @editable
    gm : game_manager = game_manager{}

    @editable
    AudioManager : audio_manager = audio_manager{}
    
    @editable
    Tutorial : tutorial_manager = tutorial_manager{}

    @editable
    DoubleJumpGranter : movement_modulator_device = movement_modulator_device{}
    @editable
    DoubleJumpTrigger : input_trigger_device = input_trigger_device{}

    @editable
    OpenGUPButtons : []button_device = array{}

    var Upgrade1Amounts : []float = array{0.1, 1.0}
    var Upgrade2Amounts : []float = array{0.10, 1.0}
    var Upgrade3Amounts : []float = array{25.0, 0.02}

    var MaxLevels1 : []int = array{10, 4}
    var MaxLevels2 : []int = array{15, 1}
    var MaxLevels3 : []int = array{20, 10}

    var GUP1CostMult : []float = array{450.0,1250.0} #THIS IS PURELY FOR THE HUD
    var GUP2CostMult : []float = array{250.0,1.0} #THIS IS PURELY FOR THE HUD
    var GUP3CostMult : []float = array{650.0,1500.0} #THIS IS PURELY FOR THE HUD
    
    var GUP1Costs : []float = array{2.50e05,2.34e19}
    var GUP2Costs : []float = array{2.50e07,1.41e13}
    var GUP3Costs : []float = array{7.50e10,3.12e17}
    
    var AllGUPCostMults : [][]float = array{}

    #MOVEMENT SPEED BUFFS
    var UpgradeTitles1 : []string = array{
        "Movement Speed",
        "Double Jump"
        }
    #AFK ZONE BUFFS
    var UpgradeTitles2 : []string = array{
        "Gather Amount",
        "Unlock Pet Modifier"
        }
    #FIGHTER BUFFS
    var UpgradeTitles3 : []string = array{
        "Max HP",
        "Mob Drop Rates"
        }
        
    var UpgradeDescs1 : []string = array{
        "Increase Movement Speed",
        "Jump Multiple Times"
        }
    var UpgradeDescs2 : []string = array{
        "Increase Gather Amount %\nin AFK Zone",
        "Unlock Pet Damage\nmodifier in AFK Zone"
        }
    var UpgradeDescs3 : []string = array{
        "Increase Max HP",
        "Increase Drop Rates\n from mobs"
        }

        var AllUpgradeTitles : [][]string = array{}
        var AllUpgradeDescs : [][]string = array{}
        var AllTextures : [][]texture = array{}
        var AllUpgradeAmounts : [][]float = array{}
        var AllMaxLevels : [][]int = array{}

    OnBegin<override>()<suspends>:void=

        for(i:int=0..OpenGUPButtons.Length):
            if(OpenGUPButton:=OpenGUPButtons[i]):
                OpenGUPButton.InteractedWithEvent.SubscribeAgent(OpenGUPHUD, i)

        DoubleJumpTrigger.PressedEvent.Subscribe(DoubleJump)

        for(Spawner : gm.PlayerSpawners):     
            Spawner.SpawnedEvent.Subscribe(OnPlayerJoined)

    OnPlayerJoined(Agent:agent):void=
        spawn{ApplyMovementBuff(Agent)}

    ApplyMovementBuff(Agent:agent)<suspends>:void=
        Sleep(1.0)
        if(CustomPlayer := gm.PlayersMap[player[Agent]]):

            if(CustomPlayer.PlayTimeSessionSeconds = 0.0):
                CustomPlayer.GUPCloseButton.OnClick().Subscribe(RemoveGUPHUD)
    
                set AllUpgradeTitles = array{UpgradeTitles1,UpgradeTitles2,UpgradeTitles3}
                set AllUpgradeDescs = array{UpgradeDescs1,UpgradeDescs2,UpgradeDescs3}
                set AllTextures = array{Textures1,Textures2,Textures3}
                set AllUpgradeAmounts = array{Upgrade1Amounts,Upgrade2Amounts,Upgrade3Amounts}
                set AllMaxLevels = array{MaxLevels1,MaxLevels2,MaxLevels3}
                set AllGUPCostMults = array{GUP1CostMult,GUP2CostMult,GUP3CostMult}
                set CustomPlayer.AllGUPTexts = array{CustomPlayer.StoneAmountText,CustomPlayer.Upgrade1Title,CustomPlayer.Upgrade1Level,CustomPlayer.Upgrade1Desc,CustomPlayer.Upgrade1Bonus,CustomPlayer.Upgrade1NextBonus,CustomPlayer.Upgrade1CostText,CustomPlayer.Upgrade1CostAmount,CustomPlayer.Upgrade2Title,CustomPlayer.Upgrade2Level,CustomPlayer.Upgrade2Desc,CustomPlayer.Upgrade2Bonus,CustomPlayer.Upgrade2NextBonus,CustomPlayer.Upgrade2CostText,CustomPlayer.Upgrade2CostAmount}
    
                set CustomPlayer.GUPStoneIcons = array{CustomPlayer.StoneIcon1,CustomPlayer.StoneIcon2}
                set CustomPlayer.GUPUpgradeTitlesBlock = array{CustomPlayer.Upgrade1Title,CustomPlayer.Upgrade2Title}
                set CustomPlayer.GUPUpgradeLevelsBlock = array{CustomPlayer.Upgrade1Level,CustomPlayer.Upgrade2Level}
                set CustomPlayer.GUPUpgradeIconsBlock = array{CustomPlayer.Upgrade1Icon,CustomPlayer.Upgrade2Icon}
                set CustomPlayer.GUPUpgradeDescTextsBlock = array{CustomPlayer.Upgrade1Desc,CustomPlayer.Upgrade2Desc}
                set CustomPlayer.GUPUpgradeBonusesBlock = array{CustomPlayer.Upgrade1Bonus,CustomPlayer.Upgrade2Bonus}
                set CustomPlayer.GUPUpgradeNextBonusesBlock = array{CustomPlayer.Upgrade1NextBonus,CustomPlayer.Upgrade2NextBonus}
                set CustomPlayer.GUPUpgradeCostsBlock = array{CustomPlayer.Upgrade1CostAmount,CustomPlayer.Upgrade2CostAmount}
                set CustomPlayer.GUPUpgradeCostTextsBlock = array{CustomPlayer.Upgrade1CostText,CustomPlayer.Upgrade2CostText}
                set CustomPlayer.GUPUpgradeButtonsBlock = array{CustomPlayer.GUPUpgradeButton1,CustomPlayer.GUPUpgradeButton2}
    
                for(i:int=0..CustomPlayer.GUPUpgradeButtonsBlock.Length):
                    if(UpgradeButton:=CustomPlayer.GUPUpgradeButtonsBlock[i]):
                        UpgradeButton.OnClick().SubscribeAgent3(OnUpgradeButton, i)

            gm.ApplyMovementBuff(CustomPlayer)

#UI BLOCKS
    var Textures1 :[]texture = array{UI.movementspeed,UI.DoubleJump}
    var Textures2 :[]texture = array{UI.AFKZone,UI.AFKZonePets}
    var Textures3 :[]texture = array{UI.HP,UI.MobRewards}
        
#CANVAS
    MakeGUPCanvas(CustomPlayer:Custom_Player):canvas=
        CustomPlayer.GUPCloseButton.SetText(StringToMessage("X"))
        CustomPlayer.Upgrade1Desc.SetJustification(text_justification.Center)
        CustomPlayer.Upgrade2Desc.SetJustification(text_justification.Center)
        CustomPlayer.Upgrade3Desc.SetJustification(text_justification.Center)
        CustomPlayer.Upgrade4Desc.SetJustification(text_justification.Center)

        for(TextBlock : CustomPlayer.AllGUPTexts):
            TextBlock.SetShadowColor(CustomPlayer.ShadowDefaults.DefaultShadowColor)
            TextBlock.SetShadowOffset(CustomPlayer.ShadowDefaults.DefaultShadowOffset)
            TextBlock.SetShadowOpacity(1.0)

        var MyGUPCanvas: canvas = canvas:
            Slots := array:
                canvas_slot:
                    Widget:=CustomPlayer.GUPHUDBG
                    ZOrder:=0
                    SizeToContent:=false
                    Alignment := vector2{X := 0.5, Y := 0.0}
                    Anchors := anchors{Minimum := vector2{X:= 0.5, Y:= 0.0}, Maximum := vector2{X:= 0.5, Y:= 0.0}}
                    Offsets:=margin{Top:=0.0,Left:=0.0,Bottom:=1076.0,Right:=2048.0}
                canvas_slot:
                    Widget:=CustomPlayer.GUPCloseButton
                    ZOrder:=9
                    SizeToContent:=false
                    Alignment := vector2{X := 0.5, Y := 0.0}
                    Anchors := anchors{Minimum := vector2{X:= 0.5, Y:= 0.0}, Maximum := vector2{X:= 0.5, Y:= 0.0}}
                    Offsets:=margin{Top:=116.0,Left:=368.0,Bottom:=CustomPlayer.GUPClose_Size.Y,Right:=CustomPlayer.GUPClose_Size.X}
                
                canvas_slot:
                    Widget:=CustomPlayer.GUPHUDTitle
                    ZOrder:=1
                    SizeToContent:=false
                    Alignment := vector2{X := 0.0, Y := 0.0}
                    Anchors := anchors{Minimum := vector2{X:= 0.5, Y:= 0.0}, Maximum := vector2{X:= 0.5, Y:= 0.0}}
                    Offsets:=margin{Top:=144.0,Left:=-176.0, Bottom:=56.0, Right:=336.0}

                canvas_slot:
                    Widget:=CustomPlayer.StoneAmountIcon
                    ZOrder:=1
                    SizeToContent:=false
                    Anchors := anchors{Minimum := vector2{X:= 0.5, Y:= 0.0}, Maximum := vector2{X:= 0.5, Y:= 0.0}}
                    Offsets:=margin{Top:=162.0,Left:=-788.0, Bottom:=40.0, Right:=40.0}
                    Alignment := vector2{X := 0.5, Y := 0.0}
                canvas_slot:
                    Widget:=CustomPlayer.StoneAmountText
                    ZOrder:=1
                    SizeToContent:=false
                    Anchors := anchors{Minimum := vector2{X:= 0.5, Y:= 0.0}, Maximum := vector2{X:= 0.5, Y:= 0.0}}
                    Offsets:=margin{Top:=162.0,Left:=-746.0}
                    Alignment := vector2{X := 0.5, Y := 0.0}

                canvas_slot:
                    Widget:=CustomPlayer.Upgrade1Title
                    ZOrder:=1
                    SizeToContent:=true
                    Anchors := anchors{Minimum := vector2{X:= 0.5, Y:= 0.0}, Maximum := vector2{X:= 0.5, Y:= 0.0}}
                    Offsets:=margin{Top:=280.0,Left:=-348.0}
                    Alignment := vector2{X:= 0.5, Y:= 0.0}
                canvas_slot:
                    Widget:=CustomPlayer.Upgrade1Level
                    ZOrder:=1
                    SizeToContent:=true
                    Anchors := anchors{Minimum := vector2{X:= 0.5, Y:= 0.0}, Maximum := vector2{X:= 0.5, Y:= 0.0}}
                    Offsets:=margin{Top:=341.0,Left:=-348.0}
                    Alignment := vector2{X:= 0.5, Y:= 0.0}
                canvas_slot:
                    Widget:=CustomPlayer.Upgrade1Icon
                    ZOrder:=1
                    SizeToContent:=false
                    Anchors := anchors{Minimum := vector2{X:= 0.5, Y:= 0.0}, Maximum := vector2{X:= 0.5, Y:= 0.0}}
                    Offsets:=margin{Top:=421.0,Left:=-348.0, Bottom:=140.0,Right:=140.0}
                    Alignment := vector2{X:= 0.5, Y:= 0.0}
                canvas_slot:
                    Widget:=CustomPlayer.Upgrade1Desc
                    ZOrder:=1
                    SizeToContent:=true
                    Anchors := anchors{Minimum := vector2{X:= 0.5, Y:= 0.0}, Maximum := vector2{X:= 0.5, Y:= 0.0}}
                    Offsets:=margin{Top:=592.0,Left:=-348.0}
                    Alignment := vector2{X:= 0.5, Y:= 0.0}
                canvas_slot:
                    ZOrder:=1
                    SizeToContent:=true
                    Anchors := anchors{Minimum := vector2{X:= 0.5, Y:= 0.0}, Maximum := vector2{X:= 0.5, Y:= 0.0}}
                    Offsets:=margin{Top:=740.0,Left:=-348.0}
                    Alignment := vector2{X:= 0.5, Y:= 0.0}
                    Widget := stack_box:
                        Orientation := orientation.Horizontal
                        Slots := array:
                            stack_box_slot:
                                Widget := CustomPlayer.Upgrade1Bonus   
                            stack_box_slot:
                                Widget := CustomPlayer.Upgrade1NextBonus
                canvas_slot:
                    Widget:=CustomPlayer.Upgrade1CostText
                    ZOrder:=1
                    SizeToContent:=true
                    Anchors := anchors{Minimum := vector2{X:= 0.5, Y:= 0.0}, Maximum := vector2{X:= 0.5, Y:= 0.0}}
                    Offsets:=margin{Top:=800.0,Left:=-448.0}
                    Alignment := vector2{X:= 0.5, Y:= 0.0}
                canvas_slot:
                    ZOrder:=1
                    SizeToContent:=true
                    Anchors := anchors{Minimum := vector2{X:= 0.5, Y:= 0.0}, Maximum := vector2{X:= 0.5, Y:= 0.0}}
                    Offsets:=margin{Top:=800.0,Left:=-188.0}
                    Alignment := vector2{X:= 1.0, Y:= 0.0}
                    Widget := stack_box:
                        Orientation := orientation.Horizontal
                        Slots := array:
                            stack_box_slot:
                                Widget := CustomPlayer.StoneIcon1    
                            stack_box_slot:
                                Widget := CustomPlayer.Upgrade1CostAmount
                canvas_slot:
                    Widget:=CustomPlayer.GUPUpgradeButton1
                    ZOrder:=3
                    SizeToContent:=true
                    Anchors := anchors{Minimum := vector2{X:= 0.5, Y:= 0.0}, Maximum := vector2{X:= 0.5, Y:= 0.0}}
                    Offsets:=margin{Top:=870.0,Left:=-348.0, Bottom:=60.0, Right:=220.0}
                    Alignment := vector2{X:= 0.5, Y:= 0.0}
                    
                canvas_slot:
                    Widget:=CustomPlayer.Upgrade2Title
                    ZOrder:=1
                    SizeToContent:=true
                    Anchors := anchors{Minimum := vector2{X:= 0.5, Y:= 0.0}, Maximum := vector2{X:= 0.5, Y:= 0.0}}
                    Offsets:=margin{Top:=282.0,Left:=336.0}
                    Alignment := vector2{X:= 0.5, Y:= 0.0}
                canvas_slot:
                    Widget:=CustomPlayer.Upgrade2Level
                    ZOrder:=1
                    SizeToContent:=true
                    Anchors := anchors{Minimum := vector2{X:= 0.5, Y:= 0.0}, Maximum := vector2{X:= 0.5, Y:= 0.0}}
                    Offsets:=margin{Top:=341.0,Left:=336.0}
                    Alignment := vector2{X:= 0.5, Y:= 0.0}
                canvas_slot:
                    Widget:=CustomPlayer.Upgrade2Icon
                    ZOrder:=1
                    SizeToContent:=false
                    Anchors := anchors{Minimum := vector2{X:= 0.5, Y:= 0.0}, Maximum := vector2{X:= 0.5, Y:= 0.0}}
                    Offsets:=margin{Top:=421.0,Left:=336.0, Bottom:=140.0,Right:=140.0}
                    Alignment := vector2{X:= 0.5, Y:= 0.0}
                canvas_slot:
                    Widget:=CustomPlayer.Upgrade2Desc
                    ZOrder:=1
                    SizeToContent:=true
                    Anchors := anchors{Minimum := vector2{X:= 0.5, Y:= 0.0}, Maximum := vector2{X:= 0.5, Y:= 0.0}}
                    Offsets:=margin{Top:=592.0,Left:=336.0}
                    Alignment := vector2{X:= 0.5, Y:= 0.0}
                canvas_slot:
                    ZOrder:=1
                    SizeToContent:=true
                    Anchors := anchors{Minimum := vector2{X:= 0.5, Y:= 0.0}, Maximum := vector2{X:= 0.5, Y:= 0.0}}
                    Offsets:=margin{Top:=740.0,Left:=336.0}
                    Alignment := vector2{X:= 0.5, Y:= 0.0}
                    Widget := stack_box:
                        Orientation := orientation.Horizontal
                        Slots := array:
                            stack_box_slot:
                                Widget := CustomPlayer.Upgrade2Bonus   
                            stack_box_slot:
                                Widget := CustomPlayer.Upgrade2NextBonus
                canvas_slot:
                    Widget:=CustomPlayer.Upgrade2CostText
                    ZOrder:=1
                    SizeToContent:=true
                    Anchors := anchors{Minimum := vector2{X:= 0.5, Y:= 0.0}, Maximum := vector2{X:= 0.5, Y:= 0.0}}
                    Offsets:=margin{Top:=800.0,Left:=240.0}
                    Alignment := vector2{X:= 0.5, Y:= 0.0}
                canvas_slot:
                    ZOrder:=1
                    SizeToContent:=true
                    Anchors := anchors{Minimum := vector2{X:= 0.5, Y:= 0.0}, Maximum := vector2{X:= 0.5, Y:= 0.0}}
                    Offsets:=margin{Top:=800.0,Left:=500.0}
                    Alignment := vector2{X:= 1.0, Y:= 0.0}
                    Widget := stack_box:
                        Orientation := orientation.Horizontal
                        Slots := array:
                            stack_box_slot:
                                Widget := CustomPlayer.StoneIcon2   
                            stack_box_slot:
                                Widget := CustomPlayer.Upgrade2CostAmount
                canvas_slot:
                    Widget:=CustomPlayer.GUPUpgradeButton2
                    ZOrder:=1
                    SizeToContent:=true
                    Anchors := anchors{Minimum := vector2{X:= 0.5, Y:= 0.0}, Maximum := vector2{X:= 0.5, Y:= 0.0}}
                    Offsets:=margin{Top:=870.0,Left:=336.0, Bottom:=60.0, Right:=220.0}
                    Alignment := vector2{X:= 0.5, Y:= 0.0}
                    
        return MyGUPCanvas

#CODE
    OpenGUPHUD(Agent:agent, which:int): void=
        if(CustomPlayer := gm.PlayersMap[Agent]):
            if(MyUI:=CustomPlayer.UIMaps["GUP"]?):
            else:
                var MyGUPCanvas : canvas = MakeGUPCanvas(CustomPlayer)
                set CustomPlayer.CurrentGUP = which
                CustomPlayer.AddFullHUD(MyGUPCanvas)
                if(set CustomPlayer.UIMaps["GUP"] = option{MyGUPCanvas}) {}
                UpdateGUPHUD(CustomPlayer)

    RemoveGUPHUD(Wdata:widget_message):void=
        if(CustomPlayer := gm.PlayersMap[Wdata.Player]):
            if(Player:=Wdata.Player, PlayerUI:=GetPlayerUI[Player],SelectedButton:=Wdata.Source):
                if(MyUI:=CustomPlayer.UIMaps["GUP"]?):
                    PlayerUI.RemoveWidget(MyUI)
                    MyUI.SetVisibility(widget_visibility.Collapsed)
                    if(set CustomPlayer.UIMaps["GUP"] = false){}

    UpdateGUPHUD(CustomPlayer: Custom_Player):void=
        set CustomPlayer.AllLevels = array{CustomPlayer.GUP1Levels,CustomPlayer.GUP2Levels,CustomPlayer.GUP3Levels}
        set CustomPlayer.AllCosts = array{GUP1Costs,GUP2Costs,GUP3Costs}
        if(MaxLevels:=AllMaxLevels[CustomPlayer.CurrentGUP],UpgradeAmounts:=AllUpgradeAmounts[CustomPlayer.CurrentGUP],Costs:=CustomPlayer.AllCosts[CustomPlayer.CurrentGUP],Levels:=CustomPlayer.AllLevels[CustomPlayer.CurrentGUP],UpgradeTitles:=AllUpgradeTitles[CustomPlayer.CurrentGUP],UpgradeDescs:=AllUpgradeDescs[CustomPlayer.CurrentGUP],Textures:=AllTextures[CustomPlayer.CurrentGUP]):
            
            CustomPlayer.StoneAmountText.SetText(StringToMessage(NumberCondensor(CustomPlayer.StoneAmount)))

            for(i:int=0..CustomPlayer.GUPUpgradeTitlesBlock.Length):
                if(StoneIcon:=CustomPlayer.StoneIcons[i],UpgradeNextBonus:=CustomPlayer.GUPUpgradeNextBonusesBlock[i],UpgradeCostText:=CustomPlayer.GUPUpgradeCostTextsBlock[i],MaxLevel:=MaxLevels[i],Level:=Levels[i],UpgradeAmount:=UpgradeAmounts[i],Upgrade:=UpgradeAmount*Level,GUPCostMult:=AllGUPCostMults[CustomPlayer.CurrentGUP],Cost:=Costs[i]*(Pow(GUPCostMult[i], Level*1.0)),UpgradeTitle:=UpgradeTitles[i],UpgradeDesc:=UpgradeDescs[i],UpgradeIcon:=Textures[i],UpgradeTitleBlock:=CustomPlayer.GUPUpgradeTitlesBlock[i],UpgradeLevel:=CustomPlayer.GUPUpgradeLevelsBlock[i],UpgradeIconBlock:=CustomPlayer.GUPUpgradeIconsBlock[i],UpgradeDescBlock:=CustomPlayer.GUPUpgradeDescTextsBlock[i],UpgradeBonus:=CustomPlayer.GUPUpgradeBonusesBlock[i],UpgradeCost:=CustomPlayer.GUPUpgradeCostsBlock[i],UpgradeButton:=CustomPlayer.GUPUpgradeButtonsBlock[i]):
                    UpgradeTitleBlock.SetText(StringToMessage(UpgradeTitle))
                    UpgradeLevel.SetText(StringToMessage("Lv. {Level}"))
                    UpgradeIconBlock.SetImage(UpgradeIcon)
                    UpgradeDescBlock.SetText(StringToMessage(UpgradeDesc))
                    if(CustomPlayer.StoneAmount >= Cost):
                        UpgradeCost.SetTextColor(MakeColorFromHex("39FF39"))
                    else:
                        UpgradeCost.SetTextColor(MakeColorFromHex("fa4d56"))

                    UpgradeCost.SetText(StringToMessage(" {NumberCondensor(Cost)}"))
                    var isPerc : logic = true

                    if(CustomPlayer.CurrentGUP = 0):
                        if(i = 1):
                            set isPerc = false
                    else if(CustomPlayer.CurrentGUP = 1):
                        if(i = 1):
                            set isPerc = false
                    else if(CustomPlayer.CurrentGUP = 2):
                        if(i = 0):
                            set isPerc = false
                            
                    if(isPerc = true):
                        UpgradeBonus.SetText(StringToMessage("{NumberCondensor(Upgrade*100.0)}% "))
                    else:
                        UpgradeBonus.SetText(StringToMessage("{NumberCondensor(Upgrade)} "))

                    if(Level < MaxLevel):
                        UpgradeButton.SetText(StringToMessage("UPGRADE"))
                        StoneIcon.SetVisibility(widget_visibility.Visible)
                        UpgradeCostText.SetText(StringToMessage("COST:"))
                        UpgradeCostText.SetTextColor(MakeColorFromHex("DDDDDD"))
                        if(isPerc = true):
                            UpgradeNextBonus.SetText(StringToMessage("(+{NumberCondensor(UpgradeAmount*100.0)}%)"))
                        else:
                            UpgradeNextBonus.SetText(StringToMessage("(+{NumberCondensor(UpgradeAmount)})"))
                    else:
                        UpgradeButton.SetText(StringToMessage("MAX"))
                        StoneIcon.SetVisibility(widget_visibility.Collapsed)
                        UpgradeCostText.SetText(StringToMessage(""))
                        UpgradeCost.SetText(StringToMessage(""))
                        UpgradeNextBonus.SetText(StringToMessage(""))

    OnUpgradeButton(Wdata:widget_message, WhichUpgrade:int):void=
        if(CustomPlayer := gm.PlayersMap[Wdata.Player]):
            if(CustomPlayer.CurrentGUP = 0): #MOVEMENT SPEED
                PurchaseMovementUpgrade(Wdata.Player, WhichUpgrade)
            if(CustomPlayer.CurrentGUP = 1): #AFK ZONE
                PurchaseAFKZoneUpgrade(Wdata.Player, WhichUpgrade)
            if(CustomPlayer.CurrentGUP = 2): #FIGHTER BUFFS
                PurchaseFighterUpgrade(Wdata.Player, WhichUpgrade)

            
    PurchaseMovementUpgrade(Agent:agent, WhichUpgrade:int):void=
        if:
            CustomPlayer := gm.PlayersMap[player[Agent]]
            MaxLevels:=AllMaxLevels[CustomPlayer.CurrentGUP]
            UpgradeAmounts:=AllUpgradeAmounts[CustomPlayer.CurrentGUP]
            Costs:=CustomPlayer.AllCosts[CustomPlayer.CurrentGUP]
            GUPCostMult:=AllGUPCostMults[CustomPlayer.CurrentGUP]
            Levels:=CustomPlayer.AllLevels[CustomPlayer.CurrentGUP]
            Cost:=Costs[WhichUpgrade]*(Pow(GUPCostMult[WhichUpgrade], Levels[WhichUpgrade]*1.0))
        then:
            if(Levels[WhichUpgrade] < MaxLevels[WhichUpgrade]):
                Tutorial.UpdateTutorialUI(CustomPlayer, "MovementSpeed", 250000.0)
                if(CustomPlayer.StoneAmount >= Cost):
                    if(Accolade:=AudioManager.Accolades[11]): #AWARD PLAYER EXP
                        Accolade.Award(CustomPlayer.Player)
                    AudioManager.LeveledUp.Play(CustomPlayer.Player)
                    CustomPlayer.Spend(Cost, "Stone", 0)
                    if(WhichUpgrade = 0):
                        set CustomPlayer.MovementSpeedLevel += 1
                        gm.ApplyMovementBuff(CustomPlayer)
                    else if(WhichUpgrade = 1):
                        set CustomPlayer.DoubleJumpUnlocked = true
                        set CustomPlayer.DoubleJumpAmount += 1
                    if(set CustomPlayer.GUP1Levels[WhichUpgrade] += 1){} #Up level
                    SaveAllData(CustomPlayer)
                else:
                    AudioManager.Denied.Play(CustomPlayer.Player)
            else:
                AudioManager.Denied.Play(CustomPlayer.Player)
            UpdateGUPHUD(CustomPlayer)

    PurchaseAFKZoneUpgrade(Agent:agent, WhichUpgrade:int):void=
        if(CustomPlayer := gm.PlayersMap[player[Agent]],MaxLevels:=AllMaxLevels[CustomPlayer.CurrentGUP],UpgradeAmounts:=AllUpgradeAmounts[CustomPlayer.CurrentGUP],Costs:=CustomPlayer.AllCosts[CustomPlayer.CurrentGUP],GUPCostMult:=AllGUPCostMults[CustomPlayer.CurrentGUP],Levels:=CustomPlayer.AllLevels[CustomPlayer.CurrentGUP],Cost:=Costs[WhichUpgrade]*(Pow(GUPCostMult[WhichUpgrade], Levels[WhichUpgrade]*1.0))):
            if(Levels[WhichUpgrade] < MaxLevels[WhichUpgrade]):
                if(CustomPlayer.StoneAmount >= Cost):
                    if(Accolade:=AudioManager.Accolades[11]): #AWARD PLAYER EXP
                        Accolade.Award(CustomPlayer.Player)
                    AudioManager.LeveledUp.Play(CustomPlayer.Player)
                    if(UpgradeAmount:=UpgradeAmounts[WhichUpgrade]):
                        CustomPlayer.Spend(Cost, "Stone", 0)
                        if(WhichUpgrade = 0):
                            set CustomPlayer.AFKStoneBoost += UpgradeAmount
                        else if(WhichUpgrade = 1):
                            set CustomPlayer.AFKPetsUnlocked = true

                    #if(set CustomPlayer.GUP2Upgrades[WhichUpgrade] += UpgradeAmounts[WhichUpgrade]) {} #Up the amount the hud has given
                    if(set CustomPlayer.GUP2Levels[WhichUpgrade] += 1){} #Up level
                    #if(set CustomPlayer.GUP2Costs[WhichUpgrade] *= GUPCostMult[WhichUpgrade]){} #Up Cost
                    SaveAllData(CustomPlayer)
                else:
                    AudioManager.Denied.Play(CustomPlayer.Player)
            else:
                AudioManager.Denied.Play(CustomPlayer.Player)
            UpdateGUPHUD(CustomPlayer)

    PurchaseFighterUpgrade(Agent:agent, WhichUpgrade:int):void=
        if(CustomPlayer := gm.PlayersMap[player[Agent]],MaxLevels:=AllMaxLevels[CustomPlayer.CurrentGUP],UpgradeAmounts:=AllUpgradeAmounts[CustomPlayer.CurrentGUP],Costs:=CustomPlayer.AllCosts[CustomPlayer.CurrentGUP],GUPCostMult:=AllGUPCostMults[CustomPlayer.CurrentGUP],Levels:=CustomPlayer.AllLevels[CustomPlayer.CurrentGUP],Cost:=Costs[WhichUpgrade]*(Pow(GUPCostMult[WhichUpgrade], Levels[WhichUpgrade]*1.0))):
            if(Levels[WhichUpgrade] < MaxLevels[WhichUpgrade]):
                if(CustomPlayer.StoneAmount >= Cost):
                    if(Accolade:=AudioManager.Accolades[11]): #AWARD PLAYER EXP
                        Accolade.Award(CustomPlayer.Player)
                    AudioManager.LeveledUp.Play(CustomPlayer.Player)
                    if(UpgradeAmount:=UpgradeAmounts[WhichUpgrade]):
                        CustomPlayer.Spend(Cost, "Stone", 0)
                        if(WhichUpgrade = 0):
                            set CustomPlayer.MaxHP += UpgradeAmount
                            gm.SetHP(CustomPlayer)
                        else if(WhichUpgrade = 1):
                            set CustomPlayer.WepMatDropRate += UpgradeAmount

                    #if(set CustomPlayer.GUP3Upgrades[WhichUpgrade] += UpgradeAmounts[WhichUpgrade]) {} #Up the amount the hud has given
                    if(set CustomPlayer.GUP3Levels[WhichUpgrade] += 1){} #Up level
                    #if(set CustomPlayer.GUP3Costs[WhichUpgrade] *= GUPCostMult[WhichUpgrade]){} #Up Cost
                    SaveAllData(CustomPlayer)
                else:
                    AudioManager.Denied.Play(CustomPlayer.Player)
            else:
                AudioManager.Denied.Play(CustomPlayer.Player)
            UpdateGUPHUD(CustomPlayer)

    DoubleJump(Agent:agent):void=
        if(CustomPlayer := gm.PlayersMap[player[Agent]]):
            if(CustomPlayer.DoubleJumpUnlocked = true):
                if(Fort :=CustomPlayer.Player.GetFortCharacter[]):
                    if(Fort.IsOnGround[]):
                        set CustomPlayer.DoubleJumpCount = 0
                    if(Fort.IsInAir[]):
                        if(CustomPlayer.DoubleJumpCount < CustomPlayer.DoubleJumpAmount):
                            set CustomPlayer.DoubleJumpCount += 1
                            DoubleJumpGranter.Activate(CustomPlayer.Player)
                        else:
