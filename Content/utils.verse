using { /Fortnite.com/Devices }
using { /Verse.org/Simulation }
using { /UnrealEngine.com/Temporary/Diagnostics }
using { /UnrealEngine.com/Temporary/UI }
using { /Fortnite.com/Game }
using { /Fortnite.com/Characters }
using { /Verse.org/Random }

(Listenable : listenable(agent)).SubscribeAgent(OutputFunc : tuple(agent, t)->void, ExtraData : t where t:type) : cancelable =
    Wrapper := wrapper_agent(t){ExtraData := ExtraData, OutputFunc := OutputFunc}
    Listenable.Subscribe(Wrapper.InputFunc)

(Listenable : listenable(?agent)).SubscribeAgent2(OutputFunc : tuple(?agent, t)->void, ExtraData : t where t:type) : cancelable =
    Wrapper := wrapper_agent2(t){ExtraData := ExtraData, OutputFunc := OutputFunc}
    Listenable.Subscribe(Wrapper.InputFunc)

(Listenable : listenable(widget_message)).SubscribeAgent3(OutputFunc : tuple(widget_message, t)->void, ExtraData : t where t:type) : cancelable =
    Wrapper := wrapper_agent3(t){ExtraData := ExtraData, OutputFunc := OutputFunc}
    Listenable.Subscribe(Wrapper.InputFunc)

(Listenable : listenable(device_ai_interaction_result)).SubscribeAgent4(OutputFunc : tuple(device_ai_interaction_result, t)->void, ExtraData : t where t:type) : cancelable =
    Wrapper := wrapper_agent4(t){ExtraData := ExtraData, OutputFunc := OutputFunc}
    Listenable.Subscribe(Wrapper.InputFunc)

(Listenable : listenable(damage_result)).SubscribeAgent5(OutputFunc : tuple(damage_result, t)->void, ExtraData : t where t:type) : cancelable =
    Wrapper := wrapper_agent5(t){ExtraData := ExtraData, OutputFunc := OutputFunc}
    Listenable.Subscribe(Wrapper.InputFunc)

(Listenable : listenable(fort_character)).SubscribeAgent6(OutputFunc : tuple(fort_character, t)->void, ExtraData : t where t:type) : cancelable =
    Wrapper := wrapper_agent6(t){ExtraData := ExtraData, OutputFunc := OutputFunc}
    Listenable.Subscribe(Wrapper.InputFunc)

(Listenable : listenable(tuple(agent, int))).SubscribeAgent7(OutputFunc : tuple(agent,int,t)->void, ExtraData : t where t:type) : cancelable =
    Wrapper := wrapper_agent7(t){ExtraData := ExtraData, OutputFunc := OutputFunc}
    Listenable.Subscribe(Wrapper.InputFunc)

(Listenable : listenable(elimination_result)).SubscribeAgentElim(OutputFunc : tuple(elimination_result, t)->void, ExtraData : t where t:type) : cancelable =
    Wrapper := wrapper_agentElim(t){ExtraData := ExtraData, OutputFunc := OutputFunc}
    Listenable.Subscribe(Wrapper.InputFunc)
 
wrapper_agent(t : type) := class():
    ExtraData : t;
    OutputFunc : tuple(agent, t) -> void
    InputFunc(Agent : agent):void = OutputFunc(Agent, ExtraData)

wrapper_agent2(t : type) := class():
    ExtraData : t;
    OutputFunc : tuple(?agent, t) -> void
    InputFunc(Agent : ?agent):void = OutputFunc(Agent, ExtraData)

wrapper_agent3(t : type) := class():
    ExtraData : t;
    OutputFunc : tuple(widget_message, t) -> void
    InputFunc(Wdata : widget_message):void = OutputFunc(Wdata, ExtraData)

wrapper_agent4(t : type) := class():
    ExtraData : t;
    OutputFunc : tuple(device_ai_interaction_result, t) -> void
    InputFunc(Wdata : device_ai_interaction_result):void = OutputFunc(Wdata, ExtraData)

wrapper_agent5(t : type) := class():
    ExtraData : t;
    OutputFunc : tuple(damage_result, t) -> void
    InputFunc(Wdata : damage_result):void = OutputFunc(Wdata, ExtraData)

wrapper_agent6(t : type) := class():
    ExtraData : t;
    OutputFunc : tuple(fort_character, t) -> void
    InputFunc(Wdata : fort_character):void = OutputFunc(Wdata, ExtraData)

wrapper_agentElim(t : type) := class():
    ExtraData : t;
    OutputFunc : tuple(elimination_result, t) -> void
    InputFunc(Wdata : elimination_result):void = OutputFunc(Wdata, ExtraData)

wrapper_agent7(t: type) := class():
    ExtraData : t;
    OutputFunc : tuple(agent, int,t) -> void
    InputFunc(Agent:agent, Index:int):void = OutputFunc(Agent,Index,ExtraData)
 
(Listenable : listenable(tuple())).SubscribeEmpty(OutputFunc : t -> void, ExtraData : t where t:type) : cancelable =
    Wrapper := wrapper_empty(t) {ExtraData := ExtraData, OutputFunc := OutputFunc}
    Listenable.Subscribe(Wrapper.InputFunc) 
    
 
wrapper_empty(t : type) := class():
    ExtraData : t;
    OutputFunc : t -> void
    InputFunc():void = OutputFunc(ExtraData)



StringToMessage<localizes>(value:string) : message = "{value}"
PlayerToMessage<localizes>(Agent:agent):message = "{Agent}"


StringToInt(IntInString: []char): int=
        var Value: int = -1
        if(IntInString <> ""):
            loop:
                set Value += 1
                if (ToString(Value) = IntInString):
                    return Value
        else:
            return Value

DecimalToBinary(DecimalNumber:float):string=
    if(DecimalNumber = 0):
        return "0"
    var binary_str : string = ""

    if(var Num :int= Floor[DecimalNumber]):
        loop:
            if(Num > 0):
                if(remainder := Mod[Num, 2]):
                    set binary_str = ToString(remainder) + binary_str
                if(newNumber :=Floor(Num / 2)):
                    set Num = newNumber
            else:
                break:

        return binary_str
    return binary_str

(Widget:widget).ForceRefresh<public>():void=
    Widget.SetVisibility(widget_visibility.Collapsed)
    Widget.SetVisibility(widget_visibility.Visible)

NumberCondensor(Amount:float, ?WhichSuffix:int=0):string=
    var suffixes : []tuple(float, string, string) = array{
        (1.00e300, "NNn", "e300"),
        (1.00e297, "ONn", "e297"),
        (1.00e294, "SpNn", "e294"),
        (1.00e291, "SxNn", "e291"),
        (1.00e288, "QtNn", "e288"),
        (1.00e285, "QaNn", "e285"),
        (1.00e282, "TNn", "e282"),
        (1.00e279, "DNn", "e279"),
        (1.00e276, "UNn", "e276"),
        (1.00e273, "Nn", "e273"),
        (1.00e270, "NOg", "e270"),
        (1.00e267, "OOg", "e267"),
        (1.00e264, "SpOg", "e264"),
        (1.00e261, "SxOg", "e261"),
        (1.00e258, "QtOg", "e258"),
        (1.00e255, "QaOg", "e255"),
        (1.00e252, "TOg", "e252"),
        (1.00e249, "DOg", "e249"),
        (1.00e246, "UOg", "e246"),
        (1.00e243, "Og", "e243"),
        (1.00e240, "NSt", "e240"),
        (1.00e237, "OSt", "e237"),
        (1.00e234, "SpSt", "e234"),
        (1.00e231, "SxSt", "e231"),
        (1.00e228, "QtSt", "e228"),
        (1.00e225, "QaSt", "e225"),
        (1.00e222, "TSt", "e222"),
        (1.00e219, "DSt", "e219"),
        (1.00e216, "USt", "e216"),
        (1.00e213, "St", "e213"),
        (1.00e210, "NSe", "e210"),
        (1.00e207, "OSe", "e207"),

        (1.00e204, "SpSe", "e204"),
        (1.00e201, "SxSe", "e201"),
        (1.00e198, "QtSe", "e198"),
        (1.00e195, "QaSe", "e195"),
        (1.00e191, "TSe", "e191"),
        (1.00e189, "DSe", "e189"),
        (1.00e186, "USe", "e186"),
        (1.00e183, "Se", "e183"),
        (1.00e180, "NQi", "e180"),
        (1.00e177, "OQi", "e177"),
        (1.00e174, "SpQi", "e174"),
        (1.00e171, "SxQi", "e171"),
        (1.00e168, "QtQi", "e168"),
        (1.00e165, "QaQi", "e165"),
        (1.00e162, "TQi", "e162"),
        (1.00e159, "DQi", "e159"),
        (1.00e156, "UQi", "e156"),
        (1.00e153, "Qi", "e153"),
        (1.00e150, "NQd", "e150"),
        (1.00e147, "OQd", "e147"),
        (1.00e144, "SpQd", "e144"),
        (1.00e141, "SxQd", "e141"),
        (1.00e138, "QtQd", "e138"),
        (1.00e135, "QaQd", "e135"),
        (1.00e132, "TQd", "e132"),
        (1.00e129, "DQd", "e129"),
        (1.00e126, "UQd", "e126"),
        (1.00e123, "Qd", "e123"),
        (1.00e120, "NTg", "e120"),
        (1.00e117, "OTg", "e117"),
        (1.00e114, "SpTg", "e114"),
        (1.00e111, "SxTg", "e111"),
        (1.00e108, "QtTg", "e108"),
        (1.00e105, "QaTg", "e105"),
        (1.00e102, "TTg", "e102"),
        (1.00e99, "DTg", "e99"),
        (1.00e96, "UTg", "e96"),
        (1.00e93, "Tg", "e93"),
        (1.00e90, "NVg", "e90"),
        (1.00e87, "OVg", "e87"),
        (1.00e84, "SpVg", "e84"),
        (1.00e81, "SxVg", "e81"),
        (1.00e78, "QtVg", "e78"),
        (1.00e75, "QaVg", "e75"),
        (1.00e72, "TVg", "e72"),
        (1.00e69, "DVg", "e69"),
        (1.00e66, "UVg", "e66"),
        (1.00e63, "Vg", "e63"),
        (1.00e60, "NDc", "e60"),
        (1.00e57, "ODc", "e57"),
        (1.00e54, "SpDc", "e54"),
        (1.00e51, "SxDc", "e51"),
        (1.00e48, "QtDc", "e48"),
        (1.00e45, "QaDc", "e45"),
        (1.00e42, "TDc", "e42"),
        (1.00e39, "DDc", "e39"),
        (1.00e36, "UDc", "e36"),
        (1.00e33, "Dc", "e33"),
        (1.00e30, "No", "e30"),
        (1.00e27, "Oc", "e27"),
        (1.00e24, "Sp", "e24"),
        (1.00e21, "Sx", "e21"),
        (1.00e18, "Qt", "e18"),
        (1.00e15, "Qa", "e15"),
        (1.00e12, "T", "e12"),
        (1.00e9, "B", "e9"),
        (1.00e6, "M", "e6"),
        (1.00e3, "K", "e3")
    }

    var GoldString : string = ""
    var TrueSuffix : string = ""
    GoldAmount := Amount + 0.0001

    for(i:int=0..suffixes.Length):
        if(suffix:=suffixes[i]):
            if (GoldAmount >= suffix(0)):
                if(WhichSuffix = 0):
                    set TrueSuffix = suffix(1)
                else:
                    set TrueSuffix = suffix(2)
                if(RoundedWholeNum := Floor[GoldAmount / suffix(0)]):
                    if(RoundedDec := Floor[((GoldAmount / suffix(0)) - (RoundedWholeNum * 1.0)) * 100]):
                        if(RoundedDec <= 9):
                            return "{RoundedWholeNum}.0{RoundedDec}{TrueSuffix}"
                        else:
                            return "{RoundedWholeNum}.{RoundedDec}{TrueSuffix}"
                            
    if(FlooredAmount := Floor[GoldAmount]):
        set GoldString = "{FlooredAmount}"
    return GoldString
    
<#NumberCondensor(Amount:float):string=
    var suffixes : []tuple(float, string) = array{
        (1.00e204, "SpSe"),
        (1.00e201, "SxSe"),
        (1.00e198, "QtSe"),
        (1.00e195, "QaSe"),
        (1.00e191, "TSe"),
        (1.00e189, "DSe"),
        (1.00e186, "USe"),
        (1.00e183, "Se"),
        (1.00e180, "NQi"),
        (1.00e177, "OQi"),
        (1.00e174, "SpQi"),
        (1.00e171, "SxQi"),
        (1.00e168, "QtQi"),
        (1.00e165, "QaQi"),
        (1.00e162, "TQi"),
        (1.00e159, "DQi"),
        (1.00e156, "UQi"),
        (1.00e153, "Qi"),
        (1.00e150, "NQd"),
        (1.00e147, "OQd"),
        (1.00e144, "SpQd"),
        (1.00e141, "SxQd"),
        (1.00e138, "QtQd"),
        (1.00e135, "QaQd"),
        (1.00e132, "TQd"),
        (1.00e129, "DQd"),
        (1.00e126, "UQd"),
        (1.00e123, "Qd"),
        (1.00e120, "NTg"),
        (1.00e117, "OTg"),
        (1.00e114, "SpTg"),
        (1.00e111, "SxTg"),
        (1.00e108, "QtTg"),
        (1.00e105, "QaTg"),
        (1.00e102, "TTg"),
        (1.00e99, "DTg"),
        (1.00e96, "UTg"),
        (1.00e93, "Tg"),
        (1.00e90, "NVg"),
        (1.00e87, "OVg"),
        (1.00e84, "SpVg"),
        (1.00e81, "SxVg"),
        (1.00e78, "QtVg"),
        (1.00e75, "QaVg"),
        (1.00e72, "TVg"),
        (1.00e69, "DVg"),
        (1.00e66, "UVg"),
        (1.00e63, "Vg"),
        (1.00e60, "NDc"),
        (1.00e57, "ODc"),
        (1.00e54, "SpDc"),
        (1.00e51, "SxDc"),
        (1.00e48, "QtDc"),
        (1.00e45, "QaDc"),
        (1.00e42, "TDc"),
        (1.00e39, "DDc"),
        (1.00e36, "UDc"),
        (1.00e33, "Dc"),
        (1.00e30, "No"),
        (1.00e27, "Oc"),
        (1.00e24, "Sp"),
        (1.00e21, "Sx"),
        (1.00e18, "Qt"),
        (1.00e15, "Qa"),
        (1.00e12, "T"),
        (1.00e9, "B"),
        (1.00e6, "M"),
        (1.00e3, "K")
    }

    var GoldString : string = ""
    GoldAmount := Amount + 0.0001

    for(i:int=0..suffixes.Length):
        if(suffix:=suffixes[i]):
            if (GoldAmount >= suffix(0)):
                if(RoundedWholeNum := Floor[GoldAmount / suffix(0)]):
                    if(RoundedDec := Floor[((GoldAmount / suffix(0)) - (RoundedWholeNum * 1.0)) * 100]):
                        if(RoundedDec <= 9):
                            return "{RoundedWholeNum}.0{RoundedDec}{suffix(1)}"
                        else:
                            return "{RoundedWholeNum}.{RoundedDec}{suffix(1)}"
                            
    if(FlooredAmount := Floor[GoldAmount]):
        set GoldString = "{FlooredAmount}"
    return GoldString#>

ConvertToTime(Time : float, whichConversion:string):string=
    if(whichConversion = "Hours"):
        if(Hours:=Floor[Time/3600.0],Minutes:=Floor[((Time) - (Hours * 3600.0)) / 60.0]):
            if(Hours <= 9 and Minutes <= 9):
                return("0{Hours}:0{Minutes}")
            else if(Hours <= 9):
                return("0{Hours}:{Minutes}")
            else if(Minutes <= 9):
                return("{Hours}:0{Minutes}")
            else:
                return("{Hours}:{Minutes}")

    if(whichConversion = "Minutes"):
        if(Minutes:=Floor[Time/60.0],Seconds:=Floor[(Time) - (Minutes * 60.0)]):
            if(Minutes <= 9 and Seconds <= 9):
                return("0{Minutes}:0{Seconds}")
            else if(Minutes <= 9):
                return("0{Minutes}:{Seconds}")
            else if(Seconds <= 9):
                return("{Minutes}:0{Seconds}")
            else:
                return("{Minutes}:{Seconds}")

    if(whichConversion = "FullHours"):
        if(Hours:=Floor[Time/3600.0],Minutes:=Floor[((Time) - (Hours * 3600.0)) / 60.0],Seconds:=Floor[(Time) - (Minutes * 60.0)]):
            if(Hours <= 9 and Minutes <= 9 and Seconds <= 9):
                return("0{Hours}:0{Minutes}:0{Seconds}")
            else if(Hours <= 9 and Seconds <= 9):
                return("0{Hours}:{Minutes}:0{Seconds}")
            else if(Minutes <= 9 and Seconds <= 9):
                return("{Hours}:0{Minutes}:0{Seconds}")
            else if(Hours <= 9 and Minutes <= 9):
                return("0{Hours}:0{Minutes}:{Seconds}")
            else if(Hours <= 9):
                return("0{Hours}:{Minutes}:{Seconds}")
            else if(Minutes <= 9):
                return("{Hours}:0{Minutes}:{Seconds}")
            else:
                return("{Hours}:{Minutes}:{Seconds}")
    return ""

ConvertToRomanNumeral(Number : int):string =
    var Symbols : []string = {"M", "CM", "D", "CD", "C", "XC", "L", "XL", "X", "IX", "V", "IV", "I"}
    var Values : []int = {1000, 900, 500, 400, 100, 90, 50, 40, 10, 9, 5, 4, 1}

    var N : int = Number
    var RomanNumeral : string = ""

    loop:
        for (X:=0..N):
            if (N <= 0):
                break
            loop:
                for (SymbolIndex := 0..Symbols.Length):
                    if (SV := Values[SymbolIndex], Symbol := Symbols[SymbolIndex], N >= SV):
                        set RomanNumeral += Symbol
                        set N -= SV
                        break
                break
        break
    return RomanNumeral

IncreaseCostIncrement(base: float, multiplier: float, increment:int):float=
    var result : float = base
    if(increment > 0):
        for (i:int=0..increment):
            set result = result * multiplier
    return result

CalcUpgradeCost(level: int, TrueBase:int, Increase:float):int=
    poweredNumber:= Pow(Increase, (level - 1)*1.0)
    if(ReturnValue := Floor[TrueBase * poweredNumber]):
        return ReturnValue
    return 0
    
# Function to determine the egg based on drop odds
SelectItem(DropOdds:[]float):int=
    var TotalOdds : float = 0.0
    var cumulative_odds : []float = array{0.0,0.0,0.0,0.0,0.0,0.0}

    for(i:int=0..DropOdds.Length):
        if(odd:=DropOdds[i]):
            set TotalOdds += odd
        if(set cumulative_odds[i] = TotalOdds) {}
    
    # Step 2: Generate a random number between 0 and total_odds
    var random_value :float= GetRandomFloat(0.0, TotalOdds)
    
    # Step 3: Determine which egg corresponds to the random number
    for(i:int=0..cumulative_odds.Length):
        if(random_value < cumulative_odds[i]):
            return i

    # Return the first egg if not found (fallback)
    return 0

ConvertSeconds(Seconds:float):tuple(float,float,float)=
    var InitialSeconds : float = Seconds
    var hours : float = Seconds/3600.0
    set InitialSeconds -= hours * 3600.0
    var min : float = InitialSeconds/60.0
    set InitialSeconds = min*60

    return (hours,min,InitialSeconds)

RemoveKeyFromMap(ExampleMap:[fort_character]float, ElementToRemove:fort_character):[fort_character]float=
    var NewMap:[fort_character]float = map{}
    # Concatenate Keys from ExampleMap into NewMap, excluding ElementToRemove
    for (Key -> Value : ExampleMap, Key <> ElementToRemove):
        set NewMap = ConcatenateMaps(NewMap, map{Key => Value})
    #Print("REMOVED FROM MAP")
    return NewMap

BubbleSort(Array:[]float): []float=
    var n :int= Array.Length
    var temp : float = 0.0
    var NewArray : []float = Array
    # Perform the bubble sort
    for(i:int=0..n):
        for(j:int=0..n-i-1):
            if:
                NewJ := NewArray[j]
                NewJ1 := NewArray[j+1]
                NewArray[j] > NewArray[j+1]
            then:
                set temp = NewJ
                if(set NewArray[j] = NewJ1) {}
                if(set NewArray[j+1] = temp) {}

    return NewArray

#PUSH SCRIPT
<#SimplificationSuffixes : []string = array{"K", "M", "B", "T", "Qa", "Qi", "Si", "Sp"}
    
GetDigits<public>(Number : float)<transacts>:int=
    if (Result := ToString(Floor[Number]).Length):
        return Result

    return 1

SimplifyNumber<public>(Number : float, ?Decimals : int = 2)<decides><transacts>:string=
    var Result : ?string = false 
    AdjustedDecimals := Min(Decimals, 3)

    DigitsCount := GetDigits(Number)
    SuffixId := Ceil[DigitsCount*1.0 / 3.0 - 2.0]

    if (SuffixId <= -1):
        set Result = option{ToString(Floor[Number])}
        Result?
    else:
        WholeDisplayNumber := Floor[Number * Pow(10.0,-((SuffixId+1)*3.0)+(AdjustedDecimals*1.0))]
        var SimplifiedNumber : string = ToString(WholeDisplayNumber)
        NumberWithDot := SimplifiedNumber.Insert[SimplifiedNumber.Length - AdjustedDecimals, "."]
        set SimplifiedNumber = NumberWithDot
        set SimplifiedNumber += SimplificationSuffixes[SuffixId]
        set Result = option{SimplifiedNumber}
        Result?#>